using { /UnrealEngine.com/Temporary/UI }
using { /Fortnite.com/Devices }

Art<public> := module:
    Zombie_Models<public>:=module:
        Zombies<public>:=module:
            Animations<public>:=module:
                Default<public>:=module:
                    Vaulting<public>:=module:
    UIStuff<public>:=module:
        MainMenu<public>:=module:
        Fonts<public>:=module:
            InteractionBackground<public>:=module:
            Round_Numbers<public>:=module:
            Point_Numbers<public>:=module:
            InteractionText<public>:=module:
                Letters<public>:=module:
                Numbers<public>:=module:
                Symbols<public>:=module:
            Quest<public>:=module:
                Letters<public>:=module:
                Numbers<public>:=module:
                Symbols<public>:=module:
        Icons<public>:=module:
            PointBackground<public>:=module:
            Powerup_Icons<public>:=module:
            Perk_Icons<public>:=module:
            QuestObjects<public>:=module:
            Weapon<public>:=module:
            Trial_UI<public>:=module:
            Part_Icons<public>:=module:    
        Leaderboards<public>:=module:
        KennyAssets<public>:=module:

Zombie_Pack<public>:=module:
    Animations<public>:=module:
        InPlace<public>:=module:
            Jumps<public>:=module:
                
Goldys<public> := module:
    Plagues<public>:= module:
    Seeker<public> := module:

# A class that is unique without using the class specifier
comparable_class<public> := class:

    WidgetComparingTool<internal> : color_block = color_block{}

    IsEqual<public>(OtherClass:comparable_class)<transacts><decides>:void=
        WidgetComparingTool = OtherClass.WidgetComparingTool

    IsNotEqual<public>(OtherClass:comparable_class)<transacts><decides>:void=
        WidgetComparingTool <> OtherClass.WidgetComparingTool

# Returns the index of the instance inside the array of class instances
# Fails if 'Instance' does not exist in the array
(ComparableInstances:[]comparable_class).FindInstance<public>(Instance:comparable_class)<transacts><decides>:int=
    var MaybeComparableIndex : ?int = false
    for:
        Index->ThisInstance:ComparableInstances
        ThisInstance.IsEqual[Instance]
    do:
        set MaybeComparableIndex = option{Index}
    MaybeComparableIndex?

comparable_device := class(creative_device):

    WidgetComparingTool<internal> : color_block = color_block{}

    IsEqual<public>(OtherClass:comparable_device)<transacts><decides>:void=
        WidgetComparingTool = OtherClass.WidgetComparingTool
        
    IsNotEqual<public>(OtherClass:comparable_device)<transacts><decides>:void=
        WidgetComparingTool <> OtherClass.WidgetComparingTool