using { /Verse.org/Simulation }

<#
# Await a list of events
AwaitAll<public>(Awaitables:[]event(t) where t:subtype(comparable))<suspends>:[]t=
    if(FirstEvent := Awaitables[0], NewAwaitablesList := Awaitables.RemoveElement[0]):
        SyncResult:=sync:
            FirstEvent.Await()
            AwaitAll(NewAwaitablesList)
        array{SyncResult(0)}+SyncResult(1)
    else:
        array{}

# Race a list of events
RaceAll<public>(Awaitables:[]event(t) where t:type):void=
    NewRaceInst:race_inst=race_inst{NewAwaitable:=event(event(type)){}}
    spawn. NewRaceInst.StartRace()
    NewRaceInst

# An instance of a race
race_inst<public> := class:

    NewAwaitable<public> : event(event(type))

    StartRace<public>(Awaitables:[]event(t) where t:type)<suspends>:?t=
        FirstEvent := Awaitables[0] or NewAwaitable.Await()
        if(NewAwaitablesList := Awaitables.RemoveElement[0], NewAwaitablesList[0]):
            RaceResult:=race:
                block:
                    AwaitResult:=FirstEvent.Await()
                    option. AwaitResult
                StartRace(NewAwaitablesList)
            RaceResult
        else:
            EventResult:=FirstEvent.Await()
            option. EventResult
        false

    AddToRace<public>(InAwaitable:event(t)):void=
        NewAwaitable.Signal(InAwaitable) #>