using { /Fortnite.com/AI }
using { /Fortnite.com/Characters }
using { /Fortnite.com/Devices }
using { /Verse.org/Colors }
using { /Verse.org/Random }
using { /Verse.org/Simulation }
using { /UnrealEngine.com/Temporary/Diagnostics }
using { /UnrealEngine.com/Temporary/SpatialMath }
using { Player }

(BetterVolumeDevices:[]better_volume_device).GetAgentsInZones(AgentList:[]agent)<transacts>:[]agent=
    var InZonePlayers : []agent = array{}
    PlayerList:=GetPlayers()
    for(ThisAgent:AgentList):
        for(ThisVolume:BetterVolumeDevices):
            if(not InZonePlayers.Find[ThisAgent], ThisVolume.IsInVolume[ThisAgent]):
                set InZonePlayers += array{ThisAgent}
    return InZonePlayers

(VolumeDevices:[]volume_device).ConvertToBVD():[]better_volume_device=
    var ConvertedVolumes : []better_volume_device = array{}
    for(ThisVolume:VolumeDevices):
        NewVolume:better_volume_device=better_volume_device{VolumeDevice:=ThisVolume}
        set ConvertedVolumes += array{NewVolume}
    return ConvertedVolumes

better_volume_device<public> := class<concrete>:

    @editable
    var VolumeDevice<public>:volume_device = volume_device{}

    # Signaled when an `agent` enters the device volume.
    AgentEntersEvent<public>:event(agent) = event(agent){}

    # Signaled when an `agent` exits the device volume.
    AgentExitsEvent<public>:event(agent) = event(agent){}

    var VolumeOrigin : transform = transform{}

    var<private> AgentsInVolumeMap<public>:[agent]logic = map{}

    var<private> IsActive : logic = false
    Init():void=
        if(not IsActive?):
            set IsActive = true
            VolumeDevice.AgentEntersEvent.Subscribe(AgentEnteredVolume)
            VolumeDevice.AgentExitsEvent.Subscribe(AgentExitedVolume)
            set VolumeOrigin = VolumeDevice.GetTransform()
            OnInit()

    OnInit<public>():void=
        return

    AgentEnteredVolume<private>(InAgent:agent):void=
        if(set AgentsInVolumeMap[InAgent] = true){}
        else:
            set AgentsInVolumeMap = ConcatenateMaps(AgentsInVolumeMap, map{InAgent => true})
        AgentEntersEvent.Signal(InAgent)
        OnAgentEnteringVolume(InAgent)

    OnAgentEnteringVolume<public>(InAgent:agent):void=
        return
        
    AgentExitedVolume<private>(InAgent:agent):void=
        if(set AgentsInVolumeMap[InAgent] = false){}
        else:
            set AgentsInVolumeMap = ConcatenateMaps(AgentsInVolumeMap, map{InAgent => false})
        AgentExitsEvent.Signal(InAgent)
        OnAgentExitingVolume(InAgent)

    OnAgentExitingVolume<public>(InAgent:agent):void=
        return

    IsInVolume<public>(Agent:agent)<transacts><decides>:void= 
    {
        AgentsInVolumeMap[Agent]?
    }

    IsOccupied<public>()<transacts><decides>:[]agent=
        var OccupiedPlayers : []agent = array{}
        for:
            ThisPlayer->RoomState:AgentsInVolumeMap
            RoomState?
            ThisPlayer.IsInSession[]
        do:
            set OccupiedPlayers += array{ThisPlayer}
        OccupiedPlayers.Length > 0
        return OccupiedPlayers 

    GetOutsidePlayers()<transacts>:[]agent=
        PlayerList:=GetPlayers()
        var UnoccupiedPlayers : []agent = array{}
        for:
            ThisPlayer:PlayerList
            AgentsInVolumeMap[ThisPlayer.User]?
        do:
            set UnoccupiedPlayers += array{ThisPlayer.User}
        return UnoccupiedPlayers

    Disable<public>():void=
        if. VolumeDevice.TeleportTo[vector3{},rotation{}]

    Enable<public>():void=
        if. VolumeDevice.TeleportTo[VolumeOrigin.Translation, VolumeOrigin.Rotation]
