using { /Fortnite.com/Devices }
using { /Fortnite.com/Characters }
using { /Fortnite.com/Game }
using { /Fortnite.com/UI }
using { /Fortnite.com/FortPlayerUtilities }
using { /Verse.org/Simulation }
using { /Verse.org/Random }
using { /Verse.org/Colors }
using { /UnrealEngine.com/Temporary/Diagnostics }
using { /UnrealEngine.com/Temporary/SpatialMath }
using { /UnrealEngine.com/Temporary/UI }
using { /Verse.org/Colors/NamedColors }
using { InteractiveModule}
using { Player }

collectable_interact_behavior<public>:=enum{ShowProp, HideProp}
collectable_setup_type<public>:=enum{SingleCollectable, MultipleCollectables}
collectable_signal_type<public>:=enum{NoPlayer, ActivatingPlayer, AllPlayers}

CollectableDevice := class(Interactive_Base):

    @editable Locations : []Interactable = array{}
    @editable CollectableType : collectable_setup_type = collectable_setup_type.SingleCollectable
    @editable OnInteractBehavior : collectable_interact_behavior = collectable_interact_behavior.HideProp
    @editable OnTrigger_HideProps : trigger_device = trigger_device{}
    @editable OnTrigger_ShowProps : trigger_device = trigger_device{}
    @editable PlayAudio_OnHidden : audio_player_device = audio_player_device{}
    @editable PlayAudio_OnShown : audio_player_device = audio_player_device{}
    @editable ActivateTrigger_OnCollectablePickedup : trigger_device = trigger_device{}
    @editable ActivateTrigger_OnALLCollectablesPickedup : trigger_device = trigger_device{}
    @editable OnAllCollectablesPickedup_SignalSendType : collectable_signal_type = collectable_signal_type.ActivatingPlayer

    var PropTransforms : [int]transform = map{}

    OnBegin<override>():void=
        LinkAndCompile(Locations)
        var CurrProp : int = 0
        for(ThisProp:Locations):
            if(ThisProp.PropLocation.IsValid[]):
                set PropTransforms = ConcatenateMaps(PropTransforms, map{CurrProp => ThisProp.GetTransform()})
            set CurrProp += 1
        OnTrigger_HideProps.TriggeredEvent.Subscribe(HideProps)
        OnTrigger_ShowProps.TriggeredEvent.Subscribe(ShowProps)

    ShowProps(InAgent:?agent):void=
        var CurrIndex : int = 0
        for(ThisProp:Locations):
            if. ThisProp.PropLocation.TeleportTo[PropTransforms[CurrIndex]]
            ThisProp.PropLocation.Show()
            set CurrIndex += 1
        PlayAudio_OnShown.Play()

    HideProps(InAgent:?agent):void=
        for(ThisProp:Locations):
            ThisProp.PropLocation.Hide()
            if. ThisProp.PropLocation.TeleportTo[vector3{}, rotation{}]
        PlayAudio_OnHidden.Play()

    Reset<override>():void=
        ShowProps(false)

    GetInteractionText<override>(InAgent:agent, InInteractingLocation:Interactable):string=
        return ""

    OnInteracted<override>(InAgent:agent, InteractingLocation:Interactable):void=
        if(CollectableType = collectable_setup_type.SingleCollectable):
            if(OnInteractBehavior = collectable_interact_behavior.ShowProp):
                ShowProps(false)
            else:
                HideProps(false)
            ActivateTrigger_OnCollectablePickedup.Trigger(InAgent)
        else:
            FullyHide(InteractingLocation.PropLocation)
            InteractingLocation.SetEnabledState(false)
            ActivateTrigger_OnCollectablePickedup.Trigger(InAgent)
            LocationLength:=GetEnabledLocations().Length
            if(LocationLength <= 0):
                case(OnAllCollectablesPickedup_SignalSendType):
                    collectable_signal_type.NoPlayer =>
                        ActivateTrigger_OnALLCollectablesPickedup.Trigger()
                    collectable_signal_type.ActivatingPlayer =>
                        ActivateTrigger_OnALLCollectablesPickedup.Trigger(InAgent)
                    collectable_signal_type.AllPlayers =>
                        PlayerList:=GetPlayers()
                        for(ThisPlayer:PlayerList):
                            ActivateTrigger_OnALLCollectablesPickedup.Trigger(ThisPlayer.User)

    FullyHide(InProp:creative_prop):void=
        if. InProp.TeleportTo[vector3{}, rotation{}]
        InProp.Hide()