using { /UnrealEngine.com/Temporary/SpatialMath }
using { /Fortnite.com/Devices }
using { /Fortnite.com/Game }

CustomCollision := module:

    #assuming you're using the provided meshes in the content drawer

    volume_shapes<public> := enum{cube; sphere; ellipsoid; cylinder}

    #idk if this exists or not and was too lazy to check lol
    Average(F1:float, F2:float)<transacts>:float={(F1 + F2) / 2.0}

    IsInBounds_Cylinder<public>(Cylinder:positional, Point:vector3)<transacts>:logic=
        CylinderTrans := Cylinder.GetTransform()
        Radius := (Average(CylinderTrans.Scale.X, CylinderTrans.Scale.Y) * 100.0)
        Offset := CylinderTrans.Scale.Z * 100.0
        TopPoint := vector3{X := CylinderTrans.Translation.X, Y := CylinderTrans.Translation.Y, Z := CylinderTrans.Translation.Z + Offset}
        BottomPoint := vector3{X := CylinderTrans.Translation.X, Y := CylinderTrans.Translation.Y, Z := CylinderTrans.Translation.Z - Offset}
        CylinderAxis := vector3{X := TopPoint.X - BottomPoint.X, Y := TopPoint.Y - BottomPoint.Y, Z := TopPoint.Z - BottomPoint.Z}
        HeightVector := vector3:
            X := Point.X - BottomPoint.X
            Y := Point.Y - BottomPoint.Y
            Z := Point.Z - BottomPoint.Z
        ProjLength := DotProduct(HeightVector, CylinderAxis) / DotProduct(CylinderAxis, CylinderAxis)
        if(ProjLength < 0.0 or ProjLength > 1.0){return false}
        ClosestPointOnAxis := vector3:
            X := (BottomPoint.X) + (ProjLength * CylinderAxis.X)
            Y := (BottomPoint.Y) + (ProjLength * CylinderAxis.Y)
            Z := (BottomPoint.Z) + (ProjLength * CylinderAxis.Z)
        DistanceToAxis := Distance(Point, ClosestPointOnAxis)
        if((DistanceToAxis) <= Radius){return true}
        else return false

    #assuming scale is uniform
    IsInBounds_Sphere<public>(Sphere:positional, Point:vector3)<transacts>:logic=
        SelfTrans := Sphere.GetTransform()
        Loc := SelfTrans.Translation
        PlayerRadius := 25.0 #almost treated as a threshold
        SelfRadius := 100.0 * Average(SelfTrans.Scale.X, SelfTrans.Scale.Y) #assuming the sphere's scale is even on (x,y,z) | if not use ellipsoid
        Dist := Distance(Loc, Point)
        if(Dist <= PlayerRadius + SelfRadius){return true} else{return false}

    #Assuming we're starting with blenders default uv sphere which has a radius of 100cm | Can Scale And Rotate These
    IsInBounds_Ellipsoid<public>(Ellipsoid:positional, Point:vector3)<transacts>:logic=
        EllipsoidTrans := Ellipsoid.GetTransform()
        EllipsoidRotation := EllipsoidTrans.Rotation
        EllipsoidLocation := EllipsoidTrans.Translation
        EllipsoidScale := vector3:
            X := EllipsoidTrans.Scale.X * 100.0
            Y := EllipsoidTrans.Scale.Y * 100.0
            Z := EllipsoidTrans.Scale.Z * 100.0
        InverseRotation := EllipsoidRotation.Invert()
        #Get Local Position of point relative to checkpoint
        LocalPoint := vector3:
            X := Point.X - EllipsoidLocation.X
            Y := Point.Y - EllipsoidLocation.Y
            Z := Point.Z - EllipsoidLocation.Z
        #Rotate the local point by the inverse rotation to align with the ellipsoid's axis
        InversedLocalPoint := InverseRotation.RotateVector(LocalPoint)
        #Scale the rotated point by the dimensions of the ellipsoid
        ScaledPoint := vector3:
            X := InversedLocalPoint.X / EllipsoidScale.X
            Y := InversedLocalPoint.Y / EllipsoidScale.Y
            Z := InversedLocalPoint.Z / EllipsoidScale.Z
        #Calculate the squared distance of the scaled point from the origin
        SquaredDist := (ScaledPoint.X * ScaledPoint.X) + (ScaledPoint.Y * ScaledPoint.Y) + (ScaledPoint.Z * ScaledPoint.Z)
        if(SquaredDist <= 1.0){return true} else{return false}

    #can also rewrite this to take 2 positionals if you need
    #This is an AABB check | do not rotate the cubes, will not work
    IsInBounds_Cube<public>(P1:positional, Point:vector3):logic=
        P1Min := GetMin(P1)
        P1Max := GetMax(P1)
        PointMin := GetImaginaryMin(Point, 1.0)
        PointMax := GetImagniaryMax(Point, 1.0)
        if(PointMax.X < P1Min.X or PointMin.X > P1Max.X){return false}
        if(PointMax.Y < P1Min.Y or PointMin.Y > P1Max.Y){return false}
        if(PointMax.Z < P1Min.Z or PointMin.Z > P1Max.Z){return false}
        return true

    #for assigning an imaginary cubes bounds to a point (most likely a player) / define it's L, W, & H
    GetImaginaryMin<public>(PlayerLoc:vector3, RayLWH:float):vector3=
        X := PlayerLoc.X - ( (1.0 * RayLWH) / 2.0 )
        Y := PlayerLoc.Y - ( (1.0 * RayLWH) / 2.0 )
        Z := PlayerLoc.Z - ( (1.0 * RayLWH) / 2.0 )
        MinV3 := vector3{X := X, Y := Y, Z := Z}
        return MinV3

    #for assigning an imaginary cubes bounds to a point (most likely a player) / define it's L, W, & H
    GetImagniaryMax<public>(PlayerLoc:vector3, RayLWH:float):vector3=
        X := PlayerLoc.X + ( (1.0 * RayLWH) / 2.0 )
        Y := PlayerLoc.Y + ( (1.0 * RayLWH) / 2.0 )
        Z := PlayerLoc.Z + ( (1.0 * RayLWH) / 2.0 )
        MaxV3 := vector3{X := X, Y := Y, Z := Z}
        return MaxV3

    GetMin<public>(Positional:positional):vector3= #Calculates the minimum corner coordinates of an Axis-Aligned Bounding Box
        DefaultLWH := 200.0 #default LWH of the cube i've imported
        Scale := Positional.GetTransform().Scale
        Location := Positional.GetTransform().Translation
        MinX := Location.X - ((Scale.X * DefaultLWH) / 2.0)  #This is assuming the cubes origin is in the center of the mesh and you used blenders default cube which is 200cm in L, W, and H for your props
        MinY := Location.Y - ((Scale.Y * DefaultLWH) / 2.0)
        MinZ := Location.Z - ((Scale.Z * DefaultLWH) / 2.0)
        Min_V3 := vector3{X := MinX, Y := MinY, Z := MinZ}
        return Min_V3

    GetMax<public>(Positional:positional):vector3= #Calculates the max corner coordinates of an Axis-Aligned Bounding Box
        DefaultLWH := 200.0 #default LWH of the cube i've imported
        Scale := Positional.GetTransform().Scale
        Location := Positional.GetTransform().Translation
        MaxX := Location.X + ((Scale.X * DefaultLWH) / 2.0) #This is assuming the cubes origin is in the center of the mesh and you used blenders default cube which is 200cm in L, W, and H for your props
        MaxY := Location.Y + ((Scale.Y * DefaultLWH) / 2.0)
        MaxZ := Location.Z + ((Scale.Z * DefaultLWH) / 2.0)
        Max_V3 := vector3{X := MaxX, Y := MaxY, Z := MaxZ}
        return Max_V3

using { /Verse.org/Simulation }
using { /Fortnite.com/Devices }
using { /Fortnite.com/Characters }
using { /UnrealEngine.com/Temporary/SpatialMath }
using { /Fortnite.com/Game }
using { SubscribableEvents }
using { CustomCollision }

#a volume device that lets you choose the shape | (set the appropriate mesh in editor)
custom_volume_device := class(creative_device):

    @editable Shape : volume_shapes = volume_shapes.sphere

    AgentEntersEvent<public>: agent_event = agent_event{}
    AgentExitsEvent<public>: agent_event = agent_event{}

    var Running : logic = false

    DisabledEvent<private> : event() = event(){}

    var PlayerStatus<private> : [player]logic = map{}

    OnBegin<override>()<suspends>:void=
        for(ThisPlayer:GetPlayspace().GetPlayers()){Register(ThisPlayer)}
        spawn{Main()}

    Enable():void={if(not Running?){spawn{Main()}}}
    Disable():void={DisabledEvent.Signal()}

    #initialize player map | check if they're in / out of vol for intial value
    Register(ThisPlayer:player):void=
        if(FC := ThisPlayer.GetFortCharacter[], Loc := FC.GetTransform().Translation):
            Callback := case(Shape):
                volume_shapes.cube => IsInBounds_Cube
                volume_shapes.sphere => IsInBounds_Sphere
                volume_shapes.ellipsoid => IsInBounds_Ellipsoid
                volume_shapes.cylinder => IsInBounds_Cylinder
            Status := Callback(Self, Loc)
            if(set PlayerStatus[ThisPlayer] = Status){}

    #i don't like iterating through maps
    UnRegister(LeavingPlayer:player):void=
        var NewPlayerStatus : [player]logic = map{}
        for(ThisPlayer:GetPlayspace().GetPlayers(), ThisPlayer <> LeavingPlayer):
            if(Status := PlayerStatus[ThisPlayer], set NewPlayerStatus[ThisPlayer] = Status){}
        set PlayerStatus = NewPlayerStatus

    Main()<suspends>:void=
        if(Running?){return}    #some safety since we're spawning this
        set Running = true

        #get an IsInBounds Method From Shape
        Callback := case(Shape):
            volume_shapes.cube => IsInBounds_Cube
            volume_shapes.sphere => IsInBounds_Sphere
            volume_shapes.ellipsoid => IsInBounds_Ellipsoid
            volume_shapes.cylinder => IsInBounds_Cylinder

        race:
            block{DisabledEvent.Await()}
            block:
                loop:   #no need to have break conditions in a race | we like jenk
                    Sleep(0.0)  #1 tick you can play with this number for sure
                    for(ThisPlayer:GetPlayspace().GetPlayers(), FC := ThisPlayer.GetFortCharacter[]):
                        Loc := FC.GetTransform().Translation
                        CurrentStatus := Callback(Self, Loc) #check if in volume
                        if(LastStatus := PlayerStatus[ThisPlayer]):
                            #check if new status is different than last | signal exit or enter if it is
                            if:
                                CurrentStatus <> LastStatus
                                set PlayerStatus[ThisPlayer] = CurrentStatus
                            then:
                                if(CurrentStatus?){AgentEntersEvent.Signal(ThisPlayer)}
                                else{AgentExitsEvent.Signal(ThisPlayer)}
        set Running = false