using { /Verse.org/Simulation }
using { /Fortnite.com/Devices }
using { /Fortnite.com/Characters }
using { /Verse.org/Concurrency }
    
SubscribableEvents := module:
    
    AgentFunc := type{func(:agent):void}

    cancelable_agent_callback := class<unique>(cancelable):
        var callback : ?AgentFunc = false
        Cancel<override>()<transacts>:void={set callback = false}
        Invoke(Payload:agent):void={if(Callback := callback?){Callback(Payload)}}
            
    agent_wrapper<public>(t:type) := class():
        ExtraData<public> : t
        OutputFunc<public>  : tuple(agent, t) -> void
        InputFunc<public> (Agent:agent):void=OutputFunc(Agent, ExtraData)

    agent_event <public> := class(signalable(agent), awaitable(agent)):

        var Callbacks : []cancelable_agent_callback = array{}

        Subscribe<public>(Callback:AgentFunc)<transacts>:cancelable=
            NewCallback := cancelable_agent_callback{callback := option{Callback}}
            set Callbacks = Callbacks + array{NewCallback}
            for(_Callback:Callbacks, not _Callback.callback?){set Callbacks = Callbacks.RemoveAllElements(_Callback)}
            return NewCallback

        SubscribeAgent<public>(OutputFunc : tuple(agent, t)->void, ExtraData : t where t:type):cancelable=
            Wrapper := agent_wrapper(t){ExtraData := ExtraData, OutputFunc := OutputFunc}
            Subscribe(Wrapper.InputFunc)

        var IsAwaiting : logic = false

        #for Await() since my brain isn't quite big enough to deal with how Await could possibly work lol
        AwaitableEvent : event(agent) = event(agent){}
            
        Signal<override>(InAgent:agent):void=
            if(IsAwaiting?){AwaitableEvent.Signal(InAgent)}
            for(Callback:Callbacks){Callback.Invoke(InAgent)}
            
        Await<override>()<suspends>:agent=
            set IsAwaiting = true
            Agent := AwaitableEvent.Await()
            set IsAwaiting = false
            return Agent