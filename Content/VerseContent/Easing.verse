
using { /Fortnite.com/Devices }
using { /Verse.org/Simulation }
using { /UnrealEngine.com/Temporary/Diagnostics }
using { /UnrealEngine.com/Temporary/SpatialMath }

interpolation_type<public> := enum{
    Linear
    EaseIn
    EaseOut
    EaseInOut
    EaseInBack
    EaseOutBack
    EaseInOutBack
    EaseOutBounce
}

(InProgress:float).ApplyInterpolation(InterpolationType:interpolation_type):float=
    case(InterpolationType):
        interpolation_type.Linear => return InProgress
        interpolation_type.EaseIn => return Ease_In(InProgress)
        interpolation_type.EaseOut => return Ease_Out(InProgress)
        interpolation_type.EaseInOut => return Ease_InOut(InProgress)
        interpolation_type.EaseInBack => return Ease_InBack(InProgress)
        interpolation_type.EaseOutBack => return Ease_OutBack(InProgress)
        interpolation_type.EaseInOutBack => return Ease_InOutBack(InProgress)
        interpolation_type.EaseOutBounce => return Ease_OutBounce(InProgress)

Ease_In(Progress:float):float=
    return 1.0 - Cos((Progress * 3.1415) / 2.0)

Ease_Out(Progress:float):float=
    return Sin((Progress * 3.1415) / 2.0)

Ease_InOut(Progress:float):float=
    return -(Cos(3.1415 * Progress) - 1.0) / 2.0

Ease_InBack(Progress:float):float=
    return 2.70158 * Progress * Progress * Progress - 1.70158 * Progress * Progress

Ease_OutBack(Progress:float):float=
    return 1.0 + 2.70158 * Pow(Progress - 1.0, 3.0) + 1.70158 * Pow(Progress - 1.0, 2.0)

Ease_InOutBack(Progress:float):float=
    c2 := 2.5949095
    
    if(Progress < 0.5):
      return (Pow(2 * Progress, 2.0) * ((c2 + 1.0) * 2.0 * Progress - c2)) / 2.0
    else:
      return (Pow(2 * Progress - 2.0, 2.0) * ((c2 + 1.0) * (Progress * 2.0 - 2.0) + c2) + 2.0) / 2.0
Ease_OutBounce(Progress:float):float=
    var InProgress : float = Progress
    n1 := 7.5625;
    d1 := 2.75;

    if (InProgress < 1.0 / d1) {
        return n1 * InProgress * InProgress;
    } else if (InProgress < 2.0 / d1) {
        return n1 * (set InProgress -= 1.5 / d1) * InProgress + 0.75;
    } else if (InProgress < 2.5 / d1) {
        return n1 * (set InProgress -= 2.25 / d1) * InProgress + 0.9375;
    } else {
        return n1 * (set InProgress -= 2.625 / d1) * InProgress + 0.984375;
    }