# Basic queue using integers as keys representing slots
slot_queue<public> := class:

    # Slot -> Popped Event
    # If pop event = true, then the Slot exists on the queue
    # If pop event = false, then the Slot is not on the queue
    var QueueMap<private> : [int]?event() = map{}
    
    # The amount of Slots on the queue. Used for setting each Slot identity
    var SlotCount<private> : int = 0

    # Used to debug this queue
    var<public> DebugState : logic = false

    # EVENTS
    AddedEvent<public> : event(int) = event(int){}
    RemovedEvent<public> : event(int) = event(int){}

    # Adds a slot to the queue and awaits for its removal
    # Doesn't complete until the slot is popped from the queue OR the Slot Queue is already empty
    LatchOn<public><final>()<suspends>:void=
        EmptyCheck := logic{GetQueue().Length <= 0}
        ThisRemoveEvent:=Add()
        DebugPrint("Entering latch on queue [{ThisRemoveEvent(0)}]")
        if(not EmptyCheck?):
            ThisRemoveEvent(1).Await()
            if. set QueueMap[ThisRemoveEvent(0)] = false
        DebugPrint("Exiting latch on with Slot [{ThisRemoveEvent(0)}]")

    # Adds an index to the bottom of the queue
    # Returns the popped event if successful
    Add<public><final>(?Slot:int=-1):tuple(int, event())=
        InSlot := if. Slot >= 0 then. Slot else{
            ThisQueueCount := SlotCount
            set SlotCount += 1
            ThisQueueCount
        }
        if(ThisPopEvent:=QueueMap[InSlot]?):
            return (InSlot, ThisPopEvent)
        else:
            RemoveEvent : event() = event(){}
            if. set QueueMap[InSlot] = option{RemoveEvent}
            else. set QueueMap = ConcatenateMaps(QueueMap, map{InSlot => option{RemoveEvent}})
            AddedEvent.Signal(InSlot)
            return (InSlot, RemoveEvent)

    # Removes the slot at the top of the queue
    # Returns the popped slot, if it exists
    Pop<public><final>(?FireRemoveEvent:logic=true):?int=
        DebugPrint("Pop called")
        if:
            TopQueueSlot := GetQueue()[0]
            RemovalEvent := QueueMap[TopQueueSlot]?
            set QueueMap[TopQueueSlot] = false
        then:
            DebugPrint("Pop success on zombie [{TopQueueSlot}]")
            if. FireRemoveEvent? then. RemovalEvent.Signal()
            QueuedZombieList:=GetQueue().Length
            if(QueuedZombieList <= 0):
                set QueueMap = map{}
            else if:
                NextQueueSlot := GetQueue()[0]
                NextRemovalEvent := QueueMap[NextQueueSlot]?
            then:
                DebugPrint("Signaled exit on zombie [{NextQueueSlot}]")
                NextRemovalEvent.Signal()
            else:
                DebugPrint("ERR: Failed to signal exit on next queued zombie!")
            RemovedEvent.Signal(TopQueueSlot)
            return option{TopQueueSlot}
        else:
            Print("ERR: Failed to Remove a zombie info from a zombie queue!") 
        return false

    # Forcefully pops the slot off of the queue
    ForcePop<public><final>(Slot:int):void=
        if:
            QueueSlot := GetQueue().Find[Slot]
            RemovalEvent := QueueMap[Slot]?
            set QueueMap[Slot] = false
        then:
            RemovalEvent.Signal()
            RemovedEvent.Signal(Slot)
            if:
                QueueSlot = 0
                NextQueueRemove := QueueMap[GetQueue()[0]]?
            then:
                NextQueueRemove.Signal()
        GetQueue()

    # Pops all slots off the queue
    PopAll<public><final>():[]int=
        QueueList := GetQueue()
        PoppedQueueSlots:=for:
            QueueSlot:QueueList
            RemovalEvent:=QueueMap[QueueSlot]?
        do:
            RemovalEvent.Signal()
            RemovedEvent.Signal(QueueSlot)
            QueueSlot
        set QueueMap = map{}
        PoppedQueueSlots

    # Waits for the queue to change (added/removal events)
    # (1) TRUE = The event was an addition
    # (1) FALSE = The event was a removal
    AwaitUpdate<public><final>()<suspends>:tuple(int, logic)=
        race:
            (AddedEvent.Await(), true)
            (RemovedEvent.Await(), false)

    # Fetch all queued slots
    GetQueue<public><final>()<transacts>:[]int=
        var UpdatedQueue : [int]?event() = map{}
        QueueList:=for:
            ThisSlot->MaybeExitEvent:QueueMap
            ThisExitEvent:=MaybeExitEvent?
        do:
            set UpdatedQueue = ConcatenateMaps(UpdatedQueue, map{ThisSlot => option. ThisExitEvent})
            ThisSlot
        set QueueMap = UpdatedQueue
        return QueueList

    GetQueuePosition<public><final>(InSlot:int)<transacts><decides>:int=
        GetQueue().Find[InSlot]

    # Checks if the slot is in the queue
    IsInQueue<public><final>(Slot:int)<transacts><decides>:void=
        QueueMap[Slot]?

    # Checks if the queue is empty of slots
    IsEmpty<public><final>()<transacts><decides>:void=
        GetQueue().Length <= 0

    # Prints strings to the log only if DebugState is true
    DebugPrint<private><final>(String:string):void=
        if(DebugState?):
            Print(String)