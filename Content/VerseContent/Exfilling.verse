using { /Fortnite.com/Devices }
using { /Fortnite.com/Characters }
using { /Fortnite.com/Game }
using { /Fortnite.com/UI }
using { /Fortnite.com/FortPlayerUtilities }
using { /Verse.org/Simulation }
using { /Verse.org/Random }
using { /Verse.org/Colors }
using { /UnrealEngine.com/Temporary/Diagnostics }
using { /UnrealEngine.com/Temporary/SpatialMath }
using { /UnrealEngine.com/Temporary/UI }
using { /Verse.org/Colors/NamedColors }
using { InteractiveModule }
using { Room }
using { Zombie }
using { Player }
using { UI }
using { Spawning }
using { Trials }

(InAgent:agent).HasExfilledBefore<public>()<transacts><decides>:void=
    GlobalExfilManager[GetSession()].ExfilCompletedSaveSwitch_ClassicMode.GetCurrentState[InAgent] or GlobalExfilManager[GetSession()].ExfilCompletedSaveSwitch_HardcoreMode.GetCurrentState[InAgent]

IsExfilActive<public>()<transacts><decides>:void=
    GlobalExfilManager[GetSession()].ExfilActiveState?

var GlobalExfilState : weak_map(session, logic) = map{}
var GlobalExfilManager : weak_map(session, ExfilManager) = map{}

ExfilManager<public> := class(Interactive_Base):

    @editable SpawnControls : []spawn_group = array{}
    @editable ExfilEliminations_Easy : int = 35
    @editable ExfilEliminations_Classic : int = 50
    @editable ExfilEliminations_Hardcore : int = 60

    @editable ExfilDoorsToClose : []door = array{}
    @editable ExfilBorders : []creative_prop = array{}
    @editable AllowExfilPurchase_OnTrigger : trigger_device = trigger_device{}
    @editable OnExfilActivated_ActivateTrigger : trigger_device = trigger_device{}
    @editable OnExfilCompleted_ActivateTrigger : trigger_device = trigger_device{}
    @editable OnExfilCompleted_ActivateTriggerPerPlayer : trigger_device = trigger_device{}
    @editable ExfilCompletedSaveSwitch_ClassicMode : switch_device = switch_device{}
    @editable ExfilCompletedSaveSwitch_HardcoreMode : switch_device = switch_device{}
    @editable ExfilCompletedSaveSwitch_UnderOneHour : switch_device = switch_device{}
    @editable OnExfilFailed_ActivateTrigger : trigger_device = trigger_device{}
    @editable DelayToEndGameOnExfil : float = 0.0

    @editable ActivateExfilLocation : Interactable = Interactable{}

    @editable PotentialKeyLocations : []Interactable = array{}

    var TotalCost : int = 100000
    var RemainingCost : int = 100000

    var IsOpened : logic = false
    var IsBought : logic = false
    var ExfilActiveState : logic = false
    var ExfilSuccessState : logic = false

    OnBegin<override>()<suspends>:void=
        if. set GlobalExfilState[GetSession()] = false
        if. set GlobalExfilManager[GetSession()] = Self
        for(ThisBorder:ExfilBorders):
            ThisBorder.Hide()
        AllowExfilPurchase_OnTrigger.TriggeredEvent.Subscribe(AllowExfilPurchase)
        LinkAndCompile(array{ActivateExfilLocation}+PotentialKeyLocations)
        for(ThisKeyLoc:PotentialKeyLocations):
            ThisKeyLoc.SetEnabledState(false)
            ThisKeyLoc.PropLocation.Hide()
        Sleep(2.0)
        if(ThisRandomKeyLocation:=Shuffle(PotentialKeyLocations)[0]):
            ThisRandomKeyLocation.SetEnabledState(true)
            ThisRandomKeyLocation.PropLocation.Show()

    AllowExfilPurchase<private>(Agent:?agent):void=
        for(ThisKey:PotentialKeyLocations):
            ThisKey.SetEnabledState(false)
            ThisKey.PropLocation.Hide()
        set IsOpened = true
        PlayerCount:=GetPlayerCount()
        set TotalCost = case(PlayerCount):
            1 => 80000
            2 => 200000
            3 => 250000
            4 => 300000
            _ => 100000
        set RemainingCost = TotalCost
    
    ForceEnableExfil<public>():void=
        AllowExfilPurchase(false)
        set IsBought = true

    GetInteractionText<override>(InAgent:agent, InteractableLoc:Interactable):string=
        if(IsBought?):
            if(ExfilActiveState?):
                SetInteractableState(false)
                return "exfil in progress"
            else:
                SetInteractableState(true)
                return "start exfil [<C_R>ends game after<E>]"
        else if(IsOpened?):
            #CashColor := CLerp(NamedColors.Yellow, NamedColors.Green, ((TotalCost-RemainingCost)*1.0)/(TotalCost*1.0))
            SetInteractableState(true)
            return "contribute to exfil [<C_Y>${TotalCost-RemainingCost}/${TotalCost}<E>]"
        else if(PotentialKeyLocations.FindInstance[InteractableLoc]):
            SetInteractableState(true)
            return "Pick up truck key"
        else:
            SetInteractableState(false)
            return "<C_R>locked<E>"

    OnInteracted<override>(InAgent:agent, InteractableLoc:Interactable):void=
        if(InteractableLoc.IsEqual[ActivateExfilLocation]):
            if(not IsBought?):
                if(ThisPlayerInfo:=InAgent.GetPlayerInfo[]):
                    ThisCost := if. RemainingCost  > ThisPlayerInfo.Points then. ThisPlayerInfo.Points else. RemainingCost
                    PurchasedState := if. (RemainingCost-ThisPlayerInfo.Points) <= 0 then. true else. false
                    ThisPlayerInfo.BuyItem(ThisCost, ?PlayPurchaseSound:=PurchasedState)
                    set RemainingCost -= ThisCost
                    if(RemainingCost <= 0):
                        set IsBought = true
            else:
                spawn. ExfilSequence()
        else:
            AllowExfilPurchase(false)
            FalsePart:?buildable_part=false
            spawn. InAgent.GetUIManager().PartsUI.OnPartPickedUp(FalsePart) # -> Shows the key pickup on everyones screen

    ExfilSequence()<suspends>:void=
        if(ThisPlayerManager:=GlobalPlayerManager[GetSession()], not ExfilActiveState?, RandomPlayer:=GetAgents()[0]):
            set ExfilActiveState = true
            if(ThisWaveDevice:=GlobalWaveController[GetSession()]):
                ThisWaveDevice.SetWaveEndPermission(false)
                ThisWaveDevice.SetSpawningState(false)
                ThisWaveDevice.StopWaves()
            Sleep(0.5)
            ThisZombieList:=GetZombieList()
            ThisZombieList.Eliminate()
            for(ThisDoor:ExfilDoorsToClose):
                ThisDoor.SetEnabledState(false)
                ThisDoor.SetBarrierState(true)
            for(ThisBorder:ExfilBorders):
                ThisBorder.Show()
            # Drag all players outside of the spawn room and Convoy
            PlayerList:=GetAgents()
            for(ThisPlayer:PlayerList):
                if:
                    ThisRoom:=ThisPlayer.GetOccupyingRoom[]
                    ThisRoom.RoomText <> "main courtyard" and ThisRoom.RoomText <> "ruined convoy"
                    ThisSpawner:=ThisPlayerManager.SpawnPads[ThisPlayer.GetPlayerInfo[].UniqueIndex]
                    ThisPlayer.GetFortCharacter[].TeleportTo[ThisSpawner.GetTransform().Translation+vector3{Z:=70.0}, rotation{}]

            ElimsToExfil:=block{
                ThisExfilAmount:=case(GetGameDifficulty()):
                    game_difficulty.Easy => ExfilEliminations_Easy
                    game_difficulty.Classic => ExfilEliminations_Classic
                    game_difficulty.Hardcore => ExfilEliminations_Hardcore
                ExfilMultiplyAmount:=case(GetAgents().Length):
                    1 => 1.0
                    2 => 1.5
                    3 => 2.0
                    4 => 2.5
                    _ => 2.5
                Int[ThisExfilAmount*ExfilMultiplyAmount] or ThisExfilAmount
            }
            MaxExfilTime := if. GetGameDifficulty()=game_difficulty.Hardcore then. 80.0 else. 99.0
            if(ThisTrialManager:=GlobalTrialManager[GetSession()]):
                ThisTrialManager.Reset()
            SpawnsAtOnceOverride:=case(PlayerList.Length):
                1 => 4
                2 => 6
                3 => 8
                4 => 10
                _ => 8
            AllPlayerUI:=for:
                ThisPlayer:GetAgents(?OnlyActives:=false)
                ThisUI:=ThisPlayer.GetUIManager()
            do:
                ThisPlayerManager.EXFIL_ActiveMusic.Register(ThisPlayer)
                ThisUI.TrialUI.Show(ThisPlayer)
                ThisUI.TrialUI.SetRewardVisibilityState(false)
                ThisUI
            StartTime:=GetSimulationElapsedTime()
            ThisPlayerManager.EXFIL_ActiveMusic.Play()
            OnExfilActivated_ActivateTrigger.Trigger(RandomPlayer)
            Winner:=race:
                block:
                    if(PlagueController:=zombie_type_list.Plague.GetController[]):
                        loop:
                            NewPlague:=PlagueController.ZombieSpawned.Await()
                            NewPlague.Heal(10000.0)
                    false
                block:
                    loop:
                        for(ThisSpawnGroup:SpawnControls):
                            set ThisSpawnGroup.SpawnsAtOnce = SpawnsAtOnceOverride
                            if(ThisSpawnGroup.ZombieType = zombie_type_list.Plague):
                                if(GetZombieList(?TypeFilter:=array. zombie_type_list.Plague).Length < 3):
                                    ThisSpawnGroup.Spawn()
                            else:
                                ThisSpawnGroup.Spawn(?AmountOverride:=option. (ThisSpawnGroup.Amount-ThisSpawnGroup.GetActiveZombies().Length))
                        Sleep(2.5)
                    false
                block:
                    var TotalPauseTime : float = 0.0
                    SpawnGroupEliminationCollection:=for:
                        ThisSpawnGroup:SpawnControls
                    do:
                        ThisSpawnGroup.ZombieEliminatedEvent
                    for(ThisUI:AllPlayerUI):
                        ThisUI.TrialUI.TrialTimerBar.SetProgress(100.0)
                        ThisUI.TrialUI.TrialTimerBar.Show()
                        spawn. ThisUI.TrialUI.TrialTimerBar.ChangeProgressOverTime(MaxExfilTime, 0.0)
                    branch:
                        loop:
                            AwaitGameState(game_state.Paused)
                            StartPauseTime:=GetSimulationElapsedTime()
                            AwaitGameState(game_state.InGame)
                            set TotalPauseTime += (GetSimulationElapsedTime()-StartPauseTime)
                    loop:
                        var TotalElimCount : int = 0
                        for(ThisSpawnGroup:SpawnControls):
                            set TotalElimCount += ThisSpawnGroup.GetEliminatedCount()
                        TimeLeft := Int[(MaxExfilTime-(GetSimulationElapsedTime()-StartTime))+TotalPauseTime] > 0 or 0
                        for(ThisUI:AllPlayerUI):
                            ThisUI.TrialUI.SetDescription("eliminate to exfil : ({ElimsToExfil-TotalElimCount})\\ntime left : ({TimeLeft})")
                        if(TotalElimCount>=ElimsToExfil):
                            break
                        SafeSleep(1.0)
                    true
                block:
                    for(ThisUI:AllPlayerUI):
                        if(ThisWidget:=color_block[ThisUI.TrialUI.TrialTimerBar.GetBarWidget[]]):
                            ThisWidget.SetColor(MakeColorFromHex("39E75F"))
                    SafeSleep(MaxExfilTime*0.5)
                    for(ThisUI:AllPlayerUI):
                        if(ThisWidget:=color_block[ThisUI.TrialUI.TrialTimerBar.GetBarWidget[]]):
                            ThisWidget.SetColor(MakeColorFromHex("ffe146"))
                    SafeSleep(MaxExfilTime*0.25)
                    for(ThisUI:AllPlayerUI):
                        if(ThisWidget:=color_block[ThisUI.TrialUI.TrialTimerBar.GetBarWidget[]]):
                            ThisWidget.SetColor(MakeColorFromHex("ce2525"))
                    Sleep(Inf)
                    false
                block:
                    SafeSleep(MaxExfilTime+1.0)
                    false
                block:
                    AwaitGameState(game_state.GameOver)
                    false
            for(ThisSpawnGroup:SpawnControls):
                ThisSpawnGroup.Cancel()
            for(ThisUI:AllPlayerUI):
                ThisUI.TrialUI.Hide()
            ThisPlayerManager.EXFIL_ActiveMusic.Stop()
            ThisPlayerManager.EXFIL_ActiveMusic.UnregisterAll()
            if(Winner?):
                # Exfil was successful
                if. set GlobalExfilState[GetSession()] = true
                OnExfilCompleted_ActivateTrigger.Trigger()
                ZombieList:=GetZombieList()
                ZombieList.Eliminate()
                SafeSleep(DelayToEndGameOnExfil)
                Print("Exfil successful")
                EndTime:=GetTimeSinceStart()
                NewPlayerList:=GetAgents()
                for(ThisPlayer:NewPlayerList):
                    OnExfilCompleted_ActivateTriggerPerPlayer.Trigger(ThisPlayer)
                    if(ThisChair := ThisPlayer.GetPlayerPackage[].ExfilChair):
                        ThisChair.Seat(ThisPlayer)
                    if(GetGameDifficulty()<>game_difficulty.Easy, EndTime < 3600.0):
                        ExfilCompletedSaveSwitch_UnderOneHour.TurnOn(ThisPlayer)
                    case(GetGameDifficulty()):
                        game_difficulty.Classic => ExfilCompletedSaveSwitch_ClassicMode.TurnOn(ThisPlayer)
                        game_difficulty.Hardcore => ExfilCompletedSaveSwitch_HardcoreMode.TurnOn(ThisPlayer)
                        _ =>
                    if. ThisFortChar:=ThisPlayer.GetFortCharacter[] then. ThisFortChar.PutInStasis(stasis_args{AllowFalling:=false, AllowTurning:=false, AllowEmotes:=false})
                ThisPlayerManager.EXFIL_SuccessCinematic.Play()
                Sleep(2.2)
                ThisPlayerManager.GameOverMusic.Play()
                ThisPlayerManager.EXFIL_SuccessCinematic.StoppedEvent.Await()
            else:
                # Exfil failed, time ran out
                Print("Exfil FAILED")
                OnExfilFailed_ActivateTrigger.Trigger(RandomPlayer)
            SetGameState(game_state.GameOver)