using { /Verse.org/Simulation }
using { /Fortnite.com/Devices }
using { /Fortnite.com/AI }
using { /Fortnite.com/Animation/PlayAnimation }
using { /Verse.org/Assets }
using { /Fortnite.com/AI/movement_types }
using { /UnrealEngine.com/Temporary/SpatialMath }
using { Player }
using { Weapons }
using { UI }
using { Zombie }
using { ZombieTypes }
using { Spawning }
using { Perks }
using { Room }

GetTimeSinceStart<public>()<transacts>:float=
    if(GameManager:=GlobalGameManager[GetSession()]):
        EndTimeStamp := if. GetEndTime() > 0.0 then. GetEndTime() else. GetSimulationElapsedTime()
        return (EndTimeStamp-GameManager.GameStartTime)-GameManager.TotalPausedTime
    return 0.0

GetEndTime<public>()<transacts>:float=
    return GlobalGameManager[GetSession()].GameEndTime or 0.0

<#
# Safely navigates an NPC to a destination by handling pausing events
(Navigatable:navigatable).SafeNavigate(Target:navigation_target, ?ReachRadius:float=0.0, ?MovementType:movement_type=Walking, ?AllowPartialPath:logic=false, ?CancelEvents:[]event()=array{})<suspends>:logic=
    var CompletedState : logic = false
    loop:
        AwaitGameState(game_state.InGame)
        NavResult:=race:
            block:
                loop:
                    Result:=Navigatable.NavigateTo(Target, ?ReachRadius:=ReachRadius, ?MovementType:=MovementType, ?AllowPartialPath:=AllowPartialPath)
                    if(Result=navigation_result.Reached):
                        set CompletedState = true
                        break
                    else if(Result=navigation_result.Unreachable):
                        break
                    Sleep(0.0)
                true
            block:
                AwaitGameState(game_state.Paused)
                false
            block:
                CancelEvents.RaceAll()
                true
        if(NavResult?):
            break
        Navigatable.StopNavigation()
    return CompletedState #>

# An animating function that will be auto paused when the game is paused. Will play from the point it was paused at.
<#
(PlayAnimationController:play_animation_controller).SafeAnimate<public>(AnimationSequence:animation_sequence, Time:float, ?PlayRate:float=1.0, ?StartPositionSeconds:float=0.0, ?BlendInTime:float=0.0, ?BlendOutTime:float=0.0)<suspends>:void=
    PlayAnimationController.PlayAndAwait(AnimationSequence, ?PlayRate:=PlayRate, ?StartPositionSeconds:=StartPositionSeconds, ?BlendInTime:=BlendInTime, ?BlendOutTime:=BlendOutTime)
    <#
    if. Time < 0.0 then. return
    var LastAnimInstance : ?play_animation_instance = false
    var TimePlayed : float = 0.0
    loop:
        StartTime:=GetSimulationElapsedTime()
        AnimState:=race:
            block:
                AwaitGameState(game_state.Paused)
                false
            block:
                ThisBlendInTime := if. TimePlayed < Time then. 0.0 else. BlendInTime
                #Print("AnimInfo: StartTime[{StartPositionSeconds+TimePlayed}], BlendInTime[{ThisBlendInTime}]")
                AnimInstance:=PlayAnimationController.Play(AnimationSequence, ?PlayRate:=PlayRate, ?StartPositionSeconds:=StartPositionSeconds+TimePlayed, ?BlendInTime:=ThisBlendInTime, ?BlendOutTime:=BlendOutTime)
                set LastAnimInstance = option{AnimInstance}
                defer{
                    AnimInstance.Stop()
                }
                loop:
                    AnimInstance.Await()
                    if. IsGameState[game_state.InGame] then. break
                Print("Last animation state: {AnimInstance.GetState().ToString()}")
                true
        if. AnimState? then. break
        set TimePlayed += GetSimulationElapsedTime()-StartTime
        PlayAnimationController.Play(AnimationSequence, ?PlayRate:=0.0, ?StartPositionSeconds:=TimePlayed)
        AwaitGameState(game_state.InGame) #>
    Print("Finished playing animation")#>

SafeMoveTo<public>(Prop:creative_object, Position:vector3, Rotation:rotation, Time:float)<suspends>:void=
    var TimePlayed : float = 0.0
    loop:
        StartTime := GetSimulationElapsedTime()
        FinishedState:=race:
            block:
                Prop.MoveTo(Position, Rotation, Time-TimePlayed)
                true
            block:
                AwaitGameState(game_state.Paused)
                false
        if(FinishedState?):
            break
        CurrPos:=Prop.GetTransform()
        if. Prop.TeleportTo[CurrPos]
        if. Prop.TeleportTo[CurrPos]
        set TimePlayed += GetSimulationElapsedTime()-StartTime
        AwaitGameState(game_state.InGame)


# A sleep that will be auto paused when the game is paused
SafeSleep<public>(Time:float)<suspends>:void=
    var TimeLeft : float = Time
    loop:
        StartTime:=GetSimulationElapsedTime()
        TimerState:=race:
            block:
                AwaitGameState(game_state.Paused)
                false
            block:
                Sleep(TimeLeft)
                true
        if. TimerState? then. break
        set TimeLeft -= GetSimulationElapsedTime()-StartTime
        AwaitGameState(game_state.InGame)

game_state<public> := enum{
    Menu
    InGame # When gameplay is active
    Paused
    GameOver
}

var GlobalGameManager<public> : weak_map(session, Game_Manager) = map{}

Game_Manager<public> := class(creative_device):

    @editable
    Input_GameplayAdminPanel : input_trigger_device = input_trigger_device{}

    @editable
    FreezingMutatorZone : mutator_zone_device = mutator_zone_device{}

    var GameState<public> : game_state = game_state.Menu
    var GameStateChangedEvent<public> : event(game_state) = event(game_state){}

    var GameDifficulty<public> : game_difficulty = game_difficulty.Easy
    var GameDifficultyChangedEvent<public> : event(game_difficulty) = event(game_difficulty){}

    var CountdownState<public> : logic = false
    var GameStartTime<public> : float = 0.0
    var GameEndTime<public> : float = 0.0
    var TotalPausedTime<public> : float = 0.0
    var LastPauseTimestamp<public> : float = 0.0
    StopCountdownEvent<public> : event() = event(){}

    # ADMIN MENU PANEL
    var<public> AutoUnlockDoors : logic = false
    var<public> AutoPowerMap : logic = false
    var<public> StartingWaveOverride : int = -1
    var<public> DPState : logic = false

    @editable
    ActivateTrigger_OnMaxPlayers<public> : trigger_device = trigger_device{} 

    @editable
    ActivateTrigger_OnGameStart<public> : trigger_device = trigger_device{}
    @editable
    ActivateTrigger_OnGameEnd<public> : trigger_device = trigger_device{}

    OnBegin<override>():void=
        if. set GlobalGameManager[GetSession()] = Self
        else. Print("Failed to link game manager")
        Input_GameplayAdminPanel.PressedEvent.Subscribe(ShowGameAdminPanel)

    OnVoteReadyToggle<public>(InPlayer:agent, InState:logic):void=
        ThisPlayerList := GetPlayers(?OnlyActives:=false)
        ReadyPlayerList := for:
            ThisPlayer : ThisPlayerList
            ThisPlayer.ReadyUpState?
        do:
            0
        if(ThisPlayerList.Length <= 2):
            if(ReadyPlayerList.Length >= ThisPlayerList.Length):
                spawn. CountdownStart()
            else:
                StopCountdownEvent.Signal()
        else if(ReadyPlayerList.Length > Ceil[1.0*ThisPlayerList.Length/2.0]):
            spawn. CountdownStart()
        else:
            StopCountdownEvent.Signal()
            
    CountdownStart<public>()<suspends>:void=
        if(not CountdownState?):
            set CountdownState = true
            var CountdownCompletedState : logic = false
            var CountdownTime : int = if. IsSinglePlayer[] then. 3 else. 8
            ThisPlayerList := GetAgents(?OnlyActives:=false)
            for(ThisPlayer : ThisPlayerList):
                spawn. ThisPlayer.GetUIManager().MainMenuUI.StartCountdown()

            branch:
                Sleep(CountdownTime*1.0)
                set CountdownCompletedState = true
                StopCountdownEvent.Signal()
            StopCountdownEvent.Await()
            set CountdownState = false
            for(ThisPlayer : ThisPlayerList):
                ThisPlayer.GetUIManager().MainMenuUI.CancelCountdownEvent.Signal()
            if(CountdownCompletedState?):
                if(ThisPlayerList.Length >= 4):
                    ActivateTrigger_OnMaxPlayers.Trigger()
                SetGameState(game_state.InGame)
        else:
            StopCountdownEvent.Signal()

    ShowGameAdminPanel<private>(Agent:agent):void=
        if(IsGameState[game_state.InGame], ThisPlayerInfo:=Agent.GetPlayerInfo[], ThisPlayerInfo.GameplayMenuEnabledState?):
            Agent.GetUIManager().SetGamePanelState(true)

IsCountdown<public>()<transacts><decides>:void=
    GlobalGameManager[GetSession()].CountdownState?

SetGameState<public>(InState:game_state)<suspends>:void=
    if(IsGameState[game_state.GameOver]):
        return
    PreviousState := GetGameState()
    if(set GlobalGameManager[GetSession()].GameState = InState):
        GetGameStateChangedEvent().Signal(InState)
        case(InState):
            game_state.Menu =>
            game_state.InGame =>
                if(PreviousState = game_state.Menu):
                    if(ThisPlayerManager:=GlobalPlayerManager[GetSession()]):
                        # Spawn players into the map from the menu
                        PlayerList:=GetPlayers(?OnlyActives:=false)
                        for(ThisPlayer:PlayerList):
                            ThisPlayer.User.GetUIManager().ApplyUI(ThisPlayer.User)
                            spawn. ThisPlayer.User.GetUIManager().FadeInBlackScreen(0.1)
                        Sleep(1.0)
                        ThisPlayerManager.ActivateTrigger_OnGameStart.Trigger()
                        ThisPlayerManager.FreezingMutatorZone.Disable()
                        for(ThisPlayer:PlayerList):
                            spawn. ThisPlayer.User.GetUIManager().FadeOutBlackScreen(1.75)
                            spawn. ThisPlayer.PlayState(player_state.Active)
                        set ThisPlayerManager.GameStartTime = GetSimulationElapsedTime()
                        # ADMIN SETTINGS
                        if(ThisPlayerManager.AutoUnlockDoors?):
                            RoomList:=GetRooms()
                            for(ThisRoom:RoomList):
                                ThisRoom.ForceOpenDoors()
                        if(ThisPlayerManager.AutoPowerMap?, ThisPowerManager:=GlobalPowerDevice[GetSession()]):
                            ThisPowerManager.Enable()
                        if(ThisPlayerManager.StartingWaveOverride > 0, ThisWaveController := GlobalWaveController[GetSession()], ThisWave := type{_X:int where _X >= 1}[ThisPlayerManager.StartingWaveOverride]):
                            set ThisWaveController.Controls.StartingWave = ThisWave
                        #
                    if(ThisWaveController:=GlobalWaveController[GetSession()]):
                        ThisWaveController.StartWaves()
                    if(ThisSaviorPerk := saveyoursoda_perk[GetPerk[perk_list.SaviorSoda]]):
                        ThisSaviorPerk.Reset()

                    spawn. LoopSpawnPlague() # -> Process for spawning plague boss
                else if(PreviousState = game_state.Paused):
                    if(ThisPlayerManager:=GlobalPlayerManager[GetSession()]):
                        ThisPlayerManager.PauseCinematic.Stop()
                        set ThisPlayerManager.TotalPausedTime += (GetSimulationElapsedTime()-ThisPlayerManager.LastPauseTimestamp)
            game_state.Paused =>
                PlayerList := GetPlayers(?OnlyActives:=false)
                for(ThisPlayer:PlayerList):
                    spawn. ThisPlayer.PlayState(player_state.Paused)
                ZombieList := GetZombieList()
                for(ThisZombie:ZombieList):
                    spawn. ThisZombie.PlayState(zombie_state.Paused)
                if(ThisPlayerManager:=GlobalPlayerManager[GetSession()]):
                    ThisPlayerManager.PauseCinematic.Play()
                    set ThisPlayerManager.LastPauseTimestamp = GetSimulationElapsedTime()
            game_state.GameOver =>
                Print("===GAME OVER===")
                # Eliminate all players, disable zombie spawning and show end cinematic/screen
                if(WaveDevice:=GlobalWaveController[GetSession()]):
                    WaveDevice.SetSpawningState(false)
                ZombieList:=GetZombieList()
                for(ThisZombie:ZombieList):
                    spawn. ThisZombie.PlayState(zombie_state.GameOver)
                if(ThisPlayerManager:=GlobalPlayerManager[GetSession()]):
                    set ThisPlayerManager.GameEndTime = GetSimulationElapsedTime()
                    ThisPlayerManager.ActivateTrigger_OnGameEnd.Trigger()
                    PlayerList := GetAgents(?OnlyActives:=false)
                    for(ThisPlayer:PlayerList):
                        if(ThisPlayerInfo:=ThisPlayer.GetPlayerInfo[], ThisPlayerInfo.IsActive[]):
                            spawn. ThisPlayerInfo.PlayState(player_state.Downed)
                    if(FirstPlayer:=PlayerList[0]):
                        for(ThisPlayer:PlayerList):
                            if. ThisPlayerInfo:=ThisPlayer.GetPlayerInfo[] then. ThisPlayerInfo.SetMoveState(false)
                            ThisPlayer.GetUIManager().ShowUI_EndScreen()
                        if(GlobalExfilState[GetSession()]?):
                            ThisPlayerManager.GameOverCinematic.SetPlaybackTime(6.0)
                            ThisPlayerManager.GameOverCinematic.Play()
                            <#
                            for(ThisPlayer:PlayerList):
                                # Play black screen fade-in
                                ThisUI:=ThisPlayer.GetUIManager()
                                ThisUI.PausedScreenCover.Show()
                                ThisUI.PausedScreenCover.SetOpacity(1.0)
                            Sleep(5.9)
                            for(ThisPlayer:PlayerList):
                                # Play black screen fade-out
                                ThisUI:=ThisPlayer.GetUIManager()
                                spawn. ThisUI.FadeOutBlackScreen(0.65)
                                ThisUI.PausedScreenCover.Hide()
                                ThisUI.PausedScreenCover.SetOpacity(0.85)#>
                        else:
                            ThisPlayerManager.GameOverCinematic.Play()
                            ThisPlayerManager.GameOverMusic.Play()
                            if(IsExfilActive[], ThisExfilManage:=GlobalExfilManager[GetSession()]):
                                ThisExfilManage.OnExfilFailed_ActivateTrigger.Trigger()
                            Sleep(5.0)
                            for(ThisPlayer:PlayerList):
                                # Play black screen fade-in
                                spawn. ThisPlayer.GetUIManager().FadeInBlackScreen(0.85)
                            Sleep(0.9)
                            for(ThisPlayer:PlayerList):
                                # Play black screen fade-out
                                spawn. ThisPlayer.GetUIManager().FadeOutBlackScreen(0.65)
                        ThisPlayerManager.GameOverCinematic.StoppedEvent.Await()
                        ThisPlayerManager.RoundDevice.EndRound(FirstPlayer)

IsGameState<public>(DesiredState:game_state)<transacts><decides>:void=
    GetGameState() = DesiredState

GetGameState<public>()<transacts>:game_state=
    GlobalGameManager[GetSession()].GameState or game_state.Menu

GetGameStateChangedEvent<public>()<transacts>:event(game_state)=
    GlobalGameManager[GetSession()].GameStateChangedEvent or event(game_state){}

AwaitGameState<public>(DesiredState:game_state)<suspends>:void=
    loop:
        if. GlobalGameManager[GetSession()] then. break
        Sleep(0.1)
    if(not IsGameState[DesiredState], GameStateListen := GetGameStateChangedEvent()):
        loop:
            NewState := GameStateListen.Await()
            if(NewState = DesiredState):
                return

IsGameDifficulty<public>(DesiredState:game_difficulty)<transacts><decides>:void=
    GlobalGameManager[GetSession()].GameDifficulty = DesiredState

GetGameDifficultyChangedEvent<public>()<transacts><decides>:event(game_difficulty)=
    return GlobalGameManager[GetSession()].GameDifficultyChangedEvent

GetGameDifficulty<public>()<transacts>:game_difficulty=
    return GlobalGameManager[GetSession()].GameDifficulty or game_difficulty.Classic

# Switches the game difficulty from the current one to the next
# ORDER: Easy -> Classic -> Hardcore -> Easy...
ToggleGameDifficulty<public>():void=
    CurrGameDifficulty:=GetGameDifficulty()
    NextGameDifficulty := case(CurrGameDifficulty):
        game_difficulty.Easy =>
            game_difficulty.Classic
        game_difficulty.Classic =>
            game_difficulty.Hardcore
        game_difficulty.Hardcore =>
            game_difficulty.Easy
    if(ThisPlayerManager:=GlobalPlayerManager[GetSession()]):
        set ThisPlayerManager.GameDifficulty = NextGameDifficulty
        ThisPlayerManager.GameDifficultyChangedEvent.Signal(NextGameDifficulty)

(Difficulty:game_difficulty).ToString<public>():string=
    case(Difficulty):
        game_difficulty.Easy => "Easy"
        game_difficulty.Classic => "Classic"
        game_difficulty.Hardcore => "Hardcore"

game_difficulty<public> := enum{
    Easy
    Classic
    Hardcore
}