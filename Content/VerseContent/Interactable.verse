using { /Fortnite.com/Devices }
using { /Verse.org/Simulation }
using { /Verse.org/Colors }
using { /UnrealEngine.com/Temporary/Diagnostics }
using { /UnrealEngine.com/Temporary/SpatialMath }
using { /Fortnite.com/Characters }
using { InteractiveModule }
using { Player }
using { UI }

# A list of different types of interaction found throughout the game
interactable_type<public> := enum{
    Weapon  # Anything weapon related, which includes wall weapons, weapon crate, and power punch
    Ammo    # Ammo boxes
    Door
    Perk
    Barrier
    Revive # The revive interactables used to revive other downed players
    Misc    # Anything not already listed above
}

(InputDevice:input_trigger_device).AwaitPressedEvent<public>(InAgent:agent)<suspends>:void=
    if(InputDevice.IsHeld[InAgent]){}
    else:
        loop:
            PressedEvent:=InputDevice.PressedEvent.Await()
            if(PressedEvent = InAgent):
                break
            
(InputDevice:input_trigger_device).AwaitReleasedEvent<public>(InAgent:agent)<suspends>:void=
    if(InputDevice.IsHeld[InAgent]):
        loop:
            ReleasedEvent:=InputDevice.ReleasedEvent.Await()
            if(ReleasedEvent(0) = InAgent):
                break

var EventCollection_InteractableChanged : weak_map(session, [agent]?event(?Interactable)) = map{}
var EventCollection_InterruptInteraction : weak_map(session, [agent]?event()) = map{}
var CurrentInteractablePerPlayer : weak_map(session, [agent]?Interactable) = map{}

(InAgent:agent).GetInteractableChangedEvent<public>()<transacts>:event(?Interactable)=
    ThisEvent := EventCollection_InteractableChanged[GetSession()][InAgent]? or block{
        NewEvent : event(?Interactable) = event(?Interactable){}
        if. set EventCollection_InteractableChanged[GetSession()][InAgent] = option{NewEvent}
        else if. CurrentCollection := EventCollection_InteractableChanged[GetSession()] then if. set EventCollection_InteractableChanged[GetSession()] = ConcatenateMaps(CurrentCollection, map{InAgent=>option{NewEvent}})
        NewEvent
    }
    return ThisEvent

(InAgent:agent).GetInterruptInteractionEvent<public>()<transacts>:event()=
    ThisEvent := EventCollection_InterruptInteraction[GetSession()][InAgent]? or block{
        NewEvent : event() = event(){}
        if. set EventCollection_InterruptInteraction[GetSession()][InAgent] = option{NewEvent}
        else if. CurrentCollection := EventCollection_InterruptInteraction[GetSession()] then if. set EventCollection_InterruptInteraction[GetSession()] = ConcatenateMaps(CurrentCollection, map{InAgent=>option{NewEvent}})
        NewEvent
    }
    return ThisEvent

InteractiveDebugDraw := class(debug_draw_channel) {}

InteractiveModule<public>:=module:

    # Creates an Interactable to be used
    # Owners - If not empty, these players along with 'Owner' will only be able to view/interact with this object
    CreateInteractable<public>(Location:vector3, Rotation:rotation, ?Owners:[]agent=array{}):Interactable=
        NewInteractable : Interactable = Interactable{}
        # Only the 'Owners' (if any) will be able to interact with this new Interactable
        # If no Owners are specified, everyone will be able to interact
        if(Owners.Length > 0):
            NewInteractable.SetEnabledState(false)
            for(ThisOwner:Owners):
                NewInteractable.SetEnabledState(true, ?InAgent:=option{ThisOwner})
        NewInteractable.SetLocation(transform{Translation:=Location, Rotation:=Rotation})
        NewInteractable.Link()
        return NewInteractable

    # Fetches the desired Interactable for where the player currently stands in the world
    # Will respect Interactive rules, such as enabled states, maximum distance, interactive angle, and player placement
    # If this fails, no Interactables are currently near/accessible to the player
    FindDesiredInteractable<public>(InAgent:agent):?Interactable=
        DesiredInteractables:=FindDesiredInteractables(InAgent)
        if(FortChar:=InAgent.GetFortCharacter[], FirstInteractable:=DesiredInteractables[0]):
            FortCharPos := FortChar.GetTransform().Translation
            var ClosestInteractabletuple : tuple(Interactable, float) = (FirstInteractable, Distance(FortCharPos, FirstInteractable.GetTransform().Translation))
            for(ThisInteractable:DesiredInteractables):
                if(InteractableDistance := Distance(ThisInteractable.GetTransform().Translation, FortCharPos) < ClosestInteractabletuple(1)):
                    if(ThisInteractable.InteractivePriority > ClosestInteractabletuple(0).InteractivePriority):
                        set ClosestInteractabletuple = (ThisInteractable, InteractableDistance)
            return option{ClosestInteractabletuple(0)}
        else{return false}

    # Fetches all Interactables that meet the conditions of Interaction
    FindDesiredInteractables(InAgent:agent)<transacts>:[]Interactable=
        if(InteractableCollection:=GlobalInteractableCollection[GetSession()], ThisFortChar:=InAgent.GetFortCharacter[], ThisPlayerInfo:=InAgent.GetPlayerInfo[]):
            EnabledInteractables:=for:
                ThisInteractable:InteractableCollection
                ThisInteractable.CheckEnabledState[option{InAgent}]
                ThisInteractable.IsInInteractableDistance[ThisFortChar]
                ThisInteractable.IsInInteractableAngle[ThisFortChar]
                ThisInteractable.IsPlayerInFront[ThisFortChar]
                not ThisPlayerInfo.ExcludedInteractableTypes.Find[ThisInteractable.InteractiveType]
            do:
                ThisInteractable
            return EnabledInteractables
        return array{}

    # Container for all Interactables
    var<internal> GlobalInteractableCollection : weak_map(session, []Interactable) = map{}

    # The basis for all Interactable components
    # Each 'Interactable' is interactive within the world (or can be)
    Interactable<public> := class<concrete>(comparable_class):

        @editable Identity<public> : string = "Unset Interactable Location Identity"
        # A default prompt that can be used for when getting interaction texts #
        @editable var<private> DefaultPrompt<public> : string = "Unset Interactable Prompt"
        # A default description that can be used to display extra information #
        @editable DefaultDescription<public> : string = ""
        @editable DescriptionIconPath<public> : string = ""
        @editable AllowDescriptionOverride<public> : logic = false
        # The location of interaction via a creative prop #
        @editable var<private> PropLocation<public> : creative_prop = creative_prop{}
        # Do we need to hide the interactive location automatically?
        @editable HidePropLocation<private> : logic = true
        # Is the location of interaction static, meaning it does not move? #
        @editable var<internal> IsLocationStatic<public> : logic = true
        # Is the location is enabled on game start? #
        @editable var<private> IsEnabled<internal> : logic = true
        # What behavior should execute when this location is interacted with? #
        @editable InteractedWithBehavior<public> : interact_behavior = interact_behavior.None
        # Can this interactable be used while the player is downed? #
        @editable var AllowUseWhileDowned<public> : logic = false
        # Does this interactable need the player to be looking at it? #
        @editable var RequiresLookAt<public> : logic = true
        # The maximum look angle this interactable can be interacted from #
        @editable var InteractionRadius<public> : type{_X : float where _X > 0.0} = 50.0
        # A list of other colliders that act together
        @editable ExtraInteractionRadiusColliders<public> : []vector3 = array{}
        # Does the player have to stand still while interacting? #
        @editable var RequiresStandStill<public> : logic = false
        # Does the player have to be in front of this interactable? #
        @editable var RequiresPlayerMustBeInFront<public> : logic = true
        # The time it takes to interact with this interactable location #
        @editable var InteractTime<public> : float = 0.75
        # The interactable will be interacted with automatically upon looking at #
        # NOTE: Still considers "InteractTime" value #
        @editable var InteractByViewing<public> : logic = false
        # The maximum distance allowed to interact with this interactable #
        @editable var InteractDistance<public> : float = 200.0
        # Ignores the interact distance check #
        @editable var IgnoreDistanceCheck<public> : logic = false
        # The offset of the interaction location
        @editable var InteractionOffset<public> : vector3 = vector3{}
        # The offset from this interactable that the interaction sprite will appear at #
        @editable var InteractionSpriteOffset<public> : vector3 = vector3{}
        # The multiplier to the size of the interaction sprite #
        @editable var InteractionSpriteSizeMultiplier<public> : type{_X : float where 0.0 < _X} = 1.0
        # If interaction is disabled with this interactable, should the interaction sprite still be visible? #
        @editable var AlwaysShowInteractionSprite<public> : logic = false
        # Enables the location on a trigger signal #
        @editable OnTrigger_Enable : trigger_device = trigger_device{}
        # Activates the trigger when this interactable is interacted with #
        @editable ActivateTrigger_OnInteracted : trigger_device = trigger_device{}
        # Can the interact time be influenced by outside multipliers? #
        @editable AllowInteractTimeInfluence : logic = true
        # Display the interacting animation #
        @editable UseUIAnimation<public> : logic = true

        @editable EnableDebug : logic = false

        @editable InteractivePriority : int = 0

        # VARIABLES
        var OriginalPrompt<private> : string = ""
        var InitState<private> : logic = false
        var<private> EnabledState : logic = false

        var<internal> StaticLocation<public> : transform = transform{}

        var EnabledStatePerPlayer<private> : [agent]logic = map{}
        var InteractionTimePerPlayer<private> : [agent]float = map{}
        var InteractableStatePerPlayer<private> : [agent]logic = map{}
        var IsInteractingStatePerPlayer<private> : [agent]logic = map{}
        var<private> IsInteractable<public> : logic = true

        var ParentDevice<public> : ?Interactive_Base = false
        var InteractiveType<public> : interactable_type = interactable_type.Misc

        # EVENTS
        InteractingEvent<public>:event(agent)=event(agent){}
        InteractedEvent<public>:event(agent)=event(agent){}
        StoppedInteractingEvent<public>:event(agent)=event(agent){}

        # Initializes this Interactable
        # MUST be called for this Interactable to be active and use-able
        #   GlobalLink - True by default. If False, this Interactable will NOT be linked globally, and must be by some outside setup
        Link<public><final>():void=
            if(not InitState?):
                set InitState = true
                set EnabledState = IsEnabled
                set OriginalPrompt = DefaultPrompt
                OnTrigger_Enable.TriggeredEvent.Subscribe(EnableLocation)
                if(PropLocation.IsValid[]):
                    if. HidePropLocation? then. PropLocation.Hide()
                    set StaticLocation = PropLocation.GetTransform()
                else:
                    set IsLocationStatic = true
                if. not GlobalInteractableCollection[GetSession()] then if. set GlobalInteractableCollection[GetSession()] = array{}
                if. CurrentInteractableCollection := GlobalInteractableCollection[GetSession()]
                then if. set GlobalInteractableCollection[GetSession()] = CurrentInteractableCollection+array{Self}
                spawn. OnLinked()

        OnLinked<public>()<suspends>:void=
            AwaitGameState(game_state.InGame)
            if(EnableDebug?):
                Print("Interactive debug enabled")
                ThisLoc := GetTransform()
                # Draws debug for interaction distance and interaction radius
                InteractionDebugDraw : debug_draw = debug_draw{Channel := InteractiveDebugDraw}
                InteractionDebugDraw.DrawSphere(ThisLoc.Translation, ?Radius:=InteractionRadius, ?Color:=NamedColors.Yellow, ?Thickness:=5.0, ?DrawDurationPolicy:=debug_draw_duration_policy.Persistent)
                InteractionDebugDraw.DrawSphere(ThisLoc.Translation, ?Radius:=InteractDistance, ?Color:=NamedColors.Green, ?Thickness:=2.5, ?DrawDurationPolicy:=debug_draw_duration_policy.Persistent)
                for(ThisExtraCollider:ExtraInteractionRadiusColliders):
                    ThisLocation:=ThisLoc.GetLocalVector(ThisExtraCollider)
                    InteractionDebugDraw.DrawSphere(ThisLocation, ?Radius:=InteractionRadius, ?Color:=NamedColors.LightYellow, ?Thickness:=2.5, ?DrawDurationPolicy:=debug_draw_duration_policy.Persistent, ?Duration:=60.0)

        Unlink<public><final>():void=
            if(InitState?):
                if:
                    InteractableIndex := GlobalInteractableCollection[GetSession()].FindInstance[Self]
                    UpdatedInteractableCollection := GlobalInteractableCollection[GetSession()].RemoveElement[InteractableIndex]
                    set GlobalInteractableCollection[GetSession()] = UpdatedInteractableCollection
                else. Print("ERROR: Failed to unlink an Interactable")
                set InitState = false

        # Resets this Interactable to its default settings
        Reset<public><final>():void=
            SetEnabledState(IsEnabled)
            ResetText()

        # Activates when this Interactable is being interacted with
        OnInteractionStarted<public>(InAgent:agent):void=
            if. set IsInteractingStatePerPlayer[InAgent] = true
            else. set IsInteractingStatePerPlayer = ConcatenateMaps(IsInteractingStatePerPlayer, map{InAgent => true})
            InteractingEvent.Signal(InAgent)
            if(ThisParent:=ParentDevice?):
                spawn. ThisParent.OnInteractStart(InAgent, Self)

        OnInteractionStopped<public>(InAgent:agent):void=
            if. set IsInteractingStatePerPlayer[InAgent] = false
            else. set IsInteractingStatePerPlayer = ConcatenateMaps(IsInteractingStatePerPlayer, map{InAgent => false})
            StoppedInteractingEvent.Signal(InAgent)
            if(ThisParent:=ParentDevice?):
                spawn. ThisParent.OnInteractStopped(InAgent, Self)

        # Activates when this Interactable is interacted with, handling settings and signaling
        Interact<public><final>(InAgent:agent):void=
            case(InteractedWithBehavior):
                interact_behavior.DisableForAll =>
                    SetEnabledState(false)
                interact_behavior.DisableForPlayer =>
                    if(not EnabledStatePerPlayer[InAgent]? and EnabledStatePerPlayer[InAgent]):
                        Print("~InteractableLog: This interactable is disabled for you (one time use)!")
                        return
                    else if(not EnabledStatePerPlayer[InAgent]):
                        set EnabledStatePerPlayer = ConcatenateMaps(EnabledStatePerPlayer, map{InAgent => false})
                    else if(set EnabledStatePerPlayer[InAgent] = false){}
                interact_behavior.None =>
            InteractedEvent.Signal(InAgent)
            ActivateTrigger_OnInteracted.Trigger(InAgent)
            if(ThisParent:=ParentDevice?):
                spawn. ThisParent.OnInteracted(InAgent, Self)

        GetInteractingPlayers<public><final>()<transacts>:[]agent=
            InteractingPlayerList:=for:
                ThisPlayer->InteractingState:IsInteractingStatePerPlayer
                InteractingState?
            do:
                ThisPlayer

        SetText<public><final>(InString:string):void=
            set DefaultPrompt = InString

        ResetText<public><final>():void=
            set DefaultPrompt = OriginalPrompt

        GetInteractionText<public>(Agent:agent):string=
            if(ThisParentDevice:=ParentDevice?):
                DeviceText:=ThisParentDevice.GetInteractionText(Agent, Self)
                if(DeviceText <> "NA"):
                    return DeviceText
            return DefaultPrompt

        GetDescription<public><final>(Agent:agent):tuple(string, string)=
            if(ThisParentDevice:=ParentDevice?):
                return ThisParentDevice.GetDescription(Agent, Self)
            else:
                return (DefaultDescription, "")

        # Sets whether or not players can interact with this Interactable
        #   InAgent - If true, this player will only be affected
        SetInteractableState<public><final>(InState:logic, ?InAgent:?agent=false):void=
            if(ThisAgent:=InAgent?):
                if(set InteractableStatePerPlayer[ThisAgent] = InState){}
                else:
                    set InteractableStatePerPlayer = ConcatenateMaps(InteractableStatePerPlayer, map{ThisAgent => InState})
            else:
                set IsInteractable=InState
                for(ThisPlayer : GetAgents(?OnlyActives:=false)):
                    if(set InteractableStatePerPlayer[ThisPlayer] = InState){}

        IsInteractableForPlayer<public><final>(InAgent:agent)<transacts><decides>:void=
            if(ThisInteractableState:=InteractableStatePerPlayer[InAgent]):
                ThisInteractableState?
            else:
                false

        # Sets whether or not players can view with this Interactable
        #   InAgent - If true, this player will only be affected
        SetEnabledState<public><final>(InState:logic, ?InAgent:?agent=false):void=
            if(ThisAgent:=InAgent?):
                if(set EnabledStatePerPlayer[ThisAgent] = InState){}
                else:
                    set EnabledStatePerPlayer = ConcatenateMaps(EnabledStatePerPlayer, map{ThisAgent => InState})
            else:
                set EnabledState = InState
                for(ThisPlayer : GetAgents(?OnlyActives:=false)):
                    if(set EnabledStatePerPlayer[ThisPlayer] = InState){}

        CheckEnabledState<public><final>(InAgent:?agent)<transacts><decides>:void=
            if(ThisEnabledState:=EnabledStatePerPlayer[InAgent?]):
                ThisEnabledState?
            else:
                EnabledState?

        SetInteractionTime<public><final>(InTime:float, ?InAgent:?agent=false):void=
            if(ThisAgent:=InAgent?):
                if(set InteractionTimePerPlayer[ThisAgent] = InTime){}
                else:
                    set InteractionTimePerPlayer = ConcatenateMaps(InteractionTimePerPlayer, map{ThisAgent => InTime})
            else:
                set InteractTime = InTime
                for(ThisPlayer : GetAgents(?OnlyActives:=false)):
                    if(set InteractionTimePerPlayer[ThisPlayer] = InTime){}

        GetInteractionTime<public><final>(?InAgent:?agent=false, ?TimeMultiply:float=1.0)<transacts>:float=
            ThisInteractTime := InteractionTimePerPlayer[InAgent?] or InteractTime
            if(AllowInteractTimeInfluence?):
                return ThisInteractTime*TimeMultiply
            else:
                return ThisInteractTime

        EnableLocation<public><final>(Agent:?agent):void=
            set EnabledState = true

        IsValid<public><final>()<transacts><decides>:void=
            PropLocation.IsValid[] or IsLocationStatic?

        GetTransform<public><final>(?ApplyOffset:logic=true)<transacts>:transform=
            ThisLocation:=block{
                if(IsLocationStatic?):
                    StaticLocation
                else:
                    PropLocation.GetTransform()
            }
            if(ApplyOffset?):
                return transform{
                    Translation:=ThisLocation.GetLocalVector(InteractionOffset)
                    Rotation:=ThisLocation.Rotation
                }
            else:
                return ThisLocation

        GetSpriteTransform<public><final>()<transacts>:transform=
            if(InteractionSpriteSizeMultiplier <= 0.0):
                transform{
                    Translation := vector3{X:=0.0,Y:=0.0,Z:=0.0}
                    Rotation := rotation{}
                    Scale := vector3{X:=1.0,Y:=1.0,Z:=1.0}
                }
            else:
                transform{
                    Translation := GetTransform(?ApplyOffset:=false).GetLocalVector(InteractionSpriteOffset)
                    Rotation := rotation{}
                    Scale := vector3{X:=1.0,Y:=1.0,Z:=1.0}*InteractionSpriteSizeMultiplier
                }

        SetLocation<public>(InProp:creative_prop):void=
            if(InProp.IsValid[]):
                set PropLocation = InProp
                set StaticLocation = InProp.GetTransform()

        SetLocation<public>(InLocation:transform):void=
            set StaticLocation = InLocation
            set IsLocationStatic = true

        MeetsInteractionRequirements<public><final>(InFortChar:fort_character)<transacts><decides>:void=
            IsInInteractableDistance[InFortChar]
            IsInInteractableAngle[InFortChar]
            IsPlayerInFront[InFortChar]

        IsInInteractableDistance<public><final>(InFortChar:fort_character)<transacts><decides>:void=
            ThisTransform:=GetTransform()
            InFortChar.IsActive[]
            IgnoreDistanceCheck? or Distance(InFortChar.GetTransform().Translation, ThisTransform.Translation) <= InteractDistance

        IsInInteractableAngle<public><final>(InFortChar:fort_character)<transacts><decides>:void=
            InViewLocation := InFortChar.GetViewLocation()
            InViewRotation := InFortChar.GetViewRotation()
            not RequiresLookAt? or (RaySphereIntersection[InViewLocation, InViewRotation, GetTransform().Translation, InteractionRadius] or IsInInteractableColliders(InViewLocation, InViewRotation)?)

        IsInInteractableColliders<private>(InViewLocation:vector3, InViewRotation:rotation)<transacts>:logic=
            InteractiveLoc := GetTransform()
            for(ThisCollider:ExtraInteractionRadiusColliders):
                if(RaySphereIntersection[InViewLocation, InViewRotation, InteractiveLoc.GetLocalVector(ThisCollider), InteractionRadius]):
                    return true
            return false

        IsPlayerInFront<public><final>(InFortChar:fort_character)<transacts><decides>:void=
            ThisTransform:=GetTransform()
            not RequiresPlayerMustBeInFront? or StaticLocation.GetVectorLookPercent(InFortChar.GetTransform().Translation) > 0.0

        OnInput_Shown<public>(InAgent:agent)<suspends>:void=
            return
        
        OnInput_Hidden<public>(InAgent:agent)<suspends>:void=
            return

        OnInput_Pressed<public>(InAgent:agent)<suspends>:void=
            return

        OnInput_Released<public>(InAgent:agent)<suspends>:void=
            return

    # The behavior to follow when an Interactable is interacted with
    interact_behavior<public> := enum{
        DisableForPlayer
        DisableForAll
        None
    }

    # Interface for interactable components, such as the 'Interactive Device'
    interactive<public> := interface:

        # Compiles all the given interactables for this interactive device, and links it globally
        # Call it ONCE within the inheriting class
        LinkAndCompile<public>(Interactables:[]Interactable):void

        # Fetches all Interactable locations within the inheriting class
        GetLocations<public>()<transacts>:[]Interactable

        # Activates when this Interactive is interacted with
        OnInteracted<public>(InAgent:agent, InteractingLocation:Interactable)<suspends>:void

    var InteractiveDeviceCollection : weak_map(session, [Interactive_Base][]Interactable) = map{}
    # A device that contains Interactive components
    # Allows for multiple interactive points, with different settings
    # Meant to be overridden as this class by itself is Abstract and has NO functionality 
    Interactive_Base<public> := class<unique>(creative_device, interactive):

        InteractiveType<public> : interactable_type = interactable_type.Misc
        
        interactivepriority_tt<localizes>:message="Should this Interactable have priority over others when interaction is overlapping?\n0 = Highest priority"
        @editable{ToolTip:=interactivepriority_tt}
        InteractivePriority : int = 0

        var<private> InteractableCollection<public> : []Interactable = array{}

        var<private> EnabledStatePerPlayer : [agent]logic = map{}

        var<private> IsLinked : logic = false

        DebugStateChanged:logic=false # -> Override in inheriting classes to use

        # Signaled when this Interactive device is interacted with
        #   Agent - The player who activated this interactable
        #   Interactable - The Interactable that was interacted with
        InteractedEvent:event(tuple(agent, Interactable))=event(tuple(agent, Interactable)){}

        # Compiles all the given interactables for this interactive device, and links it globally
        LinkAndCompile<override><final>(Interactables:[]Interactable):void=
            if(not IsLinked?):
                set IsLinked = true
                # Checking each Interactable for validity
                ValidInteractables := for:
                    ThisInteractable:Interactables
                    ThisInteractable.IsValid[]
                do:
                    ThisInteractable.Link() # -> Link Interactable with this Device
                    set ThisInteractable.ParentDevice = option. Self
                    set ThisInteractable.InteractiveType = InteractiveType
                    ThisInteractable
                if(ValidInteractables.Length < Interactables.Length):
                    Print("ERROR: An Interactable failed to link {Interactables.Length-ValidInteractables.Length}/{Interactables.Length} Interactables!")

                set InteractableCollection = ValidInteractables

                if:
                    not InteractiveDeviceCollection[GetSession()]
                    set InteractiveDeviceCollection[GetSession()] = map{}
                if:
                    CurrentInteractiveCollection := InteractiveDeviceCollection[GetSession()]
                    set InteractiveDeviceCollection[GetSession()] = ConcatenateMaps(CurrentInteractiveCollection, map{Self => ValidInteractables})
            else:
                Print("ERROR: Tried to LinkAndCompile an interactive device when it already is!")

        # Fetch all interactive locations
        GetLocations<override><final>()<transacts>:[]Interactable={InteractableCollection}

        IsEnabled<public><final>(?InAgent:?agent=false)<transacts><decides>:[]Interactable=
            var EnabledInteractableLocations : []Interactable = array{}
            CurrLocations := GetLocations()
            for:
                ThisInteractable:CurrLocations
                ThisInteractable.CheckEnabledState[InAgent]
            do:
                set EnabledInteractableLocations += array{ThisInteractable}
            EnabledInteractableLocations.Length > 0
            return EnabledInteractableLocations

        GetEnabledLocations<public><final>(?InAgent:?agent=false):[]Interactable=
            var LocationArray : []Interactable = array{}
            for:
                ThisLocation:InteractableCollection
                ThisLocation.CheckEnabledState[InAgent]
            do:
                set LocationArray += array{ThisLocation}
            return LocationArray

        # Gets the closest location to the player, as well as if they are looking at it #
        GetClosestLocation<public><final>(InAgent:agent):?Interactable=
            if:
                InFortChar := InAgent.GetFortCharacter[]
                InFortChar.IsActive[]
                var ClosestLocation : ?Interactable = false
                var ClosestLocationDistance : float = 0.0
            then:
                Locations:=GetLocations()
                for:
                    ThisLocation:Locations
                    ThisLocation.CheckEnabledState[option{InAgent}]
                    ThisLocationDistance := Distance(InFortChar.GetTransform().Translation, ThisLocation.GetTransform().Translation)
                    ThisLocationDistance < ClosestLocationDistance or not ClosestLocation?
                do:
                    set ClosestLocation = option{ThisLocation}
                    set ClosestLocationDistance = ThisLocationDistance
                return ClosestLocation
            return false

        # Fetches the interaction text #
        # Meant to be overridden to suit the needs of a child class #
        # Returning "NA" results in the interactable using its' default interaction text
        GetInteractionText<public>(InAgent:agent, Interactive:Interactable):string=
            return "NA"

        # TUPLE 0 -> The Icon path string
        # TUPLE 1 -> The description text content
        GetDescription<public>(InAgent:agent, Interactive:Interactable):tuple(string, string)=
            return ("", "")

        # Activated when this interactable is interacted with #
        Interact<final><public>(InAgent:agent, InInteractingLocation:Interactable):void=
            if(InInteractingLocation.CheckEnabledState[option{InAgent}]):
                case(InInteractingLocation.InteractedWithBehavior):
                    interact_behavior.DisableForAll =>
                        InInteractingLocation.SetEnabledState(false)
                    interact_behavior.DisableForPlayer =>
                        if(not EnabledStatePerPlayer[InAgent]? and EnabledStatePerPlayer[InAgent]):
                            Print("~InteractableLog: This interactable is disabled for you (one time use)!")
                            return
                        else if(not EnabledStatePerPlayer[InAgent]):
                            set EnabledStatePerPlayer = ConcatenateMaps(EnabledStatePerPlayer, map{InAgent => false})
                        else if(set EnabledStatePerPlayer[InAgent] = false){}
                    interact_behavior.None =>
                InteractedEvent.Signal(InAgent, InInteractingLocation)
                spawn. OnInteracted(InAgent, InInteractingLocation)

        # A function meant to be overridden to suit the needs of a child class #
        OnInteracted<override>(InAgent:agent, InteractingLocation:Interactable)<suspends>:void=
            Print("~Log: An Interactable has an unset 'OnInteracted()' function! Make sure to override it.")
            return
        
        OnInteractStart<public>(InAgent:agent, InteractingLocation:Interactable)<suspends>:void=
            return

        OnInteractStopped<public>(InAgent:agent, InteractingLocation:Interactable)<suspends>:void=
            return

        # Activates when the player is able to interact with this interactable #
        WhenInsideInteractionRadius<public>(InAgent:agent, InteractingLocation:Interactable)<suspends>:void=
            return

        # Activates when the player has left the interaction radius #
        WhenOutsideInteractionRadius<public>(InAgent:agent, InteractingLocation:Interactable)<suspends>:void=
            return

        # Sets the state of the interactable #
        SetEnabledState<public><final>(InState:logic, ?InAgent:?agent=false):void=
            if(DebugStateChanged?):
                if(InState?){Print("~State changed to true!")}
                else{Print("~State changed to false!")}
            if(not InAgent?):
                PlayerList:=GetPlayers(?OnlyActives:=false)
                for(ThisPlayerInfo:PlayerList):
                    for(ThisInteractionLocation:InteractableCollection):
                        ThisInteractionLocation.SetEnabledState(InState, ?InAgent:=option{ThisPlayerInfo.User})
            else if(InAgent?):
                for(ThisInteractionLocation:InteractableCollection):
                    ThisInteractionLocation.SetEnabledState(InState, ?InAgent:=InAgent)

        # Sets the interactive state of this Interactive device for all players
        #   InAgent - If set, the Interactive state will only apply for the given player
        SetInteractableState<public><final>(InState:logic, ?InAgent:?agent=false):void=
            if(InAgent?):
                for(ThisInteractionLocation:InteractableCollection):
                    ThisInteractionLocation.SetInteractableState(InState, ?InAgent:=InAgent)
            else:
                PlayerList:=GetPlayers(?OnlyActives:=false)
                for(ThisPlayerInfo:PlayerList):
                    for(ThisInteractionLocation:InteractableCollection):
                        ThisInteractionLocation.SetInteractableState(InState, ?InAgent:=option{ThisPlayerInfo.User})

        # Resets all Interactables inside this Interactive device to their default states
        Reset<public>():void=
            for(ThisInteractable:InteractableCollection):
                ThisInteractable.Reset()

        # A function to test if the player meets the interaction requirements
        #   The requirements should be written in inheriting classes within this function when overriding
        MeetsInteractionRequirements<public>(InAgent:agent)<transacts><decides>:void=
            return

    # An example device using the Interactive_Base class to make a simple interactive location
    Basic_Interactable := class(Interactive_Base):

        @editable InteractiveLocation : Interactable = Interactable{}

        OnBegin<override>():void=
            LinkAndCompile(array{InteractiveLocation})