using { /Fortnite.com/Devices }
using { /Fortnite.com/Characters }
using { /Fortnite.com/Game }
using { /Fortnite.com/UI }
using { /Fortnite.com/FortPlayerUtilities }
using { /Verse.org/Simulation }
using { /Verse.org/Random }
using { /Verse.org/Colors }
using { /UnrealEngine.com/Temporary/Diagnostics }
using { /UnrealEngine.com/Temporary/SpatialMath }
using { /UnrealEngine.com/Temporary/UI }
using { /Verse.org/Colors/NamedColors }
using { /Fortnite.com/Animation/PlayAnimation }
using { Spawning }

damage_result_custom<public> := struct:

    Instigator<public> : ?agent
    Target<public> : agent
    Amount<public> : float = 0.0
    DamageType<public> : damage_type

damage_type<public> := enum{
    Body
    Head
    Melee
}

AllowStatePrint:logic=true
PrintState(InMSG:string):void=
    if(AllowStatePrint?):
        Print("{InMSG}")

(InFortChar:fort_character).IsLookingAt(InVector:vector3, PrecisionPercentage:float)<transacts><decides>:float=
    UnitVector := (InVector - InFortChar.GetTransform().Translation).MakeUnitVector[]
    return DotProduct(InFortChar.GetViewRotation().GetLocalForward(), UnitVector) >= PrecisionPercentage

(InFortChar:fort_character).GetLookPercent(InVector:vector3)<transacts>:float=
    var LookPercent : float = 0.0
    if(UnitVector := (InVector - InFortChar.GetTransform().Translation).MakeUnitVector[]):
        set LookPercent = DotProduct(InFortChar.GetViewRotation().GetLocalForward(), UnitVector)
    return LookPercent

(InTransform:transform).GetVectorLookPercent(InVector:vector3)<transacts>:float=
    var LookPercent : float = 0.0
    if(UnitVector := (InVector - InTransform.Translation).MakeUnitVector[]):
        set LookPercent = DotProduct(InTransform.Rotation.GetLocalForward(), UnitVector)
    return LookPercent

# Checks if the sources' view is intersecting with the target sphere
RaySphereIntersection<public>(Source:vector3, Rotation:rotation, TargetTranslation:vector3, TargetRadius:float)<decides><transacts>:float=
    SrcToTarget := Source - TargetTranslation
    DirectionForward := Rotation.GetLocalForward()
    A := DotProduct(DirectionForward, DirectionForward)
    B := 2.0 * DotProduct(DirectionForward, SrcToTarget)
    C := DotProduct(SrcToTarget, SrcToTarget) - TargetRadius * TargetRadius
    Discriminant := B * B - 4.0 * A * C

    Discriminant >= 0.0 # If this fails, the sphere is not intersecting the line
    
    # Calculate the intersection points
    T1 := (-B + Sqrt(Discriminant)) / (2.0 * A)
    T2 := (-B - Sqrt(Discriminant)) / (2.0 * A)

    OptResult : ?float = if ((T1 >= 0.0) and (T2 >= 0.0)):
        option. Min(T1, T2)
    else if (T1 >= 0.0):
        option. T1
    else if (T2 >= 0.0):
        option. T2
    else:
        false
    
    OptResult?


(FortChar:fort_character).CameraRay(?AtDistance:float=1.0):vector3 = FortChar.GetViewRotation().GetLocalForward() * AtDistance

(FortChar:fort_character).GetVelocity<public>(TimeBetween:float)<suspends>:float=
    InitialLocation := FortChar.GetTransform().Translation
    StartTime := GetSimulationElapsedTime()

    Sleep(TimeBetween)

    NewLocation := FortChar.GetTransform().Translation
    EndTime := GetSimulationElapsedTime()

    DistanceMoved := Distance(InitialLocation, NewLocation)
    TimeElapsed := EndTime - StartTime

    Velocity := DistanceMoved / TimeElapsed

    #Print("Player Velocity: {Velocity}", ?Duration:=5.0, ?Color:=LimeGreen)
    return Velocity

GetRandomPointInCircle(CenterPoint : vector3, Radius : float):vector3=
        RandomFloat := GetRandomFloat(0.0, 1.0)
        Theta := RandomFloat * (PiFloat * 2)

        var RandomRadius : float = GetRandomFloat(-Radius, Radius)
        XCoord := RandomRadius * Cos(Theta)
        set RandomRadius = GetRandomFloat(-Radius, Radius)
        YCoord := RandomRadius * Sin(Theta)

        NewPoint := vector3{X:= XCoord, Y:= YCoord} + CenterPoint

(InTransform:transform).GetLocalVector(DirectionAndDistance:vector3)<transacts>:vector3=
    return InTransform.Translation + (DirectionAndDistance.X * InTransform.Rotation.GetLocalForward()) + (DirectionAndDistance.Y * InTransform.Rotation.GetLocalRight()) + (DirectionAndDistance.Z * InTransform.Rotation.GetLocalUp())
    <#
    var LocalVector : vector3 = InTransform.Translation
    # FORWARDS/BACKWARDS #
    #         X          #
    set LocalVector = LocalVector + (InTransform.Rotation.GetLocalForward() * DirectionAndDistance.X)
    # RIGHT/LEFT #
    #      Y     #
    set LocalVector = LocalVector + (InTransform.Rotation.GetLocalRight() * DirectionAndDistance.Y)
    # UP/DOWN #
    #    Z    #
    set LocalVector = LocalVector + (InTransform.Rotation.GetLocalUp() * DirectionAndDistance.Z)
    return LocalVector #>

(InRotation:rotation).ApplyRandomYawRotation():rotation=
    var RandomRotation : rotation = rotation{}
    if(YPR:=InRotation.GetYawPitchRollDegrees(), Pitch:=YPR[1], Roll:=YPR[2]):
        set RandomRotation = MakeRotationFromYawPitchRollDegrees(GetRandomFloat(0.0, 360.0), Pitch, Roll)
    return RandomRotation

(InPosition:vector3).GetRotationToPosition(OtherPosition:vector3)<transacts>:rotation=
    var FacingRotation : rotation = rotation{}
    if(LookDirection := (OtherPosition - InPosition).MakeUnitVector[]):
        Yaw := RadiansToDegrees(ArcTan(LookDirection.Y, LookDirection.X))
        set FacingRotation = MakeRotationFromYawPitchRollDegrees(Yaw, 0.0, 0.0)
    return FacingRotation

(InTransform:transform).GetRotationToFacePosition(Position:vector3)<transacts>:rotation=
    var NewRotation : rotation = InTransform.Rotation
    if (LookDirection := (Position - InTransform.Translation).MakeUnitVector[]):
        Yaw := RadiansToDegrees(ArcTan(LookDirection.Y, LookDirection.X)) - 90.0
        Pitch := 0.0
        Roll := 0.0
        set NewRotation = MakeRotationFromYawPitchRollDegrees(Yaw, Pitch, Roll)
    return NewRotation

(InDistance:float).LocationFromDistance(FirstLocation:vector3, SecondLocation:vector3, ?DistanceAsPercent:logic=false):vector3=
    var ChosenDistance : float = InDistance
    if(DistanceAsPercent?):
        set ChosenDistance = Distance(FirstLocation, SecondLocation)*(InDistance/100.0)
    TTRotation:=FirstLocation.GetRotationToPosition(SecondLocation)
    TravelTransform : transform = transform{
        Translation:=FirstLocation
        Rotation:=TTRotation
    }
    return TravelTransform.GetLocalVector(vector3{X:=InDistance})

(Character: fort_character).PlayerFeetTranslated()<transacts>:vector3 =
    var HeightOffsetToPlayerFeet : float = 77.15
    if(Character.IsCrouching[]):
        set HeightOffsetToPlayerFeet -= 25.0
    
    return Character.GetTransform().Translation - vector3{Z:= HeightOffsetToPlayerFeet}

(Character: fort_character).GetYawRotation()<transacts><decides>:rotation=
    MakeRotationFromYawPitchRollDegrees(Character.GetTransform().Rotation.GetYawPitchRollDegrees()[0], 0.0, 0.0)

(InWidget:widget).Show():void=
    InWidget.SetVisibility(widget_visibility.Visible)
(InWidget:widget).Hide():void=
    InWidget.SetVisibility(widget_visibility.Hidden)
(InWidget:widget).Collapse():void=
    InWidget.SetVisibility(widget_visibility.Collapsed)

# Changes an entire string to be lowercase #
(String:string).ToLower()<transacts>:string =
    LowercaseString := for (CharToConvert : String):
        CharToConvert.ToLower()

# Changes a char value to be the lowercase version #
(InChar:char).ToLower()<transacts>:char =
    UpperToLowerMap: [char]char = map:
        'A' => 'a', 'B' => 'b', 'C' => 'c', 'D' => 'd', 'E' => 'e',
        'F' => 'f', 'G' => 'g', 'H' => 'h', 'I' => 'i', 'J' => 'j',
        'K' => 'k', 'L' => 'l', 'M' => 'm', 'N' => 'n', 'O' => 'o',
        'P' => 'p', 'Q' => 'q', 'R' => 'r', 'S' => 's', 'T' => 't',
        'U' => 'u', 'V' => 'v', 'W' => 'w', 'X' => 'x', 'Y' => 'y',
        'Z' => 'z'
    if:
        UpperToLowerMap[InChar]
        NewChar := UpperToLowerMap[InChar]
    then:
        return NewChar
    else:
        return InChar
    
FormatStringToCenter(StartingSpaces:int, InString:string):string=
    var ReturnStr : string = ""
    for(ThisSpaceCount := 0..(StartingSpaces-InString.Length-1)):
        set ReturnStr += " "
    set ReturnStr += InString
    return ReturnStr

(InLogic:logic).ToString():string=
    if. InLogic? then. "TRUE"
    else. "FALSE"

(InDamageable:damageable).ToAgent()<transacts><decides>:agent=
    fort_character[InDamageable].GetAgent[]

# Removes an array of widgets from a canvas #
(InCanvas:canvas).RemoveWidgets(WidgetArray:[]widget):void=
    for(ThisWidget:WidgetArray):
        InCanvas.RemoveWidget(ThisWidget)

# Removes an array of overlay slots from an overlay #
(InOverlay:overlay).RemoveWidgets(SlotArray:[]overlay_slot):void=
    for(ThisSlot:SlotArray):
        InOverlay.RemoveWidget(ThisSlot.Widget)

(InInt:int).FixToRange(LowInt:int, HighInt:int)<transacts>:int=
    if(InInt < LowInt):
        return LowInt
    else if(InInt > HighInt):
        return HighInt
    else:
        return InInt

(InFloat:float).FixToRange(LowFloat:float, HighFloat:float)<transacts>:float=
    if(InFloat < LowFloat):
        return LowFloat
    else if(InFloat > HighFloat):
        return HighFloat
    else:
        return InFloat

# Lerp that doesn't allow lower than 0 or greater than 100 percent
LockLerp(V1:float, V2:float, P:float)<transacts>:float=
    Lerp(V1,V2,Clamp(P, 0.0,1.0))

(FloatArray:[]float).GetMax()<transacts>:float=
    return block{
        var MaxValue : float = FloatArray[0]
        for(ThisValue:FloatArray):
            if(ThisValue > MaxValue):
                set MaxValue = ThisValue
        MaxValue
    } or -1.0

(FloatArray:[]float).GetMin()<transacts>:float=
    return block{
        var MaxValue : float = FloatArray[0]
        for(ThisValue:FloatArray):
            if(ThisValue < MaxValue):
                set MaxValue = ThisValue
        MaxValue
    } or -1.0

(FloatArray:[]float).SortByMax()<transacts>:[]float=
    var UnsortedElements : []float = FloatArray
    var SortedArray : []float = array{}
    for(ThisElement:FloatArray):
        ThisMin := UnsortedElements.GetMax()
        if. NewUnsortedArray := UnsortedElements.RemoveFirstElement[ThisMin] then. set UnsortedElements = NewUnsortedArray else. return SortedArray
        set SortedArray += array{ThisMin}
    return SortedArray

(FloatArray:[]float).SortByMin()<transacts>:[]float=
    var UnsortedElements : []float = FloatArray
    var SortedArray : []float = array{}
    for(ThisElement:FloatArray):
        ThisMin := UnsortedElements.GetMin()
        if. NewUnsortedArray := UnsortedElements.RemoveFirstElement[ThisMin] then. set UnsortedElements = NewUnsortedArray else. return SortedArray
        set SortedArray += array{ThisMin}
    return SortedArray


(Props:[]creative_prop).TeleportTo(InTranslation:vector3, InRotation:rotation):void=
    for(ThisProp:Props):
        if. ThisProp.TeleportTo[InTranslation, InRotation]

#################
# BEZIER CURVES #

QuadBezier(T : float, P0 : float, P1 : float, P2 : float)<transacts>:float=
    return (1.0 - T) * (1.0 - T) * P0 + 2 * (1.0 - T) * T * P1 + T * T * P2
    
QuadBezier_3D(Quality : int, Start : vector3, End : vector3, Mid : vector3):[]vector3=
    Dist := Distance(Start, End)
    
    X := array{Start.X, Mid.X, End.X}
    Y := array{Start.Y, Mid.Y, End.Y}
    Z := array{Start.Z, Mid.Z, End.Z}
    
    var tX:float = 0.0
    var tY:float = 0.0
    var tZ:float = 0.0
    
    var Poss : []vector3 = array{}
    
    for (i := 1..Quality):
        t := i * (1.0 / (Quality * 1.0))
    
        if (new_tX := QuadBezier(t, X[0], X[1], X[2])) { set tX = new_tX}
        if (new_tY := QuadBezier(t, Y[0], Y[1], Y[2])) { set tY = new_tY}
        if (new_tZ := QuadBezier(t, Z[0], Z[1], Z[2])) { set tZ = new_tZ}
    
        Pos := vector3{X:=tX, Y:=tY, Z:=tZ}
        set Poss += array{Pos}
    Poss
    
# Calculates the delta movement from the given vector to the desired location
(InVector:vector3).GetDeltaToPosition(ToVector:vector3):vector3=
    vector3{
        X := ToVector.X - InVector.X
        Y := ToVector.Y - InVector.Y
        Z := ToVector.Z - InVector.Z
    }

#################

# Returns the 'Tab' or index in elements that passed the random chance keeping weights in mind #
# Weights are the threshold for how 'common' or 'rare' an element is #
# If a weight is 1.5, it has a 1.5% chance of being selected #
# ALL weights should be split amongst each other, adding up to 100% #
<#
RandomWeightedChance(Weights : []float, ?ForcedChance:?float=false)<transacts>:int=
    var ChanceNumber : float = GetRandomFloat(0.0, Weights.GetMax())
    Print("~Random chance num: {ChanceNumber}")
    if(ThisForcedChance:=ForcedChance?):
        set ChanceNumber = ThisForcedChance
    var TabSelected : int = -1
    var SelectedTabWeight : float = 0.0
    var CurrTab : int = 0
    for(ThisWeight:Weights):
        if(ThisWeight <= ChanceNumber, ThisWeight > SelectedTabWeight, ThisWeight > 0.0):
            set TabSelected = CurrTab
            set SelectedTabWeight = ThisWeight
        else if(ThisWeight = Weights.GetMax()):
            set TabSelected = CurrTab
            set SelectedTabWeight = ThisWeight
        set CurrTab += 1
    if(TabSelected = -1):
        set CurrTab = 0
        set SelectedTabWeight = -1.0
        for(ThisWeight:Weights):
            if(ThisWeight < SelectedTabWeight, ThisWeight > 0.0):
                set SelectedTabWeight = ThisWeight
                set TabSelected = CurrTab
            set CurrTab += 1
    return TabSelected #>

RandomWeightedChance(Weights : []float, ?ForcedChance:?float=false)<transacts>:int=
    var TotalWeight : float = 0.0
    for (Weight : Weights){set TotalWeight += Weight}

    RandomValue : float = GetRandomFloat(0.0, TotalWeight)

    var CurrentWeight : float = 0.0
    for (Index->Weight : Weights):
        set CurrentWeight += Weight
        if (RandomValue < CurrentWeight):
            return Index
    0

#################

# CHARACTER SPACING #

CharacterSizeMapping:[char]int = map:
    ' ' => 8, '!' => 11, '"' => 13, '#' => 20, '$' => 20, '%' => 27, '&' => 24, ''' => 6, '(' => 11, ')' => 11, '*' => 13, '+' => 16, ',' => 9, '-' => 12,
    '.' => 9, '/' => 15, '0' => 21, '1' => 14, '2' => 19, '3' => 20, '4' => 20, '5' => 20, '6' => 21, '7' => 18, '8' => 21, '9' => 20, ':' => 9, ';' => 10,
    '<' => 16, '=' => 15, '>' => 16, '?' => 21, '@' => 34, 'A' => 22, 'B' => 20, 'C' => 22, 'D' => 21, 'E' => 18, 'F' => 17, 'G' => 22, 'H' => 20, 'I' => 11,
    'J' => 19, 'K' => 21, 'L' => 17, 'M' => 30, 'N' => 21, 'O' => 24, 'P' => 19, 'Q' => 23, 'R' => 20, 'S' => 20, 'T' => 19, 'U' => 21, 'V' => 21, 'W' => 29,
    'X' => 21, 'Y' => 22, 'Z' => 18, '[' => 12, '\' => 15, ']' => 12, '^' => 17, '_' => 16, '`' => 18, 'a' => 18, 'b' => 19, 'c' => 18, 'd' => 19, 'e' => 18,
    'f' => 14, 'g' => 18, 'h' => 18, 'i' => 9, 'j' => 9, 'k' => 18, 'l' => 9, 'm' => 28, 'n' => 18, 'o' => 19, 'p' => 20, 'q' => 19, 'r' => 15, 's' => 17,
    't' => 14, 'u' => 18, 'v' => 18, 'w' => 26, 'x' => 19, 'y' => 18, 'z' => 15, '{' => 15, '|' => 10, '}' => 15, '~' => 17

(Character:char).GetSize()<transacts>:int =
    CharacterSizeMapping[Character] or 0

GetStringSize(String:string)<transacts>:float =
    var TotalStringSize:float = 0.0
    for (CharacterIndex -> Character : String):
        CharacterSize := Character.GetSize()
        set TotalStringSize += CharacterSize * 1.0
    TotalStringSize

(Text:string).WrapText(AvailableWidth:float):string =
    var WrappedText:string = ""
    var CurrentTextSize:int = 0
    var TextSizeFromLastSpaceCharacter:int = 0
    var LastSpaceCharacterIndex:int = 0

    for (CharacterIndex -> Character : Text):
        CharacterSize := Character.GetSize()
        set CurrentTextSize += CharacterSize
        set TextSizeFromLastSpaceCharacter += CharacterSize

        if (CurrentTextSize * 1.0 >= AvailableWidth):
            if (set WrappedText[LastSpaceCharacterIndex] = '\n') {}
            set CurrentTextSize = TextSizeFromLastSpaceCharacter

        if (Character = ' '):
            set LastSpaceCharacterIndex = CharacterIndex
            set TextSizeFromLastSpaceCharacter = 0

        set WrappedText += ToString(Character)

    WrappedText

FormatSecondsToTime(Time: float, ?AutoHideHour:logic=true):string = {
    SecondsRemaining := Floor[Time] or 0
    Hours: int = Floor(SecondsRemaining / 3600) or 0
    HourString: string = if(Hours > 10) then "{Hours}:" else if(Hours > 0 or not AutoHideHour?) then "0{Hours}:" else ""
    Minutes: int = Floor(Mod[SecondsRemaining, 3600] / 60) or 0
    MinuteString: string = if(Minutes < 10) then "0{Minutes}" else "{Minutes}"
    Seconds: int = Mod[SecondsRemaining, 60] or 0
    SecondsString: string = if(Seconds < 10) then "0{Seconds}" else "{Seconds}"
    
    return "{HourString}{MinuteString}:{SecondsString}"
}

# Used to linearly interpolate/extrapolate between `FromColor` (when `Parameter = 0.0`) and `ToColor` (when `Parameter = 1.0`). Expects that all arguments are finite.
CLerp(FromColor:color, ToColor:color, Percent:float):color=
    SRGB1:=MakeSRGBFromColor(FromColor)
    SRGB2:=MakeSRGBFromColor(ToColor)
    MakeColorFromSRGB(Lerp(SRGB1(0), SRGB2(0), Percent), Lerp(SRGB1(1), SRGB2(1), Percent), Lerp(SRGB1(2), SRGB2(2), Percent)) 

<#
MakeHexFromColor<public>(InColor:color):string=
    SRGBValues := MakeSRGBFromColor(InColor)
    HexVal := block{
        block{
            SRGBValues(0)/16.0
        }+block{

        }+block{

        }
    }
#>

(FortCharacter:fort_character).Kill():void=
    FortCharacter.SetVulnerability(true)
    FortCharacter.SetMaxHealth(1.0)
    FortCharacter.Damage(999999.0)

(Vector2:vector2).Multiply(OtherVector2:vector2):vector2=
    return vector2{
        X := Vector2.X*OtherVector2.X
        Y := Vector2.Y*OtherVector2.Y
    }

AbsVector(InVector:vector2):vector2=
    return vector2{
        X:=Abs(InVector.X)
        Y:=Abs(InVector.Y)
    }

(Array:[]type).GetEnd<public>()<transacts><decides>:type=
    return Array[Array.Length-1]

(AnimState:play_animation_state).ToString<public>()<transacts>:string=
    case(AnimState):
        # The animation is blending in.
        play_animation_state.BlendingIn => "BlendingIn"
        # The animation has blended in, is playing, and has not begun blending out.
        play_animation_state.Playing => "Playing"
        # The animation is playing and is blending out.
        play_animation_state.BlendingOut => "BlendingOut"
        # The animation completed successfully.
        play_animation_state.Completed => "Completed"
        # The animation was stopped internally.
        play_animation_state.Stopped => "Stopped"
        # The animation was interrupted externally.
        play_animation_state.Interrupted => "Interrupted"
        # An error occurred at creation or during playback.
        play_animation_state.Error => "Error"

(AnimResult:play_animation_result).ToString<public>()<transacts>:string=
    case(AnimResult):
        # The animation completed successfully.
        play_animation_result.Completed => "Completed"
        # The animation was interrupted whilst playing.
        play_animation_result.Interrupted => "Interrupted"
        # The animation encountered an error during initialization or whilst playing.
        play_animation_result.Error => "Error"
# Await a list of events
AwaitAll<public>(Awaitables:[]event(t) where t:subtype(comparable))<suspends>:[]t=
    if(FirstEvent := Awaitables[0], NewAwaitablesList := Awaitables.RemoveElement[0]):
        SyncResult:=sync:
            FirstEvent.Await()
            AwaitAll(NewAwaitablesList)
        array{SyncResult(0)}+SyncResult(1)
    else:
        array{}

RaceAll<public>(Awaitables:[]event(t) where t:subtype(comparable))<suspends>:?t=
    if(FirstEvent := Awaitables[0]):
        if(NewAwaitablesList := Awaitables.RemoveElement[0], NewAwaitablesList[0]):
            RaceResult:=race:
                block:
                    AwaitResult:=FirstEvent.Await()
                    option. AwaitResult
                RaceAll(NewAwaitablesList)
            RaceResult
        else:
            EventResult:=FirstEvent.Await()
            option. EventResult
    else:
        Sleep(Inf)
        false