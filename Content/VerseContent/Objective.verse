using { /Fortnite.com/Devices }
using { /Fortnite.com/Characters }
using { /Fortnite.com/Game }
using { /Fortnite.com/UI }
using { /Fortnite.com/FortPlayerUtilities }
using { /Verse.org/Simulation }
using { /Verse.org/Random }
using { /Verse.org/Colors }
using { /Verse.org/Simulation/Tags }
using { /UnrealEngine.com/Temporary/Diagnostics }
using { /UnrealEngine.com/Temporary/SpatialMath }
using { /UnrealEngine.com/Temporary/UI }
using { /Fortnite.com/Devices/CreativeAnimation }
using { /Fortnite.com/Devices/CreativeAnimation/InterpolationTypes }
using { /Fortnite.com/AI }
using { Zombie }
using { Spawning }
using { Player }
using { InteractiveModule }

var CurrentActiveObjective<public> : weak_map(session, ?objective_base) = map{}

objective_state<public> := enum{
    Inactive
    Active
    Cooldown
}
objective_base<public> := class(Powerable):

    @editable
    InteractiveLocation : Interactable = Interactable{}

    @editable BasicSettings : objective_basic_settings = objective_basic_settings{}

    var<private> Location : vector3 = vector3{}
    var<private> AttackingZombies : [agent]zombie_ai = map{}
    GetActiveEnemies()<transacts>:[]zombie_ai=
        var ActiveZombieList : []zombie_ai = array{}
        for(ThisZombie:AttackingZombies):
            if(ThisZombie.IsActive[]):
                set ActiveZombieList += array{ThisZombie}
        return ActiveZombieList

    OnPowerableBegin<override>()<suspends>:void=
        if(BasicSettings.HideProp?):
            BasicSettings.Prop.Hide()
        if(BasicSettings.Prop.IsValid[]):
            set Location = BasicSettings.Prop.GetTransform().Translation
            BasicSettings.OnTrigger_StartObjective.TriggeredEvent.Subscribe(StartViaTrigger)
        else:
            Print("~ERR: Objective [{BasicSettings.Name}] has an empty or invalid prop location!")
        LinkAndCompile(array. InteractiveLocation)
        OnObjectiveBegin()

    OnObjectiveBegin<public>()<suspends>:void=
        return

    StartViaTrigger<private>(Agent:?agent):void=
        spawn{Start()}

    var<private> CompletedState : logic = false
    GetPowerInteractionText<override>(InAgent:agent, Interactive:Interactable):string=
        if(TimerCooldownState?):
            SetInteractableState(false)
            SecondsLeft := if. IntegerTime:=Ceil[30.0-(GetSimulationElapsedTime()-TimeSinceResetStarted)] then. IntegerTime else. -1
            return "On cooldown - {SecondsLeft} seconds left"
        else if(WaveCooldownState?):
            SetInteractableState(false)
            return InteractiveLocation.DefaultPrompt
        else if(not CompletedState? and not CurrentActiveObjective[GetSession()]?):
            SetInteractableState(true)
            return InteractiveLocation.DefaultPrompt
        else:
            SetInteractableState(false)
            return ""

    OnInteracted<override>(InAgent:agent, InInteractingLocation:Interactable):void=
        spawn. Start()

    var<private> IsActive : logic = false
    Start<public><final>()<suspends>:void=
        if(not IsActive?):
            defer{
                set IsActive = false
                if. set CurrentActiveObjective[GetSession()] = false
            }
            set IsActive = true
            if. set CurrentActiveObjective[GetSession()] = option{Self}
            SetEnabledState(false)
            set AttackingZombies = map{}
            BasicSettings.Prop.Show()
            AllObjectiveSpawners:=GetAllSpawners(zombie_type_list.Objective)
            for(ThisSpawner:AllObjectiveSpawners):
                ThisSpawner.Disable()
            for(ThisZombieSpawner:BasicSettings.SpawnControl.ForcedSpawnLocations):
                spawn{TrackZombieSpawner(ThisZombieSpawner)}
            branch. OnStarted()
            BasicSettings.SpawnControl.SetSpawnRate(BasicSettings.SpawnRate)
            race:
                StopEvent.Await()
                block:
                    SafeSleep(BasicSettings.InitialSpawnDelay)
                    loop:
                        EnemyListLength:=GetActiveEnemies().Length
                        BasicSettings.SpawnControl.SetSpawnAmount(BasicSettings.MaximumTargetingZombies-EnemyListLength)
                        BasicSettings.SpawnControl.Spawn()
                        BasicSettings.SpawnControl.AwaitGroupSpawning()
                        SafeSleep(0.25)
            

    OnStarted<public>()<suspends>:void=
        SafeSleep(BasicSettings.Time)
        Stop()

    TrackZombieSpawner<private>(InZombieSpawner:spawn_location)<suspends>:void=
        SpawnedZombieList:=InZombieSpawner.GetSpawnerZombies()
        for(ThisZombie:SpawnedZombieList):
            set AttackingZombies = ConcatenateMaps(AttackingZombies, map{ThisZombie.Agent => ThisZombie})
            spawn. OnEnemySpawned(ThisZombie)
        InZombieSpawner.Enable()
        race:
            StopEvent.Await()
            loop:
                ThisZombieAgent:=InZombieSpawner.ZombieSpawnedEvent.Await()
                if(ThisZombieInfo:=ThisZombieAgent.GetZombieAI[]):
                    set AttackingZombies = ConcatenateMaps(AttackingZombies, map{ThisZombieAgent => ThisZombieInfo})
                    spawn. OnEnemySpawned(ThisZombieInfo)
        InZombieSpawner.Disable()

    # A function to be OVERRIDDEN #
    # Activated when an enemy that is set on this objective spawns #
    OnEnemySpawned<public>(InZombieInfo:zombie_ai)<suspends>:void=
        return

    StopEvent:event(logic)=event(logic){}
    Stop<public><final>(?EndState:logic=true):void=
        if(IsActive?):
            set IsActive = false
            set AttackingZombies = map{}
            if. set CurrentActiveObjective[GetSession()] = false
            StopEvent.Signal(EndState)
            if(BasicSettings.HideProp?):
                BasicSettings.Prop.Hide()
            if(EndState?):
                BasicSettings.ActivateTrigger_OnObjectiveWIN.Trigger()
            else:
                BasicSettings.ActivateTrigger_OnObjectiveLOSE.Trigger()
            BasicSettings.SpawnControl.Cancel()
            if(BasicSettings.EliminateZombiesAfterCompletion?):
                BasicSettings.SpawnControl.Eliminate()
            spawn. OnStopped()
            SetEnabledState(true)
            if(EndState?):
                set CompletedState = true
                case(BasicSettings.CompletionBehavior):
                    objective_completion_behavior.Reset => ResetObjective()
                    objective_completion_behavior.Disable => SetEnabledState(false)
                    objective_completion_behavior.ResetAfterRoundsPassed => spawn{AwaitReset(EndState)}
                    objective_completion_behavior.ResetAfterTimePassed =>spawn{AwaitReset(EndState)}
            else:
                case(BasicSettings.FailureBehavior):
                    objective_completion_behavior.Reset => ResetObjective()
                    objective_completion_behavior.Disable => SetEnabledState(false)
                    objective_completion_behavior.ResetAfterRoundsPassed => spawn{AwaitReset(EndState)}
                    objective_completion_behavior.ResetAfterTimePassed =>spawn{AwaitReset(EndState)}

    var TimeSinceResetStarted : float = 0.0
    var WaveCooldownState : logic = false
    var TimerCooldownState : logic = false
    AwaitReset(EndState:logic)<suspends>:void=
        set TimeSinceResetStarted = GetSimulationElapsedTime()
        ResetBehavior := if. EndState? then. BasicSettings.CompletionBehavior else. BasicSettings.FailureBehavior
        ResetAmount := if. EndState? then. BasicSettings.CompletionResetAmount else. BasicSettings.FailureResetAmount
        if(ResetBehavior = objective_completion_behavior.ResetAfterRoundsPassed):
            set WaveCooldownState = true
            if(ThisWaveDevice:=GlobalWaveController[GetSession()]):
                for(ThisRound:=0..ResetAmount-1):
                    if(ThisLocation:=GetLocations()[0]):
                        if(ResetAmount = 1):
                            ThisLocation.SetText("On Cooldown - 1 wave left")
                        else:
                            ThisLocation.SetText("On Cooldown - {ThisRound+1} wave(s) left")
                    ThisWaveDevice.WaveStartedEvent.Await()
        else:
            set TimerCooldownState = true
            SafeSleep(ResetAmount*1.0)
        if(ThisLocation:=GetLocations()[0]):
            ThisLocation.ResetText()
        ResetObjective()

    ResetObjective():void=
        Print("Resetting objective")
        set CompletedState = false
        set TimerCooldownState = false
        set WaveCooldownState = false
        PowerSettings.Reset()
        SetEnabledState(true)
        
    OnStopped<public>()<suspends>:void=
        return

    ApplyDifficulty<public>():void=
        case(GetGameDifficulty()):
            game_difficulty.Easy =>
            game_difficulty.Classic =>
                set BasicSettings.MaximumTargetingZombies = 10
                set BasicSettings.SpawnRate = 1.8
                set BasicSettings.FailureBehavior = objective_completion_behavior.ResetAfterTimePassed
                set BasicSettings.FailureResetAmount = 30
            game_difficulty.Hardcore =>
                set BasicSettings.MaximumTargetingZombies = 12
                set BasicSettings.SpawnRate = 1.5
                set BasicSettings.FailureBehavior = objective_completion_behavior.ResetAfterRoundsPassed
                set BasicSettings.FailureResetAmount = 1

objective_completion_behavior<public>:=enum{Reset, ResetAfterRoundsPassed, ResetAfterTimePassed, Disable}
objective_basic_settings := class<concrete>:

    @editable Name : string = "Unset Objective Name"
    @editable Prop : creative_prop = creative_prop{}
    @editable HideProp : logic = false
    @editable Time : type{_X : float where _X >= 0.0} = 30.0
    @editable SpawnControl : spawn_group = spawn_group{}
    @editable CountAsRoundEnemies : logic = false
    @editable EliminateZombiesAfterCompletion : logic = false
    @editable var MaximumTargetingZombies : type{_X : int where _X > 0} = 10
    @editable var SpawnRate : type{_X : float where 0.0 < _X} = 2.0
    @editable InitialSpawnDelay : type{_X : float where 0.0 <= _X} = 2.5
    @editable OnTrigger_StartObjective : trigger_device = trigger_device{}
    @editable ActivateTrigger_OnObjectiveWIN : trigger_device = trigger_device{}
    @editable ActivateTrigger_OnObjectiveLOSE : trigger_device = trigger_device{}
    @editable CompletionBehavior : objective_completion_behavior = objective_completion_behavior.Disable
    @editable CompletionResetAmount : int = 0
    @editable var FailureBehavior : objective_completion_behavior = objective_completion_behavior.Reset
    @editable var FailureResetAmount : int = 0
    @editable DebugHits : logic = false
    
objective_defend := class(objective_base):

    @editable TimerProp : creative_prop = creative_prop{}
    @editable DefendSettings : objective_protect_settings = objective_protect_settings{}

    TimerMaterialControl : MI_ProgressRadial_material = MI_ProgressRadial_material{}

    var<internal> CurrentHealth : float = 100.0
    GetHealth()<transacts>:float=CurrentHealth
    GetHealthPercent()<transacts>:float=(CurrentHealth/DefendSettings.MaxHealth)

    OnObjectiveBegin<override>()<suspends>:void=
        DefendSettings.SafeState_Beacon.Disable()
        DefendSettings.UnsafeState_Beacon.Disable()
        TimerProp.Hide()
        TimerProp.SetMaterial(TimerMaterialControl)

    OnEnemySpawned<override>(InZombieInfo:zombie_ai)<suspends>:void=
        Print("Objective enemy spawned")
        InZombieInfo.SetMarker(zombie_marker.Objective)
        InZombieInfo.ForceSetState(zombie_state.Targeting)
        spawn. InZombieInfo.TargetEntity(MakeNavigationTarget(Location), DefendSettings.ReachRadius, ?ForcedLocation:=option{Location})
        race:
            StopEvent.Await()
            InZombieInfo.DeathEvent.Await()
            loop:
                InZombieInfo.AttackedEvent.Await()
                Damage(30.0)

    OnStarted<override>()<suspends>:void=
        set CurrentHealth = DefendSettings.MaxHealth
        DefendSettings.SafeState_Beacon.Enable()
        set TimerMaterialControl.Progress = 0.0
        TimerProp.Show()
        branch. TrackSafetyState()
        branch:
            if(DefendSettings.DoRegenerate?):
                loop:
                    race:
                        DamagedEvent.Await()
                        block:
                            SafeSleep(DefendSettings.RegenDelay)
                            loop:
                                Heal(DefendSettings.RegenAmount)
                                SafeSleep(DefendSettings.RegenRate)
        branch:
            if(BasicSettings.Time > 0.0):
                TimeStarted:=GetSimulationElapsedTime()
                loop:
                    CurrProg := Lerp(0.0, 1.0, (GetSimulationElapsedTime()-TimeStarted)/BasicSettings.Time)
                    set TimerMaterialControl.Progress = CurrProg
                    SafeSleep(1.0)
        if(BasicSettings.Time > 0.0):
            SafeSleep(BasicSettings.Time)
        else:
            Sleep(Inf)
        Stop()

    OnStopped<override>():void=
        TimerProp.Hide()
        DefendSettings.SafeState_Beacon.Disable()
        DefendSettings.UnsafeState_Beacon.Disable()

    DamagedEvent:event()=event(){}
    Damage<public><final>(InAmount:float):void=
        Print("~Objective damaged for {InAmount}")
        set CurrentHealth -= InAmount
        if(CurrentHealth <= 0.0):
            Stop(?EndState:=false)
        DamagedEvent.Signal()
        DefendSettings.DamagedSound_AudioDevice.Play()

    TrackSafetyState<private>()<suspends>:void=
        defer{
            DefendSettings.SafeState_Beacon.Disable()
            DefendSettings.UnsafeState_Beacon.Disable()
        }
        loop:
            race:
                block:
                    DamagedEvent.Await()
                    DefendSettings.SafeState_Beacon.Disable()
                    DefendSettings.UnsafeState_Beacon.Enable()
                block:
                    SafeSleep(2.0)
                    DefendSettings.SafeState_Beacon.Enable()
                    DefendSettings.UnsafeState_Beacon.Disable()

    Heal<public><final>(InAmount:float):void=
        set CurrentHealth += InAmount
        if(CurrentHealth > DefendSettings.MaxHealth):
            set CurrentHealth = DefendSettings.MaxHealth

objective_protect_settings := class<concrete>:

    @editable ReachRadius : float = 150.0
    @editable MaxHealth : type{_X : float where _X > 0.0} = 500.0
    @editable DoRegenerate : logic = true
    @editable RegenDelay : float = 3.0
    @editable RegenRate : float = 1.0
    @editable RegenAmount : float = 25.0
    @editable DamagedSound_AudioDevice : audio_player_device = audio_player_device{}
    @editable SafeState_Beacon : beacon_device = beacon_device{}
    @editable UnsafeState_Beacon : beacon_device = beacon_device{}

    <#
objective_follow := class(objective_defend):

    @editable FollowSettings : objective_follow_settings = objective_follow_settings{}
    @editable Paths : []objective_follow_path = array{}

    OnObjectiveBegin<override>():void=
        DefendSettings.SafeState_Beacon.Disable()
        DefendSettings.UnsafeState_Beacon.Disable()
        for(ThisPath:Paths):
            ThisPath.InitPoints()
        if(FollowSettings.UseNPC?):
            FollowSettings.NPCSpawner.SpawnedEvent.Subscribe(NPCSpawned)

    OnStarted<override>()<suspends>:void=
        set CurrentHealth = DefendSettings.MaxHealth
        if(FollowSettings.UseNPC?):
            FollowSettings.NPCSpawner.Spawn()
            loop:
                if(NPCAgent?.GetFortCharacter[].IsActive[]):
                    break
                Sleep(0.1)
        race:
            block:
                if(BasicSettings.Time > 0.0):
                    Sleep(BasicSettings.Time)
                else{Sleep(Inf)}
            block:
                if(DefendSettings.DoRegenerate?):
                    loop:
                        race:
                            DamagedEvent.Await()
                            block:
                                Sleep(DefendSettings.RegenDelay)
                                loop:
                                    Heal(DefendSettings.RegenAmount)
                                    Sleep(DefendSettings.RegenRate)
                else{Sleep(Inf)}
            StopEvent.Await()
            WatchPlayers()
            MoveAlongPath()
        Stop()

    OnEnemySpawned<override>(InZombieInfo:zombie_ai)<suspends>:void=
        if(FollowSettings.DoZombiesTargetObjective?):
            spawn{InZombieInfo.TargetEntity(MakeNavigationTarget(), DefendSettings.ReachRadius)}

    var NPCAgent : ?agent = false
    NPCSpawned(ThisNPCAgent:agent):void=
        if(CurrentNPCAgent:=NPCAgent?, ThisFortChar:=CurrentNPCAgent.GetFortCharacter[]):
            Print("~Eliminating previous Objective NPC")
            ThisFortChar.Damage(999999.0)
        set NPCAgent = option{ThisNPCAgent}
        if(not FollowSettings.ShowNPC?, ThisFortChar:=ThisNPCAgent.GetFortCharacter[]):
            ThisFortChar.Hide()
        if(ThisNPCFortChar:=ThisNPCAgent.GetFortCharacter[]):
            spawn{AwaitOnNPCSpawned(ThisNPCFortChar)}
        else{Print("~ERR: Failed to fetch spawned NPC fort_character for Follow Objective [{BasicSettings.Name}]")}
    
    AwaitOnNPCSpawned(InNPCFortChar:fort_character)<suspends>:void=
        loop:
            if(InNPCFortChar.IsActive[]):
                break
            Sleep(0.1)
        if(InNPCFortChar.TeleportTo[BasicSettings.Prop.GetTransform().Translation-FollowSettings.OffsetFromPropToNPC, rotation{}]):
            Print("~Success on teleport AI to objective prop")
        else{Print("~ERR: Failed to teleport the AI of a follow objective to the starting point [{BasicSettings.Name}]!")}

    ChangeMoveState:event(logic)=event(logic){}
    MoveAlongPath<public>()<suspends>:void=
        if(BasicSettings.Prop.IsValid[]):
            if(ThisRandomPath := Shuffle(Paths)[0], ThisRandomPath.Points.Length > 1):
                var CurrentPathPointIndex : int = 0
                var TraveledPoints : []objective_follow_point = array{}
                var CurrentPathPoint : objective_follow_point = objective_follow_point{}
                var ReachedPointState : logic = true
                var IsWaiting : logic = false
                var IsWaitingEndResult : logic = false
                var CurrentLoop : int = 0
                loop:
                    Winner:=race:
                        block:
                            loop:
                                ThisMoveState:=ChangeMoveState.Await()
                                if(not ThisMoveState?):
                                    if(IsWaiting?):
                                        set IsWaitingEndResult = false
                                    else:
                                        break
                                else if(IsWaiting?):
                                    set IsWaitingEndResult = true
                            -1
                        block:
                            loop:
                                var DestinationPoint : objective_follow_point = objective_follow_point{}
                                if(ReachedPointState?):
                                    if(ThisRandomPath.ChoosePointsRandomly?):
                                        if(ThisRandomPoint:=ThisRandomPath.GetRandomPoint[TraveledPoints]):
                                            set DestinationPoint = ThisRandomPoint
                                        else if(CurrentLoop >= ThisRandomPath.LoopTimes, ThisRandomPath.LoopTimes > -1):
                                            # Traveled all points randomly #
                                            Print("~Traveled all points randomly")
                                            break
                                        else:
                                            set CurrentLoop += 1
                                            set CurrentPathPointIndex = 0
                                            set TraveledPoints = array{CurrentPathPoint}
                                    else:
                                        if(ThisPathPoint:=ThisRandomPath.Points[CurrentPathPointIndex]):
                                            set DestinationPoint = ThisPathPoint
                                        else if(CurrentLoop >= ThisRandomPath.LoopTimes, ThisRandomPath.LoopTimes > -1):
                                            # Reached end of entire path in order #
                                            Print("~Traveled all points in order")
                                            break
                                        else:
                                            set CurrentLoop += 1
                                            set CurrentPathPointIndex = 0
                                            set TraveledPoints = array{CurrentPathPoint}
                                else:
                                    set DestinationPoint = CurrentPathPoint
                                if(DestinationPoint.Point.IsValid[]):
                                    set ReachedPointState = false
                                    set CurrentPathPoint = DestinationPoint
                                    PointTransform := DestinationPoint.Point.GetTransform()
                                    if(not FollowSettings.UseNPC?):
                                        ThisPropTransform := BasicSettings.Prop.GetTransform()
                                        SpeedTime := Distance(ThisPropTransform.Translation, PointTransform.Translation) / (60.0*DestinationPoint.Speed)
                                        RotationToPoint := ThisPropTransform.GetRotationToFacePosition(PointTransform.Translation)
                                        if(BasicSettings.Prop.TeleportTo[ThisPropTransform.Translation, RotationToPoint]){}
                                        BasicSettings.Prop.MoveTo(PointTransform.Translation+FollowSettings.OffsetFromPropToPoints, BasicSettings.Prop.GetTransform().Rotation, SpeedTime)
                                    else if(ThisNPCFortChar:=NPCAgent?.GetFortCharacter[], Navigatable:=ThisNPCFortChar.GetNavigatable[]):
                                        Print("~Going to new point")
                                        Navigatable.SetMovementSpeedMultiplier(DestinationPoint.Speed)
                                        race:
                                            loop:
                                                FirstLocation := ThisNPCFortChar.GetTransform().Translation
                                                Sleep(0.1)
                                                SecondLocation:=ThisNPCFortChar.GetTransform().Translation
                                                var ThisRotation : rotation = rotation{}
                                                if(FollowSettings.RotatePropWithNPC?, not IsAlmostEqual[FirstLocation,SecondLocation, 0.1]):
                                                    set ThisRotation = FirstLocation.GetRotationToPosition(SecondLocation)
                                                spawn{BasicSettings.Prop.MoveTo(SecondLocation+FollowSettings.OffsetFromPropToNPC, ThisRotation, 0.1)}
                                            loop:
                                                NavResult:=Navigatable.NavigateTo(MakeNavigationTarget(PointTransform.Translation), ?AllowPartialPath:=true)
                                                if(NavResult=navigation_result.Reached):
                                                    Print("~Point reached")
                                                    break
                                                else if(NavResult=navigation_result.Unreachable):
                                                    Print("~ERR: A point within an objective path is unreachable by the AI! Check objective [{BasicSettings.Name}]")
                                        if(BasicSettings.Prop.TeleportTo[ThisNPCFortChar.GetTransform().Translation+FollowSettings.OffsetFromPropToNPC, rotation{}]){}
                                    else:
                                        Print("~ERR: Cannot fetch the fort_character of an AI for the Follow Objective [{BasicSettings.Name}]!")
                                        Stop()
                                    set CurrentPathPointIndex += 1
                                    set TraveledPoints += array{DestinationPoint}
                                    set ReachedPointState = true
                                    if(DestinationPoint.MatchPointRotation?):
                                        ThisRotation:=DestinationPoint.Point.GetTransform().Rotation
                                        if(ThisNPCFortChar:=NPCAgent?.GetFortCharacter[], ThisNPCFortChar.TeleportTo[ThisNPCFortChar.GetTransform().Translation, ThisRotation]){}
                                        if(BasicSettings.Prop.TeleportTo[BasicSettings.Prop.GetTransform().Translation, ThisRotation]){}
                                    if(ThisSoulBox:=DestinationPoint.SoulboxToComplete?):
                                        set IsWaiting = true
                                        if(ThisNavigatable:=NPCAgent?.GetFortCharacter[].GetNavigatable[]):
                                            spawn{ThisNavigatable.Wait()}
                                        ThisSoulBox.Enable()
                                        ThisSoulBox.CompletedEvent.Await()
                                        set IsWaiting = false
                                    else if(DestinationPoint.WaitTime > 0.0):
                                        set IsWaiting = true
                                        if(ThisNavigatable:=NPCAgent?.GetFortCharacter[].GetNavigatable[]):
                                            spawn{ThisNavigatable.Wait(?Duration:=DestinationPoint.WaitTime)}
                                        Sleep(DestinationPoint.WaitTime)
                                        set IsWaiting = false
                                        if(not IsWaitingEndResult?):
                                            ChangeMoveState.Signal(false)
                                else:
                                    Print("~ERR: A 'Point' in a path inside the Follow Objective [{BasicSettings.Name}] is invalid or empty!")
                                Sleep(0.0)
                            1
                    if(Winner = -1):
                        Print("~Player outside radius!")
                        PropTransform := BasicSettings.Prop.GetTransform()
                        if(BasicSettings.Prop.TeleportTo[PropTransform]){}
                        if(Navigatable:=NPCAgent?.GetFortCharacter[].GetNavigatable[]):
                            spawn{Navigatable.Wait()}
                        loop:
                            ThisMoveState:=ChangeMoveState.Await()
                            if(ThisMoveState?):
                                break
                    else:
                        Print("~Stopping follow AI from winner")
                        Stop()
            else:   
                Print("~ERR: No 'Paths' or 'Points' within a path have been set up in Follow Objective [{BasicSettings.Name}]!")
        else:
            Print("~ERR: The objective prop in Objective [{BasicSettings.Name}] is empty or invalid and cannot move!")
        Print("~Reached end of MoveAlongPath()")

    OnStopped<override>():void=
        Print("~Follow objective stopped")
        if(BasicSettings.Prop.TeleportTo[Location, rotation{}]){}
        if(ThisNPCFortChar:=NPCAgent?.GetFortCharacter[]):
            ThisNPCFortChar.SetVulnerability(true)
            ThisNPCFortChar.Damage(999999.0)
        set NPCAgent = false

    WatchPlayers<public><final>()<suspends>:void=
        loop:
            PlayersOutside := GetPlayersOutOfRadius()
            if(PlayersOutside.Length > 0):
                if(FollowSettings.DoStopOnPlayerOutsideRadius?):
                    ChangeMoveState.Signal(false)
                for(ThisPlayer:PlayersOutside):
                    if(not OutsideStatePerPlayer[ThisPlayer]? or not OutsideStatePerPlayer[ThisPlayer]):
                        if(set OutsideStatePerPlayer[ThisPlayer] = true){}
                        else:
                            set OutsideStatePerPlayer = ConcatenateMaps(OutsideStatePerPlayer, map{ThisPlayer => true})
                        spawn{ThisPlayer.OnOutsideRadius()}
            else:
                ChangeMoveState.Signal(true)
            Sleep(0.0)
    
    var<private> OutsideStatePerPlayer : [agent]logic = map{}
    (InPlayer:agent).OnOutsideRadius<public>()<suspends>:void=
        if(FollowSettings.DoDamagePlayerOutsideRadius?, ThisPlayerInfo:=InPlayer.GetPlayerInfo[], ThisFortChar:=InPlayer.GetFortCharacter[]):
            loop:
                if(not ThisFortChar.IsActive[] or Distance(ThisFortChar.GetTransform().Translation, BasicSettings.Prop.GetTransform().Translation) <= FollowSettings.FollowRadius):
                    if(set OutsideStatePerPlayer[InPlayer] = false){}
                    break
                else:
                    ThisPlayerInfo.Damage(FollowSettings.DamageAmount)
                    Sleep(FollowSettings.DamageRate)
                Sleep(0.0)

    GetPlayersInRadius<public><final>():[]agent=
        ActivePlayerList:=GetActivePlayers()
        var PlayersInRadius : []agent = array{}
        for(ThisPlayer:ActivePlayerList):
            if(ThisFortChar:=ThisPlayer.GetFortCharacter[]):
                if(Distance(ThisFortChar.GetTransform().Translation, BasicSettings.Prop.GetTransform().Translation) <= FollowSettings.FollowRadius):
                    set PlayersInRadius += array{ThisPlayer}
        return PlayersInRadius

    GetPlayersOutOfRadius<public><final>():[]agent=
        ActivePlayerList:=GetActivePlayers()
        var PlayersInRadius : []agent = array{}
        for(ThisPlayer:ActivePlayerList):
            if(ThisFortChar:=ThisPlayer.GetFortCharacter[]):
                if(Distance(ThisFortChar.GetTransform().Translation, BasicSettings.Prop.GetTransform().Translation) > FollowSettings.FollowRadius):
                    set PlayersInRadius += array{ThisPlayer}
        return PlayersInRadius

objective_follow_settings := class<concrete>:

    @editable UseNPC : logic = false
    @editable ShowNPC : logic = false
    @editable RotatePropWithNPC : logic = true
    @editable NPCSpawner : npc_spawner_device = npc_spawner_device{}
    @editable DoZombiesTargetObjective : logic = true
    @editable FollowRadius : float = 200.0
    @editable DoStopOnPlayerOutsideRadius : logic = true
    @editable DoDamagePlayerOutsideRadius : logic = false
    @editable DamageAmount : float = 20.0
    @editable DamageRate : float = 1.0
    @editable OffsetFromPropToPoints : vector3 = vector3{}
    @editable OffsetFromPropToNPC : vector3 = vector3{}
    
objective_follow_path := class<concrete>:

    @editable ChoosePointsRandomly : logic = false
    @editable LoopTimes : type{_X:int where _X >= -1} = 0
    @editable Points : []objective_follow_point = array{}

    InitPoints<public>():void=
        var CurrPointIndex : int = 0
        for(ThisPoint:Points):
            ThisPoint.Init(CurrPointIndex)
            set CurrPointIndex += 1

    GetUntraveledPoints<public><final>(TraveledPointList:[]objective_follow_point)<transacts>:[]objective_follow_point=
        if(TraveledPointList.Length > 0):
            var UntraveledPoints : []objective_follow_point = array{}
            for(ThisPoint:Points):
                var IsInList : logic = false
                for(ThisTraveledPoint:TraveledPointList):
                    if(ThisPoint.IsEqual[ThisTraveledPoint]):
                        set IsInList = true
                if(not IsInList?):
                    set UntraveledPoints += array{ThisPoint}
            return UntraveledPoints
        else:
            return Points
    
    GetRandomPoint(TraveledPoints:[]objective_follow_point)<transacts><decides>:objective_follow_point=
        Shuffle(GetUntraveledPoints(TraveledPoints))[0]

objective_follow_point := class<concrete>:

    @editable Point : creative_prop = creative_prop{}
    @editable Speed : float = 1.0
    @editable WaitTime : float = 0.0
    @editable SoulboxToComplete : ?soulbox = false
    @editable MatchPointRotation : logic = false

    var<private> Location : vector3 = vector3{}
    var<private> Index : int = -1

    Init<public>(InIndex:int):void=
        if(Point.IsValid[]):
            set Location = Point.GetTransform().Translation
        else:
            Print("~ERR: A Point within a Path inside a Follow Objective is invalid or empty!")
        set Index = InIndex

    IsEqual<public><final>(OtherPoint:objective_follow_point)<transacts><decides>:void=
        OtherPoint.Index = Index
#>