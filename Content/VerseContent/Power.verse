using { /Fortnite.com/Devices }
using { /Fortnite.com/Characters }
using { /Fortnite.com/Game }
using { /Fortnite.com/UI }
using { /Fortnite.com/FortPlayerUtilities }
using { /Verse.org/Simulation }
using { /Verse.org/Random }
using { /Verse.org/Colors }
using { /UnrealEngine.com/Temporary/Diagnostics }
using { /UnrealEngine.com/Temporary/SpatialMath }
using { /UnrealEngine.com/Temporary/UI }
using { /Verse.org/Colors/NamedColors }
using { InteractiveModule }

########  ########     ###     ######    #######  ##    ## ########     ###     ######   ######## 
##     ## ##     ##   ## ##   ##    ##  ##     ## ###   ## ##     ##   ## ##   ##    ##  ##       
##     ## ##     ##  ##   ##  ##        ##     ## ####  ## ##     ##  ##   ##  ##        ##       
##     ## ########  ##     ## ##   #### ##     ## ## ## ## ########  ##     ## ##   #### ######   
##     ## ##   ##   ######### ##    ##  ##     ## ##  #### ##   ##   ######### ##    ##  ##       
##     ## ##    ##  ##     ## ##    ##  ##     ## ##   ### ##    ##  ##     ## ##    ##  ##       
########  ##     ## ##     ##  ######    #######  ##    ## ##     ## ##     ##  ######   ########

<###########################

PowerDevice

###########################>
CheckPowerState<public>(?Channel:int=0)<transacts><decides>:void=
    GlobalPowerDevice[GetSession()].IsEnabled[?Channel:=Channel]

GetPowerEnabledEvent<public>(?Channel:int=0)<transacts><decides>:event()=
    GlobalPowerDevice[GetSession()].PowerEnabledEventsPerChannel[Channel]

var GlobalPowerDevice : weak_map(session, power_manager) = map{}
power_manager := class(creative_device):

    defaultstate_tooltip<localizes>:message="What state should the power be in on game start?"
    @editable{ToolTip:=defaultstate_tooltip} DefaultState : logic = false
    @editable OnTrigger_EnablePower : trigger_device = trigger_device{}
    @editable OnTrigger_DisablePower : trigger_device = trigger_device{}
    @editable ActivateTrigger_OnPowerReset : trigger_device = trigger_device{}

    channels_tooltip<localizes>:message="Allows for power to be routed for specific things"
    @editable{ToolTip:=channels_tooltip} Channels : []power_channel = array{}

    OnBegin<override>():void=
        OnTrigger_EnablePower.TriggeredEvent.Subscribe(EnablePower)
        OnTrigger_DisablePower.TriggeredEvent.Subscribe(DisablePower)
        set PowerState = DefaultState
        InitChannels()
        if(set GlobalPowerDevice[GetSession()] = Self){}
        else{Print("~ERROR: Failed to link power device globally!")}

    InitChannels():void=
        for(ThisChannel:Channels):
            ThisChannel.Init()
            set PowerEnabledEventsPerChannel += array{event(){}}

    var<private> PowerState : logic = false
    PowerEnabledEvent:event(int)=event(int){}
    PowerDisabledEvent:event(int)=event(int){}

    var<private> PowerEnabledEventsPerChannel : []event() = array{event(){}}

    EnablePower<private><final>(Agent:?agent, ?InChannel:int=0):void=
        set PowerState = true
        PowerEnabledEvent.Signal(InChannel)
        if(ThisEnabledEvent:=PowerEnabledEventsPerChannel[InChannel]):
            ThisEnabledEvent.Signal()

    DisablePower<private><final>(Agent:?agent, ?InChannel:int=0):void=
        set PowerState = false
        PowerDisabledEvent.Signal(InChannel)

    Enable<public><final>(?InChannel:int=0):void=
        if(InChannel = 0):
            OnTrigger_EnablePower.Trigger()
        else:
            var NoAgent:?agent=false
            EnablePower(NoAgent, ?InChannel:=InChannel)

    Disable<public><final>(?InChannel:int=0):void=
        var NoAgent:?agent=false
        DisablePower(NoAgent, ?InChannel:=InChannel)

    Reset<public><final>():void=
        ActivateTrigger_OnPowerReset.Trigger()
        set PowerState = DefaultState
        for(ThisChannel:Channels):
            ThisChannel.Reset()
        Powerables:=PowerableCollection[GetSession()] or array{}
        for(ThisPowerable:Powerables):
            ThisPowerable.PowerSettings.Reset()

    IsEnabled<public>(?Channel:int=0)<transacts><decides>:void=
        if(Channel > 0):
            FindChannel[Channel].PowerState?
        else:
            PowerState?

    FindChannel<private><final>(InChannel:int)<transacts><decides>:power_channel=
        var MaybeChannel : ?power_channel = false
        for:
            ThisChannel : Channels
            ThisChannel.Channel = InChannel
        do:
            set MaybeChannel = option{ThisChannel}
        MaybeChannel?

    SetChannelState<public>(InState:logic, Channel:int):logic=
        if(ThisChannel:=FindChannel[Channel]):
            ThisChannel.SetState(InState)
            return true
        return false

power_channel := class<concrete>:

    @editable DefaultState : logic = false
    @editable OnTrigger_EnablePowerChannel : trigger_device = trigger_device{}
    @editable OnTrigger_DisablePowerChannel : trigger_device = trigger_device{}

    channel_tooltip<localizes>:message="The channel this power source represents"
    @editable{ToolTip:=channel_tooltip} Channel : type{_X : int where 1 <= _X} = 1

    var<private> PowerState : logic = false
    Init():void=
        set PowerState = DefaultState
        OnTrigger_DisablePowerChannel.TriggeredEvent.Subscribe(DisableChannel)
        OnTrigger_EnablePowerChannel.TriggeredEvent.Subscribe(EnableChannel)

    EnableChannel<private>(Agent:?agent):void=
        set PowerState = true
        
    DisableChannel<private>(Agent:?agent):void=
        set PowerState = false

    Reset():void=
        set PowerState = DefaultState

    SetState<public>(InState:logic):void=
        set PowerState = InState

<####################################

Powerable

A foundation for objects who run on power

####################################>

var PowerableCollection<public> : weak_map(session, []Powerable) = map{}

Powerable<public> := class(Interactive_Base):

    @editable PowerSettings : powerable_settings = powerable_settings{}

    OnBegin<override>()<suspends>:void=
        if:
            not PowerableCollection[GetSession()]
            set PowerableCollection[GetSession()] = array{}
        if:
            CurrentPowerableCollection := PowerableCollection[GetSession()]
            set PowerableCollection[GetSession()] = CurrentPowerableCollection + array{Self}
        if(not IsPowered[]):
            SetInteractableState(false)
        loop:
            if(ThisPowerManager:=GlobalPowerDevice[GetSession()]):
                spawn{AwaitPowerChange()}
                break
            Sleep(0.0)
        PowerSettings.Init(Self)
        OnPowerableBegin()

    IsPowered<public><final>()<transacts><decides>:void=
        if(GlobalPowerDevice[GetSession()]):
            if(PowerSettings.UsePower?):
                CheckPowerState[?Channel:=PowerSettings.Channel] or PowerSettings.PowerState?

    AwaitPowerChange<private>()<suspends>:void=
        if(ThisPowerManager:=GlobalPowerDevice[GetSession()]):
            loop:
                race:
                    block:
                        loop:
                            if(not CheckPowerState[?Channel:=PowerSettings.Channel]):
                                Channel:=ThisPowerManager.PowerEnabledEvent.Await()
                                if(Channel=PowerSettings.Channel):
                                    PowerSettings.Enable(false)
                                    spawn{OnPowered()}
                            else:
                                Channel:=ThisPowerManager.PowerDisabledEvent.Await()
                                if(Channel=PowerSettings.Channel):
                                    PowerSettings.Disable(false)
                                    spawn{OnUnPowered()}
                    block:
                        PowerSettings.ResetEvent.Await()
                Sleep(0.0)

    OnPowered<public>()<suspends>:void=
        return

    OnUnPowered<public>()<suspends>:void=
        return

    OnPowerableBegin<public>()<suspends>:void=
        return

    ForcePowerState<public><final>(InState:logic):void=
        if(InState?):
            PowerSettings.Enable(false)
        else:
            PowerSettings.Disable(false)

    GetInteractionText<override>(InAgent:agent, Interactive:Interactable):string=
        if(IsPowered[]):
            GetPowerInteractionText(InAgent, Interactive)
        else:
            SetInteractableState(false)
            return "<C_R>power required<E>"

    GetPowerInteractionText<public>(InAgent:agent, Interactive:Interactable):string=
        return "Unset device interaction text"

powerable_settings := class<concrete>:

    var ParentPowerable : ?Powerable = false

    @editable UsePower : logic = true
    @editable Channel : type{_X : int where 0 <= _X} = 0
    @editable OnTrigger_EnableSelfPower : trigger_device = trigger_device{}
    @editable OnTrigger_DisableSelfPower : trigger_device = trigger_device{}
    @editable UnpoweredProp : creative_prop = creative_prop{}
    @editable PoweredProp : creative_prop = creative_prop{}

    var<private> PoweredPropTransform : transform = transform{}

    var<private> PowerState : logic = false
    var<private> ResetEvent : event() = event(){}

    Init(InPowerable:Powerable):void=
        set ParentPowerable = option{InPowerable}
        if(PoweredProp.IsValid[]):
            set PoweredPropTransform = PoweredProp.GetTransform()
        if(not UsePower?):
            set PowerState = true
        else if(CheckPowerState[?Channel:=Channel]):
            UnpoweredProp.Hide()
        else:
            if(PoweredProp.IsValid[]):
                if(PoweredProp.TeleportTo[vector3{}, rotation{}]){}
            PoweredProp.Hide()
        OnTrigger_EnableSelfPower.TriggeredEvent.Subscribe(Enable)
        OnTrigger_DisableSelfPower.TriggeredEvent.Subscribe(Disable)

    Enable(Agent:?agent):void=
        set PowerState = true
        if(PoweredProp.IsValid[], PoweredProp.TeleportTo[PoweredPropTransform]){}
        UnpoweredProp.Hide()
        PoweredProp.Show()

    Disable(Agent:?agent):void=
        set PowerState = false
        UnpoweredProp.Show()
        if(PoweredProp.IsValid[], PoweredProp.TeleportTo[vector3{}, rotation{}]){}
        PoweredProp.Hide()

    Reset():void=
        if(UsePower?):
            if(CheckPowerState[?Channel:=Channel]):
                Enable(false)
            else:
                Disable(false)
            ResetEvent.Signal()