using { /Fortnite.com/Devices }
using { /Fortnite.com/Characters }
using { /Fortnite.com/Game }
using { /Fortnite.com/UI }
using { /Fortnite.com/FortPlayerUtilities }
using { /Verse.org/Simulation }
using { /Verse.org/Random }
using { /Verse.org/Colors }
using { /Verse.org/Simulation/Tags }
using { /UnrealEngine.com/Temporary/Diagnostics }
using { /UnrealEngine.com/Temporary/SpatialMath }
using { /UnrealEngine.com/Temporary/UI }
using { /Fortnite.com/Devices/CreativeAnimation }
using { /Fortnite.com/Devices/CreativeAnimation/InterpolationTypes }
using { Player }
using { UI }
using { Spawning }

GetQuestTitle<public>():string=
    ActiveQuest[GetSession()]?.BasicSettings.QuestTitle or ""

var FirstQuest : weak_map(session, ?quest_base) = map{}
var ActiveQuest : weak_map(session, ?quest_base) = map{}

quest_base<public> := class<unique>(creative_device):

    @editable BasicSettings<public> : quest_settings = quest_settings{}

    var<private> CurrentProgress : int = 0
    var<private> CompletedState<public> : logic = false

    CompletedEvent : event() = event(){}

    OnBegin<override>()<suspends>:void=
        BasicSettings.OnTrigger_ActivateQuest.TriggeredEvent.Subscribe(ActivateViaTrigger)
        BasicSettings.OnTrigger_CountProgress.TriggeredEvent.Subscribe(IncrementProgressViaTrigger)
        BasicSettings.OnTrigger_ForceComplete.TriggeredEvent.Subscribe(ForceComplete)
        Sleep(0.0)
        if(BasicSettings.ActivateOnGameStart?):
            AwaitGameState(game_state.InGame)
            if. set FirstQuest[GetSession()] = option{Self}
            SafeSleep(1.0)
            Activate()
    
    ActivateViaTrigger<private>(Agent:?agent):void=
        Activate()

    IncrementProgressViaTrigger<private>(Agent:?agent):void=
        IncrementProgress()

    ForceComplete<private>(Agent:?agent):void=
        Deactivate()

    var<private> IsActive : logic = false
    Activate<public><final>():void=
        if(not IsActive?, not CompletedState?):
            set IsActive = true
            spawn. AwaitActivationConditions()
        else if(CompletedState?):
            # Starts the next quest if this quest has been completed before activation
            if(ThisNextQuest:=BasicSettings.OnCompleted_StartQuest?):
                spawn{StartNextQuest()}

    AwaitActivationConditions<public>()<suspends>:void=
        Winner:=race:
            block:
                AwaitWave(BasicSettings.MinimumWaveToActivate)
                true
            block:
                CompletedEvent.Await()
                false
        if(Winner?):
            if(set ActiveQuest[GetSession()] = option{Self}){}
            PlayerList:=GetAgents(?OnlyActives:=false)
            for(ThisPlayer:PlayerList):
                ThisPlayer.GetUIManager().QuestUI.ChangeText(Self)
            OnActivated()

    OnActivated<public>()<suspends>:void=
        return

    IncrementProgress<public><final>():void=
        set CurrentProgress += 1
        PlayerList:=GetAgents(?OnlyActives:=false)
        for(ThisPlayer:PlayerList):
            ThisPlayer.GetUIManager().QuestUI.IncrementProgress()
        if(CurrentProgress >= BasicSettings.TotalProgressToComplete):
            spawn{PreDeactivation()}
    
    PreDeactivation<private>()<suspends>:void=
        SafeSleep(0.25)
        PlayerList:=GetAgents(?OnlyActives:=false)
        for(ThisPlayer:PlayerList):
            spawn{ThisPlayer.GetUIManager().QuestUI.CompletedAnimation()}
        SafeSleep(GreenFlash.Length*0.15)
        for(ThisPlayer:PlayerList):
            ThisPlayer.GetUIManager().QuestUI.Hide()
        SafeSleep(0.25)
        Deactivate()

    Deactivate<public><final>(?StartNextQuestOverride:logic=true):void=
        if(IsActive?):
            set IsActive = false
            if(set ActiveQuest[GetSession()] = false){}
            CompletedEvent.Signal()
            if(ThisNextQuest:=BasicSettings.OnCompleted_StartQuest?, StartNextQuestOverride?):
                spawn{StartNextQuest()}
            else:
                PlayerList:=GetAgents(?OnlyActives:=false)
                for(ThisPlayer:PlayerList):
                    ThisPlayer.GetUIManager().QuestUI.Hide()
            spawn{OnDeactivated()}
        set CompletedState = true

    StartNextQuest<private>()<suspends>:void=
        BasicSettings.OnCompleted_ActivateTrigger.Trigger()
        if(ThisNextQuest:=BasicSettings.OnCompleted_StartQuest?):
            if(ThisNextQuest <> Self):
                SafeSleep(2.0)
                ThisNextQuest.Activate()
            else:
                Print("~ERR: Infinite quest cycle detected. Make sure the next quest is not linked to itself!")

    OnDeactivated<public>()<suspends>:void=
        return

    GetProgressRequired<public><final>()<transacts>:int=
        return BasicSettings.TotalProgressToComplete

    Reset():void=
        Deactivate(?StartNextQuestOverride:=false)
        set CurrentProgress = 0

quest_settings<public> := class<concrete>:

    @editable QuestTitle<public> : string = "Unset Quest Title"
    @editable QuestDescription : string = "Unset Quest Description"
    @editable ActivateOnGameStart : logic = false
    @editable TotalProgressToComplete : int = 1
    @editable MinimumWaveToActivate : type{_X:int where _X >= 1} = 1
    @editable OnTrigger_CountProgress : trigger_device = trigger_device{}
    @editable OnTrigger_ActivateQuest : trigger_device = trigger_device{}
    @editable OnTrigger_ForceComplete : trigger_device = trigger_device{}
    @editable OnCompleted_ActivateTrigger : trigger_device = trigger_device{}
    @editable OnCompleted_StartQuest : ?quest_base = false
