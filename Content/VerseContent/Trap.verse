using { /Fortnite.com/Devices }
using { /Fortnite.com/Characters }
using { /Fortnite.com/Game }
using { /Fortnite.com/UI }
using { /Fortnite.com/FortPlayerUtilities }
using { /Verse.org/Simulation }
using { /Verse.org/Random }
using { /Verse.org/Colors }
using { /UnrealEngine.com/Temporary/Diagnostics }
using { /UnrealEngine.com/Temporary/SpatialMath }
using { /UnrealEngine.com/Temporary/UI }
using { /Verse.org/Colors/NamedColors }
using { Player }
using { Zombie }
using { InteractiveModule }

Trap_base := class(Powerable):

    @editable BasicSettings : trap_basic_settings = trap_basic_settings{}

    @editable
    DefaultActivatePrompt : string = "Turn on trap"

    @editable 
    InteractiveLocations : []Interactable = array{}

    var<private> ActiveState : logic = false

    var<private> CurrentUser : ?agent = false

    OnPowerableBegin<override>():void=
        BasicSettings.ActivatedTrapProp.Hide()
        for(ThisKillZone:BasicSettings.KillZones):
            ThisKillZone.AgentEntersEvent.Subscribe(AgentEnteredKillZone)
            ThisKillZone.AgentExitsEvent.Subscribe(AgentExitedKillZone)
        LinkAndCompile(InteractiveLocations)

    var<private> AgentInKillZonesState : [agent]logic = map{}
    AgentEnteredKillZone<private>(InAgent:agent):void=
        if(ActiveState?):
            var JustAdded : logic = false
            if(not AgentInKillZonesState[InAgent]):
                set AgentInKillZonesState = ConcatenateMaps(AgentInKillZonesState, map{InAgent => true})
                set JustAdded = true
            if(not AgentInKillZonesState[InAgent]? or JustAdded?):
                if(set AgentInKillZonesState[InAgent] = true):
                    if(ThisPlayerInfo:=InAgent.GetPlayerInfo[]):
                        if(BasicSettings.InstaKillPlayers?):
                            ThisPlayerInfo.Damage(1000.0)
                        else:
                            ThisPlayerInfo.Damage(BasicSettings.PlayerDamage)
                    else if(ThisZombie:=InAgent.GetZombieAI[], not BasicSettings.ZombieTypesToIgnore.Find[ThisZombie.Controller.Type]):
                        if(BasicSettings.InstaKillZombies?):
                            #if(ThisUserInfo:=CurrentUser?.GetPlayerInfo[]):
                                #ThisUserInfo.ZombieEliminatedByTrapEvent.Signal()
                            ThisZombie.Eliminate(?Instigator:=CurrentUser)
                            if(BasicSettings.GrantPointsToInstigator?, ThisPlayerInfo:=CurrentUser?.GetPlayerInfo[]):
                                ThisPlayerInfo.IncreasePoints(50)
                                ThisPlayerInfo.EliminatedZombieByTrapEvent.Signal(ThisZombie.Agent)
                        else:
                            ThisZombie.Damage(BasicSettings.ZombieDamage, ?IsTrap:=true)
                    spawn{AwaitAgentExit(InAgent)}

    AgentExitedKillZone<private>(InAgent:agent):void=
        AgentExitedKillZoneEvent.Signal(InAgent)

    AgentExitedKillZoneEvent<private>:event(agent)=event(agent){}
    AwaitAgentExit<private>(InAgent:agent)<suspends>:void=
        race:
            TurnedOffEvent.Await()
            loop:
                SafeSleep(BasicSettings.DamageTickrate)
                if(ThisPlayerInfo:=InAgent.GetPlayerInfo[]):
                    ThisPlayerInfo.Damage(BasicSettings.PlayerDamage)
                else if(ThisZombieInfo:=InAgent.GetZombieAI[]):
                    ThisZombieInfo.Damage(BasicSettings.ZombieDamage)
            loop:
                ExitingAgent:=AgentExitedKillZoneEvent.Await()
                if(ExitingAgent=InAgent):
                    break
            block:
                if(ThisPlayerInfo:=InAgent.GetPlayerInfo[]):
                    ThisPlayerInfo.DownedEvent.Await()
                else if(ThisZombieInfo:=InAgent.GetZombieAI[]):
                    ThisZombieInfo.DeathEvent.Await()
                else:
                    Sleep(0.0)
        if(set AgentInKillZonesState[InAgent] = false){}

    GetPowerInteractionText<override>(InAgent:agent, Interactive:Interactable):string=
        if(ActiveState?):
            SetInteractableState(false)
            return "trap active"
        else if(CooldownState?):
            SetInteractableState(false)
            return "trap on cooldown"
        else:
            SetInteractableState(true)
            return DefaultActivatePrompt+" [<C_Y>${BasicSettings.Cost}<E>]"

    OnInteracted<override>(InAgent:agent, InInteractingLocation:Interactable):void=
        if(not CooldownState? and not ActiveState?):
            if(ThisPlayerInfo:=InAgent.GetPlayerInfo[]):
                BuyState:=ThisPlayerInfo.BuyItem(BasicSettings.Cost)
                if(BuyState?):
                    TurnOn()
                    set CurrentUser = option{InAgent}

    TurnOn<public><final>():void=
        spawn{TrapLifecycle()}

    TrapLifecycle<private>()<suspends>:void=
        if(not ActiveState?):
            set ActiveState = true
            BasicSettings.ActivateTrigger_OnTrapActivated.Trigger()
            BasicSettings.ActivatedTrapProp.Show()
            BasicSettings.DeactivatedTrapProp.Hide()
            SafeSleep(BasicSettings.Duration)
            set CooldownState = true
            TurnOff()

    TurnedOffEvent:event()=event(){}
    TurnOff<public><final>():void=
        spawn{AwaitCooldown()}
        set CurrentUser = false

    Reset<override><final>():void=
        set ActiveState = false
        set CooldownState = false
        TurnedOffEvent.Signal()
        BasicSettings.DeactivatedTrapProp.Show()
        BasicSettings.ActivatedTrapProp.Hide()
        (super:)Reset()

    var<private> CooldownState : logic = false
    AwaitCooldown<private>()<suspends>:void=
        if(ActiveState?):
            set ActiveState = false
            TurnedOffEvent.Signal()
            BasicSettings.ActivateTrigger_OnTrapEnd.Trigger()
            BasicSettings.DeactivatedTrapProp.Show()
            BasicSettings.ActivatedTrapProp.Hide()
            SafeSleep(BasicSettings.Cooldown)
            BasicSettings.ActivateTrigger_OnTrapReady.Trigger()
            BasicSettings.ActivateAudio_OnTrapReady.Play()
            set CooldownState = false

trap_basic_settings := class<concrete>:

    @editable Cost : int = 500
    @editable ZombieDamage : float = 500.0
    @editable InstaKillZombies : logic = true
    @editable GrantPointsToInstigator : logic = true
    @editable ZombieTypesToIgnore : []zombie_type_list = array{}
    @editable PlayerDamage : float = 30.0
    @editable InstaKillPlayers : logic = false
    @editable DamageTickrate : float = 0.75
    @editable Duration : float = 20.0
    @editable Cooldown : float = 60.0
    @editable ActivatedTrapProp : creative_prop = creative_prop{}
    @editable DeactivatedTrapProp : creative_prop = creative_prop{}
    @editable KillZones : []volume_device = array{}
    @editable ActivateTrigger_OnTrapReady : trigger_device = trigger_device{}
    @editable ActivateTrigger_OnTrapActivated : trigger_device = trigger_device{}
    @editable ActivateTrigger_OnTrapEnd : trigger_device = trigger_device{}
    @editable ActivateAudio_OnTrapReady : audio_player_device = audio_player_device{}