using { /Verse.org/Simulation }
using { /Fortnite.com/Devices }
using { /Fortnite.com/Characters }
using { /Fortnite.com/Game }
using { /Verse.org/Assets }
using { /Verse.org/Random }
using { /UnrealEngine.com/Temporary/SpatialMath }
using { /Fortnite.com/Animation/PlayAnimation }
using { /Fortnite.com/AI }
using { /Fortnite.com/Devices/CreativeAnimation }
using { /Fortnite.com/AI/movement_types }

shortcut_type<public> := enum{
    Walk
    Vault
    Leap
    Teleport
}

var GlobalShortcutCollection<public> : weak_map(session, []zombie_shortcut) = map{}

zombie_shortcut<public> := class<concrete>(comparable_device):

    # SETTINGS
    @editable
    StartPosition : creative_prop = creative_prop{}

    @editable
    MiddlePosition : creative_prop = creative_prop{}

    @editable
    EndPosition : creative_prop = creative_prop{}

    @editable
    TravelType : shortcut_type = shortcut_type.Leap
    
    @editable
    Speed : type{_X:float where _X > 0.0} = 1.0

    @editable
    IsCancelable : logic = false

    @editable
    IsEnabled : logic = true

    @editable
    UsedInMap : logic = true

    @editable
    ZombieDetectionZones : []volume_device = array{}

    # VARIABLES
    var<private> StartTransform : transform = transform{}
    var<private> MiddleTransform : transform = transform{}
    var<private> EndTransform<public> : transform = transform{}
    var<private> EndRotation : rotation = rotation{}
    var<private> ZAdditive : float = 0.0

    var ZombieDetectionVolumes<private> : []better_volume_device = array{}

    ShortcutQueue : zombie_queue = zombie_queue{}

    OnBegin<override>()<suspends>:void=
        if:
            StartPosition.IsValid[]
            MiddlePosition.IsValid[] or TravelType = shortcut_type.Walk or TravelType = shortcut_type.Teleport
            EndPosition.IsValid[]
        then:
            set StartTransform = StartPosition.GetTransform()
            set MiddleTransform = if. MiddlePosition.IsValid[] then. MiddlePosition.GetTransform() else. transform{}
            set EndTransform = EndPosition.GetTransform()
            set EndRotation = StartTransform.Translation.GetRotationToPosition(EndTransform.Translation)
            if(StartTransform.Translation.Z = EndTransform.Translation.Z):
                set ZAdditive = Abs(MiddleTransform.Translation.Z-StartTransform.Translation.Z)
            else if(StartTransform.Translation.Z > EndTransform.Translation.Z):
                set ZAdditive = Abs(MiddleTransform.Translation.Z-StartTransform.Translation.Z)
            else if(StartTransform.Translation.Z < EndTransform.Translation.Z):
                set ZAdditive = Abs(MiddleTransform.Translation.Z-EndTransform.Translation.Z)
            for(ThisZombieDetectZone:ZombieDetectionZones):
                NewDetectionVolume:better_volume_device=better_volume_device{VolumeDevice:=ThisZombieDetectZone}
                set ZombieDetectionVolumes += array{NewDetectionVolume}
                NewDetectionVolume.Init()
            if:
                not GlobalShortcutCollection[GetSession()]
                set GlobalShortcutCollection[GetSession()] = array{}
            if:
                CurrentShortcutCollection := GlobalShortcutCollection[GetSession()]
                set GlobalShortcutCollection[GetSession()] = CurrentShortcutCollection + array{Self}
            #loop:
            #    ShortcutQueue.RefreshQueue()
            #    Sleep(5.0)

    # Travels to the start of the shortcut sequence
    TravelTo<public>(InZombieInfo:zombie_ai, ?ReachRadius:float=15.0)<suspends>:logic=
        InZombieInfo.CancelCurrentTargeting.Signal()
        var PoppedState : logic = false
        defer{
            # This might not be working for some reason internally, not a priority issue atm
            if(not PoppedState?):
                ShortcutQueue.ForcePop(InZombieInfo)
        }
        NavigateState:=race:
            block:
                NavResult:=InZombieInfo.SafeNavigate(MakeNavigationTarget(StartTransform.Translation), ?ReachRadius:=ReachRadius, ?MovementType:=if. InZombieInfo.IsAggro[] then. Running else. Walking)
                if(NavResult?):
                    ShortcutQueue.LatchOn(InZombieInfo)
                NavResult
            block:
                InZombieInfo.DeathEvent.Await()
                false
            block:
                InZombieInfo.RespawnEvent.Await()
                false
        ShortcutQueue.Pop()
        set PoppedState = true
        InZombieInfo.StopPropAnimation()
        return NavigateState

    # The given zombie will use this shortcut when this is called
    Use<public>(InZombieInfo:zombie_ai, ?ReachRadius:float=15.0, ?CancelEvents:[]event(any)=array{})<suspends>:void=
        if(InZombieInfo.IsActive[]):
            case(TravelType):
                shortcut_type.Leap =>
                    if(ThisJumpPackage:=Shuffle(InZombieInfo.Controller.Animations.Leaps)[0]):
                        defer{
                            InZombieInfo.SetMoveState(true)
                        }
                        InZombieInfo.SetMoveState(false)
                        if. InZombieInfo.FortCharacter.TeleportTo[StartTransform.Translation, StartTransform.Rotation]
                        ThisJumpPackage.PlayAnimation(StartTransform, InZombieInfo)
                        CurveStartPosition := transform{Translation:=StartTransform.Translation+ThisJumpPackage.EndOffset,Rotation:=StartTransform.Rotation}
                        CurveEndPosition := EndTransform.Translation+vector3{Z:=InZombieInfo.Controller.Animations.ZOffset}
                        InZombieInfo.SetMoveState(false)
                        if. InZombieInfo.FortCharacter.TeleportTo[StartTransform.Translation, StartTransform.Rotation]
                        AnimateTask:=spawn. AnimateCurve(InZombieInfo, CurveStartPosition, CurveEndPosition)
                        AnimateTask.Await()
                        InZombieInfo.SetMoveState(true)
                    else if:
                        InZombieInfo.FortCharacter.TeleportTo[EndTransform.Translation, rotation{}]
                shortcut_type.Vault =>
                    if. InZombieInfo.FortCharacter.TeleportTo[StartTransform.Translation, StartTransform.Rotation]
                    if(ThisVaultPackage:=Shuffle(InZombieInfo.Controller.Animations.Vaulting)[0]):
                        ThisVaultPackage.PlayAnimation(StartTransform, InZombieInfo)
                shortcut_type.Walk =>
                    InZombieInfo.SafeNavigate(MakeNavigationTarget(EndTransform.Translation), ?ReachRadius:=20.0, ?MovementType:=if. InZombieInfo.IsAggro[] then. Running else. Walking)
                shortcut_type.Teleport =>
                    if. InZombieInfo.FortCharacter.TeleportTo[EndTransform.Translation, EndTransform.Rotation]
            CheckEndTransform:=transform{
                Translation:=EndTransform.GetLocalVector(vector3{X:=-15.0})
                Rotation:=EndTransform.Rotation
            }
            var Tries : int = 0
            loop:
                if(CheckEndTransform.GetVectorLookPercent(InZombieInfo.FortCharacter.GetTransform().Translation) >= 0.4):
                    break
                else if. InZombieInfo.FortCharacter.TeleportTo[EndTransform.GetLocalVector(vector3{X:=2.5*Tries}), EndTransform.Rotation]
                set Tries += 1
                Sleep(0.1)

    # Animates the zombie in a curve. A helper function
    AnimateCurve<private>(InZombieAI:zombie_ai, StartPos:transform, EndPos:vector3)<suspends>:void=
        GetCurvePoints := QuadBezier_3D(45, StartPos.Translation, EndPos, MiddleTransform.Translation+vector3{Z:=ZAdditive})
        ZombiePosition := InZombieAI.FortCharacter.GetTransform()
        if(ShortcutHelper := SpawnProp(InZombieAI.Controller.GetSkinInfo[InZombieAI.SkinTag].InAirAnimation, ZombiePosition.Translation, ZombiePosition.Rotation)(0)?, ThisAnimControl := ShortcutHelper.GetAnimationController[]):
            CompletedEvent:event()=event(){}
            branch:
                defer{
                    ShortcutHelper.Dispose()
                }
                if(var PrevCurvePoint : vector3 = GetCurvePoints[0]):
                    var TotalTime : float = 0.0
                    AnimPoints := for(ThisCurvePoint : GetCurvePoints):
                        ThisPointTime := Distance(ShortcutHelper.GetTransform().Translation, ThisCurvePoint) / (7500.0 * Speed)
                        set TotalTime += ThisPointTime
                        CurvePoint : keyframe_delta = keyframe_delta{
                            DeltaLocation := PrevCurvePoint.GetDeltaToPosition(ThisCurvePoint)
                            DeltaRotation := rotation{}
                            Time := ThisPointTime
                        }
                        set PrevCurvePoint = ThisCurvePoint
                        CurvePoint
                    ThisAnimControl.SetAnimation(AnimPoints, ?Mode:=animation_mode.OneShot)
                    ThisAnimControl.Play()
                    InZombieAI.Hide()
                    branch:
                        ThisRotation := StartPos.GetRotationToFacePosition(EndPos)
                        loop:
                            if(InZombieAI.FortCharacter.TeleportTo[ShortcutHelper.GetTransform().Translation-vector3{Z:=InZombieAI.Controller.Animations.ZOffset}, ThisRotation]){}
                            #else{Print("~ERR: Failed to teleport zombie to quick travel end location!")}
                            Sleep(0.0)
                    race:
                        ThisAnimControl.MovementCompleteEvent.Await()
                        InZombieAI.DeathEvent.Await()
                        InZombieAI.RespawnEvent.Await()
                    InZombieAI.Show()
                    if. InZombieAI.FortCharacter.TeleportTo[EndTransform.Translation, EndTransform.Rotation]
                CompletedEvent.Signal()
            race:
                CompletedEvent.Await()
                InZombieAI.CancelShortcut.Await()
            ShortcutHelper.Dispose()

    # Succeeds if the zombie should use this shortcut to get to the target quicker than normal navigation
    # Fails if the reverse is true
    ShouldUse<public>(NavDestination:vector3, InTarget:vector3, InPosition:vector3, InstigatingAgent:agent)<transacts><decides>:float=
        IsEnabled?
        UsedInMap?
        Distance(NavDestination, InTarget) > Distance(EndTransform.Translation, InTarget)
        (ZombieDetectionVolumes.Length <= 0 and (GetShortenedDistance(InPosition, InTarget) < (Distance(InPosition, NavDestination)+Distance(NavDestination, InTarget)))) or ZombieDetectionVolumes.GetAgentsInZones(array. InstigatingAgent).Length > 0
        return Distance(InPosition, StartTransform.Translation)+Distance(EndTransform.Translation, InTarget)

    # Gets the distance from a position to a target if it were to use this shortcut
    # Position -> StartPosition + EndPosition -> Target
    GetShortenedDistance<public>(InPosition:vector3, InTarget:vector3)<transacts>:float=
        return Distance(InPosition, StartTransform.Translation)+Distance(EndTransform.Translation, InTarget)

GetUsefulShortcut<public>(InTarget:vector3, CurrDestination:vector3, InstigatingAgent:agent)<transacts><decides>:zombie_shortcut=
    ShortcutList := GlobalShortcutCollection[GetSession()]
    CurrPos:=InstigatingAgent.GetFortCharacter[].GetTransform().Translation
    var MaybeUsefulShortcutTuple : tuple(?zombie_shortcut, float) = (option{ShortcutList[0]}, ShortcutList[0].GetShortenedDistance(CurrPos, InTarget))
    CurrTotalDistance:=Distance(CurrPos, CurrDestination)+Distance(CurrDestination, InTarget)
    for:
        ThisShortcut:ShortcutList
        ShortenedDistance:=ThisShortcut.ShouldUse[CurrDestination, InTarget, CurrPos, InstigatingAgent]
        ShortenedDistance < MaybeUsefulShortcutTuple(1)
    do:
        set MaybeUsefulShortcutTuple = (option{ThisShortcut}, ShortenedDistance)
    MaybeUsefulShortcutTuple(0)?.ShouldUse[CurrDestination, InTarget, CurrPos, InstigatingAgent]
    MaybeUsefulShortcutTuple(1) < CurrTotalDistance
    MaybeUsefulShortcutTuple(0)?