using { /Fortnite.com/Devices }
using { /Fortnite.com/Characters }
using { /Fortnite.com/Game }
using { /Fortnite.com/UI }
using { /Fortnite.com/FortPlayerUtilities }
using { /Verse.org/Simulation }
using { /Verse.org/Random }
using { /Verse.org/Colors }
using { /UnrealEngine.com/Temporary/Diagnostics }
using { /UnrealEngine.com/Temporary/SpatialMath }
using { /UnrealEngine.com/Temporary/UI }
using { Player }
using { Weapons }

########  ########     ###     ######    #######  ##    ## ########     ###     ######   ######## 
##     ## ##     ##   ## ##   ##    ##  ##     ## ###   ## ##     ##   ## ##   ##    ##  ##       
##     ## ##     ##  ##   ##  ##        ##     ## ####  ## ##     ##  ##   ##  ##        ##       
##     ## ########  ##     ## ##   #### ##     ## ## ## ## ########  ##     ## ##   #### ######   
##     ## ##   ##   ######### ##    ##  ##     ## ##  #### ##   ##   ######### ##    ##  ##       
##     ## ##    ##  ##     ## ##    ##  ##     ## ##   ### ##    ##  ##     ## ##    ##  ##       
########  ##     ## ##     ##  ######    #######  ##    ## ##     ## ##     ##  ######   ########

# SoulBox Device which counts eliminated creatures within the set radius #

soulbox<public> := class(creative_device):

    @editable SoulboxLocation<public> : creative_prop = creative_prop{}
    @editable SoulFlareVFX_BP<public> : creative_prop_asset = DefaultCreativePropAsset
    @editable Animation_Module : soulbox_animation = soulbox_animation{}
    @editable Controls_Module : soulbox_controls = soulbox_controls{}
    @editable Signal_Recieve : soulbox_signalrecieve = soulbox_signalrecieve{}
    @editable Signal_Send : soulbox_signalsend = soulbox_signalsend{}

    var<private> EnabledState<public>: logic = false
    CompletedEvent:event()=event(){}
    DeactivateEvent:event()=event(){}

    var<private> SoulsConsumed<public>: int = 0
    var<private> CompletedState : logic = false

    OnBegin<override>()<suspends>:void=
        if(SoulboxLocation.IsValid[]):
            InitRecieve()
            if(Controls_Module.EnabledOnGameStart?):
                Enable()
        else:
            Print("~ERROR: SoulboxLocation param in SoulBox Device is empty or invalid!")
        OnSoulboxBegin()

    OnSoulboxBegin<public>()<suspends>:void=
        return

    InitRecieve():void=
        Signal_Recieve.EnableOn_Trigger.TriggeredEvent.Subscribe(EnableViaTrigger)
        Signal_Recieve.EnableOn_Trigger.TriggeredEvent.Subscribe(DisableViaTrigger)

    EnableViaTrigger<private>(Agent:?agent):void=
        Enable()

    DisableViaTrigger<private>(Agent:?agent):void=
        Disable()

    Enable<public>():void=
        if(not EnabledState?):
            set EnabledState = true
            spawn{SoulLoop()}

    Reset<public>():void=
        set SoulsConsumed = 0
        set SoulsBeingConsumed = 0
        set CompletedState = false
        DeactivateEvent.Signal()

    Disable<public>():void=
        if(EnabledState?):
            set EnabledState = false
            DeactivateEvent.Signal()

    var SoulsBeingConsumed : int = 0
    SoulLoop<private>()<suspends>:void=
        Reset()
        SoulboxLocation.Show()
        AllowdZombieElimEvents := for:
            ThisZombieType:Controls_Module.AllowedZombieTypes
            ThisController:=ThisZombieType.GetController[]
        do:
            ThisController.ZombieEliminated
        branch:
            loop:
                EliminatedZombieInfo := RaceAll(AllowdZombieElimEvents)
                if(ZombiePosition := EliminatedZombieInfo?.GetZombieAI[].DeathTransform.Translation):
                    if(Distance(ZombiePosition, SoulboxLocation.GetTransform().Translation) <= Controls_Module.SoulboxRadius):
                        if(SoulsBeingConsumed < Controls_Module.MaxConsumingSoulsAtATime or Controls_Module.MaxConsumingSoulsAtATime = -1):
                            set SoulsConsumed += 1
                            set SoulsBeingConsumed += 1
                            spawn. MoveSoulFlare(ZombiePosition)
                            if(SoulsConsumed >= Controls_Module.SoulsRequired):
                                break
                        else:
                            # The Distance between the dead zombie and soul box is larger than the maximum amount #
                Sleep(0.0)
        DeactivateEvent.Await()
        set EnabledState = false
        case(Controls_Module.CompletionBehavior):
            soulbox_completion.ResetAndRestart =>
                set SoulsConsumed = 0
                spawn{SoulLoop()}
                return
            soulbox_completion.ResetAndDisable =>
                set SoulsConsumed = 0
                Disable()
            soulbox_completion.Disable =>
                set SoulsConsumed = 0
                Disable()

    OnSoulboxCompleted<public>()<suspends>:void=
        Print("A")
        return

    ForceComplete<public>():void=
        set CompletedState = true
        Signal_Send.ActivateTrigger_OnSoulBoxCompleted.Trigger()
        spawn. OnSoulboxCompleted()
        DeactivateEvent.Signal()
        CompletedEvent.Signal()
        SoulboxLocation.Hide()

    MoveSoulFlare(InPosition : vector3)<suspends>:void=
        MaybeSoulFlareProp := SpawnProp(SoulFlareVFX_BP, InPosition, rotation{})
        if(SoulFlareProp := MaybeSoulFlareProp(0)?):
            MovementSpeed := (Distance(InPosition, SoulboxLocation.GetTransform().Translation)) / (Controls_Module.SoulVFXMovementSpeed*60.0)
            SoulFlareProp.MoveTo(SoulboxLocation.GetTransform().Translation, rotation{}, MovementSpeed)
            SoulFlareProp.Dispose()
            set SoulsBeingConsumed -= 1
            SoulConsumedEvent.Signal()
            if(Animation_Module.ConsumingEffect_ActiveTime > 0.0):
                if(SpawnedConsumingEffect:=SpawnProp(Animation_Module.ConsumingEffect_BP, SoulboxLocation.GetTransform())(0)?):
                    Sleep(Animation_Module.ConsumingEffect_ActiveTime)
                    SpawnedConsumingEffect.Dispose()
            else:
                Sleep(0.0)
        Signal_Send.ActivateTrigger_OnSoulConsumed.Trigger()
        if(SoulsConsumed >= Controls_Module.SoulsRequired, not CompletedState?):
            set CompletedState = true
            Signal_Send.ActivateTrigger_OnSoulBoxCompleted.Trigger()
            spawn. OnSoulboxCompleted()
            if(Animation_Module.ChangeBoxAnimation?):
                if(FullAnimation:=SpawnProp(Animation_Module.FullAnim_BP, SoulboxLocation.GetTransform())(0)?):
                    Sleep(Animation_Module.FullAnim_ActiveTime)
                    FullAnimation.Dispose()
            if(Animation_Module.FullEffect_ActiveTime > 0.0):
                if(SpawnedFullEffect:=SpawnProp(Animation_Module.FullEffect_BP, SoulboxLocation.GetTransform())(0)?):
                    Sleep(Animation_Module.FullEffect_ActiveTime)
                    SpawnedFullEffect.Dispose()
            DeactivateEvent.Signal()
            CompletedEvent.Signal()
            SoulboxLocation.Hide()

    SoulConsumedEvent : event() = event(){}
    SpawnConsumingAnim()<suspends>:void=
        if(SoulsBeingConsumed <= 1):
            if(Animation_Module.ChangeBoxAnimation?, ConsumingAnimation:=SpawnProp(Animation_Module.ConsumingSoulAnim_BP, SoulboxLocation.GetTransform())(0)?):
                loop:
                    SoulConsumedEvent.Await()
                    if(SoulsBeingConsumed <= 0):
                        ConsumingAnimation.Dispose()
            else:
                Sleep(0.0)

    IsCompleted<public>()<transacts><decides>:void=
        CompletedState?

soulbox_animation := class<concrete>:

    @editable ChangeBoxAnimation : logic = false
    @editable IdleAnim_BP : creative_prop_asset = DefaultCreativePropAsset
    @editable ConsumingSoulAnim_BP : creative_prop_asset = DefaultCreativePropAsset
    @editable ConsumingAnim_ActiveTime : float = 0.0
    @editable ConsumingEffect_BP : creative_prop_asset = DefaultCreativePropAsset
    @editable ConsumingEffect_ActiveTime : float = 0.0
    @editable FullAnim_BP : creative_prop_asset = DefaultCreativePropAsset
    @editable FullAnim_ActiveTime : float = 0.0
    @editable FullEffect_BP : creative_prop_asset = DefaultCreativePropAsset
    @editable FullEffect_ActiveTime : float = 0.0

soulbox_completion<public> := enum{ResetAndRestart, ResetAndDisable, Disable}
soulcount := type{_X : int where 0 < _X}
soulbox_controls := class<concrete>:

    @editable EnabledOnGameStart : logic = false
    @editable CompletionBehavior : soulbox_completion = soulbox_completion.Disable
    @editable AllowedZombieTypes : []zombie_type_list = array{zombie_type_list.Default}
    @editable SoulsRequired : soulcount = 25
    @editable SoulboxRadius : float = 500.0
    @editable SoulVFXMovementSpeed : int = 5

    @editable MaxConsumingSoulsAtATime : type{_X : int where -1 <= _X} = -1

soulbox_signalrecieve := class<concrete>:

    @editable EnableOn_Trigger : trigger_device = trigger_device{}
    @editable DisableOn_Trigger : trigger_device = trigger_device{}
    @editable ResetOn_Trigger : trigger_device = trigger_device{}

soulbox_signalsend := class<concrete>:

    @editable ActivateTrigger_OnSoulConsumed : trigger_device = trigger_device{}
    @editable ActivateTrigger_OnSoulBoxCompleted : trigger_device = trigger_device{}