using { /Verse.org/Simulation }
using { /Fortnite.com/Characters }
using { /Fortnite.com/Game }
using { /Fortnite.com/AI }
using { /Fortnite.com/Animation/PlayAnimation }
using { /UnrealEngine.com/Temporary/SpatialMath }
using { /Verse.org/Random }
using { /Verse.org/Colors }
using { /Fortnite.com/Devices }
using { /Fortnite.com/AI/movement_types }
using { Spawning }
using { Player }
using { Weapons }
using { Powerups }
using { Room }

<#
    Start(){
        branch:
            SpawnFunction()
    }

    SpawnFunction(){
        defer{Clear zombies from spawn location}
        race:
            
    }
#>

# Gets the zombies that are targeting the agent
(InAgent:agent).GetTargetedAmount<public>()<transacts>:int=
    TargetedCount:=for:
        ThisZombie:GetZombieList()
        ThisZombie.LastAgentTarget? = InAgent or ThisZombie.SpawnLocation?.LinkedRoom?.GetPlayersInRoom().Find[InAgent]
    do:
        0
    return TargetedCount.Length

empty_ai<public> := class(npc_behavior):

DEATHPOS<public> : vector3 = vector3{X:=10621.607136,Y:=-6213.146759,Z:=-3277.233194}
ATTACKDISTANCE<public> : float = 90.0

zombie_state<public> := enum{
    SpawningProcess
    Targeting # True for any zombie that is targeting any player/zombie
    Attacking
    Shortcutting # True for any zombie using a shortcut
    Stunned # True when the zombie is immobile in any form (frozen, electrocuted..etc)
    Inactive # Can be true for zombies not yet initialized and dead zombies
    Paused
    GameOver
}

(ThisState:zombie_state).ToString():string=
    case(ThisState):
        zombie_state.SpawningProcess => "Spawning"
        zombie_state.Targeting => "Targeting"
        zombie_state.Inactive => "Inactive"
        _ => "Unknown"

# Holds all current zombie instances
var GlobalZombieList : weak_map(session, [agent]zombie_ai) = map{}

# Refreshes the global zombie list with only active zombies
RefreshZombieList<public>():void=
    var UpdatedList : [agent]zombie_ai = map{}
    AliveZombies := GetZombieList(?OnlyActive:=true)
    for(ThisZombie:AliveZombies):
        if. set UpdatedList[ThisZombie.Agent] = ThisZombie
        else. set UpdatedList = ConcatenateMaps(UpdatedList, map{ThisZombie.Agent => ThisZombie})
    if. set GlobalZombieList[GetSession()] = UpdatedList
    else. Print("ERR: Failed to refresh the global zombie list!")

zombie_marker<public> := enum{
    None
    Skull
    Turned
    Objective
    Tracked
}

GetZombieList<public>(?OnlyActive:logic=true, ?OnlyAggro:logic=false, ?OnlyWave:logic=false, ?TypeFilter:[]zombie_type_list=array{})<transacts>:[]zombie_ai=
    return block{
        AllZombieList := GlobalZombieList[GetSession()]
        ThisZombieList:=for:
            ThisZombie:AllZombieList
            not OnlyActive? or ThisZombie.IsActive[]
            not OnlyAggro? or ThisZombie.IsAggro[]
            not OnlyWave? or ThisZombie.IsWaveEnemy?
            TypeFilter.Length = 0 or TypeFilter.HasType[ThisZombie.Controller.Type]
        do:
            ThisZombie
    } or array{}

(InAgent:agent).GetZombieAI<public>()<transacts><decides>:zombie_ai=
    return GlobalZombieList[GetSession()][InAgent]

(ZombieAIList:[]zombie_ai).ConvertToAgents()<transacts>:[]agent=
    ZombieAgentList:=for:
        ThisZombieAI:ZombieAIList
    do:
        ThisZombieAI.Agent

zombie_ai<public> := class(comparable_class):

    # SETTINGS/IDENTITY
    Identity<public> : int = -1
    Agent<public> : agent
    FortCharacter<public> : fort_character
    Controller<public> : zombie_type
    SkinTag<public> : string

    # VARIABLES
    var State : zombie_state = zombie_state.Inactive
    var Health : float = 100.0
    var StartingHealth<public> : float = 100.0
    var DamageAmount<public> : float = 35.0
    var<private> MovementSpeed<public> : float = 0.5
    var StartingMovementSpeed<private> : float = 0.5
    var<private> AggroLevel<public> : float = 0.0

    var<private> IsWaveEnemy<public> : logic = false
    var SpawnGroup : ?spawn_group_base = false
    var SpawnLocation<public> : ?spawn_location = false
    var ForcedSpawnLocation<public> : ?spawn_location = false

    var DeathTransform<public> : transform = transform{}
    var DeathAgent : ?agent = false
    var TimeOfDeath<public> : float = -1.0

    var TurnedOwner<public> : ?agent = false
    var LastAgentTarget<public> : ?agent = false
    var LastInstigator<public> : ?agent = false

    var AnimationProp : creative_prop = creative_prop{}
    var MarkerProp : creative_prop = creative_prop{}

    var<private> MoveState<public> : logic = true

    var AllowAttacks<private> : logic = true
    var AllowAttackDamage<private> : logic = true
    var<private> IsAttacking : logic = false
    var<public> AutoTargetOnSpawn<public> : logic = true

    var IsRandomRoaming<private> : logic = false
    var SoundState<private> : logic = true
    var CurrentSound<private> : ?zombie_sound = false

    # EVENTS
    StateChangedEvent<private> : event(zombie_state) = event(zombie_state){}
    CancelStateBranch<public> : event() = event(){} # Cancels the current zombie state activity. Use with caution.
    TargetChangedEvent<public> : event(?agent) = event(?agent){}
    CancelCurrentTargeting<public> : event() = event(){} # Cancels navigation to the current target
    CancelRandomRoam<public> : event() = event(){}
    DamagedEvent<public> : event(?agent) = event(?agent){}
    DeathEvent<public> : event() = event(){}
    RespawnEvent<public> : event() = event(){}
    SpawnedEvent<public> : event() = event(){}
    AttackingEvent<public> : event(?agent) = event(?agent){}
    AttackedEvent<public> : event() = event(){}

    MoveStateChanged<public> : event(logic) = event(logic){}
    
    ExitAttack<public> : event() = event(){}
    CancelAttack<public> : event() = event(){}
    CancelShortcut<public> : event() = event(){}

    NewSoundEvent<public> : event() = event(){}
    StopVFXEvent<public> : event(int) = event(int){}

    # SUBSCRIBES
    var FortCharacterDamageSubscribe : ?cancelable = false

    OnBegin<public><final>()<suspends>:void=
        CurrentZombieList := GlobalZombieList[GetSession()] or map{}
        if. set GlobalZombieList[GetSession()] = ConcatenateMaps(CurrentZombieList, map{Agent => Self})
        else. Print("ERR: Could not globally link a zombie!")
        set StartingHealth = Health
        FortCharacter.SetMaxHealth(999999.0)
        FortCharacter.SetHealth(999999.0)
        DmgEvent:=FortCharacter.DamagedEvent().Subscribe(OnFortCharacterDamaged)
        set FortCharacterDamageSubscribe = option{DmgEvent}
        AwaitGameState(game_state.InGame)
        Sleep(0.1)
        ResetSpeed()
        set StartingMovementSpeed = MovementSpeed
        # Spawn the zombie
        spawn. PlayState(zombie_state.SpawningProcess)
        if(ThisZombieTools:=GlobalZombieTools[GetSession()], ThisZombieTools.SkibidiState?):
            ApplyVFX(ThisZombieTools.SkibidiVFX, 0)
        if(AggroLevel >= Controller.Controls.AggressiveLimit):
            OnAggressiveState()
        # Active functions during lifetime
        branch:
            sync:
                OnZombieBegin()
                UpdateSound()
        DeathEvent.Await()

    # Override to further manipulate this zombies' behavior in parent classes
    OnZombieBegin<public>()<suspends>:void=
        return

    # STATE MACHINE
    PlayState<public>(InState:zombie_state)<suspends>:void=
        if(IsGameState[game_state.Paused]):
            PausedState := State
            set State = zombie_state.Paused
            AwaitGameState(game_state.InGame)
            set State = PausedState
            return
        if(not IsActive[], InState <> zombie_state.SpawningProcess):
            Print("Tried to play a state when zombie is inactive!")
            return
        PreviousState := State
        set State = InState
        StateChangedEvent.Signal(InState)
        CancelStateBranch.Signal()
        #Print("NewZombieState - {InState.ToString()}")
        Sleep(0.0)
        branch:
            case(InState):
                zombie_state.Inactive =>
                    CancelCurrentTargeting.Signal()
                    CancelShortcut.Signal()
                    CancelRandomRoam.Signal()
                    SetSoundState(false)
                    if(Connected_DmgEvent:=FortCharacterDamageSubscribe?):
                        Connected_DmgEvent.Cancel()
                        set FortCharacterDamageSubscribe = false
                    if(FortCharacter.IsActive[]):
                        set DeathTransform = FortCharacter.GetTransform()
                        StopPropAnimation()
                        FortCharacter.Hide()
                        if. FortCharacter.TeleportTo[DEATHPOS, rotation{}]
                        Sleep(0.1)
                    if(ThisRandomDeathFX:=Shuffle(Controller.Animations.DeathFX)[0]):
                        spawn. ThisRandomDeathFX.Play(Self, ?TransformOverride:=option{DeathTransform})
                    FortCharacter.ReleaseFromStasis()
                    FortCharacter.SetVulnerability(true)
                    FortCharacter.SetMaxHealth(1.0)
                    FortCharacter.Damage(999999.0)
                    StateChangedEvent.Signal(zombie_state.Inactive)
                    SetMarker(zombie_marker.None)
                    var CanSpawnAmmoDrop : logic = true
                    if(ThisPowerup:=Controller.Controls.DropPowerupOnDeath.ConvertToPowerup[]):
                        MaybePowerup:=SpawnPowerupOnZombie(Self, ThisPowerup, ?ByChance:=logic{not Controller.Controls.AlwaysDropPowerup?})
                        if(MaybePowerup?):
                            set CanSpawnAmmoDrop = false
                    if(not IsSpawning[], CanSpawnAmmoDrop?, InstigatingPlayerInfo:=DeathAgent?.GetPlayerInfo[], InstigatingPlayerInfo.AmmoDropChance > 0.0, ThisWeaponManager:=GlobalWeaponManager[GetSession()], GetRandomFloat(0.0, 100.0) <= InstigatingPlayerInfo.AmmoDropChance):
                        ThisWeaponManager.AmmoCollection.SpawnAmmoDrop(transform{Translation:=DeathTransform.Translation-vector3{Z:=76.0}})
                    set Controller.LastEliminatedZombie = option{Self}
                    if(ThisSpawnGroup:=SpawnGroup?):
                        ThisSpawnGroup.OnZombieEliminated(Agent)
                    if(ThisZombieTools:=GlobalZombieTools[GetSession()], ThisInstigator:=DeathAgent?):
                        ThisZombieTools.ActivateTrigger_OnZombieEliminated.Trigger(ThisInstigator)
                    set TurnedOwner = false
                    set LastAgentTarget = false
                    set LastInstigator = false
                    Controller.ZombieEliminated.Signal(Agent)
                    DeathEvent.Signal()
                    CancelStateBranch.Signal()
                zombie_state.SpawningProcess =>
                    DragToNull()
                    Sleep(0.0)
                    if(FortCharacter.IsActive[]):
                        if(ThisSpawnLocation := ForcedSpawnLocation?):
                            ThisSpawnLocation.SpawnZombie(Self)
                        else:
                            loop:
                                # Basic spawning method by choosing a random open spawn location
                                if(ThisOpenSpawner:=GetOpenSpawnLocation[Controller.Type]):
                                    if. ThisRoom:=GetRoomViaSpawner[ThisOpenSpawner] then. Print("[{Identity}] zombie found a spawner [{ThisOpenSpawner.Identity}] at room [{ThisRoom.RoomText}]")
                                    else. Print("[{Identity}] zombie found a spawner [{ThisOpenSpawner.Identity}] in an unknown room")
                                    SpawningTask := spawn. ThisOpenSpawner.SpawnZombie(Self)
                                    SpawningTask.Await()
                                    break
                                Sleep(0.5)
                        SpawnedEvent.Signal()
                    else:
                        Eliminate()
                zombie_state.Targeting =>
                    FortCharacter.Show()
                    Show()
                    set LastAgentTarget = false
                    var LastShortcut : ?zombie_shortcut = false
                    var IsShortcutting : logic = false
                    NewShortcutEvent : event(?zombie_shortcut) = event(?zombie_shortcut){}
                    sync:
                        block:
                            # Finds new shortcuts and signals when they're found
                            if(Navigatable:=FortCharacter.GetNavigatable[], Controller.Controls.AllowShortcutting?):
                                # Find a shortcut to travel to the current target quicker
                                loop:
                                    if(TargetFortChar:=LastAgentTarget?.GetFortCharacter[], TargetFortChar.IsActive[], ThisShortcut:=GetUsefulShortcut[TargetFortChar.GetTransform().Translation, Navigatable.GetCurrentDestination[], Agent]):
                                        if(ThisShortcut.IsNotEqual[LastShortcut?] or not LastShortcut?, not IsShortcutting?):
                                            set LastShortcut = option{ThisShortcut}
                                            NewShortcutEvent.Signal(option. ThisShortcut)
                                            Sleep(1.0)
                                    else if(ThisShortcut:=LastShortcut?, not IsShortcutting? or ThisShortcut.IsCancelable?):
                                        Print("Canceling shortcut")
                                        set LastShortcut = false
                                        set IsShortcutting = false
                                        set LastAgentTarget = false
                                        NewShortcutEvent.Signal(false)
                                    Sleep(0.5)
                        loop:
                            Sleep(0.0)
                            # Travels to and uses valid shortcuts
                            race:
                                TargetChangedEvent.Await()
                                NewShortcutEvent.Await()
                                block:
                                    set IsShortcutting = false
                                    CancelShortcut.Signal()
                                    if(ThisShortcut := LastShortcut?):
                                        Print("Found a shortcut to use [{Identity}]")
                                        ThisShortcut.TravelTo(Self)
                                        set IsShortcutting = true
                                        Print("Using shortcut")
                                        ThisShortcut.Use(Self)
                                        set LastShortcut = false
                                        set IsShortcutting = true
                                        set LastAgentTarget = false
                                        CancelShortcut.Signal()
                                    else:
                                        NewShortcutEvent.Await()
                            Sleep(0.25)
                        block:
                            Sleep(1.0)
                            # Respawns the zombie based on distance to the player being targeted
                            if(Controller.Controls.AllowPlayerDistanceBasedRespawn?):
                                loop:
                                    if(ThisTarget:=LastAgentTarget?):
                                        if(Distance(ThisTarget.GetFortCharacter[].GetTransform().Translation, FortCharacter.GetTransform().Translation) > Controller.Controls.MaxDistanceFromPlayer):
                                            if(MoveState?):
                                                Respawn()
                                            else:
                                                MoveStateChanged.Await()
                                            break
                                        Sleep(5.0)
                                    else:
                                        Sleep(1.0)
                        block:
                            if(Controller.Controls.MaximumAFKTime > -1.0):
                                var StartAFKTime : float = 0.0
                                var IsStill : logic = false
                                loop:
                                    if(IsGameState[game_state.Paused]):
                                        set IsStill = false
                                        AwaitGameState(game_state.InGame)
                                    CurrPosition := FortCharacter.GetTransform().Translation
                                    Sleep(0.25)
                                    if(IsAlmostEqual[CurrPosition, FortCharacter.GetTransform().Translation, 0.1]):
                                        if(not IsStill?):
                                            set IsStill = true
                                            set StartAFKTime = GetSimulationElapsedTime()
                                        else if((GetSimulationElapsedTime()-StartAFKTime) > Controller.Controls.MaximumAFKTime):
                                            Respawn()
                                    else:
                                        set IsStill = false
                        loop:
                            # Targets entities (players/zombies)
                            if. IsGameState[game_state.Paused] then. AwaitGameState(game_state.InGame)
                            if(not IsShortcutting?):
                                if(TurnedOwner?):
                                    # Find a zombie target
                                    ZombieList := GetZombieList()
                                    if(ThisClosestZombie:=GetClosestZombie[ZombieList], ThisClosestZombie.Agent <> LastAgentTarget? or not LastAgentTarget?):
                                        set LastAgentTarget = option{ThisClosestZombie.Agent}
                                        spawn. TargetEntity(MakeNavigationTarget(ThisClosestZombie.Agent), Controller.Controls.DistanceFromPlayerToAttack, ?TargetedAgent:=option{ThisClosestZombie.Agent})
                                else:
                                    # Find a player target
                                    PlayerList := GetPlayers()
                                    if(ThisClosestPlayer:=GetClosestPlayer[], ThisClosestPlayer.User <> LastAgentTarget? or not LastAgentTarget?):
                                        CancelRandomRoam.Signal()
                                        set LastAgentTarget = option{ThisClosestPlayer.User}
                                        spawn. TargetEntity(MakeNavigationTarget(ThisClosestPlayer.User), Controller.Controls.DistanceFromPlayerToAttack, ?TargetedAgent:=option{ThisClosestPlayer.User})
                                    else if(not LastAgentTarget?, not IsRandomRoaming?):
                                        # No entities were found, free roam
                                        spawn. RandomRoam()
                            SafeSleep(Controller.Controls.SwitchTargetDelay)
                zombie_state.GameOver =>
                    if(not SpawnLocation?):
                        spawn. RandomRoam()
                    else:
                        Eliminate()
                _ =>
            Sleep(0.0)
        CancelStateBranch.Await()
        CancelRandomRoam.Signal()

    ForceSetState<public>(InState:zombie_state, ?CancelPreviousState:logic=true):void=
        set State = InState
        StateChangedEvent.Signal(InState)
        if. CancelPreviousState? then. CancelStateBranch.Signal()

    GetState<public>()<transacts>:zombie_state=
        State

    # Waits until this zombie is of the desired state
    AwaitState<public>(DesiredState:zombie_state)<suspends>:void=
        if(DesiredState <> State):
            loop:
                NewState := StateChangedEvent.Await()
                if(NewState = DesiredState or NewState = zombie_state.Inactive):
                    break

    # Waits until this zombie is in a different state than the stay state
    AwaitDifferentState<public>(StayState:zombie_state)<suspends>:void=
        if(StayState = State):
            loop:
                NewState := StateChangedEvent.Await()
                if(NewState <> StayState or NewState = zombie_state.Inactive):
                    break

    AwaitStateChanged<public>()<suspends>:void=
        StateChangedEvent.Await()

    SafeNavigate<public>(Target:navigation_target, ?ReachRadius:float=0.0, ?MovementType:movement_type=Walking, ?AllowPartialPath:logic=false, ?CancelEvents:[]event()=array{})<suspends>:logic=
        if(not AnimationProp.IsValid[], MovementSpeed >= Controller.Controls.SprintingSpeed, RandomSpeedAnim:=Shuffle(Controller.GetSkinInfo[SkinTag].SprintList)[0]):
            SetPropAnimation(RandomSpeedAnim)
        var CompletedState : logic = false
        if(Navigatable:=FortCharacter.GetNavigatable[]):
            loop:
                AwaitGameState(game_state.InGame)
                if(not MoveState?):
                    FortCharacter.PutInStasis(stasis_args{AllowFalling:=false, AllowTurning:=false})
                    loop:
                        NewMoveState:=MoveStateChanged.Await()
                        if(NewMoveState?):
                            break
                    FortCharacter.ReleaseFromStasis()
                NavResult:=race:
                    block:
                        loop:
                            Result:=Navigatable.NavigateTo(Target, ?ReachRadius:=ReachRadius, ?MovementType:=MovementType, ?AllowPartialPath:=AllowPartialPath)
                            if(IsGameState[game_state.InGame]):
                                if(Result=navigation_result.Reached):
                                    set CompletedState = true
                                    break
                                else if(Result=navigation_result.Unreachable):
                                    break
                            Sleep(0.0)
                        true
                    block:
                        AwaitGameState(game_state.Paused)
                        false
                    block:
                        loop:
                            NewMoveState:=MoveStateChanged.Await()
                            if(not NewMoveState?):
                                break
                        false
                    block:
                        RaceAll(CancelEvents+array{DeathEvent, RespawnEvent})
                        Print("Cancelling SafeNavigate")
                        true
                if(NavResult?):
                    break
                Navigatable.StopNavigation()
        return CompletedState

    RandomRoam<public>(?MaximumRoamRadius:float=100.0)<suspends>:void=
        if(not IsRandomRoaming?, Navigatable:=FortCharacter.GetNavigatable[]):
            set IsRandomRoaming = true
            CancelCurrentTargeting.Signal()
            defer{
                set IsRandomRoaming = false
            }
            branch:
                loop:
                    NavState:=SafeNavigate(MakeNavigationTarget(GetRandomPointInCircle(FortCharacter.GetTransform().Translation, MaximumRoamRadius)), ?AllowPartialPath:=true, ?MovementType:=if. IsAggro[] then. Running else. Walking)
                    if(not NavState?):
                        CancelRandomRoam.Signal()
            CancelRandomRoam.Await()
            set IsRandomRoaming = false
            set LastAgentTarget = false

    TargetEntity<public>(Target:navigation_target, ReachRadius:float, ?TargetedAgent:?agent=false, ?ForcedLocation:?vector3=false, ?TargetedProp:?creative_prop=false)<suspends>:void=
        var CurrentNavTarget : navigation_target = Target
        if(Navigatable:=FortCharacter.GetNavigatable[]):
            TargetChangedEvent.Signal(TargetedAgent)
            CancelCurrentTargeting.Signal()
            CancelNavigation:event()=event(){}
            branch:
                AwaitGameState(game_state.Paused)
                if(not IsAttacking?):
                    Navigatable.StopNavigation()
                    set LastAgentTarget = false
                    CancelCurrentTargeting.Signal()
            branch:
                if(ThisPlayerInfo:=TargetedAgent?.GetPlayerInfo[]):
                    race:
                        ThisPlayerInfo.LeaveEvent.Await()
                        ThisPlayerInfo.DownedEvent.Await()
                        loop:
                            ThisState:=ThisPlayerInfo.DetectableStateChangedEvent.Await()
                            if(not ThisState?):
                                break
                        loop:
                            NewState := StateChangedEvent.Await()
                            if(NewState <> zombie_state.Targeting):
                                break
                    Navigatable.StopNavigation()
                    set LastAgentTarget = false
                    CancelCurrentTargeting.Signal()
                else if(ZombieInfo:=TargetedAgent?.GetZombieAI[]):
                    race:
                        ZombieInfo.DeathEvent.Await()
                        ZombieInfo.RespawnEvent.Await()
                        loop:
                            NewState := StateChangedEvent.Await()
                            if(NewState <> zombie_state.Targeting):
                                break
                    Navigatable.StopNavigation()
                    set LastAgentTarget = false
                    CancelCurrentTargeting.Signal()
            branch:
                if(ThisProp := TargetedProp?):
                    loop:
                        set CurrentNavTarget = MakeNavigationTarget(ThisProp.GetTransform().Translation)
                        CancelNavigation.Signal()
                        Sleep(0.25)
            branch:
                loop:
                    race:
                        block:
                            NavState:=SafeNavigate(CurrentNavTarget, ?ReachRadius:=ReachRadius+RRFIX, ?CancelEvents:=array{CancelNavigation, RespawnEvent, CancelCurrentTargeting}, ?MovementType:=if. IsAggro[] then. Running else. Walking)
                            if(not NavState?):
                                CancelCurrentTargeting.Signal()
                        loop:
                            if(Distance(TargetedAgent?.GetFortCharacter[].GetTransform().Translation, FortCharacter.GetTransform().Translation) <= ReachRadius):
                                break
                            Sleep(0.5)
                    if(not TargetedAgent? or DistanceXY(TargetedAgent?.GetFortCharacter[].GetTransform().Translation, FortCharacter.GetTransform().Translation) <= (ReachRadius-RRFIX)+3.0):
                        AttackEntity(TargetedAgent, ?ForcedLocation:=ForcedLocation)
                        Sleep(0.1)
                    else:
                        Sleep(0.0)
            CancelCurrentTargeting.Await()
        return

    AttackEntity<public>(TargetedAgent:?agent, ?ForcedLocation:?vector3=false)<suspends>:void=
        if(AllowAttacks?):
            StopPropAnimation()
            defer{
                set IsAttacking = false
            }
            set IsAttacking = true
            AttackingEvent.Signal(TargetedAgent)
            if:
                Navigation:=FortCharacter.GetNavigatable[]
                Focus:=FortCharacter.GetFocusInterface[]
            then:
                MaybeAttackAnim:=GetAndPlayAnimation(zombie_state.Attacking)
                if(ThisAttackPackage:=attack_animation[MaybeAttackAnim?]):
                    AttackManipulators:=ThisAttackPackage.Manipulators.GetManipulators(AggroLevel, Controller.Controls.AggressiveLimit)
                    branch:
                        Navigation.Wait(?Duration:=(ThisAttackPackage.ActiveTime-AttackManipulators(2))/AttackManipulators(0))
                    branch:
                        if(ThisAttackPackage.MaintainFocus?):
                            if(ThisTarget:=TargetedAgent?, OtherFortChar:=ThisTarget.GetFortCharacter[]):
                                race:
                                    loop:
                                        if(OtherFortChar.IsActive[]):
                                            RotationToFace := FortCharacter.GetTransform().Translation.GetRotationToPosition(OtherFortChar.GetTransform().Translation)
                                            if. FortCharacter.TeleportTo[FortCharacter.GetTransform().Translation, RotationToFace]
                                        Sleep(0.0)
                                    SafeSleep(ThisAttackPackage.TimeOfAttack)
                            else if(ThisLocation:=ForcedLocation?):
                                RotationToFace := FortCharacter.GetTransform().Translation.GetRotationToPosition(ThisLocation)
                                if. FortCharacter.TeleportTo[FortCharacter.GetTransform().Translation, RotationToFace] 
                    branch:
                        if(ThisAttackPackage.CancelDistanceFromTarget > 0.0, ThisAttackPackage.CancelDistanceFromTarget > Controller.Controls.DistanceFromPlayerToAttack, ThisTargetFortChar:=TargetedAgent?.GetFortCharacter[]):
                            loop:
                                if(ThisTargetFortChar.IsActive[], Distance(ThisTargetFortChar.GetTransform().Translation, FortCharacter.GetTransform().Translation) > ThisAttackPackage.CancelDistanceFromTarget):
                                    CancelAttack.Signal()
                                    break
                                Sleep(0.25)
                    branch:
                        if(ThisTargetInfo:=TargetedAgent?.GetPlayerInfo[]):
                            ThisTargetInfo.ZombieAttackStarted.Signal(Agent)
                        TOASleep := (ThisAttackPackage.TimeOfAttack-AttackManipulators(2))/AttackManipulators(0)
                        SafeSleep(TOASleep)
                        HitVector := FortCharacter.GetTransform().GetLocalVector(ThisAttackPackage.AttackPointOffset)
                        if(AllowAttackDamage?, ThisTarget:=TargetedAgent?, TargetFortChar:=ThisTarget.GetFortCharacter[], TargetFortChar.IsActive[]):
                            TargetPos:=TargetFortChar.GetTransform().Translation
                            DistanceToTarget := Distance(HitVector, TargetPos) 
                            if:
                                DistanceToTarget <= ThisAttackPackage.AttackRadius
                            then:
                                if:
                                    ThisPlayerInfo:=ThisTarget.GetPlayerInfo[]
                                then:
                                    # Target is player #
                                    if(ThisPlayerInfo.IsActive[]):
                                        if(ThisPlayerInfo.CheckShieldHit[FortCharacter.GetTransform().Translation]):
                                            ThisPlayerInfo.DamageShield(ThisAttackPackage.Damage) 
                                        else:
                                            ThisPlayerInfo.Damage(ThisAttackPackage.Damage, ?Instigator:=option{Agent})
                                        OnPlayerAttacked(ThisPlayerInfo.User, ThisAttackPackage)
                                else:
                                    # Target is zombie #
                                    if(ThisZombieInfo:=ThisTarget.GetZombieAI[]):
                                        ThisZombieInfo.Eliminate()
                                        OnZombieAttack(ThisZombieInfo, ThisAttackPackage)
                                        # Give owning player points #
                                        if(OwnerPlayerInfo:=TurnedOwner?.GetPlayerInfo[]):
                                            OwnerPlayerInfo.IncreasePoints(10, ?ColorOverride:=option{NamedColors.LightGreen})
                            else:
                                #Print("Zombie failed attack, player not in range. {DistanceToTarget}")
                        AttackedEvent.Signal()
                        TimeLeftSleep := (ThisAttackPackage.ActiveTime-ThisAttackPackage.TimeOfAttack-AttackManipulators(2))/AttackManipulators(0)
                        SafeSleep(TimeLeftSleep)
                        ExitAttack.Signal()
                        <#
                        block:
                            TargetList := if. TargetedAgent?.GetPlayerInfo[] then. GetActivePlayers() else if. TargetedAgent?.GetCreatureInfo[] then. GetActiveZombieAgents() else GetActivePlayers()
                            Sleep(ThisAttackPackage.TimeOfAttack)
                            AttackedEvent.Signal()
                            HitVector := ZombieFortChar.GetTransform().GetLocalVector(vector3{X:=95.0})
                            for(ThisTarget:TargetList):
                                if(ThisFortChar:=ThisTarget.GetFortCharacter[], ThisFortChar.IsActive[], ThisTarget<>ZombieAgent):
                                    TargetPos:=ThisFortChar.GetTransform().Translation
                                    if:
                                        Distance(HitVector, TargetPos) <= HITRADIUS
                                    then:
                                        if:
                                            ThisPlayerInfo:=ThisTarget.GetPlayerInfo[]
                                        then:
                                            ThisPlayerInfo.AttackerEvent.Signal(ZombieAgent)
                                            # Target is player #
                                            if(not ThisPlayerInfo.IsDowned?):
                                                if(ThisPlayerInfo.CheckShieldHit[ZombieFortChar.GetTransform(), ThisFortChar.GetTransform()]):
                                                    ThisPlayerInfo.DamageShield(Damage) 
                                                else:
                                                    ThisPlayerInfo.Damage(Damage, ?InstigatingAgent:=option{ZombieAgent})
                                                OnPlayerAttacked(ThisTarget, ThisAttackPackage)
                                        else:
                                            # Target is zombie #
                                            if(ThisCreatureInfo:=ThisTarget.GetCreatureInfo[]):
                                                ThisCreatureInfo.Eliminate()
                                            # Give owning player points #
                                            if(OwnerPlayerInfo:=TurnedOwner?.GetPlayerInfo[]):
                                                OwnerPlayerInfo.IncreasePoints(10, ?PointColor:=option{NamedColors.LightGreen})
                            #>
                            #Sleep(ThisAttackPackage.ActiveTime-ThisAttackPackage.TimeOfAttack)
                    branch:
                        ThisAttackPackage.PlayAnimation(Self, ?AllowManipulation_PlayRate:=true, ?AllowManipulation_BlendInTime:=true, ?AllowManipulation_StartPositionSeconds:=true)
                    race:
                        CancelAttack.Await()
                        ExitAttack.Await()
                    ThisAttackPackage.StopAnimation(Self)

    OnPlayerAttacked<public>(InAgent:agent, ThisAttackPackage:attack_animation):void=
        return

    OnZombieAttack<public>(ThisZombie:zombie_ai, ThisAttackPackage:attack_animation):void=
        return

    UpdateSound<private>()<suspends>:void=
        if(NewSound:=Shuffle(Controller.GetSkinInfo[SkinTag].GetValidSounds(AggroLevel))[0], NewSoundProp:=SpawnProp(NewSound.Blueprint, FortCharacter.GetTransform().Translation, rotation{})(0)?):
            NewSoundEvent.Signal()
            set CurrentSound = option. NewSound
            defer{
                NewSoundProp.Dispose()
            }
            Rot:=rotation{}
            race:
                loop:
                    if(SoundState?):
                        NewSoundProp.MoveTo(FortCharacter.GetTransform().Translation, Rot, 0.125)
                    else:
                        if. NewSoundProp.TeleportTo[vector3{}, rotation{}]
                        Sleep(0.1)
                NewSoundEvent.Await()
                DeathEvent.Await()
            NewSoundProp.Dispose()

    SetSoundState<public>(InState:logic):void=
        set SoundState = InState
        set CurrentSound = false
        if. not InState? then. NewSoundEvent.Signal()

    DragToNull<public>():void=
        defer{
            SetMoveState(true)
        }
        for(Attempt:=0..30):
            if(FortCharacter.TeleportTo[Controller.NullLocation.GetTransform().Translation+vector3{X:=1.0,Y:=1.0,Z:=25.0*Attempt}, rotation{}]):
                return
        for(Attempt:=0..30):
            if(FortCharacter.TeleportTo[Controller.NullLocation.GetTransform().Translation+vector3{X:=125.0,Y:=1.0,Z:=25.0*Attempt}, rotation{}]):
                return
        for(Attempt:=0..30):
            if(FortCharacter.TeleportTo[Controller.NullLocation.GetTransform().Translation+vector3{X:=1.0,Y:=125.0,Z:=25.0*Attempt}, rotation{}]):
                return

    GetAnimation<public>(StateType:zombie_state):?timed_animation=
        var MaybeAnimPackage : ?timed_animation = false
        case(StateType):
            zombie_state.Attacking => 
                if(ThisRandomAttack:=Controller.Animations.GetAttack[AggroLevel]):
                    set MaybeAnimPackage = option{ThisRandomAttack}
            _ =>
        return MaybeAnimPackage

    PlayAnimation<public>(TimedAnimation:timed_animation, AnimationState:zombie_state, ?ForcePlayrate:float=1.0, ?ForceBlendIn:float=0.0, ?ForceBlendOut:float=0.0)<suspends>:void=
        if(AnimationController:=FortCharacter.GetPlayAnimationController[], ThisSequence:=TimedAnimation.AnimationSequence?):
            case(AnimationState):
                zombie_state.Attacking => 
                    if(ThisAttack:=attack_animation[TimedAnimation]):
                        AnimManipulators := ThisAttack.Manipulators.GetManipulators(AggroLevel, Controller.Controls.AggressiveLimit)
                        AnimationController.PlayAndAwait(ThisSequence, ?PlayRate:=AnimManipulators(0), ?StartPositionSeconds:=AnimManipulators(2), ?BlendInTime:=AnimManipulators(1))
                _ =>

    GetAndPlayAnimation<public>(AnimationState:zombie_state, ?ApplyManipulators:logic=true):?timed_animation=
        MaybeAnim:=GetAnimation(AnimationState)
        if(ThisAnim:=MaybeAnim?):
            spawn. ThisAnim.PlayAnimation(Self, ?AllowManipulation_BlendInTime:=ApplyManipulators, ?AllowManipulation_PlayRate:=ApplyManipulators, ?AllowManipulation_StartPositionSeconds:=ApplyManipulators)
            return option{ThisAnim}
        return false

    # MANIPULATORS
    Heal<public>(Amount:float):void=
        set Health += Amount
        if(Health >= Controller.Controls.MaximumHealth):
            set Health = Controller.Controls.MaximumHealth

    # Returns the state of elimination from this damage call
    # True => Zombie was eliminated by this call
    # False => Zombie is still alive after the damage
    Damage<public>(Amount:float, ?Instigator:?agent=false, ?CheckInstigatorWeapon:logic=false, ?IsTrap:logic=false):logic=
        if(IsActive[], Health > 0.0):
            var ThisDamageAmount : float = Amount
            var DmgType : damage_type = damage_type.Body
            if(ThisPlayer := Instigator?.GetPlayerInfo[], CheckInstigatorWeapon?):
                # Get damage from weapon settings
                if(ThisWeapon:=ThisPlayer.User.GetHeldWeapon[]):
                    set ThisDamageAmount = ThisWeapon.GetDamageAmount(ThisPlayer.User, Amount)
                    if(ThisWeapon.IsHeadshot[Amount]):
                        Print("Is headshot")
                        set DmgType = damage_type.Head
                    ThisWeapon.OnZombieHit.Signal(Agent)
                else:
                    # Player damaged via pickaxe or unknown/unset weapon
                    set ThisDamageAmount = 150.0 * ThisPlayer.DamageMultiplier
                    set DmgType = damage_type.Melee
                ThisPlayer.DamagedZombieEvent.Signal(damage_result_custom{Instigator:=option{ThisPlayer.User}, Target:=Agent, DamageType:=DmgType})
                set ThisDamageAmount *= ThisPlayer.GetDamageMultiplier()
            if(IsInstakill[], Controller.Controls.InstakillConversion > 0.0):
                InstakillAdditive := block{
                    if(Controller.Controls.InstakillConversion >= 100.0):
                        Controller.Controls.MaximumHealth
                    else:
                        (Controller.Controls.InstakillConversion*ThisDamageAmount)
                }
                set ThisDamageAmount += InstakillAdditive
            set Health -= ThisDamageAmount
            Print("Zombie damaged for {ThisDamageAmount}. Health: {Health}")
            DamagedEvent.Signal(Instigator)
            OnDamaged()
            if(Health <= 0.0):
                set DeathTransform = FortCharacter.GetTransform()
                if(ThisPlayer := Instigator?.GetPlayerInfo[]):
                    set DeathAgent = option{ThisPlayer.User}
                    set TimeOfDeath = GetSimulationElapsedTime()
                    if(ThisPlayer.User.GetHeldWeapon[].GivePoints_OnElim? or not ThisPlayer.User.GetHeldWeapon[]):
                        ThisPlayer.IncreasePoints(Controller.GetKillReward(DmgType))
                        ThisPlayer.SetTotalKills(ThisPlayer.TotalKills+1)
                        ThisPlayer.EliminatedZombieEvent.Signal(Agent)
                        if. IsTrap? then. ThisPlayer.EliminatedZombieByTrapEvent.Signal(Agent)
                Eliminate(?Instigator:=Instigator)
                return true
            else if(ThisPlayer := Instigator?.GetPlayerInfo[]):
                if(ThisPlayer.User.GetHeldWeapon[].GivePoints_OnDamage? or not ThisPlayer.User.GetHeldWeapon[]):
                    ThisPlayer.IncreasePoints(Controller.GetDamageReward(DmgType))
        return false

    OnDamaged<public>():void=
        return

    OnFortCharacterDamaged<private>(DmgResult:damage_result):void=
        # Print("FortCharacter damaged for {DmgResult.Amount}")
        Instigator := option{DmgResult.Instigator?.GetInstigatorAgent[]}
        Damage(DmgResult.Amount, ?Instigator := Instigator, ?CheckInstigatorWeapon:=logic{Instigator?.GetPlayerInfo[]})

    # Sets the ability of this zombie to attack
    # True - Zombie can attack entities
    # False - Zombie cannot attack entities, but may still target/follow them
    SetAttackPermission<public>(InState:logic):void=
        set AllowAttacks = InState

    # Sets the ability of this zombie to deal damage to players/enemies
    # True - Zombie deals damage to enemies
    # False - Zombie still attacks, but no damage is dealt
    SetDamagePermission<public>(InState:logic):void=
        set AllowAttackDamage = InState

    SetAggroLevel<public>(Amount:float, ?StackMode:logic=false):void=
        PreviousAggroLevel := AggroLevel
        if(StackMode?):
            set AggroLevel += Amount
        else:
            set AggroLevel = Amount
        if(AggroLevel >= Controller.Controls.AggressiveLimit, PreviousAggroLevel < Controller.Controls.AggressiveLimit):
            OnAggressiveState()
        AggSpeedRate := Controller.Controls.AggressionToSpeedRate
        if(Controller.Controls.AggressionToSpeedRate > 0.0, ThisRageSpeed:=Clamp(Lerp(Controller.Controls.StartingSpeed, Controller.Controls.MaximumSpeed, (AggroLevel*AggSpeedRate)/(Controller.Controls.AggressiveLimit*AggSpeedRate)), MovementSpeed, Controller.Controls.MaximumSpeed)):
            Print("NewRageSpeed: {ThisRageSpeed}")
            SetSpeed(ThisRageSpeed)
        if(not Controller.GetSkinInfo[SkinTag].GetValidSounds(AggroLevel).Find[CurrentSound?] or not CurrentSound?):
            spawn. UpdateSound()
        
    OnAggressiveState<public>():void=
        Print("Zombie is aggressive")
        SetSpeed(Controller.Controls.MaximumSpeed)

    SetSpeed<public>(Multiplier:float):void=
        ThisSpeed := if. Multiplier > Controller.Controls.MaximumSpeed then. Controller.Controls.MaximumSpeed else. Multiplier
        set MovementSpeed = ThisSpeed
        if(Navigatable:=FortCharacter.GetNavigatable[]):
            Navigatable.SetMovementSpeedMultiplier(ThisSpeed)
        else. Print("ERR: Failed to fetch navigatable in a zombie!")
        if(not AnimationProp.IsValid[], IsState[zombie_state.Targeting], not IsAttacking?, ThisSpeed >= Controller.Controls.SprintingSpeed, RandomSpeedAnim:=Shuffle(Controller.GetSkinInfo[SkinTag].SprintList)[0]):
            SetPropAnimation(RandomSpeedAnim)
            
    ResetSpeed<public>():void=
        #Print("ZombieSpeed: {Lerp(Controller.Controls.StartingSpeed, Controller.Controls.MaximumSpeed, AggroLevel/Controller.Controls.AggressiveLimit)}")
        SetSpeed(LockLerp(Controller.Controls.StartingSpeed, Controller.Controls.MaximumSpeed, AggroLevel/Controller.Controls.AggressiveLimit))
        #if(ThisWaveController:=GlobalWaveController[GetSession()]):
            #SetSpeed(ThisWaveController.GetSpeedValue())

    SetMoveState<public>(InState:logic):void=
        set MoveState = InState
        MoveStateChanged.Signal(InState)
        if(InState?):
            FortCharacter.ReleaseFromStasis()
        else:
            FortCharacter.PutInStasis(stasis_args{AllowFalling:=false, AllowTurning:=false})

    SetDamageAmount<public>(Amount:float):void=
        set DamageAmount = Amount

    ResetDamage<public>():void=
        set DamageAmount = Controller.Controls.Damage

    Eliminate<public>(?Instigator:?agent=false):void=
        if(IsActive[]):
            if. ThisInstigator := Instigator? then. set DeathAgent = option. ThisInstigator
            spawn. PlayState(zombie_state.Inactive)
        spawn. OnEliminated()

    OnEliminated<public>()<suspends>:void=
        return

    Respawn<public>(?ClearForcedRoom:logic=true):void=
        defer{
            RespawnEvent.Signal()
        }
        if(IsActive[]):
            Print("Zombie [{Identity}] respawning")
            if. ClearForcedRoom? then. set ForcedSpawnLocation = false
            set SpawnLocation = false
            CancelCurrentTargeting.Signal()
            CancelRandomRoam.Signal()
            StopPropAnimation()
            spawn. PlayState(zombie_state.SpawningProcess)

    EnableTurnedMode<public>(InAgent:agent):void=
        set TurnedOwner = option{InAgent}

    SetMarker<public>(InMarkerType:zombie_marker):void=
        var NewMarkerProp : creative_prop = creative_prop{}
        var CustomOffset : vector3 = vector3{Z:=98.0}
        if(ZombieCollection:=GlobalZombieTools[GetSession()]):
            case(InMarkerType):
                zombie_marker.None =>
                    NewMarkerEvent.Signal()
                    MarkerProp.Dispose()
                zombie_marker.Skull =>
                    if(SpawnedSkullMarker:=SpawnProp(ZombieCollection.SkullMarker, FortCharacter.GetTransform().Translation, rotation{})(0)?):
                        set NewMarkerProp = SpawnedSkullMarker
                        set CustomOffset = vector3{Z:=120.0}
                zombie_marker.Turned =>
                    if(SpawnedTurnedMarker:=SpawnProp(ZombieCollection.TurnedMarker, FortCharacter.GetTransform().Translation, rotation{})(0)?):
                        set NewMarkerProp = SpawnedTurnedMarker
                zombie_marker.Objective =>
                    if(SpawnedObjectiveMarker:=SpawnProp(ZombieCollection.ObjectiveMarker, FortCharacter.GetTransform().Translation, rotation{})(0)?):
                        set NewMarkerProp = SpawnedObjectiveMarker
                zombie_marker.Tracked =>
                    if(SpawnedTrackedMarker:=SpawnProp(ZombieCollection.TrackedMarkerBP, FortCharacter.GetTransform().Translation, rotation{})(0)?):
                        set NewMarkerProp = SpawnedTrackedMarker
        else:
            Print("ERR: Could not fetch zombie collection")
        if(NewMarkerProp.IsValid[]):
            spawn. TrackMarker(NewMarkerProp, InMarkerType, ?Offset:=CustomOffset)
        else if(InMarkerType <> zombie_marker.None):
            Print("ERR: Failed to spawn a marker prop!")

    NewMarkerEvent:event()=event(){}
    TrackMarker(InMarkerProp:creative_prop, MarkerType:zombie_marker, ?Offset:vector3=vector3{Z:=98.0})<suspends>:void=
        defer{
            InMarkerProp.Dispose()
        }
        NewMarkerEvent.Signal()
        MarkerProp.Dispose()
        set MarkerProp = InMarkerProp
        if(IsActive[] or MarkerType = zombie_marker.Tracked):
            race:
                NewMarkerEvent.Await()
                DeathEvent.Await()
                loop:
                    if(FortCharacter.IsActive[]):
                        ThisTransform:=FortCharacter.GetTransform()
                        spawn{InMarkerProp.MoveTo(ThisTransform.Translation+Offset, rotation{}, 0.09365)}
                    Sleep(0.1)
        else:
            Print("ERROR: Could not set a marker on a zombie due to their fort character being inactive!")
            MarkerString := case(MarkerType):
                zombie_marker.Skull => "Skull"
                zombie_marker.None => "None"
                zombie_marker.Objective => "Objective"
                zombie_marker.Tracked => "Tracked"
                zombie_marker.Turned => "Turned"
            Print("Failed zombie marker application with type {MarkerString}")
        InMarkerProp.Dispose()

    NewPropAnimEvent<private>:event()=event(){}
    SetPropAnimation<public>(InAsset:creative_prop_asset, ?Offset:vector3=vector3{}, ?RotationOffset:float=0.0):void=
        if(SpawnedAsset:=SpawnProp(InAsset, FortCharacter.GetTransform().Translation, FortCharacter.GetTransform().Rotation)(0)?):
            spawn. TrackPropAnimation(SpawnedAsset, ?Offset:=Offset, ?RotationOffset:=RotationOffset)
            FortCharacter.Hide()

    StopPropAnimation<public>():void=
        NewPropAnimEvent.Signal()
        FortCharacter.Show()

    TrackPropAnimation<private>(InAnimationProp:creative_prop, ?Offset:vector3=vector3{}, ?RotationOffset:float=0.0)<suspends>:void=
        defer{
            InAnimationProp.Dispose()
            FortCharacter.Show()
        }
        NewPropAnimEvent.Signal()
        AnimationProp.Dispose()
        set AnimationProp = InAnimationProp
        if(IsActive[]):
            race:
                NewPropAnimEvent.Await()
                DeathEvent.Await()
                loop:
                    if(FortCharacter.IsActive[]):
                        ThisTransform:=FortCharacter.GetTransform()
                        spawn{InAnimationProp.MoveTo(ThisTransform.Translation+Offset, ThisTransform.Rotation.ApplyYaw(RotationOffset), 0.09365)}
                    Sleep(0.1)
        InAnimationProp.Dispose()
        FortCharacter.Show()

    GetGroundTransform<public>()<transacts>:transform=
        FortCharTransform := if. FortCharacter.IsActive[] then. FortCharacter.GetTransform() else. transform{}
        transform{
            Translation := FortCharTransform.Translation-vector3{Z:=40.0}
            Rotation := FortCharTransform.Rotation
        }

    GetClosestPlayer<public>()<transacts><decides>:player_info=
        var MaybeClosestPlayerTuple : tuple(?player_info, float) = (false, 0.0)
        for:
            ThisPlayer:GetPlayers()
            ThisPlayer.IsDetectableByZombies[]
            FortChar:=ThisPlayer.GetFortCharacter[]
            ZombieTranslation:=FortCharacter.GetTransform().Translation
            DistanceToZombie:=Distance(ZombieTranslation, FortChar.GetTransform().Translation) 
            DistanceToZombie < MaybeClosestPlayerTuple(1) or not MaybeClosestPlayerTuple(0)?
        do:
            set MaybeClosestPlayerTuple = (option{ThisPlayer}, DistanceToZombie)
        MaybeClosestPlayerTuple(0)?

    GetClosestZombie<public>(ZombieList:[]zombie_ai)<transacts><decides>:zombie_ai=
        var MaybeClosestZombieTuple : tuple(?zombie_ai, float) = (false, 0.0)
        for:
            ThisZombie:ZombieList
            ZombieTranslation:=FortCharacter.GetTransform().Translation
            ThisZombie.Controller.Controls.CanBeTargetedByTurned?
            DistanceToZombie:=Distance(ZombieTranslation, FortCharacter.GetTransform().Translation) 
            DistanceToZombie < MaybeClosestZombieTuple(1) or not MaybeClosestZombieTuple(0)?
        do:
            set MaybeClosestZombieTuple = (option{ThisZombie}, DistanceToZombie)
        MaybeClosestZombieTuple(0)?

    IsActive<public>()<transacts><decides>:void=
        State <> zombie_state.Inactive

    # Is the zombie currently in the playable area of the map ?
    IsInMap<public>()<transacts><decides>:void=
        State = zombie_state.Targeting or State = zombie_state.Attacking or State = zombie_state.Shortcutting or State = zombie_state.Stunned

    IsState<public>(InState:zombie_state)<transacts><decides>:void=
        State = InState

    IsAggro<public>()<transacts><decides>:void=
        AggroLevel > Controller.Controls.AggressiveLimit

    IsSpawning<public>()<transacts><decides>:spawn_location=
        SpawnLocation?

    Hide<public>():void=
        AnimationProp.Hide()
        FortCharacter.Hide()

    Show<public>():void=
        if(AnimationProp.IsValid[]):
            AnimationProp.Show()
        else:
            FortCharacter.Show()

    ApplyVFX<public>(InVFXPowerup:visual_effect_powerup_device, Channel:int):void=
        spawn. VFXLoop(InVFXPowerup, Channel)

    StopVFX<public>(Channel:int):void=
        StopVFXEvent.Signal(Channel)

    VFXLoop<private>(InVFXPowerup:visual_effect_powerup_device, InChannel:int)<suspends>:void=
        StopVFXEvent.Signal(InChannel)
        branch:
            loop:
                InVFXPowerup.Pickup(Agent)
                Sleep(1.0)
        race:
            loop:
                ThisChannel:=StopVFXEvent.Await()
                if(ThisChannel = InChannel):
                    break
            DeathEvent.Await()

(ZombieList:[]zombie_ai).Eliminate<public>():void=
    for(ThisZombie:ZombieList):
        ThisZombie.Eliminate()