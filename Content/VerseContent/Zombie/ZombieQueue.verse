using { /Verse.org/Simulation }

# Basic queue for zombies
zombie_queue<public> := class:

    # ZombieAgent -> Popped Event
    # If pop event = true, the zombie is on the Queue
    # If pop event = false, the zombie is not on the Queue
    var<private> ZombieQueueMap : [agent]?event() = map{}

    # Adds the zombie to the queue and awaits for its removal
    # Doesn't complete until the zombie is popped from the queue OR the zombie queue is already empty
    LatchOn<public><final>(ZombieInfo:zombie_ai)<suspends>:void=
        RefreshQueue()
        EmptyCheck := logic{GetQueueSize() <= 0}
        ThisRemoveEvent:=Add(ZombieInfo)
        DebugPrint("Entering LatchOn queue : [{ZombieInfo.Identity}]")
        if(not EmptyCheck?):
            ThisRemoveEvent.Await()
            DebugPrint("Exiting LatchOn : [{ZombieInfo.Identity}]")
        else:
            DebugPrint("Exiting LatchOn starting from top : [{ZombieInfo.Identity}]")

    # Adds a zombie to the bottom of the queue
    Add<public><final>(ZombieInfo:zombie_ai):event()=
        if(ThisEvent := ZombieQueueMap[ZombieInfo.Agent]?):
            ThisEvent
        else:
            RemoveEvent : event() = event(){}
            if. set ZombieQueueMap[ZombieInfo.Agent] = option{RemoveEvent}
            else. set ZombieQueueMap = ConcatenateMaps(ZombieQueueMap, map{ZombieInfo.Agent => option{RemoveEvent}})
            return RemoveEvent

    # Removes the zombie at the top of the queue
    # Returns the popped zombie, if it exists
    Pop<public><final>(?FireRemoveEvent:logic=true):?zombie_ai=
        DebugPrint("Pop called")
        if:
            TopQueueSlot := GetQueuedZombies()[0]
            RemovalEvent := ZombieQueueMap[TopQueueSlot.Agent]?
            set ZombieQueueMap[TopQueueSlot.Agent] = false
        then:
            RefreshQueue()
            DebugPrint("Pop success on zombie [{TopQueueSlot.Identity}]")
            if. FireRemoveEvent? then. RemovalEvent.Signal()
            QueuedZombieList:=GetQueuedZombies().Length
            if(QueuedZombieList <= 0):
                set ZombieQueueMap = map{}
            else if:
                NextQueueSlot := GetQueuedZombies()[0]
                NextRemovalEvent := ZombieQueueMap[NextQueueSlot.Agent]?
            then:
                DebugPrint("Signaled exit on zombie [{NextQueueSlot.Identity}]")
                NextRemovalEvent.Signal()
            else:
                DebugPrint("ERR: Failed to signal exit on next queued zombie!")
            return option{TopQueueSlot}
        else:
            DebugPrint("ERR: Failed to Remove a zombie info from a zombie queue!") 
        return false

    # Forcefully pops the zombie off of the queue
    ForcePop<public><final>(ZombieInfo:zombie_ai):void=
        if:
            QueueIndex := GetQueuedZombies().FindInstance[ZombieInfo]
            RemovalEvent := ZombieQueueMap[ZombieInfo.Agent]?
            set ZombieQueueMap[ZombieInfo.Agent] = false
        then:
            RefreshQueue()
            RemovalEvent.Signal()
            QueuedZombieList:=GetQueuedZombies()
            if(QueuedZombieList.Length <= 0):
                set ZombieQueueMap = map{}
            else if:
                QueueIndex = 0
                TopRemoveEvent := ZombieQueueMap[QueuedZombieList[0].Agent]?
            then:
                TopRemoveEvent.Signal()
        else:
            DebugPrint("ERR: Failed to ForcePop a zombie off of a queue!")


    # Pops all zombies off the queue
    PopAll<public><final>():[]zombie_ai=
        QueuedZombieList := GetQueuedZombies()
        PoppedZombies:=for:
            QueueSlot:QueuedZombieList
            RemovalEvent:=ZombieQueueMap[QueueSlot.Agent]?
        do:
            RemovalEvent.Signal()
            QueueSlot
        set ZombieQueueMap = map{}
        PoppedZombies

    RefreshQueue<public><final>():void=
        var NewQueueMap : [agent]?event() = map{}
        for(ThisZombieAgent->MaybeRemoveEvent:ZombieQueueMap):
            if(ThisZombieAgent.GetZombieAI[].IsActive[], ThisRemoveEvent:=MaybeRemoveEvent?):
                set NewQueueMap = ConcatenateMaps(NewQueueMap, map{ThisZombieAgent => option. ThisRemoveEvent})
        set ZombieQueueMap = NewQueueMap

    # Fetch all queued zombies
    GetQueuedZombies<public><final>()<transacts>:[]zombie_ai=
        QueuedZombies:=for:
            ThisZombieAgent->MaybeExitEvent:ZombieQueueMap
            MaybeExitEvent?
            ThisZombieInfo:=ThisZombieAgent.GetZombieAI[]
        do:
            ThisZombieInfo

    GetQueueSize<public><final>()<transacts>:int=
        QueuedZombieAmount:=for:
            ThisZombieAgent->MaybeExitEvent:ZombieQueueMap
            MaybeExitEvent?
            ThisZombieAgent.GetZombieAI[]
        do:
            0
        return QueuedZombieAmount.Length

    GetQueueIndex<public><final>(ZombieInfo:zombie_ai)<transacts><decides>:int=
        GetQueuedZombies().FindInstance[ZombieInfo]

    GetQueueSlot<public><final>(Index:int)<transacts><decides>:zombie_ai=
        GetQueuedZombies()[Index]

    # Fetch the removal event for a zombie, if in the queue
    GetRemovedEvent<public><final>(ZombieInfo:zombie_ai)<transacts><decides>:event()=
        ZombieQueueMap[ZombieInfo.Agent]?

    # Checks if the zombie is in the queue
    IsInQueue<public><final>(ZombieInfo:zombie_ai)<transacts><decides>:void=
        ZombieQueueMap[ZombieInfo.Agent]?

    IsEmpty<public><final>()<transacts><decides>:void=
        GetQueuedZombies().Length <= 0
            
    var DebugState<public> : logic = false
    DebugPrint(String:string):void=
        if(DebugState?):
            Print(String)