using { /Verse.org/Simulation }
using { /Fortnite.com/Devices }
using { /Fortnite.com/Characters }
using { /Fortnite.com/Game }
using { /Verse.org/Assets }
using { /Verse.org/Random }
using { /UnrealEngine.com/Temporary/SpatialMath }
using { /Fortnite.com/Animation/PlayAnimation }

CreateIdentity<public>()<transacts>:int=
    return block{
        ZombieTools := GlobalZombieTools[GetSession()]
        set ZombieTools.TotalZombieSpawnedCount += 1
        ZombieTools.TotalZombieSpawnedCount 
    } or 1

var GlobalZombieTools<public> : weak_map(session, zombie_tools) = map{}

zombie_tools<public> := class(creative_device):

    var TotalZombieSpawnedCount : int = 0

    @editable
    ActivateTrigger_OnZombieEliminated : trigger_device = trigger_device{}

    @editable
    OnTrigger_SkullMarkWaveZombies : trigger_device = trigger_device{}

    @editable
    SkullMarker<public> : creative_prop_asset = DefaultCreativePropAsset

    @editable
    TurnedMarker<public> : creative_prop_asset = DefaultCreativePropAsset

    @editable
    ObjectiveMarker<public> : creative_prop_asset = DefaultCreativePropAsset

    @editable
    TrackedMarkerBP<public> : creative_prop_asset = DefaultCreativePropAsset

    @editable
    ShortcutProp<public> : creative_prop_asset = DefaultCreativePropAsset

    @editable
    NavHelpProp<public> : creative_prop_asset = DefaultCreativePropAsset

    @editable
    PlagueSummoningAnim<public> : creative_prop_asset = DefaultCreativePropAsset

    @editable
    PoisonVFX<public> : creative_prop_asset = DefaultCreativePropAsset

    @editable
    SkibidiVFX<public> : visual_effect_powerup_device = visual_effect_powerup_device{}

    SkibidiStateChanged<public> : event(logic) = event(logic){}

    var<private> SkibidiState<public> : logic = false

    OnBegin<override>():void=
        if. set GlobalZombieTools[GetSession()] = Self
        else. Print("ERROR: Failed to globally link zombie tools")
        OnTrigger_SkullMarkWaveZombies.TriggeredEvent.Subscribe(MarkAllWave)

    MarkAllWave<private>(Agent:?agent):void=
        ThisWaveList := GetZombieList(?OnlyWave:=true)
        for(ThisZombie:ThisWaveList):
            ThisZombie.SetMarker(zombie_marker.Skull)

    SetSkibidiState<public>(InState:logic):void=
        set SkibidiState = InState
        ZombieList := GetZombieList()
        if(InState?):
            for(ThisZombie:ZombieList):
                ThisZombie.ApplyVFX(SkibidiVFX, 0)
        else:
            for(ThisZombie:ZombieList):
                ThisZombie.StopVFX(0)
