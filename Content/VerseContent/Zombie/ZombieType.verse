using { /Verse.org/Simulation }
using { /Fortnite.com/Devices }
using { /Fortnite.com/Characters }
using { /Fortnite.com/Game }
using { /Verse.org/Assets }
using { /Verse.org/Random }
using { /UnrealEngine.com/Temporary/SpatialMath }
using { /Fortnite.com/Animation/PlayAnimation }
using { Spawning }
using { Powerups }
using { Weapons }
using { InteractiveModule }
using { ZombieTypes }

# ZOMBIE TYPE LIST

# Add more zombie type enums here as needed
zombie_type_list<public> := enum{
    Default
    Seeker
    Objective

    Plague
    Summoned
}

GetZombieTypes<public>()<transacts>:[]zombie_type=
    GlobalZombieTypes[GetSession()] or array{}

(ZombieTypeList:[]zombie_type_list).HasType<public>(DesiredType:zombie_type_list)<transacts><decides>:void=
    var HasType : logic = false
    for:
        ThisZombieType:ZombieTypeList
        ThisZombieType = DesiredType
    do:
        set HasType = true
    HasType?

(ZombieTypeList:[]zombie_type_list).DoesNotHaveType<public>(DesiredType:zombie_type_list)<transacts><decides>:void=
    var HasType : logic = false
    for:
        ThisZombieType:ZombieTypeList
        ThisZombieType <> DesiredType
    do:
        set HasType = true
    not HasType?

(DesiredType:zombie_type_list).GetController<public>()<transacts><decides>:zombie_type=
    var MaybeZombieController : ?zombie_type = false
    ZombieTypeList := GlobalZombieTypes[GetSession()]
    for:
        ThisZombieType : ZombieTypeList
        ThisZombieType.Type = DesiredType
    do:
        set MaybeZombieController = option{ThisZombieType}
    MaybeZombieController?

# Holds all zombie type devices
var GlobalZombieTypes : weak_map(session, []zombie_type) = map{}

zombie_type<public> := class(creative_device):

    # SETTINGS
    @editable Name<public> : string = "Default"
    @editable Type<public> : zombie_type_list = zombie_type_list.Default
    @editable Controls<public> : zombie_type_controls = zombie_type_controls{}
    @editable Animations<public> : zombie_type_animations = zombie_type_animations{}
    @editable Spawners<public> : []zombie_type_spawner = array{}
    @editable NullLocation<public> : creative_prop = creative_prop{}

    # VARIABLES
    var<private> SpawnedZombies<public> : []zombie_ai = array{}
    SpawnQueue<private> : slot_queue = slot_queue{}

    var LastSpawnLocation<public> : ?spawn_location = false
    var LastEliminatedZombie<public> : ?zombie_ai = false

    # EVENTS
    ZombieSpawned<public> : event(zombie_ai) = event(zombie_ai){}
    ZombieDamaged<public> : event(damage_result_custom) = event(damage_result_custom){}
    ZombieEliminated<public> : event(agent) = event(agent){}

    # STARTUP
    OnBegin<override>()<suspends>:void=
        if:
            not GlobalZombieTypes[GetSession()]
            set GlobalZombieTypes[GetSession()] = array{}
        if:
            CurrentZombieTypesList := GlobalZombieTypes[GetSession()]
            set GlobalZombieTypes[GetSession()] = CurrentZombieTypesList + array{Self}
        else. Print("ERROR: Failed to globally connect a Zombie Type device!")
    # SPAWN HANDLING
    SpawnAndAwait<public><final>(?SpawnGroup:?spawn_group_base=false, ?ForceSpawnLocation:?spawn_location=false, ?OverrideWaveEnemy:tuple(logic, logic)=(false, false), ?OverrideAggression:tuple(logic, float)=(false, -1.0))<suspends>:?zombie_ai=
        if(GetActiveZombies().Length >= (20*Spawners.Length)):
            Print("Maximum zombies AAT reached")
            ZombieEliminated.Await()
            Print("Zombie spawning after awaiting in AAT")
        Sleep(0.0)
        SpawnQueue.LatchOn()
        defer{
            SpawnQueue.Pop()
        }
        var Attempts : int = 0
        loop:
            if(ThisRandomSkinSpawner := Shuffle(Spawners)[0]):
                MaybeZombieAgent := SpawnFromSpawner(ThisRandomSkinSpawner)
                if(SpawnedZombieAgent := MaybeZombieAgent?):
                    loop:
                        if(SpawnedZombieAgent.GetFortCharacter[].IsActive[]):
                            break
                        Print("Waiting for zombie fort character")
                        Sleep(0.0)
                    if:
                        ThisFortCharacter:=SpawnedZombieAgent.GetFortCharacter[]
                        ThisWaveDevice:=GlobalWaveController[GetSession()]
                    then:
                        UpdatedStartHealth := GetZombieHealthIncrease()+Controls.StartingHealth
                        UpdatedAggression := if. OverrideAggression(0)? then. OverrideAggression(1) else. ThisWaveDevice.GetAggresionPercent()
                        CountedAsWaveEnemy := if. OverrideWaveEnemy(0)? then. OverrideWaveEnemy(1) else. Controls.CountAsWaveEnemy
                        var NewAIClass : zombie_ai = zombie_ai{Controller:=Self, Agent:=SpawnedZombieAgent, FortCharacter:=ThisFortCharacter, SkinTag:=ThisRandomSkinSpawner.SkinTag, Identity:=CreateIdentity(), Health:=UpdatedStartHealth, SpawnGroup:=SpawnGroup, ForcedSpawnLocation:=ForceSpawnLocation, IsWaveEnemy:=CountedAsWaveEnemy, AggroLevel:=UpdatedAggression}
                        # Change class if needed
                        case(Type):
                            zombie_type_list.Seeker => set NewAIClass = seeker_ai{Controller:=Self, Agent:=SpawnedZombieAgent, FortCharacter:=ThisFortCharacter, SkinTag:=ThisRandomSkinSpawner.SkinTag, Identity:=CreateIdentity(), Health:=UpdatedStartHealth, SpawnGroup:=SpawnGroup, ForcedSpawnLocation:=ForceSpawnLocation, IsWaveEnemy:=CountedAsWaveEnemy, AggroLevel:=UpdatedAggression}
                            zombie_type_list.Objective => set NewAIClass.AutoTargetOnSpawn = false
                            zombie_type_list.Plague => set NewAIClass = plague_ai{Controller:=Self, Agent:=SpawnedZombieAgent, FortCharacter:=ThisFortCharacter, SkinTag:=ThisRandomSkinSpawner.SkinTag, Identity:=CreateIdentity(), Health:=UpdatedStartHealth, SpawnGroup:=SpawnGroup, ForcedSpawnLocation:=ForceSpawnLocation, IsWaveEnemy:=CountedAsWaveEnemy, AggroLevel:=UpdatedAggression}
                            _ =>
                        set SpawnedZombies += array{NewAIClass}
                        ZombieSpawned.Signal(NewAIClass)
                        spawn. NewAIClass.OnBegin()
                        return option{NewAIClass}
                else if(Attempts > 5):
                    Print("Maximum zombies spawned, waiting for an open NPC slot")
                    ZombieEliminated.Await()
                else if(Attempts > 10):
                    break
                else:
                    Sleep(0.25)
                    set Attempts += 1
            else:
                Print("ERROR: No skin spawners set up in zombie type {Name}!")
                return false
        return false

    SpawnFromSpawner<private><final>(SkinSpawner:zombie_type_spawner)<suspends>:?agent=
        race:
            block:
                Sleep(0.1)
                SkinSpawner.NPCSpawner.Spawn()
                Sleep(1.1)
            block:
                ThisZombieAgent:=SkinSpawner.NPCSpawner.SpawnedEvent.Await()
                return option{ThisZombieAgent}
        return false

    # GETTERS
    GetActiveZombies<public><final>()<transacts>:[]zombie_ai=
        ActiveZombieList:=for:
            ZombieInfo:SpawnedZombies
            ZombieInfo.IsActive[]
        do:
            ZombieInfo

    GetZombieHealthIncrease<public>(?MaximumHealth:?float=false)<transacts>:float=
        var ThisHealthIncrease : float = 0.0
        ThisWave := GetWaveCount()
        if(ThisWave<>1):
            if(ThisWave < 10):
                set ThisHealthIncrease = (100.0 * (ThisWave-1))
            else:
                set ThisHealthIncrease = (950.0 * Pow(Controls.WaveHealthRate, (ThisWave-9)*1.0)) - 150.0
        MaxHP := MaximumHealth? or Controls.MaximumHealth
        if(ThisHealthIncrease < MaxHP):
            return ThisHealthIncrease
        else:
            return MaxHP

    GetSpeedValue<public>(?InZombieInfo:?zombie_ai=false, ?ForceAggressive:logic=false, ?NoModify:logic=false):float=
        if(ThisWaveDevice:=GlobalWaveController[GetSession()]):
            ThisWave := GetWaveCount()
            ThisZombieCap := ThisWaveDevice.GetWaveZombieCap()
            var RandomSpeed : float = (Pow(Log(10.0, (ThisWave*17.0)/(1.0*ThisWaveDevice.Controls.AggressiveWaveMaximum)), 2.0)+0.486)*(Controls.MaximumSpeed/2.0)
            if(ThisZombieInfo:=InZombieInfo?, not NoModify?):
                AggroZombieCount := GetZombieList(?OnlyAggro:=true).Length
                if(ThisWave >= ThisWaveDevice.Controls.AggressiveWaveStart, ThisZombieCap*(Controls.AggressiveLimit/100.0) > AggroZombieCount*1.0):
                    set RandomSpeed = Controls.MaximumSpeed
                    ThisZombieInfo.SetAggroLevel(100.0)
            if(RandomSpeed > Controls.MaximumSpeed):
                return Controls.MaximumSpeed
            if(RandomSpeed < 0.5):
                return 0.5
            return RandomSpeed
        return 0.5

    GetDamageReward<public>(InType:damage_type)<transacts>:int=
        case(InType):
            damage_type.Body => Controls.PointSettings.Damage_BodyShot_Reward
            damage_type.Head => Controls.PointSettings.Damage_HeadShot_Reward
            damage_type.Melee => Controls.PointSettings.Damage_Melee_Reward

    GetKillReward<public>(InType:damage_type)<transacts>:int=
        case(InType):
            damage_type.Body => Controls.PointSettings.Eliminating_BodyShot_Reward
            damage_type.Head => Controls.PointSettings.Eliminating_HeadShot_Reward
            damage_type.Melee => Controls.PointSettings.Eliminating_Melee_Reward

    KillAll<public>():void=
        for(ThisZombie:SpawnedZombies):
            ThisZombie.Eliminate()

    # TEMPORARY - remove when animation sequences work in editable
    GetAnimationSequence<public>(AnimType:animation_sequence_type)<transacts><decides>:animation_sequence=
        FalseAnimSequence : ?animation_sequence = false
        AnimToReturn:=case(Type):
            _ => # DEFAULT case - Use for default type as well
                case(AnimType):
                    animation_sequence_type.BasicSpawn1 => option. Art.Zombie_Models.Zombies.Animations.Default.ZombieAnimation_GroundSpawn_1 # -302 Z axis
                    animation_sequence_type.BasicAttack1 => option. Art.Zombie_Models.Zombies.Animations.Default.ZombieAnimation_Attack_1
                    animation_sequence_type.BasicVault1 => option. Art.Zombie_Models.Zombies.Animations.Default.Vaulting.ZombieAnimation_Vaulting_1
                    animation_sequence_type.BasicLeapStart1 => option. Zombie_Pack.Animations.InPlace.Jumps.IP_StartJump_01
                    animation_sequence_type.BasicLeapLoop1 => option. Zombie_Pack.Animations.InPlace.Jumps.IP_LoopJump_01
                    _ => FalseAnimSequence
        AnimToReturn?

    GetSkinInfo<public>(InSkinTag:string)<transacts><decides>:zombie_type_spawner=
        var MaybeSkinInfo : ?zombie_type_spawner = false
        for:
            ThisSkin:Spawners
            ThisSkin.SkinTag = InSkinTag
            not MaybeSkinInfo?
        do:
            set MaybeSkinInfo = option{ThisSkin}
        MaybeSkinInfo?

zombie_type_spawner<public> := class<concrete>:

    SkinTag_TT<localizes>:message="The string that will represent this skin type"
    @editable{ToolTip:=SkinTag_TT}
    SkinTag : string = "Default"

    NPCSpawner_TT<localizes>:message="The NPC Spawner that is set up with\nthe NPC definition of this skin type"
    @editable{ToolTip:=NPCSpawner_TT}
    NPCSpawner : npc_spawner_device = npc_spawner_device{}

    @editable
    SprintList : []creative_prop_asset = array{}

    @editable
    StunnedList<public> : []creative_prop_asset = array{}

    @editable
    SoundList : []zombie_sound = array{}

    @editable
    InAirAnimation : creative_prop_asset = DefaultCreativePropAsset

    GetValidSounds<public>(InAggressionLevel:float)<transacts>:[]zombie_sound=
        for:
            ThisSound:SoundList
            ThisSound.MinimumAggressionLevel <= InAggressionLevel
            ThisSound.MaximumAggressionLevel >= InAggressionLevel
        do:
            ThisSound

    GetSoundProp<public>(InAggressionLevel:float)<transacts><decides>:creative_prop_asset=
        var ValidSounds : []creative_prop_asset = array{}
        for:
            ThisSound:SoundList
            ThisSound.MinimumAggressionLevel <= InAggressionLevel
            ThisSound.MaximumAggressionLevel >= InAggressionLevel
        do:
            ThisSound.Blueprint
        Shuffle(ValidSounds)[0] or Shuffle(SoundList)[0].Blueprint

    GetStunnedProp<public>()<transacts><decides>:creative_prop_asset=
        Shuffle(StunnedList)[0]

zombie_sound<public> := class<concrete><unique>:

    @editable
    Blueprint : creative_prop_asset = DefaultCreativePropAsset

    @editable
    MinimumAggressionLevel : type{_X:float where _X >= 0.0, _X <= 100.0} = 0.0

    @editable
    MaximumAggressionLevel : type{_X:float where _X >= 0.0, _X <= 100.0} = 0.0

zombie_type_controls<public> := class<concrete>:

    @editable StartingHealth<public> : float = 150.0
    @editable MaximumHealth<public> : float = 12000.0
    @editable WaveHealthRate<public> : float = 1.1
    @editable Damage<public> : float = 35.0
    @editable var StartingSpeed : float = 0.5
    @editable MaximumSpeed<public> : type{_X : float where 0.5 <= _X} = 1.6
    @editable SprintingSpeed<public> : type{_X : float where _X >= 0.5, _X <= 2.0} = 1.2
    #@editable SpeedMultiplier : float = 1.0
    #@editable MaxDistanceToSpawnFromPlayer<public> : float = 30.0
    @editable RespawnDelay<public> : type{_X : float where _X >= 0.25} = 0.25
    @editable SwitchTargetDelay<public> : float = 0.8

    instakillconversion_tooltip<localizes>:message="Set to 100 to instakill, and 0 to void instakill\nAny other value will multiply the incoming weapon damage"
    @editable{ToolTip:=instakillconversion_tooltip} InstakillConversion<public> : type{_X : float where 0.0 <= _X, 100.0 >= _X} = 100.0

    maxdistancefromplayer_tooltip<localizes>:message="When a zombie is at this distance or further from any player, the zombie is marked as 'out of range'"
    @editable{ToolTip:=maxdistancefromplayer_tooltip} MaxDistanceFromPlayer<public> : type{_X : float where 25.0 <= _X} = 5000.0

    @editable AllowPlayerDistanceBasedRespawn<public> : logic = true

    @editable var DistanceFromPlayerToAttack<public> : float = 100.0

    @editable MaximumAFKTime<public> : type{_X : float where _X >= -1.0} = 15.0

    @editable CanBeAggressive<public> : logic = true

    # When does this zombie become fully aggressive ?
    @editable AggressiveLimit<public> : type{_X : float where _X >= 0.0, _X <= 100.0} = 100.0

    @editable var AggressionToSpeedRate<public> : type{_X : float where _X >= 0.0} = 0.5

    @editable CanBeTurned<public> : logic = true

    @editable CanBeTargetedByTurned<public> : logic = true

    @editable CanBeNuked<public> : logic = true

    @editable AllowShortcutting<public> : logic = true

    #@editable var IgnoreMaxSpawnLimit<public> : logic = false

    @editable CountAsWaveEnemy<public> : logic = true

    @editable EliminateAtWaveEnd<public> : logic = true

    @editable DropPowerupOnDeath<public> : powerup_type = powerup_type.RandomRegular
    @editable AlwaysDropPowerup<public> : logic = false

    @editable AmmoDropChance<public> : type{_X:float where _X > 0.0} = 5.0

    @editable PointSettings<public> : zombie_type_pointsinfo = zombie_type_pointsinfo{}

    @editable WaveZombieCaps<public> : []float = array{}

zombie_type_pointsinfo<public> := class<concrete>:

    @editable
    Damage_BodyShot_Reward<public> : int = 10

    @editable
    Damage_HeadShot_Reward<public> : int = 20

    @editable
    Damage_Melee_Reward<public> : int = 10

    @editable
    Eliminating_BodyShot_Reward<public> : int = 70

    @editable
    Eliminating_HeadShot_Reward<public> : int = 90

    @editable
    Eliminating_Melee_Reward<public> : int = 120

zombie_type_animations<public> := class<concrete>:
    
    @editable ZOffset<public> : float = 0.0
    @editable Spawns<public> : []spawn_animation = array{}
    @editable Attacks<public> : []attack_animation = array{}
    @editable Vaulting<public> : []offset_animation = array{}
    @editable Leaps<public> : []leap_animation = array{}
    @editable DeathFX<public> : []timed_fx = array{}

    GetAttack<public>(AggressionLevel:float)<transacts><decides>:attack_animation=
        ValidAttacks:=for:
            ThisAttack:Attacks
            ThisAttack.EnableOnAggressionLevel <= AggressionLevel
        do:
            ThisAttack
        Shuffle(ValidAttacks)[0]

# ANIMATION CLASSES

attack_animation<public> := class(timed_animation):

    @editable
    AttackPointOffset<public> : vector3 = vector3{X:=95.0}

    @editable
    AttackRadius<public> : float = 100.0

    @editable
    TimeOfAttack<public> : float = 0.0

    @editable
    Damage<public> : float = 35.0

    @editable
    MaintainFocus<public> : logic = true

    @editable
    CancelDistanceFromTarget<public> : float = -1.0

    @editable
    EnableOnAggressionLevel<public> : type{_X:float where _X >= 0.0, _X <= 100.0} = 0.0

leap_animation<public> := class(offset_animation):

    @editable
    InAirAnimation<public> : timed_animation = timed_animation{}
    
spawn_animation<public> := class(offset_animation):

    @editable
    StartOffset<public> : vector3 = vector3{}

    PlayAnimation<override>(SpawnPos:transform, ZombieAI:zombie_ai, ?CustomFXPos:?transform=false, ?CustomEndOffset:?vector3=false, ?AllowManipulation_PlayRate:logic=false, ?AllowManipulation_BlendInTime:logic=false, ?AllowManipulation_StartPositionSeconds:logic=false)<suspends>:void=
        ZombieAI.SetMoveState(false)
        if(ZombieAI.FortCharacter.TeleportTo[SpawnPos.GetLocalVector(StartOffset), SpawnPos.Rotation]):
            if:
                AnimControl := ZombieAI.FortCharacter.GetPlayAnimationController[]
                ThisSequence:=ZombieAI.Controller.GetAnimationSequence[AnimationType]
            then:
                ManipulatorPackage := Manipulators.GetManipulators(ZombieAI.AggroLevel, ZombieAI.Controller.Controls.AggressiveLimit)
                branch:
                    if(SpawnedFXAsset:=SpawnProp(FXAsset?, FX_Offset+(CustomFXPos?.Translation or SpawnPos.Translation), CustomFXPos?.Rotation or SpawnPos.Rotation)(0)?):
                        defer{
                            SpawnedFXAsset.Dispose()
                        }
                        Sleep(ActiveTime/(if. AllowManipulation_PlayRate? then. ManipulatorPackage(0) else. 1.0))
                    Sleep(0.0)
                AnimInstance:=AnimControl.Play(ThisSequence, 
                    ?PlayRate:=if. AllowManipulation_PlayRate? then. ManipulatorPackage(0) else. 1.0, 
                    ?BlendInTime:=if. AllowManipulation_BlendInTime? then. ManipulatorPackage(1) else. 0.25, 
                    ?BlendOutTime:=0.2,
                    ?StartPositionSeconds:=if. AllowManipulation_StartPositionSeconds? then. ManipulatorPackage(2) else. 0.0)
                ZombieAI.SetMoveState(false)
                SafeSleep(ActiveTime)
                AnimInstance.Stop()
            else. Print("ERROR: Failed to play zombie animation")
        else. Print("Failed to start spawn animation at start offset!")
        if. ZombieAI.FortCharacter.TeleportTo[SpawnPos.GetLocalVector(if. ThisOffset:=CustomEndOffset? then. ThisOffset else. EndOffset), SpawnPos.Rotation]
        else. Print("Failed to teleport to spawn animation end offset properly")
        ZombieAI.SetMoveState(true)

offset_animation<public> := class(timed_animation):

    @editable 
    EndOffset<public> : vector3 = vector3{}

    PlayAnimation<public>(FromOffset:transform, ZombieAI:zombie_ai, ?CustomEndOffset:?vector3=false)<suspends>:void=
        if:
            AnimControl := ZombieAI.FortCharacter.GetPlayAnimationController[]
            ThisSequence:=ZombieAI.Controller.GetAnimationSequence[AnimationType]
        then:
            AnimInst:=AnimControl.Play(ThisSequence)
            ZombieAI.SetMoveState(false)
            SafeSleep(ActiveTime)
            AnimInst.Stop()
            if. ZombieAI.FortCharacter.TeleportTo[FromOffset.GetLocalVector(if. ThisOffset:=CustomEndOffset? then. ThisOffset else. EndOffset), FromOffset.Rotation]
        else:
            Print("ERROR: Failed to play zombie animation")
        ZombieAI.SetMoveState(true)

timed_animation<public> := class<concrete>:

    @editable
    Manipulators<public> : blending_limits = blending_limits{}

    ActiveTime_tooltip<localizes>:message="The total duration this animation plays for"
    @editable{ToolTip:=ActiveTime_tooltip} ActiveTime<public> : float = -1.0

    @editable
    AnimationSequence<public>:?animation_sequence=false

    # TEMPORARY - remove when animation sequences work in editable
    @editable
    AnimationType<public>:animation_sequence_type=animation_sequence_type.None

    @editable
    FXAsset<public> : ?creative_prop_asset = false

    @editable
    FX_Offset<public> : vector3 = vector3{}

    var AnimationInstancesPerZombie<public> : [agent]?play_animation_instance = map{}

    PlayAnimation<public>(ZombieAI:zombie_ai, ?AllowManipulation_PlayRate:logic=false, ?AllowManipulation_BlendInTime:logic=false, ?AllowManipulation_StartPositionSeconds:logic=false, ?Loop:logic=false, ?TransformOverride:?transform=false)<suspends>:void=
        #ZombieAI.SetMoveState(false)
        if:
            AnimControl := ZombieAI.FortCharacter.GetPlayAnimationController[]
            ThisSequence:=ZombieAI.Controller.GetAnimationSequence[AnimationType]
        then:
            #Print("Animation sequence is valid!")
            ManipulatorPackage := Manipulators.GetManipulators(ZombieAI.AggroLevel, ZombieAI.Controller.Controls.AggressiveLimit)
            branch:
                if(SpawnedFXAsset:=SpawnProp(FXAsset?, FX_Offset+(TransformOverride?.Translation or ZombieAI.FortCharacter.GetTransform().Translation), TransformOverride?.Rotation or ZombieAI.FortCharacter.GetTransform().Rotation)(0)?):
                    defer{
                        SpawnedFXAsset.Dispose()
                    }
                    Sleep(ActiveTime/(if. AllowManipulation_PlayRate? then. ManipulatorPackage(0) else. 1.0))
                Sleep(0.0)
            loop:
                AnimInstance:=AnimControl.Play(ThisSequence, 
                    ?PlayRate:=if. AllowManipulation_PlayRate? then. ManipulatorPackage(0) else. 1.0, 
                    ?BlendInTime:=if. AllowManipulation_BlendInTime? then. ManipulatorPackage(1) else. 0.25, 
                    ?BlendOutTime:=0.2,
                    ?StartPositionSeconds:=if. AllowManipulation_StartPositionSeconds? then. ManipulatorPackage(2) else. 0.0)
                defer{
                    if. set AnimationInstancesPerZombie[ZombieAI.Agent] = false
                }
                if. set AnimationInstancesPerZombie[ZombieAI.Agent] = option{AnimInstance}
                else. set AnimationInstancesPerZombie = ConcatenateMaps(AnimationInstancesPerZombie, map{ZombieAI.Agent => option{AnimInstance}})
                if. AnimInstance.IsPlaying[] then. AnimInstance.Await()
                if(not Loop?):
                    break
            #ZombieAI.SetMoveState(true)
        else:
            #ZombieAI.SetMoveState(true)

    StopAnimation<public>(ZombieAI:zombie_ai):void=
        if(ThisAnimationInstance := AnimationInstancesPerZombie[ZombieAI.Agent]?):
            if. set AnimationInstancesPerZombie[ZombieAI.Agent] = false
            ThisAnimationInstance.Stop()

timed_fx<public> := class<concrete>:

    @editable
    ActiveTime<public> : float = 0.0

    @editable 
    FXAsset<public> : creative_prop_asset = DefaultCreativePropAsset

    @editable
    Offset<public> : vector3 = vector3{}

    Play(ZombieAI:zombie_ai, ?TransformOverride:?transform=false)<suspends>:void=
        if(SpawnedFXAsset:=SpawnProp(FXAsset, Offset+(TransformOverride?.Translation or ZombieAI.FortCharacter.GetTransform().Translation), TransformOverride?.Rotation or ZombieAI.FortCharacter.GetTransform().Rotation)(0)?):
            defer{
                SpawnedFXAsset.Dispose()
            }
            Sleep(ActiveTime)

blending_limits<public> := class<concrete>:

    @editable
    Minimum_Playrate : type{_X:float where _X >= 0.0} = 1.0

    @editable
    Maximum_Playrate : type{_X:float where _X >= 0.0} = 1.0

    @editable
    Minimum_BlendInTime : type{_X:float where _X >= 0.0} = 0.0

    @editable
    Maximum_BlendInTime : type{_X:float where _X >= 0.0} = 0.0

    @editable
    Minimum_StartTime : type{_X:float where _X >= 0.0} = 0.0

    @editable
    Maximum_StartTime : type{_X:float where _X >= 0.0} = 0.0

    GetManipulators<public>(AggroLevel:float, MaximumAggroLevel:float):tuple(float,float,float)=
        LerpPercent:=AggroLevel/MaximumAggroLevel
        return (
            LockLerp(Minimum_Playrate, Maximum_Playrate, LerpPercent),          # (0) = PlayRate
            LockLerp(Maximum_BlendInTime, Minimum_BlendInTime, LerpPercent),    # (1) = BlendInTime
            LockLerp(Minimum_StartTime, Maximum_StartTime, LerpPercent)         # (2) = StartTime
        )


# TEMPORARY - remove when animation sequences work in editable
animation_sequence_type<public> := enum{
    None
    BasicSpawn1
    BasicAttack1
    BasicVault1
    BasicLeapStart1
    BasicLeapLoop1
}