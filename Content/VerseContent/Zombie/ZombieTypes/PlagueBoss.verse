using { /Fortnite.com/AI }
using { /Fortnite.com/AI/movement_types }
using { /Fortnite.com/Devices }
using { /Fortnite.com/Characters }
using { /Fortnite.com/Game }
using { /Fortnite.com/UI }
using { /Fortnite.com/FortPlayerUtilities }
using { /Verse.org/Simulation }
using { /Verse.org/Random }
using { /Verse.org/Colors }
using { /Verse.org/Assets }
using { /UnrealEngine.com/Temporary/Diagnostics }
using { /UnrealEngine.com/Temporary/SpatialMath }
using { /UnrealEngine.com/Temporary/UI }
using { /Fortnite.com/Animation/PlayAnimation }
using { Player }
using { Spawning }

# Spawns a plague every 6 rounds
LoopSpawnPlague<public>()<suspends>:void=
    if(ThisWaveDevice:=GlobalWaveController[GetSession()], PlagueController:=zombie_type_list.Plague.GetController[]):
        FirstAppearance := case(GetGameDifficulty()):
            game_difficulty.Easy => 13
            game_difficulty.Classic => 11
            game_difficulty.Hardcore => 9
        WaveRespawnDelay := case(GetGameDifficulty()):
            game_difficulty.Easy => 7
            game_difficulty.Classic => 6
            game_difficulty.Hardcore => 5
        # Spawns on the first seeker wave
        AwaitWave(FirstAppearance)
        loop:
            if(not GetZombieList(?TypeFilter:=array{zombie_type_list.Plague})[0]):
                SafeSleep(3.0)
                PlagueController.SpawnAndAwait()
            else:
                Print("A plague is already spawned!")
            SafeSleep(3.0)
            if(SpawnedPlagueZombie := GetZombieList(?TypeFilter:=array{zombie_type_list.Plague})[0]):
                SpawnedPlagueZombie.DeathEvent.Await()
            # Spawns every 5 waves
            for(X:=0..WaveRespawnDelay-1):
                ThisWaveDevice.WaveStartedEvent.Await()

plague_ai<public> := class(zombie_ai):

    OnZombieBegin<override>()<suspends>:void=
        AwaitState(zombie_state.Targeting)
        sync:
            loop:
                SafeSleep(5.0)
                SpawnPoisonTrail()
                SafeSleep(15.0)
            loop:
                SafeSleep(15.0)
                SummonZombies()

    SpawnPoisonTrail<public>(?Amount:int=5, ?SpawnRate:float=2.0)<suspends>:void=
        if(ThisFortChar:=Agent.GetFortCharacter[]):
            for(X := 0..Amount):
                NewPoisonInst : poison_inst = poison_inst{Parent:=Self, Location:=ThisFortChar.GetTransform().Translation}
                spawn. NewPoisonInst.Spawn()
                SafeSleep(SpawnRate)

    SummonZombies<public>()<suspends>:void=
        if(ThisFortChar:=Agent.GetFortCharacter[]):
            if(GetZombieList(?TypeFilter:=array{zombie_type_list.Plague}).Length >= 20):
                if(ThisSummonedController:=zombie_type_list.Summoned.GetController[]):
                    ThisSummonedController.ZombieEliminated.Await()
            if(IsAttacking?):
                AttackedEvent.Await()
            SpawnerList:=GetAllSpawners(zombie_type_list.Summoned)
            SpawnPositions:=for:
                SpawnPosIndex := 0..2
            do:
                SafeSleep(1.5)
                ThisFortChar.GetTransform()
            SafeSleep(1.5)
            SetMoveState(false)
            CancelCurrentTargeting.Signal()
            for(Index->SpawnPosition:SpawnPositions):
                if(ThisSpawner:=SpawnerList[Index]):
                    set ThisSpawner.SpawningPosition = transform{Translation:=SpawnPosition.Translation, Rotation:=SpawnPosition.Rotation}
                    spawn. ThisSpawner.ForceSpawnZombie(zombie_type_list.Summoned)
                else:
                    Print("ERR: Failed to fetch summoned spawner")
            if(PlagueSummonAsset:=GlobalZombieTools[GetSession()].PlagueSummoningAnim):
                SetPropAnimation(PlagueSummonAsset, ?Offset:=vector3{Z:=-77.0})
                Sleep(4.5) # -> Animation time
                StopPropAnimation()
            AwaitGameState(game_state.InGame)
            SetMoveState(true)
            spawn. PlayState(zombie_state.Targeting)
            Print("Targeting player")
            Sleep(0.5)

    OnEliminated<override>():void=
        DeathPoison : poison_inst = poison_inst{Parent:=Self, Location:=DeathTransform.Translation}
        spawn. DeathPoison.Spawn()

PoisonDebugDraw := class(debug_draw_channel) {}

var PlayersInPoisonMap : weak_map(session, [agent]logic) = map{}

POISON_RADIUS : float = 75.0
POISON_DAMAGE : float = 15.0
POISON_TICKRATE : float = 1.5
POISON_LIFETIME : float = 6.0
poison_inst<public> := class:

    Parent<public> : plague_ai
    Location<public> : vector3

    var<public> Dmg : float = POISON_DAMAGE
    var<public> Lifetime : float = POISON_LIFETIME

    Spawn<public>()<suspends>:void=
        if:
            not PlayersInPoisonMap[GetSession()]
            set PlayersInPoisonMap[GetSession()] = map{}
        if:
            SpawnedPoisonInstance := SpawnProp(GlobalZombieTools[GetSession()].PoisonVFX, Location, rotation{})(0)?
        then:
            branch:
                loop:
                    PlayerList := GetAgents()
                    for(ThisPlayer:PlayerList):
                        if:
                            not PlayersInPoisonMap[GetSession()][ThisPlayer]
                            ThisMap := PlayersInPoisonMap[GetSession()]
                            set PlayersInPoisonMap[GetSession()] = ConcatenateMaps(ThisMap, map{ThisPlayer => false})
                        if:
                            not PlayersInPoisonMap[GetSession()][ThisPlayer]?
                            Distance(ThisPlayer.GetFortCharacter[].GetTransform().Translation, Location) < POISON_RADIUS
                            set PlayersInPoisonMap[GetSession()][ThisPlayer] = true
                        then:
                            spawn. OnPlayerEnteredPoison(ThisPlayer)
                    Sleep(0.1)
            ##PoisonDebugDrawer : debug_draw = debug_draw{Channel := PoisonDebugDraw}
            #PoisonDebugDrawer.DrawSphere(Location, ?Radius:=POISON_RADIUS, ?DrawDurationPolicy:=debug_draw_duration_policy.FiniteDuration, ?Duration:=POISON_LIFETIME)
            SafeSleep(POISON_LIFETIME)
            SpawnedPoisonInstance.Dispose()

    OnPlayerEnteredPoison<private>(InAgent:agent)<suspends>:void=
        if(ThisPlayerInfo:=InAgent.GetPlayerInfo[], ThisFortChar:=InAgent.GetFortCharacter[]):
            branch:
                loop:
                    ThisPlayerInfo.Damage(POISON_DAMAGE)
                    SafeSleep(POISON_TICKRATE)
            race:
                loop:
                    if(Distance(ThisFortChar.GetTransform().Translation, Location) > POISON_RADIUS):
                        break
                    Sleep(0.1)
                ThisPlayerInfo.LeaveEvent.Await()
                ThisPlayerInfo.DownedEvent.Await()
        if. set PlayersInPoisonMap[GetSession()][InAgent] = false