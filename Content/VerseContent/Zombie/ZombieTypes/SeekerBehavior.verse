using { /Fortnite.com/AI }
using { /Fortnite.com/AI/movement_types }
using { /Fortnite.com/Devices }
using { /Fortnite.com/Characters }
using { /Fortnite.com/Game }
using { /Fortnite.com/UI }
using { /Fortnite.com/FortPlayerUtilities }
using { /Verse.org/Simulation }
using { /Verse.org/Random }
using { /Verse.org/Colors }
using { /Verse.org/Assets }
using { /UnrealEngine.com/Temporary/Diagnostics }
using { /UnrealEngine.com/Temporary/SpatialMath }
using { /UnrealEngine.com/Temporary/UI }
using { /Fortnite.com/Animation/PlayAnimation }

seeker_ai<public> := class(zombie_ai):

    OnZombieBegin<override>()<suspends>:void=
        Print("Seeeker AI on begin")
        SetSpeed(1.25)
        TrackSpeedIncrease()

    TrackSpeedIncrease()<suspends>:void=
        StartingSpeed : float = MovementSpeed
        loop:
            race:
                AttackingEvent.Await()
                block:
                    var CurrSpeed : float = MovementSpeed
                    loop:
                        Sleep(1.0)
                        set CurrSpeed *= 1.5
                        if(CurrSpeed > Controller.Controls.MaximumSpeed):
                            set CurrSpeed = Controller.Controls.MaximumSpeed
                            SetSpeed(CurrSpeed)
                            break
                        SetSpeed(CurrSpeed)
                    Sleep(Inf)
            SetSpeed(StartingSpeed)
        
<#

    InPlayerViewEvent:event()=event(){}
    OutOfPlayerViewEvent:event()=event(){}
    TrackPlayerView()<suspends>:void=
        if(ZombieFortChar:=GetFortCharacter[]):
            var WasLookedAt : logic = false
            OverrideAnimation.Signal(AnimationType.OverrideMovement)
            loop:
                var IsBeingLookedAt : logic = false
                PlayerList:=GetPlayers()
                for(ThisPlayer:PlayerList):
                    if(ThisFortChar:=ThisPlayer.GetFortCharacter[], ThisFortChar.IsLookingAt[ZombieFortChar.GetTransform().Translation, 50.0] and Distance(ThisFortChar.GetTransform().Translation, GetTransform().Translation) < 1000.0):
                        set IsBeingLookedAt = true
                if(IsBeingLookedAt?):
                    if(not WasLookedAt?):
                        spawn{FreezeDelay(ZombieFortChar)}
                        set CanDespawn = false
                        set WasLookedAt = true
                        StopTargetingPlayersEvent.Signal()
                    InPlayerViewEvent.Signal()
                else:
                    if(WasLookedAt?):
                        ZombieFortChar.ReleaseFromStasis()
                        SpawnAnimation(MovementState.Moving)
                        set WasLookedAt = false
                        spawn{AllowDespawnDelay()}
                        StartTargetingPlayersEvent.Signal()
                    OutOfPlayerViewEvent.Signal()
                Sleep(0.0)

    FreezeDelay(InFortChar:fort_character)<suspends>:void=
        race:
            OutOfPlayerViewEvent.Await()
            block:
                Sleep(0.15)
                InFortChar.PutInStasis(stasis_args{AllowFalling:=false,AllowTurning:=false})
                SpawnAnimation(MovementState.Idle)

    AllowDespawnDelay()<suspends>:void=
        race:
            InPlayerViewEvent.Await()
            block:
                Sleep(1.5)
                set CanDespawn = true

    TrackSpeedIncrease()<suspends>:void=
        StartingSpeed : float = Speed
        loop:
            race:
                InPlayerViewEvent.Await()
                block:
                    var CurrSpeed : float = Speed
                    loop:
                        Sleep(1.0)
                        set CurrSpeed *= 1.75
                        if(CurrSpeed > ZombieTypeInfo.Controls.MaximumSpeed):
                            set CurrSpeed = ZombieTypeInfo.Controls.MaximumSpeed
                            SetSpeed(CurrSpeed)
                            break
                        SetSpeed(CurrSpeed)
                    #Print("~Reached max speed")
                    Sleep(Inf)
            SetSpeed(StartingSpeed)
            OutOfPlayerViewEvent.Await()

#>