using { /Fortnite.com/Devices }
using { /Fortnite.com/Characters }
using { /Fortnite.com/Game }
using { /Fortnite.com/UI }
using { /Fortnite.com/FortPlayerUtilities }
using { /Verse.org/Simulation }
using { /Verse.org/Random }
using { /Verse.org/Colors }
using { /UnrealEngine.com/Temporary/Diagnostics }
using { /UnrealEngine.com/Temporary/SpatialMath }
using { /UnrealEngine.com/Temporary/UI }
using { /Verse.org/Colors/NamedColors }
using { Player }
using { InteractiveModule }

AmmoStation := class(Interactive_Base):

    InteractiveType<override> : interactable_type = interactable_type.Ammo

    @editable AmmoCost : int = 2500
    @editable PowerPunched_AmmoCost : int = 7500
    @editable OnAmmoPurchased_ActivateAudio : audio_player_device = audio_player_device{}
    @editable
    InteractiveLocations : []Interactable = array{}

    OnBegin<override>():void=
        LinkAndCompile(InteractiveLocations)

    WhenInsideInteractionRadius<override>(InAgent:agent, InInteractingLocation:Interactable)<suspends>:void=
        if(ThisWeapon:=InAgent.GetHeldWeapon[]):
            if(ThisWeapon.IsInInventory[InAgent]):
                if(ThisWeapon.HasMaxAmmo[InAgent]):
                    InAgent.RefreshAmmoCount()

    GetInteractionText<override>(InAgent:agent, Interactive:Interactable):string=
        if(InPlayerInfo:=InAgent.GetPlayerInfo[], ThisWeapon:=InAgent.GetHeldWeapon[]):
            if(not ThisWeapon.HasMaxAmmo[InAgent]):
                SetInteractableState(true, ?InAgent:=option{InAgent})
                if(ThisWeapon.DoesHaveVariant[InAgent]):
                    ThisAmmoCost := if. ThisWeapon.AmmoStation_PowerPunchedCostOverride > -1 then. ThisWeapon.AmmoStation_PowerPunchedCostOverride else. PowerPunched_AmmoCost
                    return "Purchase Upgraded ammo [<C_Y>${ThisAmmoCost}<E>]"
                else:
                    ThisAmmoCost := if. ThisWeapon.AmmoStation_CostOverride > -1 then. ThisWeapon.AmmoStation_CostOverride else. AmmoCost
                    return "Purchase ammo [<C_Y>${ThisAmmoCost}<E>]"
            else:
                SetInteractableState(false, ?InAgent:=option{InAgent})
                return "you have maximum ammo"
        else:
            SetInteractableState(false, ?InAgent:=option{InAgent})
            return "Requires a weapon"

    OnInteracted<override>(InAgent:agent, InteractableLoc:Interactable):void=
        var Cost : int = 0
        if(ThisWeapon:=InAgent.GetHeldWeapon[]):
            if(ThisWeapon.DoesHaveVariant[InAgent]):
                set Cost = if. ThisWeapon.AmmoStation_PowerPunchedCostOverride > -1 then. ThisWeapon.AmmoStation_PowerPunchedCostOverride else. PowerPunched_AmmoCost
            else:
                set Cost = if. ThisWeapon.AmmoStation_CostOverride > -1 then. ThisWeapon.AmmoStation_CostOverride else. AmmoCost
            if(ThisPlayerInfo:=InAgent.GetPlayerInfo[]):
                BuyState:=ThisPlayerInfo.BuyItem(Cost, ?PlayPurchaseSound:=false)
                if(BuyState?):
                    ThisWeapon.GrantMaxAmmo(InAgent)
                    OnAmmoPurchased_ActivateAudio.Play(InAgent)
        else:
            return