using { /Fortnite.com/Devices }
using { /Fortnite.com/Characters }
using { /Fortnite.com/Game }
using { /Fortnite.com/UI }
using { /Fortnite.com/FortPlayerUtilities }
using { /Verse.org/Simulation }
using { /Verse.org/Random }
using { /Verse.org/Colors }
using { /UnrealEngine.com/Temporary/Diagnostics }
using { /UnrealEngine.com/Temporary/SpatialMath }
using { /UnrealEngine.com/Temporary/UI }
using { /Fortnite.com/AI }
using { /Verse.org/Assets }
using { Player }
using { UI }
using { Zombie }
using { Spawning }
using { InteractiveModule }
using { Powerups }

(InAgent:agent).IsPowerPunching()<transacts><decides>:void=
    GlobalPackAPunch[GetSession()].PackingAgent? = InAgent

var GlobalPackAPunch : weak_map(session, PackAPunch) = map{}

PackAPunch := class(Powerable):

    InteractiveType<override> : interactable_type = interactable_type.Weapon

    @editable PrimaryCost : int = 5000
    @editable SecondaryCost : int = 2500
    @editable DamageMultiplier : float = 2.0
    @editable ActivateTrigger_OnActivation : trigger_device = trigger_device{}
    @editable CenterLocation : creative_prop = creative_prop{}
    @editable SuckInTime : type{_X : float where 0.5 <= _X} = 0.5
    @editable AwaitTime : type{_X : float where 0.5 <= _X} = 2.0
    @editable TimeToCollect : type{_X : float where 5.0 <= _X} = 8.0

    @editable ActivatedVFX : creative_prop_asset = DefaultCreativePropAsset
    @editable VFX_Lifetime : float = 5.0
    @editable ActivatedSFX : audio_player_device = audio_player_device{}

    @editable VariantAbilities : variant_ability_module = variant_ability_module{}

    @editable Switch_FirstPowerPunchUse : switch_device = switch_device{}

    @editable var InteractiveLocation : Interactable = Interactable{Identity:="PackAPunch", InteractTime:=0.8, InteractionSpriteOffset:=vector3{X:=25.0}}

    var<private> PackedState : logic = false
    var<private> PackingAgent : ?agent = false

    OnPowerableBegin<override>():void=
        if(set GlobalPackAPunch[GetSession()] = Self){}
        else{Print("~ERROR: Failed to globally link pack a punch!")}
        LinkAndCompile(array. InteractiveLocation)

    GetPowerInteractionText<override>(InAgent:agent, Interactive:Interactable):string=
        if(not PackingAgent?):
            if(HeldWeapon:=InAgent.GetHeldWeapon[], not InAgent.IsUsingItemCollection[]):
                if(HeldWeapon.PowerPunchEnabled?):
                    if(HeldWeapon.DoesHaveVariant[InAgent]):
                        if(HeldWeapon.DoublePunchEnabled?):
                            SetInteractableState(true, ?InAgent:=option{InAgent})
                            return "Re-Upgrade Weapon [<C_Y>${SecondaryCost}<E>]"
                        else:
                            SetInteractableState(false, ?InAgent:=option{InAgent})
                            return "<C_R>Cannot upgrade weapon again!<E>"
                    else:
                        SetInteractableState(true, ?InAgent:=option{InAgent})
                        return "Upgrade Weapon [<C_Y>${PrimaryCost}<E>]"
                else:
                    SetInteractableState(false, ?InAgent:=option{InAgent})
                    return "<C_R>Cannot upgrade weapon<E>"
            else:
                SetInteractableState(false, ?InAgent:=option{InAgent})
                return "<C_R>Requires a valid weapon<E>"
        else:
            if(PackedState? and InAgent = PackingAgent?):
                SetInteractableState(true, ?InAgent:=option{InAgent})
                return "Grab Upgraded Weapon"
            else:
                SetInteractableState(false)
                return ""

    GetDescription<override>(InAgent:agent, Interactive:Interactable):tuple(string, string)=
        return ("Gives your weapon a damage boost", "power punch")

    OnInteracted<override>(InAgent:agent, InInteractingLocation:Interactable):void=
        if(ThisPlayerInfo:=InAgent.GetPlayerInfo[]):
            if(not PackingAgent?):
                # Pack a punch not active yet #
                if(HeldWeapon:=InAgent.GetHeldWeapon[]):
                    if(HeldWeapon.DoesHaveVariant[InAgent]):
                        # Weapon is already pack a punched, re-pack it! #
                        BuyState:=ThisPlayerInfo.BuyItem(SecondaryCost)
                        if(BuyState?):
                            # Code for Re-Packing #
                            set PackingAgent = option{InAgent}
                            spawn{PackWeapon(InAgent, HeldWeapon, ?IsRePacking:=true)}
                    else:
                        # Weapon is not pack a punched, so pack it! #
                        BuyState:=ThisPlayerInfo.BuyItem(PrimaryCost)
                        if(BuyState?):
                            set PackingAgent = option{InAgent}
                            spawn{PackWeapon(InAgent, HeldWeapon)}
                else{Print("~ERROR: Failed to get the held weapon!")}
            else if(PackedState?, InAgent = PackingAgent?):
                # Player picked up their packed weapon #
                PickedUpEvent.Signal()

    PickedUpEvent:event()=event(){}
    PackWeapon(InAgent:agent, Weapon:weapon_definition, ?IsRePacking:logic=false)<suspends>:void=
        SetInteractableState(false)
        Weapon.RecordAmmoAmount(InAgent)
        Sleep(0.0)
        spawn{SpawnVFX()}
        Weapon.Weapon_ConditionalButton.Activate(InAgent)
        ActivatedSFX.Play()
        if(ThisPlayerInfo:=InAgent.GetPlayerInfo[]):
            if(MysteryBoxManager:=GlobalWonderBoxManager[GetSession()]):
                MysteryBoxManager.SetStateForPlayer(InAgent, false)
            if(ThisItemPlacer:=Weapon.GetPerPlayerItemPlacer[InAgent]):
                if(InteractLocation:=InteractableCollection[0].GetTransform()):
                    if(ThisItemPlacer.TeleportTo[InteractLocation.Translation, InteractLocation.Rotation]){}
                    else{Print("~ERROR: Failed to get the interactable in the pack a punch!")}
                    SafeMoveTo(ThisItemPlacer, CenterLocation.GetTransform().Translation, CenterLocation.GetTransform().Rotation, SuckInTime)
                    SafeSleep(AwaitTime)
                    SafeMoveTo(ThisItemPlacer, InteractLocation.Translation, InteractLocation.Rotation, SuckInTime)
                    if(CenterLocation.IsValid[]):
                        set PackedState = true
                        SetInteractableState(true, ?InAgent:=option{InAgent})
                        Winner:=race:
                            block:
                                # Move weapon back into pack a punch #
                                SafeMoveTo(ThisItemPlacer, CenterLocation.GetTransform().Translation, CenterLocation.GetTransform().Rotation, TimeToCollect)
                                -1
                            block:
                                PickedUpEvent.Await()
                                1
                            #block:
                            #    ThisPlayerInfo.StopUsingItemPlacerEvent.Await()
                            #    -1
                        if(Winner=1):
                            Weapon.GrantWeapon(InAgent)
                            if(IsRePacking?):
                                ThisAbility := Weapon.VariantDefinition.ApplyRandomAbility(InAgent)
                                spawn. InAgent.GetUIManager().PowerPunchDetails.ShowDetails_PowerAbility(ThisAbility)
                            else:
                                Weapon.EnableVariantDefinition(InAgent)
                                if(not Switch_FirstPowerPunchUse.GetCurrentState[InAgent]):
                                    Switch_FirstPowerPunchUse.TurnOn(InAgent)
                                    spawn. InAgent.GetUIManager().PowerPunchDetails.ShowDetails_PowerPunch()
                            Weapon.LoadWeaponAmmo(InAgent)
                        else:
                            Weapon.RemoveWeapon(InAgent)
                        ThisItemPlacer.Hide()
                    else:
                        Print("~ERROR: The 'Center' prop is empty or invalid in the pack a punch!")
            else{Print("~ERROR: Failed to get an item placer for the weapon [{Weapon.Type.ToString()}]")}
            InAgent.SetItemCollectionState(false)
        set PackedState = false
        set PackingAgent = false
        if(MysteryBoxManager:=GlobalWonderBoxManager[GetSession()]):
            MysteryBoxManager.SetStateForPlayer(InAgent, true)

    SpawnVFX<private>()<suspends>:void=
        if(ThisSpawnedVFX:=SpawnProp(ActivatedVFX, CenterLocation.GetTransform().Translation, CenterLocation.GetTransform().Rotation)(0)?):
            Sleep(VFX_Lifetime)
            ThisSpawnedVFX.Dispose()

# VARIANT ABILITY MODULE #

variant_ability_module := class<concrete>:

    @editable Explosive : variant_ability_explosive = variant_ability_explosive{}
    @editable Turned : variant_ability_turned = variant_ability_turned{}
    @editable DeadWire : variant_ability_deadwire = variant_ability_deadwire{}
    @editable Frostbite : variant_ability_frostbite = variant_ability_frostbite{}
    #@editable HeatWave : variant_ability_heatwave = variant_ability_heatwave{}
    @editable LuckyShot : variant_ability_luckyshot = variant_ability_luckyshot{}
    @editable Overclocked : variant_ability_overclocked = variant_ability_overclocked{}

    GetVariantAbilities()<transacts>:[]variant_ability_base=
        array{
            Explosive
            Turned
            DeadWire
            Frostbite
            #HeatWave - Removed due to similarities with "Explosive" ability
            LuckyShot
            Overclocked
        }

    GetRandomVariantAbility(?ExcludingList:[]variant_ability_base=array{})<transacts><decides>:variant_ability_base=
        AbilitySelection:=for:
            ThisAbility:GetVariantAbilities()
            not ExcludingList.Find[ThisAbility]
        do:
            ThisAbility
        return Shuffle(AbilitySelection)[0]

VariantAbilityDebugDraw := class(debug_draw_channel) {}

variant_ability_activationtype<public> := enum{OnZombieEliminated, OnZombieDamaged}
variant_ability_cooldowntype<public> := enum{ActivationType, Time, None}
variant_ability_base<public> := class<concrete><unique>:

    @editable BasicSettings<public> : variant_ability_settings_base = variant_ability_settings_base{}

    var<private> IsActive : logic = false

    var<private> RemoveEventMap : [agent]event() = map{}
    GetRemoveEvent(InAgent:agent):event()=
        if(ThisRemoveEvent:=RemoveEventMap[InAgent]):
            return ThisRemoveEvent
        else:
            NewRemoveEvent : event() = event(){}
            set RemoveEventMap = ConcatenateMaps(RemoveEventMap, map{InAgent => NewRemoveEvent})
            return NewRemoveEvent

    Apply<public><final>(InAgent:agent, InWeapon:weapon_definition)<suspends>:void=
        DebugPrint("~PackAPunch Ability [{BasicSettings.Name}] applied!")
        if(not RemoveEventMap[InAgent]):
            set RemoveEventMap = ConcatenateMaps(RemoveEventMap, map{InAgent => event(){}})
        if(ThisPlayerInfo:=InAgent.GetPlayerInfo[]):
            spawn{OnApplied(InAgent)}
            var TargetedZombie : ?agent = false
            var Progress : int = 0
            if(BasicSettings.ActivateAbilityIfNew?):
                set Progress = BasicSettings.ThresholdAmount
            var CooldownProgress : int = 0
            race:
                loop:
                    set CooldownProgress = 0
                    loop:
                        if(IsActiveForPlayer[InAgent]):
                            case(BasicSettings.ActivationType):
                                variant_ability_activationtype.OnZombieDamaged =>
                                    ThisPlayerInfo.DamagedZombieEvent.Await()
                                variant_ability_activationtype.OnZombieEliminated =>
                                    ThisPlayerInfo.DamagedZombieEvent.Await()
                            if(InAgent.GetHeldWeapon[] = InWeapon):
                                set Progress += 1
                                if(Progress >= BasicSettings.ThresholdAmount):
                                    break
                        else:
                            Sleep(0.1)
                    set Progress = 0
                    loop:
                        if(IsActiveForPlayer[InAgent]):
                            case(BasicSettings.ActivationType):
                                variant_ability_activationtype.OnZombieDamaged =>
                                    DmgInfo:=ThisPlayerInfo.DamagedZombieEvent.Await()
                                    set TargetedZombie = option{DmgInfo.Target}
                                variant_ability_activationtype.OnZombieEliminated =>
                                    EliminatedZombie:=ThisPlayerInfo.EliminatedZombieEvent.Await()
                                    set TargetedZombie = option{EliminatedZombie}
                            if(ThisZombie:=TargetedZombie?, InAgent.GetHeldWeapon[] = InWeapon):
                                ActivatedState := ActivateOnChance(ThisZombie, InAgent)
                                if(ActivatedState?):
                                    break
                        else:
                            Sleep(0.1)
                    loop:
                        if(IsActiveForPlayer[InAgent]):
                            case(BasicSettings.CooldownType):
                                variant_ability_cooldowntype.ActivationType =>
                                    case(BasicSettings.ActivationType):
                                        variant_ability_activationtype.OnZombieDamaged =>
                                            ThisPlayerInfo.DamagedZombieEvent.Await()
                                            set CooldownProgress += 1
                                        variant_ability_activationtype.OnZombieEliminated =>
                                            ThisPlayerInfo.DamagedZombieEvent.Await()
                                            set CooldownProgress += 1
                                    if(CooldownProgress >= BasicSettings.CooldownAmount):
                                        break
                                variant_ability_cooldowntype.Time =>
                                    SafeSleep(BasicSettings.CooldownAmount*1.0)
                                    break
                                variant_ability_cooldowntype.None =>
                                    break
                        else:
                            Sleep(0.1)
                GetRemoveEvent(InAgent).Await()

    OnApplied<public>(InAgent:agent)<suspends>:void=
        return

    ActivateOnChance<public><final>(AtZombie:agent, InAgent:agent):logic=
        if(GetRandomFloat(0.0, 100.0) <= BasicSettings.ActivationChance):
            DebugPrint("~PackAPunch Ability [{BasicSettings.Name}] activated!")
            spawn{OnActivated(AtZombie, InAgent)}
            return true
        else:
            return false

    OnActivated<public>(AtZombie:agent, InAgent:agent)<suspends>:void=
        return

    var<private> ActiveStatePerPlayer : [agent]logic = map{}
    SetActivationState(InAgent:agent, InState:logic):void=
        if(set ActiveStatePerPlayer[InAgent] = InState){}
        else:
            set ActiveStatePerPlayer = ConcatenateMaps(ActiveStatePerPlayer, map{InAgent => InState})
    IsActiveForPlayer(InAgent:agent)<transacts><decides>:void=
        if(ActiveStatePerPlayer[InAgent]):
            ActiveStatePerPlayer[InAgent]?
        else:
            set ActiveStatePerPlayer = ConcatenateMaps(ActiveStatePerPlayer, map{InAgent => true})

    Remove<public><final>(InAgent:agent):void=
        GetRemoveEvent(InAgent).Signal()

    OnRemoved<public>(OutAgent:agent):void=
        return

    AbilityIcon : ?texture = false
    GetAbilityIcon()<transacts><decides>:texture=
        AbilityIcon?

    DebugPrint(InString:string):void=
        if(BasicSettings.EnableDebug?):
            Print(InString)

variant_ability_settings_base<public> := class<concrete>:

    @editable Name<public> : string = "Unset Ability Name"
    @editable Description<public> : string = "Unset Ability Description"
    # The amount of kills/damage events before the ability has a chance to activate again #
    @editable var<private> ThresholdAmount : type{_X : int where 0 <= _X} = 10
    # The chance that this ability will activate #
    @editable var<private> ActivationChance : type{_X : float where 0.0 <= _X} = 10.0
    # How should this ability be activated? #
    @editable var<private> ActivationType : variant_ability_activationtype = variant_ability_activationtype.OnZombieDamaged
    # How much until the cooldown ends? #
    # Depends on what "CooldownType" is set to #
    @editable CooldownAmount : type{_X : int where 1 <= _X} = 10
    # The type of cooldown to use #
    @editable CooldownType : variant_ability_cooldowntype = variant_ability_cooldowntype.ActivationType

    @editable ActivateAbilityIfNew : logic = true

    @editable EnableDebug : logic = false

variant_ability_explosive := class(variant_ability_base):

    @editable ExplosionVFX : creative_prop_asset = DefaultCreativePropAsset
    @editable VFX_ActiveTime : type{_X : float where 0.0 <= _X} = 0.0
    @editable Damage : type{_X : float where 0.0 <= _X} = 250.0
    @editable Radius : type{_X : float where 1.0 <= _X} = 250.0
    @editable FalloffRadius : type{_X : float where 1.0 <= _X} = 175.0

    @editable BasicSettings<override> : variant_ability_settings_base = variant_ability_settings_base{Name:="Explosive"}

    OnActivated<override>(AtZombie:agent, InAgent:agent)<suspends>:void=
        spawn. InAgent.GetUIManager().StartAbilityFlashing()

        #AbilityDebugDrawInst : debug_draw = debug_draw{Channel := VariantAbilityDebugDraw}
        if(AtFortChar:=AtZombie.GetFortCharacter[], AtZombieInfo:=AtZombie.GetZombieAI[]):
            var AtPosition : vector3 = AtFortChar.GetTransform().Translation
            if(not AtZombieInfo.IsActive[]):
                set AtPosition = AtZombieInfo.DeathTransform.Translation
            #AbilityDebugDrawInst.DrawSphere(AtPosition, ?Radius:=Radius, ?Color:=NamedColors.Yellow, ?NumSegments:=10, ?Thickness:=12.0, ?DrawDurationPolicy:=debug_draw_duration_policy.FiniteDuration, ?Duration:=2.0)
            var SpawnedExplosionVFX : creative_prop = creative_prop{}
            if(VFX_ActiveTime > 0.0, NewExplosionVFX := SpawnProp(ExplosionVFX, AtPosition, rotation{})(0)?):
                set SpawnedExplosionVFX = NewExplosionVFX
            PercentFalloff := FalloffRadius/Radius
            ZombieList := GetZombieList()
            for(ThisZombie : ZombieList):
                DistanceFromTarget := Distance(ThisZombie.FortCharacter.GetTransform().Translation, AtPosition)
                PercentFar := DistanceFromTarget/Radius
                PercentClose := 1.0-PercentFar
                if(PercentClose >= 0.0):
                    # Target is within the radius #
                    if(PercentFar >= PercentFalloff):
                        # This zombie is between the FalloffRadius and the Radius, so a percent damage is applied #
                        PercentDamage := Damage * (1.0-(DistanceFromTarget-FalloffRadius)/(Radius-FalloffRadius))
                        ThisZombie.FortCharacter.Damage(damage_args{Amount:=PercentDamage, Instigator:=option{InAgent.GetInstigator()}})
                        Print("~Zombie at distance [{DistanceFromTarget}] damaged for [{PercentDamage}]")
                    else:
                        # This zombie is closer than the falloffRadius, so full damage is applied #
                        ThisZombie.FortCharacter.Damage(damage_args{Amount:=Damage, Instigator:=option{InAgent.GetInstigator()}})
                        Print("~Zombie damaged for [{Damage}]")
            SafeSleep(VFX_ActiveTime)
            SpawnedExplosionVFX.Dispose()
        else{Print("~ERROR: Failed to get the death translation of a zombie!")}
        InAgent.GetUIManager().StopAbilityFlashing.Signal()

    AbilityIcon<override> : ?texture = option{Art.UIStuff.Icons.Weapon.Explosive_Icon}

variant_ability_turned := class(variant_ability_base):

    @editable TurnedTitle_BP : creative_prop_asset = DefaultCreativePropAsset
    @editable TurnedActiveTime : type{_X : float where 0.0 <= _X} = 30.0

    @editable BasicSettings<override> : variant_ability_settings_base = variant_ability_settings_base{Name:="Turned"}

    var<private> ActiveStatePerAgent : [agent]logic = map{}

    OnActivated<override>(AtZombie:agent, InAgent:agent)<suspends>:void=
        spawn. InAgent.GetUIManager().StartAbilityFlashing()
        if(not ActiveStatePerAgent[InAgent]):
            set ActiveStatePerAgent = ConcatenateMaps(ActiveStatePerAgent, map{InAgent => false})
        if(not ActiveStatePerAgent[InAgent]?):
            if(set ActiveStatePerAgent[InAgent] = true){}
            if(ZombieInfo:=AtZombie.GetZombieAI[], WaveDevice:=GlobalWaveController[GetSession()]):
                if(not ZombieInfo.SpawnLocation?, ZombieInfo.Controller.Controls.CanBeTurned?):
                    ZombieInfo.EnableTurnedMode(InAgent)
                    ZombieInfo.SetSpeed(WaveDevice.GetSpeedValue(?NoModify:=true)+0.15)
                    if(DefaultType := zombie_type_list.Default.GetController[]):
                        ZombieInfo.SetDamageAmount(DefaultType.GetZombieHealthIncrease())
                    ZombieInfo.SetMarker(zombie_marker.Turned)
                    race:
                        ZombieInfo.DeathEvent.Await()
                        SafeSleep(TurnedActiveTime)
                    ZombieInfo.Eliminate()
                    InAgent.GetUIManager().StopAbilityFlashing.Signal()
            if(set ActiveStatePerAgent[InAgent] = false){}
        else:
            Print("~Turned ability cannot active again yet!")

    AbilityIcon<override> : ?texture = option{Art.UIStuff.Icons.Weapon.Turned_Icon}

variant_ability_deadwire := class(variant_ability_base):

    @editable Electric_VFX : creative_prop_asset = DefaultCreativePropAsset
    @editable VFX_ActiveTime : type{_X : float where 0.0 <= _X} = 0.0
    @editable Damage : type{_X : float where 0.0 <= _X} = 150.0
    @editable MaximumHits : type{_X : int where 0 <= _X} = 20
    @editable TickTime : type{_X : float where 0.0 < _X} = 0.5
    @editable MaximumChainDistance : type{_X : float where 0.0 <= _X} = 250.0

    @editable BasicSettings<override> : variant_ability_settings_base = variant_ability_settings_base{Name:="Dead Wire"}

    OnActivated<override>(AtZombie:agent, InAgent:agent)<suspends>:void=
        spawn. InAgent.GetUIManager().StartAbilityFlashing(?Loops:=3)
        NewChainInst : deadwire_chain = deadwire_chain{Parent:=Self}
        spawn{NewChainInst.ChainHit(AtZombie, InAgent)}

    AbilityIcon<override> : ?texture = option{Art.UIStuff.Icons.Weapon.Deadwire_Icon}

deadwire_chain := class:

    Parent : variant_ability_deadwire

    var TrackedZombies : []agent = array{}

    #AbilityDebugDrawInst : debug_draw = debug_draw{Channel := VariantAbilityDebugDraw}

    ChainHit(InZombie:agent, Instigator:agent)<suspends>:void=
        if(InZombieInfo:=InZombie.GetZombieAI[], InZombieInfo.IsActive[], InZombieFortChar:=InZombie.GetFortCharacter[]):
            SafeSleep(Parent.TickTime)
            spawn{SpawnElectricVFX(InZombieInfo)}
            var PrevPosition : vector3 = InZombieFortChar.GetTransform().Translation
            InZombieFortChar.Damage(damage_args{Amount:=Parent.Damage, Instigator:=option{Instigator.GetInstigator()}})
            if(not InZombieInfo.IsActive[]):
                set PrevPosition = InZombieInfo.DeathTransform.Translation
            #AbilityDebugDrawInst.DrawSphere(PrevPosition, ?Radius:=Parent.MaximumChainDistance, ?Color:=NamedColors.Yellow, ?NumSegments:=10, ?Thickness:=12.0, ?DrawDurationPolicy:=debug_draw_duration_policy.FiniteDuration, ?Duration:=2.0)
            ActiveZombies := GetZombieList()
            for(ThisZombie:ActiveZombies):
                if(Distance(ThisZombie.FortCharacter.GetTransform().Translation, PrevPosition) < Parent.MaximumChainDistance):
                    if(not TrackedZombies.Find[ThisZombie.Agent]):
                        set TrackedZombies += array{ThisZombie.Agent}
                        spawn{ChainHit(ThisZombie.Agent, Instigator)}

    SpawnElectricVFX(InZombieInfo:zombie_ai)<suspends>:void=
        if(InZombieFortChar:=InZombieInfo.FortCharacter, SpawnedElectricVFX := SpawnProp(Parent.Electric_VFX, InZombieFortChar.GetTransform().Translation, rotation{})(0)?):
            race:
                loop:
                    SafeMoveTo(SpawnedElectricVFX, InZombieFortChar.GetTransform().Translation, rotation{}, 0.09365)
                InZombieInfo.DeathEvent.Await()
                SafeSleep(Parent.VFX_ActiveTime)
            SpawnedElectricVFX.Dispose()

variant_ability_frostbite := class(variant_ability_base):

    @editable FrostVFX : creative_prop_asset = DefaultCreativePropAsset
    @editable VFX_ActiveTime : type{_X : float where 0.0 <= _X} = 0.0
    @editable SpeedDebuffPercent : type{_X : float where 0.0 < _X, 100.0 > _X} = 75.0
    @editable Damage : type{_X : float where 0.0 <= _X} = 150.0
    @editable ActiveTime : type{_X : float where 0.0 <= _X} = 7.5
    @editable Radius : type{_X : float where 0.0 <= _X} = 250.0

    @editable BasicSettings<override> : variant_ability_settings_base = variant_ability_settings_base{Name:="Frostbite"}

    OnActivated<override>(AtZombie:agent, InAgent:agent)<suspends>:void=
        spawn. InAgent.GetUIManager().StartAbilityFlashing()
        #AbilityDebugDrawInst : debug_draw = debug_draw{Channel := VariantAbilityDebugDraw}
        var AffectedZombies : []tuple(zombie_ai, float) = array{}
        NewSpeedMult := 1.0-(SpeedDebuffPercent/100.0)
        if(InZombieInfo:=AtZombie.GetZombieAI[], AtFortChar:=AtZombie.GetFortCharacter[]):
            var AtPosition : vector3 = AtFortChar.GetTransform().Translation
            if(not InZombieInfo.IsActive[]):
                set AtPosition = InZombieInfo.DeathTransform.Translation
            #AbilityDebugDrawInst.DrawSphere(AtPosition, ?Radius:=Radius, ?Color:=NamedColors.Yellow, ?NumSegments:=10, ?Thickness:=12.0, ?DrawDurationPolicy:=debug_draw_duration_policy.FiniteDuration, ?Duration:=2.0)
            spawn{SpawnVFX(AtPosition)}
            set AffectedZombies += array{(InZombieInfo, InZombieInfo.MovementSpeed)}
            var SlowedZombieAgents : []agent = array{}
            ActiveZombies := GetZombieList()
            for(ThisZombie : ActiveZombies):
                if(not SlowedZombieAgents.Find[ThisZombie.Agent], not ThisZombie.TurnedOwner?, Distance(ThisZombie.FortCharacter.GetTransform().Translation, AtPosition) < Radius):
                    set SlowedZombieAgents += array{ThisZombie.Agent}
                    set AffectedZombies += array{(ThisZombie, ThisZombie.MovementSpeed)}
                    ThisZombie.SetSpeed(ThisZombie.MovementSpeed*NewSpeedMult)
                    AtFortChar.Damage(damage_args{Amount:=Damage, Instigator:=option{InAgent.GetInstigator()}})
            SafeSleep(ActiveTime)
            for(ThisAffectedZombie:AffectedZombies):
                ThisAffectedZombie(0).SetSpeed(ThisAffectedZombie(1))
        InAgent.GetUIManager().StopAbilityFlashing.Signal()

    SpawnVFX(AtPosition:vector3)<suspends>:void=
        if(SpawnedVFX := SpawnProp(FrostVFX, AtPosition, rotation{})(0)?):
            SafeSleep(VFX_ActiveTime)
            SpawnedVFX.Dispose()
        else{Print("~ERROR: Failed to spawn frostbite vfx!")}

    AbilityIcon<override> : ?texture = option{Art.UIStuff.Icons.Weapon.Frostbite_Icon}

# NOT IN USE, POSSIBLY REMOVE FROM CODE
variant_ability_heatwave := class(variant_ability_base):

    @editable FlameVFX : creative_prop_asset = DefaultCreativePropAsset
    @editable ActiveTime : type{_X : float where 0.0 <= _X} = 4.5
    @editable TickDamage : type{_X : float where 0.0 <= _X} = 50.0
    @editable TickRate : type{_X : float where 0.1 <= _X} = 1.0
    @editable Radius : type{_X : float where 1.0 <= _X} = 5.0

    @editable BasicSettings<override> : variant_ability_settings_base = variant_ability_settings_base{Name:="HeatWave"}

    OnActivated<override>(AtZombie:agent, InAgent:agent)<suspends>:void=
        if(AtZombieInfo:=AtZombie.GetZombieAI[], AtZombieFortChar:=AtZombie.GetFortCharacter[]):
            var FlamePosition : vector3 = AtZombieFortChar.GetTransform().Translation
            if(not AtZombieInfo.IsActive[]):
                set FlamePosition = AtZombieInfo.DeathTransform.Translation
            var SpawnedVFX : creative_prop = creative_prop{}
            if(SpawnedFlameVFX := SpawnProp(FlameVFX, FlamePosition, rotation{})(0)?):
                set SpawnedVFX = SpawnedFlameVFX
            race:
                loop:
                    ActiveZombies := GetZombieList()
                    for(ThisZombie:ActiveZombies):
                        if(Distance(ThisZombie.FortCharacter.GetTransform().Translation, FlamePosition) < Radius):
                            ThisZombie.FortCharacter.Damage(damage_args{Amount:=TickDamage, Instigator:=option{InAgent.GetInstigator()}})
                    SafeSleep(TickRate)
                SafeSleep(ActiveTime)
            SpawnedVFX.Dispose()

    #AbilityIcon<override> : ?texture = option{Art.UI.Icons.Weapon.Heatwave_Icon}

variant_ability_luckyshot := class(variant_ability_base):

    @editable LuckyPowerupVFX : creative_prop_asset = DefaultCreativePropAsset
    @editable VFX_ActiveTime : type{_X : float where 0.0 <= _X} = 0.0
    @editable MaxPowerupSpawnsPerRound : type{_X : int where 1 <= _X} = 3

    @editable BasicSettings<override> : variant_ability_settings_base = variant_ability_settings_base{Name:="LuckyShot"}

    var<private> PowerupsSpawnedPerAgent : [agent]int = map{}

    OnApplied<override>(InAgent:agent)<suspends>:void=
        spawn. InAgent.GetUIManager().StartAbilityFlashing(?Loops:=3)
        if(not PowerupsSpawnedPerAgent[InAgent]):
            set PowerupsSpawnedPerAgent = ConcatenateMaps(PowerupsSpawnedPerAgent, map{InAgent => 0})
        else if(set PowerupsSpawnedPerAgent[InAgent] = 0){}
        race:
            block:
                GetRemoveEvent(InAgent).Await()
            block:
                if(ThisWaveDevice:=GlobalWaveController[GetSession()]):
                    loop:
                        ThisWaveDevice.WaveStartedEvent.Await()
                        if(set PowerupsSpawnedPerAgent[InAgent] = 0){}
                Sleep(Inf)
    
    OnActivated<override>(AtZombie:agent, InAgent:agent)<suspends>:void=
        if(PowerupsSpawnedPerAgent[InAgent] < MaxPowerupSpawnsPerRound):
            if(AtZombieInfo:=AtZombie.GetZombieAI[], AtZombieFortChar:=AtZombie.GetFortCharacter[]):
                var SpawnPos : vector3 = AtZombieFortChar.GetTransform().Translation
                if(not AtZombieInfo.IsActive[]):
                    set SpawnPos = AtZombieInfo.DeathTransform.Translation
                if(ThisRandomPowerup:=GetRandomPowerup[?IncludeSpecials:=true]):
                    SpawnPowerupOnZombie(AtZombieInfo, ThisRandomPowerup, ?ByChance:=false)
                if{
                    CurrentSpawnedCount := PowerupsSpawnedPerAgent[InAgent]
                    set PowerupsSpawnedPerAgent[InAgent] = CurrentSpawnedCount+1
                }
                if(SpawnedLuckyVFX:=SpawnProp(LuckyPowerupVFX, SpawnPos, rotation{})(0)?):
                    SafeSleep(VFX_ActiveTime)
                    SpawnedLuckyVFX.Dispose()

    AbilityIcon<override> : ?texture = option{Art.UIStuff.Icons.Weapon.Lucky_Icon}

variant_ability_overclocked := class(variant_ability_base):
    
    @editable DamageMultipliers : []float = array{}
    @editable MaxActivationTime : float = 10.0
    @editable BasicSettings<override> : variant_ability_settings_base = variant_ability_settings_base{Name:="Overclocked"}

    OnActivated<override>(AtZombie:agent, InAgent:agent)<suspends>:void=
        spawn. InAgent.GetUIManager().StartAbilityFlashing()
        SetActivationState(InAgent, false)
        var CurrentDamageMultiplierIndex : int = 0
        if(ThisPlayerInfo:=InAgent.GetPlayerInfo[], ThisWeapon:=InAgent.GetHeldWeapon[]):
            ThisPlayerInfo.SetDamageBoostState(true)
            race:
                SafeSleep(MaxActivationTime)
                loop:
                    if(InAgent.GetHeldWeapon[] <> ThisWeapon):
                        break
                    Sleep(0.0)
                loop:
                    Winner:=race:
                        block:
                            SafeSleep(2.5)
                            -1
                        block:
                            ThisPlayerInfo.DamagedEvent.Await()
                            1
                    if(Winner=-1):
                        Print("~Minimum shoot time passsed!")
                        break
                block:
                    for(ThisDmgMultiplier:DamageMultipliers):
                        ThisPlayerInfo.SetDamageMultiplier(ThisDmgMultiplier, ?AffectBase:=false)
                        Print("~Curr Damage Multiplier: {ThisPlayerInfo.DamageMultiplier}")
                        SafeSleep((CurrentDamageMultiplierIndex+1)*1.5)
                        set CurrentDamageMultiplierIndex += 1
                    Sleep(Inf)
            ThisPlayerInfo.SetDamageBoostState(false)
            Print("~End DamageMultiplier: {ThisPlayerInfo.DamageMultiplier}")
        else:
            Print("~ERR: Failed to get playerinfo or weapon!")
        SetActivationState(InAgent, true)
        InAgent.GetUIManager().StopAbilityFlashing.Signal()
    
    AbilityIcon<override> : ?texture = option{Art.UIStuff.Icons.Weapon.Overclocked_Icon}
