using { /Fortnite.com/Devices }
using { /Fortnite.com/Characters }
using { /Fortnite.com/Game }
using { /Fortnite.com/UI }
using { /Fortnite.com/FortPlayerUtilities }
using { /Verse.org/Simulation }
using { /Verse.org/Random }
using { /Verse.org/Colors }
using { /UnrealEngine.com/Temporary/Diagnostics }
using { /UnrealEngine.com/Temporary/SpatialMath }
using { /UnrealEngine.com/Temporary/UI }
using { Player }
using { InteractiveModule }

FindWallbuy<public>(WeaponType:weapon_definition_list)<transacts><decides>:wallbuy=
    var MaybeWallbuy : ?wallbuy = false
    WallbuyList := WallbuyCollection[GetSession()]
    for(ThisWallbuy:WallbuyList):
        if(ThisWallbuy.Weapon = WeaponType):
            set MaybeWallbuy = option{ThisWallbuy}
    MaybeWallbuy?

var WallbuyCollection : weak_map(session, []wallbuy) = map{}

wallbuy<public> := class(Interactive_Base):

    InteractiveType<override> : interactable_type = interactable_type.Weapon

    @editable Weapon : weapon_definition_list = weapon_definition_list.None
    @editable WeaponCost : int = 500
    @editable AmmoCost : int = 1000
    @editable UpgradedAmmoCost : int = 2500
    @editable AmmoAmount : type{_X : int where 0 <= _X} = 0
    @editable InteractiveLocation : Interactable = Interactable{Identity:="Wallbuy", InteractTime:=0.5, RequiresLookAt:=true, InteractDistance:=175.0, InteractionSpriteOffset:=vector3{X:=5.0}}

    var WeaponDefinition : weapon_definition = weapon_definition{}

    OnBegin<override>()<suspends>:void=
        loop:
            if(GlobalWeaponManager[GetSession()]):
                if(DesiredWeaponDef:=Weapon.GetWeaponDefinition[]):
                    set SetupState = true
                    set WeaponDefinition = DesiredWeaponDef
                else:
                    Print("~ERR: Failed to setup a wallbuy with an unknown weapon definition type!")
                    Print("~Make sure the name matches the desired one within the weapon manager and the type is not set to \"None\"!")
                break
            Sleep(0.0)
        if:
            not WallbuyCollection[GetSession()]
            set WallbuyCollection[GetSession()] = array{}
        if:
            CurrentWallbuys := WallbuyCollection[GetSession()]
            set WallbuyCollection[GetSession()] = CurrentWallbuys + array{Self}
        LinkAndCompile(array. InteractiveLocation)

    WhenInsideInteractionRadius<override>(InAgent:agent, InInteractingLocation:Interactable)<suspends>:void=
        if(SetupState?):
            if(WeaponDefinition.IsInInventory[InAgent]):
                if(WeaponDefinition.HasMaxAmmo[InAgent]):
                    InAgent.RefreshAmmoCount()

    var<private> SetupState : logic = false
    GetInteractionText<override>(InAgent:agent, Interactive:Interactable):string=
        if(SetupState?):
            if(WeaponDefinition.IsInInventory[InAgent]):
                if(not WeaponDefinition.HasMaxAmmo[InAgent]):
                    SetInteractableState(true, ?InAgent:=option{InAgent})
                    if(WeaponDefinition.DoesHaveVariant[InAgent]):
                        return "Purchase Upgraded ammo [<C_Y>${UpgradedAmmoCost}<E>]"
                    else:
                        return "Purchase ammo [<C_Y>${AmmoCost}<E>]"
                else:
                    SetInteractableState(false, ?InAgent:=option{InAgent})
                    return "you have maximum ammo"
            else:
                SetInteractableState(true, ?InAgent:=option{InAgent})
                if(InAgent.CheckInventorySpace[]):
                    return "Purchase {WeaponDefinition.Type.ToString()} [<C_Y>${WeaponCost}<E>]"
                else:
                    return "Swap for {WeaponDefinition.Type.ToString()} [<C_Y>${WeaponCost}<E>]"
        else:
            SetInteractableState(false)
            return "ERR - wallbuy weapon"

    OnInteracted<override>(InAgent:agent, InInteractingLocation:Interactable):void=
        if(ThisPlayerInfo:=InAgent.GetPlayerInfo[]):
            if(WeaponDefinition.IsInInventory[InAgent]):
                ThisAmmoCost : int = if. WeaponDefinition.DoesHaveVariant[InAgent] then. UpgradedAmmoCost else. AmmoCost
                BuyState := ThisPlayerInfo.BuyItem(ThisAmmoCost)
                if(BuyState?):
                    # Grant ammo for the weapon #
                    if(AmmoAmount > 0):
                        WeaponDefinition.GrantAmmo(InAgent, AmmoAmount)
                    else:
                        WeaponDefinition.GrantMaxAmmo(InAgent)
            else:
                WeaponDefinition.ResetMaxAmmo(InAgent)
                WeaponDefinition.ResetAmmo(InAgent, ?DoGrantAmmo:=false)
                BuyState := ThisPlayerInfo.BuyItem(WeaponCost)
                if(BuyState?):
                    # Grant the weapon #
                    WeaponDefinition.GrantWeapon(InAgent)