using { /Fortnite.com/Devices }
using { /Fortnite.com/Characters }
using { /Fortnite.com/Game }
using { /Fortnite.com/UI }
using { /Fortnite.com/FortPlayerUtilities }
using { /Verse.org/Simulation }
using { /Verse.org/Random }
using { /Verse.org/Colors }
using { /UnrealEngine.com/Temporary/Diagnostics }
using { /UnrealEngine.com/Temporary/SpatialMath }
using { /UnrealEngine.com/Temporary/UI }
using { /Fortnite.com/AI }
using { Player }
using { UI }
using { Spawning }
using { InteractiveModule }

GetWeaponsOfWonderBoxRarity<public>(ThisRarity:item_rarity, InAgent:agent)<transacts>:[]weapon_definition=
    PlayerWeapons := InAgent.GetInventoryWeapons()
    WeaponsOfRarity:=for:
        ThisWeapon:GetWeaponDefinitions()
        ThisWeapon.WonderBoxEnabled?
        not PlayerWeapons.Find[ThisWeapon]
        ThisWeapon.WonderBoxRarity = ThisRarity
    do:
        ThisWeapon

GetWeaponDefinitions<public>()<transacts>:[]weapon_definition=
    var WeaponDefinitions : []weapon_definition = array{}
    if(WeaponManager:=GlobalWeaponManager[GetSession()]):
        set WeaponDefinitions = WeaponManager.WeaponDefinitions
    return WeaponDefinitions

(WeaponEnum:weapon_definition_list).ToString<public>()<transacts>:string=
    case(WeaponEnum):
        # USE ONLY LOWERCASE CHARACTERS HERE
        weapon_definition_list.RangerPistol => "ranger pistol"
        weapon_definition_list.DeathMachine => "death machine"
        weapon_definition_list.GatekeeperShotgun => "gatekeeper shotgun"
        weapon_definition_list.HyperSMG => "hyper smg"
        weapon_definition_list.ReaperSniperRifle => "reaper sniper rifle"
        weapon_definition_list.Revolver => "revolver"
        weapon_definition_list.HuntingRifle => "hunting rifle"
        weapon_definition_list.InfantryRifle => "infantry rifle"
        weapon_definition_list.HammerPumpShotgun => "pump shotgun"
        weapon_definition_list.DrumGun => "drum gun"
        weapon_definition_list.LightMachineGun => "light machine gun"
        weapon_definition_list.DoubleBarrelShotgun => "double barrel shotgun"
        weapon_definition_list.ScopedBurstSMG => "scoped burst smg"
        weapon_definition_list.StrikerAR => "striker ar"
        weapon_definition_list.StrikerBurstAR => "striker burst ar"
        weapon_definition_list.CombatSMG => "combat smg"
        weapon_definition_list.Crossbow => "crossbow"
        weapon_definition_list.HeavyAR => "heavy ar" 
        weapon_definition_list.RocketLauncher => "rocket launcher"
        weapon_definition_list.MachinistsCombatAR => "Machinists Combat AR"
        weapon_definition_list.LeverActionRifle => "lever action rifle"
        weapon_definition_list.KymeraRaygun => "kymera raygun"
        weapon_definition_list.ExplosiveRepeater => "explosive repeater"
        weapon_definition_list.MonarchPistol => "monarch pistol"
        _ => "Unknown"

(DesiredWeaponType:weapon_definition_list).GetWeaponDefinition<public>()<transacts><decides>:weapon_definition=
    var MaybeMatchingWeapon : ?weapon_definition = false
    WeaponDefs := GetWeaponDefinitions()
    for:
        ThisWeapon:WeaponDefs
        ThisWeapon.Type = DesiredWeaponType
    do:
        set MaybeMatchingWeapon = option{ThisWeapon}
    MaybeMatchingWeapon?

weapon_definition_list<public> := enum{
    None
    #SidearmPistol
    RangerPistol
    DeathMachine
    GatekeeperShotgun
    HyperSMG
    ReaperSniperRifle
    Revolver
    HuntingRifle
    InfantryRifle
    HammerPumpShotgun
    DrumGun
    LightMachineGun
    DoubleBarrelShotgun
    ScopedBurstSMG
    StrikerAR
    StrikerBurstAR
    CombatSMG
    Crossbow
    HeavyAR
    RocketLauncher
    MachinistsCombatAR
    LeverActionRifle
    KymeraRaygun
    ExplosiveRepeater
    MonarchPistol
}

(ItemRarity:item_rarity).ToString<public>()<transacts>:string=
    case(ItemRarity):
        item_rarity.Common => "Common"
        item_rarity.Rare => "Rare"
        item_rarity.Epic => "Epic"
        item_rarity.Legendary => "Legendary"
        item_rarity.Mythic => "Mythic"
        _ => "Unknown rarity"

(ItemRarity:item_rarity).Increase<public>()<transacts>:item_rarity=
    case(ItemRarity):
        item_rarity.Common => item_rarity.Rare
        item_rarity.Rare => item_rarity.Epic
        item_rarity.Epic => item_rarity.Legendary
        item_rarity.Legendary => item_rarity.Mythic
        _ => ItemRarity

(ItemRarity:item_rarity).Decrease<public>()<transacts>:item_rarity=
    case(ItemRarity):
        item_rarity.Rare => item_rarity.Common
        item_rarity.Epic => item_rarity.Rare
        item_rarity.Legendary => item_rarity.Epic
        item_rarity.Mythic => item_rarity.Legendary
        _ => ItemRarity

item_rarity<public> := enum{
    None
    Common
    Rare
    Epic
    Legendary
    Mythic
}

weapon_definition<public> := class<concrete><unique>:
    
    var ParentWeaponManager : Weapon_Manager = Weapon_Manager{}
    Init(InManager:Weapon_Manager, InAmmoGranter:item_granter_device):void=
        set ParentWeaponManager = InManager
        set AmmoGranter = InAmmoGranter
        VariantDefinition.Init(Self)

    @editable var<private> Type : weapon_definition_list = weapon_definition_list.None
    # Used for checking if a player has this weapon, or erasing it from a player #
    @editable Weapon_ConditionalButton : conditional_button_device = conditional_button_device{}
    # Contains a SINGULAR drop of ammo for this weapon #
    @editable AmmoType : ammotype = ammotype.Light
    # The maximum allowed ammo for this weapon #
    @editable MaxAmmoAmount : type{_X : int where 1 <= _X} = 500
    # Should the weapon have infinite ammo #
    @editable IsInfiniteAmmo : logic = false
    # The default amount of ammo allocated to a player upon receiving this weapon #
    @editable DefaultAmmoAmount : type{_X : int where 1 <= _X} = 150 
    # The default base damage for this weapon within fortnite #
    @editable DefaultBaseDamage : type{_X : float where 0.0 <= _X} = 0.0
    # The amount to multiply the DefaultBaseDamage by #
    @editable var<private> DamageOverride : float = -1.0
    # The multiplier for headshots, only works with DamageOverride #
    @editable HeadshotDamageMultiplier : type{_X : float where 1.0 <= _X} = 1.5
    # Overrides the firerate of the weapon to the set amount #
    #@editable CustomFireRate : type{_X : float where 0.0 <= _X} = 0.0
    # Can this weapon appear in the wonder box? # 
    @editable WonderBoxEnabled : logic = true
    # Can this weapon be pack-a-punched? #
    @editable PowerPunchEnabled : logic = true
    # Can this weapon be double power punched?
    @editable DoublePunchEnabled : logic = true

    @editable AmmoStation_CostOverride : int = -1
    @editable AmmoStation_PowerPunchedCostOverride : int = -1

    @editable AmmoDrop_Minimum : int = 0
    @editable AmmoDrop_Maximum : int = 0

    @editable GivePoints_OnDamage<public> : logic = true
    @editable GivePoints_OnElim<public> : logic = true
    
    var WonderBoxItemIndex : int = -1
    var ItemGranterIndex : int = -1

    @editable ItemPlacers : []item_placer_device = array{}
    GetPerPlayerItemPlacer(InAgent:agent)<transacts><decides>:item_placer_device=
        ItemPlacers[InAgent.GetPlayerInfo[].UniqueIndex]
        
    @editable LocalWeaponOffset : vector3 = vector3{}
    ApplyOffset(InTransform:transform):vector3=
        return InTransform.GetLocalVector(LocalWeaponOffset)

    # Can apply a custom weapon model via vfx #
    @editable UseCustomModelVFX : logic = false
    @editable CustomModelVFX : visual_effect_powerup_device = visual_effect_powerup_device{}

    #@editable CustomCamera : ?gameplay_camera_device = false

    @editable CustomCamera : ?gameplay_camera_orbit_device = false

    @editable TrialRewardRarity : item_rarity = item_rarity.None
    @editable WonderBoxRarity : item_rarity = item_rarity.Common

    # The 'Variant' or Upgraded version of this weapon #
    @editable VariantDefinition : weapon_definition_variant = weapon_definition_variant{}

    @editable IsDebug : logic = false

    # The ammo item granter device for this weapon. Set on game start by the weaponmanager #
    var AmmoGranter : item_granter_device = item_granter_device{}

    # The amount of ammo left in this weapon per player #
    var AmmoCountPerPlayer : [agent]int = map{}

    OnZombieHit<public> : event(agent) = event(agent){}

    var MaxAmmoPerPlayer : [agent]int = map{}
    GetMaxAmmo<public>(InAgent:?agent)<transacts>:int=
        if(ThisAgent:=InAgent?):
            if(DoesHaveVariant[ThisAgent]):
                if. set MaxAmmoPerPlayer[ThisAgent] = VariantDefinition.MaxAmmoAmount
                else. set MaxAmmoPerPlayer = ConcatenateMaps(MaxAmmoPerPlayer, map{ThisAgent => VariantDefinition.MaxAmmoAmount})
                return VariantDefinition.MaxAmmoAmount
            else:
                if(ThisMaxAmmo:=MaxAmmoPerPlayer[ThisAgent]):
                    return ThisMaxAmmo
                if. set MaxAmmoPerPlayer[ThisAgent] = MaxAmmoAmount
                else. set MaxAmmoPerPlayer = ConcatenateMaps(MaxAmmoPerPlayer, map{ThisAgent => MaxAmmoAmount})
        return MaxAmmoAmount

    SetMaxAmmo<public>(InAgent:agent, MaxAmount:int)<transacts>:void=
        if(set MaxAmmoPerPlayer[InAgent] = MaxAmount){}
        else:
            set MaxAmmoPerPlayer = ConcatenateMaps(MaxAmmoPerPlayer, map{InAgent => MaxAmount})

    ResetMaxAmmo<public>(InAgent:agent)<transacts>:void=
        if(set MaxAmmoPerPlayer[InAgent] = MaxAmmoAmount){}
        else:
            set MaxAmmoPerPlayer = ConcatenateMaps(MaxAmmoPerPlayer, map{InAgent => MaxAmmoAmount})

    HasMaxAmmo(InAgent:agent)<transacts><decides>:void=
        if(AmmoType <> ammotype.None):
            ThisMaxAmount:=GetMaxAmmo(option{InAgent})
            AmmoCountPerPlayer[InAgent] >= ThisMaxAmount

    var IsWeaponVariantPerPlayer : [agent]logic = map{}
    DoesHaveVariant(InAgent:agent)<transacts><decides>:void=
        IsWeaponVariantPerPlayer[InAgent]?

    LoadWeapon(InAgent:agent):void=
        if(not IsWeaponVariantPerPlayer[InAgent]):
            set IsWeaponVariantPerPlayer = ConcatenateMaps(IsWeaponVariantPerPlayer, map{InAgent=>false})
        if(DoesHaveVariant[InAgent]):
            VariantDefinition.LoadVariantWeapon(InAgent)
        else:
            ThisUI:=InAgent.GetUIManager()
            ThisUI.PowerPunchWidget.Hide()
            ThisUI.AmmoTypeWidget.Hide()
        LoadWeaponAmmo(InAgent)
        spawn{ApplyWeaponVFXLoop(InAgent)}
        HeldEvent.Signal((true, InAgent))

    # Grants the ammo count for this player #
    LoadWeaponAmmo(InAgent:agent):void=
        if(AmmoType <> ammotype.None):
            if(not AmmoCountPerPlayer[InAgent]):
                set AmmoCountPerPlayer = ConcatenateMaps(AmmoCountPerPlayer, map{InAgent=>DefaultAmmoAmount})
            if(ThisAmmoCount:=AmmoCountPerPlayer[InAgent]):
                SetAmmo(InAgent, ThisAmmoCount)
            else{Print("~ERROR: Failed to fetch a players' ammo count for weapon [{Type.ToString()}]")}
    # Records the ammo amount the player currently has #
    RecordAmmoAmount(InAgent:agent):int=
        if(AmmoType <> ammotype.None):
            var CurrentAmmoAmount : int = ParentWeaponManager.AmmoCollection.GetAmmoCount(InAgent, AmmoType)
            if(set AmmoCountPerPlayer[InAgent] = CurrentAmmoAmount){}
            else:
                set AmmoCountPerPlayer = ConcatenateMaps(AmmoCountPerPlayer, map{InAgent => DefaultAmmoAmount})
                set CurrentAmmoAmount = DefaultAmmoAmount
            return CurrentAmmoAmount
        return -1

    # Grants the given amount of ammo to the given player #
    GrantAmmo(InAgent:agent, Amount:int):void=
        if(AmmoType <> ammotype.None):
            CurrAmmo:=GetAmmoCountPerPlayer(InAgent)
            SetAmmo(InAgent, Amount+CurrAmmo)

    SetAmmo(InAgent:agent, Amount:int, ?DoGrantAmmo:logic=true):void=
        if(AmmoType <> ammotype.None):
            if. DoGrantAmmo? then. spawn. ParentWeaponManager.AmmoCollection.SetAmmo(InAgent, Amount, Self)
            var AmountToSet : int = Amount
            ThisMaxAmmo:=GetMaxAmmo(option{InAgent})
            if(Amount > ThisMaxAmmo):
                set AmountToSet = ThisMaxAmmo
            if. set AmmoCountPerPlayer[InAgent] = AmountToSet

    ResetAmmo(InAgent:agent, ?DoGrantAmmo:logic=true):void=
        if(AmmoType <> ammotype.None):
            SetAmmo(InAgent, DefaultAmmoAmount, ?DoGrantAmmo:=DoGrantAmmo)

    # Grants the maximum ammo allowed for this gun to the player #
    # Works even if the player is not holding the weapon #
    GrantMaxAmmo<public>(InAgent:agent):void=
        if(AmmoType <> ammotype.None):
            var MaxAmmo : int = GetMaxAmmo(option{InAgent})
            if(DoesHaveVariant[InAgent]):
                set MaxAmmo = VariantDefinition.GetMaxAmmo(option{InAgent})
            if(IsHeld[InAgent]):
                # Give ammo as player is holding weapon #
                SetAmmo(InAgent, MaxAmmo)
            # Set ammo amount as player is NOT holding weapon #
            else if(set AmmoCountPerPlayer[InAgent] = MaxAmmo){}
            else{Print("~ERROR: Failed to set AmmoCountPerPlayer to max in GrantMaximumAmmo()!")}

    # Gets the ammo count for this player #
    GetAmmoCountPerPlayer(InAgent:agent):int=
        ThisAmmoAmount := RecordAmmoAmount(InAgent)
        return ThisAmmoAmount

    # Enables the Variant definition for this weapon definition, overriding damage values...etc #
    EnableVariantDefinition(InAgent:agent):void=
        if(not IsWeaponVariantPerPlayer[InAgent]):
            set IsWeaponVariantPerPlayer = ConcatenateMaps(IsWeaponVariantPerPlayer, map{InAgent=>false})
        if(not DoesHaveVariant[InAgent]):
            if(set IsWeaponVariantPerPlayer[InAgent] = true){}
            VariantDefinition.LoadVariantWeapon(InAgent)
            if(VariantDefinition.GrantMaxAmmoOnVariantEnabled?):
                GrantMaxAmmo(InAgent)
        else:
            Print("~Log: Weapon already variant!")

    # Disables the Variant definition for this weapon definition #
    DisableVariantDefinition(InAgent:agent):void=
        if(DoesHaveVariant[InAgent]):
            if(set IsWeaponVariantPerPlayer[InAgent] = false){}
            VariantDefinition.Remove(InAgent)

    var StopTrackingWeaponMap : [agent]event() = map{}
    GetStopTrackingWeaponEvent(InAgent:agent):event()=
        if(ThisStopTrackingWeaponEvent := StopTrackingWeaponMap[InAgent]):
            return ThisStopTrackingWeaponEvent
        else:
            NewStopTrackingWeaponEvent : event() = event(){}
            set StopTrackingWeaponMap = ConcatenateMaps(StopTrackingWeaponMap, map{InAgent=>NewStopTrackingWeaponEvent})
            return NewStopTrackingWeaponEvent

    # Continually applies the custom model vfx to the player as long as the weapon is held #
    ApplyWeaponVFXLoop(InAgent:agent)<suspends>:void=
        if((DoesHaveVariant[InAgent] and VariantDefinition.UseCustomModelVFX?) or UseCustomModelVFX?):
            var ThisCustomWeaponVFX : visual_effect_powerup_device = CustomModelVFX
            if(DoesHaveVariant[InAgent] and VariantDefinition.UseCustomModelVFX?):
                set ThisCustomWeaponVFX = VariantDefinition.CustomModelVFX
            race:
                GetStopTrackingWeaponEvent(InAgent).Await()
                loop:
                    ThisCustomWeaponVFX.Pickup(InAgent)
                    Sleep(1.0)
            ParentWeaponManager.OnTrigger_ClearCustomModelVFX.Trigger(InAgent)

    # Returns true if the player has this weapon, otherwise false #
    IsInInventory(InAgent:agent)<transacts><decides>:void=
        Weapon_ConditionalButton.HasAllItems[InAgent]

    RemoveDuplicates(InAgent:agent):void=
        ItemCount := Weapon_ConditionalButton.GetItemCount(InAgent, 0)
        if(ItemCount > 1):
            Weapon_ConditionalButton.Activate(InAgent)
            GrantWeapon(InAgent)

    HeldEvent:event(tuple(logic, agent))=event(tuple(logic, agent)){}
    # Returns true if the player is currently holding this weapon, otherwise false #
    IsHeld(InAgent:agent)<transacts><decides>:void=
        Weapon_ConditionalButton.IsHoldingItem[InAgent]

    # Changes the base amount of damage this weapon does #
    SetDamageOverride<public>(InAmount:float):void=
        set DamageOverride = InAmount

    GetDamageAmount<public>(InAgent:agent, InDamage:float)<transacts>:float=
        ThisDmgAmount := if. DamageOverride >= 0.0 then. DamageOverride else. InDamage
        if(DefaultBaseDamage > 0.0 and InDamage > DefaultBaseDamage):
            #Print("Is headshot")
            # Is headshot #
            if(DoesHaveVariant[InAgent]):
                return ThisDmgAmount * VariantDefinition.DamageMultiplier * HeadshotDamageMultiplier
            else:
                return ThisDmgAmount * HeadshotDamageMultiplier
        else:
            # Is normal shot #
            if(DoesHaveVariant[InAgent]):
                return ThisDmgAmount * VariantDefinition.DamageMultiplier
            else:
                return ThisDmgAmount

    IsHeadshot<public>(InDamage:float)<transacts><decides>:void=
        DefaultBaseDamage > 0.0
        InDamage > DefaultBaseDamage

    GetName<public>(InAgent:agent):string=
        if(DoesHaveVariant[InAgent]):
            return VariantDefinition.UpgradedName
        return Type.ToString()

    SpawnedEvent<public>:event(Interactable) = event(Interactable){}
    GrantedEvent<public>:event(Interactable) = event(Interactable){}
    SpawnWeapon<public>(AtLocation:transform, ?Lifetime:float=-1.0, ?OwningPlayer:?agent=false, ?AllowSwap:logic=true):Interactable=
        InteractableLoc := Interactable{IsLocationStatic:=true, RequiresPlayerMustBeInFront:=false, InteractDistance:=125.0, InteractionSpriteOffset:=vector3{Z:=25.0}, StaticLocation:=transform{Translation:=AtLocation.Translation,Rotation:=AtLocation.Rotation}}
        spawn{AwaitWeaponPickup(InteractableLoc, AtLocation, Lifetime, OwningPlayer, AllowSwap)}
        return InteractableLoc

    var SpawnedInstances : []Interactable = array{}
    AwaitWeaponPickup<private>(NewInteractable:Interactable, AtLocation:transform, Lifetime:float, OwningPlayer:?agent, AllowSwap:logic)<suspends>:void=
        var PickedUpPlayer : ?agent = false
        var PickupString : string = "Grab {Type.ToString()}"
        if(not OwningPlayer?.CheckInventorySpace[]):
            set PickupString = "Swap for {Type.ToString()}"
        NewInteractable.SetText(PickupString)
        NewInteractable.Link()
        set SpawnedInstances += array{NewInteractable}
        if(OwningPlayer?):
            SpawnedEvent.Signal(NewInteractable)
            NewInteractable.SetEnabledState(false)
            NewInteractable.SetEnabledState(true, ?InAgent:=OwningPlayer)
            if(ThisPlayerInfo:=OwningPlayer?.GetPlayerInfo[], ThisWeaponPlacer:=GetPerPlayerItemPlacer[OwningPlayer?]):
                if(ThisWeaponPlacer.TeleportTo[AtLocation.Translation, AtLocation.Rotation]){}
                ThisPlayerInfo.User.SetItemCollectionState(true)
                race:
                    block:
                        InteractionTuple:=NewInteractable.InteractedEvent.Await()
                        set PickedUpPlayer = option{InteractionTuple}
                    block:
                        if(Lifetime > 0.0):
                            SafeSleep(Lifetime)
                        else:
                            Sleep(Inf)
                ThisWeaponPlacer.Hide()
                ThisPlayerInfo.User.SetItemCollectionState(false)
                if(PickingUpPlayer:=PickedUpPlayer?):
                    GrantWeapon(PickingUpPlayer, ?AllowSwap:=AllowSwap)
                    GrantedEvent.Signal(NewInteractable)
            NewInteractable.Unlink()
        if(SpawnedInstances.Length < 2):
            set SpawnedInstances = array{}
        else if(NewSpawnedList:=SpawnedInstances.RemoveElement[SpawnedInstances.FindInstance[NewInteractable]]):
            set SpawnedInstances = NewSpawnedList

    GetSpawnedInstances<public>():[]Interactable=
        return SpawnedInstances

    GrantWeapon<public>(InAgent:agent, ?AllowSwap:logic=true):void=
        if(IsInInventory[InAgent]):
            GrantMaxAmmo(InAgent)
            return
        else if(AllowSwap?, not InAgent.CheckInventorySpace[]):
            if(ThisHeldWeapon:=InAgent.GetHeldWeapon[]):
                ThisHeldWeapon.RemoveWeapon(InAgent)
            else if(FirstWeapon:=InAgent.GetInventoryWeapons()[0]):
                FirstWeapon.RemoveWeapon(InAgent)
            DisableVariantDefinition(InAgent)
        ParentWeaponManager.WeaponsItemGranter.GrantItemIndex(InAgent, ItemGranterIndex)
        if(ThisPlayerInfo:=InAgent.GetPlayerInfo[]):
            ThisPlayerInfo.PickedUpWeaponEvent.Signal()

    RemoveWeapon<public>(InAgent:agent, ?ResetStats:logic=true):void=
        if(DoesHaveVariant[InAgent], VariantDefinition.UseDifferentWeapon?):
            VariantDefinition.NewWeapon_ConditionalButton.Activate(InAgent)
            if(ResetStats?):
                VariantDefinition.Remove(InAgent)
        Weapon_ConditionalButton.Activate(InAgent)
        if(ResetStats?):
            ResetAmmo(InAgent, ?DoGrantAmmo:=false)
            ResetMaxAmmo(InAgent)

    Show<public>(InAgent:agent, Position:vector3, Rotation:rotation):void=
        if. ItemPlacers[InAgent.GetPlayerInfo[].UniqueIndex].TeleportTo[Position, Rotation]

    Hide<public>(InAgent:agent):void=
        if. ItemPlacers[InAgent.GetPlayerInfo[].UniqueIndex].TeleportTo[vector3{},rotation{}]

# Holds information about the upgraded variant of a weapon #
weapon_definition_variant := class<concrete>:

    var ParentWeaponDefinition : ?weapon_definition = false

    var AbilityPerPlayer : [agent]?variant_ability_base = map{}

    Init(InParentWeaponDef:weapon_definition):void=
        set ParentWeaponDefinition = option{InParentWeaponDef}
        if(UpgradedName = "Unset Upgraded Name" or UpgradedName = "" or UpgradedName = " "):
            set UpgradedName = InParentWeaponDef.Type.ToString()

    @editable var UpgradedName : string = "Unset Upgraded Name"

    @editable UseCustomModelVFX : logic = false
    @editable CustomModelVFX : visual_effect_powerup_device = visual_effect_powerup_device{}

    @editable UseDifferentWeapon : logic = false
    @editable NewWeapon_ItemGranter : item_granter_device = item_granter_device{}
    @editable NewWeapon_ConditionalButton : conditional_button_device = conditional_button_device{}
    # The damage amplifier amount to apply with this weapon variant #
    @editable DamageMultiplier : type{_X : float where 0.0 <= _X} = 1.5
    # The new MaxAmmoAmount for this variant weapon #
    @editable MaxAmmoAmount : type{_X : int where 1 <= _X} = 750

    @editable GrantMaxAmmoOnVariantEnabled : logic = true

    # Loads the Variant Weapon settings #
    LoadVariantWeapon<public>(InAgent:agent):void=
        if(ThisParentWeapon:=ParentWeaponDefinition?):
            if(UseCustomModelVFX?):
                ThisParentWeapon.GetStopTrackingWeaponEvent(InAgent).Signal()
                spawn{ThisParentWeapon.ApplyWeaponVFXLoop(InAgent)}
        ThisUI:=InAgent.GetUIManager()
        ThisUI.PowerPunchWidget.Show()
        if(ThisAbilityIcon:=AbilityPerPlayer[InAgent]?.GetAbilityIcon[]):
            ThisUI.AmmoAbilityTexture.SetImage(ThisAbilityIcon)
            ThisUI.AmmoTypeWidget.Show()
        else:
            ThisUI.AmmoTypeWidget.Hide()

    ApplyRandomAbility<public>(InAgent:agent):variant_ability_base=
        var CurrAbility : variant_ability_base = variant_ability_base{}
        if(CurrentAbility:=AbilityPerPlayer[InAgent]?):
            CurrentAbility.Remove(InAgent)
            set CurrAbility = CurrentAbility
        if:
            ParentWeapon := ParentWeaponDefinition?
            PackAPunchDevice:=GlobalPackAPunch[GetSession()]
            ThisRandomAbility:=PackAPunchDevice.VariantAbilities.GetRandomVariantAbility[?ExcludingList:=array{CurrAbility}]
        then:
            set CurrAbility = ThisRandomAbility
            if(not AbilityPerPlayer[InAgent]):
                set AbilityPerPlayer = ConcatenateMaps(AbilityPerPlayer, map{InAgent => option{ThisRandomAbility}})
            else if(set AbilityPerPlayer[InAgent] = option{ThisRandomAbility}){}
            spawn{ThisRandomAbility.Apply(InAgent, ParentWeapon)}
            if(ParentWeaponDefinition?.IsHeld[InAgent], ThisAbilityIcon:=ThisRandomAbility.GetAbilityIcon[]):
                InAgent.GetUIManager().AmmoAbilityTexture.SetImage(ThisAbilityIcon)
        return CurrAbility 

    Remove<public>(InAgent:agent):void=
        if(CurrentAbility:=AbilityPerPlayer[InAgent]?):
            CurrentAbility.Remove(InAgent)
        if(ParentWeapon:=ParentWeaponDefinition?):
            ParentWeapon.DisableVariantDefinition(InAgent)
        if. set AbilityPerPlayer[InAgent] = false

    var MaxAmmoPerPlayer : [agent]int = map{}
    GetMaxAmmo(InAgent:?agent)<transacts>:int=
        if(ThisMaxAmmo:=MaxAmmoPerPlayer[InAgent?]):
            return ThisMaxAmmo
        else if(ThisAgent:=InAgent?):
            set MaxAmmoPerPlayer = ConcatenateMaps(MaxAmmoPerPlayer, map{ThisAgent => MaxAmmoAmount})
        return MaxAmmoAmount