using { /Fortnite.com/Devices }
using { /Fortnite.com/Characters }
using { /Fortnite.com/Game }
using { /Fortnite.com/UI }
using { /Fortnite.com/FortPlayerUtilities }
using { /Verse.org/Simulation }
using { /Verse.org/Random }
using { /Verse.org/Colors }
using { /UnrealEngine.com/Temporary/Diagnostics }
using { /UnrealEngine.com/Temporary/SpatialMath }
using { /UnrealEngine.com/Temporary/UI }
using { /Fortnite.com/AI }
using { Player }
using { UI }
using { Spawning }
using { Zombie }
using { Powerups }

# ITEM COLLECTION

(InAgent:agent).IsUsingItemCollection<public>()<transacts><decides>:void=
    ThisWeaponManager := GlobalWeaponManager[GetSession()]
    ThisState := ThisWeaponManager.UsingItemCollectionStates[InAgent] or block{
        set ThisWeaponManager.UsingItemCollectionStates = ConcatenateMaps(ThisWeaponManager.UsingItemCollectionStates, map{InAgent => false})
        false
    }
    ThisState?

(InAgent:agent).GetItemCollection<public>(OnlyBoxItems:logic)<transacts>:[]item_placer_device=
    var ItemPlacerCollection : []item_placer_device = array{}
    if(ThisItemPlacerIndex:=InAgent.GetPlayerInfo[].UniqueIndex):
        WeaponList := GetWeaponDefinitions()
        for(ThisWeapon:WeaponList):
            if:
                not OnlyBoxItems? or ThisWeapon.WonderBoxEnabled?
            then:
                if(ThisItemPlacer:=ThisWeapon.ItemPlacers[ThisItemPlacerIndex]): 
                    set ItemPlacerCollection += array{ThisItemPlacer}
                #else:
                    #Print("~ERR: Could not find an item placer at index [{ThisItemPlacerIndex}] for weapon [{ThisWeapon.Type.ToString()}]!")
    else{Print("~ERR: Failed to get unique index in getting item placers!")}
    return ItemPlacerCollection

(InAgent:agent).SetItemCollectionState<public>(InState:logic):void=
    if. set GlobalWeaponManager[GetSession()].UsingItemCollectionStates[InAgent] = InState
    else if(CurrentItemCollectionStates := GlobalWeaponManager[GetSession()].UsingItemCollectionStates):
        if. set GlobalWeaponManager[GetSession()].UsingItemCollectionStates = ConcatenateMaps(CurrentItemCollectionStates, map{InAgent => InState})

(ItemPlacer:item_placer_device).Hide<public>()<transacts>:void=
    if. ItemPlacer.TeleportTo[vector3{Z:=-5000.0}, rotation{}]

(ItemPlacers:[]item_placer_device).Hide<public>()<transacts>:void=
    #Print("Hiding {ItemPlacers.Length} weapons")
    for(ThisItemPlacer:ItemPlacers):
        if. ThisItemPlacer.TeleportTo[vector3{},rotation{}]

# MISC WEAPON FUNCTIONS

(InAgent:agent).AwaitAimEvent()<suspends>:void=
    if(ThisWeaponManager:=GlobalWeaponManager[GetSession()]):
        if. not ThisWeaponManager.AimInput.IsHeld[InAgent] then. ThisWeaponManager.AimInput.PressedEvent.Await()
    else:
        Sleep(0.0)

(InAgent:agent).AwaitFireEvent()<suspends>:void=
    if(ThisWeaponManager:=GlobalWeaponManager[GetSession()]):
        if. not ThisWeaponManager.FireInput.IsHeld[InAgent] then. ThisWeaponManager.FireInput.PressedEvent.Await()
    else:
        Sleep(0.0)

(InAgent:agent).ResetAllAmmoCounts<public>():void=
    WeaponList:=GetWeaponDefinitions()
    for(ThisWeapon:WeaponList):
        ThisWeapon.ResetAmmo(InAgent, ?DoGrantAmmo:=false)
        ThisWeapon.ResetMaxAmmo(InAgent)

(InAgent:agent).RefreshAmmoCount<public>():void=
    #Print("Refreshing ammo count")
    if(ThisWeapon:=InAgent.GetHeldWeapon[]):
        ThisWeapon.RecordAmmoAmount(InAgent)

(InAgent:agent).GrantMaxAmmo<public>():void=
    PlayerWeapons := InAgent.GetInventoryWeapons()
    for(ThisWeapon:PlayerWeapons):
        ThisWeapon.GrantMaxAmmo(InAgent)

(InAgent:agent).RemoveWeapons<public>():void=
    WeaponDefs:=GetWeaponDefinitions()
    for(ThisWeapon:WeaponDefs):
        ThisWeapon.RemoveWeapon(InAgent)

(InZombie:zombie_ai).DropAmmo<public>():void=
    if(not InZombie.SpawnLocation?):
        if(ThisWeaponManager:=GlobalWeaponManager[GetSession()]):
            spawn{ThisWeaponManager.AmmoCollection.SpawnAmmoDrop(InZombie.GetGroundTransform())}

# Checks if any inventory space is available #
(InAgent:agent).CheckInventorySpace<public>()<transacts><decides>:void=
    InAgent.GetOpenInventorySlots() > 0

(InAgent:agent).GetOpenInventorySlots<public>()<transacts>:int=
    if(ThisPlayerInfo:=InAgent.GetPlayerInfo[]):
        return (ThisPlayerInfo.InventorySlots - InAgent.GetInventoryWeapons().Length)
    return 2

(InAgent:agent).GetHeldWeapon<public>()<transacts><decides>:weapon_definition=
    return GlobalWeaponManager[GetSession()].HeldWeaponDefPerPlayer[InAgent]?

(InAgent:agent).GetInventoryWeapons<public>()<transacts>:[]weapon_definition=
    var InventoryWeapons : []weapon_definition = array{}
    WeaponDefinitions:=GetWeaponDefinitions()
    for:
        ThisWeapon:WeaponDefinitions
        ThisWeapon.IsInInventory[InAgent]
    do:
        set InventoryWeapons += array{ThisWeapon}
    return InventoryWeapons

var GlobalWeaponManager<public> : weak_map(session, Weapon_Manager) = map{}

ammotype<public> := enum{Light, Medium, Heavy, Shells, Rockets, Arrows, None}
Weapon_Manager<public> := class(creative_device):

    @editable var<private> WeaponDefinitions : []weapon_definition = array{}
    @editable AmmoCollection<public> : ammo_collection = ammo_collection{}
    @editable WeaponsItemGranter : item_granter_device = item_granter_device{}
    @editable OnTrigger_GrantMaxAmmo : trigger_device = trigger_device{}
    @editable OnTrigger_EnableWeaponVariant : trigger_device = trigger_device{}
    @editable OnTrigger_ClearCustomModelVFX : trigger_device = trigger_device{}
    @editable FireInput<public> : input_trigger_device = input_trigger_device{}
    @editable AimInput<public> : input_trigger_device = input_trigger_device{}
    @editable AllowDebug : logic = false

    var<private> TrackedPlayers : []agent = array{}

    var HeldWeaponDefPerPlayer : [agent]?weapon_definition = map{}

    GetItemPlacers(InAgent:agent)<transacts>:[]item_placer_device=
        var ItemPlacerCollection : []item_placer_device = array{}
        if(ThisItemPlacerIndex:=InAgent.GetPlayerInfo[].UniqueIndex):
            for(ThisWeapon:WeaponDefinitions):
                if(ThisItemPlacer:=ThisWeapon.ItemPlacers[ThisItemPlacerIndex], ThisWeapon.WonderBoxEnabled?):
                    set ItemPlacerCollection += array{ThisItemPlacer}
                else:
                    Print("~ERR: Could not find an item placer at index [{ThisItemPlacerIndex}] for weapon [{ThisWeapon.Type.ToString()}]!")
        else{Print("~ERR: Failed to get unique index in getting item placers!")}
        return ItemPlacerCollection

    var WonderBoxWeapons : []weapon_definition = array{}
    GetRandomWonderBoxWeapon(InAgent:agent)<transacts><decides>:weapon_definition=
        PlayerWeapons:=InAgent.GetInventoryWeapons()
        var NewValidWeapons : []weapon_definition = array{}
        for(ThisWeapon:WeaponDefinitions):
            if(ThisWeapon.WonderBoxEnabled?, not PlayerWeapons.Find[ThisWeapon]):
                set NewValidWeapons += array{ThisWeapon}
        return Shuffle(NewValidWeapons)[0]

    OnBegin<override>():void=
        if(set GlobalWeaponManager[GetSession()] = Self){}
        else{Print("~ERROR: Failed to globally link weaponmanager!")}
        InitWeaponDefinitions()
        OnTrigger_EnableWeaponVariant.TriggeredEvent.Subscribe(EnableWeaponVariant)
        if(AllowDebug?):
            #spawn{AwaitDebugHit()}
        else{Print("~WeaponManager: Debug OFF")}

    var UsingItemCollectionStates<public> : [agent]logic = map{}

    # DEBUG PURPOSES ONLY #
    <#
    AwaitDebugHit()<suspends>:void=
        loop:
            if(DefaultZombieType:=zombie_type_list.Default.GetController[]):
                var ThisDebugWeapon : weapon_definition = weapon_definition{}
                for(ThisWeapon:WeaponDefinitions):
                    if(ThisWeapon.IsDebug?):
                        set ThisDebugWeapon = ThisWeapon
                loop:
                    DmgResult:=DefaultZombieType.ZombieDamaged.Await()
                    if:
                        ThisDamagingPlayer:=DmgResult.Instigator?
                        ThisDamagingPlayer.GetHeldWeapon[] = ThisDebugWeapon
                        ThisZombieInfo:=DmgResult.Target.GetZombieAI[]
                    then:
                        Print("~ PRINTING ZOMBIE INFO ~")
                        Print("IsActive : {ThisZombieInfo.IsActive.ToString()}")
                        Print("IsSpawning : {ThisZombieInfo.IsSpawning.ToString()}")
                        Print("IsSpawned : {ThisZombieInfo.IsSpawned.ToString()}")
                        Print("HasRespawned : {ThisZombieInfo.HasRespawned.ToString()}")
                        Print("IsQuickTravelling : {ThisZombieInfo.IsQuickTravelling.ToString()}")
                        if(ThisSpawnLocation:=ThisZombieInfo.SpawnLocation?):
                            Print("SpawnLocation : {ThisSpawnLocation.Identifier}")
                        else: 
                            Print("SpawnLocation : FALSE")
                        Print("IsInBarrier : {ThisZombieInfo.IsInBarrier.ToString()}")
                        Print("IsAtBarrier : {ThisZombieInfo.IsAtBarrier.ToString()}")
                        #Print("WasAtBarrier : {ThisZombieInfo.WasAtBarrier.ToString()}")
                        #Print("SpawnLoc_State1 : {ThisZombieInfo.SpawnLoc_State1.ToString()}")
                        #Print("SpawnLoc_State2 : {ThisZombieInfo.SpawnLoc_State2.ToString()}")
                        #Print("SpawnLoc_State3 : {ThisZombieInfo.SpawnLoc_State3.ToString()}")
                        #sPrint("SpawnLoc_State4 : {ThisZombieInfo.SpawnLoc_State4.ToString()}")
                        if(ThisNavDest:=ThisZombieInfo.GetFortCharacter[].GetNavigatable[].GetCurrentDestination[]):
                            Print("Nav Destination : {ThisNavDest}")
                        else:
                            Print("Nav Destination : FALSE")
            Sleep(0.0)
            #>

    InitWeaponDefinitions():void=
        var CurrWonderBoxWeaponIndex : int = 0
        for(WeaponIndex->ThisWeapon : WeaponDefinitions):
            ThisAmmoGranter := AmmoCollection.GetAmmoGranter(ThisWeapon.AmmoType)
            ThisWeapon.Init(Self, ThisAmmoGranter)
            set ThisWeapon.ItemGranterIndex = WeaponIndex
            if(ThisWeapon.WonderBoxEnabled?):
                set ThisWeapon.WonderBoxItemIndex = CurrWonderBoxWeaponIndex
                set WonderBoxWeapons += array{ThisWeapon}
                set CurrWonderBoxWeaponIndex += 1
            for(ThisItemPlacer:ThisWeapon.ItemPlacers):
                ThisItemPlacer.Hide()

    TrackPlayerWeapon<public>(InAgent:agent):void=
        if(not TrackedPlayers.Find[InAgent]):
            set TrackedPlayers += array{InAgent}
            spawn{TrackHeldWeapon(InAgent)}

    TrackHeldWeapon(InAgent:agent)<suspends>:void=
        var IsInputSpammed : logic = false
        var IsHoldingWeapon : logic = false
        if(ThisPlayerInfo:=InAgent.GetPlayerInfo[]):
            race:
                loop:
                    if(not IsInputSpammed?):
                        if(ThisPreviousWeapon:=HeldWeaponDefPerPlayer[InAgent]?):
                            AwaitDifferentWeapon(InAgent, ThisPreviousWeapon)
                    if(ThisWeaponDef:=GetHeldWeaponDefinition[InAgent]):
                        if(ThisPlayerManager:=GlobalPlayerManager[GetSession()]):
                            if(ThisCustomCamera:=ThisWeaponDef.CustomCamera?):
                                ThisPlayerManager.FirstPersonCamera.RemoveFrom(InAgent)
                                ThisPlayerManager.OrbitFirstPersonCamera.RemoveFrom(InAgent)
                                ThisCustomCamera.AddTo(InAgent)
                            else:
                                for(ThisWeapon:WeaponDefinitions):
                                    if(ThisCamera:=ThisWeaponDef.CustomCamera?):
                                        ThisCamera.RemoveFrom(InAgent)
                                ThisPlayerManager.OrbitFirstPersonCamera.RemoveFrom(InAgent)
                                ThisPlayerManager.FirstPersonCamera.AddTo(InAgent)
                        set IsHoldingWeapon = false
                        Winner:=race:
                            block:
                                Sleep(0.2)
                                1
                            block:
                                AwaitDifferentWeapon(InAgent, ThisWeaponDef)
                                -1
                            block:
                                if(ThisWeaponDef.IsInfiniteAmmo?, ThisWeaponDef.AmmoType <> ammotype.None):
                                    loop:
                                        ThisWeaponDef.GrantMaxAmmo(InAgent)
                                        Sleep(5.0)
                                Sleep(Inf)
                                -1
                        if(Winner=1):
                            HandleTrackInventorySlot(InAgent)
                            set IsInputSpammed = false
                        else:
                            set IsInputSpammed = true
                    else if(not IsHoldingWeapon?):
                        if(PreviousWeaponDef:=HeldWeaponDefPerPlayer[InAgent]?):
                            #Print("No weapon detected")
                            if(ThisPlayerManager:=GlobalPlayerManager[GetSession()]):
                                ThisPlayerManager.FirstPersonCamera.RemoveFrom(InAgent)
                                ThisPlayerManager.OrbitFirstPersonCamera.AddTo(InAgent)
                            PreviousWeaponDef.RecordAmmoAmount(InAgent)
                            PreviousWeaponDef.HeldEvent.Signal((false, InAgent))
                            PreviousWeaponDef.GetStopTrackingWeaponEvent(InAgent).Signal()
                            if(set HeldWeaponDefPerPlayer[InAgent] = false){}
                            ThisUI:=InAgent.GetUIManager()
                            ThisUI.PowerPunchWidget.Hide()
                            ThisUI.AmmoTypeWidget.Hide()
                            CurrentInventoryWeapons:=InAgent.GetInventoryWeapons()
                            if(not CurrentInventoryWeapons.Find[PreviousWeaponDef], not InAgent.IsPowerPunching[], not InAgent.HasDeathMachine[]):
                                PreviousWeaponDef.DisableVariantDefinition(InAgent)
                            loop:
                                if(GetHeldWeaponDefinition[InAgent]):
                                    break
                                Sleep(0.05)
                        set IsHoldingWeapon = true
                    Sleep(0.0)
                ThisPlayerInfo.LeaveEvent.Await()
            set TrackedPlayers = TrackedPlayers.RemoveAllElements(InAgent)
        
    AwaitDifferentWeapon(InAgent:agent, InWeaponDef:weapon_definition)<suspends>:void=
        loop:
            ThisWeapon := option. GetHeldWeaponDefinition[InAgent]
            if(ThisHeldWeapon:=GetHeldWeaponDefinition[InAgent]):
                if. ThisHeldWeapon <> InWeaponDef then. break
            else. break
            Sleep(0.05)

    HandleTrackInventorySlot(InAgent:agent)<suspends>:void=
        if(PreviousWeaponDef:=HeldWeaponDefPerPlayer[InAgent]?):
            PreviousWeaponDef.RecordAmmoAmount(InAgent)
            PreviousWeaponDef.GetStopTrackingWeaponEvent(InAgent).Signal()
        Sleep(0.0)
        AmmoCollection.AllAmmoRemover.Remove(InAgent)
        if(ThisWeaponDef:=GetHeldWeaponDefinition[InAgent]):
            ThisWeaponDef.LoadWeapon(InAgent)
            if(not HeldWeaponDefPerPlayer[InAgent]):
                set HeldWeaponDefPerPlayer = ConcatenateMaps(HeldWeaponDefPerPlayer, map{InAgent=>option{ThisWeaponDef}})
            else if(set HeldWeaponDefPerPlayer[InAgent] = option{ThisWeaponDef}){}
            else{Print("~ERROR: Failed to set current weapon definition per player!")}
        else:
            # Player has switched to a weapon NOT set inside the weapon definitions OR the player switched to their pickaxe #
            if(set HeldWeaponDefPerPlayer[InAgent] = false){}
            else{Print("~ERROR: Failed to clear HeldWeaponDefPerPlayer when pickaxe is equipped!")}
            ThisUI:=InAgent.GetUIManager()
            ThisUI.PowerPunchWidget.Hide()
            ThisUI.AmmoTypeWidget.Hide()

    GrantMaxAmmo<public>(InAgent:?agent):void=
        if(ThisAgent:=InAgent?):
            PlayerWeapons := GetInventoryWeaponDefinitions(ThisAgent)
            for(ThisWeapon:PlayerWeapons):
                ThisWeapon.GrantMaxAmmo(ThisAgent)

    GrantAmmo<public>(InAgent:?agent, Amount:int):void=
        if(ThisAgent:=InAgent?):
            if(HeldWeapon:=GetHeldWeaponDefinition[ThisAgent]):
                HeldWeapon.GrantAmmo(ThisAgent, Amount)

    EnableWeaponVariant(InAgent:?agent):void=
        if(ThisAgent:=InAgent?):
            if(HeldWeapon:=GetHeldWeaponDefinition[ThisAgent]):
                HeldWeapon.EnableVariantDefinition(ThisAgent)
            else:
                #Print("~Log: No weapon held to upgrade")

    # Finds the weapon definition for the weapon that is currently held by the player #
    GetHeldWeaponDefinition(InAgent:agent)<transacts><decides>:weapon_definition=
        var MaybeWeaponDef : ?weapon_definition = false
        for:
            ThisWeaponDef:WeaponDefinitions
            ThisWeaponDef.IsHeld[InAgent]
        do:
            set MaybeWeaponDef = option{ThisWeaponDef}
            MaybeWeaponDef?
        MaybeWeaponDef?

    # Gets all the weapon definitions currently in a players' inventory #
    GetInventoryWeaponDefinitions<public>(InAgent:agent):[]weapon_definition=
        var InventoryWeaponDefinitions : []weapon_definition = array{}
        for(ThisWeaponDef:WeaponDefinitions):
            if(ThisWeaponDef.IsInInventory[InAgent]):
                set InventoryWeaponDefinitions += array{ThisWeaponDef}
        return InventoryWeaponDefinitions

    GetWeaponsOfRarity<public>(InRarity:item_rarity)<transacts>:[]weapon_definition=
        var WeaponsAtRarity : []weapon_definition = array{}
        for(ThisWeapon:WeaponDefinitions):
            if(ThisWeapon.TrialRewardRarity = InRarity):
                set WeaponsAtRarity += array{ThisWeapon}
        return WeaponsAtRarity

    Reset():void=
        PlayerList := GetPlayers()
        for(ThisPlayer:PlayerList):
            for(ThisWeapon:WeaponDefinitions):
                ThisWeapon.ResetAmmo(ThisPlayer.User, ?DoGrantAmmo:=false)

cameradevice_module := class<concrete>:

    @editable FireInput : input_trigger_device = input_trigger_device{}
    @editable AimInput : input_trigger_device = input_trigger_device{}

    @editable NormalCamera : gameplay_camera_orbit_device = gameplay_camera_orbit_device{}
    @editable DefaultAimingCamera : gameplay_camera_orbit_device = gameplay_camera_orbit_device{}
        
# Holds item granters used to grant ammo #
ammo_collection<public> := class<concrete>:

    @editable LightAmmo_ItemGranter : item_granter_device = item_granter_device{}
    @editable LightAmmo_MaxCap : int = 999
    @editable MediumAmmo_ItemGranter : item_granter_device = item_granter_device{}
    @editable MediumAmmo_MaxCap : int = 999
    @editable HeavyAmmo_ItemGranter : item_granter_device = item_granter_device{}
    @editable HeavyAmmo_MaxCap : int = 999
    @editable ShellAmmo_ItemGranter : item_granter_device = item_granter_device{}
    @editable ShellAmmo_MaxCap : int = 999
    @editable RocketAmmo_ItemGranter : item_granter_device = item_granter_device{}
    @editable RocketAmmo_MaxCap : int = 12
    @editable ArrowAmmo_ItemGranter : item_granter_device = item_granter_device{}
    @editable ArrowAmmo_MaxCap : int = 999

    @editable LightAmmo_ConditionalButton : conditional_button_device = conditional_button_device{}
    @editable MediumAmmo_ConditionalButton : conditional_button_device = conditional_button_device{}
    @editable HeavyAmmo_ConditionalButton : conditional_button_device = conditional_button_device{}
    @editable ShellAmmo_ConditionalButton : conditional_button_device = conditional_button_device{}
    @editable RocketAmmo_ConditionalButton : conditional_button_device = conditional_button_device{}
    @editable ArrowAmmo_ConditionalButton : conditional_button_device = conditional_button_device{}

    @editable AllAmmoRemover : item_remover_device = item_remover_device{}

    @editable AmmoDrop_BP : creative_prop_asset = DefaultCreativePropAsset
    @editable DropLifetime : type{_X:float where _X > 0.0} = 18.0
    @editable DropPickupRadius : float = 100.0
    @editable AmmoDrop_SpawnChance<public> : float = 5.5
    @editable RandomAmmoAmount_Lower : int = 8
    @editable RandomAmmoAmount_Upper : int = 25
    @editable DropPickupEffect_BP : creative_prop_asset = DefaultCreativePropAsset
    @editable PickupEffectLifetime : type{_X:float where _X >= 0.0} = 1.5

    # Grants the given agent the given amount of the ammo type provided #
    GrantAmmo(InAgent:agent, InAmount:int, InWeaponDef:weapon_definition)<suspends>:void=
        InWeaponDef.GrantAmmo(InAgent, InAmount)

    # Sets the ammo of this weapon to the given amount #
    SetAmmo(InAgent:agent, InAmount:int, InWeaponDef:weapon_definition, ?Retry:logic=false)<suspends>:void=
        AllAmmoRemover.Remove(InAgent)
        GetAmmoGranter(InWeaponDef.AmmoType).GrantItem(InAgent)
        AmmoConsumer:=GetAmmoConsumer(InWeaponDef.AmmoType)

        # Checks if the amount is greater than the weapon definitions' maximum allowed ammo #
        var AmountToGrant : int = InAmount
        if(InWeaponDef.DoesHaveVariant[InAgent], InAmount >= InWeaponDef.VariantDefinition.GetMaxAmmo(option{InAgent})):
            set AmountToGrant = InWeaponDef.VariantDefinition.GetMaxAmmo(option{InAgent})
        else if(not InWeaponDef.DoesHaveVariant[InAgent], InAmount >= InWeaponDef.GetMaxAmmo(option{InAgent})):
            set AmountToGrant = InWeaponDef.GetMaxAmmo(option{InAgent})
        FinalAmountToGrant := GetAmmoCap(InWeaponDef.AmmoType)-AmountToGrant
        AmmoConsumer.SetItemCountRequired(0, FinalAmountToGrant)
        Sleep(0.05)
        AmmoConsumer.Activate(InAgent)
        Sleep(0.0)
        CurrAmount:=GetAmmoCount(InAgent, InWeaponDef.AmmoType)
        if(CurrAmount <> AmountToGrant, not Retry?):
            SetAmmo(InAgent, InAmount, InWeaponDef, ?Retry:=true)
        #InWeaponDef.RecordAmmoAmount(InAgent)

    # Gets the ammo item granter for a given ammo type #
    GetAmmoGranter(InType:ammotype):item_granter_device=
        case(InType):
            ammotype.Light => return LightAmmo_ItemGranter
            ammotype.Medium => return MediumAmmo_ItemGranter
            ammotype.Heavy => return HeavyAmmo_ItemGranter
            ammotype.Shells => return ShellAmmo_ItemGranter
            ammotype.Rockets => return RocketAmmo_ItemGranter
            ammotype.Arrows => return ArrowAmmo_ItemGranter
            ammotype.None => item_granter_device{}

    # Gets the ammo cap set inside the item granter for a given ammo type #
    GetAmmoCap(InType:ammotype):int=
        case(InType):
            ammotype.Light => return LightAmmo_MaxCap
            ammotype.Medium => return MediumAmmo_MaxCap
            ammotype.Heavy => return HeavyAmmo_MaxCap
            ammotype.Shells => return ShellAmmo_MaxCap
            ammotype.Rockets => return RocketAmmo_MaxCap
            ammotype.Arrows => return ArrowAmmo_MaxCap
            ammotype.None => -1

    # Gets the ammo count for the given ammo type that a player currently has #
    GetAmmoCount(InAgent:agent, InType:ammotype):int=
        case(InType):
            ammotype.Light => return LightAmmo_ConditionalButton.GetItemCount(InAgent, 0)
            ammotype.Medium => return MediumAmmo_ConditionalButton.GetItemCount(InAgent, 0)
            ammotype.Heavy => return HeavyAmmo_ConditionalButton.GetItemCount(InAgent, 0)
            ammotype.Shells => return ShellAmmo_ConditionalButton.GetItemCount(InAgent, 0)
            ammotype.Rockets => return RocketAmmo_ConditionalButton.GetItemCount(InAgent, 0)
            ammotype.Arrows => return ArrowAmmo_ConditionalButton.GetItemCount(InAgent, 0)
            ammotype.None => -1

    # Gets the ammo conditional button for the given ammo type #
    GetAmmoConsumer(InType:ammotype):conditional_button_device=
        case(InType):
            ammotype.Light => return LightAmmo_ConditionalButton
            ammotype.Medium => return MediumAmmo_ConditionalButton
            ammotype.Heavy => return HeavyAmmo_ConditionalButton
            ammotype.Shells => return ShellAmmo_ConditionalButton
            ammotype.Rockets => return RocketAmmo_ConditionalButton
            ammotype.Arrows => return ArrowAmmo_ConditionalButton
            ammotype.None => conditional_button_device{}

    SpawnAmmoDrop<public>(InTransform:transform)<suspends>:void=
        if(SpawnedAmmoDrop:=SpawnProp(AmmoDrop_BP, InTransform.Translation+vector3{Z:=15.0}, InTransform.Rotation)(0)?):
            var PickingUpPlayer : ?agent = false
            race:
                block:
                    loop:
                        ActivePlayers:=GetPlayers()
                        for(ThisPlayer:ActivePlayers):
                            if(ThisFortChar:=ThisPlayer.GetFortCharacter[], Distance(InTransform.Translation, ThisFortChar.GetTransform().Translation) < DropPickupRadius):
                                set PickingUpPlayer = option{ThisPlayer.User}
                                break
                        Sleep(0.1)
                block:
                    SafeSleep(DropLifetime)
            SpawnedAmmoDrop.Dispose()
            if(ThisPlayer:=PickingUpPlayer?):
                branch:
                    if(PickupEffectLifetime > 0.0, SpawnedPickupEffect:=SpawnProp(DropPickupEffect_BP, InTransform.Translation+vector3{Z:=15.0}, InTransform.Rotation)(0)?):
                        SafeSleep(PickupEffectLifetime)
                        SpawnedPickupEffect.Dispose()
                if(ThisWeapon:=ThisPlayer.GetHeldWeapon[] or ThisPlayer.GetInventoryWeapons()[0]):
                    ThisMinimum := if. ThisWeapon.AmmoDrop_Minimum > 0 then. ThisWeapon.AmmoDrop_Minimum else. RandomAmmoAmount_Lower
                    ThisMaximum := if. ThisWeapon.AmmoDrop_Maximum > 0 then. ThisWeapon.AmmoDrop_Maximum else. RandomAmmoAmount_Upper
                    AmmoAmount:=GetRandomInt(ThisMinimum, ThisMaximum)
                    Print("AmmoDrop: {AmmoAmount}")
                    ThisWeapon.GrantAmmo(ThisPlayer, AmmoAmount)
                else:
                    Print("~Failed to get a players weapon to grant ammo!")


