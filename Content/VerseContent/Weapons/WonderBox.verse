using { /Fortnite.com/Devices }
using { /Fortnite.com/Characters }
using { /Fortnite.com/Game }
using { /Fortnite.com/UI }
using { /Fortnite.com/FortPlayerUtilities }
using { /Verse.org/Simulation }
using { /Verse.org/Random }
using { /Verse.org/Colors }
using { /UnrealEngine.com/Temporary/Diagnostics }
using { /UnrealEngine.com/Temporary/SpatialMath }
using { /UnrealEngine.com/Temporary/UI }
using { /Fortnite.com/Devices/CreativeAnimation }
using { /Fortnite.com/Devices/CreativeAnimation/InterpolationTypes }
using { InteractiveModule }
using { Player }
using { Spawning }

########  ########     ###     ######    #######  ##    ## ########     ###     ######   ######## 
##     ## ##     ##   ## ##   ##    ##  ##     ## ###   ## ##     ##   ## ##   ##    ##  ##       
##     ## ##     ##  ##   ##  ##        ##     ## ####  ## ##     ##  ##   ##  ##        ##       
##     ## ########  ##     ## ##   #### ##     ## ## ## ## ########  ##     ## ##   #### ######   
##     ## ##   ##   ######### ##    ##  ##     ## ##  #### ##   ##   ######### ##    ##  ##       
##     ## ##    ##  ##     ## ##    ##  ##     ## ##   ### ##    ##  ##     ## ##    ##  ##       
########  ##     ## ##     ##  ######    #######  ##    ## ##     ## ##     ##  ######   ########

GetActiveWonderBox()<transacts><decides>:wonder_box=
    return GlobalWonderBoxManager[GetSession()].CurrentBoxLocation?

var GlobalWonderBoxManager<public> : weak_map(session, Wonder_Box_Manager) = map{}

boxspins_int := type{_X : int where 0 < _X}
boxspinmax_float := type{_X : float where 0.0 < _X, 100.0 > _X}
Wonder_Box_Manager<public> := class(creative_device):

    BoxLinkedEvent:event()=event(){}
    var<private> BoxLocations<public> : []wonder_box = array{}
    var<private> PrioritizedBoxes : []wonder_box = array{}
    AddWonderBox(InBox:wonder_box):void=
        if(not BoxLocations.Find[InBox]):
            set BoxLocations += array{InBox}
            BoxLinkedEvent.Signal()
            InBox.InitBox(Self)
            if(InBox.IsPrioritized?):
                set PrioritizedBoxes += array{InBox}

    @editable Controls_Module : wonderbox_controls = wonderbox_controls{}
    @editable Prop_Module : wonderbox_prop = wonderbox_prop{}
    @editable Audio_Module : wonderbox_audio = wonderbox_audio{}
    @editable Signal_Recieve : wonder_box_device_signalrecieve = wonder_box_device_signalrecieve{}
    @editable Signal_Send : wonder_box_device_signalsend = wonder_box_device_signalsend{}

    var<private> EnabledStatePerPlayer : [agent]logic = map{}
    IsEnabledForPlayer(InAgent:agent)<transacts><decides>:void=
        if(EnabledStatePerPlayer[InAgent]):
            EnabledStatePerPlayer[InAgent]?
        else:
            set EnabledStatePerPlayer = ConcatenateMaps(EnabledStatePerPlayer, map{InAgent => true})
    SetStateForPlayer(InAgent:agent, InState:logic):void=
        if(set EnabledStatePerPlayer[InAgent] = InState){}
        else:
            set EnabledStatePerPlayer = ConcatenateMaps(EnabledStatePerPlayer, map{InAgent => InState})

    var<private> AllowBoxRotation<public> : logic = true
    var<private> CurrentBoxLocation<public> : ?wonder_box = false

    var<private> BoxPrice<public> : int = 950
    var<private> RotateState<public> : logic = true
    SetRotateState<public>(InState:logic):void=
        set RotateState = InState
    
    OnBegin<override>()<suspends>:void=
        if(set GlobalWonderBoxManager[GetSession()] = Self){}
        else{Print("~ERROR: Failed to globally link WonderboxManager!")}
        set BoxPrice = Controls_Module.DefaultPrice
        InitSignals()
        loop:
            Winner:=race:
                block:
                    BoxLinkedEvent.Await()
                    0
                block:
                    Sleep(2.5)
                    1
            if(Winner=1):
                break
            Sleep(0.0)
        Sleep(0.5)
        if(GetEnabledBoxLocations().Length <= 1):
            if(Singularwonderbox := GetEnabledBoxLocations()[0]):
                spawn{Singularwonderbox.SpawnBox()}
                set CurrentBoxLocation = option{Singularwonderbox}
            else{Print("~ERR: No box locations set up or none are enabled!")}
        else if(PrioritizedBoxes.Length > 0, ThisPriorityBox:=Shuffle(PrioritizedBoxes)[0]):
            spawn{ThisPriorityBox.SpawnBox()}
            set CurrentBoxLocation = option{ThisPriorityBox}
        else:
            RotateBox()
        AwaitGameState(game_state.InGame)
        if(GetGameDifficulty() = game_difficulty.Easy):
            if(Val := type{_X:int where _X >= 1}[Controls_Module.MaximumWave-3]):
                set Controls_Module.MaximumWave = Val

    Reset()<suspends>:void=
        if(ThisCurrentBox:=CurrentBoxLocation?):
            ThisCurrentBox.DespawnBox()
        if(ThisPriorityBox:=Shuffle(PrioritizedBoxes)[0]):
            spawn{ThisPriorityBox.SpawnBox()}
            set CurrentBoxLocation = option{ThisPriorityBox}
        else:
            RotateBox()

    InitSignals():void=
        Signal_Recieve.OnTrigger_SpinBox.TriggeredEvent.Subscribe(SpinBox)
        Signal_Recieve.OnTrigger_ForceRotateBox.TriggeredEvent.Subscribe(ForceRotateBox)
        Signal_Recieve.OnTrigger_EnableBoxRotation.TriggeredEvent.Subscribe(EnableBoxRotation)
        Signal_Recieve.OnTrigger_DisableBoxRotation.TriggeredEvent.Subscribe(DisableBoxRotation)

    SpinBox(InAgent : ?agent):void=
        if(ValidAgent := InAgent?):
            if(ThisCurrentBox := CurrentBoxLocation?):
                Signal_Send.ActivateTrigger_OnBoxSpin.Trigger(ValidAgent)
                spawn{ThisCurrentBox.SpinBox(ValidAgent)}

    ForceRotateBox(InAgent : ?agent):void=
        spawn{RotateBox()}

    RotateBox(?DesiredLocation:?wonder_box=false)<suspends>:void=
        if(GetEnabledBoxLocations().Length > 1, AllowBoxRotation?):
            if(Thiswonderbox := CurrentBoxLocation?):
                if(ThisPreviousBoxUser := Thiswonderbox.PreviousBoxUser?):
                    Signal_Send.ActivateTrigger_OnBoxRotated.Trigger(ThisPreviousBoxUser)
                if(Thiswonderbox.EnabledState?):
                    Thiswonderbox.DespawnBox()
            if(Chosenwonderbox:=DesiredLocation? or GetRandomBox[]):
                Sleep(5.0)
                set CurrentBoxLocation = option{Chosenwonderbox}
                spawn{Chosenwonderbox.SpawnBox()}
            else{Print("~ERR: Failed to get a new random box location!")}
        else{Print("~ERR: Tried to rotate box when there is only one location!")}

    GetRandomBox()<transacts><decides>:wonder_box=
        var OtherBoxLocations : []wonder_box = array{}
        for(ThisBox:GetEnabledBoxLocations()):
            if(not CurrentBoxLocation? or (ThisBox <> CurrentBoxLocation?)):
                set OtherBoxLocations += array{ThisBox}
        return Shuffle(OtherBoxLocations)[0]

    GetEnabledBoxLocations()<transacts>:[]wonder_box=
        var EnabledBoxLocations : []wonder_box = array{}
        for(ThisBoxLocation:BoxLocations):
            if(ThisBoxLocation.IsLocationEnabled?, ThisBoxLocation.InteractableEnabledState?):
                set EnabledBoxLocations += array{ThisBoxLocation}
        return EnabledBoxLocations

    EnableBoxRotation<public>(InAgent : ?agent):void=
        set AllowBoxRotation = true

    DisableBoxRotation<public>(InAgent : ?agent):void=
        set AllowBoxRotation = false

    SetPrice<public>(InPrice:int):void=
        set BoxPrice = InPrice

    ResetPrice<public>():void=
        set BoxPrice = Controls_Module.DefaultPrice

wonderbox_controls := class<concrete>:

    @editable DefaultPrice : int = 950
    @editable Starting_ZOffsetToBox : float = 50.0
    @editable Ending_ZOffsetToBox : float = 150.0
    @editable WeaponRotation : float = 90.0
    @editable WeaponMovingUpDuration : boxspinmax_float = 1.0
    @editable BoxSpinningDuration : boxspinmax_float = 6.2
    @editable BoxItemDuration : boxspinmax_float = 10.0
    @editable DurationBeforeItemDescends : boxspinmax_float = 1.5
    @editable MaximumItemSpins : type{_X:int where _X >= 1} = 6
    @editable MaximumBoxSpins : boxspins_int = 10
    @editable AllowMoreSpinsThanMax : logic = true
    @editable ChancePerMaxSpin : boxspinmax_float = 60.0

    @editable
    WeaponChances_Starting : weapon_chances = weapon_chances{}

    @editable
    WeaponChances_Maximum : weapon_chances = weapon_chances{}

    @editable
    var MaximumWave : type{_X:int where _X >= 1} = 20

    GetWeaponRarityFromChance<public>()<transacts>:item_rarity=
        WavePercent := block{
            ThisPercent:=(1.0*GetWaveCount())/(MaximumWave*1.0)
            ThisPercent < 1.0
            ThisPercent
        } or 1.0
        WeaponChances := array:
            LockLerp(WeaponChances_Starting.Common_Chance, WeaponChances_Maximum.Common_Chance, WavePercent)*100.0
            LockLerp(WeaponChances_Starting.Rare_Chance, WeaponChances_Maximum.Rare_Chance, WavePercent)*100.0
            LockLerp(WeaponChances_Starting.Epic_Chance, WeaponChances_Maximum.Epic_Chance, WavePercent)*100.0
            LockLerp(WeaponChances_Starting.Legendary_Chance, WeaponChances_Maximum.Legendary_Chance, WavePercent)*100.0
            LockLerp(WeaponChances_Starting.Mythic_Chance, WeaponChances_Maximum.Mythic_Chance, WavePercent)*100.0
        <#
        for(Index->ThisChance:WeaponChances):
            case(Index):
                0 => Print("{item_rarity.Common.ToString()} chance: {ThisChance}")
                1 => Print("{item_rarity.Rare.ToString()} chance: {ThisChance}")
                2 => Print("{item_rarity.Epic.ToString()} chance: {ThisChance}")
                3 => Print("{item_rarity.Legendary.ToString()} chance: {ThisChance}")
                4 => Print("{item_rarity.Mythic.ToString()} chance: {ThisChance}")
                _ => 
        #>
        ThisRarity:=case(RandomWeightedChance(WeaponChances)):
            0 => item_rarity.Common
            1 => item_rarity.Rare
            2 => item_rarity.Epic
            3 => item_rarity.Legendary
            4 => item_rarity.Mythic
            _ => item_rarity.Common
        #Print("Wonder box rarity: {ThisRarity.ToString()}")
        ThisRarity

weapon_chances<public> := class<concrete>:

    @editable
    Common_Chance : type{_X:float where _X >= 0.0, _X <= 1.0} = 0.739

    @editable
    Rare_Chance : type{_X:float where _X >= 0.0, _X <= 1.0} = 0.20

    @editable
    Epic_Chance : type{_X:float where _X >= 0.0, _X <= 1.0} = 0.05

    @editable
    Legendary_Chance : type{_X:float where _X >= 0.0, _X <= 1.0} = 0.01

    @editable
    Mythic_Chance : type{_X:float where _X >= 0.0, _X <= 1.0} = 0.001

wonderbox_prop := class<concrete>:
    
    @editable OpenAnimation : creative_prop_asset = DefaultCreativePropAsset
    @editable Open_ActiveTime : float = 0.5

    @editable CloseAnimation : creative_prop_asset = DefaultCreativePropAsset
    @editable Close_ActiveTime : float = 0.566667

    @editable SpawningAnimation : creative_prop_asset = DefaultCreativePropAsset
    @editable Spawning_ActiveTime : float = 0.53

    @editable DespawnAnimation : creative_prop_asset = DefaultCreativePropAsset
    @editable Despawn_ActiveTime : float = 4.4

    @editable UseTauntProp : logic = false
    @editable TauntProp : creative_prop_asset = DefaultCreativePropAsset

    @editable StaticBoxProp : creative_prop_asset = DefaultCreativePropAsset

    @editable UseEmptyBoxProp : logic = false
    @editable EmptyBoxProp : creative_prop_asset = DefaultCreativePropAsset

wonderbox_audio := class<concrete>:

    @editable BoxAmbienceLoop : ?creative_prop_asset = false
    @editable BoxOpen_Audio : timed_audio = timed_audio{}
    @editable BoxClose_Audio : timed_audio = timed_audio{}
    @editable BoxSpinning_Audio : timed_audio = timed_audio{}
    @editable BoxMove_Audio : timed_audio = timed_audio{}

    PlayOpenAndSpinAudio(BoxPosition:vector3):void=
        spawn{BoxOpen_Audio.PlayAudio(BoxPosition)}
        spawn{BoxSpinning_Audio.PlayAudio(BoxPosition)}

    PlayCloseAudio(BoxPosition:vector3):void=
        spawn{BoxClose_Audio.PlayAudio(BoxPosition)}
        BoxSpinning_Audio.StopAudio()

    PlayMoveAudio(BoxPosition:vector3):void=
        spawn{BoxMove_Audio.PlayAudio(BoxPosition)}

timed_audio := class<concrete>:

    @editable AudioBP : ?creative_prop_asset = false
    @editable ActiveTime : type{_X : float where 0.0 <= _X} = 0.0
    @editable DelayTime : type{_X : float where 0.0 <= _X} = 0.0

    PlayAudio(Position:vector3)<suspends>:void=
        Sleep(DelayTime)
        if(ThisAudio:=AudioBP?, ThisSpawnedAudio:=SpawnProp(ThisAudio, Position, rotation{})(0)?):
            race:
                Sleep(ActiveTime)
                StopAudioEvent.Await()
            ThisSpawnedAudio.Dispose()

    StopAudioEvent<private>:event()=event(){}
    StopAudio():void=
        StopAudioEvent.Signal()

wonder_box_device_signalrecieve := class<concrete>:

    @editable OnTrigger_SpinBox : trigger_device = trigger_device{}
    @editable OnTrigger_ForceRotateBox : trigger_device = trigger_device{}
    @editable OnTrigger_EnableBoxRotation : trigger_device = trigger_device{}
    @editable OnTrigger_DisableBoxRotation : trigger_device = trigger_device{}

wonder_box_device_signalsend := class<concrete>:

    @editable ActivateTrigger_OnBoxSpin : trigger_device = trigger_device{}
    @editable ActivateTrigger_OnBoxWeaponSelected : trigger_device = trigger_device{}
    @editable ActivateTrigger_OnBoxWeaponDespawned : trigger_device = trigger_device{}
    @editable ActivateTrigger_OnBoxRotated : trigger_device = trigger_device{}

wonder_box<public> := class<concrete><unique>(Interactive_Base):

    InteractiveType<override> : interactable_type = interactable_type.Weapon

    @editable Identity : string = "Unset WonderBox Identity"
    @editable BoxLocation : creative_prop = creative_prop{}
    @editable Beacon : creative_prop = creative_prop{}
    @editable IsLocationEnabled : logic = true
    @editable IsPrioritized : logic = false
    @editable OnTrigger_EnableChoice : trigger_device = trigger_device{}
    @editable OnTrigger_ForceBoxSwitch : trigger_device = trigger_device{}

    @editable InteractiveComponent : Interactable = Interactable{
        InteractionSpriteOffset:=vector3{Z:=150.0}, ExtraInteractionRadiusColliders:=array{vector3{Y:=65.0},vector3{Y:=-65.0},vector3{Z:=45.0}}, InteractDistance:=200.0, InteractTime:=0.5
    }

    var<private> BoxTransform<public> : transform = transform{}

    var<private> WonderboxManager<public> : Wonder_Box_Manager = Wonder_Box_Manager{}
    var<private> BoxSpinCount<public> : int = 0
    var<private> EnabledState<public> : logic = false

    var<private> BoxClosedEvent<public> : event() = event(){}
    var<private> BoxSpinningCompleteEvent<public> : event() = event(){}
    var<private> WeaponCollectedEvent<public> : event() = event(){}
    var<private> BoxSpinningState<public> : logic = true

    var<private> PreviousBoxUser<public> : ?agent = false
    var<private> CurrentWeapon : ?weapon_definition = false
    var<private> IsInUse<public> : logic = false

    #var<private> WeaponProps : []item_placer_device = array{}

    OnBegin<override>()<suspends>:void=
        Beacon.Hide()
        OnTrigger_EnableChoice.TriggeredEvent.Subscribe(EnableViaTrigger)
        OnTrigger_ForceBoxSwitch.TriggeredEvent.Subscribe(ForceBoxSwitchTrigger)
        if(BoxLocation.IsValid[]):
            set InteractableEnabledState = IsLocationEnabled
        else:
            SetEnabledState(false)
            Print("~ERROR: 'BoxLocation' field is empty or invalid at identity [{Identity}]!")
        loop:
            if(ThisWonderboxManager:=GlobalWonderBoxManager[GetSession()]):
                ThisWonderboxManager.AddWonderBox(Self)
                break
            Sleep(0.0)
        LinkAndCompile(array{InteractiveComponent})

    InitBox(InwonderboxDevice : Wonder_Box_Manager):void=
        SetEnabledState(false)
        set WonderboxManager = InwonderboxDevice
        HideBox()
        if(BoxLocation.IsValid[]):
            set BoxTransform = BoxLocation.GetTransform()
        else{Print("~ERROR: 'BoxLocation' is empty at identity [{Identity}]!")}

    var InteractableEnabledState : logic = true
    EnableViaTrigger<private>(Agent:?agent):void=
        set InteractableEnabledState = true

    ForceBoxSwitchTrigger(Agent:?agent):void=
        spawn{WonderboxManager.RotateBox(?DesiredLocation:=option{Self})}

    OnInteracted<override>(InAgent:agent, InteractingLocation:Interactable):void=
        if(not IsInUse? and WonderboxManager.IsEnabledForPlayer[InAgent]):
            if(ThisPlayerInfo:=InAgent.GetPlayerInfo[]):
                BuyState:=ThisPlayerInfo.BuyItem(WonderboxManager.BoxPrice)
                if(BuyState?):
                    set IsInUse = true
                    SetEnabledState(false)
                    spawn{SpinBox(InAgent)}
                #else{Print("~Insufficient funds for wonderbox!")}
        else if(InAgent=PreviousBoxUser?):
            set PreviousBoxUser = false
            if(ThisWeapon:=CurrentWeapon?):
                #SPrint("~Granted wonder weapon to agent!")
                ThisWeapon.GrantWeapon(InAgent)
                WeaponCollectedEvent.Signal()
        else:
            SetEnabledState(false, ?InAgent:=option{InAgent})
            Print("~Failed all OnInteracted conditions")

    GetInteractionText<override>(InAgent:agent, Interactive:Interactable):string=
        if(not IsInUse? and WonderboxManager.IsEnabledForPlayer[InAgent] and not InAgent.IsUsingItemCollection[], EnabledState?):
            SetInteractableState(true, ?InAgent:=option{InAgent})
            return "Spin Wonder Box [<C_Y>${WonderboxManager.BoxPrice}<E>]"
        else if(InAgent=PreviousBoxUser?):
            SetInteractableState(true, ?InAgent:=option{InAgent})
            if(InAgent.CheckInventorySpace[]):
                return "Grab Weapon"
            else:
                return "Swap Weapon"
        SetInteractableState(false, ?InAgent:=option{InAgent})
        return ""

    SpinBoxNew<public>(InAgent:agent)<suspends>:void=
        SetEnabledState(false)
        if(WeaponManager:=GlobalWeaponManager[GetSession()]):
            ThisItemCollection:=InAgent.GetItemCollection(false)


    SpinBox(InAgent : agent)<suspends>:void=
        SetEnabledState(false)
        if(WeaponManager:=GlobalWeaponManager[GetSession()]):
            ThisItemCollection:=InAgent.GetItemCollection(false)
            if(WeaponAnchor:=InAgent.GetPlayerPackage[].WeaponAnchor, WeaponAnchor.TeleportTo[vector3{X:=0.0,Y:=0.0,Z:=0.0}, WeaponAnchor.GetTransform().Rotation]){}
            ThisItemCollection.Hide()
            set BoxSpinningState = true
            set PreviousBoxUser = option{InAgent}
            PlayAnimation(WonderboxManager.Prop_Module.OpenAnimation)
            WonderboxManager.Audio_Module.PlayOpenAndSpinAudio(BoxTransform.Translation)
            if(WonderboxManager.GetEnabledBoxLocations().Length > 1):
                if(WonderboxManager.RotateState?, WonderboxManager.Controls_Module.MaximumItemSpins <= BoxSpinCount):
                    if(WonderboxManager.Controls_Module.AllowMoreSpinsThanMax?, WonderboxManager.Controls_Module.MaximumBoxSpins > BoxSpinCount):
                        if(WonderboxManager.Controls_Module.ChancePerMaxSpin < GetRandomFloat(0.0, 100.0)):
                            set EnabledState = false
                    else:
                        set EnabledState = false
            spawn{MoveWeaponsUp(InAgent)}
            InAgent.SetItemCollectionState(true)
            spawn{SpinWeapons(InAgent, InAgent.GetItemCollection(true))}
            ThisItemCollection.Hide()
            SafeSleep(WonderboxManager.Controls_Module.BoxSpinningDuration)
            BoxSpinningCompleteEvent.Signal()
            MoveWeaponsDown(InAgent)
            set BoxSpinningState = false
            var PickedUpState : logic = false
            var CurrentWeaponProp : item_placer_device = item_placer_device{}
            if(EnabledState?):
                SetEnabledState(true, ?InAgent:=option{InAgent})
                SetInteractableState(true, ?InAgent:=option{InAgent})
                ThisItemCollection.Hide()
                Sleep(0.0)
                ThisWeaponRarity := WonderboxManager.Controls_Module.GetWeaponRarityFromChance()
                if(ThisRandomWeapon := Shuffle(GetWeaponsOfWonderBoxRarity(ThisWeaponRarity, InAgent))[0] or Shuffle(GetWeaponsOfWonderBoxRarity(ThisWeaponRarity.Decrease(), InAgent))[0]):
                    if(ThisRandomWeaponProp := ThisRandomWeapon.GetPerPlayerItemPlacer[InAgent], ThisRandomWeaponProp.TeleportTo[BoxTransform.Translation+vector3{Z:=WonderboxManager.Controls_Module.Ending_ZOffsetToBox}, BoxLocation.GetTransform().Rotation.ApplyYaw(DegreesToRadians(WonderboxManager.Controls_Module.WeaponRotation))]):
                        set CurrentWeaponProp = ThisRandomWeaponProp
                    else:
                        Print("~ERROR: Failed to fetch the prop associated with the weapon [{ThisRandomWeapon.Type.ToString()}]!")
                    set CurrentWeapon = option{ThisRandomWeapon}
                        race:
                            block:
                                SafeSleep(WonderboxManager.Controls_Module.BoxItemDuration + WonderboxManager.Controls_Module.DurationBeforeItemDescends)
                                WonderboxManager.Signal_Send.ActivateTrigger_OnBoxWeaponDespawned.Trigger(InAgent)
                            block:
                                WeaponCollectedEvent.Await()
                                WonderboxManager.Signal_Send.ActivateTrigger_OnBoxWeaponSelected.Trigger(InAgent)
                                set PickedUpState = true
                            block:
                                SafeSleep(WonderboxManager.Controls_Module.DurationBeforeItemDescends)
                                SafeMoveTo(CurrentWeaponProp, BoxTransform.Translation+vector3{Z:=WonderboxManager.Controls_Module.Starting_ZOffsetToBox}, BoxLocation.GetTransform().Rotation.ApplyYaw(DegreesToRadians(WonderboxManager.Controls_Module.WeaponRotation)), WonderboxManager.Controls_Module.BoxItemDuration)
                        if(WonderboxManager.RotateState?):
                            set BoxSpinCount += 1
                    else:
                        Print("~ERROR: Failed to fetch the prop for a wonder box weapon!")
                else:
                    Print("~ERROR: Failed to fetch a random wonder box weapon from the weapon manager!")
            else:
                # Box is despawning, and moving to another location #
                CurrentWeaponProp.Hide()
                if(WeaponAnchor:=InAgent.GetPlayerPackage[].WeaponAnchor, WeaponAnchor.TeleportTo[vector3{X:=0.0,Y:=0.0,Z:=0.0}, WeaponAnchor.GetTransform().Rotation]){}
                InAgent.SetItemCollectionState(false)
                if(ThisPlayerInfo:=InAgent.GetPlayerInfo[]):
                    ThisPlayerInfo.IncreasePoints(WonderboxManager.BoxPrice)
                set BoxSpinCount = 0
                ThisItemCollection.Hide()
                DespawnBox()
                WonderboxManager.Audio_Module.PlayMoveAudio(BoxTransform.Translation)
                spawn{WonderboxManager.RotateBox()}
                set IsInUse = false
                set PreviousBoxUser = false
                return
            WonderboxManager.Audio_Module.PlayCloseAudio(BoxTransform.Translation)
            if(PickedUpState?):
                spawn{PlayAnimation(WonderboxManager.Prop_Module.CloseAnimation, ?Time:=WonderboxManager.Prop_Module.Close_ActiveTime)}
            else:
                PlayAnimation(WonderboxManager.Prop_Module.CloseAnimation, ?Time:=WonderboxManager.Prop_Module.Close_ActiveTime)
            if(WeaponAnchor:=InAgent.GetPlayerPackage[].WeaponAnchor, WeaponAnchor.TeleportTo[vector3{X:=0.0,Y:=0.0,Z:=0.0}, WeaponAnchor.GetTransform().Rotation]){}
            ThisItemCollection.Hide()
            CurrentWeaponProp.Hide()
            BoxClosedEvent.Signal()
            set IsInUse = false
            set PreviousBoxUser = false
            if(ThisPlayerInfo:=InAgent.GetPlayerInfo[]):
                InAgent.SetItemCollectionState(false)
            if(EnabledState?):
                SetEnabledState(true)
        else:
            Print("~ERR: Failed to fetch the weapon manager!")
    
    SpinWeapons(InAgent:agent, InItemPlacers : []item_placer_device, ?InTimePerWeapon:float=0.1)<suspends>:void=
        var TimePerWeapon : float = InTimePerWeapon
        if(BoxSpinningState?, InItemPlacers.Length > 0):
            race:
                loop:
                    BoxWeapons := Shuffle(InItemPlacers)
                    for(ThisWeapon : BoxWeapons):
                        ThisWeapon.Show(InAgent)
                        SafeSleep(TimePerWeapon)
                        ThisWeapon.Hide()
                        set TimePerWeapon *= 1.1
                block:
                    BoxSpinningCompleteEvent.Await()

    MoveWeaponsUp(InAgent:agent)<suspends>:void=
        if(WeaponAnchor:=InAgent.GetPlayerPackage[].WeaponAnchor):
            ThisItemCollection:=InAgent.GetItemCollection(false)
            if(WeaponAnchor.TeleportTo[vector3{Z:=WonderboxManager.Controls_Module.Starting_ZOffsetToBox}, BoxLocation.GetTransform().Rotation.ApplyYaw(DegreesToRadians(WonderboxManager.Controls_Module.WeaponRotation))]):
                SafeSleep(0.75)
                SafeMoveTo(WeaponAnchor, vector3{Z:=WonderboxManager.Controls_Module.Ending_ZOffsetToBox}, WeaponAnchor.GetTransform().Rotation, WonderboxManager.Controls_Module.WeaponMovingUpDuration)
            else{Print("~ERROR: Failed to teleport WewaponAttachingProp in wonderbox!")}
        <#
        if(WeaponAnchor:=InAgent.GetPlayerPackage[].WeaponAnchor):
            ThisItemCollection:=InAgent.GetItemCollection(false)
            if(WeaponAnchor.TeleportTo[BoxTransform.Translation+vector3{Z:=WonderboxManager.Controls_Module.Starting_ZOffsetToBox}, BoxLocation.GetTransform().Rotation.ApplyYaw(DegreesToRadians(WonderboxManager.Controls_Module.WeaponRotation))]):
                Sleep(0.25)
                ThisItemCollection.Hide()
                Sleep(0.5)
                WeaponAnchor.MoveTo(BoxTransform.Translation+vector3{Z:=WonderboxManager.Controls_Module.Ending_ZOffsetToBox}, WeaponAnchor.GetTransform().Rotation, WonderboxManager.Controls_Module.WeaponMovingUpDuration)
            else{Print("~ERROR: Failed to teleport WewaponAttachingProp in wonderbox!")}
            #>

    MoveWeaponsDown(InAgent:agent):void=
        if(WeaponAnchor:=InAgent.GetPlayerPackage[].WeaponAnchor):
            spawn. SafeMoveTo(WeaponAnchor, vector3{Z:=WonderboxManager.Controls_Module.Starting_ZOffsetToBox}, WeaponAnchor.GetTransform().Rotation, WonderboxManager.Controls_Module.BoxItemDuration)

    (ItemPlacer:item_placer_device).Show(InAgent:agent):void=
        if(WeaponAnchor:=InAgent.GetPlayerPackage[].WeaponAnchor):
            ThisTransform:=WeaponAnchor.GetTransform()
            if(ItemPlacer.TeleportTo[BoxTransform.Translation+ThisTransform.Translation, ThisTransform.Rotation]){}
            else{Print("~ERROR: Could not teleport/show an item placer!")}

    SpawnBox<public>()<suspends>:void=
        PlayAnimation(WonderboxManager.Prop_Module.SpawningAnimation, ?Time:=WonderboxManager.Prop_Module.Spawning_ActiveTime)
        ShowBox()
        Beacon.Show()
        set EnabledState = true
        SetEnabledState(true)
        if(ThisAmbienceProp := SpawnProp(WonderboxManager.Audio_Module.BoxAmbienceLoop?, BoxTransform.Translation, rotation{})(0)?):
            set AmbienceProp = ThisAmbienceProp

    DespawnBox<public>()<suspends>:void=
        SetEnabledState(false)
        set EnabledState = false
        Beacon.Hide()
        #spawn{PlayTauntAnimation()}
        AmbienceProp.Dispose()
        PlayAnimation(WonderboxManager.Prop_Module.DespawnAnimation, ?Time:=WonderboxManager.Prop_Module.Despawn_ActiveTime)
        HideBox()

    var CurrentBoxProp : creative_prop = creative_prop{}
    var AmbienceProp : creative_prop = creative_prop{}
    PlayAnimation(InAnimAsset:creative_prop_asset, ?Time:float=-1.0)<suspends>:void=
        if(SpawnedBoxAnim:=SpawnProp(InAnimAsset, BoxTransform.Translation, BoxTransform.Rotation)(0)?):
            CurrentBoxProp.Dispose()
            set CurrentBoxProp = SpawnedBoxAnim
            if(Time > 0.0):
                Sleep(Time)

    PlayTauntAnimation()<suspends>:void=
        if(WonderboxManager.Prop_Module.UseTauntProp?):
            if:
                SpawnedTauntProp:=SpawnProp(WonderboxManager.Prop_Module.TauntProp, BoxTransform.Translation+vector3{Z:=WonderboxManager.Controls_Module.Ending_ZOffsetToBox}, BoxTransform.Rotation)(0)?
                AnimControl:=SpawnedTauntProp.GetAnimationController[]
            then:
                TauntAnim : []keyframe_delta = array{
                    keyframe_delta{
                        DeltaLocation := vector3{Z:=-18.0}
                        DeltaRotation := rotation{}
                        Time := 0.7
                        Interpolation := EaseOut
                    }
                    keyframe_delta{
                        DeltaLocation := vector3{Z:=+18.0}
                        DeltaRotation := rotation{}
                        Time := 0.7
                        Interpolation := EaseOut
                    }
                    keyframe_delta{
                        DeltaLocation := vector3{Z:=-25.0}
                        DeltaRotation := rotation{}
                        Time := 0.7
                        Interpolation := EaseOut
                    }
                    keyframe_delta{
                        DeltaLocation := vector3{Z:=4000.0}
                        DeltaRotation := rotation{}
                        Time := 1.25
                        Interpolation := EaseOut
                    }
                }
                AnimControl.SetAnimation(TauntAnim, ?Mode:=animation_mode.OneShot)
                AnimControl.Play()


    ShowBox():void=
        spawn{PlayAnimation(WonderboxManager.Prop_Module.StaticBoxProp)}

    HideBox():void=
        if(WonderboxManager.Prop_Module.UseEmptyBoxProp?):
            spawn{PlayAnimation(WonderboxManager.Prop_Module.EmptyBoxProp)}
        else:
            CurrentBoxProp.Dispose()