using { /Fortnite.com/Devices }
using { /Fortnite.com/Characters }
using { /Fortnite.com/Game }
using { /Fortnite.com/UI }
using { /Fortnite.com/FortPlayerUtilities }
using { /Verse.org/Simulation }
using { /Verse.org/Random }
using { /Verse.org/Colors }
using { /UnrealEngine.com/Temporary/Diagnostics }
using { /UnrealEngine.com/Temporary/SpatialMath }
using { /UnrealEngine.com/Temporary/UI }
using { /Fortnite.com/AI }
using { Player }
using { UI }
using { Spawning }
using { Powerups }
using { Weapons }
using { InteractiveModule }

(InAgent:agent).GetTrialRewardRarity<public>()<transacts>:item_rarity=
    InAgent.GetTrialPoints().ToTrialRarity()

(InAmount:int).ToTrialRarity<public>()<transacts>:item_rarity=
    if(InAmount > 100):
        if(InAmount < 200):
            return item_rarity.Common
        else if(InAmount < 300):
            return item_rarity.Rare
        else if(InAmount < 400):
            return item_rarity.Epic
        else if(InAmount < 500):
            return item_rarity.Legendary
        else:
            return item_rarity.Mythic
    return item_rarity.None

(InAgent:agent).IncreaseTrialPoints<public>(InAmount:int):void=
    PreviousRarity:=InAgent.GetTrialRewardRarity()
    Amount := if. InAmount > 500 then. 500 else:
        case(InAgent.GetTrialRewardRarity()):
            item_rarity.None => InAmount*3
            item_rarity.Common => InAmount*2
            _ => InAmount
    if(CurrentPoints := GlobalTrialManager[GetSession()].TrialPointsPerPlayer[InAgent]):
        FinalAmount := if. (Amount+CurrentPoints) > 500 then. 500 else. (Amount+CurrentPoints)
        if. set GlobalTrialManager[GetSession()].TrialPointsPerPlayer[InAgent] = FinalAmount
    else if:
        ThisTrialManager:=GlobalTrialManager[GetSession()]
        set ThisTrialManager.TrialPointsPerPlayer = ConcatenateMaps(ThisTrialManager.TrialPointsPerPlayer, map{InAgent => Amount})
    NewRarity:=InAgent.GetTrialRewardRarity()
    if(ThisTrialManager:=GlobalTrialManager[GetSession()]):
        if(PreviousRarity<>NewRarity, NewRarity<>item_rarity.None):
            ThisTrialManager.OnNewRarityReached(InAgent, NewRarity)
    InAgent.GetUIManager().TrialUI.UpdateReward(InAgent)

(InAgent:agent).SetTrialPoints<public>(InAmount:int):void=
    PreviousRarity:=InAgent.GetTrialRewardRarity()
    Amount := if. InAmount > 500 then. 500 else. InAmount
    if:
        set GlobalTrialManager[GetSession()].TrialPointsPerPlayer[InAgent] = Amount
    else if:
        ThisTrialManager:=GlobalTrialManager[GetSession()]
        set ThisTrialManager.TrialPointsPerPlayer = ConcatenateMaps(ThisTrialManager.TrialPointsPerPlayer, map{InAgent => Amount})
    InAgent.GetUIManager().TrialUI.UpdateReward(InAgent)
    NewRarity:=InAgent.GetTrialRewardRarity()
    if(ThisTrialManager:=GlobalTrialManager[GetSession()]):
        if(PreviousRarity<>NewRarity, NewRarity<>item_rarity.None):
            ThisTrialManager.OnNewRarityReached(InAgent, NewRarity)

(InAgent:agent).GetTrialPoints<public>()<transacts>:int=
    return GlobalTrialManager[GetSession()].TrialPointsPerPlayer[InAgent] or block{
        if(ThisTrialManager:=GlobalTrialManager[GetSession()]):
            set ThisTrialManager.TrialPointsPerPlayer = ConcatenateMaps(ThisTrialManager.TrialPointsPerPlayer, map{InAgent => 0})
        0
    }
    

(InRarity:item_rarity).ConvertToNumber()<transacts>:int=
    case(InRarity):
        item_rarity.None => 0
        item_rarity.Common => 1
        item_rarity.Rare => 2
        item_rarity.Epic => 3
        item_rarity.Legendary => 4
        item_rarity.Mythic => 5

MAXIMUM_TRIALPOINTS : int = 500

var GlobalTrialManager<public> : weak_map(session, TrialManager) = map{}

TrialManager<public> := class(Powerable):

    @editable RewardSpawningLocation : Interactable = Interactable{}
    @editable RewardBeacon : beacon_device = beacon_device{}
    @editable ActivateTrialLocation : Interactable = Interactable{}
    @editable TrialScreenCinematic : cinematic_sequence_device = cinematic_sequence_device{}
    @editable OnRewardRarityIncreased_PlayAudio : audio_player_device = audio_player_device{}
    @editable OnClaimingReward_PlayCinematic : cinematic_sequence_device = cinematic_sequence_device{}
    @editable OnTrialBegin_ActivateTrigger : trigger_device = trigger_device{}
    @editable OnTrialEnd_ActivateTrigger : trigger_device = trigger_device{}
    
    @editable OnRarityReached_Common_ActivateTrigger : trigger_device = trigger_device{}
    @editable OnRarityReached_Rare_ActivateTrigger : trigger_device = trigger_device{}
    @editable OnRarityReached_Epic_ActivateTrigger : trigger_device = trigger_device{}
    @editable OnRarityReached_Legendary_ActivateTrigger : trigger_device = trigger_device{}
    @editable OnRarityReached_Mythic_ActivateTrigger : trigger_device = trigger_device{}

    @editable KillTrial : trial_Kills = trial_Kills{}
    @editable KillCloseTrial : trial_KillClose = trial_KillClose{}
    @editable KillFarTrial : trial_KillFar = trial_KillFar{}
    @editable KillWhileCrouchedTrial : trial_KillCrouched = trial_KillCrouched{}
    @editable KillWithTrap : trial_KillWithTrap = trial_KillWithTrap{}
    @editable KillInPlace : trial_KillInPlace = trial_KillInPlace{}
    @editable HeadshotTrial : trial_Headshots = trial_Headshots{}
    @editable KillInRoomTrial : trial_KillInRoom = trial_KillInRoom{}
    @editable RoomTrial : trial_StayInRoom = trial_StayInRoom{}
    @editable RebuildBarriersTrial : trial_RebuildBarriers = trial_RebuildBarriers{}
    @editable AvoidDamageTrial : trial_AvoidDamage = trial_AvoidDamage{}
    @editable KillHipfireTrial : trail_KillHipfire = trail_KillHipfire{}
    @editable KillADSTrial : trial_KillADS = trial_KillADS{}
    @editable ProtectObjectiveTrial : trial_ProtectObjective = trial_ProtectObjective{}

    var TrialPointsPerPlayer : [agent]int = map{}

    # Fetches all trials, enabled or not #
    GetTrials()<transacts>:[]trial_base=
        TrialList : []trial_base = array{
            KillTrial
            KillCloseTrial
            KillFarTrial
            KillWhileCrouchedTrial
            KillWithTrap
            KillInPlace
            HeadshotTrial
            KillInRoomTrial
            RoomTrial
            RebuildBarriersTrial
            AvoidDamageTrial
            KillHipfireTrial
            KillADSTrial
            ProtectObjectiveTrial
        }

    # Fetches a random, enabled trial from the trial list #
    GetRandomTrial()<transacts><decides>:trial_base=
        var EnabledTrials : []trial_base = array{}
        TrialList := GetTrials()
        for(ThisTrial:TrialList):
            if(ThisTrial.Enabled?, ThisTrial.CanBeActivated[]):
                set EnabledTrials += array{ThisTrial}
        return Shuffle(EnabledTrials)[0]

    ActivateRandomTrial():void=
        if(not ActiveTrial?):
            if(ThisRandomTrial:=GetRandomTrial[]):
                set ActiveTrial = option{ThisRandomTrial}
                spawn{ThisRandomTrial.AwaitTrialCompletion()}
            else{Print("~ERR: Failed to fetch a random trial!")}

    TrialStartedEvent:event()=event(){}
    TrialCompletedEvent:event()=event(){}
    (InTrial:trial_base).AwaitTrialCompletion<private>()<suspends>:void=
        Print("~Trial Log: Starting Trial [{InTrial.Description}]")
        var ThisPlayerList : []agent = GetAgents()
        TrialStartedEvent.Signal()
        OnTrialBegin_ActivateTrigger.Trigger()
        branch:
            Sleep(0.1)
            ActivePlayers:=GetAgents()
            for(ThisPlayer:ActivePlayers):
                ThisUI:=ThisPlayer.GetUIManager()
                ThisUI.TrialUI.SetDescription(InTrial.GetDescription())
                ThisUI.TrialUI.Show(ThisPlayer)
                ThisUI.TrialUI.TrialTimerBar.Show()
                ThisUI.TrialUI.TrialTimerBar.SetProgress(100.0)
                spawn{ThisUI.TrialUI.TrialTimerBar.ChangeProgressOverTime(InTrial.Time, 0.0, ?InAgent:=option{ThisPlayer})}
        race:
            InTrial.Start()
            loop:
                for(ThisColor:GreyscaleArray):
                    PlayerList:=GetAgents()
                    for(ThisPlayer:PlayerList):
                        ThisPlayer.GetUIManager().TrialUI.DescriptionBackgroundArrows.SetTint(ThisColor)
                    Sleep(0.1)
                Sleep(0.0)
            block:
                for(ThisPlayer:ThisPlayerList):
                    if(ThisWidget:=color_block[ThisPlayer.GetUIManager().TrialUI.TrialTimerBar.GetBarWidget[]]):
                        ThisWidget.SetColor(MakeColorFromHex("39E75F"))
                SafeSleep(InTrial.Time*0.5)
                set ThisPlayerList = GetAgents()
                for(ThisPlayer:ThisPlayerList):
                    if(ThisWidget:=color_block[ThisPlayer.GetUIManager().TrialUI.TrialTimerBar.GetBarWidget[]]):
                        ThisWidget.SetColor(MakeColorFromHex("ffe146"))
                SafeSleep(InTrial.Time*0.25)
                set ThisPlayerList = GetAgents()
                for(ThisPlayer:ThisPlayerList):
                    if(ThisWidget:=color_block[ThisPlayer.GetUIManager().TrialUI.TrialTimerBar.GetBarWidget[]]):
                        ThisWidget.SetColor(MakeColorFromHex("ce2525"))
                Sleep(Inf)
            AwaitGameState(game_state.GameOver)
        set ThisPlayerList = GetAgents()
        for(ThisPlayer:ThisPlayerList):
            ThisUI:=ThisPlayer.GetUIManager()
            ThisUI.TrialUI.HideTrialDescription(ThisPlayer)
            ThisUI.TrialUI.DescriptionBackgroundArrows.SetTint(NamedColors.White)
        OnTrialEnd_ActivateTrigger.Trigger()
        TrialCompletedEvent.Signal()
        Sleep(2.5)
        set ActiveTrial = false

    var<private> ActiveTrial : ?trial_base = false

    OnPowerableBegin<override>():void=
        if(set GlobalTrialManager[GetSession()] = Self){}
        else{Print("~ERR: Failed to globally link trial manager!")}
        for(ThisTrial:GetTrials()):
            ThisTrial.OnTrigger_ActivateTrial.TriggeredEvent.Subscribe(ThisTrial.StartViaTrigger)
        LinkAndCompile(array{RewardSpawningLocation, ActivateTrialLocation})
        spawn{AwaitPowerEnabled()}

    AwaitPowerEnabled<private>()<suspends>:void=
        TrialScreenCinematic.Play()
        TrialScreenCinematic.SetPlayRate(0.0)
        TrialScreenCinematic.SetPlaybackTime(0.0)
        loop:
            if(ThisEnabledEvent:=GetPowerEnabledEvent[], ThisPlayerManager:=GlobalPlayerManager[GetSession()]):
                race:
                    block:
                        ThisEnabledEvent.Await()
                        loop:
                            TrialStartedEvent.Await()
                            TrialScreenCinematic.SetPlaybackTime(2.1)
                            TrialCompletedEvent.Await()
                            TrialScreenCinematic.SetPlaybackTime(0.0)
                    AwaitGameState(game_state.GameOver)
            TrialScreenCinematic.SetPlaybackTime(0.0)
            Sleep(0.0)

    Reset<override>():void=
        if(ThisActiveTrial:=ActiveTrial?):
            ThisActiveTrial.Stop()
        set ActiveTrial = false
        TrialScreenCinematic.SetPlayRate(0.0)
        TrialScreenCinematic.SetPlaybackTime(0.0)
        (super:)Reset()

    SetTrialScreen(InAgent:?agent, InTime:float):void=
        Players:=GetPlayers()
        for(ThisPlayer:Players):
            TrialScreenCinematic.Play(ThisPlayer.User)
            TrialScreenCinematic.SetPlayRate(0.0)
            TrialScreenCinematic.SetPlaybackTime(InTime)

    GetCost():int=
        return block{
            ThisWave := GetWaveCount()
            ThisWave >= 10
            ThisCost := (ThisWave-9) * 100 + 500
            if. ThisCost < 2500 then. ThisCost else. 2500
        } or 500

    GetPowerInteractionText<override>(InAgent:agent, Interactive:Interactable):string=
        if(Interactive.IsEqual[ActivateTrialLocation]):
            # Interactable is trial activation
            if(not ActiveTrial?):
                SetInteractableState(true, ?InAgent:=option{InAgent})
                return "Start Random Task [<C_Y>${GetCost()}<E>]"
            else:
                SetInteractableState(false)
                return "Task Currently Active"
        else if(ThisPlayerInfo:=InAgent.GetPlayerInfo[], not IsRewarding?, not InAgent.IsUsingItemCollection[]):
            # Interactable is reward claim
            ThisRewardRarity := ThisPlayerInfo.User.GetTrialPoints().ToTrialRarity()
            if(ThisRewardRarity = item_rarity.None):
                SetInteractableState(false, ?InAgent:=option{InAgent})
                return "No rewards to claim"
            else:
                SetInteractableState(true, ?InAgent:=option{InAgent})
                case(ThisRewardRarity):
                    item_rarity.Common => return "Claim [<CH_39E75F>Common<E>] trial reward"
                    item_rarity.Rare => return "Claim [<CH_89CFF0>Rare<E>] trial reward"
                    item_rarity.Epic => return "Claim [<CH_9D4DBB>Epic<E>] trial reward"
                    item_rarity.Legendary => return "Claim [<CH_F3AF19>Legendary<E>] trial reward"
                    item_rarity.Mythic => return "Claim [<CA_gold wave>Mythic<E>] trial reward"
                    _ => return "Claim [Unknown] rarity reward"
        else:
            SetInteractableState(false, ?InAgent:=option{InAgent})
            return ""

    GetDescription<override>(InAgent:agent, InInteractable:Interactable):tuple(string, string)=
        if(RewardSpawningLocation.IsEqual[InInteractable]):
            return ("", "")
        return ("Complete challenges for rewards!", "")

    var<private> IsRewarding : logic = false
    OnInteracted<override>(InAgent:agent, InteractableLocation:Interactable):void=
        if(ActivateTrialLocation.IsEqual[InteractableLocation]):
            if(ThisPlayerInfo:=InAgent.GetPlayerInfo[]):
                BuyState := ThisPlayerInfo.BuyItem(GetCost(), ?PlayPurchaseSound:=false)
                if(BuyState?):
                    # Activate random trial #
                    ActivateRandomTrial()
        else:
            if. not RewardsPerPlayer[InAgent] then. set RewardsPerPlayer = ConcatenateMaps(RewardsPerPlayer, map{InAgent=>trial_rewards{}})
            if(ThisRewardPackage:=RewardsPerPlayer[InAgent], not ThisRewardPackage.ClaimedState?, not IsRewarding?):
                set IsRewarding = true
                spawn{SpawnRewards(InAgent, ThisRewardPackage)}
            else:
                Print("~Reward already claimed/spawned!")

    SpawnRewards(InAgent:agent, RewardsPackage:trial_rewards)<suspends>:void=
        RewardBeacon.RemoveFromShowList(InAgent)
        sync:
            block:
                OnClaimingReward_PlayCinematic.Play()
                Sleep(0.0)
            OnClaimingReward_PlayCinematic.StoppedEvent.Await()
        RewardSpawningLocation.SetEnabledState(false)
        SpawningLocation:=RewardSpawningLocation.GetTransform()
        # Grant reward to player #
        if(ThisPlayerInfo:=InAgent.GetPlayerInfo[], ThisPlayerInfo.User.GetTrialRewardRarity() <> item_rarity.None):
            PowerupsSpawning := RewardsPackage.PowerupList.Length+RewardsPackage.PerkList
            var PowerupsSpawned : []powerup_instance = array{}
            var MaybeWeapon : ?weapon_definition = false
            set RewardsPackage.ClaimedState = true
            #ThisPlayerInfo.StopUsingItemPlacerEvent.Signal()
            if(RewardsPackage.WeaponList.Length > 0):
                #set ThisPlayerInfo.IsUsingItemCollection = true
            if(ThisWeapon:=RewardsPackage.WeaponList[0]):
                set MaybeWeapon = option{ThisWeapon}
            for(ThisPowerup:RewardsPackage.PowerupList):
                set PowerupsSpawned += array{ThisPowerup.SpawnPowerupHandler(PowerupsSpawned.Length, InAgent)}
            if(ThisPerkPowerup:=powerup_type.Perk.ConvertToPowerup[]):
                for(ThisPerkIndex:=0..RewardsPackage.PerkList-1):
                    set PowerupsSpawned += array{ThisPerkPowerup.SpawnPowerupHandler(PowerupsSpawned.Length, InAgent)}
            ThisPlayerInfo.User.SetTrialPoints(0)
            AwaitRewardClaim(MaybeWeapon, PowerupsSpawned, InAgent)
        else:
            set IsRewarding = false

    (InPowerup:powerup_base).SpawnPowerupHandler<private>(PowerupsSpawned:int, InAgent:agent):powerup_instance=
        SpawningLocation:=RewardSpawningLocation.GetTransform()
        var SpawnPos : vector3 = vector3{}
        case(PowerupsSpawned):
            0 => set SpawnPos = SpawningLocation.GetLocalVector(vector3{X:=15.0, Z:=42.0})
            1 => set SpawnPos = SpawningLocation.GetLocalVector(vector3{X:=15.0, Z:=45.0, Y:=40.0})
            2 => set SpawnPos = SpawningLocation.GetLocalVector(vector3{X:=15.0, Z:=45.0, Y:=-40.0})
            _ =>
        InPowerup.Spawn(SpawnPos, ?Lifetime:=60.0, ?OwningPlayer:=option{InAgent})
                    
    AwaitRewardClaim(MaybeWeapon:?weapon_definition, PowerupList:[]powerup_instance, InAgent:agent)<suspends>:void=
        SpawningLocation:=RewardSpawningLocation.GetTransform()
        sync:
            if(ThisWeapon:=MaybeWeapon?):
                race:
                    block:
                        Sleep(0.25)
                        WeaponPickup:=ThisWeapon.SpawnWeapon(SpawningLocation, ?Lifetime:=30.0, ?OwningPlayer:=option{InAgent})
                        WeaponPickup.InteractedEvent.Await()
                    Sleep(30.0)
                WeaponInteractList:=ThisWeapon.GetSpawnedInstances()
                for(ThisWeaponInt:WeaponInteractList):
                    ThisWeaponInt.Unlink()
            loop:
                var PowerupsLeft : int = 0
                for(ThisPowerup:PowerupList):
                    if(ThisPowerup.IsOnGround[]):
                        set PowerupsLeft += 1
                if(PowerupsLeft <= 0):
                    break
                Sleep(0.25)
        set IsRewarding = false
        RewardSpawningLocation.SetEnabledState(true)
        OnClaimingReward_PlayCinematic.PlayReverse()

    <# TIERS AND THEIR REWARDS #
    
    - Common -
    1 Reward
    Weapons : 75%
    Powerups : 24%
    Perks : 1%

    - Rare -
    1 Reward
    Weapons : 61%
    Powerups : 32%
    Perks : 4%

    - Epic -
    2 Rewards
    Weapons : 50%
    Powerups : 42%
    Perks : 8%

    - Legendary -
    2 Rewards
    Weapons : 45%
    Powerups : 40%
    Perks : 15%

    - Mythic -
    3 Rewards
    Weapons : 100%
    Powerups : 100%
    Perks : 100%
    
    #                          #>

    # Fires when a new rarity is reached #
    OnNewRarityReached(InAgent:agent, InRarity:item_rarity):void=
        OnRewardRarityIncreased_PlayAudio.Play(InAgent)
        case(InRarity):
            item_rarity.Common => OnRarityReached_Common_ActivateTrigger.Trigger(InAgent)
            item_rarity.Rare => OnRarityReached_Rare_ActivateTrigger.Trigger(InAgent)
            item_rarity.Epic => OnRarityReached_Epic_ActivateTrigger.Trigger(InAgent)
            item_rarity.Legendary => OnRarityReached_Legendary_ActivateTrigger.Trigger(InAgent)
            item_rarity.Mythic =>  OnRarityReached_Mythic_ActivateTrigger.Trigger(InAgent)
            _ =>
        UpdateRewards(InAgent)

    var<private> RewardsPerPlayer : [agent]trial_rewards = map{}
    UpdateRewards(InAgent:agent):void=
        RewardBeacon.Enable()
        PlayerList:=GetPlayers()
        for(ThisPlayerInfo:PlayerList):
            if(ThisPlayerInfo.User.GetTrialRewardRarity()<>item_rarity.None):
                RewardBeacon.AddToShowList(ThisPlayerInfo.User)
            else:
                RewardBeacon.RemoveFromShowList(ThisPlayerInfo.User)
        if(ThisPlayerInfo:=InAgent.GetPlayerInfo[]):
            var ThisUpdatedRewards : trial_rewards = trial_rewards{}
            var RewardAmount : int = 1
            var RewardRarities : tuple(float, float, float) = (0.0, 0.0, 0.0)
            case(ThisPlayerInfo.User.GetTrialRewardRarity()):
                item_rarity.Common => 
                    set RewardAmount = 1
                    set RewardRarities = (75.0, 25.0, 0.0)
                    #set RewardRarities = (100.0, 0.0, 0.0)
                item_rarity.Rare => 
                    set RewardAmount = 1
                    set RewardRarities = (61.0, 40.0, 0.0)
                    #set RewardRarities = (100.0, 0.0, 0.0)
                item_rarity.Epic => 
                    set RewardAmount = 2
                    set RewardRarities = (50.0, 42.0, 8.0)
                    #set RewardRarities = (100.0, 0.0, 0.0)
                item_rarity.Legendary => 
                    set RewardAmount = 2
                    set RewardRarities = (45.0, 40.0, 15.0)
                    #set RewardRarities = (100.0, 0.0, 0.0)
                item_rarity.Mythic => 
                    set RewardAmount = 3
                    set RewardRarities = (100.0, 100.0, 100.0)
                    #set RewardRarities = (100.0, 0.0, 0.0)
                _ => 
                    if(set RewardsPerPlayer[InAgent] = ThisUpdatedRewards){}
                    return
            set ThisUpdatedRewards = GetRewardsFromRarity(ThisPlayerInfo.User.GetTrialRewardRarity(), RewardAmount, RewardRarities)
            if(set RewardsPerPlayer[InAgent] = ThisUpdatedRewards){}
            else:
                set RewardsPerPlayer = ConcatenateMaps(RewardsPerPlayer, map{InAgent => ThisUpdatedRewards})

    GetRewardsFromRarity(RarityTier:item_rarity, RarityRewardAmount:int, Weights : tuple(float, float, float)):trial_rewards=
        var WeaponChosen : logic = false
        var ThisRewardPackage : trial_rewards = trial_rewards{}
        for(ThisReward:=0..RarityRewardAmount-1):
            ChosenRewardTab:=RandomWeightedChance(array{Weights(0), Weights(1), Weights(2)})
            case(ChosenRewardTab):
                0 => # Give weapon #
                    if(not WeaponChosen?):
                        if(ThisRandomWeapon:=Shuffle(GlobalWeaponManager[GetSession()].GetWeaponsOfRarity(RarityTier))[0]):
                            set ThisRewardPackage.WeaponList += array{ThisRandomWeapon}
                            Print("~Reward set: Weapon")
                            set WeaponChosen = true
                        else if(ThisRandomWeapon:=Shuffle(GlobalWeaponManager[GetSession()].GetWeaponsOfRarity(item_rarity.Epic))[0]):
                            set ThisRewardPackage.WeaponList += array{ThisRandomWeapon}
                            Print("~Reward set: Weapon")
                            set WeaponChosen = true
                    else if(ThisRandomPowerup:=Shuffle(GlobalPowerupManager[GetSession()].GetPowerupsOfRarity(RarityTier))[0]):
                            set ThisRewardPackage.PowerupList += array{ThisRandomPowerup}
                1 => # Give powerup #
                    if(ThisPowerupManager:=GlobalPowerupManager[GetSession()]):
                        var PowerupList:[]powerup_base=ThisPowerupManager.GetPowerupsOfRarity(RarityTier)
                        if(RarityTier = item_rarity.Mythic):
                            set PowerupList += ThisPowerupManager.GetPowerupsOfRarity(item_rarity.Legendary)
                        if(ThisRandomPowerup:=Shuffle(PowerupList)[0]):
                            set ThisRewardPackage.PowerupList += array{ThisRandomPowerup}
                            Print("~Reward set: Powerup")
                        else{Print("~ERR: Failed to choose a random Powerup!")}
                    else{Print("~ERR: Failed to fetch powerup manager")}
                2 => # Give perk (perk powerup) #
                    set ThisRewardPackage.PerkList += 1
                    Print("~Reward set: Perk")
                _ => # RNG failsafe - Gives a powerup #
                    Print("~Failed rng chance")
                    if(ThisRandomPowerup:=Shuffle(GlobalPowerupManager[GetSession()].GetPowerupsOfRarity(RarityTier))[0]):
                        set ThisRewardPackage.PowerupList += array{ThisRandomPowerup}
        return ThisRewardPackage
            
trial_rewards := class:

    var ClaimedState : logic = false

    var WeaponList : []weapon_definition = array{}
    var PowerupList : []powerup_base = array{}
    var PerkList : int = 0

trial_base := class<concrete>:

    @editable Description : string = "Unset Trial Description"
    @editable Time : float = 60.0
    @editable Enabled : logic = true
    @editable OnTrigger_ActivateTrial : trigger_device = trigger_device{}

    GetDescription<public>()<transacts>:string=
        return Description

    StartViaTrigger(Agent:?agent):void=
        spawn{Start()}

    var<private> ActiveState : logic = false
    StartedEvent:event()=event(){}
    Start<public><final>()<suspends>:void=
        if(not ActiveState?):
            set ActiveState = true
            StartedEvent.Signal()
            branch:
                OnStart()
            Winner:=race:
                block:
                    StartTimer()
                    true
                block:
                    StopEvent.Await()
                    false
            if(Winner?):
                Stop()

    OnStart<public>()<suspends>:void=
        return

    StartTimer()<suspends>:void=
        if(Time > 0.0):
            SafeSleep(Time)
        else:
            Sleep(Inf)

    StopEvent:event()=event(){}
    Stop<public><final>():void=
        set ActiveState = false
        StopEvent.Signal()
        spawn{OnStop()}
        
    OnStop<public>()<suspends>:void=
        return

    CanBeActivated<public>()<transacts><decides>:void=
        return