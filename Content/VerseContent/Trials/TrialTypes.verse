using { /Fortnite.com/Devices }
using { /Fortnite.com/Characters }
using { /Fortnite.com/Game }
using { /Fortnite.com/UI }
using { /Fortnite.com/FortPlayerUtilities }
using { /Verse.org/Simulation }
using { /Verse.org/Random }
using { /Verse.org/Colors }
using { /UnrealEngine.com/Temporary/Diagnostics }
using { /UnrealEngine.com/Temporary/SpatialMath }
using { /UnrealEngine.com/Temporary/UI }
using { /Fortnite.com/AI }
using { Player }
using { Zombie }
using { Weapons }
using { Room }

trial_Kills := class(trial_base):

    @editable Description<override> : string = "Eliminate Zombies"
    @editable PointsPerKill : int = 4

    OnStart<override>():void=
        PlayerList:=GetAgents()
        for(ThisPlayer:PlayerList):
            spawn{ThisPlayer.TrackZombieEliminations()}

    (InAgent:agent).TrackZombieEliminations<public>()<suspends>:void=
        if(ThisPlayerInfo:=InAgent.GetPlayerInfo[]):
            race:
                loop:
                    ThisPlayerInfo.EliminatedZombieEvent.Await()
                    InAgent.IncreaseTrialPoints(PointsPerKill)
                StopEvent.Await()

trial_Headshots := class(trial_base):

    @editable Description<override> : string = "Headshot Zombies"
    @editable PointsPerHeadshot : int = 5

    OnStart<override>():void=
        PlayerList:=GetAgents()
        for(ThisPlayer:PlayerList):
            spawn{ThisPlayer.TrackHeadshots()}

    (InAgent:agent).TrackHeadshots<public>()<suspends>:void=
        if(ThisPlayerInfo:=InAgent.GetPlayerInfo[]):
            race:
                loop:
                    DmgInfo:=ThisPlayerInfo.DamagedZombieEvent.Await()
                    if(DmgInfo.DamageType = damage_type.Head):
                        InAgent.IncreaseTrialPoints(PointsPerHeadshot)
                StopEvent.Await()

trial_StayInRoom := class(trial_base):

    @editable Description<override> : string = "Stay in a room"
    @editable PointRewardPerSecond : int = 3

    var CurrRoom : ?room = false

    GetDescription<override>()<transacts>:string=
        if(ThisRoom := CurrRoom?):
            return "stay in room \\n({ThisRoom.RoomText})"
        else:
            return Description

    OnStart<override>()<suspends>:void=
        if(ThisRoom:=Shuffle(GetTrialRooms())[0]):
            set CurrRoom = option{ThisRoom}
            for(ThisBorder:ThisRoom.Borders):
                ThisBorder.Show()
            ThisRoom.TrialBeacon_Away.Enable()
            ThisRoom.TrialBeacon_Near.Enable()
            ThisRoom.TrialBeacon_Near.RemoveAllFromShowList()
            loop:
                OccupiedPlayers := ThisRoom.GetPlayersInRoom()
                if(OccupiedPlayers.Length > 0):
                    #Print("~Players in room : {OccupiedPlayers.Length}")
                    for(ThisPlayer:OccupiedPlayers):
                        ThisRoom.TrialBeacon_Away.RemoveFromShowList(ThisPlayer)
                        ThisRoom.TrialBeacon_Near.AddToShowList(ThisPlayer)
                        ThisPlayer.PlayerInRoom()
                NotInRoomPlayers:=ThisRoom.GetOutsidePlayers()
                #if(NotInRoomPlayers.Length > 0):
                    #Print("~Players out of room : {NotInRoomPlayers.Length}")
                for(ThisPlayer:NotInRoomPlayers):
                    ThisRoom.TrialBeacon_Away.AddToShowList(ThisPlayer)
                    ThisRoom.TrialBeacon_Near.RemoveFromShowList(ThisPlayer)
                SafeSleep(1.0) 
        else{Print("~ERR: Failed to fetch a random room for the trial!")}

    OnStop<override>():void=
        if(ThisRoom := CurrRoom?):
            set CurrRoom = false
            for(ThisBorder:ThisRoom.Borders):
                ThisBorder.Hide()
            ThisRoom.TrialBeacon_Away.Disable()
            ThisRoom.TrialBeacon_Near.Disable()

    (InAgent:agent).PlayerInRoom<public>():void=
        InAgent.IncreaseTrialPoints(PointRewardPerSecond)

trial_KillInRoom := class(trial_base):

    @editable Description<override> : string = "Kill zombies in a room"
    @editable PointsPerKill : int = 5

    var CurrRoom : ?room = false

    GetDescription<override>()<transacts>:string=
        if(ThisRoom := CurrRoom?):
            return "kill zombies in room \\n({ThisRoom.RoomText})"
        else:
            return Description

    OnStart<override>()<suspends>:void=
        if(ThisRoom:=Shuffle(GetTrialRooms())[0]):
            set CurrRoom = option{ThisRoom}
            for(ThisBorder:ThisRoom.Borders):
                ThisBorder.Show()
            ThisRoom.TrialBeacon_Away.Enable()
            ThisRoom.TrialBeacon_Near.Enable()
            ThisRoom.TrialBeacon_Near.RemoveAllFromShowList()
            PlayerList:=GetAgents()
            for(ThisPlayer:PlayerList):
                spawn{ThisPlayer.TrackZombieEliminations()}
            race:
                StopEvent.Await()
                loop:
                    OccupiedPlayers := ThisRoom.GetPlayersInRoom()
                    if(OccupiedPlayers.Length > 0):
                        #Print("~Players in room : {OccupiedPlayers.Length}")
                        for(ThisPlayer:OccupiedPlayers):
                            ThisRoom.TrialBeacon_Away.RemoveFromShowList(ThisPlayer)
                            ThisRoom.TrialBeacon_Near.AddToShowList(ThisPlayer)
                    NotInRoomPlayers:=ThisRoom.GetOutsidePlayers()
                    #if(NotInRoomPlayers.Length > 0):
                        #Print("~Players out of room : {NotInRoomPlayers.Length}")
                    for(ThisPlayer:NotInRoomPlayers):
                        ThisRoom.TrialBeacon_Away.AddToShowList(ThisPlayer)
                        ThisRoom.TrialBeacon_Near.RemoveFromShowList(ThisPlayer)
                    SafeSleep(1.0) 
        else{Print("~ERR: Failed to fetch a random room for the trial!")}

    OnStop<override>():void=
        if(ThisRoom:=CurrRoom?):
            set CurrRoom = false
            for(ThisBorder:ThisRoom.Borders):
                ThisBorder.Hide()
            ThisRoom.TrialBeacon_Away.Disable()
            ThisRoom.TrialBeacon_Near.Disable()

    (InAgent:agent).TrackZombieEliminations<public>()<suspends>:void=
        if(ThisPlayerInfo:=InAgent.GetPlayerInfo[], ThisRoom := CurrRoom?):
            race:
                loop:
                    ThisPlayerInfo.EliminatedZombieEvent.Await()
                    if(ThisRoom.IsInRoom[InAgent]):
                        InAgent.IncreaseTrialPoints(PointsPerKill)
                StopEvent.Await()
                ThisPlayerInfo.LeaveEvent.Await()
            
trial_RebuildBarriers := class(trial_base):

    @editable Description<override> : string = "Rebuild barriers"
    @editable PointsPerBarrierBuilt : int = 2

    OnStart<override>():void=
        PlayerList:=GetAgents()
        for(ThisPlayer:PlayerList):
            spawn{ThisPlayer.TrackBarrierBuild()}

    (InAgent:agent).TrackBarrierBuild<public>()<suspends>:void=
        if(ThisPlayerInfo:=InAgent.GetPlayerInfo[]):
            race:
                loop:
                    ThisPlayerInfo.RebuildBarrierEvent.Await()
                    InAgent.IncreaseTrialPoints(PointsPerBarrierBuilt)
                StopEvent.Await()
                ThisPlayerInfo.LeaveEvent.Await()

trial_AvoidDamage := class(trial_base):

    @editable Description<override> : string = "Avoid taking damage"
    @editable MaximumPoints : int = 100
    @editable PointGrantingTickRate : float = 0.6
    @editable OnDamagePointDelay : float = 5.0

    OnStart<override>():void=
        PlayerList:=GetAgents()
        for(ThisPlayer:PlayerList):
            spawn{ThisPlayer.TrackPlayerHurt()}

    (InAgent:agent).TrackPlayerHurt<public>()<suspends>:void=
        if(ThisPlayerInfo:=InAgent.GetPlayerInfo[], PointsPerSecond:=Int[(MaximumPoints*1.0)/Time]):
            branch:
                loop:
                    race:
                        ThisPlayerInfo.DamagedEvent.Await()
                        block:
                            SafeSleep(OnDamagePointDelay)
                            loop:
                                InAgent.IncreaseTrialPoints(PointsPerSecond)
                                SafeSleep(PointGrantingTickRate)
            race:
                StopEvent.Await()
                ThisPlayerInfo.LeaveEvent.Await()

trial_KillClose := class(trial_Kills):

    @editable Description<override> : string = "Eliminate Zombies in \\nclose proximity"
    @editable PointsPerKill<override> : int = 8
    @editable MaximumCloseDistance : float = 225.0

    (InAgent:agent).TrackZombieEliminations<override>()<suspends>:void=
        if(ThisPlayerInfo:=InAgent.GetPlayerInfo[], FortChar:=InAgent.GetFortCharacter[]):
            race:
                loop:
                    ElimResult:=ThisPlayerInfo.EliminatedZombieEvent.Await()
                    if(Distance(ElimResult.GetZombieAI[].DeathTransform.Translation, FortChar.GetTransform().Translation) < MaximumCloseDistance):
                        InAgent.IncreaseTrialPoints(PointsPerKill)
                StopEvent.Await()
                ThisPlayerInfo.LeaveEvent.Await()

trial_KillFar := class(trial_Kills):

    @editable Description<override> : string = "Eliminate Zombies \\nfrom afar"
    @editable PointsPerKill<override> : int = 6
    @editable MinimumFarDistance : float = 500.0

    (InAgent:agent).TrackZombieEliminations<override>()<suspends>:void=
        if(ThisPlayerInfo:=InAgent.GetPlayerInfo[], FortChar:=InAgent.GetFortCharacter[]):
            race:
                loop:
                    ElimResult:=ThisPlayerInfo.EliminatedZombieEvent.Await()
                    if(Distance(ElimResult.GetZombieAI[].DeathTransform.Translation, FortChar.GetTransform().Translation) > MinimumFarDistance):
                        InAgent.IncreaseTrialPoints(PointsPerKill)
                StopEvent.Await()
                ThisPlayerInfo.LeaveEvent.Await()

trial_KillCrouched := class(trial_Kills):

    @editable Description<override> : string = "Eliminate Zombies \\nwhile crouched"
    @editable PointsPerKill<override> : int = 6

    (InAgent:agent).TrackZombieEliminations<override>()<suspends>:void=
        if(ThisPlayerInfo:=InAgent.GetPlayerInfo[], FortChar:=InAgent.GetFortCharacter[]):
            race:
                loop:
                    ThisPlayerInfo.EliminatedZombieEvent.Await()
                    if(FortChar.IsCrouching[]):
                        InAgent.IncreaseTrialPoints(PointsPerKill)
                StopEvent.Await()
                ThisPlayerInfo.LeaveEvent.Await()

trail_KillHipfire := class(trial_Kills):

    @editable Description<override> : string = "Eliminate Zombies \\nwhile hipfiring"
    @editable PointsPerKill<override> : int = 5

    (InAgent:agent).TrackZombieEliminations<override>()<suspends>:void=
        if(ThisPlayerInfo:=InAgent.GetPlayerInfo[], FortChar:=InAgent.GetFortCharacter[], WeaponManager:=GlobalWeaponManager[GetSession()]):
            race:
                loop:
                    ThisPlayerInfo.EliminatedZombieEvent.Await()
                    if(not WeaponManager.AimInput.IsHeld[InAgent]):
                        InAgent.IncreaseTrialPoints(PointsPerKill)
                StopEvent.Await()
                ThisPlayerInfo.LeaveEvent.Await()

trial_KillADS := class(trial_Kills):

    @editable Description<override> : string = "Eliminate Zombies \\nwhile aiming"
    @editable PointsPerKill<override> : int = 5

    (InAgent:agent).TrackZombieEliminations<override>()<suspends>:void=
        if(ThisPlayerInfo:=InAgent.GetPlayerInfo[], FortChar:=InAgent.GetFortCharacter[], WeaponManager:=GlobalWeaponManager[GetSession()]):
            race:
                loop:
                    ThisPlayerInfo.EliminatedZombieEvent.Await()
                    if(WeaponManager.AimInput.IsHeld[InAgent]):
                        InAgent.IncreaseTrialPoints(PointsPerKill)
                StopEvent.Await()
                ThisPlayerInfo.LeaveEvent.Await()

trial_KillWithTrap := class(trial_Kills):

    @editable Description<override> : string = "Eliminate Zombies \\nwith a trap"
    @editable PointsPerKill<override> : int = 3

    (InAgent:agent).TrackZombieEliminations<override>()<suspends>:void=
        if(ThisPlayerInfo:=InAgent.GetPlayerInfo[], FortChar:=InAgent.GetFortCharacter[]):
            race:
                loop:
                    ThisPlayerInfo.EliminatedZombieByTrapEvent.Await()
                    InAgent.IncreaseTrialPoints(PointsPerKill)
                StopEvent.Await()
                ThisPlayerInfo.LeaveEvent.Await()

trial_KillInPlace := class(trial_Kills):

    @editable Time<override> : float = 45.0
    @editable Description<override> : string = "Eliminate Zombies \\nwhile standing still"
    @editable PointsPerKill<override> : int = 12

    (InAgent:agent).TrackZombieEliminations<override>()<suspends>:void=
        if(ThisPlayerInfo:=InAgent.GetPlayerInfo[], FortChar:=InAgent.GetFortCharacter[], WeaponManager:=GlobalWeaponManager[GetSession()]):
            var MovingState : logic = false
            branch:
                loop:
                    Position:=FortChar.GetTransform().Translation
                    Sleep(0.0)
                    NextPosition:=FortChar.GetTransform().Translation
                    if(not IsAlmostEqual[Position, NextPosition, 0.1]):
                        set MovingState = true
                    else:
                        set MovingState = false
            race:
                loop:
                    ThisPlayerInfo.EliminatedZombieEvent.Await()
                    if(not MovingState?):
                        InAgent.IncreaseTrialPoints(PointsPerKill)
                StopEvent.Await()
                ThisPlayerInfo.LeaveEvent.Await()

trial_ProtectObjective := class(trial_base):

    @editable Description<override> : string = "Protect the objective \\nfrom enemies"
    @editable MaximumPoints : int = 100
    @editable Objectives : []objective_defend = array{}

    OnStart<override>()<suspends>:void=
        if(ThisObjective:=Shuffle(Objectives)[0]):
            spawn{ThisObjective.Start()}
            ThisObjective.StopEvent.Await()
            if(PointsToAward:=Int[ThisObjective.GetHealthPercent()*MaximumPoints]):
                PlayerList:=GetAgents()
                for(ThisPlayer:PlayerList):
                    if(ThisPlayerInfo:=ThisPlayer.GetPlayerInfo[]):
                        ThisPlayer.IncreaseTrialPoints(PointsToAward)
        else{Print("~ERR: No objectives to choose from!")}

# UNFINISHED! #
trial_Capture := class(trial_base):

    @editable Description<override> : string = "Stay in the \\ncapture zone"
    @editable PointRewardPerSecond : int = 4
    @editable MaximumTime : float = 45.0
    @editable CaptureZoneProp : creative_prop = creative_prop{}
    @editable CaptureZoneRadius : float = 250.0
    @editable CaptureZoneMovements : []trial_capture_zonemovement = array{}

    OnStart<override>()<suspends>:void=
        if(RandomMovementSequence:=Shuffle(CaptureZoneMovements)[0]):


trial_capture_zonemovement:=class<concrete>:

    @editable RandomMovement : logic = true
    @editable Speed : float = 1.0
    @editable MovementPoints : []creative_prop = array{}
