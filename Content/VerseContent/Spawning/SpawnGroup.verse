using { /Verse.org/Simulation }
using { /Fortnite.com/Devices }
using { /Verse.org/Random }
using { Zombie }
using { Player }
using { Room }

# A controller for spawning zombies in a group
# Allows for easy management of a group of zombies
spawn_group_base<public> := class<concrete>(comparable_class):

    # SETTINGS
    @editable
    ZombieType<public> : zombie_type_list = zombie_type_list.Default

    @editable
    CountAsWaveEnemies<public> : logic = true

    @editable
    OverrideAggression<public> : type{_X:float where _X >= -1.0, _X <= 100.0} = -1.0

    @editable
    OnlySpawnAvailableSlots<public> : logic = false

    @editable
    ForcedSpawnLocations<public> : []spawn_location = array{}

    @editable
    ForcedSpawnRooms<public> : []room = array{}

    @editable 
    MaximumZombiesAtATime<public> : int = -1

    # VARIABLES
    var<private> SpawnedList : []zombie_ai = array{}
    var SpawnRate<public> : float = 1.0
    var SpawnsAtOnce<public> : int = 1
    var PausedState<private> : logic = true
    var GroupSpawnedState<private> : logic = false
    var GroupEliminatedState<private> : logic = false
    var PreviousSpawnAmount<private> : int = 0

    var EliminatedCount<private> : int = 0

    # EVENTS
    SpawnedEvent<public> : event(zombie_ai) = event(zombie_ai){}
    ZombieEliminatedEvent<public> : event(agent) = event(agent){}
    GroupEliminatedEvent<public> : event() = event(){}
    CancelSpawning<private> : event() = event(){}
    ResumeSpawning<private> : event() = event(){}

    # Spawns the zombies in this group
    Spawn<public>(?AmountOverride:?int=false)<suspends>:void=
        return

    HandleSpawning<public>(Amount:int)<suspends>:void=
        if(Amount > 0):
            set PreviousSpawnAmount = Amount
            if(ThisZombieController := ZombieType.GetController[]):
                # Reset variables
                set SpawnedList = GetActiveZombies()
                set PausedState = false
                set GroupSpawnedState = false
                set GroupEliminatedState = false
                if. SpawnsAtOnce < 1 then. set SpawnsAtOnce = 1
                OverrideAggressionTuple := if. OverrideAggression >= 0.0 then. (true, OverrideAggression) else. (false, -1.0)
                if(CountAsWaveEnemies?, ThisWaveInfo:=GetWaveInfo[]):
                    ThisWaveInfo.LinkSpawnGroup(Self)
                NewAmount := if. OnlySpawnAvailableSlots? then. (Amount-GetActiveZombies().Length) else. Amount
                branch:
                    for(CurrentAmount := 0..NewAmount-1):
                        for(i := 0..SpawnsAtOnce-1):
                            if(GetSpawnedAmount() >= NewAmount):
                                # The amount spawned is greater or equal to the amount set
                                CancelSpawning.Signal()
                            else if(MaximumZombiesAtATime > -1, GetActiveZombies().Length > MaximumZombiesAtATime):
                                # The amount of alive zombies exceeds the maximum allowed amount at a time
                                ThisZombieController.ZombieEliminated.Await()
                            else:
                                # Get a random open forced spawn location (if any)
                                var OpenForcedSpawns : []spawn_location = array{}
                                for(ThisSpawner:ForcedSpawnLocations):
                                    if:
                                        ThisSpawner.AllowedTypes.HasType[ZombieType]
                                        ThisSpawner.IsAvailable[]
                                    then:
                                        set OpenForcedSpawns += array. ThisSpawner
                                for(ThisRoom:ForcedSpawnRooms):
                                    for(ThisSpawner:ThisRoom.ZombieSpawners):
                                        if:
                                            ThisSpawner.AllowedTypes.HasType[ZombieType]
                                            ThisSpawner.IsAvailable[]
                                            not OpenForcedSpawns.Find[ThisSpawner]
                                        then: 
                                            set OpenForcedSpawns += array. ThisSpawner
                                OpenForcedLocation := option. Shuffle(OpenForcedSpawns)[0]
                                MaybeNewZombie:=ThisZombieController.SpawnAndAwait(?SpawnGroup:=option{Self}, ?OverrideWaveEnemy:=(true, CountAsWaveEnemies), ?OverrideAggression:=OverrideAggressionTuple, ?ForceSpawnLocation:=OpenForcedLocation)
                                if(ThisNewZombie:=MaybeNewZombie?):
                                    set SpawnedList += array{ThisNewZombie}
                                    SpawnedEvent.Signal(ThisNewZombie)
                                else:
                                    Print("ERROR: Failed to spawn a zombie from a spawn_group! Cancelling...")
                                    CancelSpawning.Signal()
                            Sleep(0.0)
                        SafeSleep(SpawnRate)
                        if(PausedState?):
                            ResumeSpawning.Await()
                    CancelSpawning.Signal()
                CancelSpawning.Await()
        set PausedState = true
        set GroupSpawnedState = true

    # Eliminates all zombies in this group. Does NOT cancel spawns that are queued!
    Eliminate<public>():void=
        for(ThisZombie:SpawnedList):
            ThisZombie.Eliminate()

    # Cancels all spawns and queued spawns
    Cancel<public>():void=
        CancelSpawning.Signal()

    # Stops the spawning of zombies, but does not cancel it
    Pause<public>():void=
        set PausedState = true

    # Resumes the spawning of zombies if paused
    Resume<public>():void=
        set PausedState = false
        ResumeSpawning.Signal()

    # Activates when a zombie is eliminated (from the zombie_ai code)
    # Used to check if all zombies of this group have been eliminated
    OnZombieEliminated<public>(InZombieAgent:agent):void=
        ZombieEliminatedEvent.Signal(InZombieAgent)
        set EliminatedCount += 1
        ActiveZombies:=for:
            ThisZombie:SpawnedList
            ThisZombie.IsActive[]
        do:
            ThisZombie
        if(not ActiveZombies[0], SpawnedList.Length>=PreviousSpawnAmount):
            set GroupEliminatedState = true
            GroupEliminatedEvent.Signal()

    IsGroupEliminated<public>()<transacts><decides>:void=
        GroupEliminatedState?

    AwaitGroupElimination<public>()<suspends>:void=
        if(not GroupEliminatedState?):
            GroupEliminatedEvent.Await()

    AwaitGroupSpawning<public>()<suspends>:void=
        if(not GroupSpawnedState?):
            CancelSpawning.Await()
        
    IsSpawning<public>()<transacts><decides>:void=
        not PausedState?

    IsGroupSpawned<public>()<transacts><decides>:void=
        GroupSpawnedState?

    SetSpawnRate<public>(NewSpawnRate:float):void=
        set SpawnRate = NewSpawnRate

    GetActiveZombies<public>()<transacts>:[]zombie_ai=
        ActiveZombieList:=for:
            ThisZombie:SpawnedList
            ThisZombie.IsActive[]
        do:
            ThisZombie

    GetSpawnedAmount<public>()<transacts>:int=
        ActiveZombieList:=for:
            ThisZombie:SpawnedList
        do:
            0
        return ActiveZombieList.Length

    GetSpawnerAmount<public>()<transacts>:int=
        return PreviousSpawnAmount

    GetEliminatedCount<public>()<transacts>:int=
        return EliminatedCount

# A basic spawn group class that spawns a set amount of zombies
spawn_group<public> := class(spawn_group_base):

    @editable
    CountAsWaveEnemies<override> : logic = false

    @editable
    var<private> Amount<public> : int = 0

    Spawn<override>(?AmountOverride:?int=false)<suspends>:void=
        HandleSpawning(if. OverrideAmount:=AmountOverride? then. OverrideAmount else. Amount)

    SetSpawnAmount<public>(InAmount:int):void=
        set Amount = InAmount

# A device used to dynamically spawn zombies with events
SpawnGroupDevice<public> := class(creative_device):

    @editable
    Controls<public> : spawn_group = spawn_group{}

    @editable
    ActivateOnTrigger : trigger_device = trigger_device{}
    
    @editable
    ActivateTrigger_OnSpawnsCompleted : trigger_device = trigger_device{}

    @editable
    ActivateTrigger_OnGroupEliminated : trigger_device = trigger_device{}

    OnBegin<override>():void=
        ActivateOnTrigger.TriggeredEvent.Subscribe(SpawnGroupFromTrigger)

    SpawnGroupFromTrigger<private>(Agent:?agent):void=
        if(not Controls.IsSpawning[]):
            spawn. Controls.Spawn()
            spawn. AwaitTriggerEvents()

    AwaitTriggerEvents<private>()<suspends>:void=
        sync:
            block:
                Controls.AwaitGroupSpawning()
                ActivateTrigger_OnSpawnsCompleted.Trigger()
            block:
                Controls.AwaitGroupElimination()
                ActivateTrigger_OnGroupEliminated.Trigger()

# A function to easily spawn a set amount of zombies
# Returns the spawn group associated with the function call
(ZombieType:zombie_type_list).SpawnZombieGroup<public>(Amount:int):spawn_group=
    NewSpawnGroup:spawn_group=spawn_group{Amount:=Amount}
    spawn. NewSpawnGroup.Spawn()
    NewSpawnGroup