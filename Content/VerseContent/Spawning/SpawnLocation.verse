using { /Verse.org/Simulation }
using { /Fortnite.com/Devices }
using { /Fortnite.com/Characters }
using { /Fortnite.com/AI/movement_types }
using { /Fortnite.com/AI }
using { /Verse.org/Assets }
using { /Verse.org/Random }
using { /UnrealEngine.com/Temporary/SpatialMath }
using { /UnrealEngine.com/Temporary/Diagnostics }
using { InteractiveModule }
using { Zombie }
using { Powerups }
using { Room }
using { Player }

spawner_debug_draw := class(debug_draw_channel) {}

RRFIX<public> : float = -40.0

RepairAllSpawners<public>():void=
    if(SpawnCollection:=SpawnLocationCollection[GetSession()]):
        for(ThisSpawner:SpawnCollection):
            spawn. ThisSpawner.RepairBarricade(?FullRepair:=true)

# Helper function to update spawn locations, such as their enabled states and to clear queues
UpdateSpawnLocations<public>():void=
    ThisWaveCount:=GetWaveCount()
    if(SpawnerCollection := SpawnLocationCollection[GetSession()]):
        for(ThisSpawnLocation:SpawnerCollection):
            if(ThisSpawnLocation.EnableOnWave = ThisWaveCount):
                ThisSpawnLocation.Enable()
            if(ThisSpawnLocation.DisableOnWave = ThisWaveCount):
                ThisSpawnLocation.Disable()
            ThisSpawnLocation.GetSpawnerZombies() # -> Automatically clears variable

(SpawnLocations:[]spawn_location).GetOpenSpawnLocation<public>()<transacts><decides>:spawn_location=
    OpenSpawners:=for:
        ThisSpawner:SpawnLocations
        ThisSpawner.IsAvailable[]
    do:
        ThisSpawner
    return Shuffle(OpenSpawners)[0]

GetOpenSpawnLocation<public>(InZombieType:zombie_type_list)<transacts><decides>:spawn_location=
    PlayerList := GetAgents()
    var MaybeSpawnLocation : ?spawn_location = false
    # Find the player that is least targeted by zombies (balances out zombie counts per player)
    var LowestTargetedPlayerTuple : tuple(agent, int) = (PlayerList[0], PlayerList[0].GetTargetedAmount())
    for(ThisPlayer:PlayerList):
        ThisTargetedAmount := ThisPlayer.GetTargetedAmount()
        if(ThisTargetedAmount < LowestTargetedPlayerTuple(1), ThisPlayer.GetOccupyingRoom[]):
            set LowestTargetedPlayerTuple = (ThisPlayer, ThisTargetedAmount)

    SpawnCollection := SpawnLocationCollection[GetSession()]
    OpenSpawners := for:
        ThisSpawner:SpawnCollection
        ThisSpawner.AllowedTypes.HasType[InZombieType]
        ThisSpawner.IsAvailable[]
    do:
        ThisSpawner
    if(NewSpawnerList := OpenSpawners.RemoveElement[OpenSpawners.Find[InZombieType.GetController[].LastSpawnLocation?]], NewSpawnerList.Length > 0):
        set MaybeSpawnLocation = option{Shuffle(NewSpawnerList)[0]}
    else:
        set MaybeSpawnLocation = option{Shuffle(OpenSpawners)[0]}
    if(TargetedRoom := LowestTargetedPlayerTuple(0).GetOccupyingRoom[]):
        # Search the spawners in targets room
        OpenRoomSpawners := for:
            ThisSpawner:TargetedRoom.ZombieSpawners
            ThisSpawner.AllowedTypes.HasType[InZombieType]
            ThisSpawner.IsAvailable[]
        do:
            ThisSpawner
        if(NewSpawnerList := OpenRoomSpawners.RemoveElement[OpenRoomSpawners.Find[InZombieType.GetController[].LastSpawnLocation?]], NewSpawnerList.Length > 0):
            set MaybeSpawnLocation = option{Shuffle(NewSpawnerList)[0]}
        else if:
            RandomSpawn := Shuffle(OpenRoomSpawners)[0]
            # Found an available spawner in the targets' room
            set MaybeSpawnLocation = option{RandomSpawn}
        else:
            # Failed to find an available spawner in targets' room, spawn in adjacent rooms
            var AllLinkedSpawners : []spawn_location = array{}
            for(ThisLinker:TargetedRoom.LinkedRooms):
                LinkedRoom:=ThisLinker.GetLinkedRoom(TargetedRoom)
                LinkedRoomSpawners:=for:
                    ThisSpawner:LinkedRoom.ZombieSpawners
                    ThisSpawner.AllowedTypes.HasType[InZombieType]
                    ThisSpawner.IsAvailable[?IfNeededState:=true]
                do:
                    ThisSpawner
                set AllLinkedSpawners += LinkedRoomSpawners
            if(NewSpawnerList := AllLinkedSpawners.RemoveElement[AllLinkedSpawners.Find[InZombieType.GetController[].LastSpawnLocation?]], NewSpawnerList.Length > 0):
                set MaybeSpawnLocation = option{Shuffle(NewSpawnerList)[0]}
            else:
                set MaybeSpawnLocation = option{Shuffle(AllLinkedSpawners)[0]}
    return MaybeSpawnLocation?

GetAllSpawners<public>(InZombieType:zombie_type_list)<transacts>:[]spawn_location=
    if(SpawnCollection := SpawnLocationCollection[GetSession()]):
        MatchingSpawners:=for:
            ThisSpawnLocation:SpawnCollection
            ThisSpawnLocation.AllowedTypes.HasType[InZombieType]
        do:
            ThisSpawnLocation
        return MatchingSpawners
    array{}

var SpawnLocationCollection : weak_map(session, []spawn_location) = map{}

spawn_location<public> := class<unique>(Interactive_Base):

    # SETTINGS
    @editable Identity<public> : string = "Unset spawner name"
    @editable AllowedTypes<public> : []zombie_type_list = array{zombie_type_list.Default}
    @editable EnabledOnStart : logic = true
    @editable EnableOnWave : type{_X : int where 0 <= _X} = 0
    @editable DisableOnWave : type{_X : int where 0 <= _X} = 0
    @editable DisableOnSpawnCount : type{_X : int where 0 <= _X} = 0
    @editable MaximumZombiesAtATime : type{_X : int where 1 <= _X} = 3
    @editable SpawnPosition : creative_prop = creative_prop{}

    @editable
    UseSpawnAnimation : logic = false

    @editable
    CheckRoomsForSpawning : logic = true

    @editable
    BarrierModule<public> : spawn_location_barrier = spawn_location_barrier{}

    @editable
    Shortcuts : []zombie_shortcut = array{}

    @editable
    AnimationOverride : spawn_animation = spawn_animation{}

    @editable
    Enable_OnTrigger : trigger_device = trigger_device{}

    @editable
    Disable_OnTrigger : trigger_device = trigger_device{}

    # VARIABLES
    var EnabledState : logic = true
    var BarrierReference : barrier_reference = barrier_reference{}
    var BarrierObjects : []barrier_object = array{}
    var<public> SpawningPosition<public> : transform = transform{}

    SpawningQueue : zombie_queue = zombie_queue{}
    BarrierPositionQueue : slot_queue = slot_queue{}
    VaultingQueue : zombie_queue = zombie_queue{}
    var<private> SpawnerZombies : []zombie_ai = array{}

    var LinkedRoom<public> : ?room = false

    # EVENTS
    ZombieSpawningEvent<public> : event(agent) = event(agent){}
    ZombieSpawnedEvent<public> : event(agent) = event(agent){}

    OnBegin<override>()<suspends>:void=
        # Link to global spawn collection
        if:
            not SpawnLocationCollection[GetSession()]
            set SpawnLocationCollection[GetSession()] = array{}
        if:
            CurrentSpawnLocCollection := SpawnLocationCollection[GetSession()]
            set SpawnLocationCollection[GetSession()] = CurrentSpawnLocCollection + array{Self}
        set EnabledState = EnabledOnStart
        # Init barrier and interactive component
        if(BarrierModule.HasBarrier?, BarrierModule.Location.IsValid[]):
            BarrierModule.InteractiveComponent.SetLocation(BarrierModule.Location)
            LinkAndCompile(array{BarrierModule.InteractiveComponent})
            # Get and set barrier reference
            if(ThisBarrierReference:=GetBarrierReference[BarrierModule.ReferenceIndex]):
                set BarrierReference = ThisBarrierReference
            # Initialize all barrier objects
            for(ObjectIndex->ThisObjectReference : BarrierReference.BarrierObjectReferences):
                ObjectTransform := ThisObjectReference.ObjectTransform? or ThisObjectReference.ObjectLocation.GetTransform()
                var ReferenceObjectOffset : vector3 = (if. ThisObjectReference.ObjectTransform? then vector3{} else. BarrierReference.BarrierLocationReference.GetTransform().Translation) - ObjectTransform.Translation
                set ReferenceObjectOffset = ReferenceObjectOffset * vector3{X:=1.0,Y:=1.0,Z:=-1.0}
                LocalTranslationOffset := BarrierModule.Location.GetTransform().GetLocalVector(ReferenceObjectOffset)

                SubtractedRotate := BarrierModule.Location.GetTransform().Rotation.UnrotateBy(BarrierReference.BarrierLocationReference.GetTransform().Rotation)
                FinalRotation := ObjectTransform.Rotation.RotateBy(SubtractedRotate)

                MaybeObjectProp := SpawnProp(ThisObjectReference.StaticAsset, LocalTranslationOffset, FinalRotation)
                if(SpawnedObjectProp:=MaybeObjectProp(0)?):
                    NewBarrierObject : barrier_object = barrier_object{ObjectProp:=SpawnedObjectProp, Index:=ObjectIndex, Transform:=SpawnedObjectProp.GetTransform()}
                    set BarrierObjects += array{NewBarrierObject}
                else{Print("~ERROR: Could not spawn a barrier object prop!")}
        set SpawningPosition = if. SpawnPosition.IsValid[] then. SpawnPosition.GetTransform() else. transform{Scale:=vector3{X:=1.0,Y:=1.0,Z:=1.0}}
        if(ThisLinkedRoom:=GetRoomViaSpawner[Self]):
            set LinkedRoom = option. ThisLinkedRoom

    ForceSpawnZombie<public>(InZombieType:zombie_type_list, ?OverrideWaveEnemy:tuple(logic, logic)=(false, false))<suspends>:void=
        if(ThisController:=InZombieType.GetController[]):
            ThisController.SpawnAndAwait(?ForceSpawnLocation:=option{Self}, ?OverrideWaveEnemy:=OverrideWaveEnemy)
        else. Print("ERR: Failed to fetch zombie controller!")

    SpawnZombie<public>(InZombie:zombie_ai)<suspends>:void=
        set InZombie.SpawnLocation = option{Self}
        set InZombie.Controller.LastSpawnLocation = option{Self}
        set SpawnerZombies += array{InZombie}
        defer{
            #Print("Defer run success")
            if(InZombie.IsActive[]):
                set InZombie.SpawnLocation = false
            GetSpawnerZombies()
            SpawningQueue.ForcePop(InZombie)
            if(FindBarrierObject[false], GetZombiesAtBarrier().Length <= 0):
                SetEnabledState(true)
        }
        CompletedEvent:event()=event(){}
        branch:
            DragSpawnState:=race:
                block:
                    Sleep(0.0)
                    SpawningQueue.LatchOn(InZombie)
                    ZombieSpawningEvent.Signal(InZombie.Agent)
                    defer{
                        SpawningQueue.Pop()
                    }
                    if:
                        UseSpawnAnimation?
                        ThisSpawnAnimModule := block:
                            if. AnimationOverride.AnimationType <> animation_sequence_type.None then. AnimationOverride 
                            else. Shuffle(InZombie.Controller.Animations.Spawns)[0]
                    then:
                        InZombie.StopPropAnimation()
                        ThisSpawnAnimModule.PlayAnimation(SpawningPosition, InZombie)
                    var Tries : int = 0
                    loop:
                        if(InZombie.FortCharacter.IsActive[]){}
                        else:
                            Print("SPAWNING ZOMBIE NOT ACTIVE, RESPAWNING")
                            InZombie.Respawn()
                            return
                        TPPos := SpawningPosition.Translation+vector3{Z:=5.0*Tries}
                        if(InZombie.FortCharacter.TeleportTo[TPPos, SpawningPosition.Rotation.ApplyYaw(DegreesToRadians(-90.0))]):
                            break
                        else:
                            set Tries += 1
                        Sleep(0.5)
                    true
                block:
                    InZombie.RespawnEvent.Await()
                    false
                block:
                    InZombie.DeathEvent.Await()
                    false
                block:
                    SafeSleep(60.0)
                    Print("ZOMBIE PASSED MAXIMUM SPAWNER TIME, RESPAWNING")
                    InZombie.Respawn()
                    false
            if(SpawningQueue.GetQueueSize() <= 0):
                SpawningQueue.PopAll()
            if(DragSpawnState?, IsSpawning[InZombie]):
                SpawningQueue.ForcePop(InZombie)
                #Print("Using shortcuts")
                # Travel shortcuts (if any)
                for(ThisShortcut:Shortcuts):
                    ThisShortcut.TravelTo(InZombie)
                    ShortcutTask:=spawn. ThisShortcut.Use(InZombie)
                    ShortcutTask.Await()
                # Walk zombie to barricade (if applicable)
                if(BarrierModule.HasBarrier?):
                    branch:
                        if(CheckRoomsForSpawning?):
                            loop:
                                AwaitGameState(game_state.InGame)
                                if(CheckPlayerRequirement[?IfNeededState:=true]):
                                    Sleep(1.5)
                                else:
                                    Print("No players in nearby rooms, respawning")
                                    InZombie.Respawn()
                                    break
                        Sleep(0.0)
                    #Print("IN barrier race")
                    race:
                        InZombie.DeathEvent.Await()
                        InZombie.RespawnEvent.Await()
                        block:
                            BarrierQueueSlot:=BarrierPositionQueue.Add()
                            defer{
                                BarrierPositionQueue.ForcePop(BarrierQueueSlot(0))
                            }
                            loop:
                                # Zombie travelling to barricade
                                BarrierTravelLoc:=GetBarrierPosition(BarrierPositionQueue.GetQueuePosition[BarrierQueueSlot(0)] or -1)
                                #DebugDraw:debug_draw = debug_draw{Channel := spawner_debug_draw}
                                #DebugDraw.DrawSphere(BarrierTravelLoc, ?Radius:=BarrierModule.ReachRadius, ?DrawDurationPolicy:=debug_draw_duration_policy.FiniteDuration, ?Duration:=5.0)
                                NavState:=InZombie.SafeNavigate(MakeNavigationTarget(BarrierTravelLoc), ?ReachRadius:=BarrierModule.ReachRadius+RRFIX, ?MovementType:=if. InZombie.IsAggro[] then. Running else. Walking)
                                if(not NavState?):
                                    #Print("Runaway loop detected 1")
                                    return
                                ZombieToBarrierLoc:=DistanceXY(InZombie.FortCharacter.GetTransform().Translation, BarrierTravelLoc)
                                if(ZombieToBarrierLoc < BarrierModule.ReachRadius+25.0 or DistanceXY(InZombie.FortCharacter.GetTransform().Translation, BarrierModule.Location.GetTransform().Translation) < BarrierModule.ReachRadius+20.0):
                                    break
                                else:
                                    Print("Distance : {ZombieToBarrierLoc}, {BarrierQueueSlot(0)}")
                            InZombie.StopPropAnimation()
                            InZombie.BreakBarrierLoop()
                            BarrierPositionQueue.Pop()
                            loop:
                                # Zombie travelling to vaulting position
                                NavState:=InZombie.SafeNavigate(MakeNavigationTarget(BarrierModule.Location.GetTransform().Translation), ?ReachRadius:=28.0, ?MovementType:=if. InZombie.IsAggro[] then. Running else. Walking)
                                if(not NavState?):
                                    #Print("Runaway loop detected 2")
                                    return
                                DistToBarrier:=DistanceXY(InZombie.FortCharacter.GetTransform().Translation, BarrierModule.Location.GetTransform().Translation)
                                if(DistToBarrier < 65.0+12.0):
                                    #Print("hopped barrier")
                                    break
                            defer{
                                VaultingQueue.ForcePop(InZombie)
                            }
                            InZombie.StopPropAnimation()
                            VaultingQueue.LatchOn(InZombie)
                            if(RandomVaultingAnim:=Shuffle(InZombie.Controller.Animations.Vaulting)[0], InZombie.FortCharacter.TeleportTo[InZombie.FortCharacter.GetTransform().Translation, BarrierModule.Location.GetTransform().Rotation]):
                                RandomVaultingAnim.PlayAnimation(BarrierModule.Location.GetTransform(), InZombie, ?CustomEndOffset:=option{vector3{X:=45.0}})
                            VPercent := BarrierModule.Location.GetTransform().GetVectorLookPercent(InZombie.FortCharacter.GetTransform().Translation)
                            if(VPercent >= 0.6){}
                            else:
                                var Tries : int = 1
                                loop:
                                    if(InZombie.FortCharacter.TeleportTo[BarrierModule.Location.GetTransform().GetLocalVector(vector3{X:=20.0*Tries}), BarrierModule.Location.GetTransform().Rotation]):
                                        break
                                    set Tries += 1
                                    if(Tries > 10):
                                        break
                                    AwaitGameState(game_state.InGame)
                                    Sleep(0.15)
                            if(BarrierModule.Location.GetTransform().GetVectorLookPercent(InZombie.FortCharacter.GetTransform().Translation) >= 0.6){}
                            else:
                                Print("Failed to vault, respawning")
                                InZombie.Respawn()
                            VaultingQueue.Pop()
                    if(GetZombiesAtBarrier().Length <= 0, FindBarrierObject[false]):
                        SetEnabledState(true)
                if(InZombie.IsActive[], InZombie.SpawnLocation? = Self):
                    ZombieSpawnedEvent.Signal(InZombie.Agent)
                    if(InZombie.AutoTargetOnSpawn?):
                        spawn. InZombie.PlayState(zombie_state.Targeting)
            else:
                Print("Zombie failed drag spawn state!")
            CompletedEvent.Signal()
        race:
            CompletedEvent.Await()
            InZombie.AwaitStateChanged()
        RemoveZombie(InZombie)

    RemoveZombie<public>(InZombie:zombie_ai):void=
        if(InZombie.IsActive[]):
            set InZombie.SpawnLocation = false
        GetSpawnerZombies()
        SpawningQueue.ForcePop(InZombie)
        InZombie.StopPropAnimation()

    RepairBarricade<public>(?InAgent:?agent=false, ?FullRepair:logic=false)<suspends>:void=
        Range := if. FullRepair? then. BarrierObjects.Length else. 1
        for(RepairingIndex:=0..Range-1):
            if(ThisBarrierObject:=FindBarrierObject[false]):
                ThisBarrierObject.SetState(barrier_object_state.Repairing)
                ThisBarrierObject.BrokenDownProp.Dispose()
                if(ThisObjectReference:=BarrierReference.BarrierObjectReferences[ThisBarrierObject.Index]):
                    if(not FullRepair?):
                        MaybeSpawnedRepairAnimation := SpawnProp(ThisObjectReference.RepairAsset, ThisBarrierObject.Transform)
                        if(ThisSpawnedRepairAnimation := MaybeSpawnedRepairAnimation(0)?):
                            Sleep(ThisObjectReference.RepairTime)
                            ThisSpawnedRepairAnimation.Dispose()
                    ThisBarrierObject.ObjectProp.Show()
                    ThisBarrierObject.SetState(barrier_object_state.Active)
                    if(ThisPlayerInfo:=InAgent?.GetPlayerInfo[]):
                        ThisPlayerInfo.IncreasePoints(10)
                        ThisPlayerInfo.RebuildBarrierEvent.Signal()

    (InZombie:zombie_ai).BreakBarrierLoop()<suspends>:void=
        SetEnabledState(false)
        var CurrentBarrierObject : barrier_object = barrier_object{}
        defer{
            if(CurrentBarrierObject.State = barrier_object_state.Targeted):
                CurrentBarrierObject.SetState(barrier_object_state.Active)
            if(GetZombiesAtBarrier().Length <= 0):
                SetEnabledState(true)
        }
        loop:
            AwaitGameState(game_state.InGame)
            # Find a barrier object and break it
            if(ThisBarrierObject:=FindBarrierObject[true]):
                # Active barrier object found
                #Print("Found active barrier object")
                set CurrentBarrierObject = ThisBarrierObject
                InZombie.BreakBarrier(ThisBarrierObject)
                SafeSleep(0.65) # -> Small delay to make barriers break down slower
            else:
                # No Active objects found
                BrokenObjectAmount:=for:
                    ThisObject:BarrierObjects
                    ThisObject.State = barrier_object_state.Broken
                do:
                    0
                # Check if all barrier objects are in the broken state to exit
                if(BrokenObjectAmount.Length >= BarrierObjects.Length):
                    break
                # If a barrier object is still being targeted, continue to wait on it
                Sleep(0.15)
        Sleep(0.1)

    (InZombie:zombie_ai).BreakBarrier(InBarrierObject:barrier_object)<suspends>:void=
        if(BarrierModule.HasBarrier?):
            InBarrierObject.SetState(barrier_object_state.Targeted)
            if(ThisObjectReference:=BarrierReference.BarrierObjectReferences[InBarrierObject.Index]):
                MaybeAttackAnim:=InZombie.GetAndPlayAnimation(zombie_state.Attacking)
                if(ThisAttackAnimPackage:=attack_animation[MaybeAttackAnim?]):
                    AttackManipulators:=ThisAttackAnimPackage.Manipulators.GetManipulators(InZombie.AggroLevel, InZombie.Controller.Controls.AggressiveLimit)
                    SafeSleep(ThisAttackAnimPackage.TimeOfAttack-AttackManipulators(2)/AttackManipulators(0))
                    defer{
                        InBarrierObject.SetState(barrier_object_state.Broken)
                        InBarrierObject.ObjectProp.Hide()
                    }
                    InBarrierObject.SetState(barrier_object_state.Breaking)
                    InBarrierObject.ObjectProp.Hide()
                    if(ThisObjectReference.UseBreakdownAsset?):
                        MaybeSpawnedBreakdownAnimation := SpawnProp(ThisObjectReference.BreakdownAsset, InBarrierObject.Transform.Translation, BarrierModule.Location.GetTransform().Rotation)
                        if(ThisSpawnedBreakdownAnimation := MaybeSpawnedBreakdownAnimation(0)?):
                            set InBarrierObject.BrokenDownProp = ThisSpawnedBreakdownAnimation
                    else:
                        InBarrierObject.BrokenDownProp.Dispose()
                    SafeSleep(ThisObjectReference.BreakdownTime)
                else:
                    InBarrierObject.SetState(barrier_object_state.Active)
                    Print("ERR: Failed to get an attack animation for a barrier breakdown!")
            else:
                Print("~ERROR: Cannot find the desired barrier object reference!")

    FindBarrierObject(DesiredState:logic)<transacts><decides>:barrier_object=
        var MaybeBarrierObject : ?barrier_object = false
        for:
            ThisBarrierObject : BarrierObjects
            (DesiredState? and ThisBarrierObject.State = barrier_object_state.Active) or (not DesiredState? and ThisBarrierObject.State = barrier_object_state.Broken)
        do:
            set MaybeBarrierObject = option{ThisBarrierObject}
            MaybeBarrierObject?
        MaybeBarrierObject?

    GetBarrierPosition<private>(QueueIndex:int)<transacts>:vector3=
        case(QueueIndex):
            0 => BarrierModule.Location.GetTransform().GetLocalVector(vector3{X:=-5.0, Y:=-55.0})
            1 => BarrierModule.Location.GetTransform().GetLocalVector(vector3{X:=-15.0, Y:=-15.0})
            2 => BarrierModule.Location.GetTransform().GetLocalVector(vector3{X:=-15.0, Y:=15.0})
            3 => BarrierModule.Location.GetTransform().GetLocalVector(vector3{X:=-5.0, Y:=55.0})

            _ => 
                Print("Returning default barrier position for index [{QueueIndex}]")
                BarrierModule.Location.GetTransform().GetLocalVector(vector3{X:=-5.0, Y:=-55.0})

    GetSpawnerZombies<public>()<transacts>:[]zombie_ai=
        SpawnerList:=for:
            ThisZombie:SpawnerZombies
            ThisZombie.IsState[zombie_state.SpawningProcess]
            ThisZombie.SpawnLocation? = Self
        do:
            ThisZombie
        set SpawnerZombies = SpawnerList
        SpawnerList

    GetZombiesAtBarrier<public>()<transacts>:[]zombie_ai=
        BarrierList:=for:
            ThisZombie:SpawnerZombies
            ThisZombie.IsActive[]
            DistanceXY(ThisZombie.FortCharacter.GetTransform().Translation, BarrierModule.Location.GetTransform().Translation) <= BarrierModule.ReachRadius
        do:
            ThisZombie

    IsAvailable<public>(?IfNeededState:logic=false)<transacts><decides>:void=
        EnabledState?
        GetSpawnerZombies().Length < MaximumZombiesAtATime
        not CheckRoomsForSpawning? or CheckPlayerRequirement[?IfNeededState:=IfNeededState]

    CheckPlayerRequirement<private>(?IfNeededState:logic=false)<transacts><decides>:void=
        ThisRoom:=GetRoomViaSpawner[Self]
        ThisRoom.GetPlayersInRoom().Length > 0 or block{
            SatisfiedRooms:=for:
                ThisRoomLinker:ThisRoom.LinkedRooms
                ThisRoomLinker.CheckSpawnerSharing[ThisRoom, ?IfNeededState:=IfNeededState]
                ThisRoomLinker.GetLinkedRoom(ThisRoom).GetPlayersInRoom().Length > 0
            do:
                0
            SatisfiedRooms.Length > 0
        }

    GetEndLocation<public>()<transacts>:tuple(transform, int)=
        if(BarrierModule.HasBarrier?):
            (BarrierModule.Location.GetTransform(), 0)
        else if(LastShortcut:=Shortcuts[Shortcuts.Length-1]):
            (LastShortcut.EndTransform, 1)
        else:
            (SpawningPosition, 2)

    IsSpawning<public>(InZombie:zombie_ai)<transacts><decides>:void=
        InZombie.SpawnLocation? = Self

    Enable<public>():void=
        set EnabledState = true

    Disable<public>():void=
        set EnabledState = false
    
    GetInteractionText<override>(InAgent:agent, Interactive:Interactable):string=
        if(FindBarrierObject[false], GetZombiesAtBarrier().Length <= 0):
            SetEnabledState(true)
            return "rebuild barrier"
        else:
            SetEnabledState(false)
            return ""

    OnInteracted<override>(InAgent:agent, Interactive:Interactable):void=
        spawn. RepairBarricade(?InAgent:=option{InAgent})

spawn_location_barrier<public> := class<concrete>:

    @editable
    HasBarrier<public> : logic = false

    @editable
    ReferenceIndex : type{_X : int where _X >= 0} = 0

    @editable
    Location<public> : creative_prop = creative_prop{}

    @editable
    ReachRadius : float = 100.0

    @editable
    InteractiveComponent : Interactable = Interactable{InteractDistance:=125.0, InteractionSpriteOffset:=vector3{X:=30.0}, RequiresLookAt:=false}

GetBarrierReference(Index:int)<transacts><decides>:barrier_reference=
    return GlobalBarrierReferenceCollection[GetSession()][Index]

var GlobalBarrierReferenceCollection : weak_map(session, []barrier_reference) = map{}

BarrierReferenceCollection := class(creative_device):

    @editable
    BarrierReferences : []barrier_reference = array{}

    OnBegin<override>():void=
        if:
            set GlobalBarrierReferenceCollection[GetSession()] = array{}
            set GlobalBarrierReferenceCollection[GetSession()] = BarrierReferences
        else:
            Print("ERROR: Could not link barrier references!")

barrier_reference := class<concrete>:

    @editable BarrierLocationReference : creative_prop = creative_prop{}
    @editable BarrierObjectReferences : []object_reference = array{}

# A reference for a barrier object #
# Used for reference on an instanced barrier's objects #
object_reference := class<concrete>:

    @editable ObjectLocation : creative_prop = creative_prop{}
    @editable ObjectTransform : ?transform = false
    @editable StaticAsset : creative_prop_asset = DefaultCreativePropAsset
    @editable RepairAsset : creative_prop_asset = DefaultCreativePropAsset
    @editable RepairTime : float = 0.0
    @editable UseBreakdownAsset : logic = false
    @editable BreakdownAsset : creative_prop_asset = DefaultCreativePropAsset
    @editable BreakdownTime : float = 0.0

# Contains information about a barrier object instance #
# Used to track the state of the barrier object #
barrier_object := class<concrete>:

    var ObjectProp : creative_prop = creative_prop{}
    var BrokenDownProp : creative_prop = creative_prop{}
    var<private> State : barrier_object_state = barrier_object_state.Active
    
    var Index : int = -1
    var Transform : transform = transform{}

    StateChangedEvent : event(barrier_object_state) = event(barrier_object_state){}

    SetState<public>(InState:barrier_object_state):void=
        set State = InState
        StateChangedEvent.Signal(InState)

    AwaitState<public>(DesiredState:barrier_object_state)<suspends>:void=
        if(State <> DesiredState):
            loop:
                NewState:=StateChangedEvent.Await()
                if(NewState=DesiredState):
                    break

barrier_object_state := enum{
    Active
    Targeted
    Breaking
    Broken
    Repairing
}