using { /Verse.org/Simulation }
using { /Fortnite.com/Devices }
using { /UnrealEngine.com/Temporary/SpatialMath }
using { Zombie }
using { Player }
using { UI }
using { Powerups }

SkipToWave<public>(Wave:int, ?UseIntermission:logic=false):void=
    if(ThisWaveManager:=GlobalWaveController[GetSession()]):
        ThisWaveManager.SkipToWaveInternal(Wave, ?UseIntermission:=UseIntermission)

# Waits until the wave (or a higher one if force skipped) has been reached
AwaitWave<public>(DesiredWave:int)<suspends>:void=
    if(GetWaveCount() < DesiredWave, ThisWaveDevice:=GlobalWaveController[GetSession()]):
        loop:
            ThisWaveDevice.WaveStartedEvent.Await()
            if(GetWaveCount() >= DesiredWave):
                return

AwaitWaveEnd<public>()<suspends>:void=
    if(ThisWaveDevice:=GlobalWaveController[GetSession()], ThisWaveDevice.WaveActiveStatus?):
        ThisWaveDevice.WaveEndedEvent.Await()
    else:
        Sleep(0.0)

GetWaveInfo<public>()<transacts><decides>:wave_info=
    ThisWaveController := GlobalWaveController[GetSession()]
    return ThisWaveController.GetWaveInfoAtWave[ThisWaveController.TotalWaves]

GetWaveCount<public>()<transacts>:int=
    GlobalWaveController[GetSession()].TotalWaves or 0

IsSpecialWave<public>()<transacts><decides>:special_wave_info=
    special_wave_info[GetWaveInfo[]]

IsWaveActive<public>()<transacts><decides>:void=
    GlobalWaveController[GetSession()].WaveActiveStatus?

IsWaveOrGreater<public>(DesiredWave:int)<transacts><decides>:void=
    GlobalWaveController[GetSession()].TotalWaves >= DesiredWave

var GlobalWaveController<public> : weak_map(session, WaveController) = map{}

WaveController<public> := class(creative_device):

    #@editable
    #WaveInfoList : []wave_info = array{}

    @editable
    EasyWaveInfo : []wave_info = array{
        wave_info:
            TimeBetweenSpawns:=5.0
            SpawnDelay:=8.0
            WaveRepeatCount:=1
        wave_info:
            TimeBetweenSpawns:=2.5
            SpawnDelay:=4.2
            SpawnsAtOnce:=2
            WaveRepeatCount:=4
            MaximumZombiesAtATime:=18
        wave_info:
            TimeBetweenSpawns:=2.0
            SpawnDelay:=2.0
            SpawnsAtOnce:=3
            WaveRepeatCount:=8
        wave_info:
            TimeBetweenSpawns:=4.0
            SpawnDelay:=1.0
            SpawnsAtOnce:=6
            WaveRepeatCount:=13
        wave_info:
            TimeBetweenSpawns:=3.5
            SpawnDelay:=1.0
            SpawnsAtOnce:=8
            WaveRepeatCount:=16
        wave_info:
            TimeBetweenSpawns:=3.5
            SpawnDelay:=1.0
            SpawnsAtOnce:=9
            WaveRepeatCount:=20
        wave_info:
            TimeBetweenSpawns:=1.5
            SpawnDelay:=1.0
            SpawnsAtOnce:=11
            WaveRepeatCount:=25
        wave_info:
            TimeBetweenSpawns:=0.5
            SpawnDelay:=1.0
            SpawnsAtOnce:=13
    }

    @editable
    ClassicWaveInfo : []wave_info = array{
        wave_info:
            TimeBetweenSpawns:=5.0
            SpawnDelay:=8.0
        wave_info:
            TimeBetweenSpawns:=5.0
            SpawnsAtOnce:=3
            SpawnDelay:=4.2
            WaveRepeatCount:=4
            MaximumZombiesAtATime:=22
        wave_info:
            TimeBetweenSpawns:=2.0
            SpawnsAtOnce:=5
            SpawnDelay:=2.0
            WaveRepeatCount:=8
        wave_info:
            TimeBetweenSpawns:=4.0
            SpawnsAtOnce:=8
            SpawnDelay:=1.0
            WaveRepeatCount:=10
        wave_info:
            TimeBetweenSpawns:=3.0
            SpawnsAtOnce:=10
            SpawnDelay:=1.0
            WaveRepeatCount:=13
        wave_info:
            TimeBetweenSpawns:=2.0
            SpawnsAtOnce:=12
            SpawnDelay:=1.0
            WaveRepeatCount:=16
        wave_info:
            TimeBetweenSpawns:=3.0
            SpawnsAtOnce:=14
            SpawnDelay:=1.0
            WaveRepeatCount:=20
        wave_info:
            TimeBetweenSpawns:=2.0
            SpawnsAtOnce:=15
            SpawnDelay:=1.0
            WaveRepeatCount:=25
        wave_info:
            TimeBetweenSpawns:=1.6
            SpawnsAtOnce:=15
            SpawnDelay:=1.0
    }

    @editable
    HardcoreWaveInfo : []wave_info = array{
        wave_info:
            TimeBetweenSpawns:=5.0
            SpawnsAtOnce:=2
            SpawnDelay:=8.0
        wave_info:
            TimeBetweenSpawns:=5.0
            SpawnsAtOnce:=5
            SpawnDelay:=2.0
            WaveRepeatCount:=3
        wave_info:
            TimeBetweenSpawns:=3.0
            SpawnsAtOnce:=7
            SpawnDelay:=0.0
            WaveRepeatCount:=7
        wave_info:
            TimeBetweenSpawns:=3.0
            SpawnsAtOnce:=8
            SpawnDelay:=1.0
            WaveRepeatCount:=10
        wave_info:
            TimeBetweenSpawns:=2.5
            SpawnsAtOnce:=12
            SpawnDelay:=1.0
            WaveRepeatCount:=15
        wave_info:
            TimeBetweenSpawns:=1.1
            SpawnsAtOnce:=13
            SpawnDelay:=1.0
    }

    @editable
    SpecialWaveControls<public> : special_wave_controls = special_wave_controls{}

    @editable
    Controls<public> : wave_controls = wave_controls{}

    @editable
    Audio : wave_audio = wave_audio{}

    @editable
    Signal_In : wave_signal_in = wave_signal_in{}

    @editable
    Signal_Out : wave_signal_out = wave_signal_out{}

    # VARIABLES
    var<private> TotalWaves<public> : int = 0
    var<private> WaveActiveStatus : logic = false
    var<private> IsWaveLoopActive<public> : logic = false
    var AllowWaveEnd<private> : logic = true

    var WavePausedTime<public> : float = 0.0
    var QuickWaves<public> : int = 0

    # Contains info for controls per difficulty level
    var DifficultyControlMap : [game_difficulty]wave_controls = map{}
    var<public> PlagueSpawnWave : int = 17

    # EVENTS
    WaveStartedEvent<public> : event() = event(){}
    WaveEndedEvent<public> : event(logic) = event(logic){}
    StopWaveLoop<private> : event() = event(){}

    NewWaveEndPermission<private> : event(logic) = event(logic){}

    OnBegin<override>()<suspends>:void=
        # Global link
        if. set GlobalWaveController[GetSession()] = Self
        else. Print("ERROR: Failed to globally link wave controller!")
        # Init Singal In
        Signal_In.StartWaves_OnTrigger.TriggeredEvent.Subscribe(StartWaves_SignalIn)
        Signal_In.StopWaves_OnTrigger.TriggeredEvent.Subscribe(StopWaves_SignalIn)
        Signal_In.RestartWaves_OnTrigger.TriggeredEvent.Subscribe(RestartWaves_SignalIn)
        Signal_In.SkipWave_OnTrigger.TriggeredEvent.Subscribe(SkipWave_SignalIn)
        Signal_In.SkipWaveAndIntermission_OnTrigger.TriggeredEvent.Subscribe(SkipWaveAndIntermission_SignalIn)
        for(ThisSkipWave:Signal_In.SkipToWaveTriggers):
            ThisSkipWave.Init()
        # Initialize difficulty settings
        ClassicSettings := wave_controls{
            AggressiveWaveStart := Controls.AggressiveWaveStart
            AggressiveWaveMaximum := Controls.AggressiveWaveMaximum
            MaximumWaveZombieCapPerPlayer := Controls.MaximumWaveZombieCapPerPlayer
            #MaximumZombieSpeed := Controls_Module.MaximumZombieSpeed
            #SpeedRandomBounds := Controls.SpeedRandomBounds
            #MinimumSpawnRate := Controls.MinimumSpawnRate
            #SpawnRateMultiplier := Controls.SpawnRateMultiplier
            #EclipseRound := Controls_Module.EclipseRound
        }

        # Setting Classic difficulty to default values
        if:
            set DifficultyControlMap[game_difficulty.Classic] = ClassicSettings
            set DifficultyControlMap[game_difficulty.Easy] = wave_controls{
                AggressiveWaveStart := ClassicSettings.AggressiveWaveStart + 2
                AggressiveWaveMaximum := ClassicSettings.AggressiveWaveMaximum + 2 
                #MaximumZombieSpeed := ClassicSettings.MaximumZombieSpeed - 0.1
                #SpeedRandomBounds := ClassicSettings.SpeedRandomBounds
                #MinimumSpawnRate := ClassicSettings.MinimumSpawnRate
                #SpawnRateMultiplier := ClassicSettings.SpawnRateMultiplier
                #EclipseRound := ClassicSettings.EclipseRound + 3
            }
            set DifficultyControlMap[game_difficulty.Hardcore] = wave_controls{
                AggressiveWaveStart := ClassicSettings.AggressiveWaveStart - 2
                AggressiveWaveMaximum := ClassicSettings.AggressiveWaveMaximum - 4
                MaximumWaveZombieCapPerPlayer := ClassicSettings.MaximumWaveZombieCapPerPlayer + 10
                #MaximumZombieSpeed := ClassicSettings.MaximumZombieSpeed + 0.15
                #SpeedRandomBounds := ClassicSettings.SpeedRandomBounds
                #MinimumSpawnRate := ClassicSettings.MinimumSpawnRate
                #SpawnRateMultiplier := ClassicSettings.SpawnRateMultiplier
                #EclipseRound := ClassicSettings.EclipseRound - 4
            }
        # Start game if set to
        if(Controls.ActivateOnGameStartDelay >= 0.0):
            Sleep(Controls.ActivateOnGameStartDelay+0.1)
            spawn. WaveLoop()

    StartWaves<public>():void=
        spawn. WaveLoop()

    StopWaves<public>():void=
        StopWaveLoop.Signal()

    RestartWaves<public>():void=
        StopWaves()
        StartWaves()

    SkipWave<public>(?UseIntermission:logic=true):void=
        WaveEndedEvent.Signal(UseIntermission)

    SkipToWaveInternal<public>(Wave:int, ?UseIntermission:logic=false):void=
        set TotalWaves = Wave-1
        WaveEndedEvent.Signal(UseIntermission)

    WaveLoop<private>()<suspends>:void=
        if(not IsWaveLoopActive?):
            set IsWaveLoopActive = true
            set AllowWaveEnd = true
            set TotalWaves = Controls.StartingWave-1
            ApplyDifficulty()
            race:
                AwaitGameState(game_state.GameOver)
                StopWaveLoop.Await()
                loop:
                    OnWaveBegin()
                    UseIntermissionState:=AwaitWaveEndCondition()
                    OnWaveEnd()
                    if(UseIntermissionState?):
                        SafeSleep(Controls.IntermissionTime)
            set IsWaveLoopActive = false
            set WaveActiveStatus = false
            if(ThisWaveInfo:=GetWaveInfo[]):
                ThisWaveInfo.Cancel()
                ThisWaveInfo.Eliminate()

    OnWaveBegin<private>()<suspends>:void=
        set WaveActiveStatus = true
        set TotalWaves += 1
        WaveStartedEvent.Signal()
        if(IsSpecialWave[]):
            Audio.SpecialRoundStartStinger.Play()
        else:
            if(TotalWaves = 1):
                SafeSleep(2.0)
            Audio.RoundStartStinger.Play()
        SafeSleep(Controls.RoundStartDelay)
        Signal_Out.OnWaveStarted_ActivateTrigger.Trigger()
        Signal_Out.ActivateWaveStartedTrigger(TotalWaves)
        PlayerList := GetAgents(?OnlyActives:=false)
        for(ThisPlayer:PlayerList):
            ThisPlayer.GetUIManager().SetWaveText("{TotalWaves}")
        if(IsSpecialWave[]):
            Print("WaveController: A new special wave has begun! [{TotalWaves}]")
            Signal_Out.OnSpecialWaveStarted_ActivateTrigger.Trigger()
        else:
            Print("WaveController: A new wave has begun! [{TotalWaves}]")
        if(ThisWaveInfo := GetWaveInfoAtWave[TotalWaves]):
            spawn. ThisWaveInfo.Spawn()
        else:
            Print("ERROR REPORT: Failed to find a wave info at wave {TotalWaves}!")

    AwaitWaveEndCondition<public>()<suspends>:logic=
        if(ThisWaveInfo:=GetWaveInfo[]):
            branch:
                ThisWaveInfo.AwaitWaveElimination()
                WaveEndedEvent.Signal(true)
            branch:
                loop:
                    AwaitGameState(game_state.Paused)
                    StartP:=GetSimulationElapsedTime()
                    AwaitGameState(game_state.InGame)
                    set WavePausedTime += (GetSimulationElapsedTime()-StartP)
            WaveEndedEvent.Await()
            if(not AllowWaveEnd?):
                loop:
                    NewState := NewWaveEndPermission.Await()
                    if(NewState?):
                        break
        else:
            Print("ERR: Failed to fetch wave info!")
        return true

    OnWaveEnd<private>():void=
        if(not IsGameDifficulty[game_difficulty.Easy], ThisWaveInfo:=GetWaveInfo[]):
            Print("EndWaveTime: {(GetSimulationElapsedTime()-ThisWaveInfo.StartTime)-WavePausedTime}")
            Print("QuickWaveTime: {(30.0*(TotalWaves*0.30+1.15))}")
            if((GetSimulationElapsedTime()-ThisWaveInfo.StartTime-WavePausedTime) < (30.0*(TotalWaves*0.30+1.15))):
                set QuickWaves += 1
        set WaveActiveStatus = false
        if(IsSpecialWave[]):
            Audio.SpecialRoundEndStinger.Play()
            if(ThisPowerup := special_wave_info[GetWaveInfo[]].OnWaveEndDropPowerup.ConvertToPowerup[]):
                # Handle spawning at last eliminated zombie
                var LastElimZombieTuple : tuple(transform, float) = (transform{}, -1.0)
                for(ThisZombieType:GetZombieTypes()):
                    if(LastZombie := ThisZombieType.LastEliminatedZombie?, LastZombie.TimeOfDeath > LastElimZombieTuple(1)):
                        set LastElimZombieTuple = (LastZombie.DeathTransform, LastZombie.TimeOfDeath)
                if(LastElimZombieTuple(1) > -1.0):
                    ThisPowerup.Spawn(LastElimZombieTuple(0).Translation)
            Signal_Out.OnSpecialWaveEnded_ActivateTrigger.Trigger()
        else:   
            Audio.RoundEndStinger.Play()
        Signal_Out.OnWaveEnded_ActivateTrigger.Trigger()
        UpdateSpawnLocations() # -> Clears spawn location zombie data
        RefreshZombieList()
        Print("WaveController: The wave has ended...")

    # Returns the wave info that will be active for the given wave
    # Returns the last wave info data if no higher wave info's are found
    GetWaveInfoAtWave<public>(DesiredWave:int)<transacts><decides>:wave_info=
        var MaybeWaveInfo : ?wave_info = false
        # Get wave info list to look at based on difficulty
        ThisWaveInfoList := case(GetGameDifficulty()):
            game_difficulty.Easy => EasyWaveInfo
            game_difficulty.Classic => ClassicWaveInfo
            game_difficulty.Hardcore => HardcoreWaveInfo
        # Check if wave is a special wave
        if(SpecialWaveControls.SpecialWaveInfoList.Length > 0, DesiredWave >= SpecialWaveControls.FirstSpecialWaveAppearance, Mod[DesiredWave-SpecialWaveControls.FirstSpecialWaveAppearance, SpecialWaveControls.RepeatEveryXWaves] = 0):
            if(ThisSpecialWave := SpecialWaveControls.SpecialWaveInfoList[Int[((DesiredWave-SpecialWaveControls.FirstSpecialWaveAppearance)*1.0)/((SpecialWaveControls.RepeatEveryXWaves)*1.0)]] or SpecialWaveControls.SpecialWaveInfoList[SpecialWaveControls.SpecialWaveInfoList.Length-1]):
                set MaybeWaveInfo = option{ThisSpecialWave}
        # If not a special wave find the normal wave from the list 
        var currentWave : int = 1
        for(ThisIndex->ThisWaveInfo:ThisWaveInfoList):
            if(not MaybeWaveInfo?):
                for(i := 0..ThisWaveInfo.WaveRepeatCount):
                    if(currentWave = DesiredWave):
                        set MaybeWaveInfo = option{ThisWaveInfo}
                    set currentWave += 1
        # Return the found wave, or return the last wave info in the list
        return MaybeWaveInfo? or ThisWaveInfoList[ThisWaveInfoList.Length-1]

    GetAggresionPercent<public>()<transacts>:float=
        Print("InAggroP")
        AggroPercent:=if(TotalWaves < Controls.AggressiveWaveStart):
            Clamp(Pow(Controls.QuickWaveAggressionMultiplier, QuickWaves*1.0), 0.0, 100.0)
        else if(TotalWaves >= Controls.AggressiveWaveStart, Controls.AggressiveWaveStart = Controls.AggressiveWaveMaximum):
            100.0
        else:
            Clamp(((1.0*(TotalWaves-Controls.AggressiveWaveStart)/(1.0*(Controls.AggressiveWaveMaximum-Controls.AggressiveWaveStart)))*100.0)+(Pow(Controls.QuickWaveAggressionMultiplier, QuickWaves*1.0)), 0.0, 100.0)
        Print("AggroLevel: {AggroPercent}")
        return AggroPercent

    GetWaveZombieCap<public>():int=
        PlayerCount := GetPlayers(?OnlyActives:=false).Length
        var ZombieCap : int = 0
        if(TotalWaves < 5):
            var CountNum : float = 6.0
            if(TotalWaves = 1):
                set CountNum = 6.0
            else:
                set CountNum = (TotalWaves * (6.0/4.5)) + 6.0
            case(PlayerCount):
                1 => 
                2 => set CountNum *= 1.166
                3 => set CountNum *= 1.500
                4 => set CountNum *= 1.666
                _ => set CountNum *= 1.8
            if(ThisCountInt := Ceil[CountNum]):
                set ZombieCap = ThisCountInt
        else:
            var Multiplier : float = 0.09
            case(PlayerCount):
                1 => set Multiplier = 0.09
                2 => set Multiplier = 0.18
                3 => set Multiplier = 0.36
                4 => set Multiplier = 0.54
                _ => 
                    Print("~ERR: Player count exceeds 4, err in calculating zombie health!")
                    set Multiplier = 0.54
            if(ThisCountInt := Int[Multiplier * Pow(TotalWaves*1.0, 2.0) + 24.0]):
                set ZombieCap = ThisCountInt
        WaveStartTime := GetWaveInfo[].StartTime or 0.0
        PlayerList:=GetPlayers(?OnlyActives:=false)
        WavePlayerAmount:=for:
            ThisPlayer:PlayerList
            ThisPlayer.JoinTime < WaveStartTime
        do:
            0
        if(ZombieCap > (Controls.MaximumWaveZombieCapPerPlayer*WavePlayerAmount.Length)):
            return (Controls.MaximumWaveZombieCapPerPlayer*WavePlayerAmount.Length)
        else:
            return ZombieCap

    GetSpeedValue<public>(?InZombieInfo:?zombie_ai=false, ?ForceAggressive:logic=false, ?NoModify:logic=false):float=
        var RandomSpeed : float = (Pow(Log(10.0, (TotalWaves*17.0)/(1.0*Controls.AggressiveWaveMaximum)), 2.0)+0.486)*(Controls.MaximumZombieSpeed/2.0)
        if(ThisZombieInfo:=InZombieInfo?, not NoModify?):
            AggroZombieCount := GetZombieList(?OnlyAggro:=true).Length
            if(TotalWaves >= Controls.AggressiveWaveStart, GetZombieList(?OnlyWave:=true).Length*(ThisZombieInfo.Controller.Controls.AggressiveLimit/100.0) > AggroZombieCount*1.0):
                set RandomSpeed = ThisZombieInfo.Controller.Controls.MaximumSpeed
                ThisZombieInfo.SetAggroLevel(100.0)
        if(MaxSpeed := Controls.MaximumZombieSpeed, RandomSpeed > MaxSpeed):
            return MaxSpeed
        if(RandomSpeed < 0.5):
            return 0.5
        return RandomSpeed

    # Controls whether wave zombies should continue spawning or not
    SetSpawningState<public>(InState:logic):void=
        if(ThisWaveInfo:=GetWaveInfo[]):
            for(ThisSpawnGroup:ThisWaveInfo.SpawnPoints):
                if(InState?):
                    ThisSpawnGroup.Resume()
                else:
                    ThisSpawnGroup.Pause()

    SetWaveEndPermission<public>(InState:logic):void=
        set AllowWaveEnd = InState
        NewWaveEndPermission.Signal(InState)

    ApplyDifficulty<public>():void=
        CurrDifficulty:=GetGameDifficulty()
        case(CurrDifficulty):
            game_difficulty.Easy => set PlagueSpawnWave = 19
            game_difficulty.Classic =>
            game_difficulty.Hardcore => set PlagueSpawnWave = 14
        if(DifficultySettings := DifficultyControlMap[CurrDifficulty]):
            Controls.ApplySettings(DifficultySettings)

    # SIGNAL IN FUNCTIONS
    StartWaves_SignalIn<private>(Agent:?agent):void=
        StartWaves()

    StopWaves_SignalIn<private>(Agent:?agent):void=
        StopWaves()

    RestartWaves_SignalIn<private>(Agent:?agent):void=
        RestartWaves()

    SkipWave_SignalIn<private>(Agent:?agent):void=
        SkipWave()

    SkipWaveAndIntermission_SignalIn<private>(Agent:?agent):void=
        SkipWave(?UseIntermission:=false)

# Holds information about a wave
wave_info<public> := class<concrete>:

    @editable
    SpawnPoints<public> : []spawn_group_wave = array{spawn_group_wave{}}

    @editable
    WaveZombieCapPercent<public> : type{_X : float where _X >= 0.0, _X <= 1.0} = 1.0

    @editable
    MaximumZombiesAtATime<public> : type{_X:int where _X >= 1} = 25
    # Multiplied by player count
    # 1 -> *1
    # 2 -> *1.3
    # 3 -> *1.75
    # 4 -> *2.0

    @editable
    SpawnDelay<public> : float = 2.5

    @editable
    TimeBetweenSpawns<public> : float = 2.0

    @editable
    SpawnsAtOnce<public> : type{_X:int where _X >= 0} = 0

    @editable
    WaveRepeatCount<public> : type{_X : int where _X >= 0} = 0

    var LinkedSpawnGroups<private> : []spawn_group_base = array{}
    var SkullMarkedState<private> : logic = false
    var<private> StartTime<public> : float = 0.0

    CancelEvent<private> : event() = event(){}

    ZombieSpawnedEvent<public> : event(agent) = event(agent){}
    SpawningCompletedEvent<public> : event() = event(){}
    WaveZombieEliminatedEvent<public> : event(agent) = event(agent){}
    AllZombiesEliminatedEvent<public> : event() = event(){}
    ZombieGroupEliminatedEvent<private> : event() = event(){}

    Spawn<public>()<suspends>:void=
        set LinkedSpawnGroups = array{}
        set SkullMarkedState = false
        set StartTime = GetSimulationElapsedTime()
        branch:
            SafeSleep(SpawnDelay)
            for(Index->ThisSpawnPoint:SpawnPoints):
                if(ThisSpawnPoint.WaitForPreviousGroupElimination?, PreviousSpawnPoint := SpawnPoints[Index-1]):
                    PreviousSpawnPoint.AwaitGroupElimination()
                set ThisSpawnPoint.SpawnRate = TimeBetweenSpawns
                if. SpawnsAtOnce > 0 then. set ThisSpawnPoint.SpawnsAtOnce = SpawnsAtOnce
                ThisSpawnPoint.Spawn()
            SpawningCompletedEvent.Signal()
            CancelEvent.Signal()
        CancelEvent.Await()

    Cancel<public>():void=
        CancelEvent.Signal()
        for(ThisSpawnPoint:SpawnPoints):
            ThisSpawnPoint.Cancel()

    Pause<public>():void=
        for(ThisSpawnPoint:(SpawnPoints+LinkedSpawnGroups)):
            ThisSpawnPoint.Pause()

    Resume<public>():void=
        for(ThisSpawnPoint:(SpawnPoints+LinkedSpawnGroups)):
            ThisSpawnPoint.Resume()

    Eliminate<public>():void=
        for(ThisSpawnPoint:SpawnPoints):
            ThisSpawnPoint.Eliminate()

    LinkSpawnGroup<public>(SpawnGroup:spawn_group_base):void=
        if(not LinkedSpawnGroups.FindInstance[SpawnGroup] and not SpawnPoints.FindInstance[SpawnGroup]):
            set LinkedSpawnGroups += array{SpawnGroup}
            spawn. AwaitGroupEliminationHandler(SpawnGroup)

    GetEliminatedSpawnPoints<public>()<transacts>:[]spawn_group_base=
        CompletedSpawnPoints:=for:
            ThisSpawnPoint:(SpawnPoints+LinkedSpawnGroups)
            ThisSpawnPoint.IsGroupEliminated[]
        do:
            ThisSpawnPoint

    GetSpawnedSpawnPoints<public>()<transacts>:[]spawn_group_base=
        CompletedSpawnPoints:=for:
            ThisSpawnPoint:(SpawnPoints+LinkedSpawnGroups)
            ThisSpawnPoint.IsGroupSpawned[]
        do:
            ThisSpawnPoint

    IsSpawningCompleted<public>()<transacts><decides>:void=
        GetSpawnedSpawnPoints().Length >= (SpawnPoints+LinkedSpawnGroups).Length

    GetMaximumZombies<private>()<transacts>:int=
        Int[MaximumZombiesAtATime*case(GetPlayerCount()):
            1 => 1.0
            2 => 1.3
            3 => 1.75
            4 => 2.0
            _ => 2.2
        ] or 25

    GetSpawnedCount<public>()<transacts>:int=
        var TotalSpawns : int = 0
        for(ThisSpawnPoint:(SpawnPoints+LinkedSpawnGroups)):
            set TotalSpawns += ThisSpawnPoint.GetSpawnedAmount()
        return TotalSpawns

    # Waits until all spawn groups (that count as wave enemies) are eliminated
    AwaitWaveElimination<public>()<suspends>:void=
        WaveSpawnGroups:=for:
            ThisSpawnPoint:SpawnPoints
            ThisSpawnPoint.ZombieType.GetController[].Controls.CountAsWaveEnemy?
        do:
            ThisSpawnPoint
        for(ThisSpawnGroup:WaveSpawnGroups):
            spawn. AwaitGroupEliminationHandler(ThisSpawnGroup)
        var GroupsEliminated : int = 0
        loop:
            ZombieGroupEliminatedEvent.Await()
            set GroupsEliminated += 1
            if(GroupsEliminated >= WaveSpawnGroups.Length+LinkedSpawnGroups.Length):
                # All wave groups have been eliminated
                break

    # Handles the signaling of when a group has been eliminated 
    AwaitGroupEliminationHandler<private>(SpawnGroup:spawn_group_base)<suspends>:void=
        branch:
            loop:
                SpawnGroup.SpawnedEvent.Await()
                WaveEnemies := GetZombieList(?OnlyWave:=true)
                if(WaveEnemies.Length >= GetMaximumZombies()):
                    # Reached the maximum zombies at a time, pause spawning
                    Pause()
        branch:
            loop:
                EliminatedAgent:=SpawnGroup.ZombieEliminatedEvent.Await()
                WaveZombieEliminatedEvent.Signal(EliminatedAgent)
                WaveZombieList:=GetZombieList(?OnlyWave:=true)
                ZombiesLeft:=WaveZombieList.Length
                if(IsSpawningCompleted[]):
                    if(ZombiesLeft <= 3, not SkullMarkedState?):
                        set SkullMarkedState = true
                        RemainingZombies:=GetZombieList(?OnlyWave:=true)
                        for(ThisZombie:RemainingZombies):
                            if(ThisZombie.IsWaveEnemy?):
                                ThisZombie.SetMarker(zombie_marker.Skull)
                    else if(ZombiesLeft = 1, GetWaveCount() > GlobalWaveController[GetSession()].Controls.AggressiveWaveStart, LastZombie:=WaveZombieList[0]):
                        # If the aggressive waves have already started, and only one zombie left, set it to be aggro
                        LastZombie.SetAggroLevel(100.0)
                if(ZombiesLeft < GetMaximumZombies()):
                    # There is enough room for more zombies to spawn, so resume spawning
                    Resume()
        SpawnGroup.GroupEliminatedEvent.Await()
        ZombieGroupEliminatedEvent.Signal()

special_wave_info<public> := class(wave_info):
    
    @editable
    OnWaveEndDropPowerup : powerup_type = powerup_type.None

    @editable
    MaximumSpawnCap : type{_X:int where _X >= 0} = 0

special_wave_controls<public> := class<concrete>:

    @editable
    SpecialWaveInfoList<public> : []special_wave_info = array{
        special_wave_info:
            SpawnsAtOnce:=2
            SpawnPoints:=array:
                spawn_group_wave:
                    Amount:=0.25
                spawn_group_wave:
                    Amount:=0.75
                    #ZombieType:=zombie_type_list.SpecialZombieType
    }

    @editable 
    FirstSpecialWaveAppearance : type{_X:int where _X >= 0} = 0

    @editable
    RepeatEveryXWaves : type{_X:int where _X >= 0} = 0

waveloopbehavior<public> := enum{
    LoopAndContinue
    ContinueWithLastEnemyTypes
    DoNotLoopOrContinue
}

wave_controls<public> := class<concrete>:

    @editable ActivateOnGameStartDelay : type{_X : float where _X >= -1.0} = -1.0
    @editable var<public> StartingWave<public> : type{_X:int where _X > 0} = 1
    @editable IntermissionTime : type{_X : float where _X >= 0.1} = 15.0
    @editable RoundStartDelay : type{_X : float where _X >= 0.1} = 2.0
    #@editable AfterLastWaveBehavior : waveloopbehavior = waveloopbehavior.ContinueWithLastEnemyTypes

    @editable var AggressiveWaveStart<public> : int = 3
    @editable var AggressiveWaveMaximum<public> : int = 12
    @editable var QuickWaveAggressionMultiplier<public> : float = 2.2

    @editable var MaximumZombieSpeed : float = 1.65 # -> Goes between 0.5 and 2.0. Any other values outside have no different effects.
    @editable var MaximumWaveZombieCapPerPlayer : int = 75 # -> Multiplied by the number of players each wave

    @editable UpdateZombieStatsOnWaveSkip : logic = true

    ApplySettings(Settings:wave_controls):void=
        set AggressiveWaveStart = Settings.AggressiveWaveStart
        set AggressiveWaveMaximum = Settings.AggressiveWaveMaximum
        set MaximumZombieSpeed = Settings.MaximumZombieSpeed

wave_audio := class<concrete>:

    @editable RoundStartStinger : audio_player_device = audio_player_device{}
    @editable RoundEndStinger : audio_player_device = audio_player_device{}
    @editable SpecialRoundStartStinger : audio_player_device = audio_player_device{}
    @editable SpecialRoundEndStinger : audio_player_device = audio_player_device{}

wave_signal_in := class<concrete>:

    @editable StartWaves_OnTrigger : trigger_device = trigger_device{}
    @editable StopWaves_OnTrigger : trigger_device = trigger_device{}
    @editable RestartWaves_OnTrigger : trigger_device = trigger_device{}
    @editable SkipWave_OnTrigger : trigger_device = trigger_device{}
    @editable SkipWaveAndIntermission_OnTrigger : trigger_device = trigger_device{}
    @editable SkipToWaveTriggers : []skip_wave_trigger = array{}

wave_signal_out := class<concrete>:

    @editable OnWaveStarted_ActivateTrigger : trigger_device = trigger_device{}
    @editable OnWaveEnded_ActivateTrigger : trigger_device = trigger_device{}
    @editable OnSpecialWaveStarted_ActivateTrigger : trigger_device = trigger_device{}
    @editable OnSpecialWaveEnded_ActivateTrigger : trigger_device = trigger_device{}

    @editable OnWaveStartedTriggers : []on_wave_trigger = array{}

    ActivateWaveStartedTrigger<public>(InWave:int):void=
        for(ThisTrigger:OnWaveStartedTriggers):
            if(ThisTrigger.DesiredWave = InWave):
                ThisTrigger.ActivateTriggerOnWaveStart.Trigger()
                return

skip_wave_trigger := class<concrete>:

    @editable DesiredWave : type{_X : int where 1 <= _X} = 1
    @editable UseIntermission : logic = false
    @editable SkipToWave_OnTrigger : trigger_device = trigger_device{}

    Init<public>():void=
        SkipToWave_OnTrigger.TriggeredEvent.Subscribe(SkipToWaveTrig)

    SkipToWaveTrig(Agent:?agent):void=
        if(ThisWaveDevice:=GlobalWaveController[GetSession()]):
            ThisWaveDevice.SkipToWaveInternal(DesiredWave, ?UseIntermission:=UseIntermission)

on_wave_trigger := class<concrete>:

    @editable DesiredWave : type{_X : int where 1 <= _X} = 1
    @editable ActivateTriggerOnWaveStart : trigger_device = trigger_device{}

spawn_amount_category<public> := enum{
    PercentOfWaveZombieCap
    DirectAmount
}

spawn_group_wave<public> := class(spawn_group_base):

    @editable
    Amount : type{_X : float where _X >= 0.0} = 1.0

    @editable
    AmountCategory : spawn_amount_category = spawn_amount_category.PercentOfWaveZombieCap

    @editable
    SpawnsAtATime : type{_X:int where _X >= 0, _X < 20} = 0

    @editable
    WaitForPreviousGroupElimination<public> : logic = false

    Spawn<override>(?AmountOverride:?int=false)<suspends>:void=
        if. SpawnsAtATime > 0 then. set SpawnsAtOnce = SpawnsAtATime
        if(ThisWaveController := GlobalWaveController[GetSession()]):
            ThisWaveZombieCap := ThisWaveController.GetWaveZombieCap()
            #Print("WaveCap: {ThisWaveZombieCap}, Amount: {Amount}")
            AmountToSpawn := case(AmountCategory):
                spawn_amount_category.PercentOfWaveZombieCap => Int[Amount*ThisWaveZombieCap] or 0
                spawn_amount_category.DirectAmount => Int[Amount] or 0
            if(AmountToSpawn > 0):
                Print("WaveSpawnGroup: Spawning {AmountToSpawn} zombies")
                HandleSpawning(AmountToSpawn)
            else:
                Print("ERROR: Cannot spawn 0 zombies in a spawn group!")