using { /Fortnite.com/Devices }
using { /Verse.org/Simulation }
using { /UnrealEngine.com/Temporary/SpatialMath }
using { /Fortnite.com/Devices/CreativeAnimation }
using { /Fortnite.com/Devices/CreativeAnimation/InterpolationTypes }
using { InteractiveModule }
using { Room }
using { Player }

door<public> := class(Powerable):

    @editable DefaultState : logic = false
    @editable OpenOnPowered : logic = false
    @editable Price : type{_X : int where _X >= 0} = 500
    @editable BuyPrompt : string = "open door"
    @editable InteractiveLocations : []Interactable = array{Interactable{Identity:="Door Buyable", DefaultPrompt:="open door", RequiresLookAt:=false, InteractTime:=0.75, InteractDistance:=250.0, InteractionRadius:=100.0, RequiresPlayerMustBeInFront:=false}}
    @editable DoorObjects : []door_object = array{}
    @editable DoorBarrier : creative_prop = creative_prop{}
    @editable OpenOnTrigger : trigger_device = trigger_device{}
    @editable ActivateTriggerOnPurchase : trigger_device = trigger_device{}
    @editable ActivateCinematicOnPurchase : cinematic_sequence_device = cinematic_sequence_device{}
    @editable PowerSettings<override> : powerable_settings = powerable_settings{UsePower:=false}

    var<private> DoorBarrierOrigin : transform = transform{}

    var<private> BoughtState : logic = false

    var<private> LinkedRooms : []room = array{}
    LinkRoom(InRoom:room):void=
        if(not LinkedRooms.Find[InRoom]):
            set LinkedRooms += array{InRoom}

    OnPowerableBegin<override>()<suspends>:void=
        for(ThisDoorObject:DoorObjects):
            ThisDoorObject.Init()
        if(PowerSettings.UsePower?):
            if(not CheckPowerState[?Channel:=PowerSettings.Channel]):
                SetInteractableState(false)
        if(DoorBarrier.IsValid[]):
            set DoorBarrierOrigin = DoorBarrier.GetTransform()
        OpenOnTrigger.TriggeredEvent.Subscribe(OpenViaTrigger)
        LinkAndCompile(InteractiveLocations)
        if(DefaultState?):
            Open()

    OnPowered<override>():void=
        if(PowerSettings.UsePower?):
            SetEnabledState(true)
            SetInteractableState(true)
            if(OpenOnPowered?):
                spawn{Open()}

    GetPowerInteractionText<override>(InAgent:agent, Interactive:Interactable):string=
        if(not BoughtState?):
            if(Price > 0):
                if(not PowerSettings.UsePower? or IsPowered[]):
                    SetInteractableState(true)
                    return "{BuyPrompt} [<C_Y>${Price}<E>]"
                else:
                    SetInteractableState(false)
                    return "Power Required!"
            else if(PowerSettings.UsePower?):
                if(IsPowered[]):
                    SetInteractableState(true)
                    return "{BuyPrompt}"
                else:
                    SetInteractableState(false)
                    return "Power Required!"
            else:
                SetInteractableState(false)
                return "{BuyPrompt}"
        else:
            SetEnabledState(false)
            return "ERR"

    OnInteracted<override>(InAgent:agent, InInteractingLocation:Interactable):void=
        if(not BoughtState?, CheckPowerState[?Channel:=PowerSettings.Channel] or not PowerSettings.UsePower?):
            if(ThisPlayerInfo:=InAgent.GetPlayerInfo[]):
                BuyState:=ThisPlayerInfo.BuyItem(Price)
                if(BuyState?):
                    ActivateTriggerOnPurchase.Trigger(InAgent)
                    spawn{Open()}

    OpenViaTrigger<private>(InAgent:?agent):void=
        spawn{Open()}

    Open<public>()<suspends>:void=
        SetEnabledState(false)
        if(not BoughtState?):
            set BoughtState = true
            ActivateCinematicOnPurchase.Play()
            for(ThisDoorObject:DoorObjects):
                if(ThisDoorObject.AwaitAnimationFinish?):
                    ThisDoorObject.PlayAnimation()
                else:
                    spawn{ThisDoorObject.PlayAnimation()}
            if(DoorBarrier.TeleportTo[vector3{X:=0.0,Y:=0.0,Z:=-100.0}, rotation{}]){}
            for(ThisRoom:LinkedRooms):
                ThisRoom.OnRoomDoorOpened()

    Reset<override>():void=
        for(ThisDoorObject:DoorObjects):
            ThisDoorObject.Reset()
        SetEnabledState(true)
        if(DoorBarrier.TeleportTo[DoorBarrierOrigin]){}
        set BoughtState = false

    SetBarrierState<public>(State:logic):void=
        if(State?):
            if. DoorBarrier.TeleportTo[DoorBarrierOrigin]
        else if. DoorBarrier.TeleportTo[vector3{}, rotation{}]

    GetDoorTransform<public>()<transacts>:transform=
        if(DoorBarrier.IsValid[]):
            return DoorBarrier.GetTransform()
        else:
            Print("ERROR: No Door Barrier set up for a door!")
            return transform{}

door_object := class<concrete>:

    @editable Prop : creative_prop = creative_prop{}
    @editable LocalMovementDestination : vector3 = vector3{Z:=250.0}
    @editable MovementTime : float = 0.5
    @editable DelayTimeMultiplier : type{_X : float where 0.0 <= _X} = 0.5
    @editable DoMovementBob : logic = true
    @editable DoPopShrink : logic = true
    @editable DoRandomRotate : logic = true
    @editable AwaitAnimationFinish : logic = true

    var OriginalTransform : transform = transform{}

    Init():void=
        if(Prop.IsValid[]):
            set OriginalTransform = Prop.GetTransform()

    PlayAnimation()<suspends>:void=
        var TimeLeft : float = MovementTime
        Destination:=Prop.GetTransform().GetLocalVector(LocalMovementDestination)
        var EndRotation : rotation = Prop.GetTransform().Rotation
        if(DoRandomRotate?):
            set EndRotation = EndRotation.ApplyRandomYawRotation()
        if(PropAnimator:=Prop.GetAnimationController[]):
            var AnimPoints : []keyframe_delta = array{}
            if(DoMovementBob?):
                set AnimPoints += array{
                    keyframe_delta{
                        DeltaLocation := vector3{Z:=-50.0}
                        DeltaRotation := rotation{}
                        Time := MovementTime*0.175
                        Interpolation := EaseOut
                    }
                    keyframe_delta{
                        DeltaLocation := vector3{Z:=+50.0}
                        DeltaRotation := rotation{}
                        Time := MovementTime*0.175
                        Interpolation := EaseIn
                    }
                }
                set TimeLeft *= (1.0-0.35)
            if(DoPopShrink?):
                PointToShrinkStart := LocalMovementDestination*0.75
                set AnimPoints += array{
                    keyframe_delta{
                        DeltaLocation := vector3{Z:=PointToShrinkStart.Z}
                        DeltaRotation := rotation{}
                        #DeltaScale := vector3{X:=1.5,Y:=1.5,Z:=1.5}
                        Time := TimeLeft*0.4
                        Interpolation := EaseOut
                    }
                    keyframe_delta{
                        DeltaLocation := vector3{Z:=LocalMovementDestination.Z*0.125}
                        DeltaRotation := rotation{}
                        DeltaScale := vector3{X:=1.5,Y:=1.5,Z:=1.5}
                        Time := TimeLeft*0.2
                        Interpolation := EaseIn
                    }
                    keyframe_delta{
                        DeltaLocation := vector3{Z:=LocalMovementDestination.Z*0.125}
                        DeltaRotation := rotation{}
                        DeltaScale := vector3{X:=0.1, Y:=0.1, Z:=0.1}
                        Time := TimeLeft*0.2
                        Interpolation := Linear
                    }
                }
                set TimeLeft *= (1.0-0.2)
            else:
                set AnimPoints += array{
                    keyframe_delta{
                        DeltaLocation := LocalMovementDestination
                        DeltaRotation := rotation{}
                        Time := TimeLeft
                        Interpolation := EaseIn
                    }
                }
            PropAnimator.SetAnimation(AnimPoints, ?Mode:=animation_mode.OneShot)
            PropAnimator.Play()
            Sleep(MovementTime*DelayTimeMultiplier)
        else{Print("~ERR: Failed to get Animator!")}
        if(Prop.TeleportTo[vector3{X:=0.0,Y:=0.0,Z:=0.0}, rotation{}]){}
        Prop.Hide()

    Reset():void=
        if(Prop.TeleportTo[OriginalTransform]){}
        Prop.Show()

