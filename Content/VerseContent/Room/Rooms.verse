using { /Verse.org/Simulation }
using { /Fortnite.com/Devices }
using { /UnrealEngine.com/Temporary/SpatialMath }
using { Player }
using { UI }
using { Zombie }
using { Spawning }

GetTrialRooms<public>()<transacts>:[]room=
    TrialRooms:=for:
        ThisRoom:GetRooms()
        ThisRoom.AllowTrials?
    do:
        ThisRoom

(InAgent:agent).GetOccupyingRoom<public>()<transacts><decides>:room=
    RoomList := GetRooms()
    var MaybeRoom : ?room = false
    for(ThisRoom:RoomList):
        if(ThisRoom.IsInRoom[InAgent]):
            set MaybeRoom = option. ThisRoom
    MaybeRoom?

GetRoomViaSpawner<public>(InSpawner:spawn_location)<transacts><decides>:room=
    var MaybeLinkedRoom : ?room = false
    if(ThisRoomCollection:=GlobalRoomCollection[GetSession()]):
        for(ThisRoom:ThisRoomCollection):
            if(ThisRoom.ZombieSpawners.Find[InSpawner]):
                set MaybeLinkedRoom = option{ThisRoom}
    MaybeLinkedRoom?

GetRoomByName(InName:string):?room=
    RoomList:=GetRooms()
    for(ThisRoom:RoomList):
        if(ThisRoom.RoomText=InName):
            return option{ThisRoom}
    false

GetRooms<public>()<transacts>:[]room=
    return GlobalRoomCollection[GetSession()] or array{}

# ROOM LIST
room_list<public> := enum{
    Spawn
}

var GlobalRoomCollection<public> : weak_map(session, []room) = map{}

room<public> := class<unique>(creative_device):

    @editable
    RoomText<public> : string = "Unset Room Name"

    @editable
    Volumes<private> : []volume_device = array{}

    @editable
    Borders<public> : []creative_prop = array{}

    @editable
    ZombieSpawners<public> : []spawn_location = array{}

    @editable
    AllowTrials<public> : logic = false

    @editable
    TrialBeacon_Away<public> : beacon_device = beacon_device{}

    @editable
    TrialBeacon_Near<public> : beacon_device = beacon_device{}
    
    var RoomVolumes<private> : []room_volume = array{}
    var<private> LinkedRooms<public> : []room_linker = array{}

    PlayerEnteredEvent:event(agent)=event(agent){}
    PlayerExitedEvent:event(agent)=event(agent){}
    NoPlayersInRoomEvent:event()=event(){}

    OnBegin<override>():void=
        # Linking
        if:
            not GlobalRoomCollection[GetSession()]
            set GlobalRoomCollection[GetSession()] = array{}
        if:
            CurrentRoomCollection := GlobalRoomCollection[GetSession()]
            set GlobalRoomCollection[GetSession()] = CurrentRoomCollection + array{Self}
        # Create and initialize room volumes
        set RoomVolumes = for:
            ThisVolume : Volumes
        do:
            NewRoomVolume:room_volume=room_volume{VolumeDevice:=ThisVolume, ParentRoom:=Self}
            NewRoomVolume.Init()
            NewRoomVolume
        SetBorderState(false)

    LinkRoom(InLinker:room_linker):void=
        set LinkedRooms += array{InLinker}

    AgentEnteredRoom(InAgent : agent):void=
        PlayerEnteredEvent.Signal(InAgent)
        InAgent.GetUIManager().SetLocationText(RoomText)

    AgentExitedRoom(InAgent : agent):void=
        PlayerExitedEvent.Signal(InAgent)
        PlayersInRoom := GetPlayersInRoom()
        if(PlayersInRoom.Length <= 0):
            NoPlayersInRoomEvent.Signal()
        if(ThisRoom:=InAgent.GetOccupyingRoom[]):
            InAgent.GetUIManager().SetLocationText(ThisRoom.RoomText)

    IsInRoom<public>(InAgent:agent)<transacts><decides>:void=
        var IsInThisRoom : logic = false
        for(ThisZone:RoomVolumes):
            if(ThisZone.IsInVolume[InAgent]):
                set IsInThisRoom = true
        IsInThisRoom?

    GetPlayersInRoom<public>()<transacts>:[]agent=
        return GetPlayersInRoomState(true)

    GetOutsidePlayers<public>():[]agent=
        return GetPlayersInRoomState(false)

    GetPlayersInRoomState<private>(InState:logic)<transacts>:[]agent=
        var RoomStatePlayers : []agent = array{}
        PlayerList:=GetAgents()
        for(ThisPlayer:PlayerList):
            if(not InState? or IsInRoom[ThisPlayer]):
                set RoomStatePlayers += array{ThisPlayer}
        return RoomStatePlayers

    # Fires when a door connected to this room is opened
    # Used to enable Zombie Spawners in the connecting rooms
    OnRoomDoorOpened<public>():void=
        SetZombieSpawnerStates(true)
        for(RoomLinker:LinkedRooms):
            ThisConnectedState:=RoomLinker.CheckConnectionState()
            if(ThisConnectedState?):
                OtherLinkedRoom := RoomLinker.GetLinkedRoom(Self)
                OtherLinkedRoom.SetZombieSpawnerStates(true)
                for(ThisLinker:OtherLinkedRoom.LinkedRooms):
                    if(ThisLinker.ConnectedByDefault?):
                        ThisLinker.GetLinkedRoom(OtherLinkedRoom).SetZombieSpawnerStates(true)

    ForceOpenDoors<public>():void=
        for(ThisRoomLinker:LinkedRooms):
            for(ThisDoor:ThisRoomLinker.ConnectingDoors):
                spawn. ThisDoor.Open()

            
    # Sets the enabled states of all zombie spawners linked to this room
    SetZombieSpawnerStates<public>(InState:logic):void=
        for(ThisZombieSpawner:ZombieSpawners):
            if(InState?):
                ThisZombieSpawner.Enable()
            else:
                ThisZombieSpawner.Disable()

    SetBorderState<public>(InState:logic):void=
        for(ThisBorder:Borders):
            if. InState? then. ThisBorder.Show()
            else:
                ThisBorder.Hide()
                TrialBeacon_Away.Disable()
                TrialBeacon_Near.Disable()

    TraverseToRoom<private>(TargetRoom:room):[]room=
        var RoomsToTarget : []room = array{}
        loop:
            for(ThisLinker:LinkedRooms):
                if(ThisLinker.CheckConnectionState()?):
                    LinkedRoom:=ThisLinker.GetLinkedRoom(Self)
                    if(LinkedRoom = TargetRoom):
                        # Found target room
                        set RoomsToTarget += array{LinkedRoom}
                        break
            break
        return RoomsToTarget

spawner_share_type<public> := enum{
    Allow # -> Always share spawners when finding desired ones
    AllowIfNeeded # -> If there are no open spawners in the current room, get some from adjacent rooms
    Disallow # -> No spawner sharing
}

var RoomLinkerCollection : weak_map(session, []room_linker) = map{}

room_linker<public> := class<concrete>(creative_device):

    @editable RoomA : room = room{}
    @editable RoomA_ShareSpawnersPolicy : spawner_share_type = spawner_share_type.AllowIfNeeded

    @editable RoomB : room = room{}
    @editable RoomB_ShareSpawnersPolicy : spawner_share_type = spawner_share_type.AllowIfNeeded

    @editable ConnectedByDefault : logic = false
    @editable AllDoorsRequired : logic = false
    @editable ConnectingDoors : []door = array{}
    @editable ConnectingLocation<public> : ?creative_prop = false
    @editable EnableConnectionOnTrigger : trigger_device = trigger_device{}
    
    var ConnectedState : logic = false

    OnBegin<override>():void=
        RoomA.LinkRoom(Self)
        RoomB.LinkRoom(Self)
        if:
            not RoomLinkerCollection[GetSession()]
            set RoomLinkerCollection[GetSession()] = array{}
        if:
            CurrentRoomLinkers := RoomLinkerCollection[GetSession()]
            set RoomLinkerCollection[GetSession()] = CurrentRoomLinkers + array{Self}
        for(ThisConnectingDoor:ConnectingDoors):
            ThisConnectingDoor.LinkRoom(RoomA)
            ThisConnectingDoor.LinkRoom(RoomB)

    OnTriggerEnableConnection(Agent:?agent):void=
        set ConnectedState = true

    CheckConnectionState()<transacts>:logic=
        if(ConnectedState? or ConnectedByDefault?):
            return true
        if(AllDoorsRequired?):
            DoorsOpened:=for:
                ThisDoor:ConnectingDoors
                ThisDoor.BoughtState?
            do:
                0
            if(DoorsOpened.Length >= ConnectingDoors.Length):
                set ConnectedState = true
                return true
        else:
            for(ThisDoor:ConnectingDoors):
                if(ThisDoor.BoughtState?):
                    set ConnectedState = true
                    return true
        return false

    CheckSpawnerSharing<public>(InRoom:room, ?IfNeededState:logic=false)<transacts><decides>:void=
        CheckConnectionState()
        if(InRoom = RoomA):
            RoomA_ShareSpawnersPolicy <> spawner_share_type.Disallow
            RoomA_ShareSpawnersPolicy = spawner_share_type.Allow or IfNeededState?
        else if(InRoom = RoomB):
            RoomB_ShareSpawnersPolicy <> spawner_share_type.Disallow
            RoomB_ShareSpawnersPolicy = spawner_share_type.Allow or IfNeededState?
        else:
            0 > 1

    GetLinkedRoom<public>(InRoom:room)<transacts>:room=
        if(InRoom = RoomA):
            return RoomB
        else:
            return RoomA

    GetLinkedLocation<public>()<transacts>:vector3=
        return ConnectingLocation?.GetTransform().Translation or ConnectingDoors[0].DoorBarrier.GetTransform().Translation or vector3{}

(RoomLinkers:[]room_linker).GetLinkerWithRooms<public>(RoomA:room, RoomB:room)<transacts>:?room_linker=
    for(ThisRoomLinker:RoomLinkers):
        if:
            ThisRoomLinker.RoomA = RoomA or ThisRoomLinker.RoomB = RoomA
            ThisRoomLinker.RoomB = RoomB or ThisRoomLinker.RoomA = RoomB
        then:
            return option{ThisRoomLinker}
    return false

# A volume suited for rooms
room_volume := class(better_volume_device):

    var ParentRoom : room = room{}

    OnAgentEnteringVolume<override>(InAgent:agent):void=
        ParentRoom.AgentEnteredRoom(InAgent)

    OnAgentExitingVolume<override>(InAgent:agent):void=
        if(not ParentRoom.IsInRoom[InAgent]):
            ParentRoom.AgentExitedRoom(InAgent)