using { /Fortnite.com/Devices }
using { /Fortnite.com/Characters }
using { /Fortnite.com/Game }
using { /Fortnite.com/UI }
using { /Fortnite.com/FortPlayerUtilities }
using { /Verse.org/Simulation }
using { /Verse.org/Random }
using { /Verse.org/Colors }
using { /Verse.org/Simulation/Tags }
using { /UnrealEngine.com/Temporary/Diagnostics }
using { /UnrealEngine.com/Temporary/SpatialMath }
using { /UnrealEngine.com/Temporary/UI }
using { Player }
using { Zombie }
using { UI }
using { Spawning }
using { Powerups }
using { Weapons }

# Spawn a given powerup on a zombie
# Handles situations such as in barriers or shortcuts and will bring the powerup to a valid location
SpawnPowerupOnZombie<public>(InZombie:zombie_ai, InPowerup:powerup_base, ?ByChance:logic=false):?powerup_instance=
    if(not ByChance? or GlobalPowerupManager[GetSession()].RunPowerupRNG[]):
        ThisSpawnLocation := if. InZombie.IsActive[] then. InZombie.FortCharacter.GetTransform() else. InZombie.DeathTransform
        PowerupInstance:=InPowerup.Spawn(ThisSpawnLocation.Translation)
        if(ThisSpawnLoc:=InZombie.IsSpawning[]):
            EndLocationTuple := ThisSpawnLoc.GetEndLocation()
            EndLocationOffset := if. EndLocationTuple(1) = 0 then. vector3{X:=75.0} else. vector3{}
            spawn. PowerupInstance.MoveTo(ThisSpawnLoc.GetEndLocation()(0).GetLocalVector(EndLocationOffset))
        return option. PowerupInstance
    return false

# Spawn a random powerup given the location, and type filters
SpawnRandomPowerup<public>(AtPosition:vector3, ?IncludeRegulars:logic=true, ?IncludeSpecials:logic=false, ?ByChance:logic=false):?powerup_instance=
    if(not ByChance? or GlobalPowerupManager[GetSession()].RunPowerupRNG[]):
        if(ThisRandomPowerup:=GetRandomPowerup[?IncludeRegulars:=IncludeRegulars, ?IncludeSpecials:=IncludeSpecials]):
            ThisInstance:=ThisRandomPowerup.Spawn(AtPosition)
            return option{ThisInstance}
        else{Print("~ERROR: Failed to fetch and spawn a random powerup!")}
    false

# Fetches a random powerup based on the given filters
GetRandomPowerup<public>(?IncludeRegulars:logic=true, ?IncludeSpecials:logic=false)<transacts><decides>:powerup_base=
    PowerupMenu := GlobalPowerupManager[GetSession()].PowerupList
    var PowerupList : []powerup_base = array{}
    if(IncludeRegulars?):
        set PowerupList += PowerupMenu.GetPowerups()
    if(IncludeSpecials?):
        set PowerupList += PowerupMenu.GetSpecialPowerups()
    return Shuffle(PowerupList)[0]

GetPowerupList<public>()<transacts>:[]powerup_base=
    return GlobalPowerupManager[GetSession()].PowerupList.GetAllPowerups() or array{}

var GlobalPowerupManager<public> : weak_map(session, PowerupManager) = map{}

PowerupManager<public> := class(creative_device):

    @editable PowerupList<public> : powerupmanager_powerups = powerupmanager_powerups{}

    @editable SpawnControls : powerupmanager_spawncontrols = powerupmanager_spawncontrols{}

    @editable Default_PickupEffect_BP : creative_prop_asset = DefaultCreativePropAsset
    @editable Default_PickupEffect_Lifetime : float = 0.0

    var CurrentMaxPowerupsPerRound : int = -1
    var CurrentPowerupSpawnChance : float = -1.0

    var RoundPowerupsSpawned : int = 0
    var AllowChanceSpawning : logic = true

    OnBegin<override>():void=
        if(set GlobalPowerupManager[GetSession()] = Self){}
        else{Print("ERROR: Failed to globally link powerup manager!")}
        InitPowerups()
        set CurrentMaxPowerupsPerRound = SpawnControls.Inital_MaxPowerupsPerRound
        set CurrentPowerupSpawnChance = SpawnControls.Initial_PowerupSpawnChance
        spawn{IncrementValuesOverTime()}

    InitPowerups():void=
        ThisPowerupList:=PowerupList.GetAllPowerups()
        for(ThisPowerup : ThisPowerupList):
            ThisPowerup.BasicSettings.OnTrigger_SpawnPowerup.TriggeredEvent.Subscribe(ThisPowerup.SpawnViaTrigger)
            ThisPowerup.BasicSettings.OnTrigger_GrantPowerup.TriggeredEvent.Subscribe(ThisPowerup.SpawnAndActivate)

    GetAllPowerups()<transacts>:[]powerup_base=
        return PowerupList.GetPowerups()+PowerupList.GetSpecialPowerups()

    GetPowerupsOfRarity<public>(InRarity:item_rarity)<transacts>:[]powerup_base=
        var PowerupsListOfRarity : []powerup_base = array{}
        ThisPowerupList:=PowerupList.GetPowerups()
        for(ThisPowerup:ThisPowerupList):
            if(ThisPowerup.BasicSettings.TrialRewardRarity = InRarity):
                set PowerupsListOfRarity += array{ThisPowerup}
        return PowerupsListOfRarity

    # Runs a check on the spawning rng to see if a powerup should spawn
    RunPowerupRNG<public>()<transacts><decides>:void=
        AllowChanceSpawning?, RoundPowerupsSpawned < CurrentMaxPowerupsPerRound
        GetRandomFloat(0.0, 100.0) <= CurrentPowerupSpawnChance
        set RoundPowerupsSpawned += 1

    AllowChanceSpawningDelay<private>()<suspends>:void=
        set AllowChanceSpawning = false
        Sleep(3.0)
        set AllowChanceSpawning = true

    SpawnPowerup_OnChance<public><final>(InPowerup:powerup_base, AtPosition:vector3):?powerup_instance=
        if(RoundPowerupsSpawned < CurrentMaxPowerupsPerRound):
            if(GetRandomFloat(0.0, 100.0) <= CurrentPowerupSpawnChance):
                set RoundPowerupsSpawned += 1
                ThisInstance:=InPowerup.Spawn(AtPosition)
                return option{ThisInstance}
        false

    var CurrXVal<private> : int = 0
    var CurrYVal<private> : int = 0
    IncrementValuesOverTime<private>()<suspends>:void=
        if(ThisWaveDevice:=GlobalWaveController[GetSession()]):
            loop:
                ThisWaveDevice.WaveEndedEvent.Await()
                set RoundPowerupsSpawned = 0
                set CurrXVal += 1
                set CurrYVal += 1
                if(CurrXVal >= SpawnControls.X_Value):
                    set CurrXVal = 0
                    set CurrentMaxPowerupsPerRound += SpawnControls.IncrementalPerXRounds_MaxPowerupsPerRound
                if(CurrYVal >= SpawnControls.Y_Value):
                    set CurrYVal = 0
                    set CurrentPowerupSpawnChance += SpawnControls.IncrementalPerYRounds_PowerupSpawnChance
        else:
            Sleep(0.5)
            spawn{IncrementValuesOverTime()}

    Reset():void=
        set CurrentMaxPowerupsPerRound = SpawnControls.Inital_MaxPowerupsPerRound
        set CurrentPowerupSpawnChance = SpawnControls.Initial_PowerupSpawnChance
        set RoundPowerupsSpawned = 0
        set CurrXVal = 0
        set CurrYVal = 0
        ThisPowerupList:=PowerupList.GetAllPowerups()
        for(ThisPowerup:ThisPowerupList):
            ThisPowerup.DespawnAll()

(PowerupEnum:powerup_type).ConvertToPowerup<public>()<transacts><decides>:powerup_base=
    PowerupEnum <> powerup_type.None
    var MaybePowerup : ?powerup_base = false
    case(PowerupEnum):
        powerup_type.RandomRegular =>
            set MaybePowerup = option{GetRandomPowerup[?IncludeSpecials:=false]}
        powerup_type.RandomSpecial =>
            set MaybePowerup = option{GetRandomPowerup[?IncludeRegulars:=false, ?IncludeSpecials:=true]}
        powerup_type.RandomAll =>
            set MaybePowerup = option{GetRandomPowerup[?IncludeSpecials:=true]}
        powerup_type.BonusPoints =>
            set MaybePowerup = option{GlobalPowerupManager[GetSession()].PowerupList.BonusPoints}
        powerup_type.DeathMachine =>
            set MaybePowerup = option{GlobalPowerupManager[GetSession()].PowerupList.DeathMachine}
        powerup_type.DoublePoints =>
            set MaybePowerup = option{GlobalPowerupManager[GetSession()].PowerupList.DoublePoints}
        powerup_type.Instakill =>
            set MaybePowerup = option{GlobalPowerupManager[GetSession()].PowerupList.Instakill}
        powerup_type.Nuke =>
            set MaybePowerup = option{GlobalPowerupManager[GetSession()].PowerupList.Nuke}
        powerup_type.Firesale =>
            set MaybePowerup = option{GlobalPowerupManager[GetSession()].PowerupList.Firesale}
        powerup_type.MaxAmmo =>
            set MaybePowerup = option{GlobalPowerupManager[GetSession()].PowerupList.MaxAmmo}
        powerup_type.Quickdraw =>
            set MaybePowerup = option{GlobalPowerupManager[GetSession()].PowerupList.Quickdraw}
        powerup_type.Incognito =>
            set MaybePowerup = option{GlobalPowerupManager[GetSession()].PowerupList.Incognito}
        powerup_type.Perk =>
            set MaybePowerup = option{GlobalPowerupManager[GetSession()].PowerupList.Perk}
        _ =>
    MaybePowerup? 

powerup_type<public>:=enum{
    None
    RandomRegular
    RandomSpecial
    RandomAll
    BonusPoints
    DeathMachine
    DoublePoints
    Instakill
    Nuke
    Firesale
    MaxAmmo
    Quickdraw
    Incognito
    Perk
}
powerupmanager_powerups<public> := class<concrete>:

    @editable BonusPoints : powerup_bonuspoints = powerup_bonuspoints{}
    @editable DeathMachine : powerup_deathmachine = powerup_deathmachine{}
    @editable DoublePoints : powerup_doublepoints = powerup_doublepoints{}
    @editable Instakill : powerup_instakill = powerup_instakill{}
    @editable Nuke : powerup_nuke = powerup_nuke{}
    @editable Firesale : powerup_firesale = powerup_firesale{}
    @editable MaxAmmo : powerup_maxammo = powerup_maxammo{}
    @editable Quickdraw : powerup_quickdraw = powerup_quickdraw{}
    @editable Incognito : powerup_incognito = powerup_incognito{}

    @editable Perk : powerup_perk = powerup_perk{}

    GetAllPowerups<public>()<transacts>:[]powerup_base=
        return GetPowerups()+GetSpecialPowerups()
    
    GetPowerups<public>()<transacts>:[]powerup_base=
        PowerupList : []powerup_base = array{
            # Add new powerups here #
            BonusPoints
            DeathMachine
            DoublePoints
            Instakill
            Nuke
            Firesale
            MaxAmmo
            Quickdraw
            Incognito
        }
        
    GetSpecialPowerups<public>()<transacts>:[]powerup_base=
        SpecialPowerups : []powerup_base = array{
            # Add new special powerups here #
            Perk
        }

powerupmanager_spawncontrols := class<concrete>:

    # Initial #
    initial_powerupperround_tooltip<localizes>:message="What should the initial amount of maximum powerups to spawn per round be?"
    @editable{ToolTip:=initial_powerupperround_tooltip} Inital_MaxPowerupsPerRound : type{_X : int where 0 <= _X} = 2
    initial_powerupspawnchance_tooltip<localizes>:message="what should be the intial spawn chance for a powerup?"
    @editable{ToolTip:=initial_powerupspawnchance_tooltip} Initial_PowerupSpawnChance : type{_X : float where 0.0 <= _X} = 2.0

    # Absolute/Cap #
    absolute_maxpowerupperround_tooltip<localizes>:message="What should the maximum powerups to spawn per round be?\nThe amount of powerups spawned is capped at this value."
    @editable{ToolTip:=absolute_maxpowerupperround_tooltip} Absolute_MaxPowerupsPerRound : type{_X : int where 0 <= _X} = 5
    absolute_powerupspawnchance_tooltip<localizes>:message="What should the maximum chance for a powerup to spawn be?\nThe spawn chance for a powerup is capped at this value."
    @editable{ToolTip:=absolute_powerupspawnchance_tooltip} Absolute_PowerupSpawnChance : type{_X : float where 0.0 <= _X} = 5.0

    # Incremental #
    incremental_maxpowerupsperround_tooltip<localizes>:message="Increments the max powerups per round by this value every 'X' rounds."
    @editable{ToolTip:=incremental_maxpowerupsperround_tooltip} IncrementalPerXRounds_MaxPowerupsPerRound : int = 1
    @editable X_Value : type{_X : int where 1 <= _X} = 10

    incremental_powerupspawnchance_tooltip<localizes>:message="Increments the powerup spawn chance by this value every 'Y' rounds."
    @editable{ToolTip:=incremental_powerupspawnchance_tooltip} IncrementalPerYRounds_PowerupSpawnChance : float = 0.1
    @editable Y_Value : type{_X : int where 1 <= _X} = 2

# The basic class that any powerup should inherit from #
# Gives basic functionality for spawning and despawning powerups #
powerup_base<public> := class<concrete>:

    @editable BasicSettings<public> : powerup_base_settings = powerup_base_settings{}

    var CurrentInstances : []powerup_instance = array{}

    var ActiveState<private> : logic = false
    var ActiveStatePerPlayer<private> : [agent]logic = map{}

    # Spawns the powerup via a trigger, and on top of the trigger #
    SpawnViaTrigger(Agent:?agent):void=
        Spawn(BasicSettings.OnTrigger_SpawnPowerup.GetTransform().Translation)

    # Used to spawn this powerup instance at a specific position #
    # The on-ground lifetime can be overridden if needed #
    # Returns a instance class that can be used to tell when this instance was picked up, or to despawn it #
    Spawn<public><final>(AtPosition:vector3, ?Lifetime:float=0.0, ?OwningPlayer:?agent=false):powerup_instance=
        NewInstance : powerup_instance = powerup_instance{ParentPowerup:=Self, SpawnedPosition:=AtPosition, OwningPlayer:=OwningPlayer}
        if(SpawnedPowerupBP := SpawnProp(BasicSettings.Powerup_BP, AtPosition, MakeRotationFromYawPitchRollDegrees(GetRandomFloat(0.0,360.0), 0.0, 0.0))(0)?):
            set NewInstance.PowerupPropInstance = SpawnedPowerupBP
            if(Lifetime = 0.0):
                spawn{PreActivate(NewInstance, BasicSettings.GroundLifetime)}
            else:
                spawn{PreActivate(NewInstance, Lifetime)}
            set CurrentInstances += array{NewInstance}
        else{Print("~ERROR: Failed to spawn a powerup (due to max spawned props reached!)")}
        return NewInstance

    PreActivate<private><final>(PowerupInstance:powerup_instance, Lifetime:float)<suspends>:void=
        var PickedUpAgent : ?agent = false
        race:
            PowerupInstance.TrackPickedUpEvent()
            block:
                PickingUpAgent:=PowerupInstance.PickedUpEvent.Await()
                set PickedUpAgent = PickingUpAgent
            block:
                if(Lifetime > 0.0):
                    SafeSleep(Lifetime)
                else{Sleep(Inf)}
            block:
                if(Lifetime > 0.0):
                    StartTime := GetSimulationElapsedTime()
                    loop:
                        # Wait until time reaches 30% left #
                        if(((GetSimulationElapsedTime() - StartTime)/Lifetime) >= 0.50):
                            break
                        Sleep(0.1)
                    TimeLeft:=Lifetime-(GetSimulationElapsedTime()-StartTime)
                    #Print("~Time left: {TimeLeft}")
                    var NewSleepTime : float = (TimeLeft / 5.0)
                    #var SleepTimeMult : float = 0.25
                    branch:
                        #Print("~New time: {NewSleepTime}")
                        SafeSleep(TimeLeft*0.30)
                        set NewSleepTime /= 1.5
                        #Print("~New time: {NewSleepTime}")
                        SafeSleep(TimeLeft*0.21)
                        set NewSleepTime /= 2.0
                        #Print("~New time: {NewSleepTime}")
                        SafeSleep(TimeLeft*0.14)
                        set NewSleepTime /= 2.2
                        #Print("~New time: {NewSleepTime}")
                        SafeSleep(TimeLeft*0.1)
                        set NewSleepTime /= 2.35
                        #Print("~New time: {NewSleepTime}")
                        #Print("~Reached end of setting sleep time")
                        Sleep(Inf)
                    loop:
                        PowerupInstance.PowerupPropInstance.Hide()
                        SafeSleep(0.18)
                        PowerupInstance.PowerupPropInstance.Show()
                        SafeSleep(NewSleepTime)

                    <#
                    var TimeLeftInt : int = 5
                    if(Lifetime > 10.0):
                        Sleep(10.0)
                    else if(TimeLeft := Int[Lifetime/2.0]){set TimeLeftInt = TimeLeft}
                    var IntervalTime : float = (TimeLeftInt*2.0)/5.0
                    for(IntervalSteps := 1..(TimeLeftInt*TimeLeftInt)):
                        PowerupInstance.PowerupPropInstance.Hide()
                        Sleep(0.1)
                        PowerupInstance.PowerupPropInstance.Show()
                        Sleep(IntervalTime/(IntervalSteps*1.0)) #>
                Sleep(Inf)
        if(PickingUpAgent:=PickedUpAgent?):
            spawn{Activate(PickingUpAgent, PowerupInstance)}
            PowerupInstance.Despawn()
        else:
            PowerupInstance.Despawn()
        if(CurrentInstances.Length <= 0):
            spawn{OnNoInstancesActive()}

    # Despawns all current instances of this powerup type #
    DespawnAll<public><final>():void=
        for(ThisInstance:CurrentInstances):
            ThisInstance.Despawn()
        spawn{OnNoInstancesActive()}
    
    # A function that gets called when this powerup instance is despawned #
    # This function is only ran when the powerup instance despawns, NOT deactivates! #
    OnDespawned<public>():void=
        return

    SpawnAndActivate<public><final>(InAgent:?agent):void=
        Inst:=Spawn(vector3{X:=0.0,Y:=0.0,Z:=0.0}, ?OwningPlayer:=InAgent)
        if(ThisAgent:=InAgent?):
            spawn{Inst.ForcePickup(ThisAgent)}
        else:
            ActivePlayers:=GetAgents()
            if(ThisPlayer:=ActivePlayers[0]):
                spawn{Inst.ForcePickup(ThisPlayer)}
            else{Print("ERR: No valid players to pickup powerup")}

    # When a player picks up this powerup instance, this function is activated to handle things #
    ActivatedEvent<public>:event(agent)=event(agent){}
    Activate<private><final>(PickedUpAgent:agent, PowerupInstance:powerup_instance)<suspends>:void=
        Print("~{BasicSettings.Name} powerup activated!")
        ActivatedEvent.Signal(PickedUpAgent)
        if(not IsActive[] or not IsActive[?InAgent:=option{PickedUpAgent}] and BasicSettings.IncludeOnUI?):
            ActivateUI(option{PickedUpAgent}, PowerupInstance)
        ThisAffectType:=GetAffectType()
        case(ThisAffectType):
            powerup_affecting_type.AllPlayers => 
                PlayerList:=GetPlayers()
                for(ThisPlayer:PlayerList):
                    if(not ActiveStatePerPlayer[ThisPlayer.User]):
                        set ActiveStatePerPlayer = ConcatenateMaps(ActiveStatePerPlayer, map{ThisPlayer.User => true})
                    if(set ActiveStatePerPlayer[ThisPlayer.User] = true){}
            powerup_affecting_type.ActivatingPlayer =>
                if(not ActiveStatePerPlayer[PickedUpAgent]):
                    set ActiveStatePerPlayer = ConcatenateMaps(ActiveStatePerPlayer, map{PickedUpAgent => true})
                if(set ActiveStatePerPlayer[PickedUpAgent] = true){}
            powerup_affecting_type.ExcludeActivatingPlayer =>
                PlayerList:=GetPlayers()
                for(ThisPlayer:PlayerList):
                    if(ThisPlayer.User<>PickedUpAgent):
                        if(not ActiveStatePerPlayer[ThisPlayer.User]):
                            set ActiveStatePerPlayer = ConcatenateMaps(ActiveStatePerPlayer, map{ThisPlayer.User => true})
                        if(set ActiveStatePerPlayer[ThisPlayer.User] = true){}
        set ActiveState = true
        branch:
            var PickupEffectProp : creative_prop = creative_prop{}
            if(BasicSettings.OverridePickupEffect?):
                if(SpawnedPickupEffect:=SpawnProp(BasicSettings.PickupEffect_BP, PowerupInstance.SpawnedPosition, rotation{})(0)?):
                    set PickupEffectProp = SpawnedPickupEffect
                    SafeSleep(BasicSettings.PickupEffect_ActiveTime)
            else:
                if(ThisLifetime:=GlobalPowerupManager[GetSession()].Default_PickupEffect_Lifetime, SpawnedPickupEffect:=SpawnProp(GlobalPowerupManager[GetSession()].Default_PickupEffect_BP, PowerupInstance.SpawnedPosition, rotation{})(0)?):
                    set PickupEffectProp = SpawnedPickupEffect
                    SafeSleep(ThisLifetime)
            PickupEffectProp.Dispose()
        race:
            block:
                OnActivated(PickedUpAgent, PowerupInstance)
                Sleep(Inf)
            loop:
                ActivatingAgent:=ActivatedEvent.Await()
                if(ActivatingAgent=PickedUpAgent or (ThisAffectType <> powerup_affecting_type.ActivatingPlayer)):
                    break
            block:
                SafeSleep(BasicSettings.ActivatedTime)
                Deactivate(option{PickedUpAgent})
            loop:
                DeactivatingAgent:=DeactivatedEvent.Await()
                if(DeactivatingAgent?=PickedUpAgent or (ThisAffectType <> powerup_affecting_type.ActivatingPlayer) or not DeactivatingAgent?):
                    break
                Sleep(0.1)
    
    # The function that runs when this powerup is picked up #
    # MEANT TO BE OVERRIDDEN BY NEW POWERUP TYPES! #
    OnActivated<public>(PickedUpAgent:agent, PowerupInstance:powerup_instance)<suspends>:void=
        Print("~powerup log: A powerup was picked up but doesn't override the \"Activate()\" function!")
        return

    # Deactivates the powerup if it is currently active #
    DeactivatedEvent<public>:event(?agent)=event(?agent){}
    Deactivate<public><final>(OutAgent:?agent):void=
        OnDeactivated(OutAgent)
        set ActiveState = false
        DeactivatedEvent.Signal(OutAgent)
        if(ThisOutAgent:=OutAgent?):
            if(set ActiveStatePerPlayer[ThisOutAgent] = false){}
        else:
            for(ThisPlayer->ThisActiveState:ActiveStatePerPlayer):
                if(set ActiveStatePerPlayer[ThisPlayer] = false){}

    # The funcion that runs when this powerup is deactivated #
    # MEANT TO BE OVERRIDDEN BY NEW POWERUP TYPES! #
    OnDeactivated<public>(OutAgent:?agent):void=
        return

    # Activates when no powerup instances of this powerup type are active on the ground #
    OnNoInstancesActive<public>()<suspends>:void=
        return

    # Shows this powerup on everyones ui #
    ActivateUI<private><final>(InAgent:?agent, PowerupInstance:powerup_instance):void=
        if(ThisOwner:=PowerupInstance.OwningPlayer?):
            ThisOwner.GetUIManager().PowerupUI.AddPowerup(Self)
            BasicSettings.ActivateHUD_OnPickup.Show(ThisOwner)
        else:
            case(BasicSettings.AffectingAgents):
                powerup_affecting_type.AllPlayers =>
                    PlayerList:=GetAgents(?OnlyActives:=false)
                    for(ThisPlayer:PlayerList):
                        BasicSettings.ActivateHUD_OnPickup.Show(ThisPlayer)
                        ThisPlayer.GetUIManager().PowerupUI.AddPowerup(Self)
                powerup_affecting_type.ActivatingPlayer =>
                    if(ThisAgent:=InAgent?):
                        BasicSettings.ActivateHUD_OnPickup.Show(ThisAgent)
                        ThisAgent.GetUIManager().PowerupUI.AddPowerup(Self)
                powerup_affecting_type.ExcludeActivatingPlayer =>
                    PlayerList:=GetAgents(?OnlyActives:=false)
                    for(ThisPlayer:PlayerList):
                        if((not InAgent? or (ThisPlayer <> InAgent?))):
                            BasicSettings.ActivateHUD_OnPickup.Show(ThisPlayer)
                            ThisPlayer.GetUIManager().PowerupUI.AddPowerup(Self)
                            
    # Checks if the powerup is still active #
    IsActive(?InAgent:?agent=false)<transacts><decides>:void=
        if(ThisPlayer:=InAgent?):
            ActiveStatePerPlayer[ThisPlayer]?
        else:
            ActiveState?
    # Checks if this powerup persists for a duration of time #
    IsPersistent<public><final>()<transacts><decides>:void=
        BasicSettings.ActivatedTime > 0.0

    IsSpecial()<transacts><decides>:void=
        BasicSettings.IsSpecial?

    GetAffectType()<transacts>:powerup_affecting_type=
        return BasicSettings.AffectingAgents

    GetName<public><final>():string=
        return BasicSettings.Name

    GetActiveTime<public><final>():float=
        return BasicSettings.ActivatedTime

    GetGroundLifetime<public><final>():float=
        return BasicSettings.GroundLifetime

    GetAffectedPlayers<public><final>()<transacts>:[]agent=
        var AffectingPlayers : []agent = array{}
        for(ThisPlayer->ThisActiveState:ActiveStatePerPlayer):
            if(ThisActiveState?):
                set AffectingPlayers += array{ThisPlayer}
        return AffectingPlayers

powerup_affecting_type<public>:=enum{AllPlayers, ActivatingPlayer, ExcludeActivatingPlayer}
powerup_base_settings<public> := class<concrete>:

    @editable Name<public> : string = "Unset Powerup Name"
    @editable TrialRewardRarity : item_rarity = item_rarity.None
    @editable GroundLifetime<public> : float = 30.0
    @editable ActivatedTime<public> : float = 0.0
    @editable IncludeOnUI<public> : logic = true
    @editable AffectingAgents<public> : powerup_affecting_type = powerup_affecting_type.AllPlayers
    @editable IsSpecial<public> : logic = false
    @editable Powerup_BP<public> : creative_prop_asset = DefaultCreativePropAsset
    @editable OverridePickupEffect<public> : logic = false
    @editable PickupEffect_BP<public> : creative_prop_asset = DefaultCreativePropAsset
    @editable PickupEffect_ActiveTime<public> : float = 0.0
    @editable OnTrigger_SpawnPowerup<public> : trigger_device = trigger_device{}
    @editable OnTrigger_GrantPowerup<public> : trigger_device = trigger_device{}
    @editable ActivateHUD_OnPickup<public> : hud_message_device = hud_message_device{}

powerup_instance<public> := class<unique>:

    var PowerupPropInstance<public> : creative_prop = creative_prop{}
    var<internal> SpawnedPosition : vector3 = vector3{}
    var OwningPlayer : ?agent = false

    var ParentPowerup<internal> : powerup_base = powerup_base{}

    PICKUPRADIUS<private> : float = 90.0
    PickedUpEvent<public>:event(?agent)=event(?agent){}
    var<private> PickedUpState : logic = false
    var<private> DespawnedState : logic = false
    var IsTrackingPickup<private> : logic = false
    TrackPickedUpEvent<final>()<suspends>:void=
        Sleep(0.1)
        set IsTrackingPickup = true
        loop:
            PlayerList:=GetPlayers()
            for(ThisPlayer:PlayerList):
                if(ThisFortChar:=ThisPlayer.GetFortCharacter[], ThisFortChar.IsActive[]):
                    if(not OwningPlayer? or ThisPlayer.User = OwningPlayer?):
                        if(Distance(ThisFortChar.GetTransform().Translation, SpawnedPosition) < PICKUPRADIUS):
                            set PickedUpState = true
                            PickedUpEvent.Signal(option{ThisPlayer.User})
                            Sleep(Inf)                   
            Sleep(0.0)

    ForcePickup(InAgent:agent)<suspends>:void=
        loop:
            if(IsTrackingPickup?):
                break
            Sleep(0.25)
        PickedUpEvent.Signal(option{InAgent})
        
    # Checks if the powerup instance is still valid #
    IsValid<public><final>()<transacts><decides>:void=
        PowerupPropInstance.IsValid[]

    # Checks if the powerup instance is still on the ground #
    IsOnGround<public><final>()<transacts><decides>:void=
        not PickedUpState? and not DespawnedState?

    IsPickedUp<public><final>()<transacts><decides>:void=
        PickedUpState?

    # Disposes this powerup instance #
    Despawn<public><final>():void=
        if(not DespawnedState?):
            PowerupPropInstance.Dispose()
            if(not PickedUpState?):
                ParentPowerup.OnDespawned()
                set DespawnedState = true
            if(RemovedArray:=ParentPowerup.CurrentInstances.RemoveElement[ParentPowerup.CurrentInstances.Find[Self]]):
                set ParentPowerup.CurrentInstances = RemovedArray
            else{Print("~ERROR: Failed to remove a powerup instance from the CurrentInstances list of a powerup type!")}

    MoveTo<public><final>(DesiredPosition:vector3, ?Speed:float=1.0)<suspends>:void=
        if(not PickedUpState?):
            race:
                PowerupPropInstance.MoveTo(DesiredPosition, rotation{}, (Distance(SpawnedPosition, DesiredPosition)/(250.0*Speed)))
                loop:
                    set SpawnedPosition = PowerupPropInstance.GetTransform().Translation
                    Sleep(0.0)
                PickedUpEvent.Await()