using { /Fortnite.com/Devices }
using { /Fortnite.com/Characters }
using { /Fortnite.com/Game }
using { /Fortnite.com/UI }
using { /Fortnite.com/FortPlayerUtilities }
using { /Verse.org/Simulation }
using { /Verse.org/Random }
using { /Verse.org/Colors }
using { /Verse.org/Simulation/Tags }
using { /UnrealEngine.com/Temporary/Diagnostics }
using { /UnrealEngine.com/Temporary/SpatialMath }
using { /UnrealEngine.com/Temporary/UI }
using { Player }
using { Perks }
using { Spawning }
using { Zombie }
using { Weapons }
using { InteractiveModule }

IsInstakill<public>()<transacts><decides>:void=
    InstakillLogic[GetSession()]?

var InstakillLogic : weak_map(session, logic) = map{}
powerup_instakill<public> := class(powerup_base):

    BasicSettings<override>:powerup_base_settings=powerup_base_settings{
        Name:="Instakill"
    }

    OnActivated<override>(InAgent:agent, PowerupInstance:powerup_instance):void=
        if(set InstakillLogic[GetSession()] = true){}

    OnDeactivated<override>(OutAgent:?agent):void=
        if(set InstakillLogic[GetSession()] = false){}

# Grants a random amount of points #
powerup_bonuspoints<public> := class(powerup_base):

    @editable MinimumPointAllocation : int = 500
    @editable MaximumPointAllocation : int = 0
    @editable MaximumMultiplier : float = 2.5
    @editable MaximumMultiplierWave : int = 18

    BasicSettings<override>:powerup_base_settings=powerup_base_settings{
        Name:="Bonus Points"
        IncludeOnUI:=false
        AffectingAgents:=powerup_affecting_type.ActivatingPlayer
    }

    var<private> PowerupInstanceToAmount : [powerup_instance]int = map{}

    GrantWithCustomAmount(ThisPlayer:agent, MoneyAmount:int):void=
        ThisPowerupInst := SpawnWithCustomAmount(vector3{}, MoneyAmount)
        spawn. ThisPowerupInst.ForcePickup(ThisPlayer)

    SpawnWithCustomAmount(AtPosition:vector3, MoneyAmount:int, ?Lifetime:float=0.0):powerup_instance=
        ThisPowerupInstance:=Spawn(AtPosition, ?Lifetime:=Lifetime)
        set PowerupInstanceToAmount = ConcatenateMaps(PowerupInstanceToAmount, map{ThisPowerupInstance => MoneyAmount})
        return ThisPowerupInstance

    OnActivated<override>(InAgent:agent, PowerupInstance:powerup_instance):void=
        if(ThisPlayerInfo:=InAgent.GetPlayerInfo[]):
            if(ThisFixedAmount:=PowerupInstanceToAmount[PowerupInstance]):
                Print("Found forced amount")
                ThisPlayerInfo.IncreasePoints(ThisFixedAmount)
            else:
                ThisPointMultiplier := LockLerp(1.0, MaximumMultiplier, (GetWaveCount()*1.0)/(MaximumMultiplierWave*1.0))
                Print("This Point multiplier: {ThisPointMultiplier}")
                if(RandomPointAmount:=Int[GetRandomInt(MinimumPointAllocation, MaximumPointAllocation) * ThisPointMultiplier], FixedRandomPoints:=RandomPointAmount-Mod[RandomPointAmount, 50]):
                    FixedPointAmount := if. FixedRandomPoints > 10000 then 10000 else. FixedRandomPoints
                    ThisPlayerInfo.IncreasePoints(FixedPointAmount)
                else:
                    ThisPlayerInfo.IncreasePoints(MinimumPointAllocation)

    OnNoInstancesActive<override>():void=
        set PowerupInstanceToAmount = map{}

(InAgent:agent).HasDeathMachine<public>()<transacts><decides>:void=
    DeathMachinePowerup:=powerup_deathmachine[powerup_type.DeathMachine.ConvertToPowerup[]].ActivatedAgentMap[InAgent]?

powerup_deathmachine<public> := class(powerup_base):

    @editable ItemGranter_DeathMachine : item_granter_device = item_granter_device{}
    @editable OnTrigger_SaveInventory : trigger_device = trigger_device{}
    @editable OnTrigger_LoadInventory : trigger_device = trigger_device{}

    BasicSettings<override>:powerup_base_settings=powerup_base_settings{
        Name:="Death Machine"
        AffectingAgents:=powerup_affecting_type.ActivatingPlayer
    }

    var ActivatedAgentMap : [agent]logic = map{}
    var SavedWeaponsPerPlayer : [agent][]weapon_definition = map{}

    var EnabledLocationsMap : [agent][]Interactable = map{}

    OnActivated<override>(InAgent:agent, PowerupInstance:powerup_instance)<suspends>:void=
        if(not ActivatedAgentMap[InAgent]):
            set ActivatedAgentMap = ConcatenateMaps(ActivatedAgentMap, map{InAgent => false})
        if(not ActivatedAgentMap[InAgent]?):
            if(set ActivatedAgentMap[InAgent] = true):
                OnTrigger_SaveInventory.Trigger(InAgent)
                if(DeathMachineWeaponDef:=weapon_definition_list.DeathMachine.GetWeaponDefinition[]):
                    InAgent.RemoveWeapons()
                    DeathMachineWeaponDef.GrantWeapon(InAgent, ?AllowSwap:=false)
                spawn{GrantMaxAmmo(InAgent)}
                if(ThisPlayerInfo:=InAgent.GetPlayerInfo[]):
                    ThisPlayerInfo.ExcludeInteractableType(interactable_type.Weapon)
                    ThisPlayerInfo.ExcludeInteractableType(interactable_type.Ammo)
                    ThisPlayerInfo.PickedUpWeaponEvent.Await()
            else{Print("~ERROR: Failed to set an agent to activated within powerup_deathmachine!")}

    GrantMaxAmmo(InAgent:agent)<suspends>:void=
        Sleep(0.5)
        if(DeathMachine_WeaponDef := InAgent.GetInventoryWeapons()[0]):
            if(DefaultController := zombie_type_list.Default.GetController[]):
                ZombieHealth:=DefaultController.GetZombieHealthIncrease()+150.0
                DeathMachine_WeaponDef.SetDamageOverride(ZombieHealth)
            DeathMachine_WeaponDef.GrantMaxAmmo(InAgent)

    OnDeactivated<override>(OutAgent:?agent):void=
        if(ThisAgent:=OutAgent?):
            Print("~Death machine deactivated")
            if(set ActivatedAgentMap[ThisAgent] = false){}
            OnTrigger_LoadInventory.Trigger(ThisAgent)
            if(ThisPlayerInfo:=ThisAgent.GetPlayerInfo[]):
                ThisPlayerInfo.AllowInteractableType(interactable_type.Weapon)
                ThisPlayerInfo.AllowInteractableType(interactable_type.Ammo)
                #ThisPlayerInfo.SetInteractionState.Signal(true)
        else{Print("~ERROR: Failed to get the OutAgent in OnDeactivated within powerup_deathmachine!")}

IsDoublePoints<public>()<transacts><decides>:void=
    DoublePointsState[GetSession()]?

var DoublePointsState : weak_map(session, logic) = map{}
powerup_doublepoints<public> := class(powerup_base):

    BasicSettings<override>:powerup_base_settings=powerup_base_settings{
        Name:="Double Points"
    }

    var IsActivated : logic = false

    OnActivated<override>(InAgent:agent, PowerupInstance:powerup_instance):void=
        if(not IsActivated?):
            set IsActivated = true
            if(set DoublePointsState[GetSession()] = true){}
            
    OnDeactivated<override>(OutAgent:?agent):void=
        set IsActivated = false
        if(set DoublePointsState[GetSession()] = false){}

powerup_nuke<public> := class(powerup_base):

    @editable PointsGranted : type{_X : int where 0 <= _X} = 400
    @editable RestTime : type{_X : float where 0.0 <= _X} = 10.0

    BasicSettings<override>:powerup_base_settings=powerup_base_settings{
        Name:="Nuke"
        IncludeOnUI:=false
    }

    var IsActivated : logic = false

    OnActivated<override>(InAgent:agent, PowerupInstance:powerup_instance):void=
        if(not IsActivated?):
            set IsActivated = true
            if(ThisWaveDevice:=GlobalWaveController[GetSession()]):
                ThisWaveDevice.SetSpawningState(false)
        spawn{AwaitZombiesEliminated()}

    AwaitZombiesEliminated()<suspends>:void=
        if(DefaultZombieController:=zombie_type_list.Default.GetController[]):
            ActiveZombieList := DefaultZombieController.GetActiveZombies()
            for(ThisZombie:ActiveZombieList):
                ThisZombie.SetDamageAmount(0.0)
        ZombieList:=GetZombieList()
        ZombiesEliminatedTask:=spawn{EliminateZombies(ZombieList)}
        race:
            ActivatedEvent.Await()
            block:
                ZombiesEliminatedTask.Await()
                SafeSleep(RestTime)
                Reset()
        PlayerList:=GetPlayers()
        for(ThisPlayer:PlayerList):
            ThisPlayer.IncreasePoints(PointsGranted)
    
    EliminateZombies(ZombieList:[]zombie_ai)<suspends>:void=
        for(ThisZombie -> ZombieInfo : ZombieList):
            ZombieInfo.SetDamageAmount(0.0)
        for(ThisZombie -> ZombieInfo : ZombieList):
            if(ZombieInfo.Controller.Controls.CanBeNuked?):
                ZombieInfo.Eliminate()
                Sleep(0.15)
        SafeSleep(1.0)

    Reset():void=
        set IsActivated = false
        if(ThisWaveDevice:=GlobalWaveController[GetSession()]):
            ThisWaveDevice.SetSpawningState(true)
            
powerup_firesale<public> := class(powerup_base):

    BasicSettings<override>:powerup_base_settings=powerup_base_settings{
        Name:="Firesale"
    }

    var IsActivated : logic = false

    OnActivated<override>(InAgent:agent, PowerupInstance:powerup_instance):void=
        if(not IsActivated?):
            set IsActivated = true
            Print("~Activated firesale")
            if(WonderBoxManager:=GlobalWonderBoxManager[GetSession()]):
                WonderBoxManager.SetRotateState(false)
                WonderBoxManager.SetPrice(10)
                for(ThisWonderBox:WonderBoxManager.BoxLocations):
                    if(not ThisWonderBox.EnabledState?):
                        spawn{ThisWonderBox.SpawnBox()}
            else{Print("~ERROR: Could not fetch WonderBoxManager!")}

    OnDeactivated<override>(OutAgent:?agent):void=
        Print("~Deactivated firesale")
        if(WonderBoxManager:=GlobalWonderBoxManager[GetSession()]):
            WonderBoxManager.ResetPrice()
            for(ThisWonderBox:WonderBoxManager.BoxLocations):
                if(ThisWonderBox <> WonderBoxManager.CurrentBoxLocation?):
                    spawn{ThisWonderBox.AwaitThisBoxClose()}
            WonderBoxManager.SetRotateState(true)
        else{Print("~ERROR: Could not fetch WonderBoxManager!")}
        set IsActivated = false

    (ThisWonderBox:wonder_box).AwaitThisBoxClose()<suspends>:void=
        if(ThisWonderBox.IsInUse?):
            ThisWonderBox.BoxClosedEvent.Await()
        spawn{ThisWonderBox.DespawnBox()}

powerup_maxammo<public> := class(powerup_base):

    BasicSettings<override>:powerup_base_settings=powerup_base_settings{
        Name:="Max Ammo"
        IncludeOnUI:=false
    }

    OnActivated<override>(InAgent:agent, PowerupInstance:powerup_instance):void=
        if(ThisWeaponManager:=GlobalWeaponManager[GetSession()]):
            PlayerList:=GetPlayers()
            for(ThisPlayer:PlayerList):
                ThisPlayer.User.GrantMaxAmmo()
        else{Print("~ERROR: Could not find global weapon manager!")}

powerup_quickdraw<public> := class(powerup_base):

    BasicSettings<override>:powerup_base_settings=powerup_base_settings{
        Name:="Quickdraw"
    }

    @editable NewInteractTimeMultiplier : float = 0.75

    OnActivated<override>(InAgent:agent, PowerupInstance:powerup_instance):void=
        RepairAllSpawners()

powerup_perk<public> := class(powerup_base):

    BasicSettings<override>:powerup_base_settings=powerup_base_settings{
        Name:="Perk Up"
        IncludeOnUI:=false
        IsSpecial:=true
    }

    OnActivated<override>(InAgent:agent, PowerupInstance:powerup_instance):void=
        PerkList:=GetPerkList()
        if(PerkList.Length > 0):
            if(not PowerupInstance.OwningPlayer?):
                AffectingType:=GetAffectType()
                case(AffectingType):
                    powerup_affecting_type.AllPlayers =>
                        PlayerList := GetAgents()
                        for(ThisPlayer:PlayerList):
                            SelectAndGrantPerk(ThisPlayer)
                    powerup_affecting_type.ActivatingPlayer =>
                        SelectAndGrantPerk(InAgent)
                    powerup_affecting_type.ExcludeActivatingPlayer =>
                        PlayerList := GetAgents()
                        for(ThisPlayer:PlayerList):
                            if(ThisPlayer<>InAgent, ThisPlayerInfo:=ThisPlayer.GetPlayerInfo[]):
                                SelectAndGrantPerk(ThisPlayer)
            else:
                SelectAndGrantPerk(InAgent)
        else{Print("~ERROR: Failed to fetch perk list!")}

    SelectAndGrantPerk<public>(InAgent:agent):void=
        PerkList:=GetPerkList()
        if(ThisPlayerInfo:=InAgent.GetPlayerInfo[]):
            if:
                ThisPlayerInfo.ActivePerks.Length < PerkList.Length
                ThisRandomPerk:=GetRandomUnactivatedPerk[InAgent]
            then:
                ThisRandomPerk.Apply(InAgent, ?IgnoreCost:=true, ?IgnorePerkLimit:=true)
            else if(ThisRandomPerk:=PerkList[GetRandomInt(0, PerkList.Length-1)]):
                if(ThisRandomPerk.IsMaxTier[InAgent]):
                    ThisPlayerInfo.IncreasePoints(ThisRandomPerk.GetTierCost[1] or 500)
                else:
                    ThisRandomPerk.Apply(InAgent, ?IgnoreCost:=true)

    GetRandomUnactivatedPerk<private>(InAgent:agent)<transacts><decides>:perk_base=
        var MaybeRandomPerk : ?perk_base = false
        PerkList:=GetPerkList()
        var UnactivatedPerkList : []perk_base = array{}
        for(ThisPerk:PerkList):
            if(not ThisPerk.IsActive[InAgent]):
                set UnactivatedPerkList += array{ThisPerk}
        if(UnactivatedPerkList.Length <= 0):
            for(ThisPerk:PerkList):
                if(not ThisPerk.IsMaxTier[InAgent]):
                    set UnactivatedPerkList += array{ThisPerk}
        set MaybeRandomPerk = option{Shuffle(UnactivatedPerkList)[0]}
        MaybeRandomPerk?

powerup_incognito<public> := class(powerup_base):

    BasicSettings<override>:powerup_base_settings=powerup_base_settings{
        Name:="Incognito"
        ActivatedTime:=30.0
        AffectingAgents := powerup_affecting_type.ActivatingPlayer
        #IsSpecial:=true
    }

    OnActivated<override>(InAgent:agent, PowerupInstance:powerup_instance):void=
        Print("~Incognito activated")
        if(ThisPlayerInfo:=InAgent.GetPlayerInfo[]):
            Print("~Player is undetectable")
            ThisPlayerInfo.SetDetectableState(false)
        else{Print("~ERROR: Failed to make a player undetectable by zombies!")}

    OnDeactivated<override>(OutAgent:?agent):void=
        Print("~Incognito deactivated")
        if(ThisPlayerInfo:=OutAgent?.GetPlayerInfo[]):
            Print("~Player is detectable")
            ThisPlayerInfo.SetDetectableState(true)
        #else{Print("~ERROR: Failed to make a player detectable by zombies!")}
