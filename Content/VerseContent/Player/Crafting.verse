using { /Fortnite.com/Devices }
using { /Fortnite.com/Characters }
using { /Fortnite.com/Game }
using { /Fortnite.com/UI }
using { /Fortnite.com/FortPlayerUtilities }
using { /Verse.org/Simulation }
using { /Verse.org/Random }
using { /Verse.org/Colors }
using { /UnrealEngine.com/Temporary/Diagnostics }
using { /UnrealEngine.com/Temporary/SpatialMath }
using { /UnrealEngine.com/Temporary/UI }
using { InteractiveModule }
using { UI }

########  ########     ###     ######    #######  ##    ## ########     ###     ######   ######## 
##     ## ##     ##   ## ##   ##    ##  ##     ## ###   ## ##     ##   ## ##   ##    ##  ##       
##     ## ##     ##  ##   ##  ##        ##     ## ####  ## ##     ##  ##   ##  ##        ##       
##     ## ########  ##     ## ##   #### ##     ## ## ## ## ########  ##     ## ##   #### ######   
##     ## ##   ##   ######### ##    ##  ##     ## ##  #### ##   ##   ######### ##    ##  ##       
##     ## ##    ##  ##     ## ##    ##  ##     ## ##   ### ##    ##  ##     ## ##    ##  ##       
########  ##     ## ##     ##  ######    #######  ##    ## ##     ## ##     ##  ######   ########

<##############################################

Crafting Station & Buildable devices

##############################################>

(InAgent:agent).GetBuildableStorage<public>()<transacts><decides>:buildable_storage=
    return InAgent.GetPlayerInfo[].BuildableStorage

var CraftingStationCollection : weak_map(session, []crafting_station) = map{}

# A crafting station which allows the crafting of any buildable #
# Can support any buildable as long as it inherits from the class "buildable_base" #
# Can only support one buildable at a time #
crafting_station<public> := class(Interactive_Base):

    @editable ReferenceCenter : creative_prop = creative_prop{}
    @editable Beacon : beacon_device = beacon_device{}

    @editable
    InteractiveLocation : Interactable = Interactable{InteractTime:=3.0, InteractDistance:=160.0}

    var<private> CurrentBuildable : ?buildable_base = false
    var<private> BuiltParts : []buildable_part = array{}

    var<private> IsBuilt : logic = false

    var<private> EnabledState : logic = true

    OnBegin<override>():void=
        # Makes this crafting station non-interactable by default since no players will have parts to start with #
        if:
            not CraftingStationCollection[GetSession()]
            set CraftingStationCollection[GetSession()] = array{}
        if:
            CurrentCraftingStations := CraftingStationCollection[GetSession()]
            set CraftingStationCollection[GetSession()] = CurrentCraftingStations + array{Self}
        SetInteractableState(true)
        LinkAndCompile(array. InteractiveLocation)

    GetInteractionText<override><final>(InAgent:agent, Interactive:Interactable):string=
        if(EnabledState?, InPlayerInfo:=InAgent.GetPlayerInfo[]):
            if(not IsBuilt?):
                if(BuildableStorage:=InAgent.GetBuildableStorage[]):
                    if(ThisBuildable:=CurrentBuildable?):
                        MatchingPartList := GetBuildableParts(InAgent, ThisBuildable)
                        if(MatchingPartList.Length > 1, MatchingPartList.Length+BuiltParts.Length = ThisBuildable.BuildableParts.Length):
                            InteractiveLocation.SetInteractableState(true, ?InAgent:=option{InAgent})
                            return "Place all parts down [{ThisBuildable.Name}]"
                        else if(MatchingPartList.Length > 1):
                            InteractiveLocation.SetInteractableState(true, ?InAgent:=option{InAgent})
                            return "Place parts down [{ThisBuildable.Name}]"
                        else if(ThisMatchingPart := MatchingPartList[0]):
                            InteractiveLocation.SetInteractableState(true, ?InAgent:=option{InAgent})
                            return "Place part down [{ThisMatchingPart(0).Name}]"
                        else if(BuildableStorage.GetFirstCollectedPart[Self]):
                            InteractiveLocation.SetInteractableState(false, ?InAgent:=option{InAgent})
                            return "No matching parts for [{ThisBuildable.Name}]"
                        else:
                            InteractiveLocation.SetInteractableState(false, ?InAgent:=option{InAgent})
                            return "No parts to place down!"
                    else:
                        MaybeFreeBuildableType := GetUnassignedBuildable()
                        if(ThisBuildable:=MaybeFreeBuildableType?):
                            BuildablePartList := GetBuildableParts(InAgent, ThisBuildable)
                            if(BuildablePartTuple:=BuildablePartList[0]):
                                if(BuildablePartTuple(0).BelongsToCraftingStation[Self]):
                                    InteractiveLocation.SetInteractableState(true, ?InAgent:=option{InAgent})
                                    MatchingParts := GetBuildableParts(InAgent, BuildablePartTuple(0).ParentBuildable)
                                    if(MatchingParts.Length = BuildablePartTuple(0).ParentBuildable.BuildableParts.Length):
                                        return "Place all parts down [{BuildablePartTuple(0).GetBuildableName()}]"
                                    else if(MatchingParts.Length > 1):
                                        return "Place parts down [{BuildablePartTuple(0).GetBuildableName()}]"
                                    else:
                                        return "Start building [{BuildablePartTuple(0).GetBuildableName()}]"
                                else:
                                    InteractiveLocation.SetInteractableState(false, ?InAgent:=option{InAgent})
                                    return "Belongs to another crafting station!"
                            else:
                                InteractiveLocation.SetInteractableState(false, ?InAgent:=option{InAgent})
                                return "No parts to place down!"
                        else:
                            InteractiveLocation.SetInteractableState(false, ?InAgent:=option{InAgent})
                            return "Belongs to another crafting station!"
                else{Print("~ERROR: Could not fetch buildable storage!")}
                InteractiveLocation.SetInteractableState(false, ?InAgent:=option{InAgent})
                return ""
            else if(ThisBuildable:=CurrentBuildable?):
                if(ThisBuildable.CheckPickupPolicy[InAgent]):
                    InteractiveLocation.SetInteractableState(true, ?InAgent:=option{InAgent})
                    return ThisBuildable.PickupPrompt
                else:
                    InteractiveLocation.SetInteractableState(false, ?InAgent:=option{InAgent})
                    return ThisBuildable.PickupPolicyFailedReason
            else:
                return "ERROR: Is enabled, and IsBuild but failed getting Buildable!"
        else:
            # This block activates when the crafting station is disabled #
            InteractiveLocation.SetInteractableState(false, ?InAgent:=option{InAgent})
            return "Crafting station disabled!"

    GetDescription<override>(InAgent:agent, Interactive:Interactable):tuple(string, string)=
        if(ThisBuildable:=CurrentBuildable?):
            if(Shield_Buildable[ThisBuildable]):
                return (ThisBuildable.Description, "shield")
            else:
                return (ThisBuildable.Description, "")
        else:
            return ("", "")

    OnInteracted<override><final>(InAgent:agent, InteractingLocation:Interactable):void=
        if(EnabledState?):
            if(not IsBuilt?):
                if(ThisBuildable:=CurrentBuildable?):
                    ReadyPartList:=GetBuildableParts(InAgent, ThisBuildable)
                    for(ThisPart:ReadyPartList):
                        InAgent.PlacePart(ThisPart(0), ?OwningPlayer:=option{ThisPart(1)})
                    if(BuiltParts.Length = ThisBuildable.BuildableParts.Length):
                        set IsBuilt = true
                        SetInteractableState(true)
                        SetEnabledState(true)
                        if(ReferenceCenter.IsValid[]):
                            for(ThisPart:ThisBuildable.BuildableParts):
                                ThisPart.DisposePlacedPart()
                            ThisBuildable.SpawnBuildable(ReferenceCenter.GetTransform())
                        ThisBuildable.OnBuildComplete()
                        InteractiveLocation.SetInteractionTime(0.75)
                else:
                    MaybeFreeBuildableType := GetUnassignedBuildable()
                    if(ThisBuildable:=MaybeFreeBuildableType?):
                        BuildablePartList := GetBuildableParts(InAgent, ThisBuildable)
                        set CurrentBuildable = option{ThisBuildable}
                        set ThisBuildable.CurrentCraftingStation = option{Self}
                        ReadyPartList:=GetBuildableParts(InAgent, ThisBuildable)
                        for(ThisPart:ReadyPartList):
                            InAgent.PlacePart(ThisPart(0), ?OwningPlayer:=option{ThisPart(1)})
                        if(BuiltParts.Length = ThisBuildable.BuildableParts.Length):
                            set IsBuilt = true
                            SetInteractableState(true)
                            SetEnabledState(true)
                            if(ReferenceCenter.IsValid[]):
                                for(ThisPart:ThisBuildable.BuildableParts):
                                    ThisPart.DisposePlacedPart()
                                ThisBuildable.SpawnBuildable(ReferenceCenter.GetTransform())
                            ThisBuildable.OnBuildComplete()
                            InteractiveLocation.SetInteractionTime(0.75)
            else if(ThisBuildable:=CurrentBuildable?):
                if(ThisBuildable.CheckPickupPolicy[InAgent]):
                    # Interacting with a finished Buildable
                    spawn{ThisBuildable.OnPickupInteraction(InAgent)}
                    ThisBuildable.SetEquippedState(InAgent, true)
                    if(ThisBuildable.PickupPolicy = buildable_pickup_policy.DisableOnPickup):
                        SetInteractableState(false)
                else:
                    Print("~Failed pickup due to PickupPolicy!")

        else:
            # This block activates when this crafting station is interacted with while disabled #
            # Can be used for when players must buy back the crafting station (think of brutus from MOTD) #

    OnInteractStart<override>(InAgent:agent, InteractingLocation:Interactable)<suspends>:void=
        var CraftingTuple : tuple(creative_prop_asset, float) = (DefaultCreativePropAsset, 0.0)
        if(ThisBuildable:=CurrentBuildable?):
            set CraftingTuple = (ThisBuildable.CraftingBlueprint, ThisBuildable.CraftingBPLifetime)
        if(BuildableStorage:=InAgent.GetBuildableStorage[], ThisPart:=BuildableStorage.GetMatchingPartsToBuildable[CurrentBuildable?][0] or BuildableStorage.GetFirstCollectedPart[Self]):
            set CraftingTuple = (ThisPart.ParentBuildable.CraftingBlueprint, ThisPart.ParentBuildable.CraftingBPLifetime)
        if(ThisSpawnedCraftingBP:=SpawnProp(CraftingTuple(0), ReferenceCenter.GetTransform().Translation+vector3{Z:=50.0}, rotation{})(0)?):
            race:
                SafeSleep(CraftingTuple(1))
                loop:
                    Agent:=InteractionStoppedEvent.Await()
                    if(Agent=InAgent):
                        break
            ThisSpawnedCraftingBP.Dispose()
    
    InteractionStoppedEvent:event(agent)=event(agent){}
    OnInteractStopped<override>(InAgent:agent, InteractingLocation:Interactable)<suspends>:void=
        Sleep(0.0)
        InteractionStoppedEvent.Signal(InAgent)

    (InAgent:agent).PlacePart(InPart:buildable_part, ?OwningPlayer:?agent=false):void=
        if(not BuiltParts.Find[InPart], BuildableStorage:=InAgent.GetBuildableStorage[]):
            if(not CurrentBuildable?):
                set CurrentBuildable = option{InPart.ParentBuildable}
                set InPart.ParentBuildable.CurrentCraftingStation = option{Self}
            if(ThisPlayer:=OwningPlayer?, ThisBuildableStorage:=ThisPlayer.GetBuildableStorage[]):
                ThisBuildableStorage.RemoveCollectedPart(InPart)
            else:
                BuildableStorage.RemoveCollectedPart(InPart)
            BuildableStorage.PartPlacedDown(InPart)
            set BuiltParts += array{InPart}
            if(ReferenceCenter.IsValid[]):
                InPart.PlacePartDown(InAgent, ReferenceCenter.GetTransform())

    IsOccupied<public><final>()<transacts><decides>:buildable_base=
        return CurrentBuildable?

    Enable<public><final>():void={set EnabledState = true}
    Disable<public><final>():void={set EnabledState = false}
    Reset<override><final>():void=
        SetInteractableState(true)
        set CurrentBuildable = false
        set BuiltParts = array{}
        set IsBuilt = false
        set EnabledState = true
        (super:)Reset()

GetBuildableParts(RequestingUser:agent, DesiredBuildable:buildable_base):[]tuple(buildable_part, agent)=
    var ReadyParts : []tuple(buildable_part, agent) = array{}
    Players:=GetPlayers()
    if(ThisBuildableStorage:=RequestingUser.GetBuildableStorage[]):
        if(OwnedReadyParts:=ThisBuildableStorage.GetMatchingPartsToBuildable[DesiredBuildable]):
            FixedPartList:=for:
                ThisPart:OwnedReadyParts
            do:
                (ThisPart, RequestingUser)
            set ReadyParts = FixedPartList
        if(DesiredBuildable.AllowSharedCrafting?):
            for(ThisPlayer:Players, ThisPlayer<>RequestingUser):
                if(ThisBuildableStorage1:=ThisPlayer.User.GetBuildableStorage[]):
                    if(ThisReadyPartList:=ThisBuildableStorage1.GetMatchingPartsToBuildable[DesiredBuildable]):
                        FixedPartList:=for:
                            ThisPart:ThisReadyPartList
                        do:
                            (ThisPart, ThisPlayer.User)
                        set ReadyParts += FixedPartList
    return ReadyParts

GetUnassignedBuildable()<transacts>:?buildable_base=
    if(BuildableList := GlobalBuildables[GetSession()]):
        for(ThisBuildable:BuildableList):
            if(not ThisBuildable.CurrentCraftingStation?):
                return option{ThisBuildable}
    false    
            
# Holds all the collected parts a player has picked up and currently carries #
# Belongs in a playerinfo class #
buildable_storage<public> := class:

    var<private> CollectedParts : []buildable_part = array{}
    var<private> PlacedDownParts : []buildable_part = array{}

    GetFirstCollectedPart(InCraftingTable:crafting_station, ?SearchGlobally:logic=true)<transacts><decides>:buildable_part=
        var ReturningPart : ?buildable_part = false
        for(ThisPart:CollectedParts):
            if(not ReturningPart?):
                if(ThisPart.BelongsToCraftingStation[InCraftingTable] or not InCraftingTable.CurrentBuildable?):
                    set ReturningPart = option{ThisPart}
        ReturningPart?

    GetRandomCollectedPart()<transacts><decides>:buildable_part=
        return CollectedParts[GetRandomInt(0, CollectedParts.Length-1)]

    GetMatchingPartsToBuildable(InBuildable:buildable_base)<transacts><decides>:[]buildable_part=
        MatchingParts:=for:
            ThisPart : CollectedParts
            ThisPart.BelongsToBuildable[InBuildable]
        do:
            ThisPart

    GetAllParts():[]buildable_part=
        return CollectedParts+PlacedDownParts

    GetGlobalStorages():[]buildable_storage=
        var GlobalStorages : []buildable_storage = array{}
        PlayerList:=GetPlayers()
        for(ThisPlayer:PlayerList):
            if(ThisStorage:=ThisPlayer.User.GetBuildableStorage[]):
                set GlobalStorages += array{ThisStorage}
        return GlobalStorages
                
    PartPlacedDown(InPart:buildable_part):void=
        set PlacedDownParts += array{InPart}
        if(NewArray:=CollectedParts.RemoveFirstElement[InPart]):
            set CollectedParts = NewArray

    AddCollectedPart(InPart:buildable_part):void=
        set CollectedParts += array{InPart}

    RemoveCollectedPart(InPart:buildable_part):void=
        if(UpdatedCollectedParts := CollectedParts.RemoveElement[CollectedParts.Find[InPart]]):
            set CollectedParts = UpdatedCollectedParts
        else{Print("~ERROR: Failed to find or remove a buildable part from a player!")}

    ResetParts():void=
        for(ThisPart:CollectedParts):
            ThisPart.ResetPosition()
        set CollectedParts = array{}

    Reset():void=
        set CollectedParts = array{}
        set PlacedDownParts = array{}

var GlobalBuildables : weak_map(session, []buildable_base) = map{}
buildable_pickup_policy<public> := enum{NoRestrictions, OncePerLifetime, DisableOnPickup}
# The base class for any buildable, like a shield or super weapon #
# Used for easily setting where the parts of a buildable should go, and retrieving information about the buildable and its parts #
# Meant to be inherited by other classes for specific use case buildables #
buildable_base<public> := class(Interactive_Base):

    pickup_policy_TT<localizes>:message="What restrictions should be in place when a player tries to pick this Buildable up?"
    pickup_policy_failurereason_TT<localizes>:message="The message to display when this Buildable cannot be picked up by a player due to them not meeting the Pickup Policy"

    @editable Name : string = "Unset Buildable Name"
    @editable PickupPrompt : string = "Pickup Unset Buildable"
    @editable Description : string = ""
    @editable AllowSharedCrafting : logic = false
    @editable{ToolTip:=pickup_policy_TT} PickupPolicy : buildable_pickup_policy = buildable_pickup_policy.NoRestrictions
    @editable{ToolTip:=pickup_policy_failurereason_TT} PickupPolicyFailedReason : string = "Unset reason"
    @editable ReferenceCenter : creative_prop = creative_prop{}
    @editable BuildableBlueprint : creative_prop_asset = DefaultCreativePropAsset
    @editable CraftingBlueprint : creative_prop_asset = DefaultCreativePropAsset
    @editable CraftingBPLifetime : float = 0.0
    @editable BuiltBuildableOffset : creative_prop = creative_prop{}
    @editable BuildableParts : []buildable_part = array{}

    # The crafting station that this buildable is being/already built at
    var CurrentCraftingStation : ?crafting_station = false
    GetCraftingStation<public><final>()<transacts><decides>:crafting_station=
        return CurrentCraftingStation?

    OnBegin<override><final>():void=
        var InteractableLocations : []Interactable = array{}
        var PartIndex:int=0
        for(ThisBuildablePart:BuildableParts):
            ThisBuildablePart.Init(Self,PartIndex)
            set InteractableLocations += ThisBuildablePart.PotentialLocations
            set PartIndex += 1
        LinkAndCompile(InteractableLocations)
        GeneratePartLocations()
        if:
            not GlobalBuildables[GetSession()]
            set GlobalBuildables[GetSession()] = array{}
        if:
            CurrentBuildables := GlobalBuildables[GetSession()]
            set GlobalBuildables[GetSession()] = CurrentBuildables + array{Self}
    var<private> SetupComplete : logic = false
    GeneratePartLocations<public><final>():void=
        DisposePlacedBuildable()
        for(ThisBuildablePart:BuildableParts):
            ThisBuildablePart.DisposePlacedPart()
            ThisBuildablePart.Reset()
            if(ThisRandomLocation:=ThisBuildablePart.GetRandomLocation[], ThisIndex:=ThisBuildablePart.PotentialLocations.FindInstance[ThisRandomLocation]):
                ThisRandomLocation.SetEnabledState(true)
                ThisRandomLocation.PropLocation.Show()
                set ThisBuildablePart.CurrentLocationIndex = ThisIndex
            else:
                Print("ERR: Failed to set a buildable part as enabled")

    # Stores the part in the agents' CollectedPartsMap #
    OnInteracted<override><final>(InAgent:agent, InteractingLocation:Interactable):void=
        var MaybeBuildablePart : ?buildable_part = false
        for:
            ThisPart:BuildableParts
            ThisPart.PotentialLocations.FindInstance[InteractingLocation]
        do:
            set MaybeBuildablePart = option. ThisPart
        if(BuildableStorage := InAgent.GetBuildableStorage[], CollectedPart := MaybeBuildablePart?):
            InteractingLocation.SetEnabledState(false)
            BuildableStorage.AddCollectedPart(CollectedPart)
            CollectedPart.HidePotentialLocations()
            CollectedPart.PickupPart(InAgent)
        else{Print("~ERROR: Failed to fetch buildable storage or convert an interactable location to a buildable part!")}

    GetInteractionText<override>(InAgent:agent, InteractingLocation:Interactable):string=
        for(ThisPart:BuildableParts):
            if(ThisPart.PotentialLocations.FindInstance[InteractingLocation]):
                return ThisPart.PickupPrompt
        return "err"

    CheckPickupPolicy(InAgent:agent)<transacts><decides>:void=
        if(not PlayerToEquippedState[InAgent]){set PlayerToEquippedState = ConcatenateMaps(PlayerToEquippedState, map{InAgent=>false})}
        case(PickupPolicy):
            buildable_pickup_policy.OncePerLifetime =>
                not PlayerToEquippedState[InAgent]?
            buildable_pickup_policy.DisableOnPickup =>
                PlayerToEquippedState.Length <= 0
            _ =>
            
    var<private> PlayerToEquippedState : [agent]logic = map{}
    SetEquippedState<public>(InAgent:agent, InState:logic):void=
        if(not PlayerToEquippedState[InAgent]):
            set PlayerToEquippedState = ConcatenateMaps(PlayerToEquippedState, map{InAgent => InState})
        else if(set PlayerToEquippedState[InAgent] = InState){}
        else{Print("~ERR: Failed to change the equipped state of a buildable [{Name}]!")}

    # A function meant to be OVERRIDDEN by inheriting classes #
    # Activates when a player tries to pickup this buildable from a crafting station #
    # Can be made to limit buildable pickups for one player, on a timer/cost, or for all players #
    OnPickupInteraction<public>(InAgent:agent)<suspends>:void=
        return

    OnPartPickedUp<public>(InAgent:agent):void=
        return

    OnAllPartsPickedUp<public>():void=
        return

    OnPartPlacedDown<public>(InAgent:agent):void=
        return

    OnBuildComplete<public>():void=
        return

    var<private> UnequipEventPerPlayer : [agent]event() = map{}
    GetUnequipEvent(InAgent:agent):event()=
        if(ThisUnequipEvent:=UnequipEventPerPlayer[InAgent]):
            return ThisUnequipEvent
        else:
            NewEvent:event()=event(){}
            set UnequipEventPerPlayer = ConcatenateMaps(UnequipEventPerPlayer, map{InAgent => NewEvent})
            return NewEvent
        
    Unequip<public><final>(InAgent:agent):void=
        GetUnequipEvent(InAgent).Signal()
        SetEquippedState(InAgent, false)
        spawn{OnUnequipped(InAgent)}
    
    OnUnequipped<public>(InAgent:agent)<suspends>:void=
        return

    var SpawnedBuildableProp : creative_prop = creative_prop{}
    SpawnBuildable(InLocation:transform):void=
        SpawnedBuildableProp.Dispose()
        ObjectTransform := BuiltBuildableOffset.GetTransform()
        var ReferenceObjectOffset : vector3 = ReferenceCenter.GetTransform().Translation - ObjectTransform.Translation
        set ReferenceObjectOffset = ReferenceObjectOffset * vector3{X:=1.0,Y:=1.0,Z:=-1.0}
        LocalTranslationOffset := InLocation.GetLocalVector(ReferenceObjectOffset)

        SubtractedRotate := InLocation.Rotation.UnrotateBy(ReferenceCenter.GetTransform().Rotation)
        FinalRotation := ObjectTransform.Rotation.RotateBy(SubtractedRotate)
    
        MaybeObjectProp := SpawnProp(BuildableBlueprint, LocalTranslationOffset, FinalRotation)
        if(SpawnedObjectProp:=MaybeObjectProp(0)?):
            set SpawnedBuildableProp = SpawnedObjectProp

    DisposePlacedBuildable():void=
        SpawnedBuildableProp.Dispose()

    GetFreeParts()<transacts>:[]buildable_part=
        FreePartList:=for:
            ThisPart:BuildableParts
            not ThisPart.GetOwningCraftingStation[]
        do:
            ThisPart

# Holds information about a crafting/buildable part #
buildable_part<public> := class<concrete><unique>:

    var<private> ParentBuildable : buildable_base = buildable_base{}
    Init(InBuildable:buildable_base, Index:int):void=
        set ParentBuildable = InBuildable
        set PartIndex = Index

    @editable Name : string = "Unset Buildable Part Name"
    @editable PickupPrompt : string = "Pick up part"
    @editable PotentialLocations<public> : []Interactable = array{}
    #@editable BuildablePartProp : creative_prop = creative_prop{}
    @editable PartBlueprint : creative_prop_asset = DefaultCreativePropAsset
    @editable ReferenceOffset : creative_prop = creative_prop{}

    var PartIndex:int=-1
    var CurrentLocationIndex : int = 0

    var SpawnedProp : creative_prop = creative_prop{}
    var OwningPlayer : ?agent = false

    GetRandomLocation<public><final>()<transacts><decides>:Interactable=
        RandomLocation := Shuffle(PotentialLocations)[0]
        RandomLocation.IsValid[]
        return RandomLocation

    PlacePartDown(InAgent:agent, InLocation:transform):void=
        if(ParentBuildable.ReferenceCenter.IsValid[]):
            SpawnedProp.Dispose()
            ObjectTransform := ReferenceOffset.GetTransform()
            var ReferenceObjectOffset : vector3 = ParentBuildable.ReferenceCenter.GetTransform().Translation - ObjectTransform.Translation
            set ReferenceObjectOffset = ReferenceObjectOffset * vector3{X:=1.0,Y:=1.0,Z:=-1.0}
            LocalTranslationOffset := InLocation.GetLocalVector(ReferenceObjectOffset)

            SubtractedRotate := InLocation.Rotation.UnrotateBy(ParentBuildable.ReferenceCenter.GetTransform().Rotation)
            FinalRotation := ObjectTransform.Rotation.RotateBy(SubtractedRotate)
    
            MaybeObjectProp := SpawnProp(PartBlueprint, LocalTranslationOffset, FinalRotation)
            if(SpawnedObjectProp:=MaybeObjectProp(0)?):
                set SpawnedProp = SpawnedObjectProp
            else{Print("~ERROR: Could not spawn a part prop for the buildable [{ParentBuildable.Name}]!")}
        else{Print("~Buildable ERR: The center reference point of a buildable is invalid or empty!")}
        ParentBuildable.OnPartPlacedDown(InAgent)

    DisposePlacedPart():void=
        SpawnedProp.Dispose()

    HidePotentialLocations():void=
        for(ThisPartLocation:PotentialLocations):
            ThisPartLocation.PropLocation.Hide()
        
    ResetPosition():void=
        if(ThisLocation:=PotentialLocations[CurrentLocationIndex]):
            ThisLocation.PropLocation.Show()
            ThisLocation.SetEnabledState(true)

    Reset():void=
        set OwningPlayer = false
        for(ThisLocation:PotentialLocations):
            ThisLocation.SetEnabledState(false)
            ThisLocation.PropLocation.Hide()

    PickupPart(InAgent:agent):void=
        set OwningPlayer = option{InAgent}
        spawn{InAgent.GetUIManager().PartsUI.OnPartPickedUp(option. Self)}
        spawn{OnPartPickedUp(InAgent)}
        ParentBuildable.OnPartPickedUp(InAgent)
        if(ParentBuildable.GetFreeParts().Length <= 0):
            ParentBuildable.OnAllPartsPickedUp()

    OnPartPickedUp(InAgent:agent)<suspends>:void=
        return

    GetBuildableName<public><final>():string=
        return ParentBuildable.Name

    BelongsToBuildable<public><final>(InBuildable:buildable_base)<transacts><decides>:void=
        InBuildable = ParentBuildable

    BelongsToCraftingStation<public><final>(InCraftingStation:crafting_station)<transacts><decides>:void=
        not ParentBuildable.CurrentCraftingStation? or ParentBuildable.CurrentCraftingStation? = InCraftingStation
    
    GetOwningCraftingStation<public><final>()<transacts><decides>:crafting_station=
        ParentBuildable.CurrentCraftingStation?

Shield_Buildable<public> := class(buildable_base):

    @editable ShieldVFX : visual_effect_powerup_device = visual_effect_powerup_device{}
    @editable Stinger_PartPickedUp : audio_player_device = audio_player_device{}
    @editable Stinger_ShieldHit : audio_player_device = audio_player_device{}
    @editable Stinger_ShieldBreak : audio_player_device = audio_player_device{}
    @editable ProtectionAmount : type{_X : float where 0.0 <= _X} = 700.0

    OnPickupInteraction<override>(InAgent:agent)<suspends>:void=
        if(ThisPlayerInfo:=InAgent.GetPlayerInfo[]):
            set ThisPlayerInfo.ShieldClass = option{Self}
            set ThisPlayerInfo.ShieldAmount = ProtectionAmount
            ThisUI := InAgent.GetUIManager()
            ThisUI.ShieldWidget.Show()
            ThisUI.ShieldHealthBar.SetProgress(0.0)
            spawn{ThisUI.ShieldHealthBar.ChangeProgressOverTime(0.75, 100.0)}
            ThisUI.ShieldHealthBar.Show()
            race:
                loop:
                    ShieldVFX.Pickup(InAgent)
                    Sleep(1.0)
                GetUnequipEvent(InAgent).Await()

    OnUnequipped<override>(InAgent:agent)<suspends>:void=
        if(ThisPlayerInfo:=InAgent.GetPlayerInfo[]):
            Stinger_ShieldBreak.Play(InAgent)
            set ThisPlayerInfo.ShieldClass = false
            set ThisPlayerInfo.ShieldAmount = 0.0
            ThisUI := InAgent.GetUIManager()
            ThisUI.ShieldHealthBar.ChangeProgressOverTime(0.5, 0.0)
            ThisUI.ShieldHealthBar.Hide()
            ThisUI.ShieldWidget.Hide()

    OnPartPickedUp<override>(InAgent:agent):void=
        if(not GetCraftingStation[], CraftingStations := CraftingStationCollection[GetSession()]):
            for(ThisStation:CraftingStations):
                ThisStation.Beacon.Enable()
        Stinger_PartPickedUp.Play(InAgent)

    OnPartPlacedDown<override>(InAgent:agent):void=
        if(ThisCraftingStation:=GetCraftingStation[], CraftingStations := CraftingStationCollection[GetSession()]):
            for(ThisStation:CraftingStations):
                if(ThisStation <> ThisCraftingStation):
                    ThisStation.Beacon.Disable()
    
    OnBuildComplete<override>():void=
        if(ThisCraftingStation:=GetCraftingStation[]):
            ThisCraftingStation.Beacon.Disable()