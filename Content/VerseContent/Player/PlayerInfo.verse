using { /Verse.org/Simulation }
using { /Verse.org/Colors }
using { /Fortnite.com/Characters }
using { /Fortnite.com/Devices }
using { /UnrealEngine.com/Temporary/SpatialMath }
using { /Fortnite.com/FortPlayerUtilities }
using { Perks }
using { Weapons }
using { InteractiveModule }
using { UI }
using { Spawning }
using { Powerups }

REVIVEOFFSET : vector3 = vector3{Z:=25.0}

AUTOHEAL_DELAY : float = 3.0
AUTOHEAL_TICKTIME : float = 1.0
AUTOHEAL_AMOUNT : float = 25.0

POINTCAP : int = 1000000

(InAgent:agent).GetPlayerInfo<public>()<transacts><decides>:player_info=
    return GlobalPlayerManager[GetSession()].PlayerList[InAgent]?

(InAgent:agent).GetPlayerPackage<public>()<transacts><decides>:player_package=
    GlobalPlayerManager[GetSession()].PlayerPackages[InAgent.GetPlayerInfo[].UniqueIndex]

(InAgent:agent).IsInSession<public>()<transacts><decides>:void=
    GlobalPlayerManager[GetSession()].PlayerList[InAgent]?.IsInSession[]

player_package<public> := class<concrete>:

    @editable
    WeaponAnchor<public> : creative_prop = creative_prop{}

    @editable
    InteractionIcon<public> : creative_prop = creative_prop{}

    @editable
    DownedProp<public> : creative_prop = creative_prop{}

    @editable
    DownedCamera<public> : gameplay_camera_orbit_device = gameplay_camera_orbit_device{}

    @editable
    HurtScreenCinematic<public> : cinematic_sequence_device = cinematic_sequence_device{}

    @editable
    ExfilChair<public> : chair_device = chair_device{}

    @editable 
    Color<public> : color = NamedColors.White

    var IsInUse<public> : ?agent = false

(PlayerState:player_state).ToString<public>()<transacts>:string=
    case(PlayerState):
        player_state.Inactive => "Inactive"
        player_state.Active => "Active"
        player_state.Downed => "Downed"
        player_state.Paused => "Paused"
        player_state.Spectating => "Spectating"
        player_state.OutOfSession => "OutOfSession"

player_state<public> := enum{
    Inactive # When the player is dead or not yet initialized
    Active # When the player is alive in game
    Downed
    Paused
    Spectating
    OutOfSession
}

player_info<public> := class<unique>:

    User<public> : agent
    BuildableStorage<public> : buildable_storage = buildable_storage{}
    JoinTime<public> : float

    # VARIABLES
    var ReadyUpState<public> : logic = false
    var<private> Points<public> : int = 0
    var<private> PointMultiplier : float = 1.0

    var<private> Health : float = 100.0
    var<private> MaxHealth : float = 100.0
    var<private> RegenDelay : float = 3.0
    var<private> RegenAmount : float = 25.0
    var<private> RegenTickrate : float = 1.0
    var<private> VulnerableState<public> : logic = true
    var<private> DownedTime<public> : float = 40.0

    var<private> DamageMultiplierBoostActive<public> : logic = false
    var<private> SavedDamageMultiplier<public> : float = 1.0
    var<private> DamageMultiplier<public> : float = 1.0
    var<private> AmmoDropChance<public> : float = 5.5

    var<private> InventorySlots<public> : int = 2

    var<private> DetectableByZombies<public> : logic = true

    var<private> UniqueIndex<public> : int = -1

    var<private> ExcludedInteractableTypes<public> : []interactable_type = array{}

    var<private> ActivePerks<public> : []perk_base = array{}
    var PerkSlots<private> : int = 7

    var<public> LastMapLocation<public> : ?transform = false

    # - Shield
    var ShieldClass<public> : ?Shield_Buildable = false
    var ShieldAmount<public> : float = 0.0

    var TickRatePercent<public> : float = 100.0
    var InteractTimeMultiplier<public> : float = 1.0

    # - Game stats
    var TotalKills<public> : int = 0
    var TotalRevives<public> : int = 0
    var TotalDowns<public> : int = 0

    var ScoreboardState<public> : logic = false

    ReviveInteractable<public> : Interactable = Interactable{RequiresPlayerMustBeInFront:=false, InteractDistance:=125.0, InteractTime := 3.5}

    # ADMIN SETTINGS / Menu panel
    var<public> GameplayMenuEnabledState<public> : logic = false
    var<public> StartingPointsOverride : int = -1
    #

    # EVENTS
    LeaveEvent<public> : event() = event(){}
    DamagedEvent<public> : event(?agent) = event(?agent){}
    ZombieAttackStarted<public> : event(agent) = event(agent){}
    DownedEvent<public> : event() = event(){}
    DetectableStateChangedEvent<public> : event(logic) = event(logic){}
    PickedUpWeaponEvent<public> : event() = event(){}

    DamagedZombieEvent<public> : event(damage_result_custom) = event(damage_result_custom){}
    EliminatedZombieEvent<public> : event(agent) = event(agent){}
    EliminatedZombieByTrapEvent<public> : event(agent) = event(agent){}

    HurtScreenEvent : event() = event(){}

    RebuildBarrierEvent<public> : event() = event(){}


    # INITIALIZED FUNCTIOn
    OnBegin<public>()<suspends>:void=
        loop:
            if(GlobalWeaponManager[GetSession()], GlobalExfilManager[GetSession()], User.GetFortCharacter[].IsActive[]):
                break
            Sleep(0.1)
        Sleep(0.0)
        ReviveInteractable.Link()
        ThisUI := User.GetUIManager()
        ThisUI.ApplyUI(User)
        if(GetGameDifficulty() = game_difficulty.Hardcore):
            set AmmoDropChance = -1.0
        if(ThisWeaponManager:=GlobalWeaponManager[GetSession()]):
            ThisWeaponManager.TrackPlayerWeapon(User)
        if(IsWaveActive[], JoinTime > (GetWaveInfo[].StartTime+5.0)):
            spawn. PlayState(player_state.Spectating)
        else:
            spawn. PlayState(player_state.Active)
        race:
            LeaveEvent.Await()
            AutoHealLoop()

    # STATE MACHINE
    var<private> State<public> : player_state = player_state.Inactive
    StateChangedEvent<public> : event(player_state) = event(player_state){}
    CancelStateBranch<public> : event() = event(){}

    PlayState<public>(InState:player_state)<suspends>:void=
        if(IsGameState[game_state.Paused]):
            ThisUI:=User.GetUIManager()
            ThisUI.PausedScreenCover.Show()
            ThisUI.EndGame_WavesSurvived_Text.SetText("game paused")
            ThisUI.EndGame_WavesSurvived_Text.Show()
            PausedState := State
            set State = player_state.Paused
            if. ThisFortChar := User.GetFortCharacter[] then. ThisFortChar.PutInStasis(stasis_args{AllowFalling:=false, AllowTurning:=false, AllowEmotes:=false})
            AwaitGameState(game_state.InGame)
            set State = PausedState
            if. ThisFortChar := User.GetFortCharacter[] then. ThisFortChar.ReleaseFromStasis()
            ThisUI.PausedScreenCover.Hide()
            ThisUI.EndGame_WavesSurvived_Text.Hide()
            return
        PreviousState := State
        set State = InState
        Print("New player state: {State.ToString()}")
        StateChangedEvent.Signal(InState)
        CancelStateBranch.Signal()
        branch:
            case(InState):
                player_state.Active =>
                    case(GetGameDifficulty()):
                        game_difficulty.Easy =>
                            set MaxHealth = 110.0
                        game_difficulty.Classic =>
                        game_difficulty.Hardcore => 
                            ChangeRegenSettings(?Delay:=4.0, ?TickTime:=1.25)
                            set PerkSlots = 4
                    if(PreviousState = player_state.Spectating, FortChar:=User.GetFortCharacter[], SpawnPos:=GlobalPlayerManager[GetSession()].SpawnPads[UniqueIndex]):
                        User.Respawn(SpawnPos.GetTransform().Translation+vector3{Z:=77.0}, rotation{})
                        Sleep(0.5)
                        if(not FortChar.IsActive[]):
                            User.Respawn(SpawnPos.GetTransform().GetLocalVector(vector3{X:=50.0,Z:=77.0}), rotation{})
                            loop:
                                if(FortChar.IsActive[]):
                                    break
                                Sleep(0.25)
                    SetVulnerability(false)
                    User.GetUIManager().RevivingStartedAlertText.Hide()
                    # Show gameplay UI, allow movement and firing
                    if:
                        ThisPackage:=User.GetPlayerPackage[]
                        ThisFortChar:=User.GetFortCharacter[]
                        DownedLocation := ThisPackage.DownedProp.GetTransform()
                        ThisPackage.DownedProp.TeleportTo[vector3{}, rotation{}]
                    then:
                        if(PreviousState = player_state.Downed):
                            if. User.GetFortCharacter[].TeleportTo[DownedLocation.Translation+vector3{Z:=10.0}, DownedLocation.Rotation]
                            else if. ThisFortChar.TeleportTo[DownedLocation.Translation+vector3{X:=25.0}, rotation{}]
                            else if. ThisFortChar.TeleportTo[DownedLocation.Translation+vector3{X:=-25.0}, rotation{}]
                            else if. ThisFortChar.TeleportTo[DownedLocation.Translation+vector3{Y:=-25.0}, rotation{}]
                            else if. ThisFortChar.TeleportTo[DownedLocation.Translation+vector3{Y:=-25.0}, rotation{}]
                            else if. ThisFortChar.TeleportTo[GlobalPlayerManager[GetSession()].SpawnPads[UniqueIndex].GetTransform().Translation, rotation{}]
                            else if. ThisFortChar.TeleportTo[vector3{X:=6513.132099,Y:=-4583.196645,Z:=571.052726}, rotation{}] #-> Last resort, tp to location in air in spawn
                        ThisPackage.DownedCamera.RemoveFrom(User)
                        ThisFortChar.ReleaseFromStasis()
                    set Health = MaxHealth
                    if(PreviousState <> player_state.Downed, WeaponManager:=GlobalWeaponManager[GetSession()]):
                        Sleep(0.1)
                        if(User.HasExfilledBefore[], CoolPistol:=weapon_definition_list.MonarchPistol.GetWeaponDefinition[]):
                            CoolPistol.GrantWeapon(User)
                        else if(ThisWeapon := GlobalPlayerManager[GetSession()].StartingWeapon.GetWeaponDefinition[]):
                            ThisWeapon.GrantWeapon(User)
                        else. Print("ERR: No valid starting weapon set!")
                        AwaitWave(1)
                        if(ThisPlayerManager:=GlobalPlayerManager[GetSession()]):
                            if(GameplayMenuEnabledState?):
                                ThisPlayerManager.Input_GameplayAdminPanel.Register(User)
                            else:
                                ThisPlayerManager.Input_GameplayAdminPanel.Unregister(User)
                            if(GetWaveCount() = 1):
                                SetPoints(if. StartingPointsOverride > -1  then. StartingPointsOverride else. ThisPlayerManager.StartingPoints)
                            else:
                                ThisStartingPoints := block{
                                    BasePointAmount := Int[LockLerp(ThisPlayerManager.StartingPoints*1.0, ThisPlayerManager.MaximumStartingPoints*1.0, (1.0*GetWaveCount())/20.0)]
                                    BasePointAmount-Mod[BasePointAmount, 50]
                                } or 15000
                                SetPoints(ThisStartingPoints)
                    SafeSleep(3.0)
                    SetVulnerability(true)

                    if(ThisFortChar:=User.GetFortCharacter[]):
                        PerkList:=GetPerkList()
                        loop:
                            CrouchTuple:=ThisFortChar.CrouchedEvent().Await()
                            if(CrouchTuple(1)?, var ClosestPerkMachine : tuple(perk_base, float) = (PerkList[0], Distance(PerkList[0].InteractiveComponent.GetTransform().Translation, ThisFortChar.GetTransform().Translation))):
                                for(ThisPerk:PerkList):
                                    ThisDistance := Distance(ThisPerk.InteractiveComponent.GetTransform().Translation, ThisFortChar.GetTransform().Translation)
                                    if(ThisDistance < ClosestPerkMachine(1)):
                                        set ClosestPerkMachine = (ThisPerk, ThisDistance)
                                if(ClosestPerkMachine(1) < 175.0):
                                    ClosestPerkMachine(0).GetSecretMoney(User)

                player_state.Spectating =>
                    ChangeToDefaultGameClass(User)
                    SetVulnerability(true)
                    if(PreviousState = player_state.Downed):
                        if:
                            ThisPackage:=User.GetPlayerPackage[]
                            DownedLocation := ThisPackage.DownedProp.GetTransform()
                            ThisPackage.DownedProp.TeleportTo[vector3{}, rotation{}]
                        then:
                            ThisPackage.DownedCamera.RemoveFrom(User)
                    # Eliminate the player
                    if(ThisFortChar:=User.GetFortCharacter[]):
                        Sleep(0.1)
                        ThisFortChar.ReleaseFromStasis()
                        ThisFortChar.SetVulnerability(true)
                        ThisFortChar.SetMaxHealth(1.0)
                        ThisFortChar.SetMaxShield(1.0)
                        ThisFortChar.Damage(999999.0)
                        ThisFortChar.Damage(999999.0)
                    User.RemoveWeapons()
                    UpdatePointsUI()

                player_state.Downed =>
                    SetTotalDowns(TotalDowns+1)
                    Print("Player: Downed state")
                    if(IsGameState[game_state.GameOver]):
                        # Downed when game is over, just show downed prop + tp player
                        if(ThisFortChar:=User.GetFortCharacter[]):
                            ThisFortChar.PutInStasis(stasis_args{AllowFalling:=false, AllowTurning:=false, AllowEmotes:=false})
                            if:
                                not GlobalExfilState[GetSession()]?
                                ThisPackage:=User.GetPlayerPackage[]
                                ThisPackage.DownedProp.TeleportTo[ThisFortChar.GetTransform().Translation+REVIVEOFFSET, ThisFortChar.GetTransform().Rotation]
                            then:
                                ThisPackage.DownedCamera.AddTo(User)
                                DragToNull()
                        Sleep(0.0)
                        CancelStateBranch.Signal()
                        Sleep(0.1)
                    DownedEvent.Signal()
                    defer{
                        if(DevicePackage:=User.GetPlayerPackage[]):
                            DevicePackage.HurtScreenCinematic.Play(User)
                            DevicePackage.HurtScreenCinematic.SetPlayRate(0.0)
                            DevicePackage.HurtScreenCinematic.SetPlaybackTime(12.0)
                    }
                    if(DevicePackage:=User.GetPlayerPackage[]):
                        DevicePackage.HurtScreenCinematic.SetPlaybackTime(12.0)
                        DevicePackage.HurtScreenCinematic.SetPlayRate(1.0)
                        DevicePackage.HurtScreenCinematic.Play(User)
                    if(ThisPlayerManager:=GlobalPlayerManager[GetSession()]):
                        ThisPlayerManager.KnockedStinger.Play(User)
                    PlayerList:=GetAgents(?OnlyActives:=false)
                    for(ThisPlayer:PlayerList):
                        if(ThisPlayer <> User):
                            spawn. ThisPlayer.GetUIManager().DisplayKnockedAlert(User)
                    # Teleport the downed prop to the player
                    if:
                        ThisPackage:=User.GetPlayerPackage[]
                        ThisFortChar:=User.GetFortCharacter[]
                        ThisPackage.DownedProp.TeleportTo[ThisFortChar.GetTransform().Translation+REVIVEOFFSET, ThisFortChar.GetTransform().Rotation]
                    then:
                        LastPlayerLocation := ThisFortChar.GetTransform()
                        ThisFortChar.PutInStasis(stasis_args{AllowFalling:=false, AllowTurning:=false, AllowEmotes:=false})
                        ReviveInteractable.SetLocation(transform{Translation:=ThisFortChar.GetTransform().Translation,Rotation:=rotation{}})
                        ReviveInteractable.SetEnabledState(false)
                        ThisPackage.DownedCamera.AddTo(User)
                        DragToNull()
                        if(IsSinglePlayer[]):
                            Print("Is single player")
                            # Only one player in game, do not down
                            if(SaviorSodaPerk := saveyoursoda_perk[GetPerk[perk_list.SaviorSoda]], ActivePerks.Find[SaviorSodaPerk]):
                                # Solo player has savior soda, let it handle revive
                                SaviorSodaPerk.HandleSoloRevive(User)
                                CancelStateBranch.Signal()
                                Sleep(0.1)
                            else:
                                # Solo player does not have savior soda, continue to next statement to end game
                                Print("no savior soda")
                        if(GetPlayers().Length <= 0, IsGameState[game_state.InGame]):
                            # No more players are remaining, end the game
                            spawn. SetGameState(game_state.GameOver)
                        else:
                            RemoveAllPerks()
                            var DownedIconProp : creative_prop = creative_prop{}
                            if(SpawnedDownedIcon := SpawnProp(GlobalPlayerManager[GetSession()].KnockedIcon, LastPlayerLocation.Translation+vector3{Z:=20.0}, LastPlayerLocation.Rotation)(0)?):
                                set DownedIconProp = SpawnedDownedIcon
                            defer{
                                ReviveInteractable.SetLocation(transform{Translation:=vector3{},Rotation:=rotation{}})
                                ReviveInteractable.SetEnabledState(false)
                                DownedIconProp.Dispose()
                            }
                            # Ready the revive interactable 
                            ReviveInteractable.SetText("Revive player")
                            for(ThisPlayerInfo:GetPlayers(?OnlyActives:=false)):
                                if(ThisPlayerInfo.User <> User):
                                    ReviveInteractable.SetEnabledState(true, ?InAgent:=option{ThisPlayerInfo.User})
                            ThisUI:=User.GetUIManager()
                            ThisUI.DownedHealthUI.SetProgress(100.0)
                            ThisUI.DownedHealthUI.Show()
                            # Track downed time left
                            RevivedState:=race:
                                block:
                                    AwaitGameState(game_state.GameOver)
                                    false
                                block:
                                    AwaitWaveEnd()
                                    true
                                block:
                                    ThisReviver:=ReviveInteractable.InteractedEvent.Await()
                                    if(ThisPlayerInfo:=ThisReviver.GetPlayerInfo[]):
                                        ThisPlayerInfo.SetTotalRevives(ThisPlayerInfo.TotalRevives+1)
                                    true
                                block:
                                    var DownedTimeLeft : float = DownedTime
                                    loop:
                                        spawn. ThisUI.DownedHealthUI.ChangeProgressOverTime(DownedTimeLeft, 0.0)
                                        StartTime:=GetSimulationElapsedTime()
                                        RevivingState:=race:
                                            block:
                                                SafeSleep(DownedTimeLeft)
                                                false
                                            block:
                                                ReviveInteractable.InteractingEvent.Await()
                                                true
                                            block:
                                                loop:
                                                    ThisUI.DownedHealthSkull.Resize(vector2{X:=35.0,Y:=35.0}, Lerp(0.25, 0.45, (DownedTimeLeft-(GetSimulationElapsedTime()-StartTime))/DownedTime), ?InterpolationType:=interpolation_type.EaseInOutBack)
                                                    ThisUI.DownedHealthSkull.Resize(vector2{X:=26.0,Y:=26.0}, Lerp(0.25, 0.45, (DownedTimeLeft-(GetSimulationElapsedTime()-StartTime))/DownedTime), ?InterpolationType:=interpolation_type.EaseInOutBack)
                                                false
                                        if(RevivingState?):
                                            spawn. User.GetUIManager().DisplayRevivingAlert()
                                            ThisUI.DownedHealthUI.HaltProgressChanging()
                                            ThisUI.DownedHealthSkull.SetDesiredSize(vector2{X:=26.0,Y:=26.0})
                                            # Player is being revived, wait until no more players are reviving to start down timer again
                                            set DownedTimeLeft -= GetSimulationElapsedTime()-StartTime
                                            loop:
                                                ReviveInteractable.StoppedInteractingEvent.Await()
                                                if(ReviveInteractable.GetInteractingPlayers().Length <= 0):
                                                    User.GetUIManager().RevivingStartedAlertText.Hide()
                                                    break
                                        else:
                                            # Exceeded downed time
                                            break
                                    false
                            ThisUI.DownedHealthUI.Hide()
                            ThisUI.DownedHealthUI.HaltProgressChanging()
                            ReviveInteractable.SetLocation(transform{Translation:=vector3{},Rotation:=rotation{}})
                            ReviveInteractable.SetEnabledState(false)
                            User.GetUIManager().RevivingStartedAlertText.Hide()
                            DownedIconProp.Dispose()
                            if(DevicePackage:=User.GetPlayerPackage[]):
                                DevicePackage.HurtScreenCinematic.Play(User)
                                DevicePackage.HurtScreenCinematic.SetPlayRate(0.0)
                                DevicePackage.HurtScreenCinematic.SetPlaybackTime(12.0)
                            if(RevivedState?):
                                # Player was revived, set to active
                                spawn. PlayState(player_state.Active)
                            else if(not IsGameState[game_state.GameOver]):
                                Print("Exceeded downed time, player spectating")
                                spawn. PlayState(player_state.Spectating)

                player_state.OutOfSession =>
                    if(ThisPlayerPackage:=User.GetPlayerPackage[]):
                        set ThisPlayerPackage.IsInUse = false
                    # Cleanup after player has left the session
                    PlayerList:=GetAgents(?OnlyActives:=false)
                    for(ThisPlayer:PlayerList):
                        ThisUI:=ThisPlayer.GetUIManager()
                        ThisUI.Scoreboard.RemoveStats(User)
                        ThisUI.GlobalPoints.RemovePlayer(User)
                        ThisUI.MainMenuUI.RemovePlayerSlot(User)
                        ReviveInteractable.SetEnabledState(false)
                        ReviveInteractable.Unlink()
                    PlayersLeft := GetAgents()
                    if(PlayersLeft.Length <= 0):
                        SetGameState(game_state.GameOver)
                _ =>
            Sleep(0.0)
        CancelStateBranch.Await()

    # ONE-OFF LOOPING FUNCTIONS
    AutoHealLoop<private>()<suspends>:void=
        loop:
            race:
                block:
                    LeaveEvent.Await()
                    return
                DamagedEvent.Await()
                block:
                    SafeSleep(RegenDelay)
                    loop:
                        Heal(RegenAmount)
                        SafeSleep(RegenTickrate)
            Sleep(0.0)

    ActivateHurtScreen<private>()<suspends>:void=
        HurtScreenEvent.Signal()
        if(DevicePackage:=User.GetPlayerPackage[], ThisPlayerManager:=GlobalPlayerManager[GetSession()]):
            race:
                block:
                    PercentHealthLeft := (Health / MaxHealth) * 100.0
                    HurtPulseTime := 5.0 - ((Health / MaxHealth) * 5.0)
                    if(PercentHealthLeft <= 35.0):
                        ThisPlayerManager.HeartbeatStinger.Play(User)
                        DevicePackage.HurtScreenCinematic.SetPlaybackTime(9.0)
                    else if(PercentHealthLeft < 50.0):
                        DevicePackage.HurtScreenCinematic.SetPlaybackTime(6.0)
                    else if(PercentHealthLeft < 75.0):
                        DevicePackage.HurtScreenCinematic.SetPlaybackTime(3.0)
                    else if(PercentHealthLeft < 100.0):
                        DevicePackage.HurtScreenCinematic.SetPlaybackTime(0.0)
                    else{Print("~No hurt screen playback time to set to!")}
                    DevicePackage.HurtScreenCinematic.Play(User)
                    DevicePackage.HurtScreenCinematic.SetPlayRate(0.0)
                    SafeSleep(HurtPulseTime)
                    DevicePackage.HurtScreenCinematic.SetPlayRate(1.0)
                    DevicePackage.HurtScreenCinematic.Play(User)
                    SafeSleep(2.0)
                    DevicePackage.HurtScreenCinematic.SetPlayRate(0.0)
                    DevicePackage.HurtScreenCinematic.SetPlaybackTime(12.0)
                    ThisPlayerManager.HeartbeatStinger.Stop(User)
                HurtScreenEvent.Await()
                DownedEvent.Await()

    # MANIPULATORS
    # - Points -
    IncreasePoints<public>(Amount:int, ?IgnoreMultiplier:logic=false, ?ColorOverride:?color=false):void=
        if(Amount > 0):
            PointAmount := block{
                if:
                    not IgnoreMultiplier?
                    ThisAmount := Int[Amount*PointMultiplier]
                then:
                    ThisAmount
                else:
                    Amount
            } * if. IsDoublePoints[] then. 2 else. 1
            set Points += PointAmount
            if(Points > POINTCAP):
                set Points = POINTCAP
            UpdatePointsUI(?PointIncrement:=option. PointAmount)
        
    SetPoints<public>(Amount:int):void=
        set Points = Amount
        UpdatePointsUI()
        
    UpdatePointsUI<public>(?PointIncrement:?int=false):void=
        if(ThisFortChar := User.GetFortCharacter[]):
            ThisFortChar.SetMaxShield(Points*1.0)
            ThisFortChar.SetShield(Points*1.0)
        ThisUI:=User.GetUIManager()
        ThisUI.SetPointsText(Points, ?PointIncrement:=PointIncrement)
        ThisUI.Scoreboard.UpdateStats(User)
        PlayerList:=GetAgents(?OnlyActives:=false)
        for(ThisPlayer:PlayerList):
            if(ThisPlayer <> User):
                ThisPlayer.GetUIManager().GlobalPoints.UpdateGlobalPoints(User)

    SetPointMultiplier<public>(Magnitude:float):void=
        set PointMultiplier = Magnitude

    BuyItem<public>(Amount:int, ?PlayPurchaseSound:logic=true):logic=
        if(Points >= Amount):
            set Points -= Amount
            UpdatePointsUI()
            if(PlayPurchaseSound?, BuyStinger:=GlobalPlayerManager[GetSession()].PurchaseStinger):
                BuyStinger.Play(User)
            return true
        else:
            spawn. User.GetUIManager().InsufficientEffect_Points()
            return false

    # - Perks -
    AddPerk<public>(InPerk:perk_base, ?IgnoreLimit:logic=false)<transacts><decides>:void=
        CheckPerkLimit[] or IgnoreLimit?
        not ActivePerks.Find[InPerk]
        set ActivePerks += array{InPerk}

    RemovePerk<public>(InPerk:perk_base):void=
        if(ActivePerks.Find[InPerk]):
            set ActivePerks = ActivePerks.RemoveAllElements(InPerk)
            InPerk.Remove(User)
            User.GetUIManager().PerkUICollection.RemovePerk(InPerk)

    RemoveAllPerks<public>():void=
        for(ThisPerk:ActivePerks):
            ThisPerk.Remove(User)
        set ActivePerks = array{}

    SetAmmoDropChance<public>(InChance:float):void=
        set AmmoDropChance = InChance

    SetDetectableState<public>(InState:logic):void=
        set DetectableByZombies = InState
        DetectableStateChangedEvent.Signal(InState)

    ExcludeInteractableType<public>(Type:interactable_type):void=
        if(not ExcludedInteractableTypes.Find[Type]):
            set ExcludedInteractableTypes += array{Type}

    AllowInteractableType<public>(Type:interactable_type):void=
        NewExcludingList := ExcludedInteractableTypes.RemoveAllElements(Type)
        set ExcludedInteractableTypes = NewExcludingList

    SetDamageMultiplier<public>(Amount:float, ?AffectBase:logic=true):void=
        if(DamageMultiplierBoostActive? and AffectBase?):
            set SavedDamageMultiplier = Amount
        else:
            set DamageMultiplier = Amount

    GetDamageMultiplier<public>(?GetBase:logic=true):float=
        if(DamageMultiplierBoostActive? and GetBase?):
            return SavedDamageMultiplier
        else:
            return DamageMultiplier

    SetDamageBoostState<public>(InState:logic):void=
        set DamageMultiplierBoostActive = InState
        if(InState?):
            set SavedDamageMultiplier = DamageMultiplier
        else:
            set DamageMultiplier = SavedDamageMultiplier

    SetVulnerability<public>(InState:logic):void=
        set VulnerableState = InState

    # - Health - 
    Damage<public>(Amount:float, ?Instigator:?agent=false):void=
        if(IsActive[], VulnerableState?):
            set Health -= Amount
            Print("PlayerInfo: Damage event, health at {Health}")
            DamagedEvent.Signal(Instigator)
            if. ThisPlayerManager:=GlobalPlayerManager[GetSession()] then. ThisPlayerManager.HurtEffectStinger.Play(User)
            spawn. ActivateHurtScreen()
            if(Health <= 0.0):
                spawn. PlayState(player_state.Downed)

    DamageShield<public>(Amount:float):void=
        set ShieldAmount -= Amount
        if(ThisShield:=ShieldClass?):
            ThisShield.Stinger_ShieldHit.Play(User)
            spawn. User.GetUIManager().ShieldHealthBar.ChangeProgressOverTime(0.15, (ShieldAmount/ThisShield.ProtectionAmount)*100.0)
            if(ShieldAmount <= 0.0):
                ThisShield.Unequip(User)

    Heal<public>(Amount:float):void=
        if(not IsDowned[]):
            set Health += Amount
            if(Health > MaxHealth):
                set Health = MaxHealth

    SetMaxHealth<public>(Amount:float):void=
        set MaxHealth = Amount

    ChangeRegenSettings<public>(?Delay:float=-1.0, ?TickTime:float=-1.0, ?Amount:float=-1.0):void=
        if(Delay >= 0.0):
            set RegenDelay = Delay
        else:
            set RegenDelay = AUTOHEAL_DELAY
        if(TickTime >= 0.0):
            set RegenTickrate = TickTime
        else:
            set RegenTickrate = AUTOHEAL_TICKTIME
        if(Amount >= 0.0):
            set RegenAmount = Amount
        else:
            set RegenAmount = AUTOHEAL_AMOUNT

    SetDownedTime<public>(MaximumTime:float):void=
        set DownedTime = MaximumTime

    SetMoveState<public>(InState:logic):void=
        if(InState?):
            if. User.GetFortCharacter[].PutInStasis(stasis_args{AllowTurning:=false, AllowFalling:=false, AllowEmotes:=false})
        else:
            if. User.GetFortCharacter[].ReleaseFromStasis()

    # GAME STATS
    SetTotalKills<public>(InAmount:int):void=
        set TotalKills = InAmount
        PlayerList:=GetAgents(?OnlyActives:=false)
        for(ThisPlayer:PlayerList):
            ThisPlayer.GetUIManager().Scoreboard.UpdateStats(User)

    SetTotalRevives<public>(InAmount:int):void=
        set TotalRevives = InAmount
        PlayerList:=GetAgents(?OnlyActives:=false)
        for(ThisPlayer:PlayerList):
            ThisPlayer.GetUIManager().Scoreboard.UpdateStats(User)

    SetTotalDowns<public>(InAmount:int):void=
        set TotalDowns = InAmount
        PlayerList:=GetAgents(?OnlyActives:=false)
        for(ThisPlayer:PlayerList):
            ThisPlayer.GetUIManager().Scoreboard.UpdateStats(User)

    # GETTERS
    GetFortCharacter<public>()<transacts><decides>:fort_character=
        return User.GetFortCharacter[]

    # CHECKERS
    IsActive<public>()<transacts><decides>:void=
        State = player_state.Active

    IsSpectating<public>()<transacts><decides>:void=
        State = player_state.Spectating

    IsDowned<public>()<transacts><decides>:void=
        State = player_state.Downed

    IsInSession<public>()<transacts><decides>:void=
        State <> player_state.OutOfSession

    IsDetectableByZombies<public>()<transacts><decides>:void=
        DetectableByZombies?
        State = player_state.Active

    CheckPerkLimit<public>()<transacts><decides>:void=
        ActivePerks.Length < PerkSlots

    CheckShieldHit<public>(AttackerPosition:vector3)<transacts><decides>:void=
        ShieldClass?
        not User.GetFortCharacter[].IsLookingAt[AttackerPosition, 0.48]

    # MISC
    DragToNull<public>()<suspends>:void=
        if(ThisFortChar:=User.GetFortCharacter[], ThisFortChar.IsActive[], ThisPlayerManager:=GlobalPlayerManager[GetSession()]):
            set LastMapLocation = option. ThisFortChar.GetTransform()
            var Tries : int = 0
            loop:
                if. ThisFortChar.TeleportTo[ThisPlayerManager.NullPos+vector3{Z:=100.0*Tries}, rotation{}] then. break
                set Tries += 1
                Sleep(0.0)
                
    DragIntoMap<public>()<suspends>:void=
        if(ThisFortChar:=User.GetFortCharacter[], ThisFortChar.IsActive[], ThisPlayerManager:=GlobalPlayerManager[GetSession()]):
            SpawnPos := GetSpawnPosition()
            if(ThisLastLoc:=LastMapLocation?):
                var Tries : int = 0
                loop:
                    if. ThisFortChar.TeleportTo[ThisLastLoc.Translation, ThisLastLoc.Rotation] then. break
                    set Tries += 1
                    Sleep(0.0)

    GetSpawnPosition<public>():transform=
        if(ThisSpawnPad:=GlobalPlayerManager[GetSession()].SpawnPads[UniqueIndex]):
            TS:=ThisSpawnPad.GetTransform()
            transform{
                Translation:=TS.Translation+vector3{Z:=78.0}
                Rotation:=TS.Rotation
            }
        spawn. SetGameState(game_state.GameOver)
        SetMoveState(false)
        transform{}