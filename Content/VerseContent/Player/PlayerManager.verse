using { /Fortnite.com/Devices }
using { /Verse.org/Simulation }
using { /UnrealEngine.com/Temporary/SpatialMath }
using { /Fortnite.com/Characters }
using { /Fortnite.com/FortPlayerUtilities }
using { Spawning }
using { Weapons }
using { InteractiveModule }
using { UI }

IsSinglePlayer<public>()<transacts><decides>:void=
    GetAgents(?OnlyActives:=false).Length = 1

GetPlayerCount<public>()<transacts>:int=
    GetAgents(?OnlyActives:=false).Length

GetPlayers<public>(?OnlyActives:logic=true)<transacts>:[]player_info=
    PlayerList := block{
        ThisPlayerManager:=GlobalPlayerManager[GetSession()]
        for:
            ThisPlayer->MaybePI:ThisPlayerManager.PlayerList
            PlayerInfo:=MaybePI?
            not OnlyActives? or PlayerInfo.IsActive[] or not IsGameState[game_state.InGame]
            PlayerInfo.IsInSession[]
        do:
            PlayerInfo
    } or array{}

GetAgents<public>(?OnlyActives:logic=true)<transacts>:[]agent=
    PlayerList := block{
        ThisPlayerManager:=GlobalPlayerManager[GetSession()]
        for:
            ThisPlayer->MaybePI:ThisPlayerManager.PlayerList
            PlayerInfo:=MaybePI?
            not OnlyActives? or PlayerInfo.IsActive[] or not IsGameState[game_state.InGame]
            PlayerInfo.IsInSession[]
        do:
            PlayerInfo.User
    } or array{}

var GlobalPlayerManager<public> : weak_map(session, Player_Manager) = map{}

Player_Manager<public> := class(Game_Manager):

    @editable 
    SpawnPads<public> : []player_spawner_device = array{}

    @editable
    NullPos : vector3 = vector3{}

    # GAME SETTINGS

    @editable
    StartingPoints<public> : type{_X:int where _X >= 0} = 500

    @editable
    MaximumStartingPoints<public> : type{_X:int where _X >= 0} = 15000

    @editable
    StartingWeapon<public> : weapon_definition_list = weapon_definition_list.RangerPistol

    # INPUT DEVICES
    @editable
    InteractiveInput<public> : input_trigger_device = input_trigger_device{}

    @editable
    ScoreboardAndPauseInput<public> : input_trigger_device = input_trigger_device{}

    # CINEMATICS
    @editable
    PauseCinematic<public> : cinematic_sequence_device = cinematic_sequence_device{}

    @editable
    GameOverCinematic<public> : cinematic_sequence_device = cinematic_sequence_device{}

    # CAMERAS
    @editable
    FirstPersonCamera<public> : gameplay_camera_first_person_device = gameplay_camera_first_person_device{}

    @editable
    OrbitFirstPersonCamera<public> : gameplay_camera_orbit_device = gameplay_camera_orbit_device{}

    # EXFIL
    @editable
    EXFIL_SuccessCinematic<public> : cinematic_sequence_device = cinematic_sequence_device{}

    @editable
    EXFIL_ActiveMusic<public> : audio_player_device = audio_player_device{}

    # MISC
    @editable
    PurchaseStinger<public> : audio_player_device = audio_player_device{}

    @editable
    HurtEffectStinger<public> : audio_player_device = audio_player_device{}

    @editable
    KnockedStinger<public> : audio_player_device = audio_player_device{} 

    @editable
    HeartbeatStinger<public> : audio_player_device = audio_player_device{}

    @editable
    KnockedIcon<public> : creative_prop_asset = DefaultCreativePropAsset

    @editable
    GameOverMusic<public> : audio_player_device = audio_player_device{}

    @editable
    RoundDevice<public> : round_settings_device = round_settings_device{}

    # PLAYER PACKAGES
    @editable
    PlayerPackages : []player_package = array{}

    @editable
    Credits : []string = array{}

    @editable
    PatchNotes : []string = array{}

    @editable
    PatchVersion : string = "V?"

    var PlayerList : [agent]?player_info = map{}

    OnBegin<override>():void=
        GetPlayspace().PlayerRemovedEvent().Subscribe(OnPlayerLeave)
        GetPlayspace().PlayerAddedEvent().Subscribe(InitPlayer)
        (super:)OnBegin() # -> link game manager
        if. set GlobalPlayerManager[GetSession()] = Self
        #spawn. SetGameState(game_state.InGame)
        spawn. RespawnPlayersLoop()
        ScoreboardAndPauseInput.PressedEvent.Subscribe(HandleScoreboardAndPause)
        for(ThisPlayer : GetPlayspace().GetPlayers()):
            InitPlayer(ThisPlayer)
        for(ThisSpawner:SpawnPads):
            ThisSpawner.SpawnedEvent.Subscribe(InitPlayer)

    InitPlayer<private>(InAgent:agent):void=
        if(not PlayerList[InAgent]?):
            if(ThisPackage := GetOpenPlayerPackage[InAgent]):
                if(IsGameState[game_state.Paused] and IsWaveActive[]):
                    # Game is paused but player joined at valid time so unpause
                    HandleScoreboardAndPause(InAgent)
                set ThisPackage(0).IsInUse = option{InAgent}
                NewPlayerInfo : player_info = player_info{User:=InAgent, JoinTime:=GetSimulationElapsedTime(), UniqueIndex:=ThisPackage(1), Points:=500}
                if. set PlayerList[InAgent] = option. NewPlayerInfo
                else. set PlayerList = ConcatenateMaps(PlayerList, map{InAgent=>option. NewPlayerInfo})
                spawn. NewPlayerInfo.OnBegin()
                spawn. LinkInteractive(InAgent)
            else if(ThisPlayer:=player[InAgent]):
                # Should not happen unless max player count is somehow exceeded
                ThisPlayer.SendToLobby()
        else if(ThisPlayerInfo:=PlayerList[InAgent]?):
            case(GetGameState()):
                game_state.Menu =>
                    spawn. ThisPlayerInfo.PlayState(player_state.Inactive)
                game_state.InGame =>
                    if(IsWaveActive[]):
                        spawn. ThisPlayerInfo.PlayState(player_state.Spectating)
                    else:
                        spawn. ThisPlayerInfo.PlayState(player_state.Active)
                game_state.Paused =>
                    if(IsWaveActive[]):
                        spawn. ThisPlayerInfo.PlayState(player_state.Spectating)
                    else:
                        spawn. ThisPlayerInfo.PlayState(player_state.Active)
                game_state.GameOver =>
                    spawn. ThisPlayerInfo.PlayState(player_state.Inactive)

    OnPlayerLeave(InAgent:agent):void=
        if(ThisPlayerInfo:=InAgent.GetPlayerInfo[]):
            spawn. ThisPlayerInfo.PlayState(player_state.OutOfSession)
            var UpdatedPlayerMap:[agent]?player_info = map{}
            # Concatenate Keys from ExampleMap into NewMap, excluding ElementToRemove
            for (ThisPlayer -> MaybePI : PlayerList):
                if(PlayerInfo := MaybePI?, PlayerInfo <> ThisPlayerInfo):
                    set UpdatedPlayerMap = ConcatenateMaps(UpdatedPlayerMap, map{ThisPlayer => option. PlayerInfo})
            set PlayerList = UpdatedPlayerMap

    TogglePauseGame<private>(Agent:agent):void=
        if(IsGameState[game_state.Paused]):
            PauseCinematic.GoToEndAndStop()
            spawn. SetGameState(game_state.InGame)
        else if(IsGameState[game_state.InGame]):
            PauseCinematic.Play()
            spawn. SetGameState(game_state.Paused)

    HandleScoreboardAndPause<private>(Agent:agent):void=
        if(IsGameState[game_state.InGame] or IsGameState[game_state.Paused]):
            if(ThisPlayerInfo:=Agent.GetPlayerInfo[]):
                ThisUI:=Agent.GetUIManager()
                if(ThisPlayerInfo.ScoreboardState?):
                    set ThisPlayerInfo.ScoreboardState = false
                    if(IsSinglePlayer[]):
                        spawn. SetGameState(game_state.InGame)
                    ThisUI.Scoreboard.Hide()
                    ThisUI.PartsUI.Hide()
                else:
                    set ThisPlayerInfo.ScoreboardState = true
                    if(IsSinglePlayer[]):
                        spawn. SetGameState(game_state.Paused)
                    ThisUI.Scoreboard.Show()
                    ThisUI.PartsUI.Show()

    <#
    HandleScoreboardAndPause1<private>(Agent:agent)<suspends>:void=
        if(ThisPlayerInfo:=Agent.GetPlayerInfo[]):
            ThisInputType:=race:
                block:
                    Sleep(0.0)
                    var InputType:int=0
                    loop:
                        ReleaseTuple:=if. not ScoreboardAndPauseInput.IsHeld[Agent] then. (Agent, 0.0) else. ScoreboardAndPauseInput.ReleasedEvent.Await()
                        if(ReleaseTuple(0) = Agent):
                            if(ReleaseTuple(1) >= 3.0):
                                set InputType = 1
                            break
                    InputType
                block:
                    loop:
                        PressedAgent:=ScoreboardAndPauseInput.PressedEvent.Await()
                        if(PressedAgent = Agent):
                            break
                    -1
            if(ThisInputType<>-1):
                if(ThisInputType=0):
                    ThisUI:=Agent.GetUIManager()
                    # Handle as scoreboard input
                    if(ThisPlayerInfo.ScoreboardState?):
                        set ThisPlayerInfo.ScoreboardState = false
                        ThisUI.Scoreboard.Hide()
                        ThisUI.PartsUI.Hide()
                    else:
                        set ThisPlayerInfo.ScoreboardState = true
                        ThisUI.Scoreboard.Show()
                        ThisUI.PartsUI.Show()
                else:
                    # Handle as pause input#>

    RespawnPlayersLoop<private>()<suspends>:void=
        AwaitGameState(game_state.InGame)
        loop:
            AwaitWaveEnd()
            ThisPlayerList:=GetPlayers(?OnlyActives:=false)
            for(ThisPlayer:ThisPlayerList):
                if(ThisPlayer.IsSpectating[]):
                    Print("Respawning player")
                    spawn. ThisPlayer.PlayState(player_state.Active)
                else if(ThisFortChar:=ThisPlayer.User.GetFortCharacter[]):
                    if(not ThisFortChar.IsActive[]):
                        spawn. ThisPlayer.PlayState(player_state.Active)

    GetOpenPlayerPackage<public>(InAgent:agent)<transacts><decides>:tuple(player_package, int)=
        var MaybeOpenPackage : tuple(?player_package, int) = (false, -1)
        for:
            Index->ThisPackage:PlayerPackages
            not ThisPackage.IsInUse?
            not MaybeOpenPackage(0)?
        do:
            set ThisPackage.IsInUse = option. InAgent
            set MaybeOpenPackage = (option{ThisPackage}, Index)
        (MaybeOpenPackage(0)?, MaybeOpenPackage(1))
        
    LinkInteractive<public>(InAgent:agent)<suspends>:void=
        ThisUI:=InAgent.GetUIManager()
        if(ThisPlayerInfo:=InAgent.GetPlayerInfo[], ThisPlayerManager:=GlobalPlayerManager[GetSession()], ThisPackage:=InAgent.GetPlayerPackage[]):
            var CurrentInteractable : ?Interactable = false
            loop:
                AwaitGameState(game_state.InGame)
                LoopState:=race:
                    block:
                        AwaitGameState(game_state.Paused)
                        true
                    block:
                        loop:
                            MaybeInteractable:=FindDesiredInteractable(InAgent)
                            if(ThisInteractable:=MaybeInteractable?):
                                if(ThisInteractable.IsNotEqual[CurrentInteractable?] or not CurrentInteractable?):
                                    # New Interactable is shown
                                    set CurrentInteractable = option{ThisInteractable}
                                    if. ThisPackage.InteractionIcon.TeleportTo[ThisInteractable.GetSpriteTransform()]
                                    InAgent.GetInteractableChangedEvent().Signal(option{ThisInteractable})
                                    spawn. ThisInteractable.OnInput_Shown(InAgent)
                                ThisUI.SetInteractionText(ThisInteractable.GetInteractionText(InAgent))
                                DescriptionTuple := ThisInteractable.GetDescription(InAgent)
                                ThisUI.SetDescription(DescriptionTuple(0), DescriptionTuple(1), ThisInteractable.ParentDevice)
                            else if(PreviousInteractable:=CurrentInteractable?):
                                # No valid interactables
                                set CurrentInteractable = false
                                if. ThisPackage.InteractionIcon.TeleportTo[vector3{}, rotation{}]
                                InAgent.GetInteractableChangedEvent().Signal(false)
                                InAgent.GetInterruptInteractionEvent().Signal()
                                InAgent.GetUIManager().SetInteractionText("")
                                ThisUI.HideDescription()
                                spawn. PreviousInteractable.OnInput_Hidden(InAgent)
                            Sleep(0.1)
                        true
                    block:
                        InteractiveInputSignal:event()=event(){}
                        HeldEvent:event()=event(){}
                        loop:
                            if(ThisInteractable:=CurrentInteractable?, ThisInteractable.IsInteractableForPlayer[InAgent]):
                                HandleInteractableInput(InAgent, ThisInteractable, ThisPlayerManager)
                                InAgent.GetInterruptInteractionEvent().Signal()
                                ThisInteractable.OnInteractionStopped(InAgent)
                            Sleep(0.2)
                        true
                    block:
                        ThisPlayerInfo.LeaveEvent.Await()
                        false
                    block:
                        AwaitGameState(game_state.GameOver)
                        false
                if(not LoopState?):
                    break
                Sleep(0.0)
        else:
            Print("ERROR: Failed to get PlayerInfo or PlayerManager!")

    HandleInteractableInput<public>(InAgent:agent, CurrentInteractable:Interactable, ThisPlayerManager:Player_Manager)<suspends>:void=
        InteractiveInputSignal:event()=event(){}
        CompletedEvent:event()=event(){}
        branch:
            Sleep(0.1)
            CurrentInteractable.GetInteractionText(InAgent)
            ThisPlayerManager.InteractiveInput.AwaitPressedEvent(InAgent)
            InteractiveInputSignal.Signal()
            CurrentInteractable.OnInteractionStarted(InAgent)
            InAgent.GetUIManager().InteractableAnimation(CurrentInteractable)
            CurrentInteractable.Interact(InAgent)
            CompletedEvent.Signal()
        race:
            CompletedEvent.Await()
            InAgent.GetInterruptInteractionEvent().Await()
            block:
                InteractiveInputSignal.Await()
                Sleep(0.0)
                ThisPlayerManager.InteractiveInput.AwaitReleasedEvent(InAgent)
            loop:
                if(not CurrentInteractable.IsInteractableForPlayer[InAgent]):
                    break
                Sleep(0.0)
            loop:
                MaybeInteractable:=FindDesiredInteractable(InAgent)
                if(MaybeInteractable?.IsNotEqual[CurrentInteractable] or not MaybeInteractable?):
                    break
                Sleep(0.1)