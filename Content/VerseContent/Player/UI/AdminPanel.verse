using { /Fortnite.com/Devices }
using { /Verse.org/Simulation }
using { /UnrealEngine.com/Temporary/Diagnostics }
using { /UnrealEngine.com/Temporary/UI }
using { /UnrealEngine.com/Temporary/SpatialMath }
using { /Verse.org/Assets }
using { /Verse.org/Colors }
using { /Fortnite.com/UI }
using { /Fortnite.com/Characters }
using { Zombie }
using { Spawning }
using { Perks }
using { Powerups }

# Check on for admin panel
# MUST be false for public versions
GlobalAdminState<public> : logic = false

admin_panel_game<public> := class(admin_panel):

    SkipToWaveSelect : admin_setting_slider = admin_setting_slider{
        SettingText := "Skip to wave"
        SettingBackground := texture_block{DefaultImage:=Art.UIStuff.KennyAssets.Rectangle_Grey}
        SliderWidget := slider_regular{DefaultMinValue:=1.0, DefaultStepSize:=1.0, DefaultMaxValue:=256.0}
        InputBackground := texture_block{DefaultImage:=Art.UIStuff.KennyAssets.Square_Grey, DefaultDesiredSize:=vector2{X:=80.0,Y:=80.0}}
        InputIcon := texture_block{DefaultImage:=Art.UIStuff.KennyAssets.icon_circle, DefaultDesiredSize:=vector2{X:=32.0,Y:=32.0}}
        MinVal := 1.0
        MaxVal := 256.0
        StepSize := 1.0
        DefaultValue := 1.0
    }

    ToggleInvincibility : admin_setting_checkbox = admin_setting_checkbox{
        SettingText := "Toggle\ninvincibility"
        SettingBackground := texture_block{DefaultImage:=Art.UIStuff.KennyAssets.Rectangle_Grey}
        CheckboxBackground := texture_block{DefaultImage:=Art.UIStuff.KennyAssets.Square_Grey, DefaultDesiredSize:=vector2{X:=80.0,Y:=80.0}}
        OnIcon := texture_block{DefaultImage:=Art.UIStuff.KennyAssets.icon_checkmark, DefaultDesiredSize:=vector2{X:=36.0,Y:=36.0}}
        OffIcon := texture_block{DefaultImage:=Art.UIStuff.KennyAssets.icon_cross, DefaultDesiredSize:=vector2{X:=36.0,Y:=36.0}}
    }

    ManipulatePoints_Self : admin_setting_slider = admin_setting_slider{
        SettingText := "Set points"
        SettingBackground := texture_block{DefaultImage:=Art.UIStuff.KennyAssets.Rectangle_Grey}
        SliderWidget := slider_regular{DefaultMinValue:=0.0, DefaultStepSize:=100.0, DefaultMaxValue:=999999.0}
        InputBackground := texture_block{DefaultImage:=Art.UIStuff.KennyAssets.Square_Grey, DefaultDesiredSize:=vector2{X:=80.0,Y:=80.0}}
        InputIcon := texture_block{DefaultImage:=Art.UIStuff.KennyAssets.icon_circle, DefaultDesiredSize:=vector2{X:=32.0,Y:=32.0}}
        MinVal := 0.0
        MaxVal := 999999.0
        StepSize := 100.0
        DefaultValue := 500.0
    }

    EliminateAllZombies : admin_setting_button = admin_setting_button{
        SettingText:="Eliminate\nall zombies"
        SettingBackground := texture_block{DefaultImage:=Art.UIStuff.KennyAssets.Rectangle_Grey}
        InputBackground:=texture_block{DefaultImage:=Art.UIStuff.KennyAssets.Square_Grey, DefaultDesiredSize:=vector2{X:=80.0,Y:=80.0}}
        InputIcon:=texture_block{DefaultImage:=Art.UIStuff.KennyAssets.icon_circle, DefaultDesiredSize:=vector2{X:=32.0,Y:=32.0}}
    }

    <#
    ToggleFreezeZombies : admin_setting_checkbox = admin_setting_checkbox{
        SettingText := "Toggle freeze\nall zombies"
        SettingBackground := texture_block{DefaultImage:=Art.UIStuff.KennyAssets.Rectangle_Grey}
        CheckboxBackground := texture_block{DefaultImage:=Art.UIStuff.KennyAssets.Square_Grey, DefaultDesiredSize:=vector2{X:=80.0,Y:=80.0}}
        OnIcon := texture_block{DefaultImage:=Art.UIStuff.KennyAssets.icon_checkmark, DefaultDesiredSize:=vector2{X:=36.0,Y:=36.0}}
        OffIcon := texture_block{DefaultImage:=Art.UIStuff.KennyAssets.icon_cross, DefaultDesiredSize:=vector2{X:=36.0,Y:=36.0}}
    } #>

    SpawnZombies : admin_setting_slider = admin_setting_slider{
        SettingText := "Spawn zombies"
        SettingBackground := texture_block{DefaultImage:=Art.UIStuff.KennyAssets.Rectangle_Grey}
        SliderWidget := slider_regular{DefaultMinValue:=1.0, DefaultStepSize:=1.0, DefaultMaxValue:=20.0}
        InputBackground := texture_block{DefaultImage:=Art.UIStuff.KennyAssets.Square_Grey, DefaultDesiredSize:=vector2{X:=80.0,Y:=80.0}}
        InputIcon := texture_block{DefaultImage:=Art.UIStuff.KennyAssets.icon_circle, DefaultDesiredSize:=vector2{X:=32.0,Y:=32.0}}
        MinVal := 1.0
        MaxVal := 20.0
        StepSize := 1.0
        DefaultValue := 1.0
    }

    SpawnPowerupsSelection : admin_setting_dropbox = admin_setting_dropbox{
        InputBackground:=texture_block{DefaultImage:=Art.UIStuff.KennyAssets.Square_Grey, DefaultDesiredSize:=vector2{X:=80.0,Y:=80.0}}
        InputIcon:=texture_block{DefaultImage:=Art.UIStuff.KennyAssets.icon_circle, DefaultDesiredSize:=vector2{X:=32.0,Y:=32.0}}
        SettingText := "Spawn Powerup"
        Options := array{
            "Bonus Points"
            "Quickdraw"
            "Death Machine"
            "Nuke"
            "Instakill"
            "Double Points"
            "Max Ammo"
            "Firesale"
            "Incognito"
            "Perk Up"
        }
        XWidth := 350.0
        SettingBackground := texture_block{DefaultImage:=Art.UIStuff.KennyAssets.Rectangle_Grey}
    }

    GiveAllPerks : admin_setting_button = admin_setting_button{
        SettingText:="Get all perks"
        SettingBackground := texture_block{DefaultImage:=Art.UIStuff.KennyAssets.Rectangle_Grey}
        InputBackground:=texture_block{DefaultImage:=Art.UIStuff.KennyAssets.Square_Grey, DefaultDesiredSize:=vector2{X:=80.0,Y:=80.0}}
        InputIcon:=texture_block{DefaultImage:=Art.UIStuff.KennyAssets.icon_circle, DefaultDesiredSize:=vector2{X:=32.0,Y:=32.0}}
    }

    AutoCompleteExfilSteps : admin_setting_button = admin_setting_button{
        SettingText:="Skip exfil\nrequirements"
        SettingBackground := texture_block{DefaultImage:=Art.UIStuff.KennyAssets.Rectangle_Grey}
        InputBackground:=texture_block{DefaultImage:=Art.UIStuff.KennyAssets.Square_Grey, DefaultDesiredSize:=vector2{X:=80.0,Y:=80.0}}
        InputIcon:=texture_block{DefaultImage:=Art.UIStuff.KennyAssets.icon_circle, DefaultDesiredSize:=vector2{X:=32.0,Y:=32.0}}
    }

    ToggleZombieCageView : admin_setting_checkbox = admin_setting_checkbox{
        SettingText := "Toggle view\nzombie cage"
        SettingBackground := texture_block{DefaultImage:=Art.UIStuff.KennyAssets.Rectangle_Grey}
        CheckboxBackground := texture_block{DefaultImage:=Art.UIStuff.KennyAssets.Square_Grey, DefaultDesiredSize:=vector2{X:=80.0,Y:=80.0}}
        OnIcon := texture_block{DefaultImage:=Art.UIStuff.KennyAssets.icon_checkmark, DefaultDesiredSize:=vector2{X:=36.0,Y:=36.0}}
        OffIcon := texture_block{DefaultImage:=Art.UIStuff.KennyAssets.icon_cross, DefaultDesiredSize:=vector2{X:=36.0,Y:=36.0}}
    }

    ToggleDebugPrints : admin_setting_checkbox = admin_setting_checkbox{
        SettingText := "Toggle Debug Prints"
        SettingBackground := texture_block{DefaultImage:=Art.UIStuff.KennyAssets.Rectangle_Grey}
        CheckboxBackground := texture_block{DefaultImage:=Art.UIStuff.KennyAssets.Square_Grey, DefaultDesiredSize:=vector2{X:=80.0,Y:=80.0}}
        OnIcon := texture_block{DefaultImage:=Art.UIStuff.KennyAssets.icon_checkmark, DefaultDesiredSize:=vector2{X:=36.0,Y:=36.0}}
        OffIcon := texture_block{DefaultImage:=Art.UIStuff.KennyAssets.icon_cross, DefaultDesiredSize:=vector2{X:=36.0,Y:=36.0}}
    }

    ToggleSkibidMode : admin_setting_checkbox = admin_setting_checkbox{
        SettingText := "Toggle skibidi\nmode"
        SettingBackground := texture_block{DefaultImage:=Art.UIStuff.KennyAssets.Rectangle_Grey}
        CheckboxBackground := texture_block{DefaultImage:=Art.UIStuff.KennyAssets.Square_Grey, DefaultDesiredSize:=vector2{X:=80.0,Y:=80.0}}
        OnIcon := texture_block{DefaultImage:=Art.UIStuff.KennyAssets.icon_checkmark, DefaultDesiredSize:=vector2{X:=36.0,Y:=36.0}}
        OffIcon := texture_block{DefaultImage:=Art.UIStuff.KennyAssets.icon_cross, DefaultDesiredSize:=vector2{X:=36.0,Y:=36.0}}
    }

    block:
        set Settings = array{ToggleInvincibility, GiveAllPerks, EliminateAllZombies, AutoCompleteExfilSteps, SkipToWaveSelect, ManipulatePoints_Self, SpawnZombies, SpawnPowerupsSelection, ToggleZombieCageView, ToggleDebugPrints}
    
    HeaderText<override> : string = "Admin Panel - Game Menu"

    OnInit<override>(InAgent:agent):void=
        spawn. InputListen(InAgent)
        ExitButton.OnClick().Subscribe(CloseGamePanel)

    
    InputListen<private>(InAgent:agent)<suspends>:void=
        if(ThisPlayerInfo:=InAgent.GetPlayerInfo[], FortChar:=InAgent.GetFortCharacter[], ThisPlayerManager:=GlobalPlayerManager[GetSession()]):
            race:
                ThisPlayerInfo.LeaveEvent.Await()
                AwaitGameState(game_state.GameOver)
                sync:
                    loop:
                        SelectedWave := SkipToWaveSelect.ValueSetEvent.Await()
                        if(ThisWave := Int[SelectedWave]):
                            SkipToWave(ThisWave, ?UseIntermission:=true)
                    loop:
                        NewState := ToggleInvincibility.StateChangedEvent.Await()
                        if(NewState?):
                            ThisPlayerInfo.SetVulnerability(false)
                        else:
                            ThisPlayerInfo.SetVulnerability(true)
                    loop:
                        NewPoints := ManipulatePoints_Self.ValueSetEvent.Await()
                        if(PointAmount := Int[NewPoints]):
                            ThisPlayerInfo.SetPoints(PointAmount)
                    loop:
                        EliminateAllZombies.InputButton.OnClick().Await()
                        GetZombieList().Eliminate()
                    loop:
                        SpawnAmount := SpawnZombies.ValueSetEvent.Await()
                        if(ThisAmount := Int[SpawnAmount]):
                            NewSpawnGroup : spawn_group = spawn_group{Amount:=ThisAmount, SpawnsAtOnce:=5}
                            spawn. NewSpawnGroup.Spawn()
                    loop:
                        GiveAllPerks.InputButton.OnClick().Await()
                        InAgent.GrantAllPerks()
                    loop:
                        AutoCompleteExfilSteps.InputButton.OnClick().Await()
                        if(ThisExfilManager:=GlobalExfilManager[GetSession()]):
                            ThisExfilManager.ForceEnableExfil()
                    loop:
                        ZombieCageState:=ToggleZombieCageView.StateChangedEvent.Await()
                        if(ZombieCageState?):
                            set ThisPlayerInfo.LastMapLocation = option. FortChar.GetTransform()
                            if. FortChar.TeleportTo[vector3{X:=9644.836848,Y:=-5401.670257,Z:=-2583.441089}, rotation{}]
                        else:
                            ThisPlayerInfo.DragIntoMap()
                        ToggleZombieCageView.SetState(ZombieCageState, ?FireEvent:=false)
                    loop:
                        NewDPState:=ToggleDebugPrints.StateChangedEvent.Await()
                        set ThisPlayerManager.DPState = NewDPState
                        PlayerList:=GetAgents()
                        for(ThisPlayer:PlayerList):
                            ThisPlayer.GetUIManager().AdminPanel_Game.ToggleDebugPrints.SetState(NewDPState, ?FireEvent:=false)
                    #loop:
                    #    SkibidiState:=ToggleSkibidMode.StateChangedEvent.Await()
                    #    if(ThisZombieTools := GlobalZombieTools[GetSession()]):
                    #        ThisZombieTools.SetSkibidiState(SkibidiState)
                    loop:
                        SpawnPowerupsSelection.InputButton.OnClick().Await()
                        NewPowerupSelected := SpawnPowerupsSelection.GetSelectedOption()
                        FalseType : ?powerup_type = false
                        MaybePowerupType:=option. case(NewPowerupSelected):
                            "Bonus Points" => powerup_type.BonusPoints
                            "Quickdraw" => powerup_type.Quickdraw
                            "Death Machine" => powerup_type.DeathMachine
                            "Nuke" => powerup_type.Nuke
                            "Instakill" => powerup_type.Instakill
                            "Double Points" => powerup_type.DoublePoints
                            "Max Ammo" => powerup_type.MaxAmmo
                            "Firesale" => powerup_type.Firesale
                            "Incognito" => powerup_type.Incognito
                            "Perk Up" => powerup_type.Perk
                        if(ThisPowerup := MaybePowerupType?.ConvertToPowerup[]):
                            if(ThisFortChar:=InAgent.GetFortCharacter[]):
                                SpawnPos := ThisFortChar.GetTransform().GetLocalVector(vector3{X:=125.0})
                                ThisPowerup.Spawn(SpawnPos)
                        else:
                            Print("ERR: Unknown powerup name [{NewPowerupSelected}]")

    CloseGamePanel<private>(MSG:widget_message):void=
        MSG.Player.GetUIManager().SetGamePanelState(false)

admin_panel_menu<public> := class(admin_panel):

    ToggleGameplayMenu : admin_setting_checkbox = admin_setting_checkbox{
        SettingText := "Toggle\ngameplay menu"
        SettingBackground := texture_block{DefaultImage:=Art.UIStuff.KennyAssets.Rectangle_Grey}
        CheckboxBackground := texture_block{DefaultImage:=Art.UIStuff.KennyAssets.Square_Grey, DefaultDesiredSize:=vector2{X:=80.0,Y:=80.0}}
        OnIcon := texture_block{DefaultImage:=Art.UIStuff.KennyAssets.icon_checkmark, DefaultDesiredSize:=vector2{X:=36.0,Y:=36.0}}
        OffIcon := texture_block{DefaultImage:=Art.UIStuff.KennyAssets.icon_cross, DefaultDesiredSize:=vector2{X:=36.0,Y:=36.0}}
    }

    AutoUnlockDoors : admin_setting_checkbox = admin_setting_checkbox{
        SettingText := "Unlock all doors"
        SettingBackground := texture_block{DefaultImage:=Art.UIStuff.KennyAssets.Rectangle_Grey}
        CheckboxBackground := texture_block{DefaultImage:=Art.UIStuff.KennyAssets.Square_Grey, DefaultDesiredSize:=vector2{X:=80.0,Y:=80.0}}
        OnIcon := texture_block{DefaultImage:=Art.UIStuff.KennyAssets.icon_checkmark, DefaultDesiredSize:=vector2{X:=36.0,Y:=36.0}}
        OffIcon := texture_block{DefaultImage:=Art.UIStuff.KennyAssets.icon_cross, DefaultDesiredSize:=vector2{X:=36.0,Y:=36.0}}
    }

    AutoPowerMap : admin_setting_checkbox = admin_setting_checkbox{
        SettingText := "Start with power on"
        SettingBackground := texture_block{DefaultImage:=Art.UIStuff.KennyAssets.Rectangle_Grey}
        CheckboxBackground := texture_block{DefaultImage:=Art.UIStuff.KennyAssets.Square_Grey, DefaultDesiredSize:=vector2{X:=80.0,Y:=80.0}}
        OnIcon := texture_block{DefaultImage:=Art.UIStuff.KennyAssets.icon_checkmark, DefaultDesiredSize:=vector2{X:=36.0,Y:=36.0}}
        OffIcon := texture_block{DefaultImage:=Art.UIStuff.KennyAssets.icon_cross, DefaultDesiredSize:=vector2{X:=36.0,Y:=36.0}}
    } 

    StartingWave : admin_setting_slider = admin_setting_slider{
        InputBackground := texture_block{DefaultImage:=Art.UIStuff.KennyAssets.Square_Grey, DefaultDesiredSize:=vector2{X:=80.0,Y:=80.0}}
        InputIcon := texture_block{DefaultImage:=Art.UIStuff.KennyAssets.icon_circle, DefaultDesiredSize:=vector2{X:=32.0,Y:=32.0}}
        SettingText := "Starting wave\noverride"
        SettingBackground := texture_block{DefaultImage:=Art.UIStuff.KennyAssets.Rectangle_Grey}
        SliderWidget := slider_regular{DefaultMinValue:=1.0, DefaultStepSize:=1.0, DefaultMaxValue:=256.0}
        MinVal := 1.0
        MaxVal := 256.0
        StepSize := 1.0
        DefaultValue := 1.0
    }

    StartingPoints : admin_setting_slider = admin_setting_slider{
        InputBackground := texture_block{DefaultImage:=Art.UIStuff.KennyAssets.Square_Grey, DefaultDesiredSize:=vector2{X:=80.0,Y:=80.0}}
        InputIcon := texture_block{DefaultImage:=Art.UIStuff.KennyAssets.icon_circle, DefaultDesiredSize:=vector2{X:=32.0,Y:=32.0}}
        SettingText := "Starting points\noverride"
        SettingBackground := texture_block{DefaultImage:=Art.UIStuff.KennyAssets.Rectangle_Grey}
        SliderWidget := slider_regular{DefaultMinValue:=0.0, DefaultStepSize:=100.0, DefaultMaxValue:=999999.0}
        MinVal := 0.0
        MaxVal := 999999.0
        StepSize := 100.0
        DefaultValue := 500.0
    }

    SettingsPerRow<override> : int = 3

    <#
    ApplyPointsToAll : admin_setting_checkbox = admin_setting_checkbox{
        SettingText := "Points override\naffects all players"
        SettingBackground := texture_block{DefaultImage:=Art.UIStuff.KennyAssets.Rectangle_Grey}
        CheckboxBackground := texture_block{DefaultImage:=Art.UIStuff.KennyAssets.Square_Grey, DefaultDesiredSize:=vector2{X:=80.0,Y:=80.0}}
        OnIcon := texture_block{DefaultImage:=Art.UIStuff.KennyAssets.icon_checkmark, DefaultDesiredSize:=vector2{X:=36.0,Y:=36.0}}
        OffIcon := texture_block{DefaultImage:=Art.UIStuff.KennyAssets.icon_cross, DefaultDesiredSize:=vector2{X:=36.0,Y:=36.0}}
    }
        #>

    block:
        #set Settings = array{AutoPowerMap, StartingWave, StartingPoints}
        set Settings = array{ToggleGameplayMenu, AutoUnlockDoors, AutoPowerMap, StartingWave, StartingPoints}

    HeaderText<override> : string = "Admin Panel - Menu"

    OnInit<override>(InAgent:agent):void=
        ExitButton.OnClick().Subscribe(CloseMenuPanel)
        spawn. InputListen(InAgent)

    InputListen<private>(InAgent:agent)<suspends>:void=
        if(ThisPlayerInfo:=InAgent.GetPlayerInfo[], ThisPlayerManager:=GlobalPlayerManager[GetSession()]):
            race:
                ThisPlayerInfo.LeaveEvent.Await()
                AwaitGameState(game_state.InGame)
                sync:
                    loop:
                        ThisState:=ToggleGameplayMenu.StateChangedEvent.Await()
                        set ThisPlayerInfo.GameplayMenuEnabledState = ThisState
                    loop:
                        ThisState:=AutoUnlockDoors.StateChangedEvent.Await()
                        set ThisPlayerManager.AutoUnlockDoors = ThisState
                        PlayerList:=GetAgents()
                        for(ThisPlayer:PlayerList):
                            ThisPlayer.GetUIManager().AdminPanel_Menu.AutoUnlockDoors.SetState(ThisState, ?FireEvent:=false)
                    loop:
                        ThisState:=AutoPowerMap.StateChangedEvent.Await()
                        set ThisPlayerManager.AutoPowerMap = ThisState
                        PlayerList:=GetAgents()
                        for(ThisPlayer:PlayerList):
                            ThisPlayer.GetUIManager().AdminPanel_Menu.AutoPowerMap.SetState(ThisState, ?FireEvent:=false)
                    loop:
                        NewWave:=StartingWave.ValueSetEvent.Await()
                        if(ThisWave := Int[NewWave]):
                            set ThisPlayerManager.StartingWaveOverride = ThisWave
                            PlayerList:=GetAgents()
                            for(ThisPlayer:PlayerList):
                                ThisPlayer.GetUIManager().AdminPanel_Menu.StartingWave.SetValue(ThisWave*1.0)
                    loop:
                        NewPoints:=StartingPoints.ValueSetEvent.Await()
                        if(ThisValue:=Int[NewPoints], ThisValue > 0):
                            set ThisPlayerInfo.StartingPointsOverride = ThisValue
                        else:
                            set ThisPlayerInfo.StartingPointsOverride = -1

    CloseMenuPanel<private>(MSG:widget_message):void=
        Hide()

# Base class for a window/panel
# Inherit to add reactive widgets
admin_panel<public> := class:

    var Root : stack_box = stack_box{Orientation:=orientation.Vertical}
    var ContentRoot : stack_box = stack_box{Orientation:=orientation.Vertical}

    HeaderText : string = "Header"
    HeaderWidget : text_block = text_block{DefaultTextColor:=NamedColors.White}
    var<private> Spacing<public> : float = 50.0
    SettingsPerRow<public> : int = 4

    var<public> Settings<public> : []admin_setting = array{}
    var Separators<private> : []color_block = array{}

    var<private> InitState : logic = false

    ExitButton<public> : button_quiet = button_quiet{}

    block:               
        set Root = stack_box:
            Orientation:=orientation.Vertical
            Slots:=array:
                # Header
                stack_box_slot:
                    Widget:=overlay:
                        Slots:=array:
                            # Header BG
                            overlay_slot:
                                Widget:=color_block{DefaultColor:=NamedColors.Black, DefaultOpacity:=0.5, DefaultDesiredSize:=vector2{X:=1280.0,Y:=96.0}}
                            # Header display text
                            overlay_slot:
                                HorizontalAlignment:=horizontal_alignment.Left
                                VerticalAlignment:=vertical_alignment.Center
                                Padding:=margin{Left:=25.0}
                                Widget:=HeaderWidget
                            # Header exit button
                            overlay_slot:
                                HorizontalAlignment:=horizontal_alignment.Right
                                VerticalAlignment:=vertical_alignment.Fill
                                Widget:=overlay:
                                    Slots:=array:
                                        overlay_slot:
                                            HorizontalAlignment:=horizontal_alignment.Fill
                                            VerticalAlignment:=vertical_alignment.Fill
                                            Widget:=color_block{DefaultColor:=NamedColors.Red, DefaultDesiredSize:=vector2{X:=350.0,Y:=96.0}}
                                        overlay_slot:
                                            HorizontalAlignment:=horizontal_alignment.Center
                                            VerticalAlignment:=vertical_alignment.Center
                                            Widget:=texture_block{DefaultImage:=Art.UIStuff.KennyAssets.icon_cross, DefaultDesiredSize:=vector2{X:=40.0,Y:=40.0}}
                                        overlay_slot:
                                            HorizontalAlignment:=horizontal_alignment.Fill
                                            VerticalAlignment:=vertical_alignment.Fill
                                            Widget:=ExitButton
                # Body
                stack_box_slot:
                    Widget:=overlay:
                        Slots:=array:
                            # Body background
                            overlay_slot:
                                HorizontalAlignment:=horizontal_alignment.Fill
                                VerticalAlignment:=vertical_alignment.Fill
                                Widget:=overlay:
                                    Slots:=array:
                                        overlay_slot:
                                            HorizontalAlignment:=horizontal_alignment.Fill
                                            VerticalAlignment:=vertical_alignment.Fill
                                            Widget:=color_block{DefaultColor:=NamedColors.Grey, DefaultOpacity:=0.5, DefaultDesiredSize:=vector2{X:=1280.0,Y:=688.0}}
                                        overlay_slot:
                                            HorizontalAlignment:=horizontal_alignment.Left
                                            VerticalAlignment:=vertical_alignment.Fill
                                            Widget:=color_block{DefaultColor:=NamedColors.Black, DefaultOpacity:=0.5, DefaultDesiredSize:=vector2{X:=8.0,Y:=680.0}}
                                        overlay_slot:
                                            HorizontalAlignment:=horizontal_alignment.Right
                                            VerticalAlignment:=vertical_alignment.Fill
                                            Widget:=color_block{DefaultColor:=NamedColors.Black, DefaultOpacity:=0.5, DefaultDesiredSize:=vector2{X:=8.0,Y:=680.0}}
                                        overlay_slot:
                                            HorizontalAlignment:=horizontal_alignment.Center
                                            VerticalAlignment:=vertical_alignment.Bottom
                                            Widget:=color_block{DefaultColor:=NamedColors.Black, DefaultOpacity:=0.5, DefaultDesiredSize:=vector2{X:=1280.0,Y:=8.0}}
                            # Body content
                            overlay_slot:
                                HorizontalAlignment:=horizontal_alignment.Center
                                VerticalAlignment:=vertical_alignment.Center
                                Padding:=margin{Top:=25.0,Left:=25.0,Bottom:=25.0,Right:=25.0}
                                Widget:=ContentRoot

    CreateSeparator<public>()<transacts>:color_block=
        NewColorBlock:=color_block{DefaultDesiredSize:=vector2{X:=Spacing,Y:=Spacing}, DefaultOpacity:=0.0}
        set Separators += array. NewColorBlock
        return NewColorBlock

    # Adds a row of settings
    AddRow<public>(InSettings:[]admin_setting):void=
        var SettingsAdded : int = 0
        RowContent := for(Index:=0..((InSettings.Length-1)*2)):
            if(ThisSettingWidget:=InSettings[SettingsAdded], Mod[Index+1, 2] <> 0):
                Print("Adding setting - {ThisSettingWidget.SettingText}")
                set SettingsAdded += 1
                ThisSettingWidget.Root
            else:
                CreateSeparator()
        Print("Row settings: {InSettings.Length}")
        Print("Row content: {RowContent.Length}")
        NewRowWidget : stack_box_slot = stack_box_slot:
            HorizontalAlignment:=horizontal_alignment.Center
            VerticalAlignment:=vertical_alignment.Top
            Widget:=stack_box:
                Orientation:=orientation.Horizontal
                Slots:=for(ThisContent:RowContent):
                    stack_box_slot:
                        HorizontalAlignment:=horizontal_alignment.Center
                        VerticalAlignment:=vertical_alignment.Top
                        Widget:=ThisContent
        ContentRoot.AddWidget(NewRowWidget)
        NewSeparator : stack_box_slot = stack_box_slot:
            HorizontalAlignment:=horizontal_alignment.Center
            VerticalAlignment:=vertical_alignment.Top
            Widget:=CreateSeparator()
        ContentRoot.AddWidget(NewSeparator)

    GetCanvasSlot<public>():canvas_slot=
        return canvas_slot{
            Widget:=Root
            Anchors:=anchors{Minimum:=vector2{X:=0.5,Y:=0.5},Maximum:=vector2{X:=0.5,Y:=0.5}}
            Alignment:=vector2{X:=0.5,Y:=0.5}
            SizeToContent:=true
        }
                    
    Init<public><final>(InAgent:agent):void=
        if(not InitState?):
            set InitState = true
            for(ThisSetting:Settings):
                ThisSetting.Init()
            for(Index->ThisSetting:Settings):
                if(ModVal := Mod[Index, SettingsPerRow]):
                    
            if(Loops := Ceil[(Settings.Length*1.0)/(SettingsPerRow*1.0)]):
                for(ThisLoop:=0..Loops-1):
                    if(RowSettings := Settings.Slice[ThisLoop*(SettingsPerRow), (ThisLoop*(SettingsPerRow))+SettingsPerRow]):
                        AddRow(RowSettings)
                    else if(HangoffSettings := Settings.Slice[ThisLoop*(SettingsPerRow), Settings.Length]):
                        AddRow(HangoffSettings)
            HeaderWidget.SetText(StringToMessage(HeaderText))
            OnInit(InAgent)

    OnInit<public>(InAgent:agent):void=
        return

    Show<public>():void=
        Root.Show()

    Hide<public>():void=
        Root.Collapse()
                    
# Base class for admin setting widgets
admin_setting<public> := class:

    var<private> Root<public> : stack_box = stack_box{Orientation:=orientation.Vertical}

    SettingText<public> : string = "Setting"
    SettingBackground<public> : texture_block

    SettingTextBlock<private> : text_block = text_block{DefaultTextColor:=NamedColors.Black, DefaultJustification:=text_justification.Center}

    var<public> Spacing : float = 32.0

    block:
        set Root = stack_box:
            Orientation:=orientation.Vertical
            Slots:=array:
                # Setting text
                stack_box_slot:
                    HorizontalAlignment:=horizontal_alignment.Center
                    VerticalAlignment:=vertical_alignment.Fill
                    Widget:=overlay:
                        Slots:=array:
                            # BG
                            overlay_slot:
                                HorizontalAlignment:=horizontal_alignment.Fill
                                VerticalAlignment:=vertical_alignment.Fill
                                Widget:=SettingBackground
                            # Text
                            overlay_slot:
                                HorizontalAlignment:=horizontal_alignment.Center
                                VerticalAlignment:=vertical_alignment.Center
                                Padding:=margin{Left:=10.0,Right:=10.0,Top:=10.0,Bottom:=10.0}
                                Widget:=SettingTextBlock
                # Separator
                stack_box_slot:
                    HorizontalAlignment:=horizontal_alignment.Center
                    VerticalAlignment:=vertical_alignment.Center
                    Widget:=color_block{DefaultOpacity:=0.0, DefaultDesiredSize:=vector2{X:=Spacing,Y:=Spacing}}

    Init<public>():void=
        SettingTextBlock.SetText(StringToMessage(SettingText))
        OnInit()

    OnInit<public>():void=return

    Collapse<public>():void=
        Root.Collapse()

    Show<public>():void=
        Root.Show()

# A basic button input setting
admin_setting_button<public> := class(admin_setting):

    InputButton<public> : button_quiet = button_quiet{}

    InputBackground<public> : texture_block
    InputIcon<public> : texture_block

    InteractedEvent<public> : event() = event(){}

    OnInit<override>():void=
        NewWidget:=stack_box_slot:
            HorizontalAlignment:=horizontal_alignment.Center
            VerticalAlignment:=vertical_alignment.Center
            Widget:=overlay:
                Slots:=array:
                    overlay_slot:
                        HorizontalAlignment:=horizontal_alignment.Fill
                        VerticalAlignment:=vertical_alignment.Fill
                        Widget:=InputBackground
                    overlay_slot:
                        HorizontalAlignment:=horizontal_alignment.Center
                        VerticalAlignment:=vertical_alignment.Center
                        Widget:=InputIcon
                    overlay_slot:
                        HorizontalAlignment:=horizontal_alignment.Fill
                        VerticalAlignment:=vertical_alignment.Fill
                        Widget:=InputButton
        Root.AddWidget(NewWidget)

# A checkbox setting
admin_setting_checkbox<public> := class(admin_setting):

    CheckboxBackground<public> : texture_block

    OnIcon<public> : texture_block
    OffIcon<public> : texture_block

    ChangeStateButton<public> : button_quiet = button_quiet{}

    DefaultState<public> : logic = false
    var<private> CurrentState<public> : logic = false

    var<private> InitState : logic = false

    StateChangedEvent<public> : event(logic) = event(logic){}

    OnInit<override>():void=
        SetState(DefaultState, ?FireEvent:=false)
        CheckBoxWidget:=stack_box_slot:
            HorizontalAlignment:=horizontal_alignment.Center
            VerticalAlignment:=vertical_alignment.Bottom
            Widget:=overlay:
                Slots:=array:
                    # BG
                    overlay_slot:
                        HorizontalAlignment:=horizontal_alignment.Center
                        VerticalAlignment:=vertical_alignment.Center
                        Widget:=CheckboxBackground
                    # State on
                    overlay_slot:
                        HorizontalAlignment:=horizontal_alignment.Center
                        VerticalAlignment:=vertical_alignment.Center
                        Widget:=OnIcon
                    # State off
                    overlay_slot:
                        HorizontalAlignment:=horizontal_alignment.Center
                        VerticalAlignment:=vertical_alignment.Center
                        Widget:=OffIcon
                    # State button
                    overlay_slot:
                        HorizontalAlignment:=horizontal_alignment.Fill
                        VerticalAlignment:=vertical_alignment.Fill
                        Widget:=ChangeStateButton
        Root.AddWidget(CheckBoxWidget)
        if(not InitState?):
            set InitState = true
            ChangeStateButton.OnClick().Subscribe(ChangeState)

    ChangeState<private>(MSG:widget_message):void=
        SetState(logic{not CurrentState?})
        
    SetState<public>(InState:logic, ?FireEvent:logic=true):void=
        set CurrentState = InState
        if. FireEvent? then. StateChangedEvent.Signal(CurrentState)
        if(CurrentState?):
            OffIcon.Hide()
            OnIcon.Show()
        else:
            OnIcon.Hide()
            OffIcon.Show()

admin_setting_slider<public> := class(admin_setting):

    InputButton<public> : button_quiet = button_quiet{}
    SliderWidget<public> : slider_regular = slider_regular{}

    MinVal<public> : float
    MaxVal<public> : float
    StepSize<public> : float

    DefaultValue<public> : float = 0.0

    InputBackground<public> : texture_block
    InputIcon<public> : texture_block

    ValueSetEvent<public> : event(float) = event(float){}

    var<private> InitState : logic = false

    OnInit<override>():void=
        SliderWidget.SetMinValue(MinVal)
        SliderWidget.SetMaxValue(MaxVal)
        SliderWidget.SetStepSize(StepSize)
        SliderWidget.SetValue(DefaultValue)
        if(not InitState?):
            set InitState = true
            InputButton.OnClick().Subscribe(ValueSet)
        Root.AddWidget(
            stack_box_slot:
                HorizontalAlignment:=horizontal_alignment.Center
                VerticalAlignment:=vertical_alignment.Center
                Widget:=SliderWidget
        )
        Root.AddWidget(
            stack_box_slot:
                Widget:=color_block{DefaultOpacity:=0.0, DefaultDesiredSize:=vector2{X:=Spacing,Y:=Spacing}}
        )
        Root.AddWidget(
            stack_box_slot:
                HorizontalAlignment:=horizontal_alignment.Center
                VerticalAlignment:=vertical_alignment.Center
                Widget:=overlay:
                    Slots:=array:
                        overlay_slot:
                            HorizontalAlignment:=horizontal_alignment.Fill
                            VerticalAlignment:=vertical_alignment.Fill
                            Widget:=InputBackground
                        overlay_slot:
                            HorizontalAlignment:=horizontal_alignment.Center
                            VerticalAlignment:=vertical_alignment.Center
                            Widget:=InputIcon
                        overlay_slot:
                            HorizontalAlignment:=horizontal_alignment.Fill
                            VerticalAlignment:=vertical_alignment.Fill
                            Widget:=InputButton
        )

    GetValue<public>():float= 
        SliderWidget.GetValue()

    SetValue<public>(InValue:float):void=
        SliderWidget.SetValue(InValue)

    ValueSet<private>(MSG:widget_message):void=
        ValueSetEvent.Signal(SliderWidget.GetValue())

admin_setting_dropbox<public> := class(admin_setting_button):
    
    Options<public> : []string

    XWidth<public> : float
    SelectedOptionBackground<public> : texture_block = texture_block{DefaultImage:=Art.UIStuff.KennyAssets.Rectangle_Grey, DefaultDesiredSize:=vector2{X:=286.0,Y:=64.0}}

    DropdownIconBackground<public> : texture_block = texture_block{DefaultImage:=Art.UIStuff.KennyAssets.Square_Grey, DefaultDesiredSize:=vector2{X:=64.0,Y:=64.0}}
    DropdownIcon_Expand<public> : texture = Art.UIStuff.KennyAssets.Arrow_S
    DropdownIcon_Collapse<public> : texture = Art.UIStuff.KennyAssets.Arrow_N
    DropdownIcon_Size<public> : vector2 = vector2{X:=38.0,Y:=26.0}

    SelectedColor<public> : color = NamedColors.Green

    var DropboxWidget<private> : ?dropbox_widget = false

    CollapseOnSelection<public> : logic = true

    OnInit<override>():void=
        NewDropbox : dropbox_widget = dropbox_widget{
            XWidth := XWidth
            SelectedOptionBackground :=SelectedOptionBackground
            DropdownIconBackground := DropdownIconBackground
            DropdownIcon_Expand := DropdownIcon_Expand
            DropdownIcon_Collapse := DropdownIcon_Collapse
            DropdownIcon_Size := DropdownIcon_Size
            SelectedColor := SelectedColor
            DropdownIcon_StateTexture := texture_block{DefaultImage:=DropdownIcon_Collapse, DefaultDesiredSize:=DropdownIcon_Size}
        }
        set DropboxWidget = option{NewDropbox}
        Root.AddWidget(
            stack_box_slot:
                HorizontalAlignment:=horizontal_alignment.Center
                VerticalAlignment:=vertical_alignment.Center
                Widget:=NewDropbox.Create(Options)
        )
        Root.AddWidget(
            stack_box_slot:
                Widget := color_block{DefaultOpacity:=0.0, DefaultDesiredSize:=vector2{X:=Spacing,Y:=Spacing}}
        )
        NewButtonSlot:=stack_box_slot:
            HorizontalAlignment:=horizontal_alignment.Center
            VerticalAlignment:=vertical_alignment.Center
            Widget:=overlay:
                Slots:=array:
                    overlay_slot:
                        HorizontalAlignment:=horizontal_alignment.Fill
                        VerticalAlignment:=vertical_alignment.Fill
                        Widget:=InputBackground
                    overlay_slot:
                        HorizontalAlignment:=horizontal_alignment.Center
                        VerticalAlignment:=vertical_alignment.Center
                        Widget:=InputIcon
                    overlay_slot:
                        HorizontalAlignment:=horizontal_alignment.Fill
                        VerticalAlignment:=vertical_alignment.Fill
                        Widget:=InputButton
        Root.AddWidget(NewButtonSlot)

    # Fetches the selected option
    GetSelectedOption<public>():string=
        return DropboxWidget?.GetSelectedItemName[] or "?"

    # Waits until an item is selected from the dropbox, NOT listening for setting activation
    AwaitSelection<public>()<suspends>:string=
        if(ThisDropbox:=DropboxWidget?):
            ThisDropbox.SelectedEvent.Await()
        else:
            Sleep(0.0)
            return "?"