using { /Verse.org/Assets }
using { /UnrealEngine.com/Temporary/SpatialMath }
using { /UnrealEngine.com/Temporary/UI }
using { /Verse.org/Colors }
using { /Verse.org/Simulation }

# WIP UI widget! #
# A string_block with multi-line functionality #
dialog_block<public> := class:

    WidgetControl<private> : dynamic_widget = dynamic_widget{}
    GetRootWidget():widget=
        return WidgetControl.GetRootWidget()

    var LineStackBox : stack_box = stack_box{Orientation:=orientation.Vertical}
    var StringBlocks : []string_block = array{}

    var Font<public>:string

    var<private> BackplateTexture:?texture=false
    BackplateWidget<private> : texture_block = texture_block{DefaultImage:=Art.UIStuff.Fonts.InteractionText.Numbers.Number_0}
    var<private> BackplateOffsets : vector2 = vector2{}
    var<private> BackplateFixedMode : logic = false

    var<private> CurrentText : string = ""

    var<private> CurrSize : float = 30.0
    var SpacingMultiplier<private> : float = 0.5
    var CurrSpacerSize<private> : float = 1.0
    var CurrSizeRatio<private> : vector2 = vector2{X:=1.0,Y:=1.0}
    var<private> CurrColor : color = color{}

    var<private> AutoWrapTextAtLength : float = -1.0

    Create<public>(Text:string, Position:vector2, Size:float, ?SizeRatio:vector2=vector2{X:=1.0,Y:=1.0}, ?SpacingSize:float:=0.25, ?SpacerSize:float:=1.0, ?SandwichedWidget:?widget=false, ?Alignment:vector2=vector2{}, ?AutoAnchor:logic=false):canvas_slot=
        set CurrSize = Size
        set SpacingMultiplier = SpacingSize
        set CurrSpacerSize = SpacerSize+1.0
        set CurrSizeRatio = SizeRatio
        SetText(Text)
        defer{
            if(ThisTexture:=BackplateTexture?):
                SetBackplate(ThisTexture)
                ShowBackplate()
            else:
                HideBackplate()
        }
        ThisSandwichedWidget := SandwichedWidget? or overlay{}
        LineBlockOverlay:=overlay{
            Slots:=array:
                overlay_slot{
                    Widget:=BackplateWidget
                    HorizontalAlignment:=horizontal_alignment.Fill
                    VerticalAlignment:=vertical_alignment.Fill
                }
                overlay_slot{
                    Widget:=ThisSandwichedWidget
                    HorizontalAlignment:=horizontal_alignment.Center
                    VerticalAlignment:=vertical_alignment.Center
                }
                overlay_slot{
                    Widget:=LineStackBox
                    HorizontalAlignment:=horizontal_alignment.Center
                    VerticalAlignment:=vertical_alignment.Center
                }
        }
        return WidgetControl.Create(LineBlockOverlay, Position, ?Alignment:=Alignment, ?AutoAnchor:=AutoAnchor)
    
    # Sets the text to the given string
    #   LineSeparator - The character/substring that represents a line break in this text
    SetText(InText:string, ?LineSeparator:string="\n"):void=
        if(InText=CurrentText):
            return
        var TextLines : []string = array{}
        var NewLineIndexes : []int = if. InText.FindSubstring[LineSeparator] then. array{0} else. array{}
        loop:
            if(LineEndIndex:=InText.FindSubstring[LineSeparator, ?DesiredAppearance:=TextLines.Length]):
                #Print("Found \n character")
                if(ThisLineText:=InText.Slice[NewLineIndexes[TextLines.Length], LineEndIndex]):
                    #Print("Slicing text line with text [{ThisLineText}]")
                    set TextLines += array{ThisLineText}
                    set NewLineIndexes += array{LineEndIndex+1}
                else:
                    #Print("Failed slice 1")
                    break
            else if(RemainingText:=InText.Slice[NewLineIndexes[TextLines.Length]+LineSeparator.Length-1, InText.Length], RemainingText.Length > 0):
                #Print("Setting remaining text")
                set TextLines += array{RemainingText}
                break
            else:
                #Print("Failed slice 2")
                set TextLines = array{InText}
                break
        #Print("InText: {InText}")
        if(InText = ""):
            #for(ThisStringBlock:StringBlocks):
            #    LineStackBox.RemoveWidget(ThisStringBlock.GetWidget())
            set StringBlocks = array{}
            set CurrentText = ""
            HideBackplate()
            LineStackBox.Hide()
            return
        else if(TextLines.Length <> StringBlocks.Length):
            Diff:=TextLines.Length-StringBlocks.Length
            #Print("Length diff: {Diff}")
            if(Diff > 0):
                for(DiffIndex:=0..Diff-1):
                    NewLine_StringBlock : string_block = string_block{Font:=Font}
                    NewWidget:=NewLine_StringBlock.Create("", vector2{}, CurrSize, ?SizeRatio:=CurrSizeRatio).Widget
                    LineStackBox.AddWidget(
                        stack_box_slot{
                            Widget:=NewLine_StringBlock.GetWidget()
                        }
                    )
                    set StringBlocks += array{NewLine_StringBlock}
            else:
                for(ThisStringBlock:StringBlocks):
                    ThisStringBlock.Hide(?DoCollapse:=true)
                <#
                WidgetsRemaining := (StringBlocks.Length-WidgetsToRemove)
                var TempStringBlocks : []string_block = array{}
                for(ThisLineIndex := WidgetsToRemove..StringBlocks.Length):
                    if(ThisStringBlock:=StringBlocks[ThisLineIndex]):
                        set TempStringBlocks += array{ThisStringBlock}
                set StringBlocks = TempStringBlocks  #>
        AllStringBlocks := GetVisibleLines()
        for(ThisStringBlock:AllStringBlocks):
            ThisStringBlock.Show()
        LineStackBox.Show()
        var LineIndex : int = 0
        for(ThisStringBlock:StringBlocks):
            if(ThisLineText:=TextLines[LineIndex]):
                ThisStringBlock.SetText(ThisLineText)
            set LineIndex += 1
        set CurrentText = InText
        ShowBackplate()
        ResizeBackplate(BackplateOffsets)

    AddText(InText:string, ?SeperateCharacters:logic=false, ?AddDelay:float=0.25, ?StartingSize:float=-1.0, ?EndSize:float=-1.0, ?OffsetSizeChangeInterpolation:interpolation_type=interpolation_type.Linear)<suspends>:void=
        set CurrentText += InText
        TextLines:=GetTextLines(InText)
        if(SeperateCharacters?):
            for(ThisLine:TextLines):
                NewStringBlock:=AddTextLine(ThisLine, ?SizeOverride:=StartingSize)
                var CharIndex : int = 0
                var HitPunctuation : logic = false
                for(ThisChar:ThisLine):
                    if(ThisChar <> ' '):
                        if(ThisChar = '!' or ThisChar = '.'):
                            NewStringBlock.MoveTextSize((CharIndex, CharIndex+1), EndSize, AddDelay, ?InterpolationType:=OffsetSizeChangeInterpolation)
                            set HitPunctuation = true
                        else:
                            NewStringBlock.MoveTextSize((CharIndex, CharIndex+1), EndSize, AddDelay, ?InterpolationType:=OffsetSizeChangeInterpolation)
                    else:
                        NewStringBlock.SetTextSize((CharIndex, CharIndex+1), EndSize)
                        if(HitPunctuation?):
                            ResizeBackplate(BackplateOffsets)
                            Sleep(0.8)
                            set HitPunctuation = false
                    set CharIndex += 1
                    ResizeBackplate(BackplateOffsets)
                    ShowBackplate()
                if(HitPunctuation?):
                    Sleep(0.8)
        else:
            for(ThisLine:TextLines):
                AddTextLine(ThisLine, ?SizeOverride:=StartingSize)
    
    AddTextLine(InText:string, ?SizeOverride:float=-1.0):string_block=
        var ThisSize : float = CurrSize
        if(SizeOverride > -1.0):
            set ThisSize = SizeOverride
        NewLine_StringBlock : string_block = string_block{Font:=Font}
        NewWidget:=NewLine_StringBlock.Create(InText, vector2{}, ThisSize, ?SizeRatio:=CurrSizeRatio).Widget
        LineStackBox.AddWidget(
            stack_box_slot{
                Widget:=NewLine_StringBlock.GetWidget()
            }
        )
        set StringBlocks += array{NewLine_StringBlock}
        return NewLine_StringBlock

    GetTextLine(DialogLine:int)<transacts><decides>:string_block=
        return StringBlocks[DialogLine]

    Resize(DesiredSize:float):void=
        for(ThisStringBlock:StringBlocks):
            ThisStringBlock.SetTextSize(DesiredSize)
        set CurrSize = DesiredSize
        ResizeBackplate(BackplateOffsets)

    MoveResizeEvent:event()=event(){}
    MoveResize(DesiredSize:float, Time:float, ?InterpolationType:interpolation_type=interpolation_type.Linear)<suspends>:void=
        OriginalSize:=CurrSize
        MoveResizeEvent.Signal()
        Winner:=race:
            block:
                if(Time > 0.0, StepMax := Int[30*Time]):
                    for(Step := 1..StepMax):
                        StepPercent := (Step*1.0)/(StepMax*1.0)
                        StepSize:=Lerp(OriginalSize, DesiredSize, StepPercent.ApplyInterpolation(InterpolationType))
                        Resize(StepSize)
                        Sleep(0.0)
                else:
                    Resize(DesiredSize)
                1
            block:
                MoveResizeEvent.Await()
                -1
        Resize(DesiredSize)

    GetTextLines(InText:string):[]string=
        var ThisText : string = InText
        var TextLines : []string = array{}
        var NewLineIndexes : []int = array{0}
        var NewLines : int = 0
        if(FirstNewLineIndex:=InText.FindSubstring["\n", ?DesiredAppearance:=0], FirstNewLineIndex = 0):
            set NewLines += 1
            if(ClippedText:=ThisText.Slice[1, ThisText.Length]):
                set ThisText = ClippedText
        loop:
            if(LineEndIndex:=ThisText.FindSubstring["\n", ?DesiredAppearance:=NewLines]):
                if(ThisLineText:=ThisText.Slice[NewLineIndexes[TextLines.Length], LineEndIndex]):
                    Print(ThisLineText)
                    set NewLineIndexes += array{LineEndIndex+1}
                    set NewLines += 1
                    if(ThisLineText.Length > 0):
                        set TextLines += array{ThisLineText}
                else:
                    break
            else:
                break
        if(RemainingText:=ThisText.Slice[NewLineIndexes[TextLines.Length], ThisText.Length], RemainingText.Length > 0):
            set TextLines += array{RemainingText}
        return TextLines

    GetScreenSize():vector2=
        var TotalScreenSize : vector2 = vector2{}
        var HighestXValue : float = 0.0
        for(ThisStringBlock:StringBlocks):
            ThisScreenSize:=ThisStringBlock.GetTextScreenSize()
            set TotalScreenSize = (TotalScreenSize + vector2{Y:=ThisScreenSize.Y})
            if(ThisScreenSize.X > HighestXValue):
                set HighestXValue = ThisScreenSize.X
        set TotalScreenSize = vector2{X:=HighestXValue, Y:=TotalScreenSize.Y}
        return TotalScreenSize

    SetBackplate(InTexture:texture):void=
        BackplateWidget.SetImage(InTexture)
        ResizeBackplate(BackplateOffsets)

    BackplateResizeEvent:event()=event(){}
    MoveBackplateSize<public><final>(DesiredOffset:vector2, Time:float, ?InterpolationType:interpolation_type=interpolation_type.Linear)<suspends>:void=
        OriginalSize:=BackplateOffsets
        BackplateResizeEvent.Signal()
        Winner:=race:
            block:
                if(Time > 0.0, StepMax := Int[30*Time]):
                    for(Step := 1..StepMax):
                        StepPercent := (Step*1.0)/(StepMax*1.0)
                        StepSize:=Lerp(BackplateOffsets, DesiredOffset, StepPercent.ApplyInterpolation(InterpolationType))
                        ResizeBackplate(StepSize)
                        Sleep(0.0)
                else:
                    ResizeBackplate(DesiredOffset)
                1
            block:
                BackplateResizeEvent.Await()
                -1
        ResizeBackplate(DesiredOffset)

    ResizeBackplate<public><final>(Offsets:vector2):void=
        set BackplateOffsets = Offsets
        if(CurrentText.Length > 0, CurrSize > 0.0):
            if(BackplateFixedMode?):
                BackplateWidget.SetDesiredSize(Offsets)
                OnBackplateResized(Offsets)
            else:
                ScreenSize:=GetScreenSize()
                BackplateWidget.SetDesiredSize(Offsets+vector2{X:=ScreenSize.X/2.0,Y:=ScreenSize.Y})
                OnBackplateResized(Offsets+vector2{X:=ScreenSize.X,Y:=ScreenSize.Y})
        else:
            BackplateWidget.SetDesiredSize(vector2{})
            OnBackplateResized(vector2{})

    OnBackplateResized(BackplateSize:vector2):void=
        return

    SetBackplateFixedState(State:logic):void=
        set BackplateFixedMode = State
        ResizeBackplate(BackplateOffsets)

    ShowBackplate():void=
        if(CurrentText.Length > 0):
            if(BackplateTexture?):
                BackplateWidget.Show()
            OnBackplateVisibilityChanged(true)

    HideBackplate():void=
        BackplateWidget.Hide()
        OnBackplateVisibilityChanged(false)

    OnBackplateVisibilityChanged(VisibleState:logic):void=
        return

    SetAutoWrapLength(InMinimum:float):void=
        set AutoWrapTextAtLength = InMinimum
        if(AutoWrapTextAtLength > -1.0):
            SetText(AutoWrapText())

    AutoWrapText():string=
        var WrappedText : string = ""
        TextLines:=GetTextLines(CurrentText)

        return WrappedText

    Show():void=
        if(CurrentText.Length > 0):
            WidgetControl.Show()

    Hide():void=
        WidgetControl.Hide()

    GetVisibleLines():[]string_block=
        return StringBlocks.Slice[0, CurrentText.FindSubstrings("\n").Length+1] or array{}