using { /Fortnite.com/UI }
using { /UnrealEngine.com/Temporary/UI }
using { /UnrealEngine.com/Temporary/SpatialMath }
using { /Verse.org/Simulation }
using { /Verse.org/Assets }
using { /Verse.org/Colors }

<#
CreateUI(Player : custom_player)<suspends> : canvas =
    DropBox := dropbox_class{XWidth := 350.0}
    Canvas := canvas:
        Slots := array:
            canvas_slot:
                Widget := DropBox.Create(array{"Slot one","Slot Two", "Slot Three"})
                Anchors := anchors{Minimum := vector2{X := 0.5, Y := 0.5}, Maximum := vector2{X := 0.5, Y := 0.5}}
                Alignment := vector2{X:=0.5,Y:=0.0}
    Player.Agent.ApplyCanvas(Canvas,true)
    Sleep(10.0)

    DropBox.ClearCheckboxLoops.Signal
    Player.Agent.RemoveCanvas(Canvas)

    return Canvas
    #>

dropbox_widget<public> := class:

    XWidth<public> : float 
    STM<localizes>(Message : string) : message = "{Message}"
    var SelectedOption : int = -1
    var DefaultSelectedOption : int = 0

    GetSelectedItemName<public>()<decides><transacts>: string =
        return Items[SelectedOption]
    
    GetSelectedIndex<public>():int =
        return SelectedOption

    ClearCheckboxLoops : event() = event(){}

    # Use this to await an item call, then put into a case statement to filter selection types
    SelectedEvent<public> : event(string) = event(string){}

    var DropBoxAssets : []overlay = array{}
    var Items : []string = array{}
    var DropboxBackgrounds<private> : []texture_block = array{}

    # Dropdown selected option box
    SelectedOptionBackground<public> : texture_block
    DisplayMessageTextBox<public> : text_block = text_block{DefaultTextColor:=NamedColors.Black, DefaultJustification:=text_justification.Center}
    var SelectedOption_PreText<public> : string = ""
    var SelectedOption_PostText<public> : string = ""

    SelectedColor<public> : color

    # Dropdown Icon
    DropdownIconBackground<public> : texture_block
    DropdownIcon_StateTexture<public> : texture_block = texture_block{DefaultImage:=Art.UIStuff.DefaultTexture}
    DropdownIcon_Expand<public> : texture
    DropdownIcon_Collapse<public> : texture
    DropdownIcon_Size<public> : vector2

    var CreatedState : logic = false

    CollapseOnSelection<public> : logic = true

    Create<public>(Content : []string): widget =
        if(CreatedState?):
            Remove()
        set CreatedState = true
        set Items = Content
        var StackOverlays : []overlay = array{}
        DropDownButton : button_quiet = button_quiet{}
        spawn:
            AwaitExpandPress(DropDownButton)

        # Set dropdown icon to default (collapsed)
        DropdownIcon_StateTexture.SetImage(DropdownIcon_Collapse)

        set StackOverlays += array:
            overlay:
                Slots := array:
                    overlay_slot:
                        Widget := stack_box:
                            Orientation := orientation.Vertical
                            Slots := array:
                                stack_box_slot:
                                    HorizontalAlignment := horizontal_alignment.Center
                                    VerticalAlignment := vertical_alignment.Center
                                    Widget := stack_box:
                                        Orientation:=orientation.Horizontal
                                        Slots := array:
                                            stack_box_slot:
                                                HorizontalAlignment := horizontal_alignment.Fill
                                                VerticalAlignment := vertical_alignment.Fill
                                                Widget := overlay:
                                                    Slots:=array:
                                                        overlay_slot:
                                                            HorizontalAlignment := horizontal_alignment.Fill
                                                            VerticalAlignment := vertical_alignment.Fill
                                                            Widget:=SelectedOptionBackground
                                                        overlay_slot:
                                                            HorizontalAlignment := horizontal_alignment.Center
                                                            VerticalAlignment := vertical_alignment.Center
                                                            Widget:=DisplayMessageTextBox
                                            stack_box_slot:
                                                HorizontalAlignment := horizontal_alignment.Right
                                                VerticalAlignment := vertical_alignment.Center
                                                Widget := overlay:
                                                    Slots := array:
                                                        overlay_slot:
                                                            Widget := DropdownIconBackground
                                                            HorizontalAlignment := horizontal_alignment.Center
                                                            VerticalAlignment := vertical_alignment.Center
                                                        overlay_slot:
                                                            Widget := DropdownIcon_StateTexture
                                                            HorizontalAlignment := horizontal_alignment.Center
                                                            VerticalAlignment := vertical_alignment.Center
                                                        overlay_slot:
                                                            Widget := DropDownButton
                                                            HorizontalAlignment := horizontal_alignment.Fill
                                                            VerticalAlignment := vertical_alignment.Fill
                                stack_box_slot:
                                    Widget := color_block{DefaultDesiredSize:=vector2{Y:=10.0},DefaultOpacity:=0.0}

        BGTexture := SelectedOptionBackground.GetImage()
        BGSize := SelectedOptionBackground.GetDesiredSize()
        set DropBoxAssets += for (I ->Item : Content):
            Button := button_quiet{DefaultText := STM(Item)}
            NewBGTexture : texture_block = texture_block{DefaultImage:=BGTexture, DefaultDesiredSize:=BGSize}
            set DropboxBackgrounds += array{NewBGTexture}
            spawn:
                ClickedItemAwait(Button,NewBGTexture,I)
            overlay:
                Slots := array:
                    overlay_slot:
                        Widget := texture_block{DefaultImage:=BGTexture, DefaultDesiredSize:=BGSize}
                        HorizontalAlignment := horizontal_alignment.Center
                        VerticalAlignment := vertical_alignment.Fill

                    overlay_slot:
                        Widget := Button
                        HorizontalAlignment := horizontal_alignment.Fill
                        VerticalAlignment := vertical_alignment.Fill

        for (Stack : DropBoxAssets):
            Stack.SetVisibility(widget_visibility.Collapsed)

        # Initialize default option to be selected
        if(ThisItem := Items[DefaultSelectedOption], BGWidget := DropboxBackgrounds[DefaultSelectedOption]):
            DisplayMessageTextBox.SetText(STM("{SelectedOption_PreText}{ThisItem}{SelectedOption_PostText}"))
            set SelectedOption = DefaultSelectedOption
            SelectedEvent.Signal(ThisItem)
            for(ThisBG:DropboxBackgrounds):
                ThisBG.SetTint(NamedColors.White)
            BGWidget.SetTint(SelectedColor)

        set StackOverlays += DropBoxAssets
        ReturnOverlay := overlay:
            Slots := array:
                overlay_slot:
                    Widget := stack_box:
                        Orientation := orientation.Vertical
                        Slots := for (Overlay : StackOverlays):
                            stack_box_slot:
                                Widget := Overlay
                    HorizontalAlignment := horizontal_alignment.Center
                    VerticalAlignment := vertical_alignment.Top

        return ReturnOverlay

    Remove<public>():void=
        ClearCheckboxLoops.Signal()

    ClickedItemAwait(Button : button_quiet, Background:texture_block, Index : int)<suspends> : void =
        race:
            ClearCheckboxLoops.Await()
            loop:
                Wdata := Button.OnClick().Await()
                if (Owner := agent[Wdata.Player]):
                    if (Expanded?):
                        if (Item := Items[Index]):
                            if(CollapseOnSelection?):
                                CollapseSelection()
                            DisplayMessageTextBox.SetText(STM("{SelectedOption_PreText}{Item}{SelectedOption_PostText}"))
                            set SelectedOption = Index
                            SelectedEvent.Signal(Item)
                            for(ThisBG:DropboxBackgrounds):
                                ThisBG.SetTint(NamedColors.White)
                            Background.SetTint(SelectedColor)

    var Expanded : logic = false
    AwaitExpandPress(Button : button_quiet)<suspends> : void =
        race:
            ClearCheckboxLoops.Await()
            loop:
                Wdata := Button.OnClick().Await()
                if (Owner := agent[Wdata.Player]):
                    if (Expanded?):
                        CollapseSelection()
                    else:
                        ExpandSelection()

    ExpandSelection<public>():void=
        set Expanded = true
        DropdownIcon_StateTexture.SetImage(DropdownIcon_Expand)
        for (Stack : DropBoxAssets):
            Stack.SetVisibility(widget_visibility.Visible)

    CollapseSelection<public>():void=
        set Expanded = false
        DropdownIcon_StateTexture.SetImage(DropdownIcon_Collapse)
        for (Stack : DropBoxAssets):
            Stack.SetVisibility(widget_visibility.Collapsed)