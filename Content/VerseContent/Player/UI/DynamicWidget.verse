
using { /Fortnite.com/Devices }
using { /Verse.org/Simulation }
using { /UnrealEngine.com/Temporary/Diagnostics }
using { /UnrealEngine.com/Temporary/UI }
using { /UnrealEngine.com/Temporary/SpatialMath }
using { /Verse.org/Assets }
using { /Verse.org/Colors }
using { /Fortnite.com/UI }

dynamic_widget<public> := class:

    var<private> ParentWidget : overlay = overlay{}
    var<private> VisibleWidget : ?widget = false

    Mover<private> : color_block = color_block{DefaultOpacity:=0.0}

    GetMoverSize():vector2=
        return Mover.GetDesiredSize()

    var<private> Origin : vector2 = vector2{}
    var CurrAlignment<private> : vector2 = vector2{}
    var CurrAnchors<private> : vector2 = vector2{}
    var CurrentOffset<private> : vector2 = vector2{}
    var<private> CurrentPosition : vector2 = vector2{}
    var<private> CurrentSize : vector2 = vector2{}

    Create<public>(InWidget:widget, Position:vector2, ?AutoAnchor:logic=false, ?AnchorOverride:?anchors=false, ?InitialOffset:vector2=vector2{}, ?Alignment:vector2=vector2{}, ?HorizontalAlignment:horizontal_alignment=horizontal_alignment.Fill, ?VerticalAlignment:vertical_alignment=vertical_alignment.Fill, ?Distribution:?float=false):canvas_slot=
        set VisibleWidget = option{InWidget}
        set CurrentSize = GetSize()
        var ThisAnchor : anchors = anchors{Minimum:=vector2{},Maximum:=vector2{}}
        if(AutoAnchor?):
            AnchorVectors:=vector2{
                X:=Position.X/1920.0
                Y:=Position.Y/1080.0
            }
            DebugPrint("Anchor vector: {AnchorVectors}")
            set ThisAnchor = anchors{
                Minimum:=AnchorVectors
                Maximum:=AnchorVectors
            }
        if(ThisAnchorOverride:=AnchorOverride?):
            set ThisAnchor = ThisAnchorOverride
        set CurrAnchors = ThisAnchor.Minimum
        set Origin = vector2{
            X:=ThisAnchor.Minimum.X*1920.0
            Y:=ThisAnchor.Minimum.Y*1080.0
        }
        set CurrentPosition = Position
        set CurrAlignment = Alignment
        defer{
            ChangePosition(CurrentPosition+InitialOffset)
            if(DebugState?):
                EnableDebug()
            DebugPrint("DW ORIGIN: {Origin}")
        }

        set ParentWidget = overlay{
            Slots:=array:
                overlay_slot{
                    Widget:=InWidget
                    HorizontalAlignment:=horizontal_alignment.Center
                    VerticalAlignment:=vertical_alignment.Center
                }
                overlay_slot{
                    Widget:=stack_box{
                        Orientation:=orientation.Horizontal
                        Slots:=array:
                            stack_box_slot{
                                Widget:=Mover
                                HorizontalAlignment:=horizontal_alignment.Fill
                                VerticalAlignment:=vertical_alignment.Fill
                            }
                    }
                }
        }
        return canvas_slot{
            Widget:=ParentWidget
            Anchors:=ThisAnchor
            Alignment:=Alignment
            Offsets:=margin{Top:=0.0,Left:=0.0}
            SizeToContent:=true
        }

    ChangePosition<public><final>(InPosition:vector2, ?IsDelta:logic=false):void=
        if(IsDelta?):
            Mover.SetDesiredSize(Mover.GetDesiredSize()+InPosition)
            set CurrentPosition = (CurrentPosition + InPosition)
        else:
            MovePos := AbsVector(InPosition-Origin)*2.0
            Mover.SetDesiredSize(MovePos)
            set CurrentPosition = InPosition
        DebugPrint("MoverSize: {Mover.GetDesiredSize()}")
        
    ResetPosition<public><final>():void=
        ChangePosition(Origin)

    ChangeSize<public><final>(DesiredSize:vector2):void=
        if(ThisTexture:=texture_block[VisibleWidget?]):
            ThisTexture.SetDesiredSize(DesiredSize)
        else if(ThisColorBlock:=color_block[VisibleWidget?]):
            ThisColorBlock.SetDesiredSize(DesiredSize)
        else:
            Print("ERR: Called ChangeSize() on a Dynamic Widget that has no sizing support!")
        set CurrentSize = DesiredSize

    GetSize<public><final>():vector2=
        if(ThisTexture:=texture_block[VisibleWidget?]):
            return ThisTexture.GetDesiredSize()
        else if(ThisColorBlock:=color_block[VisibleWidget?]):
            return ThisColorBlock.GetDesiredSize()
        else:
            return vector2{}

    ReChangePosition<public><final>():void=
        Mover.SetDesiredSize(CurrentOffset)

    # Move the transform over time #
    MoveTransformEvent<private>:event()=event(){}
    MoveTransform<public><final>(DesiredPosition:vector2, DesiredSize:vector2, Time:float, ?InterpolationType:interpolation_type=interpolation_type.Linear, ?IsDelta:logic=false, ?AlwaysComplete:logic=false)<suspends>:void=
        MoveTransformEvent.Signal()
        Winner:=race:
            block:
                sync:
                    MovePosition(DesiredPosition, Time, ?IsDelta:=IsDelta, ?InterpolationType:=InterpolationType)
                    MoveSize(DesiredSize, Time, ?IsDelta:=IsDelta, ?InterpolationType:=InterpolationType)
                1
            block:
                MoveTransformEvent.Await()
                -1
        if(AlwaysComplete?, Winner=-1):
            SetTransform(DesiredPosition, DesiredSize)

    SetTransform<public><final>(Position:vector2, Size:vector2):void=
        ChangePosition(Position)
        ChangeSize(Size)

    MoveEvent:event()=event(){}
    MovePosition<public><final>(DesiredPosition:vector2, Time:float, ?InterpolationType:interpolation_type=interpolation_type.Linear, ?IsDelta:logic=false, ?AlwaysComplete:logic=true)<suspends>:void=
        var EndPosition : vector2 = DesiredPosition
        if(IsDelta?):
            set EndPosition = (CurrentPosition + DesiredPosition)
        MoveEvent.Signal()
        race:
            block:
                OriginalPos : vector2 = CurrentPosition
                #TRI:=GetTickRateInfluence()
                StartTime:=GetSimulationElapsedTime()
                loop:
                    LerpPercent:=(GetSimulationElapsedTime()-StartTime)/Time
                    if(LerpPercent >= 1.0):
                        break
                    NextStepPos := Lerp(OriginalPos, EndPosition, LerpPercent.ApplyInterpolation(InterpolationType))
                    ChangePosition(NextStepPos)
                    Sleep(0.0)
                ChangePosition(DesiredPosition)
                <#
                if(Time > 0.0, StepMax := Int[30*Time*TRI]):
                    OriginalPos : vector2 = CurrentPosition
                    for(Step := 1..StepMax):
                        StepPercent := ((Step*1.0)/(StepMax*1.0))
                        NextStepPos := Lerp(OriginalPos, EndPosition, StepPercent.ApplyInterpolation(InterpolationType))
                        ChangePosition(NextStepPos)
                        Sleep(0.0)
                    ChangePosition(EndPosition)
                    return
                else:
                    ChangePosition(EndPosition)
                #>
            block:
                MoveEvent.Await()
        if(AlwaysComplete?):
            ChangePosition(EndPosition)

    ResizeEvent<private>:event(tuple(vector2, float))=event(tuple(vector2, float)){}
    InterruptResize<public>:event()=event(){}
    MoveSize<public>(DesiredSize:vector2, Time:float, ?InterpolationType:interpolation_type=interpolation_type.Linear, ?IsDelta:logic=false, ?AlwaysComplete:logic=true)<suspends>:void=
        EndSize : vector2 = if. IsDelta? then. (CurrentSize + DesiredSize) else. DesiredSize
        ResizeEvent.Signal((EndSize, Time))
        race:
            block:
                StartingSize := CurrentSize
                StartTime:=GetSimulationElapsedTime()
                loop:
                    LerpPercent:=(GetSimulationElapsedTime()-StartTime)/Time
                    if(LerpPercent >= 1.0):
                        break
                    NextStepPos := Lerp(StartingSize, EndSize, LerpPercent.ApplyInterpolation(InterpolationType))
                    ChangeSize(NextStepPos)
                    Sleep(0.0)
                ChangeSize(EndSize)
                #TRI:=GetTickRateInfluence()
                <#
                if(Time > 0.0, StepMax := Int[30*Time*TRI]):
                    StartingSize := CurrentSize
                    for(Step := 1..StepMax):
                        StepPercent := (Step*1.0)/(StepMax*1.0)
                        NextStepSize := Lerp(StartingSize, DesiredSize, StepPercent.ApplyInterpolation(InterpolationType))
                        ChangeSize(NextStepSize)
                        Sleep(0.0)
                    ChangeSize(DesiredSize)
                    return
                else:
                    ResizeEvent.Signal((DesiredSize, Time))
                    Sleep(Inf)
                    #>
            block:
                ResizeEvent.Await()
            block:
                InterruptResize.Await()
        if(AlwaysComplete?):
            ChangeSize(EndSize)

    # Sets the image of the widget, if applicable #
    SetImage<public>(InTexture:texture):void=
        if(TextureWidget:=texture_block[VisibleWidget?]):
            TextureWidget.SetImage(InTexture)
        else{Print("~dynamic_widget ERR: Cannot use \"SetImage()\" on this widget type!")}

    SetColor<public>(InColor:color):void=
        if(TextureWidget:=texture_block[VisibleWidget?]):
            TextureWidget.SetTint(InColor)
        else if(ColorWidget:=color_block[VisibleWidget?]):
            ColorWidget.SetColor(InColor)
        else{Print("~dynamic_widget ERR: Cannot use \"SetImage()\" on this widget type!")}

    # Hides this dynamic widget #
    Hide<public>(?DoCollapse:logic=false):void=
        if. DoCollapse? then. ParentWidget.SetVisibility(widget_visibility.Collapsed) else. ParentWidget.SetVisibility(widget_visibility.Hidden)

    # Shows this dynamic widget #
    Show<public>():void=
        ParentWidget.SetVisibility(widget_visibility.Visible)

    var<private> DebugState : logic = false
    EnableDebug():void=
        Mover.SetOpacity(0.5)
        set DebugState = true

    DisableDebug():void=
        Mover.SetOpacity(0.0)
        set DebugState = false

    DebugPrint(InString:string):void=
        if(DebugState?):
            Print(InString)

    GetWidget():?widget=
        VisibleWidget

    GetRootWidget():widget=
        ParentWidget

        <#
ui_component_base<public> := class:

    # Must set target player when constructing a UI component
    User<public> : player

    var Root<public> : widget = color_block{}
    IsInteractive<public>:logic=false
    AppliedVisibleState : logic = true

    var InitializedState<private> : logic = false
    var<private> VisibleState : logic = false

    Apply<public><final>():void=
        if(not InitializedState?):
            if. not AppliedVisibleState? then Hide()
            if(PlayerUI:=GetPlayerUI[User]):
                PlayerUI.AddWidget(Root, player_ui_slot{InputMode := if(IsInteractive?) then ui_input_mode.All else ui_input_mode.None})
                set InitializedState = true
        else:
            Show()
        spawn. OnApplied()

    OnApplied<public>()<suspends>:void=
        return

    Remove<public>():void=
        if(PlayerUI:=GetPlayerUI[User]):
            PlayerUI.RemoveWidget(Root)
            set InitializedState = false

    Show<public>():void=
        Root.SetVisibility(widget_visibility.Visible)
        set VisibleState = true

    Hide<public>(?DoCollapse:logic=false):void=
        Root.SetVisibility(if. DoCollapse? then widget_visibility.Collapsed else widget_visibility.Hidden)
        set VisibleState = false

    IsVisible<public>()<transacts><decides>:void=
        VisibleState?
        #>