using { /Fortnite.com/Devices }
using { /Verse.org/Simulation }
using { /UnrealEngine.com/Temporary/Diagnostics }
using { /UnrealEngine.com/Temporary/UI }
using { /UnrealEngine.com/Temporary/SpatialMath }
using { /Verse.org/Assets }
using { /Verse.org/Colors }
using { /Fortnite.com/UI }

keypad_widget := class:

    var<private> KeypadWidget : stack_box = stack_box{Orientation:=orientation.Vertical}

    var<private> ButtonWidgets : []button_quiet = array{}
    DeleteButton : button_loud = button_loud{}
    EnterButton : button_loud = button_loud{}

    DisplayedCode : text_block = text_block{}
    var EnteredCode : string = ""

    var Separators : []color_block = array{}
    var SeparatorSize : float = 32.0

    MaximumDigits : int = 6

    # Setup widget
    block:
        NewZeroButton : button_quiet = button_quiet{}
        set KeypadWidget = stack_box{
            Orientation:=orientation.Vertical
            Slots:=array:
                stack_box_slot{
                    Widget := DisplayedCode
                    HorizontalAlignment := horizontal_alignment.Center
                    VerticalAlignment := vertical_alignment.Center
                }
                stack_box_slot{
                    Widget := CreateSeparator()
                    HorizontalAlignment := horizontal_alignment.Center
                    VerticalAlignment := vertical_alignment.Center
                }
                stack_box_slot{
                    HorizontalAlignment := horizontal_alignment.Center
                    VerticalAlignment := vertical_alignment.Center
                    Widget := stack_box{
                        Orientation := orientation.Vertical
                        Slots := array:
                            stack_box_slot{
                                Widget := CreateRow()
                                HorizontalAlignment := horizontal_alignment.Center
                                VerticalAlignment := vertical_alignment.Center
                            }
                            stack_box_slot{
                                Widget := CreateSeparator()
                                HorizontalAlignment := horizontal_alignment.Center
                                VerticalAlignment := vertical_alignment.Center
                            }
                            stack_box_slot{
                                Widget := CreateRow()
                                HorizontalAlignment := horizontal_alignment.Center
                                VerticalAlignment := vertical_alignment.Center
                            }
                            stack_box_slot{
                                Widget := CreateSeparator()
                                HorizontalAlignment := horizontal_alignment.Center
                                VerticalAlignment := vertical_alignment.Center
                            }
                            stack_box_slot{
                                Widget := CreateRow()
                                HorizontalAlignment := horizontal_alignment.Center
                                VerticalAlignment := vertical_alignment.Center
                            }
                            stack_box_slot{
                                Widget := CreateSeparator()
                                HorizontalAlignment := horizontal_alignment.Center
                                VerticalAlignment := vertical_alignment.Center
                            }
                            stack_box_slot{
                                Widget := NewZeroButton
                                HorizontalAlignment := horizontal_alignment.Center
                                VerticalAlignment := vertical_alignment.Center
                            }
                    }
                }
                stack_box_slot{
                    Widget := CreateSeparator()
                    HorizontalAlignment := horizontal_alignment.Center
                    VerticalAlignment := vertical_alignment.Center
                }
                stack_box_slot{
                    HorizontalAlignment := horizontal_alignment.Center
                    VerticalAlignment := vertical_alignment.Center
                    Widget := stack_box{
                        Orientation := orientation.Horizontal
                        Slots := array:
                            stack_box_slot{
                                Widget := DeleteButton
                                HorizontalAlignment := horizontal_alignment.Center
                                VerticalAlignment := vertical_alignment.Center
                            }
                            stack_box_slot{
                                Widget := CreateSeparator()
                                HorizontalAlignment := horizontal_alignment.Center
                                VerticalAlignment := vertical_alignment.Center
                            }
                            stack_box_slot{
                                Widget := EnterButton
                                HorizontalAlignment := horizontal_alignment.Center
                                VerticalAlignment := vertical_alignment.Center
                            }
                    }
                }
        }
        set ButtonWidgets += array{NewZeroButton}

    CreateSeparator<private>()<transacts>:color_block=
        NewSeparator:color_block=color_block{DefaultOpacity:=0.0, DefaultDesiredSize:=vector2{X:=SeparatorSize,Y:=SeparatorSize}}
        set Separators += array. NewSeparator
        return NewSeparator

    CreateRow<private>()<transacts>:stack_box=
        NewButton1 : button_quiet = button_quiet{}
        NewButton2 : button_quiet = button_quiet{}
        NewButton3 : button_quiet = button_quiet{}
        set ButtonWidgets += array{NewButton1, NewButton2, NewButton3}
        var NewRow : stack_box = stack_box{
            Orientation:=orientation.Horizontal
            Slots := array:
                stack_box_slot{
                    Widget := NewButton1
                    HorizontalAlignment := horizontal_alignment.Center
                    VerticalAlignment := vertical_alignment.Center
                }
                stack_box_slot{
                    Widget := CreateSeparator()
                    HorizontalAlignment := horizontal_alignment.Center
                    VerticalAlignment := vertical_alignment.Center
                }
                stack_box_slot{
                    Widget := NewButton2
                    HorizontalAlignment := horizontal_alignment.Center
                    VerticalAlignment := vertical_alignment.Center
                }
                stack_box_slot{
                    Widget := CreateSeparator()
                    HorizontalAlignment := horizontal_alignment.Center
                    VerticalAlignment := vertical_alignment.Center
                }
                stack_box_slot{
                    Widget := NewButton3
                    HorizontalAlignment := horizontal_alignment.Center
                    VerticalAlignment := vertical_alignment.Center
                }
        }
        return NewRow

    Init<public>():void=
        for(ThisButton:ButtonWidgets):
            ThisButton.OnClick().Subscribe(OnNumberClicked)
        DeleteButton.OnClick().Subscribe(OnDelete)
        EnterButton.OnClick().Subscribe(OnEnter)

    GetWidget<public>()<transacts>:stack_box=
        KeypadWidget

    Show<public>():void=
        KeypadWidget.Show()
    
    Hide<public>():void=
        KeypadWidget.Hide()

    OnNumberClicked<public>(MSG:widget_message):void=
        if(ThisSelectedNum := ButtonWidgets.Find[MSG.Source]):
            if(EnteredCode.Length < MaximumDigits):
                set EnteredCode += "ThisSelectedNum"
                DisplayedCode.SetText(StringToMessage(EnteredCode))

    OnDelete<public>(MSG:widget_message):void=
        if(LastDigit := EnteredCode[EnteredCode.Length-1], UpdatedCode := EnteredCode.Slice[0, EnteredCode.Length]):
            set EnteredCode = UpdatedCode
            DisplayedCode.SetText(StringToMessage(UpdatedCode))
        else:
            set EnteredCode = ""
            DisplayedCode.SetText(StringToMessage(""))

    OnEnter<public>(MSG:widget_message):void=
        ThisCodeCommand := GetCodeCommand()
        if(ThisCodeCommand > -1):
            set EnteredCode = ""
            DisplayedCode.SetText(StringToMessage(""))
            case(ThisCodeCommand):
                0 => Print("Test code success")
                1 =>
                    # Pandvil admin settings
                    
                _ =>

    CodeMap<private> : [string]int = map{
        "1234" => 0
        "274818" => 1
    }

    GetCodeCommand<public>()<transacts>:int=
        CodeMap[EnteredCode] or -1