using { /Verse.org/Simulation }
using { /Verse.org/Random }
using { /Verse.org/Assets }

LocalizationTable : [string][string]string = map{
    "EN" => map{
        "GREETING" => "Hello",
        "FAREWELL" => "Goodbye"
    },
    "FR" => map{
        "GREETING" => "Bonjour",
        "FAREWELL" => "Au revoir"
    },
    "ES" => map{
        "GREETING" => "Hola",
        "FAREWELL" => "Adiós"
    }
}

Translate(LanguageCode : string, WordCode : string):string =
    if(LanguageCode="EN"){return WordCode}
    var SpaceCounts : int = 0
    var NewWordCode : string = WordCode
    var NewLanguageCode : string = WordCode
    var TranslatedWord : string = ""
    loop:
        if:
            Index := NewLanguageCode.FindSubstring[" ", ?DesiredAppearance:=SpaceCounts]
            ThisWord := NewLanguageCode.Slice[0, Index]
            UpdatedLanguageCode := NewLanguageCode.Slice[Index+1, LanguageCode.Length-NewLanguageCode.Length]
            set NewLanguageCode = UpdatedLanguageCode
            set TranslatedWord += LocalizationTable[LanguageCode]["a".ToLower()]
        else:
            break
    TranslatedWord

(InFont:string).GetStringTexture(InString:string, ?ErrorOrigin:string="Unknown")<transacts><decides>:texture=
    var MaybeTexture : ?texture = false
    if(InString.Length > 1, ThisTexture:=GetFontFolder[InFont, ?ErrorOrigin:=ErrorOrigin].WordTextures[InString]):
        set MaybeTexture = option{ThisTexture}
    else if(InString.Length > 0, ThisTexture:=GetFontFolder[InFont, ?ErrorOrigin:=ErrorOrigin].CharTextures[InString]):
        set MaybeTexture = option{ThisTexture}
    else:
        set MaybeTexture = option{Art.UIStuff.Fonts.InteractionText.Numbers.Number_0}
    MaybeTexture?

GetFontFolder(InFont:string, ?ErrorOrigin:string="Unknown")<transacts><decides>:font_folder=
    var MaybeFontFolder : font_folder = font_folder{}
    case(InFont):
        "Wave" => 
            set MaybeFontFolder = font_folder{
                CharTextures := map{
                    "0" => Art.UIStuff.Fonts.Round_Numbers.Round_Number_0,
                    "1" => Art.UIStuff.Fonts.Round_Numbers.Round_Number_1,
                    "2" => Art.UIStuff.Fonts.Round_Numbers.Round_Number_2,
                    "3" => Art.UIStuff.Fonts.Round_Numbers.Round_Number_3,
                    "4" => Art.UIStuff.Fonts.Round_Numbers.Round_Number_4,
                    "5" => Art.UIStuff.Fonts.Round_Numbers.Round_Number_5,
                    "6" => Art.UIStuff.Fonts.Round_Numbers.Round_Number_6,
                    "7" => Art.UIStuff.Fonts.Round_Numbers.Round_Number_7,
                    "8" => Art.UIStuff.Fonts.Round_Numbers.Round_Number_8,
                    "9" => Art.UIStuff.Fonts.Round_Numbers.Round_Number_9
                }
            }
        "Points" => 
            set MaybeFontFolder = font_folder{
                CharTextures := map{
                    "0" => Art.UIStuff.Fonts.Point_Numbers.Points_Number_0,
                    "1" => Art.UIStuff.Fonts.Point_Numbers.Points_Number_1,
                    "2" => Art.UIStuff.Fonts.Point_Numbers.Points_Number_2,
                    "3" => Art.UIStuff.Fonts.Point_Numbers.Points_Number_3,
                    "4" => Art.UIStuff.Fonts.Point_Numbers.Points_Number_4,
                    "5" => Art.UIStuff.Fonts.Point_Numbers.Points_Number_5,
                    "6" => Art.UIStuff.Fonts.Point_Numbers.Points_Number_6,
                    "7" => Art.UIStuff.Fonts.Point_Numbers.Points_Number_7,
                    "8" => Art.UIStuff.Fonts.Point_Numbers.Points_Number_8,
                    "9" => Art.UIStuff.Fonts.Point_Numbers.Points_Number_9,
                    "+" => Art.UIStuff.Fonts.Point_Numbers.Points_Plus,
                    "$" => Art.UIStuff.Fonts.Point_Numbers.Points_MoneySign
                }
            }
        "Interaction" => 
            set MaybeFontFolder = font_folder{
                CharTextures := map{
                    "a" => Art.UIStuff.Fonts.InteractionText.Letters.Letter_A
                    "b" => Art.UIStuff.Fonts.InteractionText.Letters.Letter_B
                    "c" => Art.UIStuff.Fonts.InteractionText.Letters.Letter_C
                    "d" => Art.UIStuff.Fonts.InteractionText.Letters.Letter_D
                    "e" => Art.UIStuff.Fonts.InteractionText.Letters.Letter_E
                    "f" => Art.UIStuff.Fonts.InteractionText.Letters.Letter_F
                    "g" => Art.UIStuff.Fonts.InteractionText.Letters.Letter_G
                    "h" => Art.UIStuff.Fonts.InteractionText.Letters.Letter_H
                    "i" => Art.UIStuff.Fonts.InteractionText.Letters.Letter_I
                    "j" => Art.UIStuff.Fonts.InteractionText.Letters.Letter_J
                    "k" => Art.UIStuff.Fonts.InteractionText.Letters.Letter_K
                    "l" => Art.UIStuff.Fonts.InteractionText.Letters.Letter_L
                    "m" => Art.UIStuff.Fonts.InteractionText.Letters.Letter_M
                    "n" => Art.UIStuff.Fonts.InteractionText.Letters.Letter_N
                    "o" => Art.UIStuff.Fonts.InteractionText.Letters.Letter_O
                    "p" => Art.UIStuff.Fonts.InteractionText.Letters.Letter_P
                    "q" => Art.UIStuff.Fonts.InteractionText.Letters.Letter_Q
                    "r" => Art.UIStuff.Fonts.InteractionText.Letters.Letter_R
                    "s" => Art.UIStuff.Fonts.InteractionText.Letters.Letter_S
                    "t" => Art.UIStuff.Fonts.InteractionText.Letters.Letter_T
                    "u" => Art.UIStuff.Fonts.InteractionText.Letters.Letter_U
                    "v" => Art.UIStuff.Fonts.InteractionText.Letters.Letter_V
                    "w" => Art.UIStuff.Fonts.InteractionText.Letters.Letter_W
                    "x" => Art.UIStuff.Fonts.InteractionText.Letters.Letter_X
                    "y" => Art.UIStuff.Fonts.InteractionText.Letters.Letter_Y
                    "z" => Art.UIStuff.Fonts.InteractionText.Letters.Letter_Z

                    "0" => Art.UIStuff.Fonts.InteractionText.Numbers.Number_0
                    "1" => Art.UIStuff.Fonts.InteractionText.Numbers.Number_1
                    "2" => Art.UIStuff.Fonts.InteractionText.Numbers.Number_2
                    "3" => Art.UIStuff.Fonts.InteractionText.Numbers.Number_3
                    "4" => Art.UIStuff.Fonts.InteractionText.Numbers.Number_4
                    "5" => Art.UIStuff.Fonts.InteractionText.Numbers.Number_5
                    "6" => Art.UIStuff.Fonts.InteractionText.Numbers.Number_6
                    "7" => Art.UIStuff.Fonts.InteractionText.Numbers.Number_7
                    "8" => Art.UIStuff.Fonts.InteractionText.Numbers.Number_8
                    "9" => Art.UIStuff.Fonts.InteractionText.Numbers.Number_9

                    "[" => Art.UIStuff.Fonts.InteractionText.Symbols.Symbol_BracketLeft
                    "]" => Art.UIStuff.Fonts.InteractionText.Symbols.Symbol_BracketRight
                    "$" => Art.UIStuff.Fonts.InteractionText.Symbols.Symbol_Cash
                    "-" => Art.UIStuff.Fonts.InteractionText.Symbols.Symbol_Dash
                    "!" => Art.UIStuff.Fonts.InteractionText.Symbols.Symbol_ExclamationPoint
                    "." => Art.UIStuff.Fonts.InteractionText.Symbols.Symbol_Period
                    "+" => Art.UIStuff.Fonts.InteractionText.Symbols.Symbol_Plus
                    "%" => Art.UIStuff.Fonts.InteractionText.Symbols.Symbol_Percent
                    "/" => Art.UIStuff.Fonts.InteractionText.Symbols.Symbol_Slash
                }
            }
        "Powerup" => 
            set MaybeFontFolder = font_folder{
                WordTextures := map{
                    "instakill" => Art.UIStuff.Icons.Powerup_Icons.InstaKillIcon
                    "double points" => Art.UIStuff.Icons.Powerup_Icons.DoublePointsIcon
                    "death machine" => Art.UIStuff.Icons.Powerup_Icons.DeathMachineIcon
                    "firesale" => Art.UIStuff.Icons.Powerup_Icons.FiresaleIcon
                    "incognito" => Art.UIStuff.Icons.Powerup_Icons.IncognitoIcon
                }
            }
        "Perk" => 
            set MaybeFontFolder = font_folder{
                WordTextures := map{
                    "savior soda" => Art.UIStuff.Icons.Perk_Icons.SaviorSoda_Perk_Icon
                    "maxi munitions" => Art.UIStuff.Icons.Perk_Icons.MaxMunitions_Perk_Icon
                    "marathon cola" => Art.UIStuff.Icons.Perk_Icons.Marathon_Perk_Icon
                    "tanked up" => Art.UIStuff.Icons.Perk_Icons.Tank_Perk_Icon
                    "cryo cola" => Art.UIStuff.Icons.Perk_Icons.Cryo_Perk_Icon
                    "bull-it brew" => Art.UIStuff.Icons.Perk_Icons.Bull_Perk_Icon
                    "bomb cola" => Art.UIStuff.Icons.Perk_Icons.Bomb_Perk_Icon
                }
            }
        "Quest" => 
            set MaybeFontFolder = font_folder{
                CharTextures := map{
                    "a" => Art.UIStuff.Fonts.Quest.Letters.QF_Letter_A
                    "b" => Art.UIStuff.Fonts.Quest.Letters.QF_Letter_B
                    "c" => Art.UIStuff.Fonts.Quest.Letters.QF_Letter_C
                    "d" => Art.UIStuff.Fonts.Quest.Letters.QF_Letter_D
                    "e" => Art.UIStuff.Fonts.Quest.Letters.QF_Letter_E
                    "f" => Art.UIStuff.Fonts.Quest.Letters.QF_Letter_F
                    "g" => Art.UIStuff.Fonts.Quest.Letters.QF_Letter_G
                    "h" => Art.UIStuff.Fonts.Quest.Letters.QF_Letter_H
                    "i" => Art.UIStuff.Fonts.Quest.Letters.QF_Letter_I
                    "j" => Art.UIStuff.Fonts.Quest.Letters.QF_Letter_J
                    "k" => Art.UIStuff.Fonts.Quest.Letters.QF_Letter_K
                    "l" => Art.UIStuff.Fonts.Quest.Letters.QF_Letter_L
                    "m" => Art.UIStuff.Fonts.Quest.Letters.QF_Letter_M
                    "n" => Art.UIStuff.Fonts.Quest.Letters.QF_Letter_N
                    "o" => Art.UIStuff.Fonts.Quest.Letters.QF_Letter_O
                    "p" => Art.UIStuff.Fonts.Quest.Letters.QF_Letter_P
                    "q" => Art.UIStuff.Fonts.Quest.Letters.QF_Letter_Q
                    "r" => Art.UIStuff.Fonts.Quest.Letters.QF_Letter_R
                    "s" => Art.UIStuff.Fonts.Quest.Letters.QF_Letter_S
                    "t" => Art.UIStuff.Fonts.Quest.Letters.QF_Letter_T
                    "u" => Art.UIStuff.Fonts.Quest.Letters.QF_Letter_U
                    "v" => Art.UIStuff.Fonts.Quest.Letters.QF_Letter_V
                    "w" => Art.UIStuff.Fonts.Quest.Letters.QF_Letter_W
                    "x" => Art.UIStuff.Fonts.Quest.Letters.QF_Letter_X
                    "y" => Art.UIStuff.Fonts.Quest.Letters.QF_Letter_Y
                    "z" => Art.UIStuff.Fonts.Quest.Letters.QF_Letter_Z

                    "0" => Art.UIStuff.Fonts.Quest.Numbers.QF_Number_0
                    "1" => Art.UIStuff.Fonts.Quest.Numbers.QF_Number_1
                    "2" => Art.UIStuff.Fonts.Quest.Numbers.QF_Number_2
                    "3" => Art.UIStuff.Fonts.Quest.Numbers.QF_Number_3
                    "4" => Art.UIStuff.Fonts.Quest.Numbers.QF_Number_4
                    "5" => Art.UIStuff.Fonts.Quest.Numbers.QF_Number_5
                    "6" => Art.UIStuff.Fonts.Quest.Numbers.QF_Number_6
                    "7" => Art.UIStuff.Fonts.Quest.Numbers.QF_Number_7
                    "8" => Art.UIStuff.Fonts.Quest.Numbers.QF_Number_8
                    "9" => Art.UIStuff.Fonts.Quest.Numbers.QF_Number_9

                    "(" => Art.UIStuff.Fonts.Quest.Symbols.QF_Symbol_BracketLeft
                    ")" => Art.UIStuff.Fonts.Quest.Symbols.QF_Symbol_BracketRight
                    ":" => Art.UIStuff.Fonts.Quest.Symbols.QF_Symbol_Colon
                    "." => Art.UIStuff.Fonts.Quest.Symbols.QF_Symbol_Dot
                    "!" => Art.UIStuff.Fonts.Quest.Symbols.QF_Symbol_ExclamationMark
                    "?" => Art.UIStuff.Fonts.Quest.Symbols.QF_Symbol_QuestionMark
                    "/" => Art.UIStuff.Fonts.Quest.Symbols.QF_Symbol_Slash
                }
                WordTextures := map{
                    "clipboard" => Art.UIStuff.Fonts.Quest.Symbols.Clipboard_Icon
                }
            }
        _ =>
            Print("~ERROR: Failed to find the font \"{InFont}\" from origin [{ErrorOrigin}]!")
    MaybeFontFolder.IsValid[]
    MaybeFontFolder
