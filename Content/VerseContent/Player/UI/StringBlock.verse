using { /Verse.org/Assets }
using { /UnrealEngine.com/Temporary/SpatialMath }
using { /UnrealEngine.com/Temporary/UI }
using { /Verse.org/Colors }
using { /Verse.org/Simulation }

# A simple, highly configurable widget that can display text to the screen in any font #
string_block<public> := class:

    WidgetControl<private> : dynamic_widget = dynamic_widget{}
    GetWidget():widget=
        return WidgetControl.GetRootWidget()
    GetPosition():vector2=
        return WidgetControl.CurrentPosition

    var<private> StringStackbox : stack_box = stack_box{Orientation:=orientation.Horizontal}
    var StringBoxSlots<private> : []string_block_slot = array{}

    var<private> BackplateTexture:?texture=false
    BackplateWidget<private> : texture_block = texture_block{DefaultImage:=Art.UIStuff.Fonts.InteractionText.Numbers.Number_0}
    var<private> BackplateOffsets : vector2 = vector2{}

    var<private> Font<public>:string
    ChangeFont<public><final>(InFontName:string):void=
        if(GetFontFolder[InFontName]):
            set Font = InFontName
            SetText(CurrentText)
        else{Print("~string_block ERR: Failed to find Font Name [{InFontName}]!")}

    var<private> CurrentText : string = ""

    var<private> SizeMultiplier : float = 1.0
    var<private> CurrSize : vector2 = vector2{X:=1.0,Y:=1.0}
    var SpacingMultiplier<private> : float = 0.5
    var CurrSpacerSize<private> : float = 1.0
    var<private> CurrColor : color = color{}

    ApplySizeRatios : logic = true

    var HorizontalAlignment<internal> : horizontal_alignment = horizontal_alignment.Center

    Create<public>(Text:string, Position:vector2, InSizeMultiplier:float, ?SizeRatio:vector2=vector2{X:=1.0,Y:=1.0}, ?SpacingSize:float:=0.25, ?SpacerSize:float:=1.0, ?SandwichedWidget:?widget=false, ?Alignment:vector2=vector2{}, ?AutoAnchor:logic=false, ?AnchorOverride:?anchors=false):canvas_slot=
        set CurrSize = SizeRatio
        set SizeMultiplier = InSizeMultiplier
        set SpacingMultiplier = SpacingSize
        set CurrSpacerSize = SpacerSize+1.0
        if(DebugState?):
            EnableDebug()
        SetText(Text)
        defer{
            SetSpacerSize(CurrSpacerSize)
            SetSpacingSize(SpacingMultiplier)
            if(ThisBackplateTexture:=BackplateTexture?):
                SetBackplate(ThisBackplateTexture)
            else{HideBackplate()}
        }
        ThisSandwichedWidget := SandwichedWidget? or overlay{}
        StringBlockOverlay : overlay = overlay{
            Slots:=array:
                overlay_slot{
                    Widget:=BackplateWidget
                    HorizontalAlignment:=horizontal_alignment.Fill
                    VerticalAlignment:=vertical_alignment.Fill
                }
                overlay_slot{
                    Widget:=ThisSandwichedWidget
                    HorizontalAlignment:=horizontal_alignment.Center
                    VerticalAlignment:=vertical_alignment.Center
                }
                overlay_slot{
                    Widget:=StringStackbox
                    HorizontalAlignment:=horizontal_alignment.Center
                    VerticalAlignment:=vertical_alignment.Center
                }
        }
        WidgetControl.Create(StringBlockOverlay, Position, ?Alignment:=Alignment, ?AutoAnchor:=AutoAnchor, ?AnchorOverride:=AnchorOverride)

    Reset():void=
        SetText("")

    TextChangedEvent:event(string)=event(string){}
    SetText<public><final>(InText:string, ?FixSizing:logic=true):void=
        var TextChanged : logic = false
        if(InText.ToLower()=CurrentText):
            return
        else:
            set TextChanged = true
            var InTextConv : string = InText.ToLower()
            TextWithCommands := InTextConv
            set InTextConv = GetRealText(InTextConv)
            if(InTextConv=""):
                set CurrentText = ""
                StringStackbox.Hide()
                PrevBackplateVisibleState := BackplateVisibleState
                HideBackplate()
                set BackplateVisibleState = PrevBackplateVisibleState
                return
            else if(StringBoxSlots.Length <> InTextConv.Length):
                var LengthSubtract : int = 0
                var LengthAdditive : int = 0
                InSubstringIndexes := GetWordIndexes(InTextConv)
                for(ThisSubstringIndexes : InSubstringIndexes):
                    set LengthSubtract += (ThisSubstringIndexes(1)-ThisSubstringIndexes(0))
                PrevSubstringIndexes := GetWordIndexes(CurrentText)
                for(ThisSubstringIndexes : PrevSubstringIndexes):
                    set LengthAdditive += (ThisSubstringIndexes(1)-ThisSubstringIndexes(0))
                #LengthDiff := (InTextConv.Length-LengthSubtract)-(CurrentText.Length-LengthAdditive)
                LengthDiff := (InTextConv.Length-LengthSubtract)-StringBoxSlots.Length
                #Print("Lengthdiff: {LengthDiff}")
                set CurrentText = InTextConv
                if(LengthDiff > 0):
                    for(ExtraCharCount:=0..LengthDiff-1):
                        AddStringWidget('0', false)
                    AllStringBlocks:=GetVisibleStringSlots()
                    #Print("StringBlockLength for {CurrentText}: {AllStringBlocks.Length}")
                    for(ThisStringBlock:AllStringBlocks):
                        ThisStringBlock.StringWidget.Show()
                        ThisStringBlock.SpacerWidget.Hide()
                else if(LengthDiff < 0):
                    WidgetsToRemove:=StringBoxSlots.Length-(Abs(LengthDiff))
                    for(ErasingCharIndex:=WidgetsToRemove..StringBoxSlots.Length-1):
                        if(ThisErasingWidget:=StringBoxSlots[ErasingCharIndex]):
                            #StringStackbox.RemoveWidget(ThisErasingWidget.StringWidget)
                            #StringStackbox.RemoveWidget(ThisErasingWidget.SpacerWidget)
                            ThisErasingWidget.StringWidget.Collapse()
                            ThisErasingWidget.SpacerWidget.Collapse()
                    <#
                    WidgetsRemaining := (StringBoxSlots.Length-WidgetsToRemove)
                    var TempStackBoxSlots : []string_block_slot = array{}
                    for(ThisCharSlotIndex := WidgetsToRemove..StringBoxSlots.Length):
                        if(ThisStackBoxSlot:=StringBoxSlots[ThisCharSlotIndex]):
                            set TempStackBoxSlots += array{ThisStackBoxSlot}
                    set StringBoxSlots = TempStackBoxSlots #>
            set CurrentText = InTextConv
            if(BackplateVisibleState?):
                ShowBackplate()
            var CharIndex : int = 0
            var TrackedSubstringIndex : int = 0
            StringStackbox.Show()
            SubstringIndexes := GetWordIndexes(InTextConv)
            StringBlocks:=GetVisibleStringSlots()
            #Print("StringBlocks: {StringBlocks.Length}")
            for(ThisStackBox:StringBlocks):
                TextureWidget:=ThisStackBox.StringWidget
                if(MatchingIndex:=SubstringIndexes.FindMatchingStartIndex[CharIndex]):
                    if(ThisWord:=InTextConv.Slice[CharIndex, MatchingIndex(1)+1]):
                        set ThisStackBox.String = ThisWord
                        if(ThisWordTexture:=Font.GetStringTexture[ThisWord, ?ErrorOrigin:="SetText - Line 131"]):
                            TextureWidget.Show()
                            #TextureWidget.SetDesiredSize(vector2{X:=CurrSize,Y:=CurrSize})
                            TextureWidget.SetImage(ThisWordTexture)
                        else{Print("~ERROR: Couldn't find the WordTexture with word [{ThisWord}]")}
                        set CharIndex += (MatchingIndex(1)-MatchingIndex(0))
                        set TrackedSubstringIndex += 1
                    else{Print("~ERROR: Couldn't find the word within the string!")}
                else if(ThisChar:=InTextConv[CharIndex]):
                    set ThisStackBox.String = "{ThisChar}"
                    if(ThisStackBox.String <> " ", NewCharTexture:=Font.GetStringTexture[ThisStackBox.String, ?ErrorOrigin:="SetText - Line 153"]):
                        TextureWidget.Show()
                        TextureWidget.SetImage(NewCharTexture)
                    else:
                        TextureWidget.Hide()
                else:
                    Print("~ERROR: Failed to find a character widget for a char or word!")
                    Print("~The index was {CharIndex} and the string is [{InTextConv}]")
                set CharIndex += 1
            set CurrentText = InTextConv
            if(ThisLastStringBox:=StringBoxSlots[GetTextLength()-1]):
                ThisLastStringBox.SpacerWidget.SetDesiredSize(vector2{})
            TextChangedEvent.Signal(CurrentText)
            if(FixSizing?):
                SetTextSize(SizeMultiplier)
            ExecuteCommands(TextWithCommands)

    AddText(InText:string, ?SeperateCharacters:logic=false, ?AddDelay:float=0.0, ?OffsetSize:float=0.0, ?EndOffsetSize:float=0.0, ?OffsetSizeChangeInterpolation:interpolation_type=interpolation_type.Linear)<suspends>:void=
        if(SeperateCharacters?):
            for(ThisText:InText):
                StartingOffsetSize:=CurrSize*OffsetSize
                AddStringWidget(ThisText.ToLower(), option{StartingOffsetSize})
                ResizeBackplate(BackplateOffsets)
                #Sleep(0.25)
                SetText(CurrentText+"{ThisText}", ?FixSizing:=false)
                #Print("DestinationSize of added text: {StartingOffsetSize+EndOffsetSize}")
                #Sleep(1.0)
                MoveTextSize((StringBoxSlots.Length-1, StringBoxSlots.Length), EndOffsetSize, AddDelay, ?InterpolationType:=OffsetSizeChangeInterpolation)
                SetTextSize(SizeMultiplier)
                #Print("CurrSize: {CurrSize}")
        else:
            SetText(CurrentText+InText)
        

    # Adds a widget for the given character which will be applied to the stack box #
    # Internal use only #
    AddStringWidget<private>(InChar:char, SizeOverride:?vector2, ?Tint:?color=false):void=
        if(CharTexture:=Font.GetStringTexture["{InChar}", ?ErrorOrigin:="AddStringWidget"]):
            var ThisSize : vector2 = CurrSize*SizeMultiplier
            #Print("{Font} FONT Char size: {ThisSize}, {CurrSize}, {SizeMultiplier}")
            if(OverrideSize:=SizeOverride?):
                set ThisSize = OverrideSize
            NewTextureBlock : texture_block = texture_block{DefaultImage:=CharTexture, DefaultDesiredSize:=ThisSize}
            NewStackSlot : stack_box_slot = stack_box_slot{
                Widget:=NewTextureBlock
                HorizontalAlignment := horizontal_alignment.Center
                VerticalAlignment := vertical_alignment.Center
            }
            if(ThisTintColor:=Tint?):
                NewTextureBlock.SetTint(ThisTintColor)
            NewStackBoxSlot : string_block_slot = string_block_slot{
                String:="{InChar}"
                StringWidget:=NewTextureBlock
                StackboxWidget:=option{NewStackSlot}
                CurrSize:=ThisSize
            }
            NewStackBoxSlot.SpacerWidget.SetDesiredSize(vector2{})
            <#
            if(InChar = ' '):
                NewTextureBlock.HideNoCollapse()
                NewTextureBlock.SetDesiredSize(vector2{X:=SpacingMultiplier*ThisSize,Y:=ThisSize})
                NewStackBoxSlot.SpacerWidget.Hide()
            else:
                NewStackBoxSlot.SpacerWidget.SetDesiredSize(vector2{X:=-1*(ThisSize/CurrSpacerSize), Y:=ThisSize})#>
            NewStackBoxSlot.SpacerWidget.Hide()
            <#
            TextLength:=GetTextLength()
            if(InChar = ' '):
                NewTextureBlock.HideNoCollapse()
                NewTextureBlock.SetDesiredSize(vector2{X:=SpacingMultiplier*ThisSize.X,Y:=ThisSize.Y})
                NewStackBoxSlot.SpacerWidget.Hide()
            else if(ThisLastStringBox:=StringBoxSlots[TextLength-1]):
                if(ThisLastStringBox.String = " "):
                    #ThisLastStringBox.StringWidget.SetDesiredSize(vector2{X:=SpacingMultiplier*ThisLastStringBox.CurrSize,Y:=ThisLastStringBox.CurrSize})
                    ThisLastStringBox.SpacerWidget.Hide()
                else:
                    if(ThisLastStringBox.String = "i" or ThisLastStringBox.String = "!", InChar <> 'w' and InChar <> 'm'):
                        ThisLastStringBox.SpacerWidget.SetDesiredSize(vector2{X:=-1*((ThisLastStringBox.CurrSize/CurrSpacerSize)*1.5), Y:=ThisLastStringBox.CurrSize})
                        if(PreviousStringBox:=StringBoxSlots[TextLength-2]):
                            PreviousStringBox.SpacerWidget.SetDesiredSize(vector2{X:=-1*((ThisLastStringBox.CurrSize/CurrSpacerSize)*1.5), Y:=ThisLastStringBox.CurrSize})
                    else if(ThisLastStringBox.String = "w" or ThisLastStringBox.String = "m", InChar <> 'i' and InChar <> '!'):
                        ThisLastStringBox.SpacerWidget.SetDesiredSize(vector2{X:=-1*((ThisLastStringBox.CurrSize/CurrSpacerSize)*0.5), Y:=ThisLastStringBox.CurrSize})
                        if(PreviousStringBox:=StringBoxSlots[TextLength-2], PreviousStringBox.String <> "i" and PreviousStringBox.String <> "!" and PreviousStringBox.String <> " "):
                            PreviousStringBox.SpacerWidget.SetDesiredSize(vector2{X:=-1*((ThisLastStringBox.CurrSize/CurrSpacerSize)*0.5), Y:=ThisLastStringBox.CurrSize})
                    else:
                        ThisLastStringBox.SpacerWidget.SetDesiredSize(vector2{X:=-1*(ThisLastStringBox.CurrSize/CurrSpacerSize), Y:=ThisLastStringBox.CurrSize}) #>
            set StringBoxSlots += array{NewStackBoxSlot}
            StringStackbox.AddWidget(NewStackSlot)
            StringStackbox.AddWidget(NewStackBoxSlot.SetAndGetSpacerStackboxSlot())
        else{Print("~ERROR: Failed to find a texture for the char [{InChar}] for Font [{Font}]")}

    SetBackplate<public><final>(Texture:texture):void=
        BackplateWidget.SetImage(Texture)
        ResizeBackplate(BackplateOffsets)
        ShowBackplate()

    ResizeBackplate<public><final>(Offsets:vector2):void=
        set BackplateOffsets = Offsets
        ScreenSize:=GetTextScreenSize()
        if(ScreenSize.X > 0.0):
            BackplateWidget.SetDesiredSize(Offsets+vector2{X:=ScreenSize.X,Y:=ScreenSize.Y})
        else:
            BackplateWidget.SetDesiredSize(vector2{})

    var BackplateVisibleState<private> : logic = false
    HideBackplate<public><final>():void=
        BackplateWidget.Hide()
        set BackplateVisibleState = false

    ShowBackplate<public><final>():void=
        BackplateWidget.Show()
        set BackplateVisibleState = true

    var<private> VisibleState : logic = true
    Hide<public><final>(?DoCollapse:logic=false):void=
        WidgetControl.Hide(?DoCollapse:=DoCollapse)
        set VisibleState = false

    Show<public><final>(?IgnoreTextCheck:logic=false):void=
        if(CurrentText.Length > 0 or IgnoreTextCheck?):
            WidgetControl.Show()
        set VisibleState = true

    # Changes the color of the entire text #
    SetColor<public><final>(Color:color):void=
        set CurrColor = Color
        LerpColorEvent.Signal(0, GetTextLength()-1)
        SetColorsEvent.Signal()
        StringBlocks:=GetVisibleStringSlots()
        for(ThisStackBox:StringBlocks):
            ThisStackBox.StringWidget.SetTint(Color)

    # Change the color of the text within a specified range #
    SetColor<public><final>(Range:tuple(int, int), Color:color):void=
        for(ThisIndex:=Range(0)..Range(1)):
            if(ThisCharWidget:=StringBoxSlots[ThisIndex].StringWidget):
                ThisCharWidget.SetTint(Color)

    LerpColorEvent<private>:event(tuple(int, int))=event(tuple(int, int)){}
    LerpColor<public><final>(Range:tuple(int, int), DesiredColor:color, Time:float, ?Loops:int=0, ?AlwaysComplete:logic=false)<suspends>:void=
        var ThisRange : tuple(int, int) = Range
        if(Range(1) <= -1):
            set ThisRange = (0, GetTextLength()-1)
        var OriginalColor : color = NamedColors.White
        if(ThisStringBox:=StringBoxSlots[Range(0)]):
            set OriginalColor = ThisStringBox.StringWidget.GetTint()
        LerpColorEvent.Signal(ThisRange)
        Winner:=race:
            block:
                var LoopTimes : int = 0
                loop:
                    if(Time > 0.0, StepMax := Int[30*Time]):
                        for(Step := 1..StepMax):
                            StepPercent := (Step*1.0)/(StepMax*1.0)
                            StepColor := CLerp(OriginalColor, DesiredColor, StepPercent)
                            SetColor(ThisRange, StepColor)
                            Sleep(0.0)
                    else:
                        SetColor(ThisRange, DesiredColor)
                    set LoopTimes += 1
                    if(LoopTimes > Loops, Loops>-1):
                        break
                1
            block:
                loop:
                    NewLerpRange:=LerpColorEvent.Await()
                    if(NewLerpRange.IsInRange[ThisRange]):
                        break
                -1
        if(AlwaysComplete?, Winner=-1):
            SetColor(ThisRange, DesiredColor)

    SetColorsEvent<private>:event()=event(){}
    SetColors<public><final>(InColorArray:[]color, TickTime:float, InRange:tuple(int, int), ?Loops:int=0)<suspends>:void=
        var CurrRepeats : int = 0
        var ThisRange : tuple(int, int) = InRange
        if(ThisRange = (-1, -1)):
            set ThisRange = (0, GetTextLength()-1)
        else if(ThisRange(1) = -1):
            set ThisRange = (0, GetTextLength()-1)
        SetColorsEvent.Signal()
        race:
            loop:
                for(ThisColor:InColorArray):
                    SetColor(ThisRange, ThisColor)
                    Sleep(TickTime)
                if(CurrRepeats >= Loops, Loops > -1):
                    break
                else:
                    set CurrRepeats += 1
                Sleep(0.0)
            TextChangedEvent.Await() 
            SetColorsEvent.Await()

    # Checks if a range of text indexes is within the range of another #
    (Range:tuple(int, int)).IsInRange<private>(OtherRange:tuple(int, int))<transacts><decides>:void=
        Range(0) >= OtherRange(0) or Range(1) >= OtherRange(0)

    <# MOVEMENT AND SIZING FUNCTIONS #>

    SetPosition<public><final>(InPosition:vector2):void=
        WidgetControl.ChangePosition(InPosition)

    # Moves the text to the desired position #
    MovePosition<public><final>(InPosition:vector2, Time:float, ?InterpolationType:interpolation_type=interpolation_type.Linear, ?AlwaysComplete:logic=false, ?IsDelta:logic=false)<suspends>:void=
        WidgetControl.MovePosition(InPosition, Time, ?InterpolationType:=InterpolationType, ?IsDelta:=IsDelta, ?AlwaysComplete:=AlwaysComplete)
        #Print("Completed movement")
    
    # Sets the text size, as well as updating spacing and spacer sizes to scale #
    SetTextSize<public><final>(InSize:float, ?IsScaling:logic=false):void=
        SetTextSize((0, GetTextLength()), InSize, ?IsScaling:=IsScaling)

    SetTextSize<public><final>(Range:tuple(int, int), InSize:float, ?IsScaling:logic=false):void=
        var InRange : tuple(int, int) = Range
        if(InRange = (-1, -1)):
            set InRange = (0, GetTextLength())
            set SizeMultiplier = InSize
        else if(InRange(1) = -1):
            set InRange = (InRange(0), GetTextLength())
            set SizeMultiplier = InSize
        else if(InRange = (0, GetTextLength())):
            set SizeMultiplier = InSize
        var CharIndex : int = InRange(0)
        ThisStringBoxRange:=StringBoxSlots.Slice[InRange(0), InRange(1)] or StringBoxSlots
        TextLength:=GetTextLength()
        DesiredStringBoxSize:vector2=InSize*CurrSize
        #Print("{Font} FONT size: {DesiredStringBoxSize}, {CurrSize}, {SizeMultiplier}")
        for(ThisStackBox:ThisStringBoxRange):
            if(ApplySizeRatios?):
                if(ThisStackBox.String = " "):
                    ThisStackBox.StringWidget.SetDesiredSize(vector2{X:=SpacingMultiplier*DesiredStringBoxSize.X,Y:=DesiredStringBoxSize.Y})
                else if(ThisStackBox.String = "1" or ThisStackBox.String = "i" or ThisStackBox.String = "!"):
                    ThisStackBox.StringWidget.SetDesiredSize(vector2{X:=DesiredStringBoxSize.X*0.525,Y:=DesiredStringBoxSize.Y})
                    #DebugPrint("Size of 2nd: {ThisStackBox.StringWidget.GetDesiredSize()}")
                else if(ThisStackBox.String = "[" or ThisStackBox.String = "]" or ThisStackBox.String = "/"):
                    ThisStackBox.StringWidget.SetDesiredSize(vector2{X:=DesiredStringBoxSize.X*0.525,Y:=DesiredStringBoxSize.Y})
                    #DebugPrint("Size of 3rd: {ThisStackBox.StringWidget.GetDesiredSize()}")
                else if(ThisStackBox.String = "(" or ThisStackBox.String = ")" or ThisStackBox.String = ":" or ThisStackBox.String = "."):
                    ThisStackBox.StringWidget.SetDesiredSize(vector2{X:=DesiredStringBoxSize.X*0.525,Y:=DesiredStringBoxSize.Y})
                else if(ThisStackBox.String = "w" or ThisStackBox.String = "m"):
                    ThisStackBox.StringWidget.SetDesiredSize(vector2{X:=DesiredStringBoxSize.X*1.375,Y:=DesiredStringBoxSize.X*1.375*1.1})
                else:
                    ThisStackBox.StringWidget.SetDesiredSize(DesiredStringBoxSize)
            else:
                ThisStackBox.StringWidget.SetDesiredSize(DesiredStringBoxSize)
            ThisStackBox.SpacerWidget.Hide()
                <#
                ThisStackBox.StringWidget.SetDesiredSize(vector2{X:=DesiredStringBoxSize.X, Y:=DesiredStringBoxSize.Y})
                ThisStackBox.SpacerWidget.Show()
                if(ThisStackBox.String = "i" or ThisStackBox.String = "!"):
                    ThisStackBox.SpacerWidget.SetDesiredSize(vector2{X:=-1*((DesiredStringBoxSize.X/CurrSpacerSize)*1.5), Y:=DesiredStringBoxSize.Y})
                    if(PreviousStringBox:=StringBoxSlots[CharIndex-1]):
                        if(PreviousStringBox.String <> "w" and PreviousStringBox.String <> "m"):
                            PreviousStringBox.SpacerWidget.SetDesiredSize(vector2{X:=-1*((DesiredStringBoxSize.X/CurrSpacerSize)*1.5), Y:=DesiredStringBoxSize.Y})
                        else:
                            PreviousStringBox.SpacerWidget.SetDesiredSize(vector2{X:=-1*((DesiredStringBoxSize.X/CurrSpacerSize)), Y:=DesiredStringBoxSize.Y})
                else if(ThisStackBox.String = "w" or ThisStackBox.String = "m"):
                    ThisStackBox.SpacerWidget.SetDesiredSize(vector2{X:=-1*((DesiredStringBoxSize.X/CurrSpacerSize)*0.5), Y:=DesiredStringBoxSize.Y})
                    if(PreviousStringBox:=StringBoxSlots[CharIndex-1]):
                        if(PreviousStringBox.String <> "i" and PreviousStringBox.String <> "!" and PreviousStringBox.String <> " "):
                            PreviousStringBox.SpacerWidget.SetDesiredSize(vector2{X:=-1*((DesiredStringBoxSize.X/CurrSpacerSize)*0.5), Y:=DesiredStringBoxSize.Y})
                        else if(PreviousStringBox.String = " "):
                            PreviousStringBox.StringWidget.SetDesiredSize(vector2{X:=SpacingMultiplier*DesiredStringBoxSize.X*1.5,Y:=DesiredStringBoxSize.Y})
                            PreviousStringBox.SpacerWidget.SetDesiredSize(vector2{})
                            PreviousStringBox.SpacerWidget.Hide()
                            if(SecondPreviousSB:=StringBoxSlots[CharIndex-2], SecondPreviousSB.String = "i"):
                                PreviousStringBox.StringWidget.SetDesiredSize(vector2{X:=SpacingMultiplier*DesiredStringBoxSize.X*0.5,Y:=DesiredStringBoxSize.Y})
                                SecondPreviousSB.SpacerWidget.Hide()
                        else:
                            PreviousStringBox.SpacerWidget.SetDesiredSize(vector2{X:=-1*((DesiredStringBoxSize.X/CurrSpacerSize)), Y:=DesiredStringBoxSize.Y})
                else:
                    ThisStackBox.SpacerWidget.SetDesiredSize(vector2{X:=-1*(DesiredStringBoxSize.X/CurrSpacerSize), Y:=DesiredStringBoxSize.Y})#>
            set CharIndex += 1
        if(ThisLastStringBox:=StringBoxSlots[TextLength-1]):
            ThisLastStringBox.SpacerWidget.SetDesiredSize(vector2{})
        ResizeBackplate(BackplateOffsets)

    ResizeEvent<private>:event()=event(){}
    MoveTextSize<public><final>(DesiredSize:float, Time:float, ?IsScaling:logic=false, ?InterpolationType:interpolation_type=interpolation_type.Linear, ?AlwaysComplete:logic=false)<suspends>:void=
        MoveTextSize((0, GetTextLength()), DesiredSize, Time, ?IsScaling:=IsScaling, ?InterpolationType:=InterpolationType, ?AlwaysComplete:=AlwaysComplete)

    MoveTextSize<public><final>(Range:tuple(int, int), DesiredSize:float, Time:float, ?FromSize:float=-1.0, ?IsScaling:logic=false, ?InterpolationType:interpolation_type=interpolation_type.Linear, ?AlwaysComplete:logic=false, ?SaveSize:logic=true)<suspends>:void=
        var InRange : tuple(int, int) = Range
        var OriginalSize : float = SizeMultiplier
        if(FromSize > -1.0):
            set OriginalSize = FromSize
        var EndSize : float = DesiredSize
        if(InRange = (-1, -1)):
            set InRange = (0, GetTextLength())
        else if(InRange(1) = -1):
            set InRange = (InRange(0), GetTextLength())
        <#
        if(InRange <> (0, StringBoxSlots.Length-1)):
            set OriginalSize = StringBoxSlots[StringBoxSlots.Length-1].OriginalSize or CurrSize#>
        ResizeEvent.Signal()
        Winner:=race:
            block:
                if(Time > 0.0, StepMax := Int[30*Time]):
                    for(Step := 1..StepMax):
                        StepPercent := (Step*1.0)/(StepMax*1.0)
                        StepSize:=Lerp(OriginalSize, EndSize, StepPercent.ApplyInterpolation(InterpolationType))
                        SetTextSize(InRange, StepSize, ?IsScaling:=IsScaling)
                        Sleep(0.0)
                else:
                    SetTextSize(InRange, EndSize, ?IsScaling:=IsScaling)
                1
            block:
                ResizeEvent.Await()
                -1
        if(AlwaysComplete?):
            SetTextSize(InRange, EndSize, ?IsScaling:=IsScaling)
        if(SaveSize?):
            ThisStringBoxRange := StringBoxSlots.Slice[Range(0), Range(1)] or StringBoxSlots
            for(ThisStringBox:ThisStringBoxRange):
                set ThisStringBox.CurrSize = CurrSize*DesiredSize

    # Set the spacer sizes between letters #
    SetSpacerSize<public><final>(InSize:float):void=
        set CurrSpacerSize = InSize+1.0
        SetTextSize(SizeMultiplier)

    # Set the space size between words #
    SetSpacingSize<public><final>(InSize:float):void=
        set SpacingMultiplier = InSize
        SetTextSize(SizeMultiplier)

    # Get the size that this string_block takes up on the screen #
    GetTextScreenSize<public><final>():vector2=
        #var TotalSpacerSize : float = 0.0
        var TotalCharacterSize : float = 0.0
        var HighestYSize : float = 0.0
        var CurrBoxIndex : int = 0
        StringBlocks:=GetVisibleStringSlots()
        TextLength:=GetTextLength()
        for(ThisStackBox:StringBlocks):
            ThisStringSize:=ThisStackBox.StringWidget.GetDesiredSize()
            #if(CurrBoxIndex < TextLength-1):
                #set TotalSpacerSize += ThisStackBox.SpacerWidget.GetDesiredSize().X
            set TotalCharacterSize += ThisStringSize.X
            if(ThisStringSize.Y > HighestYSize):
                set HighestYSize = ThisStringSize.Y
            set CurrBoxIndex += 1
        return vector2{X:=TotalCharacterSize, Y:=HighestYSize}

    GetTextLength<public><final>()<transacts>:int=
        var LengthSubtract : int = 0
        InSubstringIndexes := GetWordIndexes(CurrentText)
        for(ThisSubstringIndexes : InSubstringIndexes):
            set LengthSubtract += (ThisSubstringIndexes(1)-ThisSubstringIndexes(0))
        #Print("TextLength for {CurrentText}: {(CurrentText.Length-LengthSubtract)}")
        return (CurrentText.Length-LengthSubtract)

    # Set the transform, i.e. position and size #
    SetTransform<public><final>(DesiredPosition:vector2, DesiredSize:float):void=
        SetPosition(DesiredPosition)
        SetTextSize(DesiredSize)

    # Move the transform over time #
    MoveTransformEvent<private>:event()=event(){}
    MoveTransform<public><final>(DesiredPosition:vector2, DesiredSize:float, Time:float, ?AlwaysComplete:logic=false)<suspends>:void=
        MoveTransformEvent.Signal()
        Winner:=race:
            block:
                sync:
                    MovePosition(DesiredPosition, Time)
                    MoveTextSize(DesiredSize, Time)
                1
            block:
                MoveTransformEvent.Await()
                -1
        if(AlwaysComplete?, Winner=-1):
            SetTransform(DesiredPosition, DesiredSize)

    <# FUNCTIONS FOR WORD & COMMAND FINDING #>

    IndexToCharacterSlot(Index:int)<transacts><decides>:string_block_slot=
        StringBoxSlots[Index*2]

    GetWordIndexes<private>(InString:string)<transacts>:[]tuple(int, int)=
        var Indexes : []tuple(int, int) = array{}
        if(ThisFontFolder:=GetFontFolder[Font]):
            for(ThisWord -> Texture : ThisFontFolder.WordTextures):
                var Appearances : int = 0
                loop:
                    if(WordIndex := InString.FindSubstring[ThisWord, ?DesiredAppearance:=Appearances]):
                        set Indexes += array{(WordIndex, WordIndex+ThisWord.Length-1)}
                        set Appearances += 1
                    else:
                        break
        else{Print("~ERROR: Failed to find the font folder [{Font}]")}
        return Indexes

    (InIndexes:[]tuple(int, int)).FindMatchingStartIndex<private>(InIndex:int)<transacts><decides>:tuple(int, int)=
        var MaybeMatchingIndex : ?tuple(int, int) = false
        for:
            ThisIndexes : InIndexes
            ThisIndexes(0) = InIndex
        do:
            set MaybeMatchingIndex = option{ThisIndexes}
        MaybeMatchingIndex?

    GetSubstringIndexes_Start<private>(InString:string):[]int=
        var Indexes : []int = array{}
        if(ThisFontFolder:=GetFontFolder[Font]):
            for(ThisWord -> Texture : ThisFontFolder.WordTextures):
                var Appearances : int = 0
                loop:
                    if(WordIndex:=InString.FindSubstring[ThisWord, ?DesiredAppearance:=Appearances]):
                        set Indexes += array{WordIndex}
                        set Appearances += 1
                    else:
                        break
        return Indexes

    GetSubstringIndexes_End<private>(InString:string):[]int=
        var Indexes : []int = array{}
        if(ThisFontFolder:=GetFontFolder[Font]):
            for(ThisWord -> Texture : ThisFontFolder.WordTextures):
                var Appearances : int = 0
                loop:
                    if(WordIndex:=InString.FindSubstring[ThisWord, ?DesiredAppearance:=Appearances]):
                        set Indexes += array{WordIndex+ThisWord.Length}
                        set Appearances += 1
                    else:
                        break
        return Indexes

    GetWidgetAtIndex(InIndex:int)<transacts><decides>:widget=
        return StringBoxSlots[InIndex].StringWidget

    GetWidgetAtSubstring(InSubstring:string, ?DesiredAppearance:int=0):?widget=
        InSubstringConv:=InSubstring.ToLower()
        SubstringIndexes := GetWordIndexes(CurrentText)
        var CharIndex : int = 0
        var Appearances : int = 0
        StringBlocks:=GetVisibleStringSlots()
        for(ThisStackBox:StringBlocks):
            if(MatchingIndex:=SubstringIndexes.FindMatchingStartIndex[CharIndex]):
                if(ThisWord:=CurrentText.Slice[CharIndex, MatchingIndex(1)+1]):
                    if(ThisWord = InSubstringConv):
                        if(Appearances >= DesiredAppearance):
                            return option{ThisStackBox.StringWidget}
                        else{set Appearances += 1}
                    set CharIndex += (MatchingIndex(1)-MatchingIndex(0))
                else{Print("~ERROR: Couldn't find the word within the string!")}
            else if("{CurrentText[CharIndex]}" = InSubstringConv):
                if(Appearances >= DesiredAppearance):
                    return option{ThisStackBox.StringWidget}
                else{set Appearances += 1}
            set CharIndex += 1
        return false

        # <CommandType_Specification> TEXT HERE <E> #

    CommandTypes : []string = array{
        "c_"
        "ca_"
        "ch_"
    }
    ColorCommands : [string]color = map{
        "w" => NamedColors.White
        "r" => NamedColors.Red
        "o" => NamedColors.Orange
        "y" => NamedColors.Yellow
        "g" => NamedColors.Green
        "b" => NamedColors.Blue
        "i" => NamedColors.Indigo
        "v" => NamedColors.Violet
    }
    GetColorArray(InSpecification:string)<transacts><decides>:[]color=
        var ThisColorArray : []color = array{}
        case(InSpecification):
            "rainbow" => set ThisColorArray = RainbowArray
            "flash" => set ThisColorArray = FlashArray
            "gold wave" => set ThisColorArray = GoldWaveArray
            _ =>
        ThisColorArray.Length > 0
        return ThisColorArray

    GetRealText(InText:string):string=
        var RealText : string = ""
        CommandPoints := GetCommands(InText)
        if(CommandPoints.Length > 0):
            var RealTextRanges : []tuple(int, int) = array{}
            if(ThisFirstCommand:=CommandPoints[0], ThisFirstCommand.CommandStartIndex > 0):
                if(ThisFirstText:=InText.Slice[0, ThisFirstCommand.CommandStartIndex]):
                    set RealText += ThisFirstText
            var CurrCommPoint : int = 0
            for(ThisCommandPoint:CommandPoints):
                RealTextRange:=ThisCommandPoint.GetTextRange(?IncludeCommands:=true)
                if(ThisPreviousCommandPoint:=CommandPoints[CurrCommPoint-1]):
                    LastCommEnd := ThisPreviousCommandPoint.CommandEndIndex+3
                    if(ThisRealText:=InText.Slice[LastCommEnd, RealTextRange(0)]):
                        set RealText += ThisRealText
                if(ThisRealText:=InText.Slice[RealTextRange(0)+ThisCommandPoint.Command.Length+ThisCommandPoint.Specification.Length+2, RealTextRange(1)+1]):
                    set RealText += ThisRealText
                set CurrCommPoint += 1
                
            if(ThisLastCommand:=CommandPoints[CommandPoints.Length-1]):
                if(ThisLastText:=InText.Slice[ThisLastCommand.CommandEndIndex+3, InText.Length]):
                    set RealText += ThisLastText
        else:
            return InText
        return RealText

    ExecuteCommands(InCommandedText:string):void=
        SetColor(NamedColors.White)
        CommandPoints := GetCommands(InCommandedText)
        for(ThisCommandPoint:CommandPoints):
            if(ThisCommandTypeIndex:=CommandTypes.Find[ThisCommandPoint.Command]):
                case(ThisCommandTypeIndex):
                    0 => # COLOR #
                        if(ThisColor:=ColorCommands[ThisCommandPoint.Specification]):
                            Range:=ThisCommandPoint.GetTextRange(?IncludeCommands:=false)
                            SetColor((Range(0), Range(1)), ThisColor)
                        else:
                            Print("~ERR: Failed to fetch a color with the type [{ThisCommandPoint.Specification}]")
                    1 => # COLOR CHANGING #
                        if(ThisColorArray:=GetColorArray[ThisCommandPoint.Specification]):
                            Range:=ThisCommandPoint.GetTextRange(?IncludeCommands:=false)
                            #Print("Range: {Range(0)} -> {Range(1)}")
                            spawn{SetColors(ThisColorArray, 0.15, Range, ?Loops:=-1)}
                        else:
                            Print("~ERR: Failed to fetch color array [{ThisCommandPoint.Specification}]")
                    2 => # COLOR WITH HEX #
                        ThisColorHex:=MakeColorFromHex("{ThisCommandPoint.Specification}")
                        Range:=ThisCommandPoint.GetTextRange(?IncludeCommands:=false)
                        SetColor((Range(0), Range(1)), ThisColorHex)
                    _ => Print("~ERR: Unknown command")
            else:
                Print("~ERR: Failed to find a command [{ThisCommandPoint.Command}]")
        if(CommandPoints.Length <= 0):
            SetColor(NamedColors.White)

    GetCommands(InCommandedText:string):[]stringcommand=
        var CommandPoints : []stringcommand = array{}
        loop:
            if:
                CommStartIndex:=InCommandedText.FindSubstring["<", ?DesiredAppearance:=CommandPoints.Length*2]
                CommEndIndex:=InCommandedText.FindSubstring[">", ?DesiredAppearance:=CommandPoints.Length*2] > CommStartIndex
                CommSeperatorIndex:=InCommandedText.FindSubstring["_", ?DesiredAppearance:=CommandPoints.Length*2] < CommEndIndex
                CommandStopIndex:=InCommandedText.FindSubstring["<e>", ?DesiredAppearance:=CommandPoints.Length] > CommEndIndex
                CommandType:=InCommandedText.Slice[CommStartIndex+1, CommSeperatorIndex+1]
                CommandTypeSpecification:=InCommandedText.Slice[CommStartIndex+CommandType.Length+1, CommEndIndex]
            then:
                NewCommandPoint : stringcommand = stringcommand{
                    Command := CommandType
                    Specification := CommandTypeSpecification
                    CommandStartIndex := CommStartIndex
                    CommandEndIndex := CommandStopIndex
                }
                for(ThisCommandPoint:CommandPoints):
                    set NewCommandPoint.IndexSubtractive += ThisCommandPoint.GetTotalLength()
                set CommandPoints += array{NewCommandPoint}
            else:
                break
        return CommandPoints

    (InCommandPoint:stringcommand).GetTextRange(?IncludeCommands:logic=false):tuple(int, int)=
        if(IncludeCommands?):
            return (InCommandPoint.CommandStartIndex, InCommandPoint.CommandEndIndex-1)
        else:
            return (InCommandPoint.CommandStartIndex-InCommandPoint.IndexSubtractive, (InCommandPoint.CommandEndIndex-InCommandPoint.IndexSubtractive)-(3+InCommandPoint.Command.Length+InCommandPoint.Specification.Length))
            # (InCommandPoint.CommandEndIndex-InCommandPoint.IndexSubtractive)-6-InCommandPoint.Specification.Length

    var<internal> DebugState : logic = false
    EnableDebug<public>():void=
        WidgetControl.EnableDebug()
        set DebugState = true

    DisableDebug<public>():void=
        WidgetControl.DisableDebug()
        set DebugState = false

    DebugPrint(String:string):void=
        if(DebugState?):
            Print(String)

    GetRootWidget():dynamic_widget=
        return WidgetControl

    GetVisibleStringSlots():[]string_block_slot=
        return StringBoxSlots.Slice[0, GetTextLength()] or array{}

string_block_slot := class:

    var String : string
    var StackboxWidget : ?stack_box_slot = false
    StringWidget : texture_block = texture_block{DefaultImage:=Art.UIStuff.Fonts.InteractionText.Numbers.Number_0}
    SpacerWidget : color_block = color_block{DefaultOpacity:=0.0, DefaultDesiredSize:=vector2{X:=0.0,Y:=0.0}}

    var CurrSize : vector2

    var StackboxSpacerSlot : ?stack_box_slot = false
    SetAndGetSpacerStackboxSlot():stack_box_slot=
        ThisStackboxSlot:=stack_box_slot{
            Widget:=SpacerWidget
        }
        set StackboxSpacerSlot = option{ThisStackboxSlot}
        return ThisStackboxSlot
