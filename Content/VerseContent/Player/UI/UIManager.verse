using { /Fortnite.com/Devices }
using { /Fortnite.com/Characters }
using { /Fortnite.com/Game }
using { /Fortnite.com/UI }
using { /Fortnite.com/FortPlayerUtilities }
using { /Verse.org/Simulation }
using { /Verse.org/Random }
using { /Verse.org/Colors }
using { /UnrealEngine.com/Temporary/Diagnostics }
using { /UnrealEngine.com/Temporary/SpatialMath }
using { /UnrealEngine.com/Temporary/UI }
using { /Verse.org/Assets }
using { Player }
using { Room }
using { Spawning }
using { Powerups }
using { InteractiveModule }

(InAgent:agent).IsUIReady()<transacts><decides>:void=
    InAgent.GetUIManager().AppliedUICanvas?

(InAgent:agent).GetUIManager<public>()<transacts>:ui_manager=
    return GlobalUICollection[GetSession()][InAgent] or block{
        NewUI : ui_manager = ui_manager{User:=InAgent}
        if. set GlobalUICollection[GetSession()][InAgent] = NewUI
        else if:
            not GlobalUICollection[GetSession()]
            set GlobalUICollection[GetSession()] = map{}
        if:
            CurrentCollection := GlobalUICollection[GetSession()]
            set GlobalUICollection[GetSession()] = ConcatenateMaps(CurrentCollection, map{InAgent => NewUI})
        NewUI
    }

var GlobalUICollection<public> : weak_map(session, [agent]ui_manager) = map{}

StringToMessage<localizes>(String : string) : message = "{String}"
PlayerToMessage<localizes>(InPlayer : agent) : message = "{InPlayer}"
PlayerAndStringToMessage<localizes>(InPlayer:agent, InString:string) : message = "{InPlayer}{InString}"

RainbowArray<public> : []color = array{
    NamedColors.Red
    NamedColors.Orange
    NamedColors.Yellow
    NamedColors.Green
    NamedColors.Blue
    NamedColors.Indigo
    NamedColors.Violet
}
FlashArray<public> : []color = array{
    NamedColors.White
    NamedColors.White
    MakeColorFromHex("f2f2f2")
    MakeColorFromHex("e6e6e6")
    MakeColorFromHex("d9d9d9")
    MakeColorFromHex("e6e6e6")
    MakeColorFromHex("f2f2f2")
    NamedColors.White
}

GreyscaleArray<public> : []color = array{
    NamedColors.White
    NamedColors.White
    MakeColorFromHex("D6D6D6")
    MakeColorFromHex("BABABA")
    MakeColorFromHex("9F9F9F")
    MakeColorFromHex("858585")
    MakeColorFromHex("6C6C6C")
    MakeColorFromHex("535353")
    MakeColorFromHex("3C3C3C")
    MakeColorFromHex("272727")
    NamedColors.Black
    MakeColorFromHex("272727")
    MakeColorFromHex("3C3C3C")
    MakeColorFromHex("535353")
    MakeColorFromHex("6C6C6C")
    MakeColorFromHex("858585")
    MakeColorFromHex("9F9F9F")
    MakeColorFromHex("BABABA")
    MakeColorFromHex("D6D6D6")
    NamedColors.White
    NamedColors.White
}

GoldWaveArray<public> : []color = array{
    MakeColorFromHex("F7C319")
    MakeColorFromHex("FACF43")
    MakeColorFromHex("FCDB6D")
    MakeColorFromHex("FDEB9E")
    MakeColorFromHex("FDFBCF")

    MakeColorFromHex("FDFBCF")
    MakeColorFromHex("FDEB9E")
    MakeColorFromHex("FCDB6D")
    MakeColorFromHex("FACF43")
    MakeColorFromHex("F7C319")
}

GreenFlash<public> : []color = array{
    NamedColors.White
    MakeColorFromHex("baf8ba")
    MakeColorFromHex("7cf87c")
    MakeColorFromHex("3ef83e")
    MakeColorFromHex("00f800")
    MakeColorFromHex("00f800")
    MakeColorFromHex("00f800")
    MakeColorFromHex("00f800")
    MakeColorFromHex("3ef83e")
    MakeColorFromHex("7cf87c")
    MakeColorFromHex("baf8ba")
    NamedColors.White
    MakeColorFromHex("D6D6D6")
    MakeColorFromHex("BABABA")
    MakeColorFromHex("9F9F9F")
    MakeColorFromHex("858585")
    MakeColorFromHex("6C6C6C")
    MakeColorFromHex("535353")
    MakeColorFromHex("3C3C3C")
    MakeColorFromHex("272727")
    NamedColors.Black
}

WhiteFadeIN<public> : []color = array{
    NamedColors.Black
    MakeColorFromHex("272727")
    MakeColorFromHex("3C3C3C")
    MakeColorFromHex("535353")
    MakeColorFromHex("6C6C6C")
    MakeColorFromHex("858585")
    MakeColorFromHex("9F9F9F")
    MakeColorFromHex("BABABA")
    MakeColorFromHex("D6D6D6")
    NamedColors.White
}

ui_manager<public> := class:

    var User<public> : agent

    var MainCanvas<public> : canvas = canvas{}

    MainMenuUI<public> : mainmenu_ui = mainmenu_ui{}

    AdminPanel_Menu<public> : admin_panel_menu = admin_panel_menu{}

    AdminPanel_Game<public> : admin_panel_game = admin_panel_game{}

    #GameOverPrompt<public> : gameover_prompt = gameover_prompt{}

    PowerupUI<public> : powerup_ui = powerup_ui{}

    PerkUICollection<public> : perk_ui_collection = perk_ui_collection{}

    InfoUI : dynamic_widget = dynamic_widget{}

    QuestUI<public> : quest_ui = quest_ui{}

    TrialUI<public> : trial_ui = trial_ui{}

    PowerPunchDetails<public> : powerpunch_details = powerpunch_details{}

    var PartsUI<public> : parts_ui = parts_ui{}

    DownedHealthUI<public> : progress_bar = progress_bar{}
    DownedHealthSkull<public> : texture_block = texture_block{DefaultImage:=Art.UIStuff.Icons.PointBackground.Skull_Icon, DefaultDesiredSize:=vector2{X:=26.0,Y:=26.0}}

    CodeWatermark : texture_block = texture_block{DefaultImage:=Art.UIStuff.Icons.Code_UI_01, DefaultDesiredSize:=vector2{X:=400.0,Y:=250.0}}

    LeaderboardSubmitControl : dynamic_widget = dynamic_widget{}
    SubmitLeaderboardTexture : texture_block = texture_block{DefaultImage:=Art.UIStuff.Leaderboards.SubmitLeaderboards, DefaultDesiredSize:=vector2{X:=650.0,Y:=225.0}}

    ToggleInfoUI<public>(InState:logic):void=
        if(InState?):
            # The info UI is not on screen, so show it #
            Scoreboard.Show()
            PartsUI.Show()
            #if(GlobalPlayerManager[GetSession()].ToggleInfoUI_InputDevice.IsHeld[ParentPlayerInfo?.Agent]):
            #    ToggleInfoUI(false)
        else:
            # The info UI is on screen, so hide it #
            Scoreboard.Hide()
            PartsUI.Hide()

    WaveStartText : string_block = string_block{Font:="Quest"} # -> Used to display 'wave' above the wave number in the intro anim. Tells the player what the number represents
    WaveIntroNumber : string_block = string_block{Font:="Wave"} # -> Used to display the intro anim for the wave number. Using this due to the need for anchoring in the middle
    WaveText : string_block = string_block{Font:="Wave", ApplySizeRatios:=false} # -> The actual wave number that is increased every wave
    SetWaveText<public>(InText:string):void=
        WaveText.SetText(InText)
        WaveText.Show()

    LocationText : string_block = string_block{Font:="Quest", BackplateTexture := option{Art.UIStuff.Fonts.InteractionBackground.InteractionBackground}, BackplateOffsets:=vector2{X:=10.0,Y:=10.0}}
    LocationIcon : texture_block = texture_block{DefaultImage:=Art.UIStuff.Icons.Map_Icon, DefaultDesiredSize:=vector2{X:=40.0,Y:=40.0}}
    SetLocationText<public>(InText:string):void=
        LocationText.SetText(InText)
        LocationText.Show()

    PointsText : points_block = points_block{}
    SetPointsText<public>(InAmount:int, ?PointIncrement:?int=false, ?PointColor:?color=false, ?Animate:logic=true):void=
        PointsText.ChangePoints("${InAmount}")
        PointsText.Show()
        if(InAmount < 1000000):
            if(Animate?):
                if(ThisPointIncrement:=PointIncrement?):
                    spawn{PlayPointsEffect(ThisPointIncrement, ?PointColor:=PointColor)}
                    spawn{BouncePoints()}
                else:
                    spawn{ShrinkBouncePoints()}

    Scoreboard<public> : scoreboard_ui = scoreboard_ui{}

    EndGame_WavesSurvived_Text<public> : string_block = string_block{Font:="Quest"}
    EndGame_DifficultyMode_Text : string_block = string_block{Font:="Quest"}

    EndGame_Timestamp_Text<public> : string_block = string_block{Font:="Quest"}

    GlobalPoints<public> : globalpoints = globalpoints{}

    RocketEE_Text<public> : string_block = string_block{Font:="Quest"}

    UpdatePlayerIcon(InAgent:agent):void=
        if(InAgent = User):
            PointsText.UpdatePlayerIcon(InAgent)
        PlayerList:=GetAgents()
        for(ThisPlayer:PlayerList):
            ThisPlayer.GetUIManager().GlobalPoints.UpdatePlayerIcon(InAgent)

    BouncePointsEvent : event() = event(){}
    BouncePoints<private>()<suspends>:void=
        if(not InsufficientEffectState?):
            BouncePointsEvent.Signal()
            race:
                block:
                    PointsText.Resize(28.0, 0.2)
                    PointsText.Resize(25.0, 0.12)
                block:
                    BouncePointsEvent.Await()
                    spawn{PointsText.Resize(25.0, 0.0)}

    ShrinkBouncePoints<private>()<suspends>:void=
        if(not InsufficientEffectState?):
            PointsText.Resize(23.0, 0.2)
            PointsText.Resize(25.0, 0.12)

    # Stores the widgets for the point effects that fly out upon earning points
    #   True state => This Point Effect is not in use
    #   False state => This Point Effect is in use
    var PointsEffectContainer<private> : []tuple(string_block, logic) = array{}

    PlayPointsEffect(PointAmount:int, ?PointColor:?color=false)<suspends>:void=
        ThisPointEffectTuple := GetAvailablePointEffect()? or block{
                Print("Creating new point effect, reached maximum")
                NewPointEffect:=string_block{Font:="Points"}
                MainCanvas.AddWidget(NewPointEffect.Create("", vector2{X:=340.0,Y:=860.0}, 15.0, ?SizeRatio:=vector2{X:=1.0,Y:=1.6}, ?AnchorOverride:=option{anchors{Minimum:=vector2{Y:=1.0},Maximum:=vector2{Y:=1.0}}}, ?Alignment:=vector2{Y:=1.0}))
                NewPointEffect.Hide()
                set PointsEffectContainer += array{(NewPointEffect, false)}
                (NewPointEffect, PointsEffectContainer.Length-1)
            }
        ThisPointEffectTuple(0).SetText("+{PointAmount}")
        ThisPointEffectTuple(0).SetPosition(vector2{X:=340.0,Y:=860.0})
        if(ThisColor:=PointColor?):
            ThisPointEffectTuple(0).SetColor(ThisColor)
        else if(PointAmount >= 1000):
            ThisPointEffectTuple(0).SetColor(NamedColors.Orange)
        else:
            ThisPointEffectTuple(0).SetColor(NamedColors.Yellow)
        RandomEndVector := vector2{X:=415.0,Y:=GetRandomFloat(890.0, 830.0)}
        #if(ThisTickRateMult:=ParentPlayerInfo?.TickRatePercent):
        ThisPointEffectTuple(0).Show()
        ThisPointEffectTuple(0).MovePosition(RandomEndVector, 0.40)
        ThisPointEffectTuple(0).SetPosition(vector2{X:=4200.0,Y:=2500.0})
        ThisPointEffectTuple(0).Hide()
        if. set PointsEffectContainer[ThisPointEffectTuple(1)] = (ThisPointEffectTuple(0), true)

    GetAvailablePointEffect()<transacts>:?tuple(string_block, int)=
        for(Index->ThisPointTuple:PointsEffectContainer):
            if:
                ThisPointTuple(1)?
                set PointsEffectContainer[Index] = (ThisPointTuple(0), false)
            then:
                return option{(ThisPointTuple(0), Index)}
        false

    InitializePointEffects<private>():void=
        if(PointsEffectContainer.Length <= 0):
            PointEffects:=for(Index:=0..15):
                NewPointEffect:=string_block{Font:="Points"}
                (NewPointEffect, true)
            set PointsEffectContainer = PointEffects
        for(ThisPointEffect:PointsEffectContainer):
            MainCanvas.AddWidget(ThisPointEffect(0).Create("", vector2{X:=340.0,Y:=860.0}, 15.0, ?SizeRatio:=vector2{X:=1.0,Y:=1.6}, ?AnchorOverride:=option{anchors{Minimum:=vector2{Y:=1.0},Maximum:=vector2{Y:=1.0}}}, ?Alignment:=vector2{Y:=1.0}))
            ThisPointEffect(0).Hide()

    var InsufficientEffectState : logic = false
    InsufficientEffectEvent : event() = event(){}
    InsufficientEffect_Points<public>()<suspends>:void=
        set InsufficientEffectState = true
        InsufficientEffectEvent.Signal()
        race:
            block:
                sync:
                    block:
                        PointsText.MovePoints(vector2{X:=10.0}, 0.15, ?IsDelta:=true)
                        PointsText.MovePoints(vector2{X:=-12.0}, 0.07, ?IsDelta:=true)
                        PointsText.MovePoints(vector2{X:=2.0}, 0.15, ?IsDelta:=true)
                    block:
                        PointsText.SetColor(MakeColorFromHex("E61717"))
                        Sleep(0.066)
                        PointsText.SetColor(MakeColorFromHex("EA4444"))
                        Sleep(0.066)
                        PointsText.SetColor(MakeColorFromHex("EE7272"))
                        Sleep(0.066)
                        PointsText.SetColor(MakeColorFromHex("F29F9F"))
                        Sleep(0.066)
                        PointsText.SetColor(MakeColorFromHex("F6CDCD"))
                        Sleep(0.066)
                        PointsText.ResetColor()
                        Sleep(0.066)
                set InsufficientEffectState = false
            block:
                InsufficientEffectEvent.Await()
        PointsText.MovePoints(vector2{X:=0.0, Y:=0.0}, 0.0)
        PointsText.ResetColor()

    ReviveBar<public> : progress_bar = progress_bar{}

    ShieldIcon<public> : texture_block = texture_block{DefaultImage:=Art.UIStuff.Icons.QuestObjects.ShieldIcon, DefaultDesiredSize:=vector2{X:=100.0,Y:=100.0}}
    ShieldWidget<public> : dynamic_widget = dynamic_widget{}
    ShieldHealthBar<public> : progress_bar = progress_bar{}

    PowerPunchWidget<public> : dynamic_widget = dynamic_widget{}

    AmmoAbilityTexture<public> : texture_block = texture_block{DefaultImage:=Art.UIStuff.Icons.Weapon.Fist_Icon, DefaultDesiredSize:=vector2{X:=40.0,Y:=40.0}}
    AmmoTypeWidget<public> : dynamic_widget = dynamic_widget{}
    StopAbilityFlashing<public>:event()=event(){}
    StartAbilityFlashing<public>(?Loops:int=0)<suspends>:void=
        var FlashLoops : int = 0
        race:
            StopAbilityFlashing.Await()
            loop:
                for(ThisFlashColor:FlashArray):
                    AmmoAbilityTexture.SetTint(ThisFlashColor)
                    Sleep(0.0)
                set FlashLoops += 1
                if(Loops > 0, FlashLoops > Loops):
                    break
                Sleep(0.0)
        AmmoAbilityTexture.SetTint(NamedColors.White)

    InteractionText<public> : string_block = string_block{Font:="Interaction", BackplateTexture:=option{Art.UIStuff.Fonts.InteractionBackground.InteractionBackground}}
    SetInteractionText<public>(InText:string):void=
        InteractionText.SetText(InText)
        if(InteractionText.VisibleState?):
            InteractionText.Show()

    DescriptionUI<public> : description_ui = description_ui{}
    SetDescription<public>(InText:string, InIconText:string, InOrigin:?Interactive_Base):void=
        DescriptionUI.Change(InText, InIconText, InOrigin)
    
    HideDescription<public>():void=
        DescriptionUI.Hide()

    PopoutInteractionEvent:event()=event(){}
    PopoutInteractionEnded:event()=event(){}
    PopoutInteractionFinished:event()=event(){}
    Popout_Interaction()<suspends>:void=
        var PreviousTextState : int = -1
        loop:
            if(PreviousTextState = 1):
                InteractionText.TextChangedEvent.Await()
            PopoutInteractionEvent.Signal()
            Winner:=race:
                block:
                    Sleep(0.0)
                    InteractionText.MoveTransform(vector2{X:=960.0,Y:=820.0}, 32.0, 0.15)
                    InteractionText.MoveTransform(vector2{X:=960.0,Y:=840.0}, 30.0, 0.15)
                    PopoutInteractionFinished.Signal()
                    #Print("~Finished popout")
                    1
                block:
                    PopoutInteractionEvent.Await()
                    InteractionText.SetTransform(vector2{X:=960.0,Y:=840.0}, 30.0)
                    #Print("~Leaving reason: PopoutInteraction")
                    -1
                block:
                    InteractionText.TextChangedEvent.Await()
                    InteractionText.SetTransform(vector2{X:=960.0,Y:=840.0}, 30.0)
                    #Print("~Leaving reason: TextChanged")
                    -1
            PopoutInteractionEnded.Signal()
            set PreviousTextState = Winner

    InteractableAnimation<public>(InteractableLocation:Interactable)<suspends>:void=
        #Print("Playing animation")
        if(ThisPlayerInfo:=User.GetPlayerInfo[]):
            if(InteractableLocation.UseUIAnimation?):
                if(InteractableLocation.IsEqual[ThisPlayerInfo.ReviveInteractable]):
                    # Interaction is from reviving #
                    InteractionText.Hide()
                    ReviveBar.SetColor(NamedColors.White)
                    ReviveBar.Show()
                    ReviveBar.SetProgress(0.0)
                    var InteractionSuccess : logic = false
                    defer{
                        ReviveBar.Hide()
                        ReviveBar.HaltProgressChanging()
                    }
                    ReviveBar.ChangeProgressOverTime(InteractableLocation.GetInteractionTime(?InAgent:=option{User}, ?TimeMultiply:=ThisPlayerInfo.InteractTimeMultiplier), 100.0, ?InAgent:=option{User})
                    #ReviveBar.Hide()
                    #ReviveBar.HaltProgressChanging()
                    #if(Winner=-1):
                    #    InteractionText.Show()
                else:
                    # Play normal interaction #
                    Progress_Interaction(InteractableLocation.GetInteractionTime(?InAgent:=option{User}, ?TimeMultiply:=ThisPlayerInfo.InteractTimeMultiplier))
            else:
                ProgressInteractionEvent.Signal()
                SafeSleep(InteractableLocation.GetInteractionTime(?InAgent:=option{User}, ?TimeMultiply:=ThisPlayerInfo.InteractTimeMultiplier))

    ProgressInteractionEvent:event()=event(){}
    #InterruptProgressInteraction:event()=event(){}
    ProgressInteractionFinished:event()=event(){}
    #ProgressInteractionInterrupted:event()=event(){}
    InteractionInterruptedEvent:event()=event(){}

    ProgressBackgroundWidget : dynamic_widget = dynamic_widget{}
    ProgressColorBlock : color_block = color_block{DefaultColor:=NamedColors.Red, DefaultOpacity:=0.35, DefaultDesiredSize:=vector2{X:=0.0,Y:=50.0}}

    Progress_Interaction(CompletionTime:float)<suspends>:void=
        ProgressColorBlock.Show()
        DesiredSize:=vector2{X:=InteractionText.GetTextScreenSize().X+InteractionText.BackplateOffsets.X,Y:=ProgressColorBlock.GetDesiredSize().Y}
        defer{
            ProgressColorBlock.SetDesiredSize(vector2{X:=0.0,Y:=50.0})
            ProgressColorBlock.Hide()
        }
        ProgressColorBlock.Resize(DesiredSize, CompletionTime)

    RevivingStartedAlertText<public>:string_block=string_block{Font:="Interaction"}
    RevivingAlertEvent:event()=event(){}
    DisplayRevivingAlert<public>()<suspends>:void=
        RevivingStartedAlertText.SetText("you are being revived")
        RevivingStartedAlertText.Show()
        RevivingAlertEvent.Signal()
        race:
            block:
                Sleep(0.0)
                RevivingStartedAlertText.MoveTransform(vector2{X:=960.0,Y:=820.0}, 32.0, 0.15)
                RevivingStartedAlertText.MoveTransform(vector2{X:=960.0,Y:=840.0}, 30.0, 0.15)
            RevivingAlertEvent.Await()
            RevivingStartedAlertText.MoveTransform(vector2{X:=960.0,Y:=840.0}, 30.0, 0.0)
        
    var KnockedAlertQueue<private> : []agent = array{}
    KnockedPlayerAlertText:text_block=text_block{<#DefaultTextOpacity:= 0.0,#> DefaultShadowOffset:=option{vector2{X:=20.0,Y:=20.0}}, DefaultShadowColor:=NamedColors.Black, DefaultTextColor:=NamedColors.White}
    DisplayKnockedAlert<public>(InPlayer:agent)<suspends>:void=
        if(IsGameState[game_state.InGame]):
            race:
                block:
                    set KnockedAlertQueue += array{InPlayer}
                    loop:
                        if(KnockedAlertQueue[0]=InPlayer):
                            break
                        Sleep(0.75)
                    Print("playing knocked alert text")
                    KnockedPlayerAlertText.SetText(PlayerAndStringToMessage(InPlayer, " was knocked!"))
                    for(Step := 1..20):
                        StepPercent:=((Step*1.0)/20.0)
                        if(NextStepOpacity:=type {_X:float where 0.000000 <= _X, _X <= 1.000000}[Lerp(0.0, 1.0, StepPercent)]):
                            KnockedPlayerAlertText.SetTextOpacity(NextStepOpacity)
                        Sleep(0.06)
                    KnockedPlayerAlertText.SetTextOpacity(1.0)
                    Sleep(2.75)
                    for(Step := 1..30):
                        StepPercent:=((Step*1.0)/30.0)
                        if(NextStepOpacity:=type {_X:float where 0.000000 <= _X, _X <= 1.000000}[Lerp(0.5, 0.0, StepPercent)]):
                            KnockedPlayerAlertText.SetTextOpacity(NextStepOpacity)
                        Sleep(0.0)
                    KnockedPlayerAlertText.SetTextOpacity(0.0)
                    Sleep(0.5)
                    if(NewQueue:=KnockedAlertQueue.RemoveFirstElement[InPlayer]):
                        set KnockedAlertQueue = NewQueue
                AwaitGameState(game_state.GameOver)
        KnockedPlayerAlertText.SetTextOpacity(0.0)

    PausedScreenCover<public> : color_block = color_block{DefaultColor:=NamedColors.Black, DefaultOpacity:=0.85, DefaultDesiredSize:=vector2{X:=7700.0, Y:=4350.0}}

    BlackScreenBlock<public> : color_block = color_block{DefaultColor:=NamedColors.Black,DefaultOpacity:=0.0,DefaultDesiredSize:=vector2{X:=7700.0,Y:=4350.0}}
    FadeInBlackScreen<public>(TimeIn:float)<suspends>:void=
        if(TimeIn > 0.0, StepMax := Int[30*TimeIn]):
            for(Step := 1..StepMax):
                StepPercent := (Step*1.0)/(StepMax*1.0)
                if(NextStepOpacity:=type {_X:float where 0.000000 <= _X, _X <= 1.000000}[Lerp(0.0, 1.0, StepPercent)]):
                    BlackScreenBlock.SetOpacity(NextStepOpacity)
                    BlackScreenBlock.Show()
                Sleep(0.0)
        BlackScreenBlock.SetOpacity(1.0)

    FadeOutBlackScreen<public>(TimeOut:float)<suspends>:void=
        if(TimeOut > 0.0, StepMax := Int[30*TimeOut]):
            for(Step := 1..StepMax):
                StepPercent := (Step*1.0)/(StepMax*1.0)
                if(NextStepOpacity:=type {_X:float where 0.000000 <= _X, _X <= 1.000000}[Lerp(1.0, 0.0, StepPercent)]):
                    BlackScreenBlock.SetOpacity(NextStepOpacity)
                    BlackScreenBlock.Show()
                Sleep(0.0)
        BlackScreenBlock.SetOpacity(0.0)
        BlackScreenBlock.Collapse()

    var AppliedUICanvas<public> : ?canvas = false

    var GameplayUIControl : canvas = canvas{}

    #var DescriptionUIControl : canvas = canvas{}

    var AllowMenuUI : logic = false
    var<public> AllowGameplayPanel : logic = false

    # Creates and applies ui to the player using incoming player data #
    ApplyUI<public>(InAgent:agent):void=
        if(PlayerUI := GetPlayerUI[player[InAgent]]):
            if(not AppliedUICanvas?):
                #spawn{ClearFlyingPointsLoop()}
                if(ThisBuildableStorage:=InAgent.GetBuildableStorage[]):
                    set PartsUI = parts_ui{UIManager:=option{Self}, BuildableStorage:=ThisBuildableStorage}
            else if(ThisCanvas:=AppliedUICanvas?):
                PlayerUI.RemoveWidget(ThisCanvas)
            ThisUI := CreateUI(InAgent)
            set MainCanvas = ThisUI(0)
            set AppliedUICanvas = option{ThisUI(0)}
            if(ThisUI(1)? or AllowGameplayPanel?):
                PlayerUI.AddWidget(ThisUI(0), player_ui_slot{InputMode:=ui_input_mode.All})
            else:
                PlayerUI.AddWidget(ThisUI(0), player_ui_slot{InputMode:=ui_input_mode.None})
            if(ThisPlayerInfo:=InAgent.GetPlayerInfo[]):
                ThisPlayerInfo.UpdatePointsUI()
            spawn{SetupGlobalUI()}
            spawn{ShrinkBouncePoints()}
            InitializePointEffects()
        else{Print("~ERROR: Failed to get playerui!")}

    # ALL WIDGETS
    CreateUI(InAgent:agent):tuple(canvas, logic)=
        var InteractiveUIState : logic = true
        var CanvasSlots : []canvas_slot = array{}
        var GameplayUISlots : []canvas_slot = array{}

        if(IsGameState[game_state.InGame]):
            #Print("Creating gameplay UI")
            set InteractiveUIState = false
            set CanvasSlots += array{LeaderboardSubmitControl.Create(SubmitLeaderboardTexture, vector2{X:=960.0,Y:=940.0}, ?Alignment:=vector2{X:=0.5,Y:=1.0}, ?AnchorOverride:=option{anchors{Minimum:=vector2{X:=0.5,Y:=1.0}, Maximum:=vector2{X:=0.5,Y:=1.0}}})}

            # Initialize canvas widgets here #
            WaveStartText.Hide()
            #  vector2{X:=92.0,Y:=916.0}
            #set CanvasSlots += array{WaveStartText.Create("wave", vector2{X:=92.0,Y:=100.0}, 0.1, ?Alignment:=vector2{X:=0.5,Y:=0.5}, ?AutoAnchor:=true)}
            #  vector2{X:=92.0,Y:=996.0}
            set CanvasSlots += array{WaveText.Create("{GetWaveCount()}", vector2{X:=110.0,Y:=1180.0}, 75.0, ?SizeRatio:=vector2{X:=1.0,Y:=1.5}, ?Alignment:=vector2{Y:=1.0}, ?AnchorOverride:=option{anchors{Minimum:=vector2{X:=0.0,Y:=1.0}, Maximum:=vector2{X:=0.0,Y:=1.0}}})}
            #set CanvasSlots += array{WaveText.Create("{GetWaveCount[] or 1}", vector2{X:=164.0,Y:=324.0}, 90.0, ?Alignment:=vector2{Y:=1.0}, ?AnchorOverride:=option{anchors{Minimum:=vector2{X:=0.0,Y:=0.5}, Maximum:=vector2{X:=0.0,Y:=0.5}}})}

            # vector2{X:=80.0, Y:=850.0}
            # ?AnchorOverride:=option{anchors{Minimum:=vector2{X:=0.0,Y:=1.0}, Maximum:=vector2{X:=0.0,Y:=1.0}}}
            set CanvasSlots += array{PointsText.Create(InAgent, "0", vector2{X:=200.0, Y:=1300.0}, 25.0, ?Alignment:=vector2{Y:=1.0}, ?AnchorOverride:=option{anchors{Minimum:=vector2{X:=0.0,Y:=1.0}, Maximum:=vector2{X:=0.0,Y:=1.0}}})}
            #PointsText.ResizeBackplate(vector2{X:=300.0,Y:=350.0})

            # vector2{X:=64.0, Y:=84.0}
            set CanvasSlots += array{LocationText.Create(InAgent.GetOccupyingRoom[].RoomText or "main courtyard", vector2{X:=80.0, Y:=67.0}, 30.0, ?SizeRatio:=vector2{X:=1.0,Y:=1.6}, ?SpacerSize:=0.8, ?Alignment:=vector2{Y:=0.5}, ?AutoAnchor:=true)}

            set CanvasSlots += array{
                canvas_slot:
                    Widget:=overlay:
                        Slots:=array:
                            overlay_slot:
                                Widget:=texture_block{DefaultImage:=Art.UIStuff.Fonts.InteractionBackground.InteractionBackground, DefaultDesiredSize:=vector2{X:=45.0,Y:=45.0}}
                            overlay_slot:
                                Widget:=LocationIcon
                    SizeToContent:=true
                    Anchors:=anchors{
                        Minimum:=vector2{X:=0.024974,Y:=0.060487}
                        Maximum:=vector2{X:=0.024974,Y:=0.060487}
                    }
                    Alignment:=vector2{X:=0.5,Y:=0.5}
                    Offsets:=margin{Top:=0.0,Left:=0.0}
            }

            InteractionTextPos:=vector2{X:=960.0, Y:=840.0}
            InteractionBackplateOffset:=vector2{X:=20.0,Y:=20.0}
            ProgressBackgroundWidgetPosition := vector2{X:=InteractionTextPos.X+InteractionBackplateOffset.X, Y:=InteractionTextPos.Y-(InteractionBackplateOffset.Y/2.0)}
            #set CanvasSlots += array{ProgressBackgroundWidget.Create(ProgressColorBlock, ProgressBackgroundWidgetPosition, vector2{X:=0.5, Y:=InteractionText.GetScreenSize().Y+InteractionText.GetBackplateOffset().Y})}

            set CanvasSlots += array{RevivingStartedAlertText.Create("you are being revived", InteractionTextPos, 30.0, ?Alignment:=vector2{X:=0.5,Y:=0.5}, ?SizeRatio:=vector2{X:=1.0,Y:=1.5}, ?AutoAnchor:=true)}
            RevivingStartedAlertText.Hide()
            
            set CanvasSlots += array{InteractionText.Create("", InteractionTextPos, 25.0, ?SizeRatio:=vector2{X:=1.0,Y:=1.6}, ?SandwichedWidget:=option{ProgressColorBlock}, ?Alignment:=vector2{X:=0.5,Y:=0.5}, ?AutoAnchor:=true)}

            DownedHealthBar : color_block = color_block{DefaultColor:=NamedColors.Red, DefaultDesiredSize:=vector2{X:=290.0,Y:=22.0}}
            DownedHealthOverlay:overlay=overlay{
                Slots:=array:
                    overlay_slot{
                        Widget:=color_block{DefaultColor:=MakeColorFromHex("050505FF"), DefaultDesiredSize:=vector2{X:=300.0,Y:=32.0}}
                    }
                    overlay_slot{
                        Widget:=overlay{
                            Slots:=array:
                                overlay_slot{
                                    Widget:=color_block{DefaultColor:=MakeColorFromHex("111111FF"), DefaultDesiredSize:=vector2{X:=290.0,Y:=22.0}}
                                    HorizontalAlignment:=horizontal_alignment.Center
                                    VerticalAlignment:=vertical_alignment.Center
                                }
                                overlay_slot{
                                    Widget:=DownedHealthBar
                                    HorizontalAlignment:=horizontal_alignment.Left
                                    VerticalAlignment:=vertical_alignment.Center
                                }
                        }
                        HorizontalAlignment:=horizontal_alignment.Center
                        VerticalAlignment:=vertical_alignment.Center
                    }
                    overlay_slot{
                        Widget:=DownedHealthSkull
                        Padding:=margin{Left:=-175.0}
                        HorizontalAlignment:=horizontal_alignment.Center
                        VerticalAlignment:=vertical_alignment.Center
                    }
            }
            set CanvasSlots += array{DownedHealthUI.Create(DownedHealthBar, DownedHealthOverlay, vector2{X:=812.0,Y:=765.0}, vector2{X:=290.0,Y:=22.0}, ?AnchorOverride:=option. anchors{Minimum:=vector2{X:=0.5,Y:=0.5},Maximum:=vector2{X:=0.5,Y:=0.5}}, ?Alignment:=vector2{X:=0.5}, ?AutoOverlay:=false)}
            DownedHealthUI.Hide()

            #InteractionText.EnableDebug()
            <#
            set InteractionUIControl = canvas{
                Slots:=array{
                    InteractionTextSlot
                }
            }#>
            InteractionText.ResizeBackplate(InteractionBackplateOffset)
            #set CanvasSlots += array{
            #    canvas_slot:
            #        Widget:=InteractionUIControl
            #}
            if(not AppliedUICanvas?):
                spawn{Popout_Interaction()}
            InteractionText.Hide()

            set CanvasSlots += array{ReviveBar.Create(color_block{DefaultColor:=NamedColors.White}, texture_block{DefaultImage:=Art.UIStuff.Fonts.InteractionBackground.InteractionBackground}, vector2{X:=960.0,Y:=780.0}, vector2{X:=480.0,Y:=15.0}, ?AutoAnchor:=true, ?Alignment:=vector2{X:=0.5,Y:=0.5})}
            ReviveBar.Hide()

            #set CanvasSlots += array{ShieldWidget.Create(ShieldIcon, vector2{X:=1372.0,Y:=932.0}, vector2{X:=100.0,Y:=100.0})}
            ShieldIconPos := vector2{X:=1700.0,Y:=840.0}
            set CanvasSlots += array{ShieldWidget.Create(ShieldIcon, ShieldIconPos, ?Alignment:=vector2{X:=1.0,Y:=1.0}, ?AnchorOverride:=option{anchors{Minimum:=vector2{X:=1.0,Y:=1.0},Maximum:=vector2{X:=1.0,Y:=1.0}}})}
            set CanvasSlots += array{ShieldHealthBar.Create(color_block{DefaultColor:=NamedColors.DodgerBlue}, texture_block{DefaultImage:=Art.UIStuff.Fonts.InteractionBackground.InteractionBackground}, ShieldIconPos+vector2{Y:=60.0}, vector2{X:=66.0,Y:=6.5}, ?Alignment:=vector2{X:=1.0,Y:=1.0}, ?AnchorOverride:=option{anchors{Minimum:=vector2{X:=1.0,Y:=1.0},Maximum:=vector2{X:=1.0,Y:=1.0}}})}
            ShieldHealthBar.SetProgress(100.0)
            ShieldWidget.Hide()
            ShieldHealthBar.Hide()

            #set CanvasSlots += array{PowerPunchWidget.Create(texture_block{DefaultImage:=Art.UIStuff.Icons.Weapon.Fist_Icon}, vector2{X:=1504.0,Y:=1000.0}, vector2{X:=40.0,Y:=40.0})}
            set CanvasSlots += array{PowerPunchWidget.Create(
                overlay{
                    Slots:=array:
                        overlay_slot{
                            Widget:=texture_block{DefaultImage:=Art.UIStuff.Icons.Part_Icons.Part_Background_Icon, DefaultDesiredSize:=vector2{X:=100.0,Y:=100.0}}
                        }
                        overlay_slot{
                            Widget:=texture_block{DefaultImage:=Art.UIStuff.Icons.Weapon.Fist_Icon, DefaultDesiredSize:=vector2{X:=40.0,Y:=40.0}}
                            HorizontalAlignment:=horizontal_alignment.Center
                            VerticalAlignment:=vertical_alignment.Center
                        }
                }, vector2{X:=1514.0,Y:=880.0}, ?Alignment:=vector2{X:=1.0,Y:=1.0}, ?AnchorOverride:=option{anchors{Minimum:=vector2{X:=1.0,Y:=1.0},Maximum:=vector2{X:=1.0,Y:=1.0}}})}
            PowerPunchWidget.Hide()

            #set CanvasSlots += array{AmmoTypeWidget.Create(AmmoAbilityTexture, vector2{X:=1564.0,Y:=1000.0}, vector2{X:=40.0,Y:=40.0})}
            set CanvasSlots += array{AmmoTypeWidget.Create(
                overlay{
                    Slots:=array:
                        overlay_slot{
                            Widget:=texture_block{DefaultImage:=Art.UIStuff.Icons.Part_Icons.Part_Background_Icon, DefaultDesiredSize:=vector2{X:=100.0,Y:=100.0}}
                        }
                        overlay_slot{
                            Widget:=AmmoAbilityTexture
                            HorizontalAlignment:=horizontal_alignment.Center
                            VerticalAlignment:=vertical_alignment.Center
                        }
                }, vector2{X:=1574.0,Y:=880.0}, ?Alignment:=vector2{X:=1.0,Y:=1.0}, ?AnchorOverride:=option{anchors{Minimum:=vector2{X:=1.0,Y:=1.0},Maximum:=vector2{X:=1.0,Y:=1.0}}})}
            AmmoTypeWidget.Hide()

            #set CanvasSlots += array{PowerPunchDetails.Create()}
            
            set CanvasSlots += DescriptionUI.Create()
            set DescriptionUI.ParentUI = option{Self}
            #set DescriptionUIControl = canvas{
            #    Slots:=ThisDescriptionSlot
            #}
            #set CanvasSlots += array{
            #    canvas_slot:
            #        Widget:=DescriptionUIControl
            #}

            set CanvasSlots += QuestUI.Create()

            set CanvasSlots += TrialUI.Create()
            
            set CanvasSlots += array{PowerupUI.Create("Powerup", "", vector2{X:=960.0, Y:=952.0}, 100.0)}

            set CanvasSlots += array{PerkUICollection.Create()}

            #set CanvasSlots += array{GlobalPoints.Init(InAgent)}
            set CanvasSlots += array{GlobalPoints.Init(InAgent)}

            set CanvasSlots += array{dynamic_widget{}.Create(KnockedPlayerAlertText, vector2{X:=40.0,Y:=508.0}, ?AutoAnchor:=true, ?Alignment:=vector2{Y:=0.5})}

            set CanvasSlots += array{RocketEE_Text.Create("a secret weapon is available somewhere...", vector2{X:=964.0,Y:=588.0}, 20.0, ?AutoAnchor:=true, ?Alignment:=vector2{X:=0.5,Y:=0.5}, ?SizeRatio:=vector2{X:=1.0,Y:=1.5})}
            RocketEE_Text.Hide()
            <#
            set CanvasSlots += array{
                canvas_slot:
                    Widget:=KnockedPlayerAlertText
                    Offsets:=margin{Top:=644.0,Left:=40.0}
                    SizeToContent:=true
            }#>

            set CanvasSlots += array. canvas_slot{
                Widget:=PausedScreenCover
                Anchors:=anchors{Minimum:=vector2{X:=0.5,Y:=0.5},Maximum:=vector2{X:=0.5,Y:=0.5}}
                Alignment:=vector2{X:=0.5,Y:=0.5}
            }
            PausedScreenCover.Hide()

            set CanvasSlots += PartsUI.Create()

            set Scoreboard.Owner = option{InAgent}
            set CanvasSlots += array{Scoreboard.Create()}

            if(GlobalAdminState?, AllowGameplayPanel?):
                AdminPanel_Game.Init(InAgent)
                set CanvasSlots += array{AdminPanel_Game.GetCanvasSlot()}
                BlackScreenBlock.Collapse()

            set CanvasSlots += array{EndGame_WavesSurvived_Text.Create("You survived x waves", vector2{X:=960.0,Y:=285.0}, 45.0, ?Alignment:=vector2{X:=0.5,Y:=0.5}, ?AutoAnchor:=true, ?SizeRatio:=vector2{X:=1.0,Y:=1.5})}
            EndGame_WavesSurvived_Text.Hide()
            set CanvasSlots += array{EndGame_DifficultyMode_Text.Create("on difficulty", vector2{X:=960.0, Y:=360.0}, 30.0, ?Alignment:=vector2{X:=0.5,Y:=1.0}, ?AutoAnchor:=true, ?SizeRatio:=vector2{X:=1.0,Y:=1.5})}
            EndGame_DifficultyMode_Text.Hide()

            set CanvasSlots += array. EndGame_Timestamp_Text.Create("exfilled at time ?", vector2{X:=960.0, Y:=392.0}, 20.0, ?Alignment:=vector2{X:=0.5,Y:=1.0}, ?AutoAnchor:=true, ?SizeRatio:=vector2{X:=1.0,Y:=1.5})
            EndGame_Timestamp_Text.Hide()

            set CanvasSlots += array{
                canvas_slot{
                    Widget:=BlackScreenBlock
                    Anchors:=anchors{
                        Minimum:=vector2{X:=0.5,Y:=0.5}
                        Maximum:=vector2{X:=0.5,Y:=0.5}
                    }
                    Alignment:=vector2{X:=0.5,Y:=0.5}
                    SizeToContent:=true
                    ZOrder:=10
                }
            } 
        else if(IsGameState[game_state.GameOver]):
            set CanvasSlots += array. canvas_slot{
                Widget:=
                    color_block{
                        DefaultColor:=NamedColors.Black
                        DefaultDesiredSize:=vector2{X:=3000.0,Y:=3000.0}
                    }
                Anchors:=anchors{
                    Minimum:=vector2{X:=0.5,Y:=0.5}
                    Maximum:=vector2{X:=0.5,Y:=0.5}
                }
                Alignment:=vector2{X:=0.5,Y:=0.5}
                SizeToContent:=true
            }
            #set CanvasSlots += GameOverPrompt.Create()
        else:
            #Print("Creating Main Menu UI")
            set CanvasSlots += array{MainMenuUI.Create(InAgent)}

            if(GlobalAdminState?):
                AdminPanel_Menu.Init(InAgent)
                AdminPanel_Menu.Hide()
                set CanvasSlots += array{AdminPanel_Menu.GetCanvasSlot()}
            <#
            ReadyUpButton : button_quiet = button_quiet{DefaultText:=StringToMessage("Test")}
            ReadyUpButton.OnClick().Subscribe(HandleReadyupClick)
            set CanvasSlots += array{
                canvas_slot{
                    Widget:=ReadyUpButton
                    #Anchors:=anchors{Minimum:=vector2{Y:=1.0}, Maximum:=vector2{Y:=1.0}}
                    Offsets:=margin{Left:=73.164673, Top:=890.0, Right:=435.0, Bottom:=135.546936}
                    SizeToContent:=false
                    ZOrder:=2
                }
            } #>
        set CanvasSlots += array. canvas_slot{
            Widget:=CodeWatermark
            Offsets:=margin{Top:=-75.0}
            SizeToContent:=true
            Anchors:=anchors{Minimum:=vector2{X:=0.5}, Maximum:=vector2{X:=0.5}}
            Alignment:=vector2{X:=0.5}
        }

        # Show gameplay ui or menu ui depending on the gamestate #
        ShowCurrentUI()

        CoreUICanvas : canvas = canvas:
            Slots := CanvasSlots

        <#
        UIOverlay : overlay = overlay:
            Slots := array:
                overlay_slot:
                    Widget:=CoreUICanvas
                overlay_slot:
                    Widget:=BlackScreenBlock

        UICanvasSlot:canvas_slot=canvas_slot{
            Widget:=UIOverlay
        }
            #>
        LeaderboardSubmitControl.Hide()
        return (CoreUICanvas, InteractiveUIState)

    var GlobalUISetupState<private> : logic = false
    SetupGlobalUI()<suspends>:void=
        if(not GlobalUISetupState?):
            set GlobalUISetupState = true
            loop:
                if(AppliedUICanvas?):
                    break
                Sleep(0.1)
            Sleep(0.0)
            PlayerList:=GetPlayers(?OnlyActives:=false)
            for(ThisPlayer:PlayerList):
                ThisUI:=ThisPlayer.User.GetUIManager()
                if(ThisPlayer<>User):
                    spawn{GlobalPoints.CreateGlobalPoints(ThisPlayer.User)}
                    spawn{Scoreboard.CreatePlayerSlot(ThisPlayer.User)}
                    spawn{ThisUI.GlobalPoints.CreateGlobalPoints(User)}
                    spawn{ThisUI.Scoreboard.CreatePlayerSlot(User)}
                    ThisUI.MainMenuUI.AddPlayerSlot(ThisPlayer.User)
                else:
                    spawn{Scoreboard.CreatePlayerSlot(User)}
                    ThisUI.MainMenuUI.AddPlayerSlot(User)

    ShowCurrentUI():void=
        MainCanvas.Show()
        set UIVisibilityState = true
        case(GetGameState()):
            game_state.Menu => ShowUI_Menu()
            game_state.InGame => ShowUI_Gameplay()
            game_state.Paused => 
            game_state.GameOver => ShowUI_EndScreen()

    ShowUI_Gameplay():void=
        EndGame_WavesSurvived_Text.Hide()
        PowerupUI.Show()
        PerkUICollection.Show()
        InfoUI.Show()
        QuestUI.Show()
        WaveText.Show()
        LocationText.Show()
        LocationIcon.Show()
        PointsText.Show()
        GlobalPoints.Show()

    ShowUI_EndScreen<public>():void=
        Scoreboard.Show()
        WaveText.Hide()
        InteractionText.Hide()
        PointsText.Hide()
        LocationText.Hide()
        LocationIcon.Hide()
        GlobalPoints.Hide()
        PowerupUI.Hide()
        InfoUI.Hide()
        PartsUI.Hide()
        QuestUI.Hide()
        TrialUI.Hide()
        PerkUICollection.Hide()
        AmmoTypeWidget.Hide()
        PowerPunchWidget.Hide()
        ShieldHealthBar.Hide()
        ShieldWidget.Hide()
        InteractionText.Hide()
        ThisWaveCount:=GetWaveCount()
        FormattedTime:=FormatSecondsToTime(GetTimeSinceStart(), ?AutoHideHour:=false)
        if(GlobalExfilState[GetSession()]?):
            EndGame_WavesSurvived_Text.SetText("exfil completed on wave <C_R>{ThisWaveCount}<E>")
            if(GetTimeSinceStart() < 3600.0):
                EndGame_Timestamp_Text.SetText("exfilled at time  <CA_gold wave>{FormattedTime}<E>")
            else:
                EndGame_Timestamp_Text.SetText("exfilled at time  <C_Y>{FormattedTime}<E>")
            EndGame_Timestamp_Text.Show()
        else:
            if(ThisWaveCount = 1):
                EndGame_WavesSurvived_Text.SetText("you survived <C_R>{ThisWaveCount}<E> wave")
            else:
                EndGame_WavesSurvived_Text.SetText("you survived <C_R>{ThisWaveCount}<E> waves")
            if(IsExfilActive[]):
                EndGame_Timestamp_Text.SetText("failed exfil at time  <C_Y>{FormattedTime}<E>")
            else:
                EndGame_Timestamp_Text.SetText("time survived <C_Y>{FormattedTime}<E>")    
            EndGame_Timestamp_Text.Show()
        EndGame_WavesSurvived_Text.Show()
        case(GetGameDifficulty()):
            game_difficulty.Easy =>
                EndGame_DifficultyMode_Text.SetText("played on <C_G>Easy<E> mode")
            game_difficulty.Classic =>
                EndGame_DifficultyMode_Text.SetText("played on <C_O>Classic<E> mode")
            game_difficulty.Hardcore =>
                EndGame_DifficultyMode_Text.SetText("played on <C_R>Hardcore<E> mode")
        EndGame_DifficultyMode_Text.Show()
        spawn. ShowLeaderboardPrompt()

    SetGamePanelState<public>(InState:logic):void=
        set AllowGameplayPanel = InState
        ApplyUI(User)

    ShowLeaderboardPrompt()<suspends>:void=
        #LeaderboardSubmitControl.ChangePosition(vector2{X:=960.0, Y:=1200.0})
        LeaderboardSubmitControl.ChangeSize(vector2{})
        Sleep(7.5)
        LeaderboardSubmitControl.Show()
        LeaderboardSubmitControl.MoveSize(vector2{X:=650.0,Y:=225.0}, 1.0, ?InterpolationType:=interpolation_type.EaseOutBack)
        #LeaderboardSubmitControl.Show()
        #LeaderboardSubmitControl.MovePosition(vector2{X:=960.0,Y:=900.0}, 1.0, ?InterpolationType:=interpolation_type.EaseOutBack)

    var UIVisibilityState : logic = true

    ShowUI_Menu():void=
        PowerupUI.Hide()
        PerkUICollection.Hide()
        InfoUI.Hide()
        QuestUI.Hide()
        WaveText.Hide()
        LocationText.Hide()
        PointsText.Hide()
        GlobalPoints.Hide()

    HideAllUI():void=
        set UIVisibilityState = false
        MainCanvas.Hide()
        <#
        WaveText.Hide()
        LocationText.Hide()
        PointsText.Hide()
        InteractionText.Hide()
        GlobalPoints.Hide()
        Scoreboard.Hide()
        InteractionText.Hide()
        EndGame_WavesSurvived_Text.Hide()#>

    IntroWaveNumberAnimation()<suspends>:void=
        return
        <#
        WaveText.SetPosition(vector2{X:=960.0, Y:=540.0})
        WaveText.SetTextSize(0.0)
        WaveStartText.SetColor(NamedColors.Red)
        WaveStartText.SetPosition(vector2{X:=960.0, Y:=450.0})
        WaveStartText.SetTextSize(0.0)
        Sleep(2.0)
        WaveStartText.Show()
        sync:
            WaveText.MoveTextSize(125.0, 1.5)
            WaveStartText.MoveTextSize(50.0, 1.5)
        sync:
            WaveText.MoveTransform(vector2{X:=92.0,Y:=996.0}, 90.0, 4.0)
            WaveStartText.MoveTransform(vector2{X:=92.0,Y:=916.0}, 0.0, 4.0)
        WaveStartText.Hide() #>
        #spawn{WaveText.MovePosition(vector2{X:=48.0, Y:=944.0}, 4.0)}
        #WaveText.MoveTextSize(90.0, 3.5)
        
    CycleBobInteraction()<suspends>:void=
        InteractionText.EnableDebug()
        loop:
            InteractionText.MoveTransform(vector2{X:=960.0, Y:= 500.0}, 60.0, 3.0)
            InteractionText.MoveTransform(vector2{X:=960.0, Y:= 840.0}, 30.0, 3.0)

# Holds the ui for the info hud #
# Shows info such as parts collected, or rituals completed...etc #
info_ui := class<concrete>(dynamic_widget):

    
scoreboard_ui<public> := class<concrete>:

    var<private> ScoreboardStackBox : stack_box = stack_box{
        Orientation:=orientation.Vertical
    }

    var<private> ScoreboardOverlay_StatTypes : overlay = overlay{}

    var IsCreated<private> : logic = false
    var Owner : ?agent = false
    Create():canvas_slot=
        # Slot background #
        ScoreboardOverlay_StatTypes.AddWidget(overlay_slot{
            Widget:=texture_block{DefaultImage:=Art.UIStuff.Fonts.InteractionBackground.InteractionBackground, DefaultDesiredSize:=vector2{X:=1055.0,Y:=50.0}, DefaultTint:=NamedColors.Black}
            HorizontalAlignment:=horizontal_alignment.Fill
            VerticalAlignment:=vertical_alignment.Fill
        })
        # Username stat text #
        ScoreboardOverlay_StatTypes.AddWidget(overlay_slot{
            Widget:=text_block{DefaultText:=StringToMessage("Username"), DefaultTextColor:=NamedColors.White}
            HorizontalAlignment:=horizontal_alignment.Left
            VerticalAlignment:=vertical_alignment.Center
            Padding:=margin{Left:=91.5}
        })
        # Points stat text #
        ScoreboardOverlay_StatTypes.AddWidget(overlay_slot{
            Widget:=text_block{DefaultText:=StringToMessage("Score"), DefaultTextColor:=NamedColors.White}
            HorizontalAlignment:=horizontal_alignment.Left
            VerticalAlignment:=vertical_alignment.Center
            Padding:=margin{Left:=387.5}
        })
        # Kills stat text #
        ScoreboardOverlay_StatTypes.AddWidget(overlay_slot{
            Widget:=text_block{DefaultText:=StringToMessage("Kills"), DefaultTextColor:=NamedColors.White}
            HorizontalAlignment:=horizontal_alignment.Left
            VerticalAlignment:=vertical_alignment.Center
            Padding:=margin{Left:=580.0}
        })
        # Revives stat text #
        ScoreboardOverlay_StatTypes.AddWidget(overlay_slot{
            Widget:=text_block{DefaultText:=StringToMessage("Revives"), DefaultTextColor:=NamedColors.White}
            HorizontalAlignment:=horizontal_alignment.Left
            VerticalAlignment:=vertical_alignment.Center
            Padding:=margin{Left:=731.5}
        })
        # Downs stat text #
        ScoreboardOverlay_StatTypes.AddWidget(overlay_slot{
            Widget:=text_block{DefaultText:=StringToMessage("Downs"), DefaultTextColor:=NamedColors.White}
            HorizontalAlignment:=horizontal_alignment.Left
            VerticalAlignment:=vertical_alignment.Center
            Padding:=margin{Left:=915.5}
        })
        ScoreboardStackBox.AddWidget(stack_box_slot{
            Widget:=ScoreboardOverlay_StatTypes
            HorizontalAlignment:=horizontal_alignment.Fill
            VerticalAlignment:=vertical_alignment.Fill
        })
        defer{
            set IsCreated = true
            Hide()
        }
        return canvas_slot{
            Widget:=ScoreboardStackBox
            Anchors:=anchors{Minimum:=vector2{X:=0.5,Y:=0.5}, Maximum:=vector2{X:=0.5,Y:=0.5}}
            Alignment:=vector2{X:=0.5,Y:=0.5}
            SizeToContent:=true
        }

    var<private> PlayerToStatsMap : [agent]?scoreboard_widgets = map{}
    var<private> CreatedStatesPerPlayer : [agent]logic = map{}
    GetPlayerStatWidgets():[]scoreboard_widgets=
        var PlayerStatWidgets : []scoreboard_widgets = array{}
        for:
            ThisPlayer -> ScoreboardWidgets : PlayerToStatsMap
            ThisScoreboardWidgets := ScoreboardWidgets?
        do:
            set PlayerStatWidgets += array{ThisScoreboardWidgets}
        return PlayerStatWidgets

    CreatePlayerSlot(InPlayer:agent)<suspends>:void=
        if(not CreatedStatesPerPlayer[InPlayer]?):
            if. set CreatedStatesPerPlayer[InPlayer] = true
            else. set CreatedStatesPerPlayer = ConcatenateMaps(CreatedStatesPerPlayer, map{InPlayer => true})
            # Await the creation of this scoreboard for the owning player #
            loop:
                if(IsCreated?):
                    break
                Sleep(0.0)
            NewSlotBackground : texture_block = texture_block{DefaultImage:=Art.UIStuff.Fonts.InteractionBackground.InteractionBackground, DefaultDesiredSize:=vector2{X:=1055.0,Y:=50.0}}
            NewUsernameWidget : text_block = text_block{DefaultText:=PlayerToMessage(InPlayer)}
            if(ThisColor:=InPlayer.GetPlayerPackage[].Color):
                NewUsernameWidget.SetTextColor(ThisColor)
            else{Print("~ERR: Failed to fetch a players' color!")}
            NewPointsWidget : text_block = text_block{DefaultTextColor:=NamedColors.White}
            NewKillsWidget : text_block = text_block{DefaultTextColor:=NamedColors.White}
            NewRevivesWidget : text_block = text_block{DefaultTextColor:=NamedColors.White}
            NewDownsWidget : text_block = text_block{DefaultTextColor:=NamedColors.White}
            NewPlayerStatOverlay : overlay = overlay{
                Slots:=array:
                    # Background slot #
                    overlay_slot{
                        Widget:=NewSlotBackground
                        HorizontalAlignment:=horizontal_alignment.Fill
                        VerticalAlignment:=vertical_alignment.Fill
                    }
                    # Username slot #
                    overlay_slot{
                        Widget:=NewUsernameWidget
                        HorizontalAlignment:=horizontal_alignment.Left
                        VerticalAlignment:=vertical_alignment.Center
                        Padding:=margin{Left:=7.5}
                    }
                    # Points slot #
                    overlay_slot{
                        Widget:=NewPointsWidget
                        HorizontalAlignment:=horizontal_alignment.Left
                        VerticalAlignment:=vertical_alignment.Center
                        Padding:=margin{Left:=367.5}
                    }
                    # Kills slot #
                    overlay_slot{
                        Widget:=NewKillsWidget
                        HorizontalAlignment:=horizontal_alignment.Left
                        VerticalAlignment:=vertical_alignment.Center
                        Padding:=margin{Left:=541.5}
                    }
                    # Revives slot #
                    overlay_slot{
                        Widget:=NewRevivesWidget
                        HorizontalAlignment:=horizontal_alignment.Left
                        VerticalAlignment:=vertical_alignment.Center
                        Padding:=margin{Left:=719.5}
                    }
                    # Downs slot #
                    overlay_slot{
                        Widget:=NewDownsWidget
                        HorizontalAlignment:=horizontal_alignment.Left
                        VerticalAlignment:=vertical_alignment.Center
                        Padding:=margin{Left:=895.5}
                    }
            }
            NewScoreboardWidgets : scoreboard_widgets = scoreboard_widgets{
                OverlayWidget:=NewPlayerStatOverlay
                BackgroundSlotWidget:=option{NewSlotBackground}
                UsernameWidget:=NewUsernameWidget
                PointsWidget:=NewPointsWidget
                KillsWidget:=NewKillsWidget
                RevivesWidget:=NewRevivesWidget
                DownsWidget:=NewDownsWidget
            }
            if(not PlayerToStatsMap[InPlayer]):
                set PlayerToStatsMap = ConcatenateMaps(PlayerToStatsMap, map{InPlayer => option{NewScoreboardWidgets}})
            else if(set PlayerToStatsMap[InPlayer] = option{NewScoreboardWidgets}){}
            else{Print("~ERR: Failed to save a players' stat widgets!")}
            set NewScoreboardWidgets.Index = GetPlayerStatWidgets().Length-1
            if(PlayerToStatsMap[InPlayer]?):
                # Add the new scoreboard slot to the main scoreboard stack box #
                ScoreboardStackBox.AddWidget(stack_box_slot{
                    Widget:=NewPlayerStatOverlay
                    HorizontalAlignment:=horizontal_alignment.Fill
                    VerticalAlignment:=vertical_alignment.Fill
                })
            if(Mod[NewScoreboardWidgets.Index, 2] = 0):
                #Print("~Set tint to white")
                NewSlotBackground.SetTint(NamedColors.White)
            else:
                #Print("~Failed mod with index [{NewScoreboardWidgets.Index}]")
                NewSlotBackground.SetTint(NamedColors.Black)
            UpdateStats(InPlayer)

    UpdateStats<public>(InPlayer:agent):void=
        if(PlayerStats:=PlayerToStatsMap[InPlayer]?):
            PlayerStats.Update(InPlayer)
        else if(InPlayer<>Owner?):
            spawn. CreatePlayerSlot(InPlayer)

    RemoveStats<public>(OutPlayer:agent):void=
        if(PlayerStats:=PlayerToStatsMap[OutPlayer]?):
            if. set CreatedStatesPerPlayer[OutPlayer] = false
            ScoreboardStackBox.RemoveWidget(PlayerStats.OverlayWidget)
            if(set PlayerToStatsMap[OutPlayer] = false){}
            PlayerStatsWidgets := GetPlayerStatWidgets()
            var StatsIndex : int = 0
            for(ThisStatsWidgets : PlayerStatsWidgets):
                set ThisStatsWidgets.Index = StatsIndex
                if(ThisBackgroundWidget:=ThisStatsWidgets.BackgroundSlotWidget?):
                    if(Mod[StatsIndex, 2] = 0):
                        ThisBackgroundWidget.SetTint(NamedColors.White)
                    else:
                        ThisBackgroundWidget.SetTint(NamedColors.Black)

    Hide<public>():void=
        ScoreboardStackBox.Hide()
    
    Show<public>():void=
        ScoreboardStackBox.Show()

scoreboard_widgets := class:

    var OverlayWidget : overlay = overlay{}
    var BackgroundSlotWidget : ?texture_block = false
    var Index : int = -1

    var UsernameWidget : text_block = text_block{}
    var PointsWidget : text_block = text_block{}
    var KillsWidget : text_block = text_block{}
    var RevivesWidget : text_block = text_block{}
    var DownsWidget : text_block = text_block{}

    Update(InPlayer:agent):void=
        if(ThisPlayerInfo:=InPlayer.GetPlayerInfo[]):
            PointsWidget.SetText(StringToMessage(FormatStringToCenter(8, "{ThisPlayerInfo.Points}")))
            KillsWidget.SetText(StringToMessage(FormatStringToCenter(8, "{ThisPlayerInfo.TotalKills}")))
            RevivesWidget.SetText(StringToMessage(FormatStringToCenter(8, "{ThisPlayerInfo.TotalRevives}")))
            DownsWidget.SetText(StringToMessage(FormatStringToCenter(8, "{ThisPlayerInfo.TotalDowns}")))

perk_depletion_ui := class:


# Holds the ui for the powerups #
# Handles displaying powerup ui to the screen and hiding it #
powerup_ui<public> := class<concrete>:

    ParentStringBlock : string_block = string_block{Font:="Powerup"}

    Create(FontTypeString:string, DefaultText:string, InPosition:vector2, InSize:float):canvas_slot=
        return ParentStringBlock.Create(DefaultText, InPosition, InSize, ?SpacerSize:=2.15, ?Alignment:=vector2{X:=0.5, Y:=0.5}, ?AutoAnchor:=true)

    AddPowerup<public><final>(PowerupType:powerup_base):void=
        if(PowerupType.IsPersistent[]):
            PowerupText:=ParentStringBlock.CurrentText
            ParentStringBlock.SetText(PowerupText+PowerupType.BasicSettings.Name)
            ParentStringBlock.Show()
            spawn{AwaitPowerupDeactivate(PowerupType)}
            spawn{BlinkEffect(PowerupType)}

    AwaitPowerupDeactivate<private><final>(PowerupType:powerup_base)<suspends>:void=
        PowerupType.DeactivatedEvent.Await()
        RemovePowerup(PowerupType)

    RemovePowerup<public><final>(PowerupType:powerup_base):void=
        Print("Powerup deactivated")
        CurrentPowerupsText := ParentStringBlock.CurrentText
        LowercasePowerupsText := PowerupType.BasicSettings.Name.ToLower()
        if(NamePosition:=CurrentPowerupsText.FindSubstring[LowercasePowerupsText]):
            if(PowerupText1 := CurrentPowerupsText.Slice[0, NamePosition]):
                if(PowerupText2 := CurrentPowerupsText.Slice[NamePosition+PowerupType.BasicSettings.Name.Length, CurrentPowerupsText.Length]):
                    ParentStringBlock.SetText(PowerupText1+PowerupText2)
                else:
                    ParentStringBlock.SetText(PowerupText1)
            else:
                Print("~ERROR: Failed to get previous text before powerup name")
        else:
            Print("~ERROR: Failed to update powerups ui (specifically removing a powerup from ui)!")
            Print("~The current powerup ui text is: {CurrentPowerupsText}")

    BlinkEffect<private><final>(InPowerup:powerup_base)<suspends>:void=
        MaybePowerupWidget:=ParentStringBlock.GetWidgetAtSubstring(InPowerup.GetName())
        if(ThisPowerupWidget:=MaybePowerupWidget?):
            loop:
                ThisPowerupWidget.Show()
                Winner:=race:
                    block:
                        InPowerup.DeactivatedEvent.Await()
                        1
                    block:
                        InPowerup.ActivatedEvent.Await()
                        2
                    block:
                        SafeSleep(InPowerup.GetActiveTime()*0.6)
                        var SleepTime : float = 1.0
                        var Loops : int = 0
                        loop:
                            ThisPowerupWidget.Collapse()
                            SafeSleep(0.15)
                            ThisPowerupWidget.Show()
                            SafeSleep(SleepTime)
                            if(Loops >= 3):
                                set SleepTime *= 0.50
                                set Loops = 0
                            else:
                                set Loops += 1
                        -1
                if(Winner=1):
                    break
                else if(Winner=2):
                    ThisPowerupWidget.Show()
        else{Print("~ERROR: Failed to fetch the powerup widget from the string_block!")}

    Show<public>():void=
        ParentStringBlock.Show()

    Hide<public>():void=
        ParentStringBlock.Hide()

globalpoints<public> := class:

    var ParentUI : ?ui_manager = false

    StackBox : stack_box = stack_box{Orientation:=orientation.Vertical}

    # Holds global players' widgets for the representation of their points #
    var PlayerToPointsMap : [agent]?points_block = map{}

    var Owner : ?agent = false

    var IsCreated : logic = false
    Init(InOwner:agent):canvas_slot=
        set Owner = option{InOwner}
        set ParentUI = option{InOwner.GetUIManager()}
        defer{
            set IsCreated = true
        }
        return canvas_slot{
            Widget := StackBox
            SizeToContent:=true
            Anchors:=anchors{Minimum:=vector2{Y:=1.0}, Maximum:=vector2{Y:=1.0}}
            Alignment:=vector2{Y:=1.0}
            Offsets:=margin{Left:=80.0,Top:=-285.0}
        }
        
    CreateGlobalPoints(InPlayer:agent)<suspends>:void=
        if(InPlayer<>Owner?):
            loop:
                if(IsCreated?, ParentUI?.AppliedUICanvas?):
                    break
                Sleep(0.1)
            if(not PlayerToPointsMap[InPlayer]?, ThisCanvasUI:=ParentUI?.AppliedUICanvas?):
                # Create new points widget #
                NewPoints : points_block = points_block{IsGlobal:=true}
                set PlayerToPointsMap = ConcatenateMaps(PlayerToPointsMap, map{InPlayer => option{NewPoints}})
                PointsWidgetSlot := NewPoints.Create(InPlayer, "0", vector2{}, 20.0, ?IncludeBackplate:=false).Widget
                StackBox.AddWidget(stack_box_slot{Widget:=PointsWidgetSlot, Padding:=margin{Top:=5.0,Bottom:=5.0}})
                NewPoints.SetColor(InPlayer)
            UpdateGlobalPoints(InPlayer)

    UpdateGlobalPoints<public>(InPlayer:agent):void=
        if(ThisPlayerInfo:=InPlayer.GetPlayerInfo[]):
            # Update Points ui #
            if(PointAmount:=InPlayer.GetPlayerInfo[].Points):
                if(PointsUI := PlayerToPointsMap[InPlayer]?):
                    PointsUI.ChangePoints("${PointAmount}")
                    PointsUI.SetColor(InPlayer)
                else:
                    spawn. CreateGlobalPoints(InPlayer)

    UpdatePlayerIcon(InPlayer:agent):void=
        if(PointsUI := PlayerToPointsMap[InPlayer]?):
            PointsUI.UpdatePlayerIcon(InPlayer)
    <#
    var TestsAdded : int = 0
    TestAddPoints():void=
        # Create new stack box slot #
        NewPoints : points_block = points_block{}
        PointsWidget := NewPoints.Create("Points", "0", vector2{}, 25.0, ?Spacing:=-10.0).Widget
        #NewPoints.ParentWidget.SetPositionBox(vector2{X:=262.0,Y:=65.0})
        NewPoints.ChangePosition(vector2{X:=131.0,Y:=100.0})
        StackBox.AddWidget(stack_box_slot{
            Widget := PointsWidget
            HorizontalAlignment := horizontal_alignment.Fill
            VerticalAlignment := vertical_alignment.Fill
            Padding := margin{Top:=-175.0}
        })
        set TestsAdded += 1
        StackBoxOffset.SetDesiredSize(vector2{X:=262.0, Y:=940.0-(TestsAdded*50.0)}) #>

    RemovePlayer<public>(InAgent:agent):void=
        if(PointsUI:=PlayerToPointsMap[InAgent]?):
            StackBox.RemoveWidget(PointsUI.GetWidget())
            #Print("~Removed a global players' points!")
        if(set PlayerToPointsMap[InAgent] = false){}
        else{Print("~ERROR: Failed to remove a points ui from global points!")}

    Hide():void=
        for(ThisPlayer->MaybePointsUI:PlayerToPointsMap):
            if(ThisPointsUI:=MaybePointsUI?):
                ThisPointsUI.Hide()

    Show():void=
        for(ThisPlayer->MaybePointsUI:PlayerToPointsMap):
            if(ThisPointsUI:=MaybePointsUI?):
                ThisPointsUI.Show()

points_block := class:

    WidgetControl : dynamic_widget = dynamic_widget{}

    var DynamicPoints : string_block = string_block{Font:="Points"}
    var BackplateDynamic : dynamic_widget = dynamic_widget{}
    var PlayerIconDynamic : dynamic_widget = dynamic_widget{}

    var DepletionBar : progress_bar = progress_bar{}

    var Owner : ?agent = false
    var IsGlobal : logic = false
    DebugState:logic=false

    Create(InOwner:agent, DefaultText:string, InPosition:vector2, InSize:float, ?BackplateTypePath:string="NoBackplate", ?IncludeBackplate:logic=true, ?Alignment:vector2=vector2{Y:=0.5}, ?AutoAnchor:logic=false, ?AnchorOverride:?anchors=false):canvas_slot=
        set Owner = option{InOwner}
        DynamicPoints.Reset()
        if(DebugState?):
            DynamicPoints.EnableDebug()
        StringBlockWidget:= DynamicPoints.Create(DefaultText, vector2{}, InSize, ?SizeRatio:=vector2{X:=1.0,Y:=1.6}, ?SpacerSize:=0.5).Widget

        NewBackgroundTexture : texture_block = texture_block{DefaultImage:=Art.UIStuff.Icons.PointBackground.PointsBackground}
        BackgroundWidget := BackplateDynamic.Create(NewBackgroundTexture, vector2{}).Widget
        if(not IncludeBackplate?):
            BackgroundWidget.Collapse()

        var PlayerIconColor : color = NamedColors.White
        if(ThisColor:=InOwner.GetPlayerPackage[].Color):
            set PlayerIconColor = ThisColor
        NewPlayerIconTexture : texture_block = texture_block{DefaultImage:=Art.UIStuff.Icons.PointBackground.Player_Icon, DefaultTint:=PlayerIconColor}
        PlayerIconWidget := PlayerIconDynamic.Create(NewPlayerIconTexture, vector2{}).Widget
        if(not IsGlobal?):
            PlayerIconWidget.Hide()

        ChangeSize(InSize)
            
        PointsBlockOveray := overlay{
            Slots := array:
                overlay_slot{
                    Widget:=BackgroundWidget
                    HorizontalAlignment:=horizontal_alignment.Left
                    VerticalAlignment:=vertical_alignment.Center
                    Padding:=margin{Left:=-45.0}
                }
                overlay_slot{
                    Widget:=StringBlockWidget
                    HorizontalAlignment:=horizontal_alignment.Left
                    VerticalAlignment:=vertical_alignment.Center
                    Padding:=margin{Top:=5.0}
                }
                overlay_slot{
                    Widget:=PlayerIconWidget
                    HorizontalAlignment:=horizontal_alignment.Left
                    VerticalAlignment:=vertical_alignment.Center
                    Padding:=margin{Left:=-45.0, Top:=5.0}
                }
        }  
        WidgetControl.Create(PointsBlockOveray, InPosition, ?AutoAnchor:=AutoAnchor, ?Alignment:=Alignment, ?AnchorOverride:=AnchorOverride)

    ChangePosition(DesiredPosition:vector2):void=
        WidgetControl.ChangePosition(DesiredPosition)
        <#
        DynamicPoints.ChangePosition(DesiredPosition)
        NewBackplatePos : vector2 = vector2{
            X:= DesiredPosition.X-((DynamicPoints.CurrSize/45.0) * 54.0)
            Y:= DesiredPosition.Y-((DynamicPoints.CurrSize/45.0) * 168.0)
        }
        BackplateDynamic.ChangePosition(NewBackplatePos)
        NewPlayerIconPos : vector2 = vector2{
            X:= DesiredPosition.X-((DynamicPoints.CurrSize/45.0) * 75.0)
            Y:= DesiredPosition.Y+((DynamicPoints.CurrSize/45.0) * 10.5)
        }
        PlayerIconDynamic.ChangePosition(NewPlayerIconPos)#>

    MovePoints(Position:vector2, Time:float, ?IsDelta:logic=false)<suspends>:void=
        #WidgetControl.MovePosition(Position, Time, ?IsDelta:=IsDelta)
        DynamicPoints.MovePosition(Position, Time, ?IsDelta:=IsDelta)

    Resize(InSize:float, Time:float)<suspends>:void=
        DynamicPoints.MoveTextSize(InSize, Time)

    SetColor(InColor:color):void=
        DynamicPoints.SetColor(InColor)

    ResetColor():void=
        DynamicPoints.SetColor(NamedColors.White)

    SetColor(InPlayer:agent):void=
        if(ThisColor:=InPlayer.GetPlayerPackage[].Color):
            SetColor(ThisColor)
        else{Print("~ERR: Failed to get color in points_block : SetColor()")}

    ChangeSize(DesiredSize:float):void=
        DynamicPoints.SetTextSize(DesiredSize)
        NewBackplateSize : vector2 = vector2{
            X:= (DesiredSize/25.0) * 300.0
            Y:= (DesiredSize/25.0) * 375.0
        }
        BackplateDynamic.ChangeSize(NewBackplateSize)
        #BackplateDynamic.ChangePosition(vector2{X:=-(DesiredSize+10.0)})
        NewPlayerIconSize : vector2 = vector2{
            X := (DesiredSize/25.0) * 32.0
            Y := (DesiredSize/25.0) * 32.0
        }
        PlayerIconDynamic.ChangeSize(NewPlayerIconSize)
        if(IsGlobal?):
            PlayerIconDynamic.ChangePosition(vector2{X:=-(DesiredSize)-15.0}, ?IsDelta:=true)
        else:
            PlayerIconDynamic.ChangePosition(vector2{X:=-(DesiredSize)-30.0}, ?IsDelta:=true)
        #ChangePosition(DynamicPoints.GetPosition())

    ChangePoints(InText:string):void=
        DynamicPoints.SetText(InText)

    UpdatePlayerIcon(InAgent:agent):void=
        if(ThisPlayerInfo:=InAgent.GetPlayerInfo[]):
            if(ThisPlayerInfo.IsSpectating[]):
                PlayerIconDynamic.SetImage(Art.UIStuff.Icons.PointBackground.Skull_Icon)
                PlayerIconDynamic.ChangeSize(vector2{X:=32.0,Y:=32.0})
                PlayerIconDynamic.Show()
            else if(ThisPlayerInfo.IsDowned[]):
                PlayerIconDynamic.SetImage(Art.UIStuff.Fonts.InteractionText.Symbols.Symbol_Plus)
                PlayerIconDynamic.ChangeSize(vector2{X:=32.0,Y:=32.0}*2.0)
                PlayerIconDynamic.Show()
            else:
                if(IsGlobal?):
                    PlayerIconDynamic.SetImage(Art.UIStuff.Icons.PointBackground.Player_Icon)
                    PlayerIconDynamic.ChangeSize(vector2{X:=22.0,Y:=22.0})
                    PlayerIconDynamic.Show()
                else:
                    PlayerIconDynamic.Hide()
        else{Print("~ERR: Failed to fetch playerinfo!")}

    Show():void=
        WidgetControl.Show()
        DynamicPoints.Show()

    Hide():void=
        WidgetControl.Hide()

    GetWidget():widget=
        return WidgetControl.ParentWidget

########  ########     ###     ######    #######  ##    ## ########     ###     ######   ######## 
##     ## ##     ##   ## ##   ##    ##  ##     ## ###   ## ##     ##   ## ##   ##    ##  ##       
##     ## ##     ##  ##   ##  ##        ##     ## ####  ## ##     ##  ##   ##  ##        ##       
##     ## ########  ##     ## ##   #### ##     ## ## ## ## ########  ##     ## ##   #### ######   
##     ## ##   ##   ######### ##    ##  ##     ## ##  #### ##   ##   ######### ##    ##  ##       
##     ## ##    ##  ##     ## ##    ##  ##     ## ##   ### ##    ##  ##     ## ##    ##  ##       
########  ##     ## ##     ##  ######    #######  ##    ## ##     ## ##     ##  ######   ########

<###########################

STRING_BLOCK CLASS - UI

Supports numbers & letters

###########################>

<#
string_block := class:

    WidgetControl<private> : dynamic_widget = dynamic_widget{}
    GetWidget():widget=
        return WidgetControl.ParentWidget
    GetPosition():vector2=
        return WidgetControl.CurrentPosition

    var StringStackbox<private> : stack_box = stack_box{Orientation:=orientation.Horizontal}
    var StringBoxSlots<private> : []string_block_slot = array{}

    var<private> BackplateTexture:?texture=false
    BackplateWidget<private> : texture_block = texture_block{DefaultImage:=Art.UIStuff.Fonts.InteractionText.Numbers.Number_0}
    var<private> BackplateOffsets : vector2 = vector2{}

    var<private> Font:string
    ChangeFont<public><final>(InFontName:string):void=
        if(GetFontFolder[InFontName]):
            set Font = InFontName
            SetText(CurrentText)
        else{Print("~string_block ERR: Failed to find Font Name [{InFontName}]!")}

    var<private> CurrentText : string = ""

    var<private> CurrSize : float = 30.0
    var SpacingMultiplier<private> : float = 0.5
    var CurrSpacerSize<private> : float = 1.0
    var<private> CurrColor : color = color{}

    var HorizontalAlignment<internal> : horizontal_alignment = horizontal_alignment.Center

    Create<public>(Text:string, Position:vector2, Size:float, ?SpacingSize:float:=0.25, ?SpacerSize:float:=1.0, ?SandwichedWidget:?widget=false, ?Alignment:vector2=vector2{}, ?AutoAnchor:logic=false):canvas_slot=
        set CurrSize = Size
        set SpacingMultiplier = SpacingSize
        set CurrSpacerSize = SpacerSize+1.0
        SetText(Text)
        defer{
            if(DebugState?):
                EnableDebug()
            SetSpacerSize(CurrSpacerSize)
            SetSpacingSize(SpacingMultiplier)
            if(ThisBackplateTexture:=BackplateTexture?):
                SetBackplate(ThisBackplateTexture)
            else{HideBackplate()}
        }
        ThisSandwichedWidget := SandwichedWidget? or overlay{}
        StringBlockOverlay : overlay = overlay{
            Slots:=array:
                overlay_slot{
                    Widget:=BackplateWidget
                    HorizontalAlignment:=horizontal_alignment.Fill
                    VerticalAlignment:=vertical_alignment.Fill
                }
                overlay_slot{
                    Widget:=ThisSandwichedWidget
                    HorizontalAlignment:=horizontal_alignment.Center
                    VerticalAlignment:=vertical_alignment.Center
                }
                overlay_slot{
                    Widget:=StringStackbox
                    HorizontalAlignment:=horizontal_alignment.Center
                    VerticalAlignment:=vertical_alignment.Center
                }
        }
        WidgetControl.Create(StringBlockOverlay, Position, ?Alignment:=Alignment, ?AutoAnchor:=AutoAnchor)

    Reset():void=
        SetText("")

    TextChangedEvent:event(string)=event(string){}
    SetText<public><final>(InText:string):void=
        var TextChanged : logic = false
        if(InText.ToLower()=CurrentText):
            return
        else:
            set TextChanged = true
            var InTextConv : string = InText.ToLower()
            TextWithCommands := InTextConv
            set InTextConv = GetRealText(InTextConv)
            var TextSize : int = 0
            if(InTextConv=""):
                set InTextConv = ""
                for(ThisStringSlot:StringBoxSlots):
                    StringStackbox.RemoveWidget(ThisStringSlot.StringWidget)
                    StringStackbox.RemoveWidget(ThisStringSlot.SpacerWidget)
                set StringBoxSlots = array{}
                set CurrentText = ""
                Hide()
                return
            else if(InTextConv.Length <> CurrentText.Length):
                var LengthSubtract : int = 0
                var LengthAdditive : int = 0
                InSubstringIndexes := GetWordIndexes(InTextConv)
                for(ThisSubstringIndexes : InSubstringIndexes):
                    set LengthSubtract += (ThisSubstringIndexes(1)-ThisSubstringIndexes(0))
                PrevSubstringIndexes := GetWordIndexes(CurrentText)
                for(ThisSubstringIndexes : PrevSubstringIndexes):
                    set LengthAdditive += (ThisSubstringIndexes(1)-ThisSubstringIndexes(0))
                LengthDiff := (InTextConv.Length-LengthSubtract)-(CurrentText.Length-LengthAdditive)
                if(LengthDiff > 0):
                    for(ExtraCharCount:=0..LengthDiff-1):
                        AddStringWidget('0')
                else if(LengthDiff < 0):
                    WidgetsToRemove:=(Abs(LengthDiff))
                    for(ErasingCharIndex:=0..WidgetsToRemove-1):
                        if(ThisErasingWidget:=StringBoxSlots[ErasingCharIndex]):
                            StringStackbox.RemoveWidget(ThisErasingWidget.StringWidget)
                            StringStackbox.RemoveWidget(ThisErasingWidget.SpacerWidget)

                    WidgetsRemaining := (StringBoxSlots.Length-WidgetsToRemove)
                    var TempStackBoxSlots : []string_block_slot = array{}
                    for(ThisCharSlotIndex := WidgetsToRemove..StringBoxSlots.Length):
                        if(ThisStackBoxSlot:=StringBoxSlots[ThisCharSlotIndex]):
                            set TempStackBoxSlots += array{ThisStackBoxSlot}
                    set StringBoxSlots = TempStackBoxSlots 
                set CurrentText = InTextConv
            else:
                set TextSize = CurrentText.Length-1
            var CharIndex : int = 0
            var TrackedSubstringIndex : int = 0
            SubstringIndexes := GetWordIndexes(InTextConv)
            for(ThisStackBox:StringBoxSlots):
                TextureWidget:=ThisStackBox.StringWidget
                if(MatchingIndex:=SubstringIndexes.FindMatchingStartIndex[CharIndex]):
                    if(ThisWord:=InTextConv.Slice[CharIndex, MatchingIndex(1)+1]):
                        set ThisStackBox.String = ThisWord
                        if(ThisWordTexture:=Font.GetStringTexture[ThisWord, ?ErrorOrigin:="SetText - Line 131"]):
                            TextureWidget.Show()
                            TextureWidget.SetDesiredSize(vector2{X:=CurrSize,Y:=CurrSize})
                            TextureWidget.SetImage(ThisWordTexture)
                            set ThisStackBox.IsWord = true
                        else{Print("~ERROR: Couldn't find the WordTexture with word [{ThisWord}]")}
                        set CharIndex += (MatchingIndex(1)-MatchingIndex(0))
                        set TrackedSubstringIndex += 1
                    else{Print("~ERROR: Couldn't find the word within the string!")}
                else if(ThisChar:=InTextConv[CharIndex]):
                    set ThisStackBox.String = "{ThisChar}"
                    set ThisStackBox.IsWord = false
                    if(ThisStackBox.String <> " ", NewCharTexture:=Font.GetStringTexture[ThisStackBox.String, ?ErrorOrigin:="SetText - Line 143"]):
                        TextureWidget.Show()
                        TextureWidget.SetImage(NewCharTexture)
                    else:
                        TextureWidget.HideNoCollapse()
                else:
                    Print("~ERROR: Failed to find a character widget for a char or word!")
                    Print("~The index was {CharIndex} and the string is [{InTextConv}]")
                set CharIndex += 1
            #if(ThisLastStringBox:=StringBoxSlots[StringBoxSlots.Length-1]):
            #    ThisLastStringBox.SpacerWidget.SetDesiredSize(vector2{})
            set CurrentText = InTextConv
            if(TextChanged?):
                TextChangedEvent.Signal(CurrentText)
                SetTextSize(CurrSize)
                ExecuteCommands(TextWithCommands)

    # Adds a widget for the given character which will be applied to the stack box #
    # Internal use only #
    AddStringWidget<private>(InChar:char, ?Tint:?color=false):void=
        if(CharTexture:=Font.GetStringTexture["{InChar}", ?ErrorOrigin:="AddStringWidget"]):
            NewTextureBlock : texture_block = texture_block{DefaultImage:=CharTexture, DefaultDesiredSize:=vector2{X:=CurrSize,Y:=CurrSize}}
            NewStackSlot : stack_box_slot = stack_box_slot{
                Widget:=NewTextureBlock
                HorizontalAlignment := horizontal_alignment.Fill
                VerticalAlignment := vertical_alignment.Fill
            }
            if(ThisTintColor:=Tint?):
                NewTextureBlock.SetTint(ThisTintColor)
            NewStackBoxSlot : string_block_slot = string_block_slot{
                String:="{InChar}"
                StringWidget:=NewTextureBlock
                StackboxWidget:=option{NewStackSlot}
            }
            if(InChar = ' '):
                NewTextureBlock.HideNoCollapse()
                NewTextureBlock.SetDesiredSize(vector2{X:=SpacingMultiplier*CurrSize,Y:=CurrSize})
                NewStackBoxSlot.SpacerWidget.Hide()
            else:
                NewStackBoxSlot.SpacerWidget.SetDesiredSize(vector2{X:=-1*(CurrSize/CurrSpacerSize), Y:=CurrSize})
            set StringBoxSlots += array{NewStackBoxSlot}
            StringStackbox.AddWidget(NewStackSlot)
            StringStackbox.AddWidget(NewStackBoxSlot.SetAndGetSpacerStackboxSlot())
        else{Print("~ERROR: Failed to find a texture for the char [{InChar}] for Font [{Font}]")}

    SetBackplate<public><final>(Texture:texture):void=
        BackplateWidget.SetImage(Texture)
        ResizeBackplate(BackplateOffsets)

    ResizeBackplate<public><final>(Offsets:vector2):void=
        set BackplateOffsets = Offsets
        ScreenSizeX:=GetTextScreenSize().X/2.0
        BackplateWidget.SetDesiredSize(Offsets+vector2{X:=ScreenSizeX,Y:=CurrSize})

    HideBackplate<public><final>():void=
        BackplateWidget.Hide()

    ShowBackplate<public><final>():void=
        BackplateWidget.Show()

    var<private> VisibleState : logic = true
    Hide<public><final>():void=
        WidgetControl.Hide()
        set VisibleState = false

    Show<public><final>():void=
        if(CurrentText.Length > 0):
            WidgetControl.Show()
        set VisibleState = true

    # Changes the color of the entire text #
    SetColor<public><final>(Color:color):void=
        set CurrColor = Color
        LerpColorEvent.Signal(0, CurrentText.Length-1)
        SetColorsEvent.Signal()
        for(ThisStackBox:StringBoxSlots):
            ThisStackBox.StringWidget.SetTint(Color)

    # Change the color of the text within a specified range #
    SetColor<public><final>(Range:tuple(int, int), Color:color):void=
        for(ThisIndex:=Range(0)..Range(1)):
            if(ThisCharWidget:=StringBoxSlots[ThisIndex].StringWidget):
                ThisCharWidget.SetTint(Color)

    LerpColorEvent<private>:event(tuple(int, int))=event(tuple(int, int)){}
    LerpColor<public><final>(Range:tuple(int, int), DesiredColor:color, Time:float, ?AlwaysComplete:logic=false)<suspends>:void=
        var ThisRange : tuple(int, int) = Range
        if(Range(1) <= -1):
            set ThisRange = (Range(0), CurrentText.Length-1)
        var OriginalColor : color = NamedColors.White
        if(ThisStringBox:=StringBoxSlots[Range(0)]):
            set OriginalColor = ThisStringBox.StringWidget.GetTint()
        LerpColorEvent.Signal(ThisRange)
        Winner:=race:
            block:
                if(Time > 0.0, StepMax := Int[30*Time]):
                    for(Step := 1..StepMax):
                        StepPercent := (Step*1.0)/(StepMax*1.0)
                        StepColor := CLerp(OriginalColor, DesiredColor, StepPercent)
                        SetColor(ThisRange, StepColor)
                        Sleep(0.0)
                else:
                    SetColor(ThisRange, DesiredColor)
                1
            block:
                loop:
                    NewLerpRange:=LerpColorEvent.Await()
                    if(NewLerpRange.IsInRange[ThisRange]):
                        break
                -1
        if(AlwaysComplete?, Winner=-1):
            SetColor(ThisRange, DesiredColor)

    SetColorsEvent<private>:event()=event(){}
    SetColors<public><final>(InColorArray:[]color, TickTime:float, InRange:tuple(int, int), ?RepeatTimes:int=0)<suspends>:void=
        var CurrRepeats : int = 0
        var ThisRange : tuple(int, int) = InRange
        if(ThisRange = (-1, -1)):
            set ThisRange = (0, CurrentText.Length-1)
        SetColorsEvent.Signal()
        race:
            loop:
                for(ThisColor:InColorArray):
                    SetColor(ThisRange, ThisColor)
                    Sleep(TickTime)
                if(CurrRepeats >= RepeatTimes, RepeatTimes > -1):
                    break
                else:
                    set CurrRepeats += 1
                Sleep(0.0)
            TextChangedEvent.Await() 
            SetColorsEvent.Await()

    # Checks if a range of text indexes is within the range of another #
    (Range:tuple(int, int)).IsInRange<private>(OtherRange:tuple(int, int))<transacts><decides>:void=
        Range(0) >= OtherRange(0) or Range(1) >= OtherRange(0)

    <# MOVEMENT AND SIZING FUNCTIONS #>

    ChangePosition<public><final>(InPosition:vector2):void=
        WidgetControl.ChangePosition(InPosition)

    # Moves the text to the desired position #
    MovePosition<public><final>(InPosition:vector2, Time:float, ?AlwaysComplete:logic=false, ?IsDelta:logic=false)<suspends>:void=
        WidgetControl.MovePosition(InPosition, Time, ?IsDelta:=IsDelta, ?AlwaysComplete:=AlwaysComplete)
    
    # Sets the text size, as well as updating spacing and spacer sizes to scale #
    SetTextSize<public><final>(InSize:float):void=
        set CurrSize = InSize
        var CharIndex : int = 0
        for(ThisStackBox:StringBoxSlots):
            if(ThisStackBox.String = " "):
                ThisStackBox.StringWidget.SetDesiredSize(vector2{X:=SpacingMultiplier*CurrSize,Y:=CurrSize})
                ThisStackBox.SpacerWidget.Hide()
            else:
                ThisStackBox.StringWidget.SetDesiredSize(vector2{X:=CurrSize, Y:=CurrSize})
                ThisStackBox.SpacerWidget.Show()
                if(ThisStackBox.String = "i" or ThisStackBox.String = "!"):
                    ThisStackBox.SpacerWidget.SetDesiredSize(vector2{X:=-1*((CurrSize/CurrSpacerSize)*1.5), Y:=CurrSize})
                    if(PreviousStringBox:=StringBoxSlots[CharIndex-1]):
                        PreviousStringBox.SpacerWidget.SetDesiredSize(vector2{X:=-1*((CurrSize/CurrSpacerSize)*1.5), Y:=CurrSize})
                else if(ThisStackBox.String = "w" or ThisStackBox.String = "m"):
                    ThisStackBox.SpacerWidget.SetDesiredSize(vector2{X:=-1*((CurrSize/CurrSpacerSize)*0.5), Y:=CurrSize})
                    if(PreviousStringBox:=StringBoxSlots[CharIndex-1], PreviousStringBox.String <> "i" and PreviousStringBox.String <> "!" and PreviousStringBox.String <> " "):
                        PreviousStringBox.SpacerWidget.SetDesiredSize(vector2{X:=-1*((CurrSize/CurrSpacerSize)*0.5), Y:=CurrSize})
                else if(ThisStackBox.IsWord?):
                    ThisStackBox.SpacerWidget.Hide()
                else:
                    ThisStackBox.SpacerWidget.SetDesiredSize(vector2{X:=-1*(CurrSize/CurrSpacerSize), Y:=CurrSize})
            set CharIndex += 1
        if(ThisLastStringBox:=StringBoxSlots[StringBoxSlots.Length-1]):
            ThisLastStringBox.SpacerWidget.SetDesiredSize(vector2{})
        ResizeBackplate(BackplateOffsets)

    ResizeEvent<private>:event()=event(){}
    MoveTextSize<public><final>(DesiredSize:float, Time:float, ?AlwaysComplete:logic=false)<suspends>:void=
        ResizeEvent.Signal()
        OriginalSize := CurrSize
        Winner:=race:
            block:
                EndDesiredSize : vector2 = vector2{X:=DesiredSize,Y:=DesiredSize}
                if(Time > 0.0, StepMax := Int[30*Time]):
                    for(Step := 1..StepMax):
                        StepPercent := (Step*1.0)/(StepMax*1.0)
                        StepSize:=Lerp(OriginalSize, DesiredSize, StepPercent)
                        SetTextSize(StepSize)
                        Sleep(0.0)
                else:
                    SetTextSize(DesiredSize)
                1
            block:
                ResizeEvent.Await()
                -1
        if(AlwaysComplete?, Winner=-1):
            SetTextSize(DesiredSize)
        set CurrSize = DesiredSize

    # Set the spacer sizes between letters #
    SetSpacerSize<public><final>(InSize:float):void=
        set CurrSpacerSize = InSize+1.0
        SetTextSize(CurrSize)

    # Set the space size between words #
    SetSpacingSize<public><final>(InSize:float):void=
        set SpacingMultiplier = InSize
        SetTextSize(CurrSize)

    # Get the size that this string_block takes up on the screen #
    GetTextScreenSize<public><final>():vector2=
        var TotalSpacerSize : float = 0.0
        var TotalCharacterSize : float = 0.0
        var CurrBoxIndex : int = 0
        for(ThisStackBox:StringBoxSlots):
            if(CurrBoxIndex < StringBoxSlots.Length-1):
                set TotalSpacerSize += ThisStackBox.SpacerWidget.GetDesiredSize().X
            if(ThisStackBox.String = " "):
                set TotalCharacterSize += ThisStackBox.StringWidget.GetDesiredSize().X
            else:
                set TotalCharacterSize += CurrSize
            set CurrBoxIndex += 1
        return vector2{X:=TotalCharacterSize+TotalSpacerSize, Y:=CurrSize}

    # Set the transform, i.e. position and size #
    SetTransform<public><final>(DesiredPosition:vector2, DesiredSize:float):void=
        ChangePosition(DesiredPosition)
        SetTextSize(DesiredSize)

    # Move the transform over time #
    MoveTransformEvent<private>:event()=event(){}
    MoveTransform<public><final>(DesiredPosition:vector2, DesiredSize:float, Time:float, ?AlwaysComplete:logic=false)<suspends>:void=
        MoveTransformEvent.Signal()
        Winner:=race:
            block:
                sync:
                    MovePosition(DesiredPosition, Time)
                    MoveTextSize(DesiredSize, Time)
                1
            block:
                MoveTransformEvent.Await()
                -1
        if(AlwaysComplete?, Winner=-1):
            SetTransform(DesiredPosition, DesiredSize)

    <# FUNCTIONS FOR WORD & COMMAND FINDING #>

    GetWordIndexes<private>(InString:string):[]tuple(int, int)=
        var Indexes : []tuple(int, int) = array{}
        if(ThisFontFolder:=GetFontFolder[Font]):
            for(ThisWord -> Texture : ThisFontFolder.WordTextures):
                var Appearances : int = 0
                loop:
                    if(WordIndex := InString.FindSubstring[ThisWord, ?DesiredAppearance:=Appearances]):
                        set Indexes += array{(WordIndex, WordIndex+ThisWord.Length-1)}
                        set Appearances += 1
                    else:
                        break
        else{Print("~ERROR: Failed to find the font folder [{Font}]")}
        return Indexes

    (InIndexes:[]tuple(int, int)).FindMatchingStartIndex<private>(InIndex:int)<transacts><decides>:tuple(int, int)=
        var MaybeMatchingIndex : ?tuple(int, int) = false
        for:
            ThisIndexes : InIndexes
            ThisIndexes(0) = InIndex
        do:
            set MaybeMatchingIndex = option{ThisIndexes}
        MaybeMatchingIndex?

    GetSubstringIndexes_Start<private>(InString:string):[]int=
        var Indexes : []int = array{}
        if(ThisFontFolder:=GetFontFolder[Font]):
            for(ThisWord -> Texture : ThisFontFolder.WordTextures):
                var Appearances : int = 0
                loop:
                    if(WordIndex:=InString.FindSubstring[ThisWord, ?DesiredAppearance:=Appearances]):
                        set Indexes += array{WordIndex}
                        set Appearances += 1
                    else:
                        break
        return Indexes

    GetSubstringIndexes_End<private>(InString:string):[]int=
        var Indexes : []int = array{}
        if(ThisFontFolder:=GetFontFolder[Font]):
            for(ThisWord -> Texture : ThisFontFolder.WordTextures):
                var Appearances : int = 0
                loop:
                    if(WordIndex:=InString.FindSubstring[ThisWord, ?DesiredAppearance:=Appearances]):
                        set Indexes += array{WordIndex+ThisWord.Length}
                        set Appearances += 1
                    else:
                        break
        return Indexes

    GetWidgetAtIndex(InIndex:int)<transacts><decides>:widget=
        return StringBoxSlots[InIndex].StringWidget

    GetWidgetAtSubstring(InSubstring:string, ?DesiredAppearance:int=0):?widget=
        InSubstringConv:=InSubstring.ToLower()
        SubstringIndexes := GetWordIndexes(CurrentText)
        var CharIndex : int = 0
        var Appearances : int = 0
        for(ThisStackBox:StringBoxSlots):
            if(MatchingIndex:=SubstringIndexes.FindMatchingStartIndex[CharIndex]):
                if(ThisWord:=CurrentText.Slice[CharIndex, MatchingIndex(1)+1]):
                    if(ThisWord = InSubstringConv):
                        if(Appearances >= DesiredAppearance):
                            return option{ThisStackBox.StringWidget}
                        else{set Appearances += 1}
                    set CharIndex += (MatchingIndex(1)-MatchingIndex(0))
                else{Print("~ERROR: Couldn't find the word within the string!")}
            else if("{CurrentText[CharIndex]}" = InSubstringConv):
                if(Appearances >= DesiredAppearance):
                    return option{ThisStackBox.StringWidget}
                else{set Appearances += 1}
            set CharIndex += 1
        return false

        # <CommandType_Specification> TEXT HERE <E> #

    CommandTypes : []string = array{
        "c_"
        "ca_"
        "ch_"
    }
    ColorCommands : [string]color = map{
        "w" => NamedColors.White
        "r" => NamedColors.Red
        "o" => NamedColors.Orange
        "y" => NamedColors.Yellow
        "g" => NamedColors.Green
        "b" => NamedColors.Blue
        "i" => NamedColors.Indigo
        "v" => NamedColors.Violet
    }
    GetColorArray(InSpecification:string)<transacts><decides>:[]color=
        var ThisColorArray : []color = array{}
        case(InSpecification):
            "rainbow" => set ThisColorArray = RainbowArray
            "flash" => set ThisColorArray = FlashArray
            "gold wave" => set ThisColorArray = GoldWaveArray
            _ =>
        ThisColorArray.Length > 0
        return ThisColorArray

    GetRealText(InText:string):string=
        var RealText : string = ""
        CommandPoints := GetCommands(InText)
        if(CommandPoints.Length > 0):
            var RealTextRanges : []tuple(int, int) = array{}
            if(ThisFirstCommand:=CommandPoints[0], ThisFirstCommand.CommandStartIndex > 0):
                if(ThisFirstText:=InText.Slice[0, ThisFirstCommand.CommandStartIndex]):
                    set RealText += ThisFirstText
            var CurrCommPoint : int = 0
            for(ThisCommandPoint:CommandPoints):
                RealTextRange:=ThisCommandPoint.GetTextRange(?IncludeCommands:=true)
                if(ThisPreviousCommandPoint:=CommandPoints[CurrCommPoint-1]):
                    LastCommEnd := ThisPreviousCommandPoint.CommandEndIndex+3
                    if(ThisRealText:=InText.Slice[LastCommEnd, RealTextRange(0)]):
                        set RealText += ThisRealText
                if(ThisRealText:=InText.Slice[RealTextRange(0)+ThisCommandPoint.Command.Length+ThisCommandPoint.Specification.Length+2, RealTextRange(1)+1]):
                    set RealText += ThisRealText
                set CurrCommPoint += 1
                
            if(ThisLastCommand:=CommandPoints[CommandPoints.Length-1]):
                if(ThisLastText:=InText.Slice[ThisLastCommand.CommandEndIndex+3, InText.Length]):
                    set RealText += ThisLastText
        else:
            return InText
        return RealText

    ExecuteCommands(InCommandedText:string):void=
        SetColor(NamedColors.White)
        CommandPoints := GetCommands(InCommandedText)
        for(ThisCommandPoint:CommandPoints):
            if(ThisCommandTypeIndex:=CommandTypes.Find[ThisCommandPoint.Command]):
                case(ThisCommandTypeIndex):
                    0 => # COLOR #
                        if(ThisColor:=ColorCommands[ThisCommandPoint.Specification]):
                            Range:=ThisCommandPoint.GetTextRange(?IncludeCommands:=false)
                            SetColor((Range(0), Range(1)), ThisColor)
                        else:
                            Print("~ERR: Failed to fetch a color with the type [{ThisCommandPoint.Specification}]")
                    1 => # COLOR CHANGING #
                        if(ThisColorArray:=GetColorArray[ThisCommandPoint.Specification]):
                            Range:=ThisCommandPoint.GetTextRange(?IncludeCommands:=false)
                            spawn{SetColors(ThisColorArray, 0.15, Range, ?RepeatTimes:=-1)}
                        else:
                            Print("~ERR: Failed to fetch color array [{ThisCommandPoint.Specification}]")
                    2 => # COLOR WITH HEX #
                        ThisColorHex:=MakeColorFromHex("{ThisCommandPoint.Specification}")
                        Range:=ThisCommandPoint.GetTextRange(?IncludeCommands:=false)
                        SetColor((Range(0), Range(1)), ThisColorHex)
                    _ => Print("~ERR: Unknown command")
            else:
                Print("~ERR: Failed to find a command [{ThisCommandPoint.Command}]")
        if(CommandPoints.Length <= 0):
            SetColor(NamedColors.White)

    GetCommands(InCommandedText:string):[]stringcommand=
        var CommandPoints : []stringcommand = array{}
        loop:
            if:
                CommStartIndex:=InCommandedText.FindSubstring["<", ?DesiredAppearance:=CommandPoints.Length*2]
                CommEndIndex:=InCommandedText.FindSubstring[">", ?DesiredAppearance:=CommandPoints.Length*2] > CommStartIndex
                CommSeperatorIndex:=InCommandedText.FindSubstring["_", ?DesiredAppearance:=CommandPoints.Length*2] < CommEndIndex
                CommandStopIndex:=InCommandedText.FindSubstring["<e>", ?DesiredAppearance:=CommandPoints.Length] > CommEndIndex
                CommandType:=InCommandedText.Slice[CommStartIndex+1, CommSeperatorIndex+1]
                CommandTypeSpecification:=InCommandedText.Slice[CommStartIndex+CommandType.Length+1, CommEndIndex]
            then:
                NewCommandPoint : stringcommand = stringcommand{
                    Command := CommandType
                    Specification := CommandTypeSpecification
                    CommandStartIndex := CommStartIndex
                    CommandEndIndex := CommandStopIndex
                }
                for(ThisCommandPoint:CommandPoints):
                    set NewCommandPoint.IndexSubtractive += ThisCommandPoint.GetTotalLength()
                set CommandPoints += array{NewCommandPoint}
            else:
                break
        return CommandPoints

    (InCommandPoint:stringcommand).GetTextRange(?IncludeCommands:logic=false):tuple(int, int)=
        if(IncludeCommands?):
            return (InCommandPoint.CommandStartIndex, InCommandPoint.CommandEndIndex-1)
        else:
            return (InCommandPoint.CommandStartIndex-InCommandPoint.IndexSubtractive, (InCommandPoint.CommandEndIndex-InCommandPoint.IndexSubtractive)-(3+InCommandPoint.Command.Length+InCommandPoint.Specification.Length))
            # (InCommandPoint.CommandEndIndex-InCommandPoint.IndexSubtractive)-6-InCommandPoint.Specification.Length

    var<internal> DebugState : logic = false
    EnableDebug<public>():void=
        WidgetControl.EnableDebug()
        set DebugState = true

    DisableDebug<public>():void=
        WidgetControl.DisableDebug()
        set DebugState = false

string_block_slot := class:

    var String : string
    var StackboxWidget : ?stack_box_slot = false
    StringWidget : texture_block = texture_block{DefaultImage:=Art.UIStuff.Fonts.InteractionText.Numbers.Number_0}
    SpacerWidget : color_block = color_block{DefaultOpacity:=0.0, DefaultDesiredSize:=vector2{X:=0.0,Y:=0.0}}

    var IsWord : logic = false

    var StackboxSpacerSlot : ?stack_box_slot = false
    SetAndGetSpacerStackboxSlot():stack_box_slot=
        ThisStackboxSlot:=stack_box_slot{
            Widget:=SpacerWidget
        }
        set StackboxSpacerSlot = option{ThisStackboxSlot}
        return ThisStackboxSlot  #>

# Container to hold a map of character/numerical textures of the same font #
font_folder := class{
    CharTextures : [string]texture = map{
        # a -> Folder.LetterTexture_A
        # b -> Folder.LetterTexture_B
        # c -> Folder.LetterTexture_C
        # d...z

        # 0 -> Folder.NumberTexture_0
        # 1 -> Folder.NumberTexture_1
        # 2..9
    }
    WordTextures : [string]texture = map{
        # IconA -> Folder.IconTexture_A
        # IconB -> Folder.IconTexture_B
        # IconC...
    }

    # Contains the width of each character #
    CharSizes : [char]int = map{
        # a -> 18
        # b -> 19
        # c -> ...
    }

    IsValid()<transacts><decides>:void=
        CharTextures.Length > 0 or WordTextures.Length > 0
}

stringcommand := class{
    # The command to execute #
    Command : string
    # Allows for more complex commands #
    Specification : string
    # Where does the initiating command string start? #
    CommandStartIndex : int
    # Where does the stop command string start? #
    CommandEndIndex : int

    var IndexSubtractive : int = 0

    GetTotalLength():int=
        # The '+5' accounts for the "<_>" characters as well as the comm end "<E>"
        return Command.Length+Specification.Length+6
}
# Command.Length+Specification.Length+CommandStartIndex+2

# Returns the index at which "Substring" starts at #
# Returns -1 if the Substring cannot be found within "InString" #
# "DesiredAppearance" will force the function to overlook instances of the substring up to the amount given #
# If "DesiredAppearance" is 0, the function will return the first appearance of the substring. If it is 2, it will return the 3rd appearance #
(InString:string).FindSubstring(Substring:string, ?DesiredAppearance:int=0)<transacts><decides>:int=
    var MaybeSubstringIndex : ?int = false
    var AppearanceCount : int = 0
    if(InString.Length >= Substring.Length):
        for:
            ThisStringIndex := 0..InString.Length
            InString.Slice[ThisStringIndex, ThisStringIndex+Substring.Length]=Substring
            not MaybeSubstringIndex?
        do:
            if(AppearanceCount >= DesiredAppearance):
                set MaybeSubstringIndex = option{ThisStringIndex}
            set AppearanceCount += 1
    MaybeSubstringIndex?

(InString:string).FindSubstrings(Substring:string)<transacts>:[]int=
    var SubstringIndexes : []int = array{}
    var CountedSubstrings : int = 0
    loop:
        if(ThisIndex:=InString.FindSubstring[Substring, ?DesiredAppearance:=CountedSubstrings]):
            set SubstringIndexes += array{ThisIndex}
        else{break}
        set CountedSubstrings += 1
    return SubstringIndexes

(InString:string).SliceAroundSubstring(Substring:string)<transacts><decides>:string=
    var SlicedString : string = InString
    ThisSubstringIndex:=InString.FindSubstring[Substring]
    PreText := InString.Slice[0, ThisSubstringIndex]
    if(PostText := InString.Slice[ThisSubstringIndex+Substring.Length, InString.Length]):
        set SlicedString = PreText+PostText
    else:
        set SlicedString = PreText
    return SlicedString

(InString:string).SliceUpToSubstring(Substring:string)<transacts><decides>:string=
    SubstringIndex := InString.FindSubstring[Substring]
    return InString.Remove[SubstringIndex, InString.Length]

(InString:string).SliceAfterSubstring(Substring:string)<transacts><decides>:string=
    SubstringIndex := InString.FindSubstring[Substring]
    return InString.Remove[0, SubstringIndex+Substring.Length]

(InString:string).FindIndexOfClosestSubstring(AtIndex:int, Substring:string)<transacts><decides>:int=
    SubstringIndexes := InString.FindSubstrings(Substring)
    var ClosestIndex : int = SubstringIndexes[0]
    for(ThisIndex:SubstringIndexes):
        if(Abs(ThisIndex-AtIndex) < ClosestIndex):
            set ClosestIndex = ThisIndex
    return ClosestIndex

########  ########     ###     ######    #######  ##    ## ########     ###     ######   ######## 
##     ## ##     ##   ## ##   ##    ##  ##     ## ###   ## ##     ##   ## ##   ##    ##  ##       
##     ## ##     ##  ##   ##  ##        ##     ## ####  ## ##     ##  ##   ##  ##        ##       
##     ## ########  ##     ## ##   #### ##     ## ## ## ## ########  ##     ## ##   #### ######   
##     ## ##   ##   ######### ##    ##  ##     ## ##  #### ##   ##   ######### ##    ##  ##       
##     ## ##    ##  ##     ## ##    ##  ##     ## ##   ### ##    ##  ##     ## ##    ##  ##       
########  ##     ## ##     ##  ######    #######  ##    ## ##     ## ##     ##  ######   ########

<###########################################

DYNAMIC_WIDGET CLASS - UI

Allows for movement and resizing of widgets

###########################################>

<#
dynamic_widget := class:

    var<private> ParentWidget : stack_box = stack_box{Orientation:=orientation.Vertical}
    var<private> VisibleWidget : ?widget = false

    MoverX<private> : color_block = color_block{DefaultOpacity:=0.0}
    MoverY<private> : color_block = color_block{DefaultOpacity:=0.0}
    GetMoverSize():vector2=
        return vector2{
            X:=MoverX.GetDesiredSize().X
            Y:=MoverY.GetDesiredSize().Y
        }

    var<private> Origin : vector2 = vector2{}
    var CurrAlignment<private> : vector2 = vector2{}
    var CurrAnchors<private> : vector2 = vector2{}
    var CurrentOffset<private> : vector2 = vector2{}
    var<private> CurrentPosition : vector2 = vector2{}
    var<private> CurrentSize : vector2 = vector2{}

    Create<public>(InWidget:widget, Position:vector2, ?AutoAnchor:logic=false, ?AnchorOverride:?anchors=false, ?InitialOffset:vector2=vector2{}, ?Alignment:vector2=vector2{}):canvas_slot=
        set VisibleWidget = option{InWidget}
        set CurrentSize = GetSize()
        var ThisAnchor : anchors = anchors{}
        if(AutoAnchor?):
            AnchorVectors:=vector2{
                X:=Position.X/1920.0
                Y:=Position.Y/1080.0
            }
            #Print("Anchor vector: {AnchorVectors}")
            set ThisAnchor = anchors{
                Minimum:=AnchorVectors
                Maximum:=AnchorVectors
            }
        if(ThisAnchorOverride:=AnchorOverride?):
            set ThisAnchor = ThisAnchorOverride
        set CurrAnchors = ThisAnchor.Minimum
        set Origin = vector2{
            X:=ThisAnchor.Minimum.X*1920.0
            Y:=ThisAnchor.Minimum.Y*1080.0
        }
        set CurrentPosition = Position
        set CurrAlignment = Alignment
        defer{
            ChangePosition(CurrentPosition+InitialOffset)
            if(DebugState?):
                EnableDebug()
        }
        set ParentWidget = stack_box{
            Orientation:=orientation.Vertical
            Slots:=array:
                stack_box_slot{
                    Widget:=MoverY
                    HorizontalAlignment:=horizontal_alignment.Fill
                    VerticalAlignment:=vertical_alignment.Fill
                }
                stack_box_slot{
                    Widget:=stack_box{
                        Orientation:=orientation.Horizontal
                        Slots:=array:
                            stack_box_slot{
                                Widget:=MoverX
                                HorizontalAlignment:=horizontal_alignment.Fill
                                VerticalAlignment:=vertical_alignment.Fill
                            }
                            stack_box_slot{
                                Widget:=InWidget
                                HorizontalAlignment:=horizontal_alignment.Fill
                                VerticalAlignment:=vertical_alignment.Fill
                            }
                    }
                }
        }
        return canvas_slot{
            Widget:=ParentWidget
            Anchors:=ThisAnchor
            Alignment:=Alignment
            Offsets:=margin{Top:=0.0,Left:=0.0}
            SizeToContent:=true
        }

    Create(Position:vector2, ?AutoAnchor:logic=false, ?AnchorOverride:?anchors=false, ?InitialOffset:vector2=vector2{}, ?Alignment:vector2=vector2{}):canvas_slot=
        if(ThisWidget:=VisibleWidget?):
            return Create(ThisWidget, Position, ?AutoAnchor:=AutoAnchor, ?AnchorOverride:=AnchorOverride, ?InitialOffset:=InitialOffset, ?Alignment:=Alignment)
        else:
            Print("ERR: Failed to create a DynamicWidget, no Widget was set when initialized")
            return canvas_slot{Widget:=color_block{}}

    ChangePosition<public><final>(InPosition:vector2, ?IsDelta:logic=false):void=
        if(IsDelta?):
            MoverX.SetDesiredSize(vector2{X:=MoverX.GetDesiredSize().X+InPosition.X})
            MoverY.SetDesiredSize(vector2{X:=MoverY.GetDesiredSize().Y+InPosition.Y})
            set CurrentPosition = (CurrentPosition + InPosition)
        else:
            XScale:=vector2{X:=(InPosition.X-Origin.X)+((InPosition.X-Origin.X)*(CurrAlignment.X*2.0))}
            MoverX.SetDesiredSize(XScale)
            YScale:=vector2{Y:=(InPosition.Y-Origin.Y)+((InPosition.Y-Origin.Y)*(CurrAlignment.Y*2.0))}
            MoverY.SetDesiredSize(YScale)
            set CurrentPosition = InPosition

    ChangeSize<public><final>(DesiredSize:vector2):void=
        if(ThisTexture:=texture_block[VisibleWidget?]):
            ThisTexture.SetDesiredSize(DesiredSize)
        else if(ThisColorBlock:=color_block[VisibleWidget?]):
            ThisColorBlock.SetDesiredSize(DesiredSize)
        else:
            Print("ERR: Called ChangeSize() on a Dynamic Widget that has no sizing support!")
        set CurrentSize = DesiredSize

    GetSize<public><final>():vector2=
        if(ThisTexture:=texture_block[VisibleWidget?]):
            return ThisTexture.GetDesiredSize()
        else if(ThisColorBlock:=color_block[VisibleWidget?]):
            return ThisColorBlock.GetDesiredSize()
        else:
            return vector2{}

    ResetPosition<public><final>():void=
        MoverX.SetDesiredSize(vector2{X:=CurrentOffset.X})
        MoverY.SetDesiredSize(vector2{Y:=CurrentOffset.Y})

    MoveEvent:event()=event(){}
    MovePosition<public><final>(DesiredPosition:vector2, Time:float, ?IsDelta:logic=false, ?AlwaysComplete:logic=true)<suspends>:void=
        var EndPosition : vector2 = DesiredPosition
        if(IsDelta?):
            set EndPosition = (CurrentPosition + DesiredPosition)
        MoveEvent.Signal()
        race:
            block:
                TRI:=GetTickRateInfluence()
                if(Time > 0.0, StepMax := Int[30*Time*TRI]):
                    OriginalPos : vector2 = CurrentPosition
                    for(Step := 1..StepMax):
                        StepPercent := ((Step*1.0)/(StepMax*1.0))
                        NextStepPos := Lerp(OriginalPos, EndPosition, StepPercent)
                        ChangePosition(NextStepPos)
                        Sleep(0.0)
                    ChangePosition(EndPosition)
                    return
                else:
                    ChangePosition(EndPosition)
            block:
                MoveEvent.Await()
        if(AlwaysComplete?):
            ChangePosition(EndPosition)

    ResizeEvent<private>:event(tuple(vector2, float))=event(tuple(vector2, float)){}
    InterruptResize<public>:event()=event(){}
    MoveSize<public>(DesiredSize:vector2, Time:float, ?AlwaysComplete:logic=true)<suspends>:void=
        ResizeEvent.Signal((DesiredSize, Time))
        race:
            block:
                TRI:=GetTickRateInfluence()
                if(Time > 0.0, StepMax := Int[30*Time*TRI]):
                    StartingSize := CurrentSize
                    for(Step := 1..StepMax):
                        StepPercent := (Step*1.0)/(StepMax*1.0)
                        NextStepSize := Lerp(StartingSize, DesiredSize, StepPercent)
                        ChangeSize(NextStepSize)
                        Sleep(0.0)
                    ChangeSize(DesiredSize)
                    return
                else:
                    ResizeEvent.Signal((DesiredSize, Time))
                    Sleep(Inf)
            block:
                ResizeEvent.Await()
            block:
                InterruptResize.Await()
        if(AlwaysComplete?):
            ChangeSize(DesiredSize)

    # Sets the image of the widget, if applicable #
    SetImage<public>(InTexture:texture):void=
        if(TextureWidget:=texture_block[VisibleWidget?]):
            TextureWidget.SetImage(InTexture)
        else{Print("~dynamic_widget ERR: Cannot use \"SetImage()\" on this widget type!")}

    # Hides this dynamic widget #
    Hide<public>():void=
        ParentWidget.SetVisibility(widget_visibility.Hidden)

    # Shows this dynamic widget #
    Show<public>():void=
        ParentWidget.SetVisibility(widget_visibility.Visible)

    var<private> DebugState : logic = false
    EnableDebug():void=
        MoverX.SetOpacity(0.5)
        MoverY.SetOpacity(0.5)
        set DebugState = true

    DisableDebug():void=
        MoverX.SetOpacity(0.0)
        MoverY.SetOpacity(0.0)
        set DebugState = false

    DebugPrint(InString:string):void=
        if(DebugState?):
            Print(InString)
#>

progress_bar<public>:= class(dynamic_widget):

    var ProgressBarOverlay : overlay = overlay{}

    var ProgressWidget : ?widget = false

    var CompletionSize : vector2 = vector2{}

    SetDirection(InDirection:vector2):void=
        set ProgressDirection = InDirection
    var ProgressDirection : vector2 = vector2{X:=1.0}

    DepletedEvent:event()=event(){}
    FullEvent:event()=event(){}

    var<private> ProgressPercent : float = 0.0

    Create(BarWidget : widget, BackgroundWidget : widget, InPosition:vector2, InSize:vector2, ?Alignment:vector2=vector2{}, ?AnchorOverride:?anchors=false, ?AutoAnchor:logic=false, ?AutoOverlay:logic=true):canvas_slot=
        set CompletionSize = InSize
        set ProgressWidget = option{BarWidget}
        if(ThisColor:=color_block[BackgroundWidget]):
            ThisColor.SetDesiredSize(InSize)
        else if(ThisTexture:=texture_block[BackgroundWidget]):
            ThisTexture.SetDesiredSize(InSize)
        
        if(AutoOverlay?):
            # Create overlay between progress background and progress bar #
            set ProgressBarOverlay = overlay{
                Slots := array{
                    overlay_slot:
                        Widget := BackgroundWidget
                        HorizontalAlignment := horizontal_alignment.Fill
                        VerticalAlignment := vertical_alignment.Fill
                    overlay_slot:
                        Widget := BarWidget
                        HorizontalAlignment := horizontal_alignment.Left
                        VerticalAlignment := vertical_alignment.Top
                }
            }
        else:
            set ProgressBarOverlay = overlay{
                Slots:=array:
                    overlay_slot{
                        Widget:=BackgroundWidget
                        HorizontalAlignment := horizontal_alignment.Fill
                        VerticalAlignment := vertical_alignment.Fill
                    }
            }
        set ProgressWidget = option{BarWidget}
        Create(ProgressBarOverlay, InPosition, ?Alignment:=Alignment, ?AnchorOverride:=AnchorOverride, ?AutoAnchor:=AutoAnchor)

    ProgressChangingEvent:event()=event(){}
    ChangeProgressOverTime<public>(InTime:float, DesiredPercent:float, ?InAgent:?agent=false, ?AlwaysComplete:logic=false)<suspends>:void=
        TickRateInfluence := InAgent?.GetPlayerInfo[].TickRatePercent or 1.0
        ProgressChangingEvent.Signal()
        if(ProgressDirection.X > 0.0):
            CompletedEvent:event()=event(){}
            branch:
                StartingPercent:float=ProgressPercent
                StartTime:=GetSimulationElapsedTime()
                var TotalPauseTime : float = 0.0
                loop:
                    if(IsGameState[game_state.Paused]):
                        StartPauseTime:=GetSimulationElapsedTime()
                        AwaitGameState(game_state.InGame)
                        set TotalPauseTime += (GetSimulationElapsedTime()-StartPauseTime)
                    LerpPercent:=((GetSimulationElapsedTime()-StartTime)-TotalPauseTime)/InTime
                    if(LerpPercent >= 1.0):
                        break
                    NextProgressStep := Lerp(StartingPercent, DesiredPercent, LerpPercent)
                    SetProgress(NextProgressStep)
                    Sleep(0.0)
                SetProgress(DesiredPercent)
                CompletedEvent.Signal()
            race:
                block:
                    if(ThisPlayerInfo:=InAgent?.GetPlayerInfo[]):
                        ThisPlayerInfo.LeaveEvent.Await()
                    else. Sleep(Inf)
                ProgressChangingEvent.Await()
                CompletedEvent.Await()
        if. AlwaysComplete? then. SetProgress(DesiredPercent)

    HaltProgressChanging<public>():void=
        ProgressChangingEvent.Signal()

    ProgressChangedEvent:event()=event(){}
    SetProgress<public>(DesiredPercent:float):void=
        var FixedPercent : float = DesiredPercent
        if(FixedPercent > 100.0):
            set FixedPercent = 100.0
        set ProgressPercent = FixedPercent
        if(ProgressDirection.X > 0.0):
            SizeDestination : vector2 = if. DesiredPercent <= 0.0 then. vector2{X:=0.0, Y:=CompletionSize.Y} else. vector2{X:=CompletionSize.X*(FixedPercent/100.0), Y:=CompletionSize.Y}
            if(ColorWidget:=color_block[ProgressWidget?]):
                ColorWidget.SetDesiredSize(SizeDestination)
            else if(TextureWidget:=texture_block[ProgressWidget?]):
                TextureWidget.SetDesiredSize(SizeDestination)
            else:
                Print("ERR: No widgets in progress bar!")
        else:
            SizeDestination : vector2 = vector2{X:=CompletionSize.X, Y:=CompletionSize.Y*(FixedPercent/100.0)}

    Hide<override>(?DoCollapse:logic=false):void=
        if. DoCollapse? then. ProgressBarOverlay.Collapse() else. ProgressBarOverlay.Hide()

    Show<override>():void=
        ProgressBarOverlay.Show()
    
    GetBarWidget<public>()<transacts><decides>:widget=
        ProgressWidget?

(ColorBlock:color_block).Resize<public>(DesiredSize:vector2, Time:float, ?InterpolationType:interpolation_type=interpolation_type.Linear)<suspends>:void=
    StartingSize := ColorBlock.GetDesiredSize()
    StartTime:=GetSimulationElapsedTime()
    loop:
        LerpPercent:=(GetSimulationElapsedTime()-StartTime)/Time
        if(LerpPercent >= 1.0):
            break
        NextStepSize := Lerp(StartingSize, DesiredSize, LerpPercent.ApplyInterpolation(InterpolationType))
        ColorBlock.SetDesiredSize(NextStepSize)
        Sleep(0.0)
    ColorBlock.SetDesiredSize(DesiredSize)

(TextureBlock:texture_block).Resize<public>(DesiredSize:vector2, Time:float, ?InterpolationType:interpolation_type=interpolation_type.Linear)<suspends>:void=
    StartingSize := TextureBlock.GetDesiredSize()
    StartTime:=GetSimulationElapsedTime()
    loop:
        LerpPercent:=(GetSimulationElapsedTime()-StartTime)/Time
        if(LerpPercent >= 1.0):
            break
        NextStepSize := Lerp(StartingSize, DesiredSize, LerpPercent.ApplyInterpolation(InterpolationType))
        TextureBlock.SetDesiredSize(NextStepSize)
        Sleep(0.0)
    TextureBlock.SetDesiredSize(DesiredSize)
    
(ColorBlock:color_block).MoveOpacity<public>(DesiredOpacity:float, Time:float, ?InterpolationType:interpolation_type=interpolation_type.Linear, ?AlwaysFinish:logic=true)<suspends>:void=
    StartingOpacity := ColorBlock.GetOpacity()
    StartTime:=GetSimulationElapsedTime()
    loop:
        LerpPercent:=(GetSimulationElapsedTime()-StartTime)/Time
        if(LerpPercent >= 1.0):
            break
        LerpVal := LerpPercent.ApplyInterpolation(InterpolationType)
        if. NextStepSize := type{_X:float where 0.000000 <= _X, _X <= 1.000000}[Lerp(StartingOpacity, DesiredOpacity, LerpVal)] then. ColorBlock.SetOpacity(NextStepSize)
        Sleep(0.0)