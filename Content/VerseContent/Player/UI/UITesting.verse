using { /Fortnite.com/Devices }
using { /Fortnite.com/Characters }
using { /Fortnite.com/Game }
using { /Fortnite.com/UI }
using { /Fortnite.com/FortPlayerUtilities }
using { /Verse.org/Simulation }
using { /Verse.org/Random }
using { /Verse.org/Colors }
using { /UnrealEngine.com/Temporary/Diagnostics }
using { /UnrealEngine.com/Temporary/SpatialMath }
using { /UnrealEngine.com/Temporary/UI }
using { /Verse.org/Assets }
using { Perks }
using { Trials }
using { Weapons }
using { InteractiveModule }

perk_ui_collection<public> := class:

    <#
    PerkTextures : [string]texture = map{
        "savior soda" => Art.UIStuff.Icons.Perk_Icons.SaviorSoda_Perk_Icon
        "maxi munitions" => Art.UIStuff.Icons.Perk_Icons.MaxMunitions_Perk_Icon
        "marathon cola" => Art.UIStuff.Icons.Perk_Icons.Marathon_Perk_Icon
        "tanked up" => Art.UIStuff.Icons.Perk_Icons.Tank_Perk_Icon
        "cryo cola" => Art.UIStuff.Icons.Perk_Icons.Cryo_Perk_Icon
        "bullet brew" => Art.UIStuff.Icons.Perk_Icons.Bull_Perk_Icon
        "bomb cola" => Art.UIStuff.Icons.Perk_Icons.Bomb_Perk_Icon
    }#>

    Stackbox : stack_box = stack_box{
        Orientation:=orientation.Horizontal
    }

    var PerkUIMap : [perk_base]?perk_ui_single = map{}

    Create():canvas_slot=
        NewCanvasSlot : canvas_slot = canvas_slot{
            Widget := Stackbox
            Anchors:=anchors{
                Minimum:=vector2{Y:=1.0}
                Maximum:=vector2{Y:=1.0}
            }
            Alignment:=vector2{Y:=1.0}
            Offsets := margin{Left:=240.0,Top:=-61.0}
            SizeToContent := true
        }

    AddPerk<public>(InPerk:perk_base):void=
        if(not PerkUIMap[InPerk]?):
            if(ThisPerkTexture:="Perk".GetStringTexture[InPerk.Name.ToLower()]):
                NewPerkTextureBlock : texture_block = texture_block{DefaultImage:=ThisPerkTexture, DefaultDesiredSize:=vector2{X:=75.0,Y:=75.0}}
                NewTierBackplate : texture_block = texture_block{DefaultImage:=Art.UIStuff.Icons.Perk_Icons.TierPerkIcon, DefaultDesiredSize:=vector2{X:=25.0,Y:=25.0}}
                NewPerkUI : perk_ui_single = perk_ui_single{
                    PerkIcon:=NewPerkTextureBlock
                    TierBackplate:=NewTierBackplate
                }
                NewWidget:=NewPerkUI.Create()
                Stackbox.AddWidget(NewWidget)
                if(set PerkUIMap[InPerk] = option{NewPerkUI}){Print("New perk added success")}
                else:
                    set PerkUIMap = ConcatenateMaps(PerkUIMap, map{InPerk => option{NewPerkUI}})
            else:
                Print("ERROR: Failed to get texture icon for perk [{InPerk.Name}]")
        else:
            Print("Perk UI already active")

    RemovePerk<public>(InPerk:perk_base):void=
        if(ThisPerkUI := PerkUIMap[InPerk]?, ThisPerkWidget:=ThisPerkUI.ThisPerkWidget?):
            Stackbox.RemoveWidget(ThisPerkWidget.Widget)
            if(set PerkUIMap[InPerk] = false){}
        else{Print("~ERR: Failed to fetch a perk ui!")}

    ChangePerkTier<public>(InPerk:perk_base, InTier:int):void=
        Print("~Changing tier to [{InTier}]")
        if(ThisPerkUI := PerkUIMap[InPerk]?):
            ThisPerkUI.SetTier(InTier)
        else{Print("~ERR: Failed to fetch a perk ui!")}

    SetActiveState<public>(InPerk:perk_base, InState:logic):void=
        if(ThisPerkUI := PerkUIMap[InPerk]?):
            ThisPerkUI.SetActiveState(InState)

    Show<public>():void=
        Stackbox.Show()

    Hide<public>():void=
        Stackbox.Hide()
    
perk_ui_single<public> := class:

    PerkIcon<public> : texture_block
    TierBackplate : texture_block = texture_block{DefaultImage:=Art.UIStuff.Icons.Perk_Icons.TierPerkIcon}

    var ThisPerkWidget : ?stack_box_slot = false

    var TierStackbox : stack_box = stack_box{
        Orientation := orientation.Horizontal
    }
    var TierNumbers : []texture_block = array{}

    Create<public>(?IsDesc:logic=false):stack_box_slot=
        var TierSize : vector2 = vector2{X:=2.0,Y:=10.0}
        #var TierMargins : margin = margin{Right:=-7.0}
        var TierMargins : margin = margin{Left:=0.7, Right:=0.7}
        var BackplateSize : vector2 = vector2{X:=25.0,Y:=25.0}
        var BackplateMargin : margin = margin{Bottom:=-2.0}
        var StackboxMargin : margin = margin{Bottom:=5.5}
        if(IsDesc?):
            set TierSize = vector2{X:=2.0,Y:=10.0}
            #set TierMargins = margin{Top:=-10.0, Right:=-8.0}
            set BackplateSize = vector2{X:=30.0,Y:=30.0}
            set BackplateMargin = margin{Bottom:=2.0}
            set StackboxMargin = margin{Bottom:=12.0}
        TierNumber1 : texture_block = texture_block{DefaultImage:=Art.UIStuff.Fonts.InteractionText.Letters.Letter_I, DefaultDesiredSize:=TierSize}
        TierNumber2 : texture_block = texture_block{DefaultImage:=Art.UIStuff.Fonts.InteractionText.Letters.Letter_I, DefaultDesiredSize:=TierSize}
        TierNumber3 : texture_block = texture_block{DefaultImage:=Art.UIStuff.Fonts.InteractionText.Letters.Letter_I, DefaultDesiredSize:=TierSize}
        set TierNumbers = array{
            TierNumber1
            TierNumber2
            TierNumber3
        }
        set TierStackbox = stack_box{
            Orientation := orientation.Horizontal
            Slots := array:
                stack_box_slot{
                    Padding := TierMargins
                    HorizontalAlignment := horizontal_alignment.Center
                    VerticalAlignment := vertical_alignment.Center
                    Widget := TierNumber1
                }
                stack_box_slot{
                    Padding := TierMargins
                    HorizontalAlignment := horizontal_alignment.Center
                    VerticalAlignment := vertical_alignment.Center
                    Widget := TierNumber2
                }
                stack_box_slot{
                    Padding := TierMargins
                    HorizontalAlignment := horizontal_alignment.Center
                    VerticalAlignment := vertical_alignment.Center
                    Widget := TierNumber3
                }
        }
        TierBackplate.SetDesiredSize(BackplateSize)
        NewPerkWidget : stack_box_slot = stack_box_slot{
            Widget := overlay{
                Slots := array:
                    overlay_slot:
                        Widget := PerkIcon
                    overlay_slot:
                        Widget := TierBackplate
                        HorizontalAlignment := horizontal_alignment.Center
                        VerticalAlignment := vertical_alignment.Bottom
                        Padding := BackplateMargin
                    overlay_slot:
                        Widget := TierStackbox
                        HorizontalAlignment := horizontal_alignment.Center
                        VerticalAlignment := vertical_alignment.Bottom
                        Padding := StackboxMargin
            }
            #Padding := margin{Right:=-10.0}
        }
        SetTier(1)
        set ThisPerkWidget = option{NewPerkWidget}
        return NewPerkWidget

    SetTier(InTier:int):void=
        if(InTier > 1):
            TierBackplate.Show()
            for(ThisTierIndex := 0..InTier-1):
                if(ThisTierWidget := TierNumbers[ThisTierIndex]):
                    ThisTierWidget.Show()
        else:
            TierBackplate.Hide()
            for(ThisTierSlot : TierNumbers):
                ThisTierSlot.Hide()

    SetActiveState(InState:logic):void=
        if(InState?):
            PerkIcon.SetTint(NamedColors.White)
        else:
            PerkIcon.SetTint(NamedColors.Grey)

    SetIcon(InTexture:texture):void=
        PerkIcon.SetImage(InTexture)

    Hide():void=
        if(ThisWidget:=ThisPerkWidget?):
            ThisWidget.Widget.Hide()

    Show():void=
        if(ThisWidget:=ThisPerkWidget?):
            ThisWidget.Widget.Show()

quest_ui<public> := class:

    Title : string_block = string_block{Font:="Quest", BackplateTexture:=option{Art.UIStuff.Fonts.InteractionBackground.InteractionBackground}, BackplateOffsets:=vector2{X:=10.0,Y:=10.0}}
    Description : string_block = string_block{Font:="Quest"}

    Icon : string_block = string_block{Font:="Quest", BackplateTexture:=option{Art.UIStuff.Fonts.InteractionBackground.InteractionBackground}, BackplateOffsets:=vector2{X:=25.0,Y:=25.0}}

    TitleSlotPos : vector2 = vector2{X:=1830.0, Y:=66.0}
    Create():[]canvas_slot=
        TitleCanvasSlot := Title.Create(GetQuestTitle(), TitleSlotPos, 30.0, ?SizeRatio:=vector2{X:=1.0,Y:=1.6}, ?SpacerSize:=0.7, ?Alignment:=vector2{X:=1.0,Y:=0.5}, ?AutoAnchor:=true)
        DescriptionCanvasSlot := Description.Create("", vector2{X:=1808.0, Y:=92.0}, 20.0, ?SpacerSize:=0.7, ?Alignment:=vector2{X:=1.0,Y:=0.5}, ?AutoAnchor:=true)
        IconCanvasSlots := Icon.Create("clipboard", vector2{X:=TitleSlotPos.X+40.0, Y:=-150.0}, 35.0, ?AutoAnchor:=true, ?Alignment:=vector2{X:=0.5})
        Icon.Hide()
        return array{TitleCanvasSlot, DescriptionCanvasSlot, IconCanvasSlots}

    var<private> CurrentQuestTracked : ?quest_base = false

    ChangeText<public>(InQuest:quest_base):void=
        Print("Changing quest text")
        set CurrentQuestTracked = option{InQuest}
        Description.SetText(InQuest.BasicSettings.QuestDescription)
        QuestTitle := block{
            if(InQuest.GetProgressRequired() > 1):
                InQuest.BasicSettings.QuestTitle+" (0/{InQuest.GetProgressRequired()})"
            else:
                InQuest.BasicSettings.QuestTitle
        }
        spawn. NewQuestAnimation(QuestTitle)

    IncrementProgress<public>():void=
        if(ThisQuest:=CurrentQuestTracked?):
            if(ThisQuest.GetProgressRequired() > 1):
                Title.SetText(ThisQuest.BasicSettings.QuestTitle+" ({ThisQuest.CurrentProgress}/{ThisQuest.GetProgressRequired()})")

    CompletedAnimation<public>()<suspends>:void=
        Title.SetColors(GreenFlash, 0.1, (-1,-1))
        Title.SetText("")
        Title.Hide()
        Icon.MovePosition(TitleSlotPos+vector2{X:=30.0,Y:=-150.0}, 1.0, ?InterpolationType:=interpolation_type.EaseInOut)
        
    NewQuestAnimation(QuestText:string)<suspends>:void=
        if(IsGameState[game_state.InGame]):
            Print("Starting quest intro anim")
            Icon.SetPosition(TitleSlotPos+vector2{X:=30.0,Y:=-150.0})
            Icon.Show()
            Icon.MovePosition(TitleSlotPos+vector2{X:=30.0}, 1.0)
            branch:
                Title.Show(?IgnoreTextCheck:=true)
                Title.SetText("")
                Title.AddText(QuestText, ?SeperateCharacters:=true, ?AddDelay:=0.15, ?OffsetSize:=-30.0, ?EndOffsetSize:=30.0, ?OffsetSizeChangeInterpolation:=interpolation_type.EaseOutBack)
                Title.SetText(QuestText)
            race:
                Sleep(4.5)
                loop:
                    Icon.MoveTextSize(45.0, 0.5, ?InterpolationType:=interpolation_type.EaseInOut)
                    Icon.MoveTextSize(35.0, 0.5, ?InterpolationType:=interpolation_type.EaseInOut)
            Icon.MoveTextSize(45.0, 0.5, ?InterpolationType:=interpolation_type.EaseInOut)
            Icon.MoveTextSize(35.0, 1.0, ?InterpolationType:=interpolation_type.EaseInOut)
        else:
            Print("Awaiting game start event")
            AwaitGameState(game_state.InGame)
            Sleep(0.5)
            spawn. NewQuestAnimation(QuestText)

    Hide<public>():void=
        Title.Hide()
        Description.Hide()
        Icon.Hide()

    Show<public>():void=
        Title.Show()
        Description.Show()

trial_ui<public> := class:

    TrialDescription : dialog_block = dialog_block{Font:="Quest"}

    DescriptionText : string_block = string_block{Font:="Quest"}
    DescriptionBackground : texture_block = texture_block{DefaultImage:=Art.UIStuff.Icons.Trial_UI.Trial_Icon, DefaultDesiredSize:=vector2{X:=474.0,Y:=264.0}}
    DescriptionBackgroundArrows<public> : texture_block = texture_block{DefaultImage:=Art.UIStuff.Icons.Trial_UI.Trial_Arrows_Icon, DefaultDesiredSize:=vector2{X:=474.0,Y:=264.0}}

    DescriptionText_Hangoff : string_block = string_block{Font:="Quest"}
    DescriptionText_Hangoff2 : string_block = string_block{Font:="Quest"}

    TrialTimerBar<public> : progress_bar = progress_bar{}

    RewardText : string_block = string_block{Font:="Quest"}
    RewardBackground : texture_block = texture_block{DefaultImage:=Art.UIStuff.Icons.Trial_UI.Trial_Reward_Icon, DefaultDesiredSize:=vector2{X:=400.0,Y:=250.0}}
    RewardProgressBar : progress_bar = progress_bar{}
    RewardWidgetControl : dynamic_widget = dynamic_widget{}

    Create():[]canvas_slot=
        var CanvasSlots : []canvas_slot = array{}
        DescriptionText.Create("Desc Text", vector2{}<#vector2{X:=212.0,Y:=236.0}#>, 38.0, ?SizeRatio:=vector2{X:=1.0,Y:=1.5})
        DescriptionText_Hangoff.Create("Desc Text hangoff", vector2{}<#vector2{X:=212.0, Y:=254.0}#>, 25.0, ?SizeRatio:=vector2{X:=1.0,Y:=1.5})
        DescriptionText_Hangoff2.Create("Desc Text hangoff two", vector2{}, 25.0, ?SizeRatio:=vector2{X:=1.0,Y:=1.5})
        RewardText.Create("", vector2{}<#vector2{X:=256.0,Y:=288.0}#>, 15.0, ?SizeRatio:=vector2{X:=1.0,Y:=1.5})
        RewardBarWidget:=RewardProgressBar.Create(color_block{DefaultColor:=MakeColorFromHex("39E75F")}, texture_block{DefaultImage:=Art.UIStuff.Fonts.InteractionBackground.InteractionBackground}, vector2{}, vector2{X:=212.0,Y:=7.0}).Widget
        RewardProgressBar.SetProgress(0.0)
        TrialWidget:=TrialTimerBar.Create(color_block{DefaultColor:=MakeColorFromHex("39E75F")}, texture_block{DefaultImage:=Art.UIStuff.Fonts.InteractionBackground.InteractionBackground}, vector2{}, vector2{X:=388.0,Y:=7.0}).Widget
        TrialTimerBar.SetProgress(100.0)
        set CanvasSlots += array{canvas_slot{
            Widget:=overlay{
                Slots:=array:
                    overlay_slot:
                        Widget:=DescriptionBackground
                    overlay_slot:
                        Widget:=DescriptionBackgroundArrows
                        HorizontalAlignment:=horizontal_alignment.Center
                        VerticalAlignment:=vertical_alignment.Center
                    overlay_slot:
                        Widget:=TrialWidget
                        Padding:=margin{Left:=42.0,Top:=65.0}
                        HorizontalAlignment:=horizontal_alignment.Left
                        VerticalAlignment:=vertical_alignment.Top
                    overlay_slot:
                        Widget:=TrialDescription.Create("", vector2{}, 20.0, ?SizeRatio:=vector2{X:=1.0,Y:=1.5}).Widget
                        HorizontalAlignment:=horizontal_alignment.Center
                        VerticalAlignment:=vertical_alignment.Center
                        Padding:=margin{Top:=20.0}
                    <#
                    overlay_slot:
                        Widget:=DescriptionText.GetWidget()
                        Padding:=margin{Top:=18.0, Left:=-5.0}
                        HorizontalAlignment:=horizontal_alignment.Center
                        VerticalAlignment:=vertical_alignment.Center
                    overlay_slot:
                        Widget:=DescriptionText_Hangoff.GetWidget()
                        Padding:=margin{Top:=6.0, Left:=-4.0}
                        HorizontalAlignment:=horizontal_alignment.Center
                        VerticalAlignment:=vertical_alignment.Center
                    overlay_slot:
                        Widget:=DescriptionText_Hangoff2.GetWidget()
                        Padding:=margin{Top:=30.0, Left:=-4.0}
                        HorizontalAlignment:=horizontal_alignment.Center
                        VerticalAlignment:=vertical_alignment.Center#>
            }
            SizeToContent := true
            Offsets := margin{Top := 104.0, Left := 0.0}
        }}
        RewardOverlay : overlay = overlay{
            Slots:=array:
                overlay_slot:
                    Widget:=RewardBackground
                overlay_slot:
                    Widget:=RewardText.GetWidget()
                    Padding:=margin{Top:=68.0-7.0, Left:=-3.0+46.0}
                    HorizontalAlignment:=horizontal_alignment.Center
                    VerticalAlignment:=vertical_alignment.Center
                overlay_slot:
                    Widget:=RewardBarWidget
                    Padding:=margin{Left:=0.0,Bottom:=30.0}
                    VerticalAlignment:=vertical_alignment.Bottom

        }
        #set CanvasSlots += array{RewardWidgetControl.Create(RewardOverlay, vector2{X:=36.0, Y:=116.0}, vector2{X:=400.0,Y:=250.0})}
        set CanvasSlots += array{RewardWidgetControl.Create(RewardOverlay, vector2{X:=230.0, Y:=280.0})}
        <#
        TrialUICanvasSlot:=canvas_slot{
            Widget:=overlay{
                Slots:=array:
                    overlay_slot:
                        Widget:=RewardOverlay
                        HorizontalAlignment:=horizontal_alignment.Center
                        Padding:=margin{Top:=15.0}
                    overlay_slot:
                        Widget:=DescriptionOverlay
            }
            SizeToContent := true
            Offsets := margin{Top := 104.0, Left := 0.0}
        }#>
        Hide()
        return CanvasSlots

    SetDescription<public>(InText:string, ?LineSeparator:string="\\n"):void=
        DescriptionBackgroundArrows.Show()
        if(InText.FindSubstring["\\n"]):
            TrialDescription.Resize(15.0)
        else:
            TrialDescription.Resize(20.0)
        TrialDescription.SetText(InText, ?LineSeparator:="\\n")
        TrialDescription.Show()
        <#
        DescriptionText.SetText(InText)
        ThisScreenSize:=DescriptionText.GetTextScreenSize()
        if(ThisScreenSize.X >= 365.0):
            DescriptionText.SetText("")
            if(ThisIndex:=Int[InText.Length*365.0/ThisScreenSize.X], BeforeSubstring:=InText.Slice[0, ThisIndex], AfterSubstring:=InText.Slice[ThisIndex, InText.Length]):
                if(ThisClosestSpaceIndex:=InText.FindIndexOfClosestSubstring[ThisIndex, " "], BeforeSpaceSubstring:=InText.Slice[0, ThisClosestSpaceIndex], AfterSpaceSubstring:=InText.Slice[ThisClosestSpaceIndex+1, InText.Length]):
                    DescriptionText_Hangoff.SetText(BeforeSpaceSubstring)
                    DescriptionText_Hangoff2.SetText(AfterSpaceSubstring)
                else:
                    DescriptionText_Hangoff.SetText(BeforeSubstring)
                    DescriptionText_Hangoff2.SetText(AfterSubstring)
            else:
                Print("~ERR: trial_ui : SetDescription()")
            DescriptionText_Hangoff.Show()
            DescriptionText_Hangoff2.Show()
            #NewSize:=38.0 * 368.0/ThisScreenSize.X
            #DescriptionText.ChangeSize(NewSize)
            #Print("~NewSize = {NewSize} vs ScreenSize = {ThisScreenSize}")
        else:
            DescriptionText_Hangoff.Hide()
            DescriptionText_Hangoff2.Hide()
            DescriptionText.SetTextSize(38.0)
            DescriptionText.SetPosition(vector2{})#>

    Hide<public>():void=
        <#
        DescriptionText.Hide()
        DescriptionText_Hangoff.Hide()
        DescriptionText_Hangoff2.Hide()
        #>
        TrialDescription.Hide()
        DescriptionBackgroundArrows.Hide()
        DescriptionBackground.Hide()
        RewardWidgetControl.Hide()
        #RewardBackground.Hide()
        TrialTimerBar.Hide()

    HideTrialDescription<public>(InAgent:agent):void=
        <#
        DescriptionText.Hide()
        DescriptionText_Hangoff.Hide()
        DescriptionText_Hangoff2.Hide() #>
        TrialDescription.Hide()
        DescriptionBackground.Hide()
        DescriptionBackgroundArrows.Hide()
        TrialTimerBar.Hide()
        if(ThisTimerWidget:=color_block[TrialTimerBar.GetBarWidget[]]):
            ThisTimerWidget.SetColor(MakeColorFromHex("39E75F"))
        if(InAgent.GetTrialRewardRarity() = item_rarity.None):
            RewardWidgetControl.Hide()
            #RewardText.Hide()
            #RewardBackground.Hide()
        else:
            RewardWidgetControl.Show()
            RewardWidgetControl.ChangePosition(vector2{X:=230.0,Y:=0.0})
            #RewardText.Show()
            #RewardBackground.Show()

    Show<public>(InAgent:agent):void=
        #DescriptionText.Show()
        TrialDescription.Show()
        DescriptionBackground.Show()
        RewardWidgetControl.ChangePosition(vector2{X:=230.0, Y:=280.0})
        UpdateReward(InAgent)

    SetRewardVisibilityState<public>(State:logic):void=
        if(State?):
            RewardWidgetControl.Show()
        else:
            RewardWidgetControl.Hide()
            
    UpdateReward<public>(InAgent:agent):void=
        if(ThisPlayerInfo:=InAgent.GetPlayerInfo[]):
            RarityTier := InAgent.GetTrialRewardRarity()
            TrialPointAmount := InAgent.GetTrialPoints()
            case(RarityTier):
                item_rarity.None => 
                    if(TrialPointAmount > 0):
                        RewardText.SetText("None")
                        RewardText.SetColor(NamedColors.White)
                        RewardWidgetControl.Show()
                        RewardProgressBar.Show()
                        spawn{RewardProgressBar.ChangeProgressOverTime(0.75, TrialPointAmount*1.0)}
                        if(ThisWidget:=color_block[RewardProgressBar.GetBarWidget[]]):
                            ThisWidget.SetColor(NamedColors.White)
                    else:
                        RewardWidgetControl.Hide()
                        RewardProgressBar.SetProgress(0.0)
                        RewardProgressBar.Hide()
                        if(ThisWidget:=color_block[RewardProgressBar.GetBarWidget[]]):
                            ThisWidget.SetColor(MakeColorFromHex("39E75F"))
                    #RewardText.Hide()
                    #RewardBackground.Hide()
                item_rarity.Common =>
                    RewardText.SetText("<CH_39E75F>Common<E>")
                    RewardWidgetControl.Show()
                    RewardProgressBar.Show()
                    spawn{RewardProgressBar.ChangeProgressOverTime(0.75, (TrialPointAmount-100)*1.0)}
                    if(ThisWidget:=color_block[RewardProgressBar.GetBarWidget[]]):
                        ThisWidget.SetColor(MakeColorFromHex("39E75F"))
                    #RewardText.Show()
                    #RewardBackground.Show()
                item_rarity.Rare =>
                    RewardText.SetText("<CH_89CFF0>Rare<E>")
                    RewardWidgetControl.Show()
                    RewardProgressBar.Show()
                    spawn{RewardProgressBar.ChangeProgressOverTime(0.75, (TrialPointAmount-200)*1.0)}
                    if(ThisWidget:=color_block[RewardProgressBar.GetBarWidget[]]):
                        ThisWidget.SetColor(MakeColorFromHex("89CFF0"))
                    #RewardText.Show()
                    #RewardBackground.Show()
                item_rarity.Epic =>
                    RewardText.SetText("<CH_9D4DBB>Epic<E>")
                    RewardWidgetControl.Show()
                    RewardProgressBar.Show()
                    spawn{RewardProgressBar.ChangeProgressOverTime(0.75, (TrialPointAmount-300)*1.0)}
                    if(ThisWidget:=color_block[RewardProgressBar.GetBarWidget[]]):
                        ThisWidget.SetColor(MakeColorFromHex("9D4DBB"))
                    #RewardText.Show()
                    #RewardBackground.Show()
                item_rarity.Legendary =>
                    RewardText.SetText("<CH_F3AF19>Legendary<E>")
                    RewardWidgetControl.Show()
                    RewardProgressBar.Show()
                    spawn{RewardProgressBar.ChangeProgressOverTime(0.75, (TrialPointAmount-400)*1.0)}
                    if(ThisWidget:=color_block[RewardProgressBar.GetBarWidget[]]):
                        ThisWidget.SetColor(MakeColorFromHex("F3AF19"))
                    #RewardText.Show()
                    #RewardBackground.Show()
                item_rarity.Mythic =>
                    RewardText.SetText("<CA_gold wave>Mythic<E>")
                    RewardWidgetControl.Show()
                    RewardProgressBar.SetProgress(0.0)
                    RewardProgressBar.Hide()
                    #RewardText.Show()
                    #RewardBackground.Show()
            if(RarityTier = item_rarity.Legendary):
                RewardText.SetTextSize(13.0)
            else:
                RewardText.SetTextSize(15.0)
        
description_ui<public> := class:

    var ParentUI : ?ui_manager = false

    DescriptionText : string_block = string_block{Font:="Interaction", BackplateTexture:=option{Art.UIStuff.Fonts.InteractionBackground.InteractionBackground}}
    DescriptionIconControl : dynamic_widget = dynamic_widget{}

    var DescIcon : perk_ui_single = perk_ui_single{PerkIcon:=texture_block{DefaultImage:=Art.UIStuff.Icons.Perk_Icons.Tank_Perk_Icon, DefaultDesiredSize:=vector2{X:=90.0,Y:=90.0}}}

    Create():[]canvas_slot=
        var CanvasSlots : []canvas_slot = array{}
        set CanvasSlots += array{DescriptionText.Create("", vector2{X:=960.0, Y:=195.0}, 30.0, ?SizeRatio:=vector2{X:=1.0,Y:=1.6}, ?Alignment:=vector2{X:=0.5}, ?AnchorOverride:=option{anchors{Minimum:=vector2{X:=0.5},Maximum:=vector2{X:=0.5}}})}
        DescriptionText.ResizeBackplate(vector2{X:=20.0,Y:=20.0})
        IconOverlay : overlay = overlay{
            Slots := array:
                overlay_slot:
                    Widget := texture_block{DefaultImage:=Art.UIStuff.Fonts.InteractionBackground.InteractionBackground, DefaultDesiredSize:=vector2{X:=100.0,Y:=100.0}}
                overlay_slot:
                    Widget := DescIcon.Create(?IsDesc:=true).Widget
        }
        set CanvasSlots += array{DescriptionIconControl.Create(IconOverlay, vector2{X:=960.0, Y:=84.0}, ?Alignment:=vector2{X:=0.5}, ?AnchorOverride:=option{anchors{Minimum:=vector2{X:=0.5},Maximum:=vector2{X:=0.5}}})}
        Hide()
        return CanvasSlots

    Change(InText:string, InIconPath:string, InOrigin:?Interactive_Base):void=
        if(PrevDescription <> InText):
            set PrevDescription = InText
            if(InText.Length > 0):
                DescriptionText.SetText(InText)
                DescriptionText.Show()
                spawn. Popout()
            InIconString := InIconPath.ToLower()
            var HasIcon : logic = false
            if(ThisIconTexture:=InIconString.GetIconTexture[]):
                set HasIcon = true
                DescriptionIconControl.Show()
                DescIcon.SetIcon(ThisIconTexture)
            else:
                DescriptionIconControl.Hide()
                Print("ERR: Failed to find icon texture")
            if(HasIcon?):
                Show()
                if(ThisPerk:=perk_base[InOrigin?]):
                    if:
                        ThisAgent:=ParentUI?.User
                    then:
                        CurrTier:= (ThisPerk.GetTier[ThisAgent]+1) or 1
                        if(CurrTier<=3):
                            DescIcon.Show()
                            DescIcon.SetTier(CurrTier)
                        else:
                            Hide()
                    else:
                        DescIcon.SetTier(0)
                else:
                    DescIcon.SetTier(0)
                    DescIcon.Show()
            else:
                DescIcon.Hide()
                
    Hide(?ClearPrevious:logic=true):void=
        DescriptionIconControl.Hide()
        DescriptionText.Hide()
        if(ClearPrevious?):
            set PrevDescription = ""
        <#
        DescriptionText.Hide()
        DescriptionIconControl.Hide()
        DescIcon.Hide() #>

    var PrevDescription : string = ""
    Show():void=
        if(PrevDescription.Length>0):
            DescriptionText.SetText(PrevDescription)
            DescriptionText.Show()
            DescriptionIconControl.Show()
        <#
        DescriptionText.ChangeText(PrevDescription)
        DescriptionText.Show()
        DescriptionIconControl.Show()
        Desc.Show()#>
        
    popout_event : event() = event(){}
    Popout()<suspends>:void=
        popout_event.Signal()
        StartTime := GetSimulationElapsedTime()
        race:
            popout_event.Await()
            block:
                DescriptionText.MoveTextSize(40.0, 0.6, ?InterpolationType:=interpolation_type.EaseOutBack)
                DescriptionText.MoveTextSize(30.0, 0.3, ?InterpolationType:=interpolation_type.EaseOut)
        DescriptionText.SetTextSize(30.0)


    (InPath:string).GetIconTexture()<transacts><decides>:texture=
        var DesiredIcon : ?texture = false
        case(InPath):
            "savior soda" => set DesiredIcon = option{Art.UIStuff.Icons.Perk_Icons.SaviorSoda_Perk_Icon}
            "maxi munitions" => set DesiredIcon = option{Art.UIStuff.Icons.Perk_Icons.MaxMunitions_Perk_Icon}
            "marathon cola" => set DesiredIcon = option{Art.UIStuff.Icons.Perk_Icons.Marathon_Perk_Icon}
            "tanked up" => set DesiredIcon = option{Art.UIStuff.Icons.Perk_Icons.Tank_Perk_Icon}
            "cryo cola" => set DesiredIcon = option{Art.UIStuff.Icons.Perk_Icons.Cryo_Perk_Icon}
            "bull-it brew" => set DesiredIcon = option{Art.UIStuff.Icons.Perk_Icons.Bull_Perk_Icon}
            "bomb cola" => set DesiredIcon = option{Art.UIStuff.Icons.Perk_Icons.Bomb_Perk_Icon}
            "shield" => set DesiredIcon = option{Art.UIStuff.Icons.QuestObjects.ShieldIcon}
            "power punch" => set DesiredIcon = option{Art.UIStuff.Icons.Weapon.Fist_Icon}
            _ =>
        DesiredIcon?

mainmenu_ui<public> := class:

    var LobbyScoreboardStackbox : stack_box = stack_box{Orientation:=orientation.Vertical}

    #ReadyUpTexture : texture_block = texture_block{DefaultImage:=Art.UIStuff.MainMenu.Ready_Up_Button, DefaultDesiredSize:=vector2{X:=500.0,Y:=300.0}}

    ReadyUpCount : text_block = text_block{DefaultTextColor:=NamedColors.White, DefaultShadowOffset:=option{vector2{X:=5.0,Y:=5.0}}, DefaultShadowColor:=NamedColors.Black, DefaultShadowOpacity:= 1.0, DefaultJustification:=text_justification.Center}
    CountdownText : text_block = text_block{DefaultTextColor:=NamedColors.White, DefaultShadowOffset:=option{vector2{X:=5.0,Y:=5.0}}, DefaultShadowOpacity:= 1.0, DefaultShadowColor:=NamedColors.Black}
    AutostartCountdownText : text_block = text_block{DefaultTextColor:=NamedColors.Yellow, DefaultShadowOffset:=option{vector2{X:=5.0,Y:=5.0}}, DefaultShadowOpacity:= 1.0, DefaultShadowColor:=NamedColors.Black}

    ReadyUpButton : button_quiet = button_quiet{}

    DifficultyButton : button_quiet = button_quiet{}

    DifficultyInfo_Control : dynamic_widget = dynamic_widget{}

    DifficultyInfo_Button : button_quiet = button_quiet{}

    PatchNoteButton : button_quiet = button_quiet{}

    CreditsButton : button_quiet = button_quiet{}

    AdminPanelButton : button_quiet = button_quiet{}

    Info_Title : text_block = text_block{DefaultTextColor:=NamedColors.White, DefaultJustification:=text_justification.Center}
    Info_Description : text_block = text_block{DefaultTextColor:=NamedColors.White}
    var InfoOverlay : overlay = overlay{}

    var SubscribedState : logic = false
    SubscribeButtons():void=
        if(not SubscribedState?):
            set SubscribedState = true
            ReadyUpButton.OnClick().Subscribe(HandleReadyupClick)
            DifficultyButton.OnClick().Subscribe(HandleDifficultyToggle)
            DifficultyInfo_Button.OnClick().Subscribe(HandleDifficultyInfoToggle)
            PatchNoteButton.OnClick().Subscribe(TogglePatchNotes)
            CreditsButton.OnClick().Subscribe(ToggleCredits)
            AdminPanelButton.OnClick().Subscribe(OpenAdminPanel)

    var User : ?agent = false
    Create(InAgent:agent):canvas_slot=
        set User = option. InAgent
        SubscribeButtons()
        ReadyUpButton.SetText(StringToMessage("Ready Up"))
        GameDifficultyString := GetGameDifficulty().ToString()
        DifficultyButton.SetText(StringToMessage(GameDifficultyString))
        DifficultyInfo_Button.SetText(StringToMessage("INFO"))
        PatchNoteButton.SetText(StringToMessage("Patch Notes"))
        CreditsButton.SetText(StringToMessage("Credits"))
        AdminPanelButton.SetText(StringToMessage("Admin Panel"))
        <#
        set PatchNotesOverlay = overlay{
            Slots:=array:
                overlay_slot:
                    Widget:=texture_block{DefaultImage:=Art.UIStuff.Fonts.InteractionBackground.InteractionBackground, DefaultDesiredSize:=vector2{X:=740.0,Y:=50.0}, DefaultTint:=NamedColors.Black}
                    HorizontalAlignment:=horizontal_alignment.Fill
                    VerticalAlignment:=vertical_alignment.Top
                overlay_slot:
                    Widget:=PatchNoteVersion
                    HorizontalAlignment:=horizontal_alignment.Center
                    VerticalAlignment:=vertical_alignment.Top
                    Padding:=margin{Top:=5.0}
                overlay_slot:
                    Widget:=overlay{
                        Slots:=array:
                            overlay_slot:
                                Widget:=texture_block{DefaultImage:=Art.UIStuff.Fonts.InteractionBackground.InteractionBackground, DefaultDesiredSize:=vector2{X:=740.0,Y:=150.0}}
                                HorizontalAlignment:=horizontal_alignment.Fill
                                VerticalAlignment:=vertical_alignment.Fill
                                Padding:=margin{Top:=50.0}
                            overlay_slot:
                                Widget:=stack_box{
                                    Orientation:=orientation.Vertical
                                    Slots:=array:
                                        stack_box_slot:
                                            Widget:=PatchNotesText
                                            Padding:=margin{Top:=55.0,Bottom:=20.0}
                                            HorizontalAlignment:=horizontal_alignment.Fill
                                            VerticalAlignment:=vertical_alignment.Fill
                                }
                                HorizontalAlignment:=horizontal_alignment.Center
                                VerticalAlignment:=vertical_alignment.Center
                                Padding:=margin{Left:=20.0,Right:=20.0}
                    }
                    HorizontalAlignment:=horizontal_alignment.Left
                    VerticalAlignment:=vertical_alignment.Top
        }
                    #>
        #PatchNotesOverlay.Hide()
        <#
        set CreditsOverlay = overlay{
            Slots:=array:
                overlay_slot:
                    Widget:=texture_block{DefaultImage:=Art.UIStuff.Fonts.InteractionBackground.InteractionBackground, DefaultDesiredSize:=vector2{X:=740.0,Y:=50.0}, DefaultTint:=NamedColors.Black}
                    HorizontalAlignment:=horizontal_alignment.Fill
                    VerticalAlignment:=vertical_alignment.Top
                overlay_slot:
                    Widget:=text_block{DefaultText:=StringToMessage("Credits"), DefaultTextColor:=NamedColors.White}
                    HorizontalAlignment:=horizontal_alignment.Center
                    VerticalAlignment:=vertical_alignment.Top
                    Padding:=margin{Top:=5.0}
                overlay_slot:
                    Widget:=overlay{
                        Slots:=array:
                            overlay_slot:
                                Widget:=texture_block{DefaultImage:=Art.UIStuff.Fonts.InteractionBackground.InteractionBackground, DefaultDesiredSize:=vector2{X:=740.0,Y:=150.0}}
                                HorizontalAlignment:=horizontal_alignment.Fill
                                VerticalAlignment:=vertical_alignment.Fill
                                Padding:=margin{Top:=50.0}
                            overlay_slot:
                                Widget:=stack_box{
                                    Orientation:=orientation.Vertical
                                    Slots:=array:
                                        stack_box_slot:
                                            Widget:=CreditsText
                                            Padding:=margin{Top:=55.0,Bottom:=20.0}
                                            HorizontalAlignment:=horizontal_alignment.Fill
                                            VerticalAlignment:=vertical_alignment.Fill
                                }
                                HorizontalAlignment:=horizontal_alignment.Center
                                VerticalAlignment:=vertical_alignment.Center
                                Padding:=margin{Left:=20.0,Right:=20.0}
                    }
                    HorizontalAlignment:=horizontal_alignment.Left
                    VerticalAlignment:=vertical_alignment.Top
        }#>
        #CreditsOverlay.Hide()
        if(ThisPlayerManager:=GlobalPlayerManager[GetSession()]):
            ThisPlayerManager.OnVoteReadyToggle(InAgent, false)
            if(not ThisPlayerManager.CountdownState?):
                DifficultyButton.SetEnabled(true)
        SetReadyupIconState(InAgent, false)
        defer{
            PlayerList:=GetAgents(?OnlyActives:=false)
            for(ThisPlayer:PlayerList):
                if(ThisPlayer.GetPlayerPackage[]):
                    AddPlayerSlot(ThisPlayer)
                ThisPlayer.GetUIManager().MainMenuUI.AddPlayerSlot(InAgent)
        }
        #DiffInfoIcon := texture_block{DefaultImage:=Art.UIStuff.Icons.Notification_Icon, DefaultDesiredSize:=vector2{X:=98.0,Y:=98.0}}
        #DiffInfoIcon.HideNoCollapse()
        <#
        DifficultyInfo_Overlay := overlay{
            Slots:=array:
                overlay_slot{
                    Widget:=DiffInfoIcon
                    HorizontalAlignment:=horizontal_alignment.Center
                    VerticalAlignment:=vertical_alignment.Center
                }
                overlay_slot{
                    Widget:=DifficultyInfo_Button
                    HorizontalAlignment:=horizontal_alignment.Fill
                    VerticalAlignment:=vertical_alignment.Fill
                }
        } #>
        set InfoOverlay = overlay{
            Slots:=array:
                overlay_slot:
                    Widget:=texture_block{DefaultImage:=Art.UIStuff.Fonts.InteractionBackground.InteractionBackground, DefaultDesiredSize:=vector2{X:=740.0,Y:=50.0}, DefaultTint:=NamedColors.Black}
                    HorizontalAlignment:=horizontal_alignment.Fill
                    VerticalAlignment:=vertical_alignment.Top
                overlay_slot:
                    Widget:=Info_Title
                    HorizontalAlignment:=horizontal_alignment.Center
                    VerticalAlignment:=vertical_alignment.Top
                    Padding:=margin{Top:=5.0}
                overlay_slot:
                    Widget:=overlay{
                        Slots:=array:
                            overlay_slot:
                                Widget:=texture_block{DefaultImage:=Art.UIStuff.Fonts.InteractionBackground.InteractionBackground, DefaultDesiredSize:=vector2{X:=740.0,Y:=150.0}}
                                HorizontalAlignment:=horizontal_alignment.Fill
                                VerticalAlignment:=vertical_alignment.Fill
                                Padding:=margin{Top:=50.0}
                            overlay_slot:
                                Widget:=stack_box{
                                    Orientation:=orientation.Vertical
                                    Slots:=array:
                                        stack_box_slot:
                                            Widget:=Info_Description
                                            Padding:=margin{Top:=55.0,Bottom:=20.0}
                                            HorizontalAlignment:=horizontal_alignment.Fill
                                            VerticalAlignment:=vertical_alignment.Fill
                                }
                                HorizontalAlignment:=horizontal_alignment.Center
                                VerticalAlignment:=vertical_alignment.Center
                                Padding:=margin{Left:=20.0,Right:=20.0}
                    }
                    HorizontalAlignment:=horizontal_alignment.Left
                    VerticalAlignment:=vertical_alignment.Top
        }
        HideInfoOverlay()
        if(not GlobalAdminState?):
            AdminPanelButton.SetEnabled(false)
            AdminPanelButton.Collapse()
        DifficultyInfo_Control.ChangeSize(vector2{X:=100.0,Y:=100.0})
        return canvas_slot{
            Widget:=canvas{
                Slots:=array:
                    <#
                    canvas_slot{
                        Widget:=ReadyUpTexture
                        #Anchors:=anchors{Minimum:=vector2{Y:=1.0}, Maximum:=vector2{Y:=1.0}}
                        Offsets:=margin{Left:=40.0, Top:=808.0}
                        SizeToContent:=true
                        ZOrder:=1
                    } #>
                    canvas_slot{
                        Widget:=InfoOverlay
                        Anchors:=anchors{Minimum:=vector2{X:=0.5, Y:=0.168889},Maximum:=vector2{X:=0.5, Y:=0.168889}}
                        Alignment:=vector2{X:=0.5, Y:=-0.127874}
                        SizeToContent:=true
                    }
                    canvas_slot{
                        Widget:=DifficultyInfo_Button
                        Offsets:=margin{Left:=820.0, Top:=892.0, Bottom:=100.0, Right:=130.0}
                        SizeToContent:=false
                    }
                    canvas_slot{
                        Widget:=text_block{DefaultText:=StringToMessage("Please LIKE and FAVORITE the game to see more content and maps!"), DefaultTextColor:=NamedColors.Yellow, DefaultShadowOffset:=option{vector2{X:=5.0,Y:=5.0}}, DefaultShadowColor:=NamedColors.Black, DefaultShadowOpacity:= 1.0}
                        Offsets:=margin{Left:=444.0,Top:=1024.0}
                    }
                    canvas_slot{
                        Widget:=texture_block{DefaultImage:=Art.UIStuff.MainMenu.Socials, DefaultDesiredSize:=vector2{X:=724.0,Y:=382.0}}
                        Offsets:=margin{Left:=-36.0,Top:=-96.0}
                        SizeToContent:=true
                    }
                    canvas_slot{
                        Widget:=ReadyUpCount
                        Offsets:=margin{Left:=212.0, Top:=800.0}
                        SizeToContent:=true
                    }
                    canvas_slot{
                        Widget:=CountdownText
                        Offsets:=margin{Left:=144.0, Top:=800.0}
                        SizeToContent:=true
                    }
                    canvas_slot{
                        Widget:=AutostartCountdownText
                        Offsets:=margin{Left:=144.0,Top:=752.0}
                        SizeToContent:=true
                    }
                    canvas_slot{
                        Widget:=ReadyUpButton
                        #Anchors:=anchors{Minimum:=vector2{Y:=1.0}, Maximum:=vector2{Y:=1.0}}
                        Offsets:=margin{Left:=73.164673, Top:=868.0, Right:=435.0, Bottom:=135.546936}
                        SizeToContent:=false
                        ZOrder:=2
                    } 
                    canvas_slot{
                        Widget:=text_block{DefaultText:=StringToMessage("Difficulty"), DefaultTextColor:=NamedColors.White, DefaultShadowOffset:=option{vector2{X:=5.0,Y:=5.0}}, DefaultShadowColor:=NamedColors.Black, DefaultShadowOpacity:= 1.0, DefaultJustification:=text_justification.Center}
                        Offsets:=margin{Left:=588.0,Top:=840.0}
                        SizeToContent:=true
                    }
                    canvas_slot{
                        Widget:=DifficultyButton
                        Offsets:=margin{Left:=536.0, Top:=908.0, Right:=248.813812, Bottom:=55.066452}
                        SizeToContent:=false
                        ZOrder:=2
                    }
                    canvas_slot{
                        Widget:=LobbyScoreboardStackbox
                        #Anchors:=anchors{Minimum:=vector2{X:=1.0}, Maximum:=vector2{X:=1.0}}
                        Offsets:=margin{Left:=1269.921875, Top:=140.0}
                        SizeToContent:=true
                        ZOrder:=1
                    }
                    <#
                    canvas_slot{
                        Widget:=PatchNotesOverlay
                        Offsets:=margin{Top:=180.0, Left:=592.0}
                        SizeToContent:=true
                    }
                        #>
                    canvas_slot{
                        Widget:=PatchNoteButton
                        Offsets:=margin{Left:=1580.0, Top:=948.0, Right:=280.0, Bottom:=56.0}
                        SizeToContent:=false
                        ZOrder:=2
                    }
                    <#
                    canvas_slot{
                        Widget:=CreditsOverlay
                        Offsets:=margin{Top:=180.0, Left:=592.0}
                        SizeToContent:=true
                    }
                        #>
                    canvas_slot{
                        Widget:=CreditsButton
                        Offsets:=margin{Left:=1646.0, Top:=850.0, Right:=212.0, Bottom:=56.0}
                        SizeToContent:=false
                        ZOrder:=2
                    }
                    canvas_slot{
                        Widget:=AdminPanelButton
                        Offsets:=margin{Left:=1568.0,Top:=752.0,Right:=292.260193,Bottom:=56.0}
                        SizeToContent:=false
                    }
                    canvas_slot{
                        Widget:=texture_block{DefaultImage:=Art.UIStuff.MainMenu.Hurtgen_Forest_Logo, DefaultDesiredSize:=vector2{X:=600.0,Y:=400.0}}
                        Offsets:=margin{Left:=0.0, Top:=-84.0}
                        SizeToContent:=true
                    }
            }
            Offsets:=margin{Top:=0.0,Left:=0.0}
        }

        <#
        if(AllowAdmin[InAgent]):
            AdminPanelButton.Show()
        else:
            AdminPanelButton.SetEnabled(false)
            AdminPanelButton.Collapse()
        #>

    var PlayerToSlot : [agent]?menu_slot = map{}
    AddPlayerSlot(InAgent:agent):void=
        if(not PlayerToSlot[InAgent]?):
            var PlayerNameColor : color = NamedColors.White
            if(ThisPackage:=InAgent.GetPlayerPackage[]):
                set PlayerNameColor = ThisPackage.Color
            ReadyUpDisplay:=overlay{
                Slots:=array:
                    overlay_slot{
                        Widget:=texture_block{DefaultImage:=Art.UIStuff.Fonts.InteractionBackground.InteractionBackground, DefaultDesiredSize:=vector2{X:=50.0,Y:=50.0}}
                    }
                    overlay_slot{
                        Widget:=texture_block{DefaultImage:=Art.UIStuff.MainMenu.Checkmark_Icon, DefaultDesiredSize:=vector2{X:=40.0,Y:=40.0}, DefaultTint:=NamedColors.Green}
                        HorizontalAlignment:=horizontal_alignment.Center
                        VerticalAlignment:=vertical_alignment.Center
                    }
            }
            NewExfilTrophyIcon : texture_block = texture_block{DefaultImage:=Art.UIStuff.Icons.Exfil_Trophy_Beta, DefaultDesiredSize:=vector2{X:=50.0,Y:=50.0}}
            if(ThisExfilManager:=GlobalExfilManager[GetSession()]):
                var CompletedExfil : logic = false
                if(ThisExfilManager.ExfilCompletedSaveSwitch_ClassicMode.GetCurrentState[InAgent]):
                    set CompletedExfil = true
                if(ThisExfilManager.ExfilCompletedSaveSwitch_HardcoreMode.GetCurrentState[InAgent]):
                    NewExfilTrophyIcon.SetImage(Art.UIStuff.Icons.Exfil_Trophy)
                    set CompletedExfil = true
                if(not CompletedExfil?):
                    NewExfilTrophyIcon.Hide()
            spawn. FlashTrophy(InAgent, NewExfilTrophyIcon)
            NewPlayerSlotWidget : stack_box_slot = stack_box_slot{
                Widget:=overlay{
                    Slots:=array:
                        overlay_slot{
                            Widget:=texture_block{DefaultImage:=Art.UIStuff.Fonts.InteractionBackground.InteractionBackground, DefaultDesiredSize:=vector2{X:=500.0,Y:=50.0}, DefaultTint:=NamedColors.Black}
                            Padding:=margin{Bottom:=5.0}
                            VerticalAlignment:=vertical_alignment.Top
                        }
                        overlay_slot{
                            Widget:=text_block{DefaultText:=PlayerToMessage(InAgent), DefaultTextColor:=PlayerNameColor}
                            HorizontalAlignment:=horizontal_alignment.Center
                            VerticalAlignment:=vertical_alignment.Center
                        }
                        overlay_slot{
                            Widget:=ReadyUpDisplay
                            Padding:=margin{Left:=-50.0}
                            HorizontalAlignment:=horizontal_alignment.Left
                            VerticalAlignment:=vertical_alignment.Top
                        }
                        overlay_slot{
                            Widget:=NewExfilTrophyIcon
                            Padding:=margin{Right:=-60.0}
                            HorizontalAlignment:=horizontal_alignment.Right
                            VerticalAlignment:=vertical_alignment.Center
                        }
                }
            }
            ReadyUpDisplay.Hide()
            NewPlayerSlot : menu_slot = menu_slot{
                PlayerSlotWidget:=NewPlayerSlotWidget.Widget
                ReadyUpWidget:=ReadyUpDisplay
                ExfilTrophyIcon:=NewExfilTrophyIcon
            }
            if. set PlayerToSlot[InAgent] = option{NewPlayerSlot}
            else:
                set PlayerToSlot = ConcatenateMaps(PlayerToSlot, map{InAgent => option{NewPlayerSlot}})
            LobbyScoreboardStackbox.AddWidget(NewPlayerSlotWidget)

    FlashTrophy<private>(InAgent:agent, TextureBlock:texture_block)<suspends>:void=
        if(GlobalExfilManager[GetSession()].ExfilCompletedSaveSwitch_UnderOneHour.GetCurrentState[InAgent], ThisPlayerInfo:=InAgent.GetPlayerInfo[]):
            race:
                ThisPlayerInfo.LeaveEvent.Await()
                AwaitGameState(game_state.InGame)
                loop:
                    for(ThisColor:GoldWaveArray):
                        TextureBlock.SetTint(ThisColor)
                        Sleep(0.1)
                    Sleep(0.0)

    var<private> InfoOverlayType : int = -1
    SetInfoOverlay(Type:int, TitleDescTuple:tuple(string,string)):void=
        LobbyScoreboardStackbox.Hide()
        DifficultyInfo_Button.SetEnabled(false)
        PatchNoteButton.SetEnabled(false)
        CreditsButton.SetEnabled(false)
        case(Type):
            0 => DifficultyInfo_Button.SetEnabled(true)
            1 => PatchNoteButton.SetEnabled(true)
            2 => CreditsButton.SetEnabled(true)
            _ =>
        set InfoOverlayType = Type
        InfoOverlay.Show()
        Info_Title.SetText(StringToMessage(TitleDescTuple(0)))
        Info_Description.SetText(StringToMessage(TitleDescTuple(1)))

    HideInfoOverlay():void=
        LobbyScoreboardStackbox.Show()
        InfoOverlay.Hide()
        DifficultyInfo_Button.SetEnabled(true)
        PatchNoteButton.SetEnabled(true)
        CreditsButton.SetEnabled(true)

    RemovePlayerSlot<public>(InAgent:agent):void=
        if(ThisPlayerSlotWidget:=PlayerToSlot[InAgent]?.PlayerSlotWidget):
            LobbyScoreboardStackbox.RemoveWidget(ThisPlayerSlotWidget)
        else{Print("ERR: Failed to find and remove a player from the lobby scoreboard!")}

    HandleReadyupClick(MSG:widget_message):void=
        if(ThisPlayerManager:=GlobalPlayerManager[GetSession()]):
            var ReadyState : logic = false
            if(ThisPlayerInfo:=MSG.Player.GetPlayerInfo[]):
                if(ThisPlayerInfo.ReadyUpState?):
                    set ThisPlayerInfo.ReadyUpState = false
                    MSG.Player.GetUIManager().MainMenuUI.SetReadyupIconState(MSG.Player, false)
                    ReadyUpButton.SetText(StringToMessage("Ready Up"))
                    DifficultyButton.SetEnabled(true)
                else:
                    set ThisPlayerInfo.ReadyUpState = true
                    MSG.Player.GetUIManager().MainMenuUI.SetReadyupIconState(MSG.Player, true)
                    set ReadyState = true
                    ReadyUpButton.SetText(StringToMessage("Cancel"))
                    DifficultyButton.SetEnabled(false)
            ThisPlayerManager.OnVoteReadyToggle(MSG.Player, ReadyState)

    HandleDifficultyToggle(MSG:widget_message):void=
        spawn. DifficultyToggleSleep()
        #spawn. HandleDifficultyInfoBounce()
        ToggleGameDifficulty()
        NewGameDifficulty:=GetGameDifficulty()
        PlayerList:=GetAgents(?OnlyActives:=false)
        for(ThisPlayer:PlayerList):
            ThisPlayer.GetUIManager().MainMenuUI.DifficultyButton.SetText(StringToMessage(NewGameDifficulty.ToString()))

    DifficultyToggleSleep()<suspends>:void=
        PlayerList:=GetAgents(?OnlyActives:=false)
        for(ThisPlayer:PlayerList):
            ThisPlayer.GetUIManager().MainMenuUI.DifficultyButton.SetEnabled(false)
        Sleep(0.5)
        if(not IsCountdown[]):
            for(ThisPlayer:PlayerList):
                ThisPlayer.GetUIManager().MainMenuUI.DifficultyButton.SetEnabled(true)

    <#
    DiffInfoBounceEvent:event()=event(){}
    HandleDifficultyInfoBounce()<suspends>:void=
        DifficultyInfo_Control.ChangePosition(vector2{})
        DiffInfoBounceEvent.Signal()
        race:
            DiffInfoBounceEvent.Await()
            block:
                DifficultyInfo_Control.MoveSize(vector2{X:=125.0,Y:=125.0}, 0.35, ?InterpolationType:=interpolation_type.EaseOutBack)
                DifficultyInfo_Control.MoveSize(vector2{X:=100.0,Y:=100.0}, 0.3, ?InterpolationType:=interpolation_type.EaseOutBounce)
        DifficultyInfo_Control.ChangePosition(vector2{})#>

    var<private> DifficultyInfoState : logic = false
    ExitDifficultyInfo : event() = event(){}
    HandleDifficultyInfoToggle(MSG:widget_message):void=
        if(DifficultyInfoState?):
            HideInfoOverlay()
            DifficultyInfo_Button.SetText(StringToMessage("INFO"))
            ExitDifficultyInfo.Signal()
            set DifficultyInfoState = false
        else:
            DifficultyInfo_Button.SetText(StringToMessage("Exit"))
            SetInfoOverlay(0, ("Difficulty stats", "Nothing to see here..."))
            SetDifficultyInfo(GetGameDifficulty())
            spawn. TrackDifficultyToggle()
            set DifficultyInfoState = true

    TrackDifficultyToggle()<suspends>:void=
        race:
            ExitDifficultyInfo.Await()
            block:
                if(DifficultyChangedEvent:=GetGameDifficultyChangedEvent[]):
                    loop:
                        NewDifficulty:=DifficultyChangedEvent.Await()
                        SetDifficultyInfo(NewDifficulty)
                Sleep(Inf)

    SetDifficultyInfo(Difficulty:game_difficulty):void=
        case(Difficulty):
            game_difficulty.Easy => SetInfoOverlay(0, ("Easy Difficulty", "New here? Then Easy mode is for you! \nLearn the map  and game without having to deal with \nvery aggressive zombies early on, as well as recieving \n more points from zombies."))
            game_difficulty.Classic => SetInfoOverlay(0, ("Classic Difficulty", "The classic mode is the go-to mode for anyone looking \nfor a casual, but exciting zombie-slaying session. \nMight be a bit difficult for newbies :P"))
            game_difficulty.Hardcore => SetInfoOverlay(0, ("Hardcore Difficulty", "The Hardcore mode is reserved for the best of players. \nBeware of aggressive zombies early on, as well \nas slower health regeneration! A 4 Perk limit \nalso is in effect, so be careful of which perks \nyou choose to buy!"))

    SetReadyupIconState(InAgent:agent, InState:logic, ?GlobalState:logic=false):void=
        if(ThisPlayerInfo:=InAgent.GetPlayerInfo[]):
            set ThisPlayerInfo.ReadyUpState = InState
        var ThisReadyUpCount : int = 0
        if(ThisReadyupIcon:=PlayerToSlot[InAgent]?.ReadyUpWidget):
            if(InState?):
                set ThisReadyUpCount = 1
                ThisReadyupIcon.Show()
            else:
                ThisReadyupIcon.Hide()
        Players:=GetPlayers(?OnlyActives:=false)
        for(ThisPlayer:Players):
            if(not GlobalState?, ThisPlayer.User<>InAgent):
                ThisPlayer.User.GetUIManager().MainMenuUI.SetReadyupIconState(InAgent, InState, ?GlobalState:=true)
                if(ThisPlayer.ReadyUpState?):
                    set ThisReadyUpCount += 1
        if(not GlobalPlayerManager[GetSession()].CountdownState?):
            ReadyUpCount.SetText(StringToMessage("{ThisReadyUpCount}/{Players.Length} Ready"))
        
    CancelCountdownEvent<public>:event()=event(){}
    StartCountdown<public>()<suspends>:void=
        ReadyUpCount.SetText(StringToMessage(""))
        var CurrSec : int = if. IsSinglePlayer[] then. 3 else. 8
        race:
            CancelCountdownEvent.Await()
            block:
                Sleep(0.5)
                loop:
                    CountdownText.SetText(StringToMessage("Game starting in {CurrSec}"))
                    Sleep(1.0)
                    if(CurrSec <= 0):
                        break
                    set CurrSec -= 1
        CountdownText.SetText(StringToMessage(""))
        ReadyUpButton.SetText(StringToMessage("Ready Up"))
        if(ThisOwner:=User?):
            SetReadyupIconState(ThisOwner, false)

    GetPatchNotes():tuple(string,string)=
        var PatchVersion : string = "Patch notes - ???"
        var PatchNoteString : string = "Nothing to see here!"
        if(ThisPlayerManager:=GlobalPlayerManager[GetSession()]):
            set PatchVersion = ThisPlayerManager.PatchVersion
            if(ThisPlayerManager.PatchNotes.Length > 0):
                set PatchNoteString = ""
                var NoteIndex : int = 0
                for(ThisPatchNote:ThisPlayerManager.PatchNotes):
                    set PatchNoteString += ThisPatchNote
                    if(NoteIndex < ThisPlayerManager.PatchNotes.Length-1):
                        set PatchNoteString += "\n"
                    set NoteIndex += 1
        return (PatchVersion, PatchNoteString)

    GetCredits():tuple(string,string)=
        var CreditsString : string = "'Ello there"
        if(ThisPlayerManager:=GlobalPlayerManager[GetSession()], ThisPlayerManager.Credits.Length > 0):
            set CreditsString = ""
            var CreditIndex : int = 0
            for(ThisCredit:ThisPlayerManager.Credits):
                set CreditsString += ThisCredit
                if(CreditIndex < ThisPlayerManager.Credits.Length-1):
                    set CreditsString += "\n"
                set CreditIndex += 1
        return ("Credits", CreditsString)

    var PatchNotesVisibility<private> : logic = false
    TogglePatchNotes(MSG:widget_message):void=
        if(PatchNotesVisibility?):
            HideInfoOverlay()
            PatchNoteButton.SetText(StringToMessage("Patch Notes"))
            set PatchNotesVisibility = false
        else:
            SetInfoOverlay(1, GetPatchNotes())
            PatchNoteButton.SetText(StringToMessage("Exit"))
            set PatchNotesVisibility = true
        
    var CreditsVisibility<private> : logic = false
    ToggleCredits(MSG:widget_message):void=
        if(CreditsVisibility?):
            HideInfoOverlay()
            CreditsButton.SetText(StringToMessage("Credits"))
            set CreditsVisibility = false
        else if(not PatchNotesVisibility?):
            SetInfoOverlay(2, GetCredits())
            CreditsButton.SetText(StringToMessage("Exit"))
            set CreditsVisibility = true

    OpenAdminPanel(MSG:widget_message):void=
        MSG.Player.GetUIManager().AdminPanel_Menu.Show()

# Holds data/widgets about the menu slot of a player #
menu_slot := class:

    PlayerSlotWidget : widget
    ReadyUpWidget : widget

    ExfilTrophyIcon : texture_block

parts_ui<public> := class:

    UIManager : ?ui_manager = false
    BuildableStorage : buildable_storage = buildable_storage{}

    var Stackbox : stack_box = stack_box{Orientation:=orientation.Horizontal}

    var ShieldPlate_PartUIControl :overlay=overlay{}
    var Cable_PartUIControl :overlay=overlay{}
    var Crystal_PartUIControl :overlay=overlay{}

    ShieldText : string_block = string_block{Font:="Quest"}
    ShieldPlate_PartIcon:texture_block=texture_block{DefaultImage:=Art.UIStuff.Icons.Part_Icons.ShieldPlate_Icon,DefaultDesiredSize:=vector2{X:=100.0,Y:=100.0}}
    Crystal_PartIcon:texture_block=texture_block{DefaultImage:=Art.UIStuff.Icons.Part_Icons.Crystal_Icon,DefaultDesiredSize:=vector2{X:=100.0,Y:=100.0}}
    Cable_PartIcon:texture_block=texture_block{DefaultImage:=Art.UIStuff.Icons.Part_Icons.Cable_Icon,DefaultDesiredSize:=vector2{X:=100.0,Y:=100.0}}

    PickupPartWidgetControl : dynamic_widget = dynamic_widget{}
    PickupPartText : string_block = string_block{Font:="Quest"}
    var PickupPart_Overlay : overlay=overlay{}
    PickupPart_Icon : texture_block = texture_block{DefaultImage:=Art.UIStuff.Icons.Part_Icons.ShieldPlate_Icon, DefaultDesiredSize:=vector2{X:=100.0,Y:=100.0}}

    Create():[]canvas_slot=
        ShieldTexureBlock_Shadow : texture_block = texture_block{DefaultImage:=Art.UIStuff.Icons.Part_Icons.ShieldPlate_Icon,DefaultDesiredSize:=vector2{X:=100.0,Y:=100.0}}
        CrystalTextureBlock_Shadow : texture_block = texture_block{DefaultImage:=Art.UIStuff.Icons.Part_Icons.Crystal_Icon,DefaultDesiredSize:=vector2{X:=100.0,Y:=100.0}}
        CableTextureBlock_Shadow : texture_block = texture_block{DefaultImage:=Art.UIStuff.Icons.Part_Icons.Cable_Icon,DefaultDesiredSize:=vector2{X:=100.0,Y:=100.0}}
        ShieldTexureBlock_Shadow.SetTint(NamedColors.Black)
        CrystalTextureBlock_Shadow.SetTint(NamedColors.Black)
        CableTextureBlock_Shadow.SetTint(NamedColors.Black)
        Part_ShieldPlateOverlay : overlay = overlay{
            Slots:=array:
                overlay_slot:
                    Padding:=margin{Right:=-80.0}
                    Widget:=texture_block{DefaultImage:=Art.UIStuff.Icons.Part_Icons.Part_Background_Icon,DefaultDesiredSize:=vector2{X:=250.0,Y:=250.0}}
                overlay_slot:
                    Padding:=margin{Right:=-44.0}
                    #Padding:=margin{Right:=-50.0,Bottom:=-4.0}
                    Widget:=ShieldTexureBlock_Shadow
                    HorizontalAlignment:=horizontal_alignment.Right
                    VerticalAlignment:=vertical_alignment.Center
                overlay_slot:
                    Padding:=margin{Right:=-44.0}
                    Widget:=ShieldPlate_PartIcon
                    HorizontalAlignment:=horizontal_alignment.Right
                    VerticalAlignment:=vertical_alignment.Center
        }
        set ShieldPlate_PartUIControl = Part_ShieldPlateOverlay
        Part_CrystalOverlay : overlay = overlay{
            Slots:=array:
                overlay_slot:
                    Padding:=margin{Right:=-80.0}
                    Widget:=texture_block{DefaultImage:=Art.UIStuff.Icons.Part_Icons.Part_Background_Icon,DefaultDesiredSize:=vector2{X:=250.0,Y:=250.0}}
                overlay_slot:
                    Padding:=margin{Right:=-44.0}
                    #Padding:=margin{Right:=-50.0,Bottom:=-4.0}
                    Widget:=CrystalTextureBlock_Shadow
                    HorizontalAlignment:=horizontal_alignment.Right
                    VerticalAlignment:=vertical_alignment.Center
                overlay_slot:
                    Padding:=margin{Right:=-44.0}
                    Widget:=Crystal_PartIcon
                    HorizontalAlignment:=horizontal_alignment.Right
                    VerticalAlignment:=vertical_alignment.Center
        }
        set Crystal_PartUIControl = Part_CrystalOverlay
        Part_CableOverlay : overlay = overlay{
            Slots:=array:
                overlay_slot:
                    Padding:=margin{Right:=-80.0}
                    Widget:=texture_block{DefaultImage:=Art.UIStuff.Icons.Part_Icons.Part_Background_Icon,DefaultDesiredSize:=vector2{X:=250.0,Y:=250.0}}
                overlay_slot:
                    Padding:=margin{Right:=-44.0}
                    #Padding:=margin{Right:=-50.0,Bottom:=-4.0}
                    Widget:=CableTextureBlock_Shadow
                    HorizontalAlignment:=horizontal_alignment.Right
                    VerticalAlignment:=vertical_alignment.Center
                overlay_slot:
                    Padding:=margin{Right:=-44.0}
                    Widget:=Cable_PartIcon
                    HorizontalAlignment:=horizontal_alignment.Right
                    VerticalAlignment:=vertical_alignment.Center
        }
        set Cable_PartUIControl = Part_CableOverlay
        NewPickupPartOverlay : overlay = overlay{
            Slots:=array:
                overlay_slot:
                    Padding:=margin{Left:=225.0}
                    Widget:=texture_block{DefaultImage:=Art.UIStuff.Icons.Part_Icons.Part_Background_Icon, DefaultDesiredSize:=vector2{X:=250.0,Y:=250.0}}
                overlay_slot:
                    Widget:=PickupPartText.Create("Part Found!",vector2{}, 32.0, ?SizeRatio:=vector2{X:=1.0,Y:=1.5}).Widget
                overlay_slot:
                    Padding:=margin{Left:=225.0}
                    Widget:=PickupPart_Icon
        }
        set PickupPart_Overlay = NewPickupPartOverlay
        set Stackbox = stack_box{
            Orientation:=orientation.Horizontal
            Slots:=array:
                stack_box_slot{
                    Widget:=Part_ShieldPlateOverlay
                    HorizontalAlignment:=horizontal_alignment.Fill
                    VerticalAlignment:=vertical_alignment.Fill
                }
                stack_box_slot{
                    Widget:=Part_CrystalOverlay
                    HorizontalAlignment:=horizontal_alignment.Fill
                    VerticalAlignment:=vertical_alignment.Fill
                }
                stack_box_slot{
                    Widget:=Part_CableOverlay
                    HorizontalAlignment:=horizontal_alignment.Fill
                    VerticalAlignment:=vertical_alignment.Fill
                }
        }
        # vector2{X:=1500.0,Y:=640.0} 688
        # vector2{X:=2000.0.0,Y:=640.0}
        ShieldTextSlot:=ShieldText.Create("shield parts", vector2{X:=960.0,Y:=704.0}, 30.0, ?Alignment:=vector2{X:=0.5,Y:=0.5}, ?SizeRatio:=vector2{X:=1.0,Y:=1.5}, ?AnchorOverride:=option{Anchors:=anchors{Minimum:=vector2{X:=0.5,Y:=0.615},Maximum:=vector2{X:=0.5,Y:=0.615}}})
        #PickupPartIconSlot:=PickupPartWidgetControl.Create(PickupPart_Overlay, vector2{X:=2000.0,Y:=640.0}, vector2{X:=250.0,Y:=250.0})
        PickupPartIconSlot:=PickupPartWidgetControl.Create(PickupPart_Overlay, vector2{X:=2200.937744,Y:=660.727051}, ?Alignment:=vector2{X:=1.0, Y:=0.5}, ?AutoAnchor:=true)
        # 2000.0 , 640.0
        #PickupPartWidgetControl.Hide()
        
        ReturningCanvasSlots:[]canvas_slot=array{
            PickupPartIconSlot
            ShieldTextSlot
            canvas_slot{
                Widget:=Stackbox
                SizeToContent:=true
                Anchors:=anchors{
                    Minimum:=vector2{X:=0.5,Y:=0.615}
                    Maximum:=vector2{X:=0.5,Y:=0.615}
                }
                Alignment:=vector2{X:=0.656236}
            }
        }
        Hide()
        return ReturningCanvasSlots

    Hide<public>():void=
        Stackbox.Hide()
        ShieldText.Hide()
        ShieldPlate_PartIcon.Hide()
        Cable_PartIcon.Hide()
        Crystal_PartIcon.Hide()
        PickupPartWidgetControl.Hide()
        if(ThisUI:=UIManager?):
            ThisUI.InteractionText.Show()
            ThisUI.DescriptionUI.Show()

    Show<public>():void=
        Stackbox.Show()
        ShieldText.Show()
        if(ThisUI:=UIManager?):
            ThisUI.InteractionText.Hide()
            ThisUI.DescriptionUI.Hide(?ClearPrevious:=false)
            GlobalStorages:=BuildableStorage.GetGlobalStorages()
            for(ThisStorage:GlobalStorages):
                PartList:=ThisStorage.GetAllParts()
                for(ThisPart:PartList):
                    case(ThisPart.PartIndex):
                        0=> ShieldPlate_PartIcon.Show()
                        1=> Cable_PartIcon.Show()
                        2=> Crystal_PartIcon.Show()
                        _=> Print("~Report: Extra part type detected!")
        else{Print("~ERR: Could not find UI manager in parts_ui")}

    var QueuedPartPickups : []color_block = array{}
    OnPartPickedUp<public>(InPart:?buildable_part, ?GloballyActivated:logic=false)<suspends>:void=
        TextureBlockHold : color_block = color_block{}
        set QueuedPartPickups+=array{TextureBlockHold}
        var MaybeTexture : ?texture = false
        if(ThisPart:=InPart?):
            if(not GloballyActivated?, ThisPart.ParentBuildable.AllowSharedCrafting?):
                if(ThisOwner:=UIManager?.User):
                    Players:=GetAgents()
                    for(ThisPlayer:Players):
                        if(ThisPlayer<>ThisOwner):
                            spawn{ThisPlayer.GetUIManager().PartsUI.OnPartPickedUp(option. ThisPart, ?GloballyActivated:=true)}
            case(ThisPart.PartIndex):
                0=> 
                    ShieldTexture:=ShieldPlate_PartIcon.GetImage()
                    set MaybeTexture=option{ShieldTexture}
                1=> 
                    CableTexture:=Cable_PartIcon.GetImage()
                    set MaybeTexture=option{CableTexture}
                2=> 
                    CrystalTexture:=Crystal_PartIcon.GetImage()
                    set MaybeTexture=option{CrystalTexture}
                _=> Print("ERR: Unknown part index !")
        else:
            set MaybeTexture = option. Art.UIStuff.Icons.TruckKey_Icon
            if(not GloballyActivated?):
                if(ThisOwner:=UIManager?.User):
                    Players:=GetAgents()
                    for(ThisPlayer:Players):
                        if(ThisPlayer<>ThisOwner):
                            spawn{ThisPlayer.GetUIManager().PartsUI.OnPartPickedUp(false, ?GloballyActivated:=true)}
        if(ThisTexture:=MaybeTexture?):
            loop:
                if(QueuedPartPickups[0]=TextureBlockHold):
                    break
                Sleep(0.25)
            PickupPartWidgetControl.ChangePosition(vector2{X:=2200.0,Y:=640.0})
            PickupPartWidgetControl.Show()
            PickupPart_Icon.SetImage(ThisTexture)
            PickupPart_Icon.SetDesiredSize(vector2{X:=100.0,Y:=100.0})
            #PickupPartWidgetControl.MovePosition(vector2{X:=1430.0,Y:=640.0}, 0.60)
            PickupPartWidgetControl.MovePosition(vector2{X:=1530.0,Y:=640.0}, 0.75, ?InterpolationType:=interpolation_type.EaseOutBack)
            Sleep(3.2)
            PickupPartWidgetControl.MovePosition(vector2{X:=2200.0,Y:=640.0}, 0.55, ?InterpolationType:=interpolation_type.EaseInBack)
            PickupPartWidgetControl.Hide()
        else:
            Print("ERR: Failed to set part texture!")
        if(NewPickupQueue:=QueuedPartPickups.RemoveFirstElement[TextureBlockHold]):
            set QueuedPartPickups=NewPickupQueue

powerpunch_details<public> := class:

    WidgetControl : dynamic_widget = dynamic_widget{}

    Title : string_block = string_block{Font:="Interaction"}
    Description : dialog_block = dialog_block{Font:="Interaction"}

    #PowerPunch_Arrow : texture_block = texture_block{DefaultImage:=Art.UIStuff.Icons.Arrow_Icon, DefaultDesiredSize:=vector2{X:=32.0,Y:=32.0}}
    #Ability_Arrow : texture_block = texture_block{DefaultImage:=Art.UIStuff.Icons.Arrow_Icon, DefaultDesiredSize:=vector2{X:=32.0,Y:=32.0}}

    Create():canvas_slot=
        return WidgetControl.Create(overlay{
            Slots:=array:
                overlay_slot{
                    Widget := color_block{DefaultColor:=NamedColors.DarkGrey, DefaultDesiredSize:=vector2{X:=0.0,Y:=100.0}}
                    HorizontalAlignment := horizontal_alignment.Fill
                    VerticalAlignment := vertical_alignment.Fill
                    Padding := margin{Top:=-10.0,Left:=-10.0,Right:=-10.0,Bottom:=-10.0}
                }
                overlay_slot{
                    Widget := color_block{DefaultColor:=NamedColors.Grey, DefaultDesiredSize:=vector2{X:=0.0,Y:=90.0}}
                    HorizontalAlignment := horizontal_alignment.Fill
                    VerticalAlignment := vertical_alignment.Fill
                    Padding := margin{Top:=-5.0,Left:=-5.0,Right:=-5.0,Bottom:=-5.0}
                }
                overlay_slot{
                    Widget := texture_block{DefaultImage:=Art.UIStuff.Icons.Arrow_Icon, DefaultDesiredSize:=vector2{X:=32.0,Y:=32.0}}
                    HorizontalAlignment := horizontal_alignment.Center
                    VerticalAlignment := vertical_alignment.Bottom
                    Padding := margin{Bottom:=-50.0}
                }
                overlay_slot{
                    Widget := Title.Create("", vector2{}, 26.0).Widget
                    HorizontalAlignment := horizontal_alignment.Center
                    VerticalAlignment := vertical_alignment.Top
                }
                overlay_slot{
                    Widget := Description.Create("", vector2{}, 20.0).Widget
                    HorizontalAlignment := horizontal_alignment.Center
                    VerticalAlignment := vertical_alignment.Center
                    Padding := margin{Top:=18.0}
                }
            }, vector2{X:=1440.0,Y:=1096.0}, ?Distribution:=option{1.0}, ?HorizontalAlignment:=horizontal_alignment.Center, ?VerticalAlignment:=vertical_alignment.Center, ?Alignment:=vector2{X:=1.0,Y:=1.0}, ?AnchorOverride:=option{anchors{Minimum:=vector2{X:=1.0,Y:=1.0},Maximum:=vector2{X:=1.0,Y:=1.0}}})

    ShowingDetailsEvent : event() = event(){}
    ShowDetails_PowerPunch<public>()<suspends>:void=
        Title.SetText("power punch")
        Description.SetText("this icon means the gun\nyou hold is powered up!")
        ShowingDetailsEvent.Signal()
        race:
            ShowingDetailsEvent.Await()
            block:
                WidgetControl.ChangePosition(vector2{X:=1555.0,Y:=1096.0})
                WidgetControl.Show()
                WidgetControl.MovePosition(vector2{X:=1555.0,Y:=730.0}, 1.1, ?InterpolationType:=interpolation_type.EaseOutBounce)
                Sleep(5.0)
                WidgetControl.MovePosition(vector2{X:=1555.0,Y:=1096.0}, 0.6, ?InterpolationType:=interpolation_type.EaseOut)
        WidgetControl.Hide()

    ShowDetails_PowerAbility<public>(InAbility:variant_ability_base)<suspends>:void=
        Title.SetText(InAbility.BasicSettings.Name)
        Description.SetText(InAbility.BasicSettings.Description, ?LineSeparator:="\\n")
        ShowingDetailsEvent.Signal()
        race:
            ShowingDetailsEvent.Await()
            block:
                WidgetControl.ChangePosition(vector2{X:=1616.0,Y:=1096.0})
                WidgetControl.Show()
                WidgetControl.MovePosition(vector2{X:=1616.0,Y:=730.0}, 1.1, ?InterpolationType:=interpolation_type.EaseOutBounce)
                Sleep(10.0)
                WidgetControl.MovePosition(vector2{X:=1616.0,Y:=1096.0}, 0.6, ?InterpolationType:=interpolation_type.EaseOut)
        WidgetControl.Hide()
        
<#
var NewGameOverInputEvent : weak_map(session, event(int)) = map{}
var CurrentGameOverInput : weak_map(session, int) = map{}


gameover_prompt<public> := class:

    PromptText : dialog_block = dialog_block{Font:="Quest", BackplateTexture:=option{Art.UIStuff.Fonts.InteractionBackground.InteractionBackground}}

    PlayAgainButton : button_quiet = button_quiet{}
    MainMenuButton : button_quiet = button_quiet{}

    var SubscribedState : logic = false
    SubscribeButtons():void=
        if(not SubscribedState?):
            set SubscribedState = true
            PlayAgainButton.OnClick().Subscribe(KeepPlayingInput)
            MainMenuButton.OnClick().Subscribe(MainMenuInput)

    Create():[]canvas_slot=
        SubscribeButtons()
        PlayAgainButton.SetText(StringToMessage("Play again"))
        MainMenuButton.SetText(StringToMessage("Back to main menu"))
        return array:
            PromptText.Create("Choose an option", vector2{X:=960.0,Y:=400.0}, 40.0, ?Alignment:=vector2{X:=0.5,Y:=0.5}, ?AutoAnchor:=true)
            canvas_slot:
                Widget := PlayAgainButton
                Offsets := margin{Top:=-100.0}
                Anchors:=anchors{Maximum:=vector2{X:=0.5,Y:=0.5}, Minimum:=vector2{X:=0.5,Y:=0.5}}
                Alignment:=vector2{X:=0.5}
                SizeToContent:=true
            canvas_slot:
                Widget := MainMenuButton
                Offsets := margin{Top:=40.0}
                Anchors:=anchors{Maximum:=vector2{X:=0.5,Y:=0.5}, Minimum:=vector2{X:=0.5,Y:=0.5}}
                Alignment:=vector2{X:=0.5}
                SizeToContent:=true

    KeepPlayingInput(MSG:widget_message):void=
        PlayInput := SignalInput(0)
        if. PlayInput? then spawn. CountdownGameOverOption(0)

    MainMenuInput(MSG:widget_message):void=
        PlayInput := SignalInput(1)
        if. PlayInput? then spawn. CountdownGameOverOption(1)

    SignalInput(Input:int):logic=
        if:
            not NewGameOverInputEvent[GetSession()]
            set NewGameOverInputEvent[GetSession()] = event(int){}
        if(ThisNewInputEvent:=NewGameOverInputEvent[GetSession()]):
            ThisNewInputEvent.Signal(Input)
        if:
            not CurrentGameOverInput[GetSession()]
            set CurrentGameOverInput[GetSession()] = Input
        else if(CurrentGameOverInput[GetSession()] = Input):
            if. set CurrentGameOverInput[GetSession()] = Input
            return false
        if. set CurrentGameOverInput[GetSession()] = Input
        return true

    CountdownGameOverOption(Input:int)<suspends>:void=
        Winner:=race:
            block:
                case(Input):
                    0 =>
                        for(X->Second : array{5,4,3,2,1,0}):
                            PlayerList:=GetAgents()
                            for(ThisPlayer:PlayerList):
                                ThisPlayer.GetUIManager().GameOverPrompt.PlayAgainButton.SetText(StringToMessage("Play again - [{Second}]"))
                            if(Second=0):
                                Sleep(0.45)
                            else:
                                Sleep(1.0)
                    1 =>
                        for(X->Second : array{5,4,3,2,1,0}):
                            PlayerList:=GetAgents()
                            for(ThisPlayer:PlayerList):
                                ThisPlayer.GetUIManager().GameOverPrompt.MainMenuButton.SetText(StringToMessage("Back to main menu - [{Second}]"))
                            if(Second=0):
                                Sleep(0.45)
                            else:
                                Sleep(1.0)
                    _ =>
                1
            block:
                if(NewInputEvent:=NewGameOverInputEvent[GetSession()]):
                    loop:
                        ThisInput:=NewInputEvent.Await()
                        if(ThisInput<>Input):
                            break
                -1
        if(Winner=1):
            if(Input=0):
                SetGameState(game_state.InGame)
            else:
                if(ThisPlayerManager:=GlobalPlayerManager[GetSession()]):
                    set ThisPlayerManager.GameState = game_state.Menu
                    PlayerList:=GetAgents()
                    for(ThisPlayer:PlayerList):
                        ThisPlayer.GetUIManager().ApplyUI(ThisPlayer)
        else:
            PlayerList:=GetAgents()
            for(ThisPlayer:PlayerList):
                ThisUI := ThisPlayer.GetUIManager()
                ThisUI.GameOverPrompt.PlayAgainButton.SetText(StringToMessage("Play again"))
                ThisUI.GameOverPrompt.MainMenuButton.SetText(StringToMessage("Back to main menu"))

ui_component<public> := class:

    # Must set target player when constructing a UI component
    User<public> : player

    var Root<public> : widget = color_block{}
    IsInteractive<public>:logic=false
    AppliedVisibleState : logic = true

    var InitializedState<private> : logic = false
    var<private> VisibleState : logic = false

    Apply<public><final>():void=
        if(not InitializedState?):
            if. not AppliedVisibleState? then Hide()
            if(PlayerUI:=GetPlayerUI[User]):
                PlayerUI.AddWidget(Root, player_ui_slot{InputMode := if(IsInteractive?) then ui_input_mode.All else ui_input_mode.None})
                set InitializedState = true
        else:
            Show()
        spawn. OnApplied()

    OnApplied<public>()<suspends>:void=
        return

    Remove<public>():void=
        if(PlayerUI:=GetPlayerUI[User]):
            PlayerUI.RemoveWidget(Root)
            set InitializedState = false

    Show<public>():void=
        Root.SetVisibility(widget_visibility.Visible)
        set VisibleState = true

    Hide<public>(?DoCollapse:logic=false):void=
        Root.SetVisibility(if. DoCollapse? then widget_visibility.Collapsed else widget_visibility.Hidden)
        set VisibleState = false

    IsVisible<public>()<transacts><decides>:void=
        VisibleState?
#>