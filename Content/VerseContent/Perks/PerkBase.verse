using { /Fortnite.com/Devices }
using { /Fortnite.com/Characters }
using { /Fortnite.com/Game }
using { /Fortnite.com/UI }
using { /Fortnite.com/FortPlayerUtilities }
using { /Verse.org/Simulation }
using { /Verse.org/Random }
using { /Verse.org/Colors }
using { /UnrealEngine.com/Temporary/Diagnostics }
using { /UnrealEngine.com/Temporary/SpatialMath }
using { /UnrealEngine.com/Temporary/UI }
using { Player }
using { UI }
using { Zombie }
using { Weapons }
using { Spawning }
using { InteractiveModule }

########  ########     ###     ######    #######  ##    ## ########     ###     ######   ######## 
##     ## ##     ##   ## ##   ##    ##  ##     ## ###   ## ##     ##   ## ##   ##    ##  ##       
##     ## ##     ##  ##   ##  ##        ##     ## ####  ## ##     ##  ##   ##  ##        ##       
##     ## ########  ##     ## ##   #### ##     ## ## ## ## ########  ##     ## ##   #### ######   
##     ## ##   ##   ######### ##    ##  ##     ## ##  #### ##   ##   ######### ##    ##  ##       
##     ## ##    ##  ##     ## ##    ##  ##     ## ##   ### ##    ##  ##     ## ##    ##  ##       
########  ##     ## ##     ##  ######    #######  ##    ## ##     ## ##     ##  ######   ########

var GlobalPerkCollection : weak_map(session, []perk_base) = map{}

# Perk Device that when purchased, grants special effects and abilities #

# Serves as a base for all perks #
perk_base<public> := class(Powerable):

    @editable Name<public> : string = "Unset Perk Name"
    @editable Type<public> : perk_list = perk_list.None
    @editable IsRandom<public> : logic = false
    @editable Rarity<public> : item_rarity = item_rarity.None

    InteractiveType<override> : interactable_type = interactable_type.Perk
        
    @editable
    InteractiveComponent<public> : Interactable = Interactable{InteractionSpriteOffset:=vector3{X:=175.0, Z:=116.0}, ExtraInteractionRadiusColliders:=array{vector3{Z:=90.0}},RequiresLookAt:=true, InteractionRadius:= 80.0, InteractTime:=1.0, InteractDistance:=250.0, RequiresPlayerMustBeInFront:=true}

    @editable OnPerkApplied_ActivateTrigger : trigger_device = trigger_device{}

    var<private> PerkEnabledState : logic = true
    SetPerkEnabledState(InState:logic):void=
        set PerkEnabledState=InState

    var<private> PerkTierPerUser<private> : [agent]int = map{}
    GetTier<final><public>(InAgent:agent)<transacts><decides>:int=
        PerkTierPerUser[InAgent] > -1
        return PerkTierPerUser[InAgent]
    GetTierAsIndex<final><public>(InAgent:agent)<transacts><decides>:int=
        return PerkTierPerUser[InAgent] - 1

    GetTierCost<public>(Tier:int)<transacts><decides>:int=
        0 > 1
    GetNextTierCost<public><final>(InAgent:agent)<transacts><decides>:int=
        if. Users.Find[InAgent] then. return GetTierCost[GetTier[InAgent]+1] else. GetTierCost[1]

    var<private> Users<public> : []agent = array{}

    OnPowerableBegin<override>()<suspends>:void=
        if:
            not GlobalPerkCollection[GetSession()]
            set GlobalPerkCollection[GetSession()] = array{}
        if:
            UpdatedPerkCollection := GlobalPerkCollection[GetSession()]
            set GlobalPerkCollection[GetSession()] = UpdatedPerkCollection + array{Self}
        LinkAndCompile(array. InteractiveComponent)
        spawn. OnPerkBegin()

    OnPerkBegin<public>()<suspends>:void=
        return

    var<private> RemovedEventPerPlayer<private> : [agent]event() = map{}
    GetRemovedEvent<final><public>(InAgent:agent):event()=
        if(ThisRemovedEvent:=RemovedEventPerPlayer[InAgent]):
            return ThisRemovedEvent
        else:
            NewRemovedEvent : event() = event(){}
            set RemovedEventPerPlayer = ConcatenateMaps(RemovedEventPerPlayer, map{InAgent=>NewRemovedEvent})
            return NewRemovedEvent

    # When the perk is interacted with #
    OnInteracted<override><final>(InAgent:agent, InInteractingLocation:Interactable):void=
        if(PerkEnabledState?):
            Apply(InAgent)

    # A public function used to purchase this perk or a tier #
    Apply<public><final>(InAgent:agent, ?IgnoreCost:logic=false, ?IgnorePerkLimit:logic=false):void=
        ApplyPerk(InAgent, ?IgnoreCost:=IgnoreCost, ?IgnorePerkLimit:=IgnorePerkLimit)
    
    # Used to apply this perk to a player #
    ApplyPerk<final><private>(InAgent:agent, ?IgnoreCost:logic=false, ?IgnorePerkLimit:logic=false):void=
        if(ThisPlayerInfo:=InAgent.GetPlayerInfo[]):
            if(TierCost:=GetNextTierCost[InAgent], ThisPlayerInfo.CheckPerkLimit[] or IgnorePerkLimit? or Users.Find[InAgent]):
                NextTier:=GetTier[InAgent]+1 or 1
                var BuyState : logic = true
                if(not IgnoreCost?):
                    set BuyState = ThisPlayerInfo.BuyItem(TierCost)
                if(BuyState?):
                    IncreaseTier(InAgent, NextTier, ?IgnorePerkLimit:=IgnorePerkLimit)

    # Activates when a perk is purchased for the FIRST time #
    OnApplied<public>(InAgent:agent)<suspends>:void=
        return

    # Activates when a player increases the tier on this perk #
    IncreaseTier<public><final>(InAgent:agent, PerkTier:int, ?IgnorePerkLimit:logic=false):void=
        #Print("Increasing tier on perk [{Name}] to [{PerkTier}]")
        if(ThisPlayerInfo:=InAgent.GetPlayerInfo[], GetTierCost[PerkTier]):
            if(not Users.Find[InAgent], ThisPlayerInfo.AddPerk[Self, ?IgnoreLimit:=true]):
                # Perk is new to player
                set Users += array{InAgent}
                if(not PerkTierPerUser[InAgent]):
                    set PerkTierPerUser = ConcatenateMaps(PerkTierPerUser, map{InAgent=>1})
                else if(set PerkTierPerUser[InAgent] = 1){}
                else{Print("~PerkLog-[{Name}]: Could not reset perk tier!")}
                OnPerkApplied_ActivateTrigger.Trigger(InAgent)
                Print("Added perk ui")
                InAgent.GetUIManager().PerkUICollection.AddPerk(Self)
                spawn{OnApplied(InAgent)}
            else if. set PerkTierPerUser[InAgent] = PerkTier
            InAgent.GetUIManager().PerkUICollection.ChangePerkTier(Self, PerkTier)
            if(not GetNextTierCost[InAgent]):
                SetInteractableState(false, ?InAgent:=option{InAgent})
            OnTierIncreased(InAgent)
        else:
            Print("~PerkLog-[{Name}]: Could not get playerinfo or tier cost")

    OnTierIncreased<public>(InAgent:agent):void=
        return

    # Used to remove a perk from a player #
    Remove<final><public>(InAgent:agent):void=
        if(ThisPlayerInfo:=InAgent.GetPlayerInfo[]):
            if(IsActive[InAgent]):
                var CurrentTier : int = 0
                if(ThisTier:=GetTier[InAgent]):
                    set CurrentTier = ThisTier
                else{Print("~PerkLog-[{Name}]: Failed to get the tier of a player when removing!")}
                OnRemoved(InAgent, CurrentTier)
                GetRemovedEvent(InAgent).Signal()
                if(set PerkTierPerUser[InAgent] = -1){}
                else{Print("~PerkLog-[{Name}]: Failed to reset perk tier!")}
                set Users = Users.RemoveAllElements(InAgent)
                Print("Removing perk [{Name}]")
                ThisPlayerInfo.RemovePerk(Self)
                SetInteractableState(true, ?InAgent:=option{InAgent})
        else{Print("~PerkLog-[{Name}]: Could not get playerinfo in perk_base::Remove()!")}

    # A function to be overridden for custom actions #
    OnRemoved<public>(InAgent:agent, PreviousTier:int):void=
        return

    # Gets the interaction text for this perk, to be used by buttons for interaction prompts #
    GetPowerInteractionText<override><final>(InAgent:agent, Interactive:Interactable):string=
        if(PerkEnabledState?):
            if(InAgent.GetPlayerInfo[].CheckPerkLimit[] or Users.Find[InAgent]):
                if(ThisTierCost := GetNextTierCost[InAgent]):
                    if(Users.Find[InAgent], ThisTier:=GetTier[InAgent]+1):
                        SetInteractableState(true, ?InAgent:=option{InAgent})
                        return "Upgrade to Tier {ThisTier} [<C_Y>${ThisTierCost}<E>]"
                    else:
                        SetInteractableState(true, ?InAgent:=option{InAgent})
                        return "Purchase {Name} [<C_Y>${ThisTierCost}<E>]"
                else:
                    SetInteractableState(false, ?InAgent:=option{InAgent})
                    if(not GetTierCost[2]):
                        return "Perk Already Purchased!"
                    else:
                        return "Maximum Tier Reached!"
            else:
                SetInteractableState(false, ?InAgent:=option{InAgent})
                return "<C_O>perk limit reached!<E>"
        else:
            SetInteractableState(false)
            return "Perk Unavailable!"

    var SecretMoneyCollectedState<public> : logic = false
    GetSecretMoney<public>(InAgent:agent):void=
        if(not SecretMoneyCollectedState?):
            set SecretMoneyCollectedState = true
            if. ThisPlayerInfo:=InAgent.GetPlayerInfo[] then. ThisPlayerInfo.IncreasePoints(100)
            if(ThisPlayerManager:=GlobalPlayerManager[GetSession()]):
                ThisPlayerManager.PurchaseStinger.Play(InAgent)

    IsMaxTier<public>(InAgent:agent)<transacts><decides>:void=
        not GetTierCost[GetTier[InAgent]+1] and Users.Find[InAgent]

    IsActive<public>(InAgent:agent)<transacts><decides>:void=
        Users.Find[InAgent]

# Holds information about the tier of a perk #
perk_tier_base<public> := class<concrete>:

    @editable Cost : int = 1000