using { /Fortnite.com/Devices }
using { /Fortnite.com/Characters }
using { /Fortnite.com/Game }
using { /Fortnite.com/UI }
using { /Fortnite.com/FortPlayerUtilities }
using { /Verse.org/Simulation }
using { /Verse.org/Random }
using { /Verse.org/Colors }
using { /UnrealEngine.com/Temporary/Diagnostics }
using { /UnrealEngine.com/Temporary/SpatialMath }
using { /UnrealEngine.com/Temporary/UI }
using { Player }
using { UI }
using { Zombie }
using { Weapons }
using { Spawning }
using { InteractiveModule }

(InAgent:agent).GrantAllPerks<public>():void=
    if(PerkCollection:=GlobalPerkCollection[GetSession()]):
        for(ThisPerk:PerkCollection):
            for(X:=0..2):
                ThisPerk.Apply(InAgent, ?IgnoreCost:=true)
        
GetPerk<public>(PerkType:perk_list)<transacts><decides>:perk_base=
    var MaybeMatchingPerk : ?perk_base = false
    PerkCollection:=GlobalPerkCollection[GetSession()]
    for(ThisPerk:PerkCollection):
        if(ThisPerk.Type = PerkType):
            set MaybeMatchingPerk = option{ThisPerk}
    MaybeMatchingPerk?

GetPerkList<public>()<transacts>:[]perk_base=
    return GlobalPerkCollection[GetSession()] or array{}

perk_list<public> := enum{
    None
    TankedUp
    MarathonCola
    BulletBrew
    BombCola
    MaxiMunitions
    CryoCola
    SaviorSoda
}

tankedup_perk := class(perk_base):

    @editable Name<override> : string = "tanked up"
    @editable Type<override> : perk_list = perk_list.TankedUp
    @editable Tiers : []tankedup_perk_tier = array{
        tankedup_perk_tier{MaxHealth:=120.0, Cost:=2500}
        tankedup_perk_tier{MaxHealth:=130.0, AutoHeal_TickRate_Override:=0.7, Cost:=3500}
        tankedup_perk_tier{MaxHealth:=130.0, AutoHeal_Delay_Override:=2.25, Cost:=5000}
    }
    
    GetTierCost<override>(Tier:int)<transacts><decides>:int=
        return Tiers[Tier-1].Cost

    OnTierIncreased<override>(InAgent:agent):void=
        if(ThisTier := GetTierAsIndex[InAgent]):
            if(ThisTierInformation:=Tiers[ThisTier]):
                ThisTierInformation.ApplyTier(InAgent)
                
    OnRemoved<override>(InAgent:agent, PreviousTier:int):void=
        if(InPlayerInfo := InAgent.GetPlayerInfo[]):
            InPlayerInfo.SetMaxHealth(100.0)
            InPlayerInfo.ChangeRegenSettings()

    GetDescription<override>(InAgent:agent, Interactive:Interactable):tuple(string, string)=
        ThisTier:int=GetTier[InAgent]+1 or 1
        case(ThisTier):
            1 => return ("Increase your health", "tanked up")
            2 => return ("faster auto healing", "tanked up")
            3 => return ("Increased health and regeneration", "tanked up")
            _ =>
        return ("", "tanked up")

tankedup_perk_tier := class<concrete>(perk_tier_base):
    @editable MaxHealth : float = 100.0
    @editable AutoHeal_Delay_Override : float = -1.0
    @editable AutoHeal_TickRate_Override : float = -1.0
    @editable AutoHeal_Amount_Override : float = -1.0

    ApplyTier(InUser:agent):void=
        if(ThisPlayerInfo:=InUser.GetPlayerInfo[]):
            ThisPlayerInfo.SetMaxHealth(MaxHealth)
            ThisPlayerInfo.ChangeRegenSettings(?Delay:=AutoHeal_Delay_Override, ?TickTime:=AutoHeal_TickRate_Override, ?Amount:=AutoHeal_Amount_Override)

marathoncola_perk  := class(perk_base):

    @editable Name<override> : string = "marathon cola"
    @editable Type<override> : perk_list = perk_list.MarathonCola
    @editable NoTier_MovementModulator : movement_modulator_device = movement_modulator_device{}
    @editable Tiers : []marathoncola_perk_tier = array{
        marathoncola_perk_tier{Cost:=2000}
        marathoncola_perk_tier{Cost:=4000}
        marathoncola_perk_tier{Cost:=6000}
    }

    GetTierCost<override>(Tier:int)<transacts><decides>:int=
        return Tiers[Tier-1].Cost

    OnTierIncreased<override>(InAgent:agent):void=
        if(ThisTier:=GetTierAsIndex[InAgent]):
            if(ThisTieredModulator:=Tiers[ThisTier].MovementModulator):
                ThisTieredModulator.Activate(InAgent)

    OnRemoved<override>(InAgent:agent, PreviousTier:int):void=
        NoTier_MovementModulator.Activate(InAgent)

    GetDescription<override>(InAgent:agent, Interactive:Interactable):tuple(string, string)=
        ThisTier:int=GetTier[InAgent]+1 or 1
        case(ThisTier):
            1 => return ("Move 10% faster", "marathon cola")
            2 => return ("Move 15% faster", "marathon cola")
            3 => return ("Move 20% faster", "marathon cola")
            _ =>
        return ("Run faster", "marathon cola")

marathoncola_perk_tier := class(perk_tier_base):

    @editable MovementModulator : movement_modulator_device = movement_modulator_device{}

bulletbrew_perk := class(perk_base): 

    @editable Name<override> : string = "bull-it brew"
    @editable Type<override> : perk_list = perk_list.BulletBrew
    @editable Tiers : []bulletbrew_perk_tier = array{
        bulletbrew_perk_tier{DamageMultiplier:=1.2, Cost:=2500}
        bulletbrew_perk_tier{DamageMultiplier:=1.4, Cost:=4000}
        bulletbrew_perk_tier{DamageMultiplier:=1.7, Cost:=5000}
    }
    @editable var<private> DefaultDamageMultiplier : float = 1.0

    GetTierCost<override>(Tier:int)<transacts><decides>:int=
        return Tiers[Tier-1].Cost

    OnTierIncreased<override>(InAgent:agent):void=
        if(ThisTier:=GetTierAsIndex[InAgent]):
            if(ThisTieredDamage:=Tiers[ThisTier].DamageMultiplier):
                if(ThisPlayerInfo:=InAgent.GetPlayerInfo[]):
                    ThisPlayerInfo.SetDamageMultiplier(ThisTieredDamage)

    OnRemoved<override>(InAgent:agent, PreviousTier:int):void=
        if(ThisPlayerInfo:=InAgent.GetPlayerInfo[]):
            ThisPlayerInfo.SetDamageMultiplier(DefaultDamageMultiplier)

    GetDescription<override>(InAgent:agent, Interactive:Interactable):tuple(string, string)=
        ThisTier:int=GetTier[InAgent]+1 or 1
        case(ThisTier):
            1 => return ("Damage boosted by 20%", "bull-it brew")
            2 => return ("Damage boosted by 40%", "bull-it brew")
            3 => return ("Damage boosted by 75%", "bull-it brew")
            _ =>
        return ("Increase your firepower", "bull-it brew")

    ChangeDefaultDamageMultiplier<public><final>(InMultiplier:float):void=
        set DefaultDamageMultiplier = InMultiplier
        PlayerList := GetPlayers()
        for(ThisPlayer:PlayerList):
            if(not Users.Find[ThisPlayer.User]):
                ThisPlayer.SetDamageMultiplier(DefaultDamageMultiplier)

bulletbrew_perk_tier := class(perk_tier_base):

    @editable DamageMultiplier : float = 1.5

bombcola_perk := class(perk_base):

    @editable Name<override> : string = "bomb cola"
    @editable Type<override> : perk_list = perk_list.BombCola
    @editable Tiers : []bombcola_perk_tier = array{
        bombcola_perk_tier{Duration:=2.0, Cost:=2000}
        bombcola_perk_tier{FalloffRange:=180.0, MaximumRange:=250.0, Duration:=2.0, Cost:=3000}
        bombcola_perk_tier{FalloffRange:=180.0, MaximumRange:=250.0, ResetTime:=4.5, Duration:=2.0, Cost:=4500}
    }
    @editable ExcludedZombieTypes : []zombie_type_list = array{}
    @editable InputDevice_Sprint : input_trigger_device = input_trigger_device{}
    @editable InputDevice_Crouch : input_trigger_device = input_trigger_device{}

    GetTierCost<override>(Tier:int)<transacts><decides>:int=
        return Tiers[Tier-1].Cost

    OnApplied<override>(InAgent:agent)<suspends>:void=
        if(InFortChar:=InAgent.GetFortCharacter[]):
            InputDevice_Sprint.Register(InAgent)
            InputDevice_Crouch.Register(InAgent)
            race:
                GetRemovedEvent(InAgent).Await()
                loop:
                    Winner := race:
                        block:
                            loop:
                                if(not InputDevice_Crouch.IsHeld[InAgent]):
                                    break
                                Sleep(0.0)
                            -1
                        block:
                            loop:
                                SprintState := InFortChar.SprintedEvent().Await()
                                if(not SprintState(1)?):
                                    break
                            -1
                        block:
                            Sleep(0.5)
                            1
                    if(Winner = 1):
                        CurrentVelocity := InFortChar.GetVelocity(0.1)
                        if(CurrentVelocity > 0.0):
                            InAgent.GetUIManager().PerkUICollection.SetActiveState(Self, false)
                            # Player held inputs down long enough to consider it as sliding and is moving #
                            var ActiveDuration : float = 0.0
                            var Tickrate : float = 0.0
                            if(ThisTieredInformation := Tiers[GetTierAsIndex[InAgent]]):
                                set ActiveDuration = ThisTieredInformation.Duration
                                set Tickrate = ThisTieredInformation.Tickrate
                            race:
                                SafeSleep(ActiveDuration)
                                loop:
                                    if(ThisTieredInformation := Tiers[GetTierAsIndex[InAgent]], BombBP := GetBombBlueprint[InAgent]):
                                        spawn{ExplodeNearbyZombies(InAgent, InFortChar, ThisTieredInformation, BombBP)}
                                    SafeSleep(Tickrate)
                                InFortChar.JumpedEvent().Await()
                                AwaitUnslide(InAgent)
                            SafeSleep(6.0)
                            InAgent.GetUIManager().PerkUICollection.SetActiveState(Self, true)
                    Sleep(0.0)

    AwaitUnslide<private>(InAgent:agent)<suspends>:void=
        if(InFortChar:=InAgent.GetFortCharacter[]):
            loop:
                Winner := race:
                    block:
                        loop:
                            if(not InputDevice_Crouch.IsHeld[InAgent]):
                                break
                            Sleep(0.0)
                        -1
                    block:
                        loop:
                            SprintState := InFortChar.SprintedEvent().Await()
                            if(not SprintState(1)?):
                                break
                        -1
                    block:
                        Sleep(0.5)
                        1
                if(Winner = -1):
                    break

    OnRemoved<override>(InAgent:agent, PreviousTier:int):void=
        InputDevice_Sprint.Unregister(InAgent)
        InputDevice_Crouch.Unregister(InAgent)

    ExplodeNearbyZombies(InAgent:agent, InFortChar:fort_character, TierInformation:bombcola_perk_tier, BombBP:creative_prop_asset)<suspends>:void=
        if(InFortChar.IsActive[]):
            ZombieList := GetZombieList()
            InLocation := InFortChar.GetTransform().GetLocalVector(vector3{X:=75.0})
            if(ThisSpawnedBomb:=SpawnProp(BombBP, InLocation, rotation{})(0)?):
                for(ZombieAgent -> ZombieInfo : ZombieList):
                    if(ExcludedZombieTypes.DoesNotHaveType[ZombieInfo.Controller.Type]):
                        if(ZombieFortChar := ZombieInfo.FortCharacter, DistanceToBomb:=Distance(ZombieFortChar.GetTransform().Translation, InLocation) < TierInformation.MaximumRange):
                            if(DistanceToBomb > TierInformation.FalloffRange):
                                DamageAmount := (ZombieInfo.Controller.Controls.MaximumHealth*(TierInformation.DamagePercent/100.0))*(TierInformation.FalloffRange/DistanceToBomb)
                                Print("BombCola: Falloff damage [{DamageAmount}]")
                                ZombieInfo.Damage(DamageAmount, ?Instigator:=option{InAgent}, ?CheckInstigatorWeapon:=false)
                            else:
                                DamageAmount := (ZombieInfo.Controller.Controls.MaximumHealth*(TierInformation.DamagePercent/100.0))
                                Print("BombCola: damage [{DamageAmount}]")
                                ZombieInfo.Damage(DamageAmount, ?Instigator:=option{InAgent}, ?CheckInstigatorWeapon:=false)
                Sleep(TierInformation.ParticleDuration)
                ThisSpawnedBomb.Dispose()

    GetDescription<override>(InAgent:agent, Interactive:Interactable):tuple(string, string)=
        ThisTier:int=GetTier[InAgent]+1 or 1
        case(ThisTier):
            1 => return ("Slide to explode enemies", "bomb cola")
            2 => return ("Bomb range increased by 25%", "bomb cola")
            3 => return ("Bomb reset time 25% faster", "bomb cola")
            _ =>
        return ("Slide to explode", "bomb cola")

    GetBombBlueprint<private>(InAgent:agent)<transacts><decides>:creative_prop_asset=
        var BombBlueprintToUse : ?creative_prop_asset = false
        PlayerTier := GetTierAsIndex[InAgent]
        for:
            Index->ThisTier:Tiers
            Index <= PlayerTier
            ThisTier.OverrideParticle?
        do:
            set BombBlueprintToUse = option. ThisTier.ParticleBlueprint
        BombBlueprintToUse?

bombcola_perk_tier := class<concrete>(perk_tier_base):

    @editable DamagePercent : type{_X:float where _X >= 0.0, _X <= 100.0} = 100.0
    @editable FalloffRange : float = 150.0
    @editable MaximumRange : float = 200.0
    @editable Duration : float = 3.0
    @editable Tickrate : float = 0.5
    @editable ResetTime : float = 6.0
    @editable OverrideParticle : logic = false
    @editable ParticleBlueprint : creative_prop_asset = DefaultCreativePropAsset
    @editable ParticleDuration : float = 2.0

ChangeToDefaultGameClass<public>(InAgent:agent):void=
    if(ThisPerk := maximunitions_perk[GetPerk[perk_list.MaxiMunitions]]):
        ThisPerk.DefaultClass.ChangeClass(InAgent)

maximunitions_perk := class(perk_base):

    @editable Name<override> : string = "maxi munitions"
    @editable Type<override> : perk_list = perk_list.MaxiMunitions
    @editable Tiers : []maximunitions_perk_tier = array{
        maximunitions_perk_tier{Cost:=3000}
        maximunitions_perk_tier{NewAmmoDropChance:=40.0, Cost:=4000}
        maximunitions_perk_tier{NewAmmoDropChance:=40.0, NewMaxAmmoCapPercent:=180.0, Cost:=5000}
    }
    @editable DefaultClass : class_and_team_selector_device = class_and_team_selector_device{}

    GetTierCost<override>(Tier:int)<transacts><decides>:int=
        return Tiers[Tier-1].Cost

    # Increases the tier of the player #
    OnTierIncreased<override>(InAgent:agent):void=
        if(ThisTierInfo:=Tiers[GetTierAsIndex[InAgent]]):
            ThisTierInfo.ApplyThisTier(InAgent)

    (InTier:maximunitions_perk_tier).ApplyThisTier(InAgent:agent):void=
        if(ThisPlayerInfo:=InAgent.GetPlayerInfo[]):
            ThisPlayerInfo.SetAmmoDropChance(InTier.NewAmmoDropChance)
        if(InTier.NewMaxAmmoCapPercent > 100.0):
            WeaponsList:=GetWeaponDefinitions()
            for(ThisWeapon:WeaponsList):
                if(NewMaxAmmo:=Int[ThisWeapon.GetMaxAmmo(false)*(InTier.NewMaxAmmoCapPercent/100.0)]):
                    ThisWeapon.SetMaxAmmo(InAgent, NewMaxAmmo)
        InTier.InstantReloadClass.ChangeClass(InAgent)

    OnRemoved<override>(InAgent:agent, PreviousTier:int):void=
        if(ThisPlayerInfo:=InAgent.GetPlayerInfo[], ThisWeaponManager:=GlobalWeaponManager[GetSession()]):
            if(GetGameDifficulty() = game_difficulty.Hardcore):
                ThisPlayerInfo.SetAmmoDropChance(-1.0)
            else:
                ThisPlayerInfo.SetAmmoDropChance(ThisWeaponManager.AmmoCollection.AmmoDrop_SpawnChance)
        WeaponsList:=GetWeaponDefinitions()
        for(ThisWeapon:WeaponsList):
            ThisWeapon.ResetMaxAmmo(InAgent)
        DefaultClass.ChangeClass(InAgent)

    GetDescription<override>(InAgent:agent, Interactive:Interactable):tuple(string, string)=
        ThisTier:int=GetTier[InAgent]+1 or 1
        case(ThisTier):
            1 => return ("Reload instantly!", "maxi munitions")
            2 => return ("Ammo drops spawn more often", "maxi munitions")
            3 => return ("Max ammo increased by 80%", "maxi munitions")
            _ =>
        return ("Reload instantly!", "maxi munitions")

maximunitions_perk_tier := class<concrete>(perk_tier_base):

    @editable InstantReloadClass : class_and_team_selector_device = class_and_team_selector_device{}
    @editable NewAmmoDropChance : type{_X:float where _X > 0.0} = 5.0
    @editable NewMaxAmmoCapPercent : type{_X:float where 100.0 <= _X} = 100.0


# Freezes attacking zombies, preventing damage
cryocola_perk := class(perk_base):

    @editable Name<override> : string = "cryo cola"
    @editable Type<override> : perk_list = perk_list.CryoCola
    @editable Tiers : []cryocola_perk_tier = array{
        cryocola_perk_tier{RechargeRate:=50.0, Cost:=4000}
        cryocola_perk_tier{MaximumCryoBombs:=3, RechargeRate:=35.0, Cost:=5000}
        cryocola_perk_tier{Duration:=12.0, MaximumCryoBombs:=3, RechargeRate:=35.0, Cost:=6000}
    }
    @editable OffsetFromTarget : vector3 = vector3{}

    GetTierCost<override>(Tier:int)<transacts><decides>:int=
        return Tiers[Tier-1].Cost

    var BombCountPerPlayer : [agent]int = map{}
    var BombsDepletedEventPerPlayer : [agent]event() = map{}
    
    Reset<override>():void=
        for(ThisAgent->MaybeBombCount:BombCountPerPlayer):
            if(set BombCountPerPlayer[ThisAgent] = 0){}

    OnApplied<override>(InAgent:agent)<suspends>:void=
        if(DefaultTier := Tiers[0]):
            if(set BombCountPerPlayer[InAgent] = DefaultTier.MaximumCryoBombs){}
            else:
                set BombCountPerPlayer = ConcatenateMaps(BombCountPerPlayer, map{InAgent => DefaultTier.MaximumCryoBombs})
            if(ThisPlayerInfo:=InAgent.GetPlayerInfo[]):
                race:
                    GetRemovedEvent(InAgent).Await()
                    loop:
                        AttackingZombie:=ThisPlayerInfo.ZombieAttackStarted.Await()
                        if(BombCountPerPlayer[InAgent] > 0):
                            if(ThisTierInformation := Tiers[GetTierAsIndex[InAgent]]):
                                spawn. SlowZombie(InAgent, AttackingZombie, ThisTierInformation)
                ThisPlayerInfo.SetVulnerability(true)

    # Increases the tier of the player #
    OnTierIncreased<override>(InAgent:agent):void=
        if(NewBombCount:=Tiers[GetTierAsIndex[InAgent]].MaximumCryoBombs):
            if(set BombCountPerPlayer[InAgent] = NewBombCount){}
            else{Print("~ERROR: Failed to change maximum bomb count in bombcola_perk!")}
            if(BombCountPerPlayer[InAgent] <= 0):
                if. set BombCountPerPlayer[InAgent] = NewBombCount

    # Sets the bomb on a zombie and activates it after the detonation time has passed, exploding and damaging any zombies in the set range #
    SlowZombie(UserAgent:agent, ZombieAgent:agent, TierInformation:cryocola_perk_tier)<suspends>:void=
        if(ZombieFortChar:=ZombieAgent.GetFortCharacter[], ZombieInfo:=ZombieAgent.GetZombieAI[]):
            var ThisParticleAsset : creative_prop_asset = Tiers[0].ParticleAsset or DefaultCreativePropAsset
            if(TierInformation.ChangeParticleAsset?):
                set ThisParticleAsset = TierInformation.ParticleAsset
            ZombieInfo.SetDamagePermission(false)
            BombsDepletedEvent := block{
                if(ThisEvent := BombsDepletedEventPerPlayer[UserAgent]):
                    ThisEvent
                else:
                    NewEvent:event()=event(){}
                    set BombsDepletedEventPerPlayer = ConcatenateMaps(BombsDepletedEventPerPlayer, map{UserAgent => NewEvent})
                    NewEvent
            }
            Winner:=race:
                block:
                    ZombieInfo.CancelAttack.Await()
                    false
                block:
                    ZombieInfo.DeathEvent.Await()
                    false
                block:
                    BombsDepletedEvent.Await()
                    false
                block:
                    ZombieInfo.AttackedEvent.Await()
                    true
            if(Winner?, BombCountPerPlayer[UserAgent] > 0):
                ZombieInfo.SetDamagePermission(false)
                if(ThisFreezingParticle := SpawnProp(ThisParticleAsset, ZombieFortChar.GetTransform().Translation+OffsetFromTarget, ZombieFortChar.GetTransform().Rotation)(0)?):
                    if(CurrBombCount:=BombCountPerPlayer[UserAgent], set BombCountPerPlayer[UserAgent] = CurrBombCount - 1){}
                    else{Print("~ERROR: Failed to decrement the bomb count per player!")}
                    if(BombCountPerPlayer[UserAgent] < 1):
                        BombsDepletedEvent.Signal()
                        UserAgent.GetUIManager().PerkUICollection.SetActiveState(Self, false)
                    branch:
                        loop:
                            if(ThisFreezingParticle.IsValid[]):
                                ThisFreezingParticle.MoveTo(ZombieFortChar.GetTransform().Translation+OffsetFromTarget, ThisFreezingParticle.GetTransform().Rotation, 0.09365)
                            else:
                                break
                    race:
                        ZombieInfo.RespawnEvent.Await()
                        ZombieInfo.DeathEvent.Await()
                        block:
                            if(ThisStun := ZombieInfo.Controller.GetSkinInfo[ZombieInfo.SkinTag].GetStunnedProp[]):
                                ZombieInfo.SetPropAnimation(ThisStun)
                            ZombieInfo.SetMoveState(false)
                            ZombieInfo.ExitAttack.Signal()
                            ZombieInfo.SetAttackPermission(false)
                            SafeSleep(TierInformation.Duration)
                            ZombieInfo.StopPropAnimation()
                            ZombieInfo.ResetSpeed()
                    ThisFreezingParticle.Dispose()
                ZombieInfo.SetMoveState(true)
                ZombieInfo.SetAttackPermission(true)
                ZombieInfo.SetDamagePermission(true)
                SafeSleep(TierInformation.RechargeRate)
                if(IsActive[UserAgent], CurrBombCount:=BombCountPerPlayer[UserAgent], set BombCountPerPlayer[UserAgent] = CurrBombCount + 1):
                    #if. ThisPlayerInfo := UserAgent.GetPlayerInfo[] then. ThisPlayerInfo.SetNullAttacksState(true)
                    UserAgent.GetUIManager().PerkUICollection.SetActiveState(Self, true)
                    if(BombCountPerPlayer[UserAgent] > TierInformation.MaximumCryoBombs or BombCountPerPlayer[UserAgent] <= 0):
                        if(set BombCountPerPlayer[UserAgent] = TierInformation.MaximumCryoBombs){}
                        else{Print("~ERROR: Failed to set BombCountPerPlayer to the maximum bomb count!")}
            else:
                ZombieInfo.SetDamagePermission(true)
            #else{Print("~ERROR: Failed to increment the bomb count per player!")}

    GetDescription<override>(InAgent:agent, Interactive:Interactable):tuple(string, string)=
        ThisTier:int=GetTier[InAgent]+1 or 1
        case(ThisTier):
            1 => return ("Freeze enemies who hurt you", "cryo cola")
            2 => return ("Freeze more enemies at a time", "cryo cola")
            3 => return ("Enemies are frozen for much longer", "cryo cola")
            _ =>
        return ("Freezes enemies who hurt you", "cryo cola")

cryocola_perk_tier<public> := class<concrete>(perk_tier_base):
    @editable Duration : float = 4.8
    #@editable Damage : float = 50.0
    @editable MaximumCryoBombs : int = 2
    @editable RechargeRate : float = 30.0
    @editable ChangeParticleAsset : logic = false
    @editable ParticleAsset : creative_prop_asset = DefaultCreativePropAsset

saveyoursoda_perk<public> := class(perk_base):

    @editable Name<override> : string = "savior soda"
    @editable Type<override> : perk_list = perk_list.SaviorSoda
    @editable Tiers : []saviorsoda_perk_tier = array{
        saviorsoda_perk_tier{Multiplayer_ReviveTime:=3.0, Cost:=500}
        saviorsoda_perk_tier{Multiplayer_ReviveTime:=3.0, Multiplayer_DownedTime:=40.0, Singleplayer_PerksSaved:=1, Cost:=1500}
        saviorsoda_perk_tier{Multiplayer_ReviveTime:=1.75, Multiplayer_DownedTime:=45.0, Singleplayer_PerksSaved:=2, Cost:=3000}  
    }

    var<private> SinglePlayerTimesPurchased : int = 0

    GetTierCost<override>(Tier:int)<transacts><decides>:int=
        return Tiers[Tier-1].Cost

    Reset<override>():void=
        Print("Resetting savior")
        set SinglePlayerTimesPurchased = 0
        if(IsSinglePlayer[]):
            Print("Savior soda on")
            ForcePowerState(true)
        SetPerkEnabledState(true)

    OnPerkBegin<override>()<suspends>:void=
        # Ensures this perk is powered if in singleplayer
        AwaitGameState(game_state.InGame)
        Reset()

    OnTierIncreased<override>(InAgent:agent):void=
        if:
            ThisTier := Tiers[GetTierAsIndex[InAgent]]
            UserPlayerInfo := InAgent.GetPlayerInfo[]
        then:
            AgentList:=GetAgents(?OnlyActives:=false)
            for(ThisAgent:AgentList):
                if(ThisPlayerInfo:=ThisAgent.GetPlayerInfo[]):
                    ThisPlayerInfo.ReviveInteractable.SetInteractionTime(ThisTier.Multiplayer_ReviveTime, ?InAgent:=option{InAgent})
            UserPlayerInfo.SetDownedTime(ThisTier.Multiplayer_DownedTime)
                    
    GetBleedoutTime(InAgent:agent):float=
        if(ThisTier := Tiers[GetTierAsIndex[InAgent]]):
            return ThisTier.Multiplayer_DownedTime
        else:
            return 40.0

    GetSavedPerkAmount(InAgent:agent):int=
        if(IsSinglePlayer[], ThisTier:=Tiers[GetTierAsIndex[InAgent]]):
            return ThisTier.Singleplayer_PerksSaved
        else:
            return 0

    HandleSoloRevive<public>(InAgent:agent)<suspends>:void=
        if(PlayerManager:=GlobalPlayerManager[GetSession()], ThisPlayerInfo:=InAgent.GetPlayerInfo[]):
            Print("Handling solo revive in savior soda")
            ThisPlayerInfo.SetVulnerability(false)
            SafeSleep(0.75)
            spawn{InAgent.GetUIManager().DisplayRevivingAlert()}
            SafeSleep(5.0)
            var PerksToSave : int = GetSavedPerkAmount(InAgent)
            if(GetTier[InAgent] > 1, TieredPerksToSave := Tiers[GetTierAsIndex[InAgent]].Singleplayer_PerksSaved):
                set PerksToSave = TieredPerksToSave
            ThisPlayerInfo.RemovePerk(Self)
            var PerkIndex : int = ThisPlayerInfo.ActivePerks.Length-1
            for(PerkCount := 0..ThisPlayerInfo.ActivePerks.Length-1-PerksToSave):
                if(ThisPerk:=ThisPlayerInfo.ActivePerks[PerkIndex]):
                    ThisPlayerInfo.RemovePerk(ThisPerk)
                    set PerkIndex -= 1
            if(ThisPlayerPackage := InAgent.GetPlayerPackage[]):
                DownedLocation := ThisPlayerPackage.DownedProp.GetTransform().Translation
                ThisZombieList := GetZombieList()
                for(ThisZombie:ThisZombieList):
                    if(Distance(ThisZombie.FortCharacter.GetTransform().Translation, DownedLocation) <= 350.0):
                        ThisZombie.Respawn()
            Print("~Remaining perks: {ThisPlayerInfo.ActivePerks.Length}")
            InAgent.GetUIManager().RevivingStartedAlertText.Hide()
            spawn. ThisPlayerInfo.PlayState(player_state.Active)
        else{Print("~SaviorSoda ERROR: Failed to fetch playermanager or playerinfo!")}

    OnRemoved<override>(InAgent:agent, PreviousTier:int):void=
        if(IsSinglePlayer[]):
            set SinglePlayerTimesPurchased += 1
            if(SinglePlayerTimesPurchased >= 3):
                SetPerkEnabledState(false)
        AgentList:=GetAgents(?OnlyActives:=false)
        for(ThisAgent:AgentList):
            if(ThisPlayerInfo:=ThisAgent.GetPlayerInfo[]):
                ThisPlayerInfo.ReviveInteractable.SetInteractionTime(3.5, ?InAgent:=option{InAgent})
        if(ThisPlayerInfo:=InAgent.GetPlayerInfo[]):
            ThisPlayerInfo.SetDownedTime(40.0)
        #if(ThisReviveManager:=GlobalReviveManager[GetSession()]):
        #    ThisReviveManager.SetRevivingTime(InAgent, 5.0)
                        
    GetDescription<override>(InAgent:agent, Interactive:Interactable):tuple(string, string)=
        var ThisTier:int=GetTier[InAgent]+1 or 1
        if(ThisTier = -1):
            set ThisTier = 1
        if(IsSinglePlayer[]):
            case(ThisTier):
                1 => return ("Revive yourself", "{Name}")
                2 => return ("Saves one perk when revived", "{Name}")
                3 => return ("Saves two perks when revived", "{Name}")
                _ => 
        else:
            case(ThisTier):
                1 => return ("Revive others quicker", "{Name}")
                2 => return ("Saves one perk when revived", "{Name}")
                3 => return ("Increased downed time and faster reviving", "{Name}")
                _ => 
        return ("", "")

saviorsoda_perk_tier := class<concrete>(perk_tier_base):
    @editable Singleplayer_PerksSaved : int = 0

    @editable Multiplayer_DownedTime : float = 30.0
    @editable Multiplayer_ReviveTime : float = 4.5