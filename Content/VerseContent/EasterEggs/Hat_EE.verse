using { /Verse.org/Simulation }
using { /Fortnite.com/Devices }
using { /UnrealEngine.com/Temporary/SpatialMath }
using { Player }
using { Zombie }

hat_easteregg_control := class(creative_device):

    @editable
    HatTriggers : []hat_inst = array{}

    var HitCount : int = 0

    OnBegin<override>():void=
        for(ThisTrigger:HatTriggers):
            set ThisTrigger.ParentControl = Self
            ThisTrigger.Trigger.TriggeredEvent.Subscribe(ThisTrigger.OnHatHit)

hat_inst := class<concrete>:

    @editable
    Trigger : trigger_device = trigger_device{}

    @editable
    HatProp : creative_prop = creative_prop{}

    var ParentControl : hat_easteregg_control = hat_easteregg_control{}

    OnHatHit<public>(Agent:?agent):void=
        set ParentControl.HitCount += 1
        spawn. HatVFX()
        HatProp.Hide()
        if(ParentControl.HitCount >= ParentControl.HatTriggers.Length):
            if(ThisPlayerManager:=GlobalPlayerManager[GetSession()]):
                PlayerList:=GetPlayers()
                for(ThisPlayer:PlayerList):
                    ThisPlayerManager.PurchaseStinger.Play(ThisPlayer.User)
                    ThisPlayer.IncreasePoints(10000)

    HatVFX()<suspends>:void=
        if(SpawnedVFX:=SpawnProp(GlobalZombieTools[GetSession()].PoisonVFX, Trigger.GetTransform().Translation, rotation{})(0)?):
            Sleep(6.0)
            SpawnedVFX.Dispose()