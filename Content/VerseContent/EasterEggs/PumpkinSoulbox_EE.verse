using { /Verse.org/Simulation }
using { /Fortnite.com/Devices }
using { /UnrealEngine.com/Temporary/SpatialMath }
using { /UnrealEngine.com/Temporary/SpatialMath }
using { Player }
using { Zombie }
using { UI }
using { Room }

var PumpkinSoulboxCollection : weak_map(session, []lantern_soulbox) = map{}

lantern_soulbox := class(soulbox):

    @editable
    InactiveSoulbox : creative_prop = creative_prop{}

    @editable
    ActivateTrigger : trigger_device = trigger_device{}

    @editable
    OnCollectionCompleted_OpenDoor : door = door{}

    var ActiveStateVFX : creative_prop = creative_prop{}

    OnSoulboxBegin<override>()<suspends>:void=
        SoulboxLocation.Hide()
        if:
            not PumpkinSoulboxCollection[GetSession()]
            set PumpkinSoulboxCollection[GetSession()] = array{}
        if:
            CurrentList := PumpkinSoulboxCollection[GetSession()]
            set PumpkinSoulboxCollection[GetSession()] = CurrentList + array{Self}
        ActivateTrigger.TriggeredEvent.Subscribe(ActivateSoulbox)

    ActivateSoulbox<private>(Agent:?agent):void=
        spawn. MoveSoulFlare()

    MoveSoulFlare<private>()<suspends>:void=
        if(SpawnedSoulFlare := SpawnProp(SoulFlareVFX_BP, InactiveSoulbox.GetTransform().Translation, rotation{})(0)?):
            InactiveSoulbox.Hide()
            MoveSpeed := Distance(SoulboxLocation.GetTransform().Translation, InactiveSoulbox.GetTransform().Translation) / 300.0
            SafeMoveTo(SpawnedSoulFlare, SoulboxLocation.GetTransform().Translation, rotation{}, MoveSpeed)
            set ActiveStateVFX = SpawnedSoulFlare
            Enable()
    
    OnSoulboxCompleted<override>()<suspends>:void=
        ActiveStateVFX.Dispose()
        if(ThisList:=PumpkinSoulboxCollection[GetSession()]):
            PumpkinSoulboxesCompleted:=for:
                ThisSoulbox:ThisList
                ThisSoulbox.IsCompleted[]
            do:
                0
            if(PumpkinSoulboxesCompleted.Length >= ThisList.Length):
                Print("soulboxes completed!")
                SafeSleep(1.0)
                spawn. OnCollectionCompleted_OpenDoor.Open()
                PlayerList:=GetPlayers(?OnlyActives:=false)
                for(ThisPlayer:PlayerList):
                    ThisPlayer.User.GetUIManager().RocketEE_Text.Show()
                SafeSleep(8.0)
                for(ThisPlayer:PlayerList):
                    ThisPlayer.User.GetUIManager().RocketEE_Text.Hide()
        else:
            Print("ERR: SOULBOX LIST FETCH FAIL")